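# Control Rig function library that adds per-bone tweaker controls for adjusting MetaHuman face bones.
# Construction picks the legacy or MetaHuman facial root, spawns null-parented controls per child bone, and tags numbered ones "macro", others "main".
# Forward solve toggles each set via master-control channels, snaps nulls to bones except while that control is manipulated, then applies control local transforms as bone offsets.
# Backward solve only snaps nulls to bones.
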ControlRigBlueprint /EpicControlRig/Libraries/Face/CRFL_FN_BoneTweaker_Face_v01
graphs [
 CRFL FN MH BoneTweaker Backward v01_ContainedGraph {nodes [
   n1 = RigVMFunctionEntryNode()
   RigVMFunction_Sequence = &RigVMFunction_Sequence(ExecuteContext=@n1.ExecuteContext)
   n2 = RigVMDispatchNode(
    Array=/Script/ControlRig.RigUnit_FindItemsWithMetadataTag(NameSpace=Self, Tag=main tweaker) {is=struct RigVMUnitNode}.Items
    Other=&RigUnit_FindItemsWithMetadataTag(NameSpace=Self, Tag=macro tweaker).Items
    ExecuteContext=@RigVMFunction_Sequence.A
   ) {template="DISPATCH_RigVMDispatch_ArrayUnion(io Array,in Other)"}
   > n3 = RigVMDispatchNode(Array=@n2.Array) {template="DISPATCH_RigVMDispatch_ArrayIterator(in Array,out Element,out Index,out Count,out Ratio)"}
   Concat = /Script/RigVM.RigVMFunction_NameConcat(B=_GRP, A=@n3.Element.Name) {is=struct RigVMUnitNode, template="Concat::Execute(in A,in B,out Result)"}
   > &RigUnit_SetTransform(
    Item=(Type=Null,Name="None")
    Weight=1
    bInitial=False
    Item.Name=@Concat.Result
    Value=&RigUnit_GetTransform(
     Item=(Type=Bone,Name="None")
     bInitial=False
     Item.Name=/Script/RigVM.RigVMFunction_NameReplace(Old=BoneTweaker_, Name=@n3.Element.Name) {is=struct RigVMUnitNode, template="Replace::Execute(in Name,in Old,in New,out Result)"}.Result
    ).Transform
   )
   RigVMFunctionReturnNode(ExecuteContext=@RigVMFunction_Sequence.B)
  ]}
 CRFL FN MH BoneTweaker Forward v01_ContainedGraph {nodes [
   n1 = RigVMFunctionEntryNode()
   n2 = &RigVMFunction_Sequence(ExecuteContext=@n1.ExecuteContext)
   RigVMFunction_Sequence = RigVMAggregateNode(ExecuteContext=@n2.A)
   For_Each_1 = RigVMDispatchNode(Array=&RigUnit_FindItemsWithMetadataTag(NameSpace=Self, Tag=main tweaker).Items, ExecuteContext=@RigVMFunction_Sequence.A) {template="DISPATCH_RigVMDispatch_ArrayIterator(in Array,out Element,out Index,out Count,out Ratio)"}
   RigUnit_GetBoolAnimationChannel = /Script/ControlRig.RigUnit_GetBoolAnimationChannel(Channel=main tweakers, Control=BoneTweaker_Master, bInitial=False) {is=struct RigVMUnitNode, template="GetAnimationChannel::Execute(out Value,in Control,in Channel,in bInitial)"}
   > /Script/ControlRig.RigUnit_SetControlVisibility(Item=@For_Each_1.Element, bVisible=@RigUnit_GetBoolAnimationChannel.Value) {is=struct RigVMUnitNode, template="SetControlVisibility::Execute(in Item,in Pattern,in bVisible)"}
   For_Each_2 = RigVMDispatchNode(Array=&RigUnit_FindItemsWithMetadataTag(NameSpace=Self, Tag=macro tweaker).Items, ExecuteContext=@RigVMFunction_Sequence.B) {template="DISPATCH_RigVMDispatch_ArrayIterator(in Array,out Element,out Index,out Count,out Ratio)"}
   > &RigUnit_SetControlVisibility(
    Item=@For_Each_2.Element
    bVisible=&RigUnit_GetBoolAnimationChannel(Channel=macro tweakers, Control=BoneTweaker_Master, bInitial=False) {template="GetAnimationChannel::Execute(out Value,in Control,in Channel,in bInitial)"}.Value
   ) {template="SetControlVisibility::Execute(in Item,in Pattern,in bVisible)"}
   n3 = RigVMDispatchNode(
    Array=&RigUnit_FindItemsWithMetadataTag(NameSpace=Self, Tag=main tweaker).Items
    Other=&RigUnit_FindItemsWithMetadataTag(NameSpace=Self, Tag=macro tweaker).Items
    ExecuteContext=@RigVMFunction_Sequence.C
   ) {template="DISPATCH_RigVMDispatch_ArrayUnion(io Array,in Other)"}
   > n4 = RigVMDispatchNode(Array=@n3.Array) {template="DISPATCH_RigVMDispatch_ArrayIterator(in Array,out Element,out Index,out Count,out Ratio)"}
   > RigVMVariableNode(Value=@n4.Element) {variable=item}
   > n5 = RigVMVariableNode(
    Value=(Type=Bone,Name="None")
    Value.Name=&RigVMFunction_NameReplace(Old=BoneTweaker_, Name=@n4.Element.Name) {template="Replace::Execute(in Name,in Old,in New,out Result)"}.Result
   ) {variable=bone}
   n6 = /Script/ControlRig.RigUnit_IsInteracting() {is=struct RigVMUnitNode}
   n7 = RigVMVariableNode() {variable=item}
   n8 = &RigVMFunction_ControlFlowBranch(
    Condition=&RigVMFunction_MathBoolAnd(
     B=RigVMDispatchNode(Array=@n6.Items, Element=@n7.Value) {template="DISPATCH_RigVMDispatch_ArrayFind(in Array,in Element,out Index,out Success)"}.Success
     A=@n6.bIsInteracting
    ).Result
    ExecuteContext=@n5.ExecuteContext
   )
   n9 = RigVMVariableNode() {variable=item}
   &RigUnit_SetTransform(
    Item=(Type=Null,Name="None")
    Weight=1
    bInitial=False
    Item.Name=&RigVMFunction_NameConcat(B=_GRP, A=@n9.Value.Name) {template="Concat::Execute(in A,in B,out Result)"}.Result
    Value=&RigUnit_GetTransform(
     Item=(Type=Bone,Name="None")
     bInitial=False
     Item.Name=&RigVMFunction_NameReplace(Old=BoneTweaker_, Name=@n9.Value.Name) {template="Replace::Execute(in Name,in Old,in New,out Result)"}.Result
    ).Transform
    ExecutePin=@n8.False
   )
   &RigUnit_OffsetTransformForItem(
    Weight=1
    bPropagateToChildren=True
    OffsetTransform=&RigUnit_GetTransform(Space=LocalSpace, bInitial=False, Item=RigVMVariableNode() {variable=item}.Value).Transform
    Item=RigVMVariableNode() {variable=bone}.Value
    ExecutePin=@n8.Completed
   )
   RigVMFunctionReturnNode(ExecuteContext=@n2.B)
  ]}
 New Function_ContainedGraph {nodes [
   RigVMFunctionEntryNode()
   > n1 = &RigUnit_HierarchyAddNull(
    Name=BoneTweaker_GRP
    Parent=(Type=None,Name="Face_Gizmo")
    Space=LocalSpace
    Transform=(Rotation=(X=0,Y=0,Z=0,W=1),Translation=(X=0,Y=0,Z=0),Scale3D=(X=0.8,Y=0.8,Z=0.8))
   )
   > RigUnit_HierarchyAddControlTransform = /Script/ControlRig.RigUnit_HierarchyAddControlTransform(
    InitialValue=(Rotation=(X=0,Y=0,Z=0,W=1),Translation=(X=0,Y=0,Z=0),Scale3D=(X=1,Y=1,Z=1))
    Name=BoneTweaker_Master
    OffsetSpace=LocalSpace
    OffsetTransform=(Rotation=(X=0,Y=0,Z=0,W=1),Translation=(X=12,Y=5,Z=190),Scale3D=(X=1,Y=1,Z=1))
    Settings="(InitialSpace=LocalSpace,bUsePreferredRotationOrder=False,PreferredRotationOrder=YZX,Limits=(LimitTranslationX=(bMinimum=False,bMaximum=False),LimitTranslationY=(bMinimum=False,bMaximum=False),LimitTranslationZ=(bMinimum=False,bMaximum=False),LimitPitch=(b ...(truncated)"
    Parent=@n1.Item
   ) {
    is=struct RigVMUnitNode
    template="SpawnControl::Execute(in InitialValue,in Settings,in OffsetTransform,in OffsetSpace,in Parent,in Name,out Item)"
   }
   > SpawnAnimationChannel = /Script/ControlRig.RigUnit_HierarchyAddAnimationChannelBool(
    InitialValue=true
    LimitsEnabled=()
    MaximumValue=True
    MinimumValue=False
    Name=main tweakers
    Parent=@RigUnit_HierarchyAddControlTransform.Item
   ) {
    is=struct RigVMUnitNode
    template="SpawnAnimationChannel::Execute(in InitialValue,in MinimumValue,in MaximumValue,in LimitsEnabled,in Parent,in Name,out Item)"
   }
   > &RigUnit_HierarchyAddAnimationChannelBool(
    InitialValue=true
    LimitsEnabled=()
    MaximumValue=True
    MinimumValue=False
    Name=macro tweakers
    Parent=@RigUnit_HierarchyAddControlTransform.Item
   ) {template="SpawnAnimationChannel::Execute(in InitialValue,in MinimumValue,in MaximumValue,in LimitsEnabled,in Parent,in Name,out Item)"}
   > n2 = RigVMVariableNode(Value=@RigUnit_HierarchyAddControlTransform.Item) {variable=Local Tweaker Main Ctrl}
   RigVMFunction_Sequence = &RigVMFunction_Sequence(ExecuteContext=@n2.ExecuteContext)
   For_Each = RigVMDispatchNode(
    Array=/Script/ControlRig.RigUnit_CollectionChildrenArray(TypeToSearch=Bone, bDefaultChildren=True, bIncludeParent=False, bRecursive=true, Parent=RigVMVariableNode() {variable=Facial_Root}.Value) {is=struct RigVMUnitNode}.Items
    ExecuteContext=@RigVMFunction_Sequence.A
   ) {template="DISPATCH_RigVMDispatch_ArrayIterator(in Array,out Element,out Index,out Count,out Ratio)"}
   > RigVMVariableNode(Value=@For_Each.Element) {variable=Local Active Joint Element}
   > RigVMFunction_Sequence_1 = RigVMAggregateNode()
   Concat = &RigVMFunction_NameConcat(A=BoneTweaker_, B=RigVMVariableNode() {variable=Local Active Joint Element}.Value.Name) {template="Concat::Execute(in A,in B,out Result)"}
   RigVMVariableNode(Value=@Concat.Result, ExecuteContext=@RigVMFunction_Sequence_1.B) {variable=Local Active Tweak Name}
   n3 = RigVMVariableNode() {variable=Local Active Joint Element}
   > n4 = &RigUnit_HierarchyAddNull(
    Space=GlobalSpace
    Name=&RigVMFunction_NameConcat(B=_GRP, A=@Concat.Result) {template="Concat::Execute(in A,in B,out Result)"}.Result
    Transform=&RigUnit_GetTransform(bInitial=true, Item=@n3.Value).Transform
    Parent=RigVMVariableNode() {variable=Local Tweaker Ctrl Parent}.Value
   )
   To_String_2 = RigVMDispatchNode(Value=@n3.Value.Name) {template="DISPATCH_RigDispatch_ToString(in Value,out Result)"}
   RigVMFunction_MathBoolNot = &RigVMFunction_MathBoolNot(Value=/Script/RigVM.RigVMFunction_StringToInteger(ChopLeft=true, ChopRight=false, Value=@To_String_2.Result) {is=struct RigVMUnitNode}.Success)
   RigVMFunction_ControlFlowBranch = &RigVMFunction_ControlFlowBranch(ExecuteContext=@n4.ExecutePin, Condition=@RigVMFunction_MathBoolNot.Result)
   n5 = &RigUnit_HierarchyAddControlTransform(
    InitialValue=(Rotation=(X=0,Y=0,Z=0,W=1),Translation=(X=0,Y=0,Z=0),Scale3D=(X=1,Y=1,Z=1))
    OffsetSpace=LocalSpace
    OffsetTransform=(Rotation=(X=0,Y=0,Z=0,W=1),Translation=(X=0,Y=0,Z=0),Scale3D=(X=1,Y=1,Z=1))
    Settings="(InitialSpace=LocalSpace,bUsePreferredRotationOrder=False,PreferredRotationOrder=YZX,Limits=(LimitTranslationX=(bMinimum=False,bMaximum=False),LimitTranslationY=(bMinimum=False,bMaximum=False),LimitTranslationZ=(bMinimum=False,bMaximum=False),LimitPitch=(b ...(truncated)"
    Name=RigVMVariableNode() {variable=Local Active Tweak Name}.Value
    ExecutePin=@RigVMFunction_ControlFlowBranch.True
    Parent=@n4.Item
   ) {template="SpawnControl::Execute(in InitialValue,in Settings,in OffsetTransform,in OffsetSpace,in Parent,in Name,out Item)"}
   > RigVMVariableNode(Value=@n5.Item) {variable=Local Tweaker Ctrl}
   SpawnControl = &RigUnit_HierarchyAddControlTransform(
    InitialValue=(Rotation=(X=0,Y=0,Z=0,W=1),Translation=(X=0,Y=0,Z=0),Scale3D=(X=1,Y=1,Z=1))
    OffsetSpace=LocalSpace
    OffsetTransform=(Rotation=(X=0,Y=0,Z=0,W=1),Translation=(X=0,Y=0,Z=0),Scale3D=(X=1,Y=1,Z=1))
    Settings="(InitialSpace=LocalSpace,bUsePreferredRotationOrder=False,PreferredRotationOrder=YZX,Limits=(LimitTranslationX=(bMinimum=False,bMaximum=False),LimitTranslationY=(bMinimum=False,bMaximum=False),LimitTranslationZ=(bMinimum=False,bMaximum=False),LimitPitch=(b ...(truncated)"
    Name=RigVMVariableNode() {variable=Local Active Tweak Name}.Value
    ExecutePin=@RigVMFunction_ControlFlowBranch.False
    Parent=@n4.Item
   ) {template="SpawnControl::Execute(in InitialValue,in Settings,in OffsetTransform,in OffsetSpace,in Parent,in Name,out Item)"}
   > RigVMVariableNode(Value=@SpawnControl.Item) {variable=Local Tweaker Ctrl}
   n6 = &RigVMFunction_NameConcat(
    A=BoneTweaker_
    B=/Script/ControlRig.RigUnit_HierarchyGetParent(bDefaultParent=True, Child=RigVMVariableNode() {variable=Local Active Joint Element}.Value) {is=struct RigVMUnitNode}.Parent.Name
   ) {template="Concat::Execute(in A,in B,out Result)"}
   n7 = &RigVMFunction_ControlFlowBranch(
    Condition=/Script/ControlRig.RigUnit_ItemExists(Item=(Type=Control,Name="None"), Item.Name=@n6.Result) {is=struct RigVMUnitNode}.Exists
    ExecuteContext=@RigVMFunction_Sequence_1.A
   )
   RigVMVariableNode(ExecuteContext=@n7.False, Value=RigVMVariableNode() {variable=Local Tweaker Main Ctrl}.Value) {variable=Local Tweaker Ctrl Parent}
   RigVMVariableNode(Value=(Type=Control,Name="None"), ExecuteContext=@n7.True, Value.Name=@n6.Result) {variable=Local Tweaker Ctrl Parent}
   To_String_1 = RigVMDispatchNode(Value=RigVMVariableNode() {variable=Local Tweaker Ctrl}.Value.Name) {template="DISPATCH_RigDispatch_ToString(in Value,out Result)"}
   RigVMFunction_StringToInteger = &RigVMFunction_StringToInteger(ChopLeft=true, ChopRight=false, Value=@To_String_1.Result)
   n8 = &RigVMFunction_ControlFlowBranch(Condition=@RigVMFunction_StringToInteger.Success, ExecuteContext=@RigVMFunction_Sequence_1.C)
   n9 = RigVMVariableNode() {variable=Local Tweaker Ctrl}
   /Script/ControlRig.RigUnit_SetMetadataTag(NameSpace=Self, Tag=main tweaker, ExecutePin=@n8.False, Item=@n9.Value) {is=struct RigVMUnitNode}
   &RigUnit_SetMetadataTag(NameSpace=Self, Tag=macro tweaker, ExecutePin=@n8.True, Item=@n9.Value)
   RigVMFunctionReturnNode(ExecuteContext=@RigVMFunction_Sequence.B)
  ]}
 New Function_ContainedGraph {nodes [
   n1 = RigVMFunctionEntryNode()
   RigVMFunction_Sequence = &RigVMFunction_Sequence(ExecuteContext=@n1.ExecuteContext)
   RigVMFunction_ControlFlowBranch = &RigVMFunction_ControlFlowBranch(
    Condition=&RigUnit_ItemExists(Item=(Type=Bone,Name="None"), Item.Name=RigVMVariableNode() {variable=Legacy Root Bone}.Value).Exists
    ExecuteContext=@RigVMFunction_Sequence.A
   )
   n2 = &RigVMFunction_ControlFlowBranch(
    Condition=&RigUnit_ItemExists(Item=(Type=Bone,Name="None"), Item.Name=RigVMVariableNode() {variable=MH Root Bone}.Value).Exists
    ExecuteContext=@RigVMFunction_ControlFlowBranch.False
   )
   RigVMVariableNode(ExecuteContext=@n2.True, Value=RigVMVariableNode() {variable=MH Root Bone}.Value) {variable=Current Root Bone}
   RigVMVariableNode(ExecuteContext=@RigVMFunction_ControlFlowBranch.True, Value=RigVMVariableNode() {variable=Legacy Root Bone}.Value) {variable=Current Root Bone}
   RigVMFunctionReturnNode(ExecuteContext=@RigVMFunction_Sequence.B)
  ]}
 RigVMFunctionLibrary {nodes [
   CRFL FN MH BoneTweaker Backward v01 = RigVMCollapseNode(ExecuteContext=())
   CRFL FN MH BoneTweaker Construction v01 = RigVMCollapseNode(ExecuteContext=(), Facial_Root=(Type=Bone,Name="None"))
   CRFL FN MH BoneTweaker Forward v01 = RigVMCollapseNode(ExecuteContext=())
   Set Current Facial Root Bone = RigVMCollapseNode(ExecuteContext=())
  ]}
 RigVMFunction_Sequence_1_ContainedGraph {nodes [
   n1 = RigVMFunctionEntryNode()
   RigVMFunction_Sequence = &RigVMFunction_Sequence(ExecuteContext=@n1.ExecuteContext)
   n2 = &RigVMFunction_Sequence(ExecuteContext=@RigVMFunction_Sequence.B)
   RigVMFunctionReturnNode(A=@RigVMFunction_Sequence.A, B=@n2.A, C=@n2.B)
  ]}
 RigVMFunction_Sequence_2_ContainedGraph {nodes [
   n1 = RigVMFunctionEntryNode()
   n2 = &RigVMFunction_Sequence(ExecuteContext=@n1.ExecuteContext)
   n3 = &RigVMFunction_Sequence(ExecuteContext=@n2.B)
   RigVMFunctionReturnNode(A=@n2.A, B=@n3.A, C=@n3.B)
  ]}
 RigVMModel {nodes [
   RigUnit_BeginExecution = &RigUnit_BeginExecution()
   CRFL FN MH BoneTweaker Forward v01 = RigVMFunctionReferenceNode(ExecuteContext=@RigUnit_BeginExecution.ExecutePin)
  ]}
 RigVMModel Rig {nodes [
   n1 = /Script/ControlRig.RigUnit_InverseExecution() {is=struct RigVMUnitNode}
   CRFL FN MH BoneTweaker Backward v01 = RigVMFunctionReferenceNode(ExecuteContext=@n1.ExecutePin)
  ]}
 RigVMModel Rig_1 {nodes [
   n1 = &RigUnit_PrepareForExecution()
   &RigUnit_HierarchyImportFromSkeleton(NameSpace=None, bIncludeCurves=False, bIncludeMeshSockets=False, bIncludeVirtualBones=True, ExecutePin=@n1.ExecutePin)
   > Set Current Facial Root Bone = RigVMFunctionReferenceNode()
   > CRFL FN MH BoneTweaker Construction v01 = RigVMFunctionReferenceNode(Facial_Root=(Type=Bone,Name="FACIAL_C_FacialRoot"), Facial_Root.Name=RigVMVariableNode() {variable=Current Root Bone}.Value)
  ]}
]
hierarchy [
 root {loc=<0 0 0>, rot=<0 0 0 1>}
 pelvis {parent=root, loc=<0 -0.7071 93.1191>, rot=<0.0127 0.707 -0.0127 -0.707>}
 spine_01 {parent=pelvis, loc=<6.7494 -0.6388 0>, rot=<0 0 0.034 -0.9994>}
 spine_02 {parent=spine_01, loc=<7.2518 -0.035 0>, rot=<0 0 0.0238 -0.9997>}
 spine_03 {parent=spine_02, loc=<6.0104 0 0>, rot=<0 0 0.0149 -0.9999>}
 spine_04 {parent=spine_03, loc=<8.1024 0 0>, rot=<0 0 -0.0132 -0.9999>}
 spine_05 {parent=spine_04, loc=<10.0146 0 0>, rot=<0 0 -0.0015 -1>}
 neck_01 {parent=spine_05, loc=<11.3138 0.1529 0>, rot=<0 0 -0.2003 -0.9797>}
 neck_02 {parent=neck_01, loc=<4.5627 0 0>, rot=<0 0 0.0504 -0.9987>}
 head {parent=neck_02, loc=<5.3914 0 0>, rot=<0 0 0.115 -0.9934>}
 FACIAL_C_FacialRoot {parent=head, loc=<2.9712 -0.4622 0>, rot=<0.5021 -0.4979 -0.5021 -0.4979>}
 FACIAL_C_Forehead {parent=FACIAL_C_FacialRoot, loc=<0 -3.1308 9.3974>, rot=<0.0207 0 0 -0.9998>}
 FACIAL_L_ForeheadIn {parent=FACIAL_C_FacialRoot, loc=<1.5783 -2.9083 9.9779>, rot=<0.0031 -0.0674 0.0105 -0.9977>}
 FACIAL_R_ForeheadIn {parent=FACIAL_C_FacialRoot, loc=<-1.5783 -2.9083 9.9779>, rot=<0.0031 0.0674 -0.0105 -0.9977>}
 FACIAL_L_ForeheadMid {parent=FACIAL_C_FacialRoot, loc=<2.867 -3.3915 9.4422>, rot=<-0.0082 -0.1749 -0.0076 -0.9845>}
 FACIAL_L_ForeheadMid1 {parent=FACIAL_L_ForeheadMid, loc=<-0.5077 0.2381 -0.0995>, rot=<0 0 0 1>}
 FACIAL_L_ForeheadMid2 {parent=FACIAL_L_ForeheadMid, loc=<0.674 0.1716 -0.0884>, rot=<0 0 0 1>}
 FACIAL_R_ForeheadMid {parent=FACIAL_C_FacialRoot, loc=<-2.867 -3.3915 9.4422>, rot=<-0.0082 0.1749 0.0076 -0.9845>}
 FACIAL_R_ForeheadMid1 {parent=FACIAL_R_ForeheadMid, loc=<0.5077 0.2381 -0.0995>, rot=<0 0 0 1>}
 FACIAL_R_ForeheadMid2 {parent=FACIAL_R_ForeheadMid, loc=<-0.674 0.1716 -0.0884>, rot=<0 0 0 1>}
 FACIAL_L_ForeheadOut {parent=FACIAL_C_FacialRoot, loc=<4.0465 -2.9836 8.7213>, rot=<-0.0602 -0.2584 -0.0611 -0.9622>}
 FACIAL_R_ForeheadOut {parent=FACIAL_C_FacialRoot, loc=<-4.0465 -2.9836 8.7213>, rot=<-0.0602 0.2584 0.0611 -0.9622>}
 FACIAL_L_EyesackUpper {parent=FACIAL_C_FacialRoot, loc=<2.8439 -2.3877 9.3504>, rot=<0.1124 -0.0867 -0.0571 -0.9882>}
 FACIAL_R_EyesackUpper {parent=FACIAL_C_FacialRoot, loc=<-2.8439 -2.3877 9.3504>, rot=<0.1124 0.0867 0.0571 -0.9882>}
 FACIAL_L_EyelidUpperFurrow {parent=FACIAL_C_FacialRoot, loc=<2.6585 -2.0955 9.0119>, rot=<0.0331 -0.154 -0.068 -0.9852>}
 FACIAL_R_EyelidUpperFurrow {parent=FACIAL_C_FacialRoot, loc=<-2.6585 -2.0955 9.0119>, rot=<0.0331 0.154 0.068 -0.9852>}
 FACIAL_L_EyelidUpperB {parent=FACIAL_C_FacialRoot, loc=<2.6949 -1.4649 7.7068>, rot=<-0.1618 0.002 0.0014 -0.9868>}
 FACIAL_R_EyelidUpperB {parent=FACIAL_C_FacialRoot, loc=<-2.6949 -1.4649 7.7068>, rot=<-0.1618 -0.002 -0.0014 -0.9868>}
 FACIAL_L_EyelidUpperA {parent=FACIAL_C_FacialRoot, loc=<2.6949 -1.4649 7.7068>, rot=<-0.0839 0.0019 0.0016 -0.9965>}
 FACIAL_L_EyelidUpperA1 {parent=FACIAL_L_EyelidUpperA, loc=<-0.6817 0.0696 1.1303>, rot=<0.0106 0.1214 0.0865 -0.9888>}
 FACIAL_L_EyelidUpperA3 {parent=FACIAL_L_EyelidUpperA, loc=<0.7566 -0.0356 1.1039>, rot=<-0.0086 -0.1657 -0.0487 -0.9849>}
 FACIAL_R_EyelidUpperA {parent=FACIAL_C_FacialRoot, loc=<-2.6949 -1.4649 7.7068>, rot=<-0.0839 -0.0019 -0.0016 -0.9965>}
 FACIAL_R_EyelidUpperA1 {parent=FACIAL_R_EyelidUpperA, loc=<0.6817 0.0696 1.1303>, rot=<0.0106 -0.1214 -0.0865 -0.9888>}
 FACIAL_R_EyelidUpperA3 {parent=FACIAL_R_EyelidUpperA, loc=<-0.7566 -0.0356 1.1039>, rot=<-0.0086 0.1657 0.0487 -0.9849>}
 FACIAL_L_Eye {parent=FACIAL_C_FacialRoot, loc=<2.6949 -1.4649 7.7068>, rot=<0.0643 0.0017 0.0019 -0.9979>}
 FACIAL_L_Pupil {parent=FACIAL_L_Eye, loc=<0.1563 -0.1178 1.1715>, rot=<0 0 0 1>}
 FACIAL_R_Eye {parent=FACIAL_C_FacialRoot, loc=<-2.6949 -1.4649 7.7068>, rot=<0.0643 -0.0017 -0.0019 -0.9979>}
 FACIAL_R_Pupil {parent=FACIAL_R_Eye, loc=<-0.1563 -0.1178 1.1715>, rot=<0 0 0 1>}
 FACIAL_L_EyelidLowerA {parent=FACIAL_C_FacialRoot, loc=<2.6949 -1.4649 7.7068>, rot=<0.3451 0.0011 0.0023 -0.9386>}
 FACIAL_L_EyelidLowerA1 {parent=FACIAL_L_EyelidLowerA, loc=<-0.7483 -0.3072 1.1061>, rot=<0 0.0785 0 -0.9969>}
 FACIAL_L_EyelidLowerA3 {parent=FACIAL_L_EyelidLowerA, loc=<0.6735 -0.1663 1.1484>, rot=<0 -0.1822 0 -0.9833>}
 FACIAL_R_EyelidLowerA {parent=FACIAL_C_FacialRoot, loc=<-2.6949 -1.4649 7.7068>, rot=<0.3451 -0.0011 -0.0023 -0.9386>}
 FACIAL_R_EyelidLowerA1 {parent=FACIAL_R_EyelidLowerA, loc=<0.7483 -0.3072 1.1061>, rot=<0 -0.0785 0 -0.9969>}
 FACIAL_R_EyelidLowerA3 {parent=FACIAL_R_EyelidLowerA, loc=<-0.6735 -0.1663 1.1484>, rot=<0 0.1822 0 -0.9833>}
 FACIAL_L_EyelidLowerB {parent=FACIAL_C_FacialRoot, loc=<2.6949 -1.4649 7.7068>, rot=<0.3776 0.001 0.0023 -0.926>}
 FACIAL_L_EyelidLowerB1 {parent=FACIAL_L_EyelidLowerB, loc=<-0.7153 -0.1974 1.2808>, rot=<0 0.0785 0 -0.9969>}
 FACIAL_L_EyelidLowerB3 {parent=FACIAL_L_EyelidLowerB, loc=<0.7461 0.0443 1.2303>, rot=<0 -0.1822 0 -0.9833>}
 FACIAL_R_EyelidLowerB {parent=FACIAL_C_FacialRoot, loc=<-2.6949 -1.4649 7.7068>, rot=<0.3776 -0.001 -0.0023 -0.926>}
 FACIAL_R_EyelidLowerB1 {parent=FACIAL_R_EyelidLowerB, loc=<0.7153 -0.1974 1.2808>, rot=<0 -0.0785 0 -0.9969>}
 FACIAL_R_EyelidLowerB3 {parent=FACIAL_R_EyelidLowerB, loc=<-0.7461 0.0443 1.2303>, rot=<0 0.1822 0 -0.9833>}
 FACIAL_L_EyeCornerInner {parent=FACIAL_C_FacialRoot, loc=<1.7099 -1.1996 8.5226>, rot=<0.0643 0.0017 0.0019 -0.9979>}
 FACIAL_R_EyeCornerInner {parent=FACIAL_C_FacialRoot, loc=<-1.7099 -1.1996 8.5226>, rot=<0.0643 -0.0017 -0.0019 -0.9979>}
 FACIAL_L_EyeCornerOuter {parent=FACIAL_C_FacialRoot, loc=<3.9155 -1.2959 7.935>, rot=<0.0587 -0.3803 0.0263 -0.9226>}
 FACIAL_R_EyeCornerOuter {parent=FACIAL_C_FacialRoot, loc=<-3.9155 -1.2959 7.935>, rot=<0.0587 0.3803 -0.0263 -0.9226>}
 FACIAL_L_EyesackLower {parent=FACIAL_C_FacialRoot, loc=<2.8327 -0.1952 8.2854>, rot=<0.0412 -0.1392 -0.0114 -0.9893>}
 FACIAL_L_EyesackLower1 {parent=FACIAL_L_EyesackLower, loc=<-0.992 -0.1607 0.241>, rot=<0.0204 0.1999 -0.0204 -0.9794>}
 FACIAL_L_EyesackLower2 {parent=FACIAL_L_EyesackLower, loc=<0.9592 -0.1302 0.2257>, rot=<0.0082 -0.0885 0.1026 -0.9907>}
 FACIAL_R_EyesackLower {parent=FACIAL_C_FacialRoot, loc=<-2.8327 -0.1952 8.2854>, rot=<0.0412 0.1392 0.0114 -0.9893>}
 FACIAL_R_EyesackLower1 {parent=FACIAL_R_EyesackLower, loc=<0.992 -0.1607 0.241>, rot=<0.0204 -0.1999 0.0204 -0.9794>}
 FACIAL_R_EyesackLower2 {parent=FACIAL_R_EyesackLower, loc=<-0.9592 -0.1302 0.2257>, rot=<0.0082 0.0885 -0.1026 -0.9907>}
 FACIAL_L_CheekInner {parent=FACIAL_C_FacialRoot, loc=<1.8741 0.5243 8.2984>, rot=<-0.0606 -0.0987 -0.1895 -0.975>}
 FACIAL_L_CheekInner1 {parent=FACIAL_L_CheekInner, loc=<-0.5285 -0.2617 0.8209>, rot=<0.0509 -0.0536 0.0234 -0.997>}
 FACIAL_L_CheekInner2 {parent=FACIAL_L_CheekInner, loc=<-0.319 0.3869 1.159>, rot=<0.0349 -0.0565 -0.1491 -0.9866>}
 FACIAL_L_CheekInner3 {parent=FACIAL_L_CheekInner, loc=<0.5554 -0.1496 0.7785>, rot=<0.0509 -0.0536 0.0234 -0.997>}
 FACIAL_L_CheekInner4 {parent=FACIAL_L_CheekInner, loc=<0.6441 0.7311 0.9265>, rot=<0.0349 -0.0565 -0.1491 -0.9866>}
 FACIAL_R_CheekInner {parent=FACIAL_C_FacialRoot, loc=<-1.8741 0.5243 8.2984>, rot=<-0.0606 0.0987 0.1895 -0.975>}
 FACIAL_R_CheekInner1 {parent=FACIAL_R_CheekInner, loc=<0.5285 -0.2617 0.8209>, rot=<0.0509 0.0536 -0.0234 -0.997>}
 FACIAL_R_CheekInner2 {parent=FACIAL_R_CheekInner, loc=<0.319 0.3869 1.159>, rot=<0.0349 0.0565 0.1491 -0.9866>}
 FACIAL_R_CheekInner3 {parent=FACIAL_R_CheekInner, loc=<-0.5554 -0.1496 0.7785>, rot=<0.0509 0.0536 -0.0234 -0.997>}
 FACIAL_R_CheekInner4 {parent=FACIAL_R_CheekInner, loc=<-0.6441 0.7311 0.9265>, rot=<0.0349 0.0565 0.1491 -0.9866>}
 FACIAL_L_CheekOuter {parent=FACIAL_C_FacialRoot, loc=<3.3414 0.7955 6.9957>, rot=<0.1132 -0.3684 0.1262 -0.9141>}
 FACIAL_L_CheekOuter1 {parent=FACIAL_L_CheekOuter, loc=<-0.5843 -0.5617 1.2992>, rot=<-0.0677 -0.0465 -0.0404 -0.9958>}
 FACIAL_L_CheekOuter2 {parent=FACIAL_L_CheekOuter, loc=<0.0775 1.0309 1.24>, rot=<0.0801 -0.0306 -0.0549 -0.9948>}
 FACIAL_L_CheekOuter3 {parent=FACIAL_L_CheekOuter, loc=<1.0345 -0.5023 1.2182>, rot=<-0.0491 -0.0605 0.2182 -0.9728>}
 FACIAL_R_CheekOuter {parent=FACIAL_C_FacialRoot, loc=<-3.3414 0.7955 6.9957>, rot=<0.1132 0.3684 -0.1262 -0.9141>}
 FACIAL_R_CheekOuter1 {parent=FACIAL_R_CheekOuter, loc=<0.5843 -0.5617 1.2992>, rot=<-0.0677 0.0465 0.0404 -0.9958>}
 FACIAL_R_CheekOuter2 {parent=FACIAL_R_CheekOuter, loc=<-0.0775 1.0309 1.24>, rot=<0.0801 0.0306 0.0549 -0.9948>}
 FACIAL_R_CheekOuter3 {parent=FACIAL_R_CheekOuter, loc=<-1.0345 -0.5023 1.2182>, rot=<-0.0491 0.0605 -0.2182 -0.9728>}
 FACIAL_C_NoseBridge {parent=FACIAL_C_FacialRoot, loc=<0 -1.8959 10.1812>, rot=<0.1078 0 0 -0.9942>}
 FACIAL_L_NasolabialBulge1 {parent=FACIAL_C_FacialRoot, loc=<3.2494 2.3983 8.6467>, rot=<-0.025 -0.2497 -0.3656 -0.8963>}
 FACIAL_R_NasolabialBulge1 {parent=FACIAL_C_FacialRoot, loc=<-3.2494 2.3983 8.6467>, rot=<-0.025 0.2497 0.3656 -0.8963>}
 FACIAL_L_NasolabialBulge {parent=FACIAL_C_FacialRoot, loc=<3.0622 4.338 7.415>, rot=<0.1164 -0.3737 0.0498 -0.9189>}
 FACIAL_L_NasolabialBulge2 {parent=FACIAL_L_NasolabialBulge, loc=<0.0915 -0.9316 0.7307>, rot=<-0.1022 0.041 0.057 -0.9923>}
 FACIAL_L_NasolabialBulge3 {parent=FACIAL_L_NasolabialBulge, loc=<-0.1478 0.616 0.8109>, rot=<-0.0099 0.0168 -0.0699 -0.9974>}
 FACIAL_R_NasolabialBulge {parent=FACIAL_C_FacialRoot, loc=<-3.0622 4.338 7.415>, rot=<0.1164 0.3737 -0.0498 -0.9189>}
 FACIAL_R_NasolabialBulge2 {parent=FACIAL_R_NasolabialBulge, loc=<-0.0915 -0.9316 0.7307>, rot=<-0.1022 -0.041 -0.057 -0.9923>}
 FACIAL_R_NasolabialBulge3 {parent=FACIAL_R_NasolabialBulge, loc=<0.1478 0.616 0.8109>, rot=<-0.0099 -0.0168 0.0699 -0.9974>}
 FACIAL_L_NasolabialFurrow {parent=FACIAL_C_FacialRoot, loc=<2.7204 2.6119 9.0327>, rot=<0.0264 -0.1474 -0.3533 -0.9235>}
 FACIAL_R_NasolabialFurrow {parent=FACIAL_C_FacialRoot, loc=<-2.7204 2.6119 9.0327>, rot=<0.0264 0.1474 0.3533 -0.9235>}
 FACIAL_L_CheekLower {parent=FACIAL_C_FacialRoot, loc=<3.5517 3.9289 6.316>, rot=<0.0916 -0.4209 0.0868 -0.8983>}
 FACIAL_L_CheekLower1 {parent=FACIAL_L_CheekLower, loc=<-0.3052 0.2977 0.8335>, rot=<0 0.0872 0 -0.9962>}
 FACIAL_L_CheekLower2 {parent=FACIAL_L_CheekLower, loc=<0.7344 0.1917 0.5743>, rot=<0 -0.0872 0 -0.9962>}
 FACIAL_R_CheekLower {parent=FACIAL_C_FacialRoot, loc=<-3.5517 3.9289 6.316>, rot=<0.0916 0.4209 -0.0868 -0.8983>}
 FACIAL_R_CheekLower1 {parent=FACIAL_R_CheekLower, loc=<0.3052 0.2977 0.8335>, rot=<0 -0.0872 0 -0.9962>}
 FACIAL_R_CheekLower2 {parent=FACIAL_R_CheekLower, loc=<-0.7344 0.1917 0.5743>, rot=<0 0.0872 0 -0.9962>}
 FACIAL_L_Ear {parent=FACIAL_C_FacialRoot, loc=<6.4207 0.0696 1.4054>, rot=<0.0382 -0.0006 -0.0243 -0.999>}
 FACIAL_R_Ear {parent=FACIAL_C_FacialRoot, loc=<-6.4207 0.0696 1.4054>, rot=<0.0382 0.0006 0.0243 -0.999>}
 FACIAL_C_Nose {parent=FACIAL_C_FacialRoot, loc=<0 0.1218 10.1331>, rot=<-0.0665 0 0 -0.9978>}
 FACIAL_C_NoseLower {parent=FACIAL_C_Nose, loc=<0 2.611 0.1959>, rot=<0.1305 0 0 -0.9914>}
 FACIAL_C_NoseTip {parent=FACIAL_C_Nose, loc=<0 1.7381 0.7792>, rot=<0.0872 0 0 -0.9962>}
 FACIAL_L_Nostril {parent=FACIAL_C_Nose, loc=<1.3616 1.8117 0.0109>, rot=<0.0173 0.1305 0.0023 -0.9913>}
 FACIAL_R_Nostril {parent=FACIAL_C_Nose, loc=<-1.3616 1.8117 0.0109>, rot=<0.0173 -0.1305 -0.0023 -0.9913>}
 FACIAL_C_LipUpperSkin {parent=FACIAL_C_FacialRoot, loc=<0 3.1175 10.6656>, rot=<0.0207 0 0 -0.9998>}
 FACIAL_L_LipUpperOuterSkin {parent=FACIAL_C_FacialRoot, loc=<2.4826 3.4496 9.4024>, rot=<-0.0193 -0.1728 -0.0993 -0.9797>}
 FACIAL_R_LipUpperOuterSkin {parent=FACIAL_C_FacialRoot, loc=<-2.4826 3.4496 9.4024>, rot=<-0.0193 0.1728 0.0993 -0.9797>}
 FACIAL_C_MouthUpper {parent=FACIAL_C_FacialRoot, loc=<0 4.4644 7.5076>, rot=<0.0643 0 0 -0.9979>}
 FACIAL_C_LipUpper {parent=FACIAL_C_MouthUpper, loc=<0 -0.9938 2.4973>, rot=<-0.0872 0 0 -0.9962>}
 FACIAL_C_LipUpper1 {parent=FACIAL_C_LipUpper, loc=<0 -0.1465 0.8579>, rot=<0 0 0 1>}
 FACIAL_C_LipUpper2 {parent=FACIAL_C_LipUpper, loc=<0 0.6591 0.1943>, rot=<0 0 0 1>}
 FACIAL_L_LipUpper {parent=FACIAL_C_MouthUpper, loc=<1.1829 -0.8863 2.1826>, rot=<-0.0778 -0.1601 -0.0379 -0.9833>}
 FACIAL_L_LipUpper1 {parent=FACIAL_L_LipUpper, loc=<-0.0571 -0.1403 0.6326>, rot=<0 0 0 1>}
 FACIAL_L_LipUpper2 {parent=FACIAL_L_LipUpper, loc=<-0.1281 0.5669 0.1244>, rot=<0 0 0 1>}
 FACIAL_R_LipUpper {parent=FACIAL_C_MouthUpper, loc=<-1.1829 -0.8863 2.1826>, rot=<-0.0778 0.1601 0.0379 -0.9833>}
 FACIAL_R_LipUpper1 {parent=FACIAL_R_LipUpper, loc=<0.0571 -0.1403 0.6326>, rot=<0 0 0 1>}
 FACIAL_R_LipUpper2 {parent=FACIAL_R_LipUpper, loc=<0.1281 0.5669 0.1244>, rot=<0 0 0 1>}
 FACIAL_L_LipUpperOuter {parent=FACIAL_C_MouthUpper, loc=<1.9999 -0.6965 1.5448>, rot=<-0.0983 -0.2294 -0.0983 -0.9634>}
 FACIAL_L_LipUpperOuter1 {parent=FACIAL_L_LipUpperOuter, loc=<-0.0959 0.1942 0.5095>, rot=<0 0 0 1>}
 FACIAL_L_LipUpperOuter2 {parent=FACIAL_L_LipUpperOuter, loc=<-0.2529 0.5561 0.2264>, rot=<0 0 0 1>}
 FACIAL_R_LipUpperOuter {parent=FACIAL_C_MouthUpper, loc=<-1.9999 -0.6965 1.5448>, rot=<-0.0983 0.2294 0.0983 -0.9634>}
 FACIAL_R_LipUpperOuter1 {parent=FACIAL_R_LipUpperOuter, loc=<0.0959 0.1942 0.5095>, rot=<0 0 0 1>}
 FACIAL_R_LipUpperOuter2 {parent=FACIAL_R_LipUpperOuter, loc=<0.2529 0.5561 0.2264>, rot=<0 0 0 1>}
 FACIAL_L_LipCorner {parent=FACIAL_C_MouthUpper, loc=<2.619 -0.1516 1.046>, rot=<0 -0.2079 0 -0.9781>}
 FACIAL_L_LipCorner2 {parent=FACIAL_L_LipCorner, loc=<-0.1424 -0.0129 -0.7513>, rot=<0 0.2588 0 -0.9659>}
 FACIAL_L_LipCorner3 {parent=FACIAL_L_LipCorner, loc=<-0.4025 0.5791 -0.5869>, rot=<0 0.2588 0 -0.9659>}
 FACIAL_R_LipCorner {parent=FACIAL_C_MouthUpper, loc=<-2.6191 -0.1516 1.046>, rot=<0 0.2079 0 -0.9781>}
 FACIAL_R_LipCorner2 {parent=FACIAL_R_LipCorner, loc=<0.1424 -0.0129 -0.7513>, rot=<0 -0.2588 0 -0.9659>}
 FACIAL_R_LipCorner3 {parent=FACIAL_R_LipCorner, loc=<0.4025 0.5791 -0.5869>, rot=<0 -0.2588 0 -0.9659>}
 FACIAL_L_JawBulge {parent=FACIAL_C_FacialRoot, loc=<5.2193 3.4628 3.4629>, rot=<0.1506 0.0886 0.004 -0.9846>}
 FACIAL_R_JawBulge {parent=FACIAL_C_FacialRoot, loc=<-5.2193 3.4628 3.4629>, rot=<0.1506 -0.0886 -0.004 -0.9846>}
 FACIAL_L_Masseter {parent=FACIAL_C_FacialRoot, loc=<5.3206 1.6285 4.5667>, rot=<0.1331 0.1147 0.0026 -0.9844>}
 FACIAL_R_Masseter {parent=FACIAL_C_FacialRoot, loc=<-5.3206 1.6285 4.5667>, rot=<0.1331 -0.1147 -0.0026 -0.9844>}
 FACIAL_C_UnderChin {parent=FACIAL_C_FacialRoot, loc=<0 8.5606 6.1576>, rot=<0.5546 0 0 -0.8321>}
 FACIAL_L_UnderChin {parent=FACIAL_C_FacialRoot, loc=<3.1037 7.3473 4.3729>, rot=<0.4548 -0.3483 0.1072 -0.8126>}
 FACIAL_R_UnderChin {parent=FACIAL_C_FacialRoot, loc=<-3.1037 7.3473 4.3729>, rot=<0.4548 0.3483 -0.1072 -0.8126>}
 FACIAL_C_TeethUpper {parent=FACIAL_C_FacialRoot, loc=<0 4.3737 9.6558>, rot=<0.0643 0 0 -0.9979>}
 FACIAL_C_LowerLipRotation {parent=FACIAL_C_FacialRoot, loc=<0 1.9413 2.9092>, rot=<0.4017 0 0 -0.9158>}
 FACIAL_C_LipLowerSkin {parent=FACIAL_C_LowerLipRotation, loc=<0 -2.158 8.1295>, rot=<-0.2588 0 0 -0.9659>}
 FACIAL_L_LipLowerOuterSkin {parent=FACIAL_C_LowerLipRotation, loc=<2.5044 -2.0345 6.9892>, rot=<-0.2809 -0.2529 0.0021 -0.9258>}
 FACIAL_R_LipLowerOuterSkin {parent=FACIAL_C_LowerLipRotation, loc=<-2.5044 -2.0345 6.9892>, rot=<-0.2809 0.2529 -0.0021 -0.9258>}
 FACIAL_C_MouthLower {parent=FACIAL_C_LowerLipRotation, loc=<0 -1.6746 4.9707>, rot=<-0.342 0 0 -0.9397>}
 FACIAL_C_LipLower {parent=FACIAL_C_MouthLower, loc=<0 0.5544 2.7134>, rot=<0.0872 0 0 -0.9962>}
 FACIAL_C_LipLower1 {parent=FACIAL_C_LipLower, loc=<0 0.168 0.5115>, rot=<0 0 0 1>}
 FACIAL_C_LipLower2 {parent=FACIAL_C_LipLower, loc=<0 -0.8438 -0.3887>, rot=<0 0 0 1>}
 FACIAL_L_LipLower {parent=FACIAL_C_MouthLower, loc=<1.1262 0.4533 2.3553>, rot=<0.0927 -0.1295 0.0273 -0.9869>}
 FACIAL_L_LipLower1 {parent=FACIAL_L_LipLower, loc=<0.2662 0.1675 0.5708>, rot=<0 0 0 1>}
 FACIAL_L_LipLower2 {parent=FACIAL_L_LipLower, loc=<-0.0195 -0.7652 -0.2768>, rot=<0 0 0 1>}
 FACIAL_R_LipLower {parent=FACIAL_C_MouthLower, loc=<-1.1263 0.4533 2.3553>, rot=<0.0927 0.1295 -0.0273 -0.9869>}
 FACIAL_R_LipLower1 {parent=FACIAL_R_LipLower, loc=<-0.2662 0.1675 0.5708>, rot=<0 0 0 1>}
 FACIAL_R_LipLower2 {parent=FACIAL_R_LipLower, loc=<0.0195 -0.7652 -0.2768>, rot=<0 0 0 1>}
 FACIAL_L_LipLowerOuter {parent=FACIAL_C_MouthLower, loc=<1.8907 0.2989 1.9184>, rot=<0.0913 -0.2322 0.0318 -0.9679>}
 FACIAL_L_LipLowerOuter1 {parent=FACIAL_L_LipLowerOuter, loc=<0.1099 0.008 0.4186>, rot=<0 0 0 -1>}
 FACIAL_L_LipLowerOuter2 {parent=FACIAL_L_LipLowerOuter, loc=<-0.0021 -0.562 -0.1956>, rot=<0 0 0 -1>}
 FACIAL_R_LipLowerOuter {parent=FACIAL_C_MouthLower, loc=<-1.8907 0.2989 1.9184>, rot=<0.0913 0.2322 -0.0318 -0.9679>}
 FACIAL_R_LipLowerOuter1 {parent=FACIAL_R_LipLowerOuter, loc=<-0.1099 0.008 0.4186>, rot=<0 0 0 -1>}
 FACIAL_R_LipLowerOuter2 {parent=FACIAL_R_LipLowerOuter, loc=<0.0021 -0.562 -0.1956>, rot=<0 0 0 -1>}
 FACIAL_C_TeethLower {parent=FACIAL_C_LowerLipRotation, loc=<0 -3.3107 6.3943>, rot=<-0.342 0 0 -0.9397>}
 FACIAL_C_Tongue1 {parent=FACIAL_C_TeethLower, loc=<0 0.8814 -4.9974>, rot=<-0.2588 0 0 -0.9659>}
 FACIAL_C_Tongue2 {parent=FACIAL_C_Tongue1, loc=<0 0.4172 1.4811>, rot=<0.2588 0 0 -0.9659>}
 FACIAL_C_TongueUpper2 {parent=FACIAL_C_Tongue2, loc=<-0.0004 -0.5559 -0.0593>, rot=<0 0 0 1>}
 FACIAL_L_TongueSide2 {parent=FACIAL_C_Tongue2, loc=<1.7649 0.1107 0.3121>, rot=<0 0 0 1>}
 FACIAL_R_TongueSide2 {parent=FACIAL_C_Tongue2, loc=<-1.7649 0.1107 0.3121>, rot=<0 0 0 -1>}
 FACIAL_C_Tongue3 {parent=FACIAL_C_Tongue2, loc=<0 0.2069 1.3354>, rot=<0.1305 0 0 -0.9914>}
 FACIAL_C_TongueUpper3 {parent=FACIAL_C_Tongue3, loc=<-0.0003 -0.3509 -0.1572>, rot=<0 0 0 -1>}
 FACIAL_C_TongueLower3 {parent=FACIAL_C_Tongue3, loc=<0 0.8092 0.0947>, rot=<0 0 0 -1>}
 FACIAL_L_TongueSide3 {parent=FACIAL_C_Tongue3, loc=<1.5325 0.0121 0.1487>, rot=<0 0 0 -1>}
 FACIAL_R_TongueSide3 {parent=FACIAL_C_Tongue3, loc=<-1.5325 0.0121 0.1487>, rot=<0 0 0 -1>}
 FACIAL_C_Tongue4 {parent=FACIAL_C_Tongue3, loc=<0.0006 -0.1867 1.1247>, rot=<0.0436 0 0 -0.999>}
 FACIAL_C_Jaw {parent=FACIAL_C_FacialRoot, loc=<0 1.9413 2.9092>, rot=<0.4017 0 0 -0.9158>}
 FACIAL_C_Jawline {parent=FACIAL_C_Jaw, loc=<0 0.9567 8.3398>, rot=<0.2588 0 0 -0.9659>}
 FACIAL_L_Jawline {parent=FACIAL_C_Jaw, loc=<4.3349 0.5774 4.2447>, rot=<-0.1329 0.167 0.0026 -0.977>}
 FACIAL_R_Jawline {parent=FACIAL_C_Jaw, loc=<-4.3349 0.5774 4.2447>, rot=<-0.1329 -0.167 -0.0026 -0.977>}
 FACIAL_L_ChinSide {parent=FACIAL_C_Jaw, loc=<2.9641 -0.948 6.7957>, rot=<-0.2207 -0.3739 0.05 -0.8994>}
 FACIAL_R_ChinSide {parent=FACIAL_C_Jaw, loc=<-2.9641 -0.948 6.7957>, rot=<-0.2207 0.3739 -0.05 -0.8994>}
 FACIAL_C_Chin {parent=FACIAL_C_Jaw, loc=<0 -1.7958 8.7593>, rot=<-0.4226 0 0 -0.9063>}
 FACIAL_C_Chin3 {parent=FACIAL_C_Jaw, loc=<0 -0.3859 9.3079>, rot=<0.0436 0 0 -0.999>}
 clavicle_l {parent=spine_05, loc=<4.3575 -0.1466 -2.3223>, rot=<0.0465 -0.7289 -0.009 -0.6829>}
 upperarm_l {parent=clavicle_l, loc=<15.5111 0 0>, rot=<0.0226 -0.3234 -0.0088 -0.9459>}
 clavicle_r {parent=spine_05, loc=<4.3575 -0.1466 2.3223>, rot=<0.7289 0.0465 -0.6829 0.009>}
 upperarm_r {parent=clavicle_r, loc=<-15.5111 0 0>, rot=<0.0226 -0.3234 -0.0088 -0.9459>}
 CTRL_expressions_browDINL {type=Curve, loc=<0 0 0>, rot=<0 0 0 1>}
 CTRL_expressions_browDINR {type=Curve, loc=<0 0 0>, rot=<0 0 0 1>}
 CTRL_expressions_browDOUTL {type=Curve, loc=<0 0 0>, rot=<0 0 0 1>}
 CTRL_expressions_browDOUTR {type=Curve, loc=<0 0 0>, rot=<0 0 0 1>}
 CTRL_expressions_browDownL {type=Curve, loc=<0 0 0>, rot=<0 0 0 1>}
 CTRL_expressions_browDownR {type=Curve, loc=<0 0 0>, rot=<0 0 0 1>}
 CTRL_expressions_browInnerInL {type=Curve, loc=<0 0 0>, rot=<0 0 0 1>}
 CTRL_expressions_browInnerInR {type=Curve, loc=<0 0 0>, rot=<0 0 0 1>}
 CTRL_expressions_browInnerMidDownL {type=Curve, loc=<0 0 0>, rot=<0 0 0 1>}
 CTRL_expressions_browInnerMidDownR {type=Curve, loc=<0 0 0>, rot=<0 0 0 1>}
 CTRL_expressions_browInnerMidInL {type=Curve, loc=<0 0 0>, rot=<0 0 0 1>}
 CTRL_expressions_browInnerMidInR {type=Curve, loc=<0 0 0>, rot=<0 0 0 1>}
 CTRL_expressions_browInnerMidOutL {type=Curve, loc=<0 0 0>, rot=<0 0 0 1>}
 CTRL_expressions_browInnerMidOutR {type=Curve, loc=<0 0 0>, rot=<0 0 0 1>}
 CTRL_expressions_browInnerMidUpL {type=Curve, loc=<0 0 0>, rot=<0 0 0 1>}
 CTRL_expressions_browInnerMidUpR {type=Curve, loc=<0 0 0>, rot=<0 0 0 1>}
 CTRL_expressions_browInnerOutL {type=Curve, loc=<0 0 0>, rot=<0 0 0 1>}
 CTRL_expressions_browInnerOutR {type=Curve, loc=<0 0 0>, rot=<0 0 0 1>}
 CTRL_expressions_browLateralL {type=Curve, loc=<0 0 0>, rot=<0 0 0 1>}
 CTRL_expressions_browLateralR {type=Curve, loc=<0 0 0>, rot=<0 0 0 1>}
 CTRL_expressions_browOuterMidDownL {type=Curve, loc=<0 0 0>, rot=<0 0 0 1>}
 CTRL_expressions_browOuterMidDownR {type=Curve, loc=<0 0 0>, rot=<0 0 0 1>}
 CTRL_expressions_browOuterMidInL {type=Curve, loc=<0 0 0>, rot=<0 0 0 1>}
 CTRL_expressions_browOuterMidInR {type=Curve, loc=<0 0 0>, rot=<0 0 0 1>}
 CTRL_expressions_browOuterMidOutL {type=Curve, loc=<0 0 0>, rot=<0 0 0 1>}
 CTRL_expressions_browOuterMidOutR {type=Curve, loc=<0 0 0>, rot=<0 0 0 1>}
 CTRL_expressions_browOuterMidUpL {type=Curve, loc=<0 0 0>, rot=<0 0 0 1>}
 CTRL_expressions_browOuterMidUpR {type=Curve, loc=<0 0 0>, rot=<0 0 0 1>}
 CTRL_expressions_browRaiseInL {type=Curve, loc=<0 0 0>, rot=<0 0 0 1>}
 CTRL_expressions_browRaiseInR {type=Curve, loc=<0 0 0>, rot=<0 0 0 1>}
 CTRL_expressions_browRaiseOuterL {type=Curve, loc=<0 0 0>, rot=<0 0 0 1>}
 CTRL_expressions_browRaiseOuterR {type=Curve, loc=<0 0 0>, rot=<0 0 0 1>}
 CTRL_expressions_browRollNegInL {type=Curve, loc=<0 0 0>, rot=<0 0 0 1>}
 CTRL_expressions_browRollNegInR {type=Curve, loc=<0 0 0>, rot=<0 0 0 1>}
 CTRL_expressions_browRollNegOutL {type=Curve, loc=<0 0 0>, rot=<0 0 0 1>}
 CTRL_expressions_browRollNegOutR {type=Curve, loc=<0 0 0>, rot=<0 0 0 1>}
 CTRL_expressions_browRollPosInL {type=Curve, loc=<0 0 0>, rot=<0 0 0 1>}
 CTRL_expressions_browRollPosInR {type=Curve, loc=<0 0 0>, rot=<0 0 0 1>}
 CTRL_expressions_browRollPosOutL {type=Curve, loc=<0 0 0>, rot=<0 0 0 1>}
 CTRL_expressions_browRollPosOutR {type=Curve, loc=<0 0 0>, rot=<0 0 0 1>}
 CTRL_expressions_browUINL {type=Curve, loc=<0 0 0>, rot=<0 0 0 1>}
 CTRL_expressions_browUINR {type=Curve, loc=<0 0 0>, rot=<0 0 0 1>}
 CTRL_expressions_browUOUTL {type=Curve, loc=<0 0 0>, rot=<0 0 0 1>}
 CTRL_expressions_browUOUTR {type=Curve, loc=<0 0 0>, rot=<0 0 0 1>}
 CTRL_expressions_eyeBlinkL {type=Curve, loc=<0 0 0>, rot=<0 0 0 1>}
 CTRL_expressions_eyeBlinkR {type=Curve, loc=<0 0 0>, rot=<0 0 0 1>}
 CTRL_expressions_eyeCheekRaiseL {type=Curve, loc=<0 0 0>, rot=<0 0 0 1>}
 CTRL_expressions_eyeCheekRaiseR {type=Curve, loc=<0 0 0>, rot=<0 0 0 1>}
 CTRL_expressions_eyeEyelidsReduce {type=Curve, loc=<0 0 0>, rot=<0 0 0 1>}
 CTRL_expressions_eyeFaceScrunchL {type=Curve, loc=<0 0 0>, rot=<0 0 0 1>}
 CTRL_expressions_eyeFaceScrunchR {type=Curve, loc=<0 0 0>, rot=<0 0 0 1>}
 CTRL_expressions_eyelidRollNegDL {type=Curve, loc=<0 0 0>, rot=<0 0 0 1>}
 CTRL_expressions_eyelidRollNegDR {type=Curve, loc=<0 0 0>, rot=<0 0 0 1>}
 CTRL_expressions_eyelidRollNegUL {type=Curve, loc=<0 0 0>, rot=<0 0 0 1>}
 CTRL_expressions_eyelidRollNegUR {type=Curve, loc=<0 0 0>, rot=<0 0 0 1>}
 CTRL_expressions_eyelidRollPosDL {type=Curve, loc=<0 0 0>, rot=<0 0 0 1>}
 CTRL_expressions_eyelidRollPosDR {type=Curve, loc=<0 0 0>, rot=<0 0 0 1>}
 CTRL_expressions_eyelidRollPosUL {type=Curve, loc=<0 0 0>, rot=<0 0 0 1>}
 CTRL_expressions_eyelidRollPosUR {type=Curve, loc=<0 0 0>, rot=<0 0 0 1>}
 CTRL_expressions_eyeLookDownL {type=Curve, loc=<0 0 0>, rot=<0 0 0 1>}
 CTRL_expressions_eyeLookDownR {type=Curve, loc=<0 0 0>, rot=<0 0 0 1>}
 CTRL_expressions_eyeLookLeftL {type=Curve, loc=<0 0 0>, rot=<0 0 0 1>}
 CTRL_expressions_eyeLookLeftR {type=Curve, loc=<0 0 0>, rot=<0 0 0 1>}
 CTRL_expressions_eyeLookRightL {type=Curve, loc=<0 0 0>, rot=<0 0 0 1>}
 CTRL_expressions_eyeLookRightR {type=Curve, loc=<0 0 0>, rot=<0 0 0 1>}
 CTRL_expressions_eyeLookUpL {type=Curve, loc=<0 0 0>, rot=<0 0 0 1>}
 CTRL_expressions_eyeLookUpR {type=Curve, loc=<0 0 0>, rot=<0 0 0 1>}
 CTRL_expressions_eyeLowerLidDownL {type=Curve, loc=<0 0 0>, rot=<0 0 0 1>}
 CTRL_expressions_eyeLowerLidDownR {type=Curve, loc=<0 0 0>, rot=<0 0 0 1>}
 CTRL_expressions_eyeLowerLidUpL {type=Curve, loc=<0 0 0>, rot=<0 0 0 1>}
 CTRL_expressions_eyeLowerLidUpR {type=Curve, loc=<0 0 0>, rot=<0 0 0 1>}
 CTRL_expressions_eyePupilNarrowL {type=Curve, loc=<0 0 0>, rot=<0 0 0 1>}
 CTRL_expressions_eyePupilNarrowR {type=Curve, loc=<0 0 0>, rot=<0 0 0 1>}
 CTRL_expressions_eyePupilWideL {type=Curve, loc=<0 0 0>, rot=<0 0 0 1>}
 CTRL_expressions_eyePupilWideR {type=Curve, loc=<0 0 0>, rot=<0 0 0 1>}
 CTRL_expressions_eyeRelaxL {type=Curve, loc=<0 0 0>, rot=<0 0 0 1>}
 CTRL_expressions_eyeRelaxR {type=Curve, loc=<0 0 0>, rot=<0 0 0 1>}
 CTRL_expressions_eyeSquintInnerL {type=Curve, loc=<0 0 0>, rot=<0 0 0 1>}
 CTRL_expressions_eyeSquintInnerR {type=Curve, loc=<0 0 0>, rot=<0 0 0 1>}
 CTRL_expressions_eyeUpperLidUpL {type=Curve, loc=<0 0 0>, rot=<0 0 0 1>}
 CTRL_expressions_eyeUpperLidUpR {type=Curve, loc=<0 0 0>, rot=<0 0 0 1>}
 CTRL_expressions_eyeWidenL {type=Curve, loc=<0 0 0>, rot=<0 0 0 1>}
 CTRL_expressions_eyeWidenR {type=Curve, loc=<0 0 0>, rot=<0 0 0 1>}
 CTRL_expressions_jawBack {type=Curve, loc=<0 0 0>, rot=<0 0 0 1>}
 CTRL_expressions_jawChinRaiseDL {type=Curve, loc=<0 0 0>, rot=<0 0 0 1>}
 CTRL_expressions_jawChinRaiseDR {type=Curve, loc=<0 0 0>, rot=<0 0 0 1>}
 CTRL_expressions_jawChinRaiseUL {type=Curve, loc=<0 0 0>, rot=<0 0 0 1>}
 CTRL_expressions_jawChinRaiseUR {type=Curve, loc=<0 0 0>, rot=<0 0 0 1>}
 CTRL_expressions_jawFwd {type=Curve, loc=<0 0 0>, rot=<0 0 0 1>}
 CTRL_expressions_jawLeft {type=Curve, loc=<0 0 0>, rot=<0 0 0 1>}
 CTRL_expressions_jawOpen {type=Curve, loc=<0 0 0>, rot=<0 0 0 1>}
 CTRL_expressions_jawOpenExtreme {type=Curve, loc=<0 0 0>, rot=<0 0 0 1>}
 CTRL_expressions_jawRight {type=Curve, loc=<0 0 0>, rot=<0 0 0 1>}
 CTRL_expressions_mouthCheekBlowL {type=Curve, loc=<0 0 0>, rot=<0 0 0 1>}
 CTRL_expressions_mouthCheekBlowR {type=Curve, loc=<0 0 0>, rot=<0 0 0 1>}
 CTRL_expressions_mouthCheekSuckL {type=Curve, loc=<0 0 0>, rot=<0 0 0 1>}
 CTRL_expressions_mouthCheekSuckR {type=Curve, loc=<0 0 0>, rot=<0 0 0 1>}
 CTRL_expressions_mouthCornerDepressL {type=Curve, loc=<0 0 0>, rot=<0 0 0 1>}
 CTRL_expressions_mouthCornerDepressR {type=Curve, loc=<0 0 0>, rot=<0 0 0 1>}
 CTRL_expressions_mouthCornerDownL {type=Curve, loc=<0 0 0>, rot=<0 0 0 1>}
 CTRL_expressions_mouthCornerDownR {type=Curve, loc=<0 0 0>, rot=<0 0 0 1>}
 CTRL_expressions_mouthCornerNarrowL {type=Curve, loc=<0 0 0>, rot=<0 0 0 1>}
 CTRL_expressions_mouthCornerNarrowR {type=Curve, loc=<0 0 0>, rot=<0 0 0 1>}
 CTRL_expressions_mouthCornerPullL {type=Curve, loc=<0 0 0>, rot=<0 0 0 1>}
 CTRL_expressions_mouthCornerPullR {type=Curve, loc=<0 0 0>, rot=<0 0 0 1>}
 CTRL_expressions_mouthCornerRounderDL {type=Curve, loc=<0 0 0>, rot=<0 0 0 1>}
 CTRL_expressions_mouthCornerRounderDR {type=Curve, loc=<0 0 0>, rot=<0 0 0 1>}
 CTRL_expressions_mouthCornerRounderUL {type=Curve, loc=<0 0 0>, rot=<0 0 0 1>}
 CTRL_expressions_mouthCornerRounderUR {type=Curve, loc=<0 0 0>, rot=<0 0 0 1>}
 CTRL_expressions_mouthCornerSharpenDL {type=Curve, loc=<0 0 0>, rot=<0 0 0 1>}
 CTRL_expressions_mouthCornerSharpenDR {type=Curve, loc=<0 0 0>, rot=<0 0 0 1>}
 CTRL_expressions_mouthCornerSharpenUL {type=Curve, loc=<0 0 0>, rot=<0 0 0 1>}
 CTRL_expressions_mouthCornerSharpenUR {type=Curve, loc=<0 0 0>, rot=<0 0 0 1>}
 CTRL_expressions_mouthCornerUpL {type=Curve, loc=<0 0 0>, rot=<0 0 0 1>}
 CTRL_expressions_mouthCornerUpR {type=Curve, loc=<0 0 0>, rot=<0 0 0 1>}
 CTRL_expressions_mouthCornerWideL {type=Curve, loc=<0 0 0>, rot=<0 0 0 1>}
 CTRL_expressions_mouthCornerWideR {type=Curve, loc=<0 0 0>, rot=<0 0 0 1>}
 CTRL_expressions_mouthDimpleL {type=Curve, loc=<0 0 0>, rot=<0 0 0 1>}
 CTRL_expressions_mouthDimpleR {type=Curve, loc=<0 0 0>, rot=<0 0 0 1>}
 CTRL_expressions_mouthDown {type=Curve, loc=<0 0 0>, rot=<0 0 0 1>}
 CTRL_expressions_mouthFunnelDL {type=Curve, loc=<0 0 0>, rot=<0 0 0 1>}
 CTRL_expressions_mouthFunnelDR {type=Curve, loc=<0 0 0>, rot=<0 0 0 1>}
 CTRL_expressions_mouthFunnelUL {type=Curve, loc=<0 0 0>, rot=<0 0 0 1>}
 CTRL_expressions_mouthFunnelUR {type=Curve, loc=<0 0 0>, rot=<0 0 0 1>}
 CTRL_expressions_mouthLeft {type=Curve, loc=<0 0 0>, rot=<0 0 0 1>}
 CTRL_expressions_mouthLipsBlowL {type=Curve, loc=<0 0 0>, rot=<0 0 0 1>}
 CTRL_expressions_mouthLipsBlowR {type=Curve, loc=<0 0 0>, rot=<0 0 0 1>}
 CTRL_expressions_mouthLipsPressL {type=Curve, loc=<0 0 0>, rot=<0 0 0 1>}
 CTRL_expressions_mouthLipsPressR {type=Curve, loc=<0 0 0>, rot=<0 0 0 1>}
 CTRL_expressions_mouthLipsPullDL {type=Curve, loc=<0 0 0>, rot=<0 0 0 1>}
 CTRL_expressions_mouthLipsPullDR {type=Curve, loc=<0 0 0>, rot=<0 0 0 1>}
 CTRL_expressions_mouthLipsPullUL {type=Curve, loc=<0 0 0>, rot=<0 0 0 1>}
 CTRL_expressions_mouthLipsPullUR {type=Curve, loc=<0 0 0>, rot=<0 0 0 1>}
 CTRL_expressions_mouthLipsPurseDL {type=Curve, loc=<0 0 0>, rot=<0 0 0 1>}
 CTRL_expressions_mouthLipsPurseDR {type=Curve, loc=<0 0 0>, rot=<0 0 0 1>}
 CTRL_expressions_mouthLipsPurseUL {type=Curve, loc=<0 0 0>, rot=<0 0 0 1>}
 CTRL_expressions_mouthLipsPurseUR {type=Curve, loc=<0 0 0>, rot=<0 0 0 1>}
 CTRL_expressions_mouthLipsPushDL {type=Curve, loc=<0 0 0>, rot=<0 0 0 1>}
 CTRL_expressions_mouthLipsPushDR {type=Curve, loc=<0 0 0>, rot=<0 0 0 1>}
 CTRL_expressions_mouthLipsPushUL {type=Curve, loc=<0 0 0>, rot=<0 0 0 1>}
 CTRL_expressions_mouthLipsPushUR {type=Curve, loc=<0 0 0>, rot=<0 0 0 1>}
 CTRL_expressions_mouthLipsThickDL {type=Curve, loc=<0 0 0>, rot=<0 0 0 1>}
 CTRL_expressions_mouthLipsThickDR {type=Curve, loc=<0 0 0>, rot=<0 0 0 1>}
 CTRL_expressions_mouthLipsThickInwardDL {type=Curve, loc=<0 0 0>, rot=<0 0 0 1>}
 CTRL_expressions_mouthLipsThickInwardDR {type=Curve, loc=<0 0 0>, rot=<0 0 0 1>}
 CTRL_expressions_mouthLipsThickInwardUL {type=Curve, loc=<0 0 0>, rot=<0 0 0 1>}
 CTRL_expressions_mouthLipsThickInwardUR {type=Curve, loc=<0 0 0>, rot=<0 0 0 1>}
 CTRL_expressions_mouthLipsThickUL {type=Curve, loc=<0 0 0>, rot=<0 0 0 1>}
 CTRL_expressions_mouthLipsThickUR {type=Curve, loc=<0 0 0>, rot=<0 0 0 1>}
 CTRL_expressions_mouthLipsThinDL {type=Curve, loc=<0 0 0>, rot=<0 0 0 1>}
 CTRL_expressions_mouthLipsThinDR {type=Curve, loc=<0 0 0>, rot=<0 0 0 1>}
 CTRL_expressions_mouthLipsThinInwardDL {type=Curve, loc=<0 0 0>, rot=<0 0 0 1>}
 CTRL_expressions_mouthLipsThinInwardDR {type=Curve, loc=<0 0 0>, rot=<0 0 0 1>}
 CTRL_expressions_mouthLipsThinInwardUL {type=Curve, loc=<0 0 0>, rot=<0 0 0 1>}
 CTRL_expressions_mouthLipsThinInwardUR {type=Curve, loc=<0 0 0>, rot=<0 0 0 1>}
 CTRL_expressions_mouthLipsThinUL {type=Curve, loc=<0 0 0>, rot=<0 0 0 1>}
 CTRL_expressions_mouthLipsThinUR {type=Curve, loc=<0 0 0>, rot=<0 0 0 1>}
 CTRL_expressions_mouthLipsTightenDL {type=Curve, loc=<0 0 0>, rot=<0 0 0 1>}
 CTRL_expressions_mouthLipsTightenDR {type=Curve, loc=<0 0 0>, rot=<0 0 0 1>}
 CTRL_expressions_mouthLipsTightenUL {type=Curve, loc=<0 0 0>, rot=<0 0 0 1>}
 CTRL_expressions_mouthLipsTightenUR {type=Curve, loc=<0 0 0>, rot=<0 0 0 1>}
 CTRL_expressions_mouthLipsTogetherDL {type=Curve, loc=<0 0 0>, rot=<0 0 0 1>}
 CTRL_expressions_mouthLipsTogetherDR {type=Curve, loc=<0 0 0>, rot=<0 0 0 1>}
 CTRL_expressions_mouthLipsTogetherUL {type=Curve, loc=<0 0 0>, rot=<0 0 0 1>}
 CTRL_expressions_mouthLipsTogetherUR {type=Curve, loc=<0 0 0>, rot=<0 0 0 1>}
 CTRL_expressions_mouthLipsTowardsDL {type=Curve, loc=<0 0 0>, rot=<0 0 0 1>}
 CTRL_expressions_mouthLipsTowardsDR {type=Curve, loc=<0 0 0>, rot=<0 0 0 1>}
 CTRL_expressions_mouthLipsTowardsUL {type=Curve, loc=<0 0 0>, rot=<0 0 0 1>}
 CTRL_expressions_mouthLipsTowardsUR {type=Curve, loc=<0 0 0>, rot=<0 0 0 1>}
 CTRL_expressions_mouthLowerLipBiteL {type=Curve, loc=<0 0 0>, rot=<0 0 0 1>}
 CTRL_expressions_mouthLowerLipBiteR {type=Curve, loc=<0 0 0>, rot=<0 0 0 1>}
 CTRL_expressions_mouthLowerLipDepressL {type=Curve, loc=<0 0 0>, rot=<0 0 0 1>}
 CTRL_expressions_mouthLowerLipDepressR {type=Curve, loc=<0 0 0>, rot=<0 0 0 1>}
 CTRL_expressions_mouthLowerLipRollInL {type=Curve, loc=<0 0 0>, rot=<0 0 0 1>}
 CTRL_expressions_mouthLowerLipRollInR {type=Curve, loc=<0 0 0>, rot=<0 0 0 1>}
 CTRL_expressions_mouthLowerLipRollOutL {type=Curve, loc=<0 0 0>, rot=<0 0 0 1>}
 CTRL_expressions_mouthLowerLipRollOutR {type=Curve, loc=<0 0 0>, rot=<0 0 0 1>}
 CTRL_expressions_mouthLowerLipTowardsTeethL {type=Curve, loc=<0 0 0>, rot=<0 0 0 1>}
 CTRL_expressions_mouthLowerLipTowardsTeethR {type=Curve, loc=<0 0 0>, rot=<0 0 0 1>}
 CTRL_expressions_mouthPressDL {type=Curve, loc=<0 0 0>, rot=<0 0 0 1>}
 CTRL_expressions_mouthPressDR {type=Curve, loc=<0 0 0>, rot=<0 0 0 1>}
 CTRL_expressions_mouthPressUL {type=Curve, loc=<0 0 0>, rot=<0 0 0 1>}
 CTRL_expressions_mouthPressUR {type=Curve, loc=<0 0 0>, rot=<0 0 0 1>}
 CTRL_expressions_mouthRight {type=Curve, loc=<0 0 0>, rot=<0 0 0 1>}
 CTRL_expressions_mouthSharpCornerPullL {type=Curve, loc=<0 0 0>, rot=<0 0 0 1>}
 CTRL_expressions_mouthSharpCornerPullR {type=Curve, loc=<0 0 0>, rot=<0 0 0 1>}
 CTRL_expressions_mouthStretchL {type=Curve, loc=<0 0 0>, rot=<0 0 0 1>}
 CTRL_expressions_mouthStretchR {type=Curve, loc=<0 0 0>, rot=<0 0 0 1>}
 CTRL_expressions_mouthUp {type=Curve, loc=<0 0 0>, rot=<0 0 0 1>}
 CTRL_expressions_mouthUpperLipBiteL {type=Curve, loc=<0 0 0>, rot=<0 0 0 1>}
 CTRL_expressions_mouthUpperLipBiteR {type=Curve, loc=<0 0 0>, rot=<0 0 0 1>}
 CTRL_expressions_mouthUpperLipRaiseL {type=Curve, loc=<0 0 0>, rot=<0 0 0 1>}
 CTRL_expressions_mouthUpperLipRaiseR {type=Curve, loc=<0 0 0>, rot=<0 0 0 1>}
 CTRL_expressions_mouthUpperLipRollInL {type=Curve, loc=<0 0 0>, rot=<0 0 0 1>}
 CTRL_expressions_mouthUpperLipRollInR {type=Curve, loc=<0 0 0>, rot=<0 0 0 1>}
 CTRL_expressions_mouthUpperLipRollOutL {type=Curve, loc=<0 0 0>, rot=<0 0 0 1>}
 CTRL_expressions_mouthUpperLipRollOutR {type=Curve, loc=<0 0 0>, rot=<0 0 0 1>}
 CTRL_expressions_mouthUpperLipTowardsTeethL {type=Curve, loc=<0 0 0>, rot=<0 0 0 1>}
 CTRL_expressions_mouthUpperLipTowardsTeethR {type=Curve, loc=<0 0 0>, rot=<0 0 0 1>}
 CTRL_expressions_noseNasolabialDeepenL {type=Curve, loc=<0 0 0>, rot=<0 0 0 1>}
 CTRL_expressions_noseNasolabialDeepenR {type=Curve, loc=<0 0 0>, rot=<0 0 0 1>}
 CTRL_expressions_noseNostrilCompressL {type=Curve, loc=<0 0 0>, rot=<0 0 0 1>}
 CTRL_expressions_noseNostrilCompressR {type=Curve, loc=<0 0 0>, rot=<0 0 0 1>}
 CTRL_expressions_noseNostrilDepressL {type=Curve, loc=<0 0 0>, rot=<0 0 0 1>}
 CTRL_expressions_noseNostrilDepressR {type=Curve, loc=<0 0 0>, rot=<0 0 0 1>}
 CTRL_expressions_noseNostrilDilateL {type=Curve, loc=<0 0 0>, rot=<0 0 0 1>}
 CTRL_expressions_noseNostrilDilateR {type=Curve, loc=<0 0 0>, rot=<0 0 0 1>}
 CTRL_expressions_noseRotateL {type=Curve, loc=<0 0 0>, rot=<0 0 0 1>}
 CTRL_expressions_noseRotateR {type=Curve, loc=<0 0 0>, rot=<0 0 0 1>}
 CTRL_expressions_noseSquash {type=Curve, loc=<0 0 0>, rot=<0 0 0 1>}
 CTRL_expressions_noseStretch {type=Curve, loc=<0 0 0>, rot=<0 0 0 1>}
 CTRL_expressions_noseWrinkleL {type=Curve, loc=<0 0 0>, rot=<0 0 0 1>}
 CTRL_expressions_noseWrinkleR {type=Curve, loc=<0 0 0>, rot=<0 0 0 1>}
 CTRL_expressions_noseWrinkleUpperL {type=Curve, loc=<0 0 0>, rot=<0 0 0 1>}
 CTRL_expressions_noseWrinkleUpperR {type=Curve, loc=<0 0 0>, rot=<0 0 0 1>}
 CTRL_expressions_teethBackD {type=Curve, loc=<0 0 0>, rot=<0 0 0 1>}
 CTRL_expressions_teethBackU {type=Curve, loc=<0 0 0>, rot=<0 0 0 1>}
 CTRL_expressions_teethDownD {type=Curve, loc=<0 0 0>, rot=<0 0 0 1>}
 CTRL_expressions_teethDownU {type=Curve, loc=<0 0 0>, rot=<0 0 0 1>}
 CTRL_expressions_teethFwdD {type=Curve, loc=<0 0 0>, rot=<0 0 0 1>}
 CTRL_expressions_teethFwdU {type=Curve, loc=<0 0 0>, rot=<0 0 0 1>}
 CTRL_expressions_teethLeftD {type=Curve, loc=<0 0 0>, rot=<0 0 0 1>}
 CTRL_expressions_teethLeftU {type=Curve, loc=<0 0 0>, rot=<0 0 0 1>}
 CTRL_expressions_teethRightD {type=Curve, loc=<0 0 0>, rot=<0 0 0 1>}
 CTRL_expressions_teethRightU {type=Curve, loc=<0 0 0>, rot=<0 0 0 1>}
 CTRL_expressions_teethRotateBackD {type=Curve, loc=<0 0 0>, rot=<0 0 0 1>}
 CTRL_expressions_teethRotateBackU {type=Curve, loc=<0 0 0>, rot=<0 0 0 1>}
 CTRL_expressions_teethRotateFwdD {type=Curve, loc=<0 0 0>, rot=<0 0 0 1>}
 CTRL_expressions_teethRotateFwdU {type=Curve, loc=<0 0 0>, rot=<0 0 0 1>}
 CTRL_expressions_teethUpD {type=Curve, loc=<0 0 0>, rot=<0 0 0 1>}
 CTRL_expressions_teethUpU {type=Curve, loc=<0 0 0>, rot=<0 0 0 1>}
 CTRL_expressions_tongueBendDown {type=Curve, loc=<0 0 0>, rot=<0 0 0 1>}
 CTRL_expressions_tongueBendUp {type=Curve, loc=<0 0 0>, rot=<0 0 0 1>}
 CTRL_expressions_tongueDown {type=Curve, loc=<0 0 0>, rot=<0 0 0 1>}
 CTRL_expressions_tongueIn {type=Curve, loc=<0 0 0>, rot=<0 0 0 1>}
 CTRL_expressions_tongueLeft {type=Curve, loc=<0 0 0>, rot=<0 0 0 1>}
 CTRL_expressions_tongueNarrow {type=Curve, loc=<0 0 0>, rot=<0 0 0 1>}
 CTRL_expressions_tongueOut {type=Curve, loc=<0 0 0>, rot=<0 0 0 1>}
 CTRL_expressions_tongueRight {type=Curve, loc=<0 0 0>, rot=<0 0 0 1>}
 CTRL_expressions_tongueRollDown {type=Curve, loc=<0 0 0>, rot=<0 0 0 1>}
 CTRL_expressions_tongueRollLeft {type=Curve, loc=<0 0 0>, rot=<0 0 0 1>}
 CTRL_expressions_tongueRollRight {type=Curve, loc=<0 0 0>, rot=<0 0 0 1>}
 CTRL_expressions_tongueRollUp {type=Curve, loc=<0 0 0>, rot=<0 0 0 1>}
 CTRL_expressions_tongueTwistLeft {type=Curve, loc=<0 0 0>, rot=<0 0 0 1>}
 CTRL_expressions_tongueTwistRight {type=Curve, loc=<0 0 0>, rot=<0 0 0 1>}
 CTRL_expressions_tongueUp {type=Curve, loc=<0 0 0>, rot=<0 0 0 1>}
 CTRL_expressions_tongueWide {type=Curve, loc=<0 0 0>, rot=<0 0 0 1>}
 DisableFaceOverride {type=Curve, loc=<0 0 0>, rot=<0 0 0 1>}
 is_3L {type=Curve, loc=<0 0 0>, rot=<0 0 0 1>}
 is_face_bone_anim {type=Curve, loc=<0 0 0>, rot=<0 0 0 1>}
 WM_BrowLateral_L {type=Curve, loc=<0 0 0>, rot=<0 0 0 1>}
 WM_BrowLateral_Offset_L {type=Curve, loc=<0 0 0>, rot=<0 0 0 1>}
 WM_BrowLateral_Offset_R {type=Curve, loc=<0 0 0>, rot=<0 0 0 1>}
 WM_BrowLateral_R {type=Curve, loc=<0 0 0>, rot=<0 0 0 1>}
 WM_BrowRaiseInner_L {type=Curve, loc=<0 0 0>, rot=<0 0 0 1>}
 WM_BrowRaiseInner_Offset_L {type=Curve, loc=<0 0 0>, rot=<0 0 0 1>}
 WM_BrowRaiseInner_Offset_R {type=Curve, loc=<0 0 0>, rot=<0 0 0 1>}
 WM_BrowRaiseInner_R {type=Curve, loc=<0 0 0>, rot=<0 0 0 1>}
 WM_BrowRaiseOutter_L {type=Curve, loc=<0 0 0>, rot=<0 0 0 1>}
 WM_BrowRaiseOutter_Offset_L {type=Curve, loc=<0 0 0>, rot=<0 0 0 1>}
 WM_BrowRaiseOutter_Offset_R {type=Curve, loc=<0 0 0>, rot=<0 0 0 1>}
 WM_BrowRaiseOutter_R {type=Curve, loc=<0 0 0>, rot=<0 0 0 1>}
 WM_CheekRaise_L {type=Curve, loc=<0 0 0>, rot=<0 0 0 1>}
 WM_CheekRaise_Offset_L {type=Curve, loc=<0 0 0>, rot=<0 0 0 1>}
 WM_CheekRaise_Offset_R {type=Curve, loc=<0 0 0>, rot=<0 0 0 1>}
 WM_CheekRaise_R {type=Curve, loc=<0 0 0>, rot=<0 0 0 1>}
 WM_Lips_L {type=Curve, loc=<0 0 0>, rot=<0 0 0 1>}
 WM_Lips_Offset_L {type=Curve, loc=<0 0 0>, rot=<0 0 0 1>}
 WM_Lips_Offset_R {type=Curve, loc=<0 0 0>, rot=<0 0 0 1>}
 WM_Lips_R {type=Curve, loc=<0 0 0>, rot=<0 0 0 1>}
 WM_MouthStretch_L {type=Curve, loc=<0 0 0>, rot=<0 0 0 1>}
 WM_MouthStretch_Offset_L {type=Curve, loc=<0 0 0>, rot=<0 0 0 1>}
 WM_MouthStretch_Offset_R {type=Curve, loc=<0 0 0>, rot=<0 0 0 1>}
 WM_MouthStretch_R {type=Curve, loc=<0 0 0>, rot=<0 0 0 1>}
 WM_NoseWrinkle_L {type=Curve, loc=<0 0 0>, rot=<0 0 0 1>}
 WM_NoseWrinkle_Offset_L {type=Curve, loc=<0 0 0>, rot=<0 0 0 1>}
 WM_NoseWrinkle_Offset_R {type=Curve, loc=<0 0 0>, rot=<0 0 0 1>}
 WM_NoseWrinkle_R {type=Curve, loc=<0 0 0>, rot=<0 0 0 1>}
 WM_Smile_Offset_L {type=Curve, loc=<0 0 0>, rot=<0 0 0 1>}
 WM_Smile_Offset_R {type=Curve, loc=<0 0 0>, rot=<0 0 0 1>}
 WM_Smile_R {type=Curve, loc=<0 0 0>, rot=<0 0 0 1>}
 WM_Smile_L {type=Curve, loc=<0 0 0>, rot=<0 0 0 1>}
]
variables [
 Current Root Bone {type=FName, default=None}
 Legacy Root Bone {type=FName, default=faceAttach}
 MH Root Bone {type=FName, default=FACIAL_C_FacialRoot}
]
