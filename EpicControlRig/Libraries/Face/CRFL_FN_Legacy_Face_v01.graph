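# Control Rig function library for the FN legacy face rig, providing a control board that drives facial pose curves.
# A construction function spawns the controls, channels and nulls, then scales and colors them.
# Forward converts controls to curves through mirrored helper functions, blending center and side lip controls, remapping optional eye-aim rotation to look curves, and handling head follow.
# Backward runs those helpers reversed to recover controls from curves, switching eye aim off.
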ControlRigBlueprint /EpicControlRig/Libraries/Face/CRFL_FN_Legacy_Face_v01
graphs [
 CRFL FN Legacy Backward v01_ContainedGraph {nodes [
   n1 = RigVMFunctionEntryNode()
   RigVMFunction_Sequence = &RigVMFunction_Sequence(ExecuteContext=@n1.ExecuteContext)
   /Script/ControlRig.RigUnit_SetCurveValue(Curve=DisableFaceOverride, Value=1, ExecutePin=@RigVMFunction_Sequence.A) {is=struct RigVMUnitNode}
   > &RigUnit_SetCurveValue(Curve=is_3L, Value=0)
   > SequenceAggregate = RigVMAggregateNode()
   FN_FloatCtrlToCurve_FNC = RigVMFunctionReferenceNode(
    Ctrl_L_Token=L_
    Ctrl_R_Token=R_
    Curve_Side_Token_Last=False
    Pos_Neg_Curves=((Type=Curve,Name="C_glabella_up_pose"),(Type=Curve,Name="C_glabella_down_pose"))
    backward=true
    control=(Type=Control,Name="midbrow")
    mirror=false
    vector_axis=X
    ExecuteContext=@SequenceAggregate.A
   )
   > FN_VectorCtrlToCurve_FNC_7 = RigVMFunctionReferenceNode(
    Backward=true
    Control=(Type=Control,Name="L_Brow")
    Ctrl_L_Token=L_
    Ctrl_R_Token=R_
    Curve_Side_Token_Last=false
    X_Mirror=true
    X_Pos_Neg_Expr=((Type=Curve,Name="L_brow_squeeze_in_pose"))
    Y_Mirror=true
    Y_Pos_Neg_Expr=((Type=Curve,Name="L_brow_up_pose"),(Type=Curve,Name="L_brow_down_pose"))
    Z_Mirror=false
    Z_Pos_Neg_Expr=()
   )
   > FN_VectorCtrlToCurve_FNC_7_1 = RigVMFunctionReferenceNode(
    Backward=true
    Control=(Type=Control,Name="L_Eyelid")
    Ctrl_L_Token=L_
    Ctrl_R_Token=R_
    Curve_Side_Token_Last=false
    X_Mirror=true
    X_Pos_Neg_Expr=((Type=Curve,Name="L_squint_inner_pose"))
    Y_Mirror=true
    Y_Pos_Neg_Expr=((Type=Curve,Name="L_wide_pose"),(Type=Curve,Name="L_blink_pose"))
    Z_Mirror=false
    Z_Pos_Neg_Expr=()
   )
   > CRFL AnimAttrToCurve v01 = RigVMFunctionReferenceNode(Backward=true, Channel=upper_lid_down, Control=(Type=Control,Name="L_Eyelid"), Pos_Neg_Curves=((Type=Curve,Name="L_upperLidDown_pose")))
   > CRFL AnimAttrToCurve v01_1 = RigVMFunctionReferenceNode(Backward=true, Channel=upper_lid_down, Control=(Type=Control,Name="L_Eyelid"), Pos_Neg_Curves=((Type=Curve,Name="R_upperLidDown_pose")))
   > CRFL AnimAttrToCurve v01_6 = RigVMFunctionReferenceNode(Backward=true, Channel=eye_convergence, Control=(Type=Control,Name="Look"), Pos_Neg_Curves=((Type=Curve,Name="eye_convergence_pose")))
   > SetAnimationChannel = /Script/ControlRig.RigUnit_SetBoolAnimationChannel(Channel=use_eye_aim, Control=Look, Value=False, bInitial=False) {is=struct RigVMUnitNode, template="SetAnimationChannel::Execute(in Value,in Control,in Channel,in bInitial)"}
   > SetControlVisibility = /Script/ControlRig.RigUnit_SetControlVisibility(Item=(Type=Control,Name="C_Eye_Aim"), bVisible=false) {is=struct RigVMUnitNode, template="SetControlVisibility::Execute(in Item,in Pattern,in bVisible)"}
   > FN_VectorCtrlToCurve_FNC_7_1_1 = RigVMFunctionReferenceNode(
    Backward=true
    Control=(Type=Control,Name="Look")
    Ctrl_L_Token=L_
    Ctrl_R_Token=R_
    Curve_Side_Token_Last=false
    X_Mirror=true
    X_Pos_Neg_Expr=((Type=Curve,Name="look_right_pose"),((Type=Curve,Name="None"),(Type=Curve,Name="None"),Type=Curve,Name="look_left_pose"))
    Y_Mirror=true
    Y_Pos_Neg_Expr=((Type=Curve,Name="look_up_pose"),(Type=Curve,Name="look_down_pose"))
    Z_Mirror=false
    Z_Pos_Neg_Expr=()
   )
   FN_FloatCtrlToCurve_FNC_1 = RigVMFunctionReferenceNode(
    Ctrl_L_Token=L_
    Ctrl_R_Token=R_
    Curve_Side_Token_Last=False
    Pos_Neg_Curves=((Type=Curve,Name="C_nose_wrinkler_pose"))
    backward=true
    control=(Type=Control,Name="Nose")
    mirror=false
    vector_axis=X
    ExecuteContext=@SequenceAggregate.B
   )
   > FN_VectorCtrlToCurve_FNC_7_1_1_1 = RigVMFunctionReferenceNode(
    Backward=true
    Control=(Type=Control,Name="L_Cheek")
    Ctrl_L_Token=L_
    Ctrl_R_Token=R_
    Curve_Side_Token_Last=false
    X_Mirror=true
    X_Pos_Neg_Expr=((Type=Curve,Name="L_cheek_up_pose"))
    Y_Mirror=true
    Y_Pos_Neg_Expr=((Type=Curve,Name="L_squeeze_pose"))
    Z_Mirror=false
    Z_Pos_Neg_Expr=()
   )
   > FN_FloatCtrlToCurve_FNC_2 = RigVMFunctionReferenceNode(
    Ctrl_L_Token=L_
    Ctrl_R_Token=R_
    Curve_Side_Token_Last=False
    Pos_Neg_Curves=((Type=Curve,Name="mouth_up_pose"),((Type=Curve,Name="None"),(Type=Curve,Name="None"),Type=Curve,Name="mouth_down_pose"))
    backward=true
    control=(Type=Control,Name="Mouth")
    mirror=false
    vector_axis=X
   )
   > CRFL AnimAttrToCurve v01_2 = RigVMFunctionReferenceNode(Backward=true, Channel=phoneme_ch, Control=(Type=Control,Name="Mouth"), Pos_Neg_Curves=((Type=Curve,Name="phoneme_ch_pose")))
   > CRFL AnimAttrToCurve v01_3 = RigVMFunctionReferenceNode(Backward=true, Channel=phoneme_fv, Control=(Type=Control,Name="Mouth"), Pos_Neg_Curves=((Type=Curve,Name="phoneme_fv_pose")))
   > CRFL AnimAttrToCurve v01_4 = RigVMFunctionReferenceNode(Backward=true, Channel=phoneme_mbp, Control=(Type=Control,Name="Mouth"), Pos_Neg_Curves=((Type=Curve,Name="phoneme_mbp_pose")))
   > CRFL AnimAttrToCurve v01_5 = RigVMFunctionReferenceNode(Backward=true, Channel=phoneme_oo, Control=(Type=Control,Name="Mouth"), Pos_Neg_Curves=((Type=Curve,Name="phoneme_oo_pose")))
   FN_FloatCtrlToCurve_FNC_4 = RigVMFunctionReferenceNode(
    Ctrl_L_Token=L_
    Ctrl_R_Token=R_
    Curve_Side_Token_Last=false
    Pos_Neg_Curves=((Type=Curve,Name="L_upper_lip_raiser_pose"),(Type=Curve,Name="L_upper_lip_lower_pose"))
    backward=true
    control=(Type=Control,Name="L_Lip_Upper")
    mirror=true
    vector_axis=X
    ExecuteContext=@SequenceAggregate.C
   )
   > FN_VectorCtrlToCurve_FNC_7_2 = RigVMFunctionReferenceNode(
    Backward=true
    Control=(Type=Control,Name="L_Lipcorner")
    Ctrl_L_Token=L_
    Ctrl_R_Token=R_
    Curve_Side_Token_Last=false
    X_Mirror=true
    X_Pos_Neg_Expr=((Type=Curve,Name="L_lip_corner_narrow_pose"),((Type=Curve,Name="None"),(Type=Curve,Name="None"),Type=Curve,Name="L_lip_corner_wide_pose"))
    Y_Mirror=true
    Y_Pos_Neg_Expr=((Type=Curve,Name="L_smile_pose"),(Type=Curve,Name="L_frown_pose"))
    Z_Mirror=false
    Z_Pos_Neg_Expr=()
   )
   > FN_FloatCtrlToCurve_FNC_4_1 = RigVMFunctionReferenceNode(
    Ctrl_L_Token=L_
    Ctrl_R_Token=R_
    Curve_Side_Token_Last=false
    Pos_Neg_Curves=((Type=Curve,Name="L_lower_lip_down_pose"),(Type=Curve,Name="L_lower_lip_up_pose"))
    backward=true
    control=(Type=Control,Name="L_Lip_Lower")
    mirror=true
    vector_axis=X
   )
   FN_VectorCtrlToCurve_FNC_7_2_1 = RigVMFunctionReferenceNode(
    Backward=true
    Control=(Type=Control,Name="Jaw")
    Ctrl_L_Token=L_
    Ctrl_R_Token=R_
    Curve_Side_Token_Last=false
    X_Mirror=false
    X_Pos_Neg_Expr=((Type=Curve,Name="jaw_right_pose"),((Type=Curve,Name="None"),(Type=Curve,Name="None"),Type=Curve,Name="jaw_left_pose"))
    Y_Mirror=false
    Y_Pos_Neg_Expr=((Type=Curve,Name="jaw_up_pose"),(Type=Curve,Name="jaw_open_pose"))
    Z_Mirror=false
    Z_Pos_Neg_Expr=()
    ExecuteContext=@SequenceAggregate.D
   )
   > FN_FloatCtrlToCurve_FNC_3 = RigVMFunctionReferenceNode(
    Ctrl_L_Token=L_
    Ctrl_R_Token=R_
    Curve_Side_Token_Last=False
    Pos_Neg_Curves=((Type=Curve,Name="jaw_back_pose"),((Type=Curve,Name="None"),(Type=Curve,Name="None"),Type=Curve,Name="jaw_forward_pose"))
    backward=true
    control=(Type=Control,Name="Jaw_FwdBk")
    mirror=false
    vector_axis=X
   )
   > FN_VectorCtrlToCurve_FNC_7_2_1_1 = RigVMFunctionReferenceNode(
    Backward=true
    Control=(Type=Control,Name="Tongue")
    Ctrl_L_Token=L_
    Ctrl_R_Token=R_
    Curve_Side_Token_Last=false
    X_Mirror=false
    X_Pos_Neg_Expr=((Type=Curve,Name="tongue_forward_pose"),((Type=Curve,Name="None"),(Type=Curve,Name="None"),Type=Curve,Name="tongue_back_pose"))
    Y_Mirror=false
    Y_Pos_Neg_Expr=((Type=Curve,Name="tongue_up_pose"),(Type=Curve,Name="tongue_down_pose"))
    Z_Mirror=false
    Z_Pos_Neg_Expr=()
   )
   RigVMFunctionReturnNode(ExecuteContext=@RigVMFunction_Sequence.B)
  ]}
 CRFL FN Legacy Forward v01_ContainedGraph {nodes [
   n1 = RigVMFunctionEntryNode()
   RigVMFunction_Sequence = &RigVMFunction_Sequence(ExecuteContext=@n1.ExecuteContext)
   &RigUnit_SetCurveValue(Curve=DisableFaceOverride, Value=1, ExecutePin=@RigVMFunction_Sequence.A)
   > &RigUnit_SetCurveValue(Curve=is_3L, Value=0)
   > SequenceAggregate = RigVMAggregateNode()
   FN_FloatCtrlToCurve_FNC_1 = RigVMFunctionReferenceNode(
    Ctrl_L_Token=L_
    Ctrl_R_Token=R_
    Curve_Side_Token_Last=False
    Pos_Neg_Curves=((Type=Curve,Name="C_nose_wrinkler_pose"))
    backward=False
    control=(Type=Control,Name="Nose")
    mirror=false
    vector_axis=X
    ExecuteContext=@SequenceAggregate.C
   )
   > FN_VectorCtrlToCurve_FNC_7_1_1_1 = RigVMFunctionReferenceNode(
    Backward=false
    Control=(Type=Control,Name="L_Cheek")
    Ctrl_L_Token=L_
    Ctrl_R_Token=R_
    Curve_Side_Token_Last=false
    X_Mirror=true
    X_Pos_Neg_Expr=((Type=Curve,Name="L_cheek_up_pose"))
    Y_Mirror=true
    Y_Pos_Neg_Expr=((Type=Curve,Name="L_squeeze_pose"))
    Z_Mirror=false
    Z_Pos_Neg_Expr=()
   )
   > FN_FloatCtrlToCurve_FNC_2 = RigVMFunctionReferenceNode(
    Ctrl_L_Token=L_
    Ctrl_R_Token=R_
    Curve_Side_Token_Last=False
    Pos_Neg_Curves=((Type=Curve,Name="mouth_up_pose"),((Type=Curve,Name="None"),(Type=Curve,Name="None"),Type=Curve,Name="mouth_down_pose"))
    backward=False
    control=(Type=Control,Name="Mouth")
    mirror=false
    vector_axis=X
   )
   > CRFL AnimAttrToCurve v01_2 = RigVMFunctionReferenceNode(Backward=False, Channel=phoneme_ch, Control=(Type=Control,Name="Mouth"), Pos_Neg_Curves=((Type=Curve,Name="phoneme_ch_pose")))
   > CRFL AnimAttrToCurve v01_3 = RigVMFunctionReferenceNode(Backward=False, Channel=phoneme_fv, Control=(Type=Control,Name="Mouth"), Pos_Neg_Curves=((Type=Curve,Name="phoneme_fv_pose")))
   > CRFL AnimAttrToCurve v01_4 = RigVMFunctionReferenceNode(Backward=False, Channel=phoneme_mbp, Control=(Type=Control,Name="Mouth"), Pos_Neg_Curves=((Type=Curve,Name="phoneme_mbp_pose")))
   > CRFL AnimAttrToCurve v01_5 = RigVMFunctionReferenceNode(Backward=False, Channel=phoneme_oo, Control=(Type=Control,Name="Mouth"), Pos_Neg_Curves=((Type=Curve,Name="phoneme_oo_pose")))
   Add = &RigVMFunction_MathDoubleAdd(
    A=/Script/ControlRig.RigUnit_GetControlVector2D(Control=Lip_Upper) {is=struct RigVMUnitNode, template="GetControlValue::Execute(in Control,out Vector,out Minimum,out Maximum)"}.Vector.X
    B=&RigUnit_GetControlVector2D(Control=L_Lip_Upper) {template="GetControlValue::Execute(in Control,out Vector,out Minimum,out Maximum)"}.Vector.X
   )
   &RigUnit_SetCurveValue(
    Curve=L_upper_lip_raiser_pose
    Value=&RigVMFunction_MathDoubleRemap(SourceMaximum=1, SourceMinimum=0, TargetMaximum=1, TargetMinimum=0, Value=@Add.Result).Result
    ExecutePin=@SequenceAggregate.D
   )
   > &RigUnit_SetCurveValue(
    Curve=L_upper_lip_lower_pose
    Value=&RigVMFunction_MathDoubleRemap(SourceMaximum=-1, SourceMinimum=0, TargetMaximum=1, TargetMinimum=0, Value=@Add.Result).Result
   )
   n2 = &RigVMFunction_MathDoubleAdd(
    B=&RigUnit_GetControlVector2D(Control=R_Lip_Upper) {template="GetControlValue::Execute(in Control,out Vector,out Minimum,out Maximum)"}.Vector.X
    A=&RigUnit_GetControlVector2D(Control=Lip_Upper) {template="GetControlValue::Execute(in Control,out Vector,out Minimum,out Maximum)"}.Vector.X
   )
   > &RigUnit_SetCurveValue(
    Curve=R_upper_lip_raiser_pose
    Value=&RigVMFunction_MathDoubleRemap(SourceMaximum=1, SourceMinimum=0, TargetMaximum=1, TargetMinimum=0, Value=@n2.Result).Result
   )
   > &RigUnit_SetCurveValue(
    Curve=R_upper_lip_lower_pose
    Value=&RigVMFunction_MathDoubleRemap(SourceMaximum=-1, SourceMinimum=0, TargetMaximum=1, TargetMinimum=0, Value=@n2.Result).Result
   )
   > FN_VectorCtrlToCurve_FNC_7_3 = RigVMFunctionReferenceNode(
    Backward=false
    Control=(Type=Control,Name="L_Lipcorner")
    Ctrl_L_Token=L_
    Ctrl_R_Token=R_
    Curve_Side_Token_Last=false
    X_Mirror=true
    X_Pos_Neg_Expr=((Type=Curve,Name="L_lip_corner_narrow_pose"),((Type=Curve,Name="None"),(Type=Curve,Name="None"),Type=Curve,Name="L_lip_corner_wide_pose"))
    Y_Mirror=true
    Y_Pos_Neg_Expr=((Type=Curve,Name="L_smile_pose"))
    Z_Mirror=false
    Z_Pos_Neg_Expr=()
   )
   n3 = &RigUnit_GetControlVector2D(Control=L_Lipcorner) {template="GetControlValue::Execute(in Control,out Vector,out Minimum,out Maximum)"}
   n4 = /Script/ControlRig.RigUnit_GetControlFloat(Control=L_Frown) {is=struct RigVMUnitNode, template="GetControlValue::Execute(in Control,out FloatValue,out Minimum,out Maximum)"}
   Clamp = &RigVMFunction_MathDoubleClamp(Value=&RigVMFunction_MathDoubleAdd(
     A=&RigVMFunction_MathDoubleRemap(SourceMaximum=-1, SourceMinimum=0, TargetMaximum=1, TargetMinimum=0, Value=@n3.Vector.Y).Result
     B=&RigVMFunction_MathDoubleRemap(SourceMaximum=-1, SourceMinimum=0, TargetMaximum=1, TargetMinimum=0, Value=@n4.FloatValue).Result
    ).Result)
   > &RigUnit_SetCurveValue(Curve=L_frown_pose, Value=@Clamp.Result)
   n5 = &RigVMFunction_MathDoubleRemap(
    SourceMaximum=-1
    SourceMinimum=0
    TargetMaximum=1
    TargetMinimum=0
    Value=&RigUnit_GetControlVector2D(Control=R_Lipcorner) {template="GetControlValue::Execute(in Control,out Vector,out Minimum,out Maximum)"}.Vector.Y
   )
   n6 = &RigVMFunction_MathDoubleRemap(
    SourceMaximum=-1
    SourceMinimum=0
    TargetMaximum=1
    TargetMinimum=0
    Value=&RigUnit_GetControlFloat(Control=R_Frown) {template="GetControlValue::Execute(in Control,out FloatValue,out Minimum,out Maximum)"}.FloatValue
   )
   > &RigUnit_SetCurveValue(
    Curve=R_frown_pose
    Value=&RigVMFunction_MathDoubleClamp(Value=&RigVMFunction_MathDoubleAdd(A=@n5.Result, B=@n6.Result).Result).Result
   )
   n7 = &RigVMFunction_MathDoubleAdd(
    B=&RigUnit_GetControlVector2D(Control=L_Lip_Lower) {template="GetControlValue::Execute(in Control,out Vector,out Minimum,out Maximum)"}.Vector.X
    A=&RigUnit_GetControlVector2D(Control=Lip_Lower) {template="GetControlValue::Execute(in Control,out Vector,out Minimum,out Maximum)"}.Vector.X
   )
   > &RigUnit_SetCurveValue(
    Curve=L_lower_lip_down_pose
    Value=&RigVMFunction_MathDoubleRemap(SourceMaximum=1, SourceMinimum=0, TargetMaximum=1, TargetMinimum=0, Value=@n7.Result).Result
   )
   > &RigUnit_SetCurveValue(
    Curve=L_lower_lip_up_pose
    Value=&RigVMFunction_MathDoubleRemap(SourceMaximum=-1, SourceMinimum=0, TargetMaximum=1, TargetMinimum=0, Value=@n7.Result).Result
   )
   n8 = &RigVMFunction_MathDoubleAdd(
    B=&RigUnit_GetControlVector2D(Control=R_Lip_Lower) {template="GetControlValue::Execute(in Control,out Vector,out Minimum,out Maximum)"}.Vector.X
    A=&RigUnit_GetControlVector2D(Control=Lip_Lower) {template="GetControlValue::Execute(in Control,out Vector,out Minimum,out Maximum)"}.Vector.X
   )
   > &RigUnit_SetCurveValue(
    Curve=R_lower_lip_down_pose
    Value=&RigVMFunction_MathDoubleRemap(SourceMaximum=1, SourceMinimum=0, TargetMaximum=1, TargetMinimum=0, Value=@n8.Result).Result
   )
   > &RigUnit_SetCurveValue(
    Curve=R_lower_lip_up_pose
    Value=&RigVMFunction_MathDoubleRemap(SourceMaximum=-1, SourceMinimum=0, TargetMaximum=1, TargetMinimum=0, Value=@n8.Result).Result
   )
   n9 = /Script/ControlRig.RigUnit_ProjectTransformToNewParent(
    Child=(Type=Null,Name="Face_Ctrl_Master_Follow")
    NewParent=(Type=Bone,Name="head")
    OldParent=(Type=Bone,Name="head")
    bChildInitial=True
    bNewParentInitial=False
    bOldParentInitial=True
   ) {is=struct RigVMUnitNode}
   &RigUnit_SetTransform(
    Item=(Type=Null,Name="Face_Ctrl_Master_Follow")
    Weight=1
    bInitial=False
    Value=/Script/RigVM.RigVMFunction_MathTransformLerp(
     T=/Script/ControlRig.RigUnit_GetFloatAnimationChannel(Channel=gui_follow_head, Control=Face_Ctrl_Master, bInitial=False) {is=struct RigVMUnitNode, template="GetAnimationChannel::Execute(out Value,in Control,in Channel,in bInitial)"}.Value
     B=@n9.Transform
     A=&RigUnit_GetTransform(Item=(Type=Null,Name="Face_Ctrl_Master_Follow"), bInitial=true).Transform
    ) {is=struct RigVMUnitNode, template="Interpolate::Execute(in A,in B,in T,out Result)"}.Result
    ExecutePin=@SequenceAggregate.A
   )
   n10 = /Script/ControlRig.RigUnit_GetBoolAnimationChannel(Channel=eyeAim_follow_head, Control=C_Eye_Aim, bInitial=False) {is=struct RigVMUnitNode, template="GetAnimationChannel::Execute(out Value,in Control,in Channel,in bInitial)"}
   n11 = &RigUnit_ProjectTransformToNewParent(
    Child=(Type=Null,Name="C_Eye_Aim_GRP")
    NewParent=(Type=Bone,Name="head")
    OldParent=(Type=Bone,Name="head")
    bChildInitial=True
    bNewParentInitial=False
    bOldParentInitial=True
   )
   > &RigUnit_SetTransform(
    Item=(Type=Null,Name="C_Eye_Aim_GRP")
    Weight=1
    bInitial=False
    Value=&RigVMFunction_MathTransformLerp(
     T=&RigVMFunction_MathBoolToFloat(Value=@n10.Value).Result
     B=@n11.Transform
     A=&RigUnit_GetTransform(Item=(Type=Null,Name="C_Eye_Aim_GRP"), bInitial=true).Transform
    ) {template="Interpolate::Execute(in A,in B,in T,out Result)"}.Result
   )
   FN_VectorCtrlToCurve_FNC_7_2_1 = RigVMFunctionReferenceNode(
    Backward=false
    Control=(Type=Control,Name="Jaw")
    Ctrl_L_Token=L_
    Ctrl_R_Token=R_
    Curve_Side_Token_Last=false
    X_Mirror=false
    X_Pos_Neg_Expr=((Type=Curve,Name="jaw_right_pose"),((Type=Curve,Name="None"),(Type=Curve,Name="None"),Type=Curve,Name="jaw_left_pose"))
    Y_Mirror=false
    Y_Pos_Neg_Expr=((Type=Curve,Name="jaw_up_pose"),(Type=Curve,Name="jaw_open_pose"))
    Z_Mirror=false
    Z_Pos_Neg_Expr=()
    ExecuteContext=@SequenceAggregate.E
   )
   > FN_FloatCtrlToCurve_FNC_3 = RigVMFunctionReferenceNode(
    Ctrl_L_Token=L_
    Ctrl_R_Token=R_
    Curve_Side_Token_Last=False
    Pos_Neg_Curves=((Type=Curve,Name="jaw_back_pose"),((Type=Curve,Name="None"),(Type=Curve,Name="None"),Type=Curve,Name="jaw_forward_pose"))
    backward=False
    control=(Type=Control,Name="Jaw_FwdBk")
    mirror=false
    vector_axis=X
   )
   > FN_VectorCtrlToCurve_FNC_7_2_1_1 = RigVMFunctionReferenceNode(
    Backward=false
    Control=(Type=Control,Name="Tongue")
    Ctrl_L_Token=L_
    Ctrl_R_Token=R_
    Curve_Side_Token_Last=false
    X_Mirror=false
    X_Pos_Neg_Expr=((Type=Curve,Name="tongue_forward_pose"),((Type=Curve,Name="None"),(Type=Curve,Name="None"),Type=Curve,Name="tongue_back_pose"))
    Y_Mirror=false
    Y_Pos_Neg_Expr=((Type=Curve,Name="tongue_up_pose"),(Type=Curve,Name="tongue_down_pose"))
    Z_Mirror=false
    Z_Pos_Neg_Expr=()
   )
   GetAnimationChannel = &RigUnit_GetBoolAnimationChannel(Channel=Flipbook, Control=Face_Ctrl_Master, bInitial=False) {template="GetAnimationChannel::Execute(out Value,in Control,in Channel,in bInitial)"}
   RigVMFunction_MathBoolNot = &RigVMFunction_MathBoolNot(Value=@GetAnimationChannel.Value)
   &RigUnit_SetControlVisibility(Item=(Type=Control,Name="Look"), bVisible=@RigVMFunction_MathBoolNot.Result, ExecutePin=@SequenceAggregate.F) {template="SetControlVisibility::Execute(in Item,in Pattern,in bVisible)"}
   > &RigUnit_SetControlVisibility(Item=(Type=Control,Name="Nose"), bVisible=@RigVMFunction_MathBoolNot.Result) {template="SetControlVisibility::Execute(in Item,in Pattern,in bVisible)"}
   > &RigUnit_SetControlVisibility(Item=(Type=Control,Name="Jaw_FwdBk"), bVisible=@RigVMFunction_MathBoolNot.Result) {template="SetControlVisibility::Execute(in Item,in Pattern,in bVisible)"}
   > &RigUnit_SetControlVisibility(Item=(Type=Control,Name="Tongue"), bVisible=@RigVMFunction_MathBoolNot.Result) {template="SetControlVisibility::Execute(in Item,in Pattern,in bVisible)"}
   FN_FloatCtrlToCurve_FNC = RigVMFunctionReferenceNode(
    Ctrl_L_Token=L_
    Ctrl_R_Token=R_
    Curve_Side_Token_Last=False
    Pos_Neg_Curves=((Type=Curve,Name="C_glabella_up_pose"),(Type=Curve,Name="C_glabella_down_pose"))
    backward=False
    control=(Type=Control,Name="midbrow")
    mirror=false
    vector_axis=X
    ExecuteContext=@SequenceAggregate.B
   )
   > FN_VectorCtrlToCurve_FNC_7 = RigVMFunctionReferenceNode(
    Backward=false
    Control=(Type=Control,Name="L_Brow")
    Ctrl_L_Token=L_
    Ctrl_R_Token=R_
    Curve_Side_Token_Last=false
    X_Mirror=true
    X_Pos_Neg_Expr=((Type=Curve,Name="L_brow_squeeze_in_pose"))
    Y_Mirror=true
    Y_Pos_Neg_Expr=((Type=Curve,Name="L_brow_up_pose"),(Type=Curve,Name="L_brow_down_pose"))
    Z_Mirror=false
    Z_Pos_Neg_Expr=()
   )
   > FN_VectorCtrlToCurve_FNC_7_1 = RigVMFunctionReferenceNode(
    Backward=false
    Control=(Type=Control,Name="L_Eyelid")
    Ctrl_L_Token=L_
    Ctrl_R_Token=R_
    Curve_Side_Token_Last=false
    X_Mirror=true
    X_Pos_Neg_Expr=((Type=Curve,Name="L_squint_inner_pose"))
    Y_Mirror=true
    Y_Pos_Neg_Expr=((Type=Curve,Name="L_wide_pose"),(Type=Curve,Name="L_blink_pose"))
    Z_Mirror=false
    Z_Pos_Neg_Expr=()
   )
   > CRFL AnimAttrToCurve v01 = RigVMFunctionReferenceNode(Backward=False, Channel=upper_lid_down, Control=(Type=Control,Name="L_Eyelid"), Pos_Neg_Curves=((Type=Curve,Name="L_upperLidDown_pose")))
   > CRFL AnimAttrToCurve v01_1 = RigVMFunctionReferenceNode(Backward=False, Channel=upper_lid_down, Control=(Type=Control,Name="L_Eyelid"), Pos_Neg_Curves=((Type=Curve,Name="R_upperLidDown_pose")))
   n12 = &RigUnit_ProjectTransformToNewParent(
    Child=(Type=Null,Name="Eye_Aim_NUL_GRP")
    NewParent=(Type=Bone,Name="head")
    OldParent=(Type=Bone,Name="head")
    bChildInitial=True
    bNewParentInitial=False
    bOldParentInitial=True
   )
   > &RigUnit_SetTransform(Item=(Type=Null,Name="Eye_Aim_NUL_GRP"), Weight=1, bInitial=False, Value=@n12.Transform)
   > /Script/ControlRig.RigUnit_AimItem(
    DebugSettings=(bEnabled=False,Scale=10,WorldOffset=(Rotation=(X=0,Y=0,Z=0,W=1),Translation=(X=0,Y=0,Z=0),Scale3D=(X=1,Y=1,Z=1)))
    Item=(Type=Null,Name="L_Eye_Aim_NUL")
    Primary=(Weight=1,Axis=(X=0,Y=-1,Z=0),Target=(X=1,Y=0,Z=0),Kind=Location,Space=(Type=Bone,Name="None"))
    Secondary=(Weight=1,Axis=(X=1,Y=0,Z=0),Target=(X=0,Y=0,Z=1),Kind=Direction,Space=(Type=Bone,Name="head"))
    Weight=1
    Primary.Target=&RigUnit_GetTransform(Item=(Type=Control,Name="L_Eye_Aim"), bInitial=False).Transform.Translation
   ) {is=struct RigVMUnitNode}
   > &RigUnit_AimItem(
    DebugSettings=(bEnabled=False,Scale=10,WorldOffset=(Rotation=(X=0,Y=0,Z=0,W=1),Translation=(X=0,Y=0,Z=0),Scale3D=(X=1,Y=1,Z=1)))
    Item=(Type=Null,Name="R_Eye_Aim_NUL")
    Primary=(Weight=1,Axis=(X=0,Y=-1,Z=0),Target=(X=1,Y=0,Z=0),Kind=Location,Space=(Type=Bone,Name="None"))
    Secondary=(Weight=1,Axis=(X=1,Y=0,Z=0),Target=(X=0,Y=0,Z=1),Kind=Direction,Space=(Type=Bone,Name="head"))
    Weight=1
    Primary.Target=&RigUnit_GetTransform(Item=(Type=Control,Name="R_Eye_Aim"), bInitial=False).Transform.Translation
   )
   > CRFL AnimAttrToCurve v01_6 = RigVMFunctionReferenceNode(Backward=False, Channel=eye_convergence, Control=(Type=Control,Name="Look"), Pos_Neg_Curves=((Type=Curve,Name="eye_convergence_pose")))
   n13 = &RigUnit_GetBoolAnimationChannel(Channel=use_eye_aim, Control=Look, bInitial=False) {template="GetAnimationChannel::Execute(out Value,in Control,in Channel,in bInitial)"}
   > SetControlVisibility = &RigUnit_SetControlVisibility(Item=(Type=Control,Name="C_Eye_Aim"), bVisible=@n13.Value) {template="SetControlVisibility::Execute(in Item,in Pattern,in bVisible)"}
   n14 = &RigVMFunction_ControlFlowBranch(Condition=@n13.Value, ExecuteContext=@SetControlVisibility.ExecutePin)
   FN_VectorCtrlToCurve_FNC_7_1_1 = RigVMFunctionReferenceNode(
    Backward=false
    Control=(Type=Control,Name="Look")
    Ctrl_L_Token=L_
    Ctrl_R_Token=R_
    Curve_Side_Token_Last=false
    X_Mirror=true
    X_Pos_Neg_Expr=((Type=Curve,Name="look_right_pose"),((Type=Curve,Name="None"),(Type=Curve,Name="None"),Type=Curve,Name="look_left_pose"))
    Y_Mirror=true
    Y_Pos_Neg_Expr=((Type=Curve,Name="look_up_pose"),(Type=Curve,Name="look_down_pose"))
    Z_Mirror=false
    Z_Pos_Neg_Expr=()
    ExecuteContext=@n14.False
   )
   Cast = /Script/RigVM.RigVMFunction_MathQuaternionToRotator(Value=&RigUnit_GetTransform(Item=(Type=Null,Name="L_Eye_Aim_NUL"), Space=LocalSpace, bInitial=False).Transform.Rotation) {is=struct RigVMUnitNode, template="Cast::Execute(in Value,out Result)"}
   Remap = &RigVMFunction_MathDoubleRemap(SourceMaximum=30, SourceMinimum=0, TargetMaximum=1, TargetMinimum=0, Value=@Cast.Result.Roll)
   &RigUnit_SetCurveValue(Curve=look_up_pose, ExecutePin=@n14.True, Value=@Remap.Result)
   > &RigUnit_SetCurveValue(
    Curve=look_down_pose
    Value=&RigVMFunction_MathDoubleRemap(SourceMaximum=0, SourceMinimum=-45, TargetMaximum=0, TargetMinimum=1, Value=@Cast.Result.Roll).Result
   )
   > &RigUnit_SetCurveValue(
    Curve=look_left_pose
    Value=&RigVMFunction_MathDoubleRemap(SourceMaximum=0, SourceMinimum=-35, TargetMaximum=0, TargetMinimum=1, Value=@Cast.Result.Yaw).Result
   )
   > &RigUnit_SetCurveValue(
    Curve=look_right_pose
    Value=&RigVMFunction_MathDoubleRemap(SourceMaximum=35, SourceMinimum=0, TargetMaximum=1, TargetMinimum=0, Value=@Cast.Result.Yaw).Result
   )
   RigVMFunctionReturnNode(ExecuteContext=@RigVMFunction_Sequence.B)
   FN_VectorCtrlToCurve_FNC_7_2 = RigVMFunctionReferenceNode(
    Backward=false
    Control=(Type=Control,Name="L_Lipcorner")
    Ctrl_L_Token=L_
    Ctrl_R_Token=R_
    Curve_Side_Token_Last=false
    ExecuteContext=()
    X_Mirror=true
    X_Pos_Neg_Expr=((Type=Curve,Name="L_lip_corner_narrow_pose"),((Type=Curve,Name="None"),(Type=Curve,Name="None"),Type=Curve,Name="L_lip_corner_wide_pose"))
    Y_Mirror=true
    Y_Pos_Neg_Expr=((Type=Curve,Name="L_smile_pose"),(Type=Curve,Name="L_frown_pose"))
    Z_Mirror=false
    Z_Pos_Neg_Expr=()
   )
  ]}
 New Function_ContainedGraph {nodes [
   n1 = RigVMFunctionEntryNode()
   n2 = &RigVMFunction_Sequence(ExecuteContext=@n1.ExecuteContext)
   RigUnit_HierarchyAddNull = &RigUnit_HierarchyAddNull(
    Name=FACE_CONTROL_GRP
    Parent=(Type=Bone,Name="None")
    Space=LocalSpace
    Transform=(Rotation=(X=0,Y=0,Z=0,W=1),Translation=(X=0,Y=0,Z=0),Scale3D=(X=1,Y=1,Z=1))
    ExecutePin=@n2.A
   )
   > RigVMVariableNode(Value=@RigUnit_HierarchyAddNull.Item) {variable=Local Top Grp}
   > n3 = &RigUnit_HierarchyAddNull(
    Name=Face_Ctrl_Master_Follow
    Space=LocalSpace
    Transform=(Rotation=(X=-0.0029,Y=0.7071,Z=0.0029,W=-0.7071),Translation=(X=0,Y=-1.6873,Z=152.0342),Scale3D=(X=1,Y=1,Z=1))
    Parent=@RigUnit_HierarchyAddNull.Item
   )
   > n4 = &RigUnit_HierarchyAddNull(
    Name=Face_Ctrl_Master_GRP
    Space=LocalSpace
    Transform=(Rotation=(X=0,Y=0.7071,Z=0,W=0.7071),Translation=(X=-7,Y=18,Z=0),Scale3D=(X=1,Y=1,Z=1))
    Parent=@n3.Item
   )
   > RigUnit_HierarchyAddControlTransform = /Script/ControlRig.RigUnit_HierarchyAddControlTransform(
    InitialValue=(Rotation=(X=0,Y=0,Z=0,W=1),Translation=(X=0,Y=0,Z=0),Scale3D=(X=1,Y=1,Z=1))
    Name=Face_Ctrl_Master
    OffsetSpace=LocalSpace
    OffsetTransform=(Rotation=(X=0,Y=0,Z=0,W=1),Translation=(X=0,Y=0,Z=0),Scale3D=(X=1,Y=1,Z=1))
    Settings="(InitialSpace=LocalSpace,bUsePreferredRotationOrder=False,PreferredRotationOrder=YZX,Limits=(LimitTranslationX=(bMinimum=False,bMaximum=False),LimitTranslationY=(bMinimum=False,bMaximum=False),LimitTranslationZ=(bMinimum=False,bMaximum=False),LimitPitch=(b ...(truncated)"
    Parent=@n4.Item
   ) {
    is=struct RigVMUnitNode
    template="SpawnControl::Execute(in InitialValue,in Settings,in OffsetTransform,in OffsetSpace,in Parent,in Name,out Item)"
   }
   > n5 = &RigUnit_HierarchyAddControlTransform(
    InitialValue=(Rotation=(X=0,Y=0,Z=0,W=1),Translation=(X=0,Y=0,Z=0),Scale3D=(X=1,Y=1,Z=1))
    Name=Face_Gizmo
    OffsetSpace=LocalSpace
    OffsetTransform=(Rotation=(X=0,Y=0,Z=0,W=1),Translation=(X=0,Y=0,Z=0),Scale3D=(X=0.75,Y=0.75,Z=0.75))
    Settings="(InitialSpace=LocalSpace,bUsePreferredRotationOrder=False,PreferredRotationOrder=YZX,Limits=(LimitTranslationX=(bMinimum=False,bMaximum=False),LimitTranslationY=(bMinimum=False,bMaximum=False),LimitTranslationZ=(bMinimum=False,bMaximum=False),LimitPitch=(b ...(truncated)"
    Parent=@RigUnit_HierarchyAddControlTransform.Item
   ) {template="SpawnControl::Execute(in InitialValue,in Settings,in OffsetTransform,in OffsetSpace,in Parent,in Name,out Item)"}
   > /Script/ControlRig.RigUnit_HierarchyAddAnimationChannelFloat(
    InitialValue=1.0
    LimitsEnabled=(Enabled=(bMinimum=True,bMaximum=True))
    MaximumValue=100.0
    MinimumValue=0.0
    Name=gui_follow_head
    Parent=@RigUnit_HierarchyAddControlTransform.Item
   ) {
    is=struct RigVMUnitNode
    template="SpawnAnimationChannel::Execute(in InitialValue,in MinimumValue,in MaximumValue,in LimitsEnabled,in Parent,in Name,out Item)"
   }
   > /Script/ControlRig.RigUnit_HierarchyAddAnimationChannelBool(
    InitialValue=False
    LimitsEnabled=()
    MaximumValue=True
    MinimumValue=False
    Name=Flipbook
    Parent=@RigUnit_HierarchyAddControlTransform.Item
   ) {
    is=struct RigVMUnitNode
    template="SpawnAnimationChannel::Execute(in InitialValue,in MinimumValue,in MaximumValue,in LimitsEnabled,in Parent,in Name,out Item)"
   }
   > RigVMVariableNode(Value=@n5.Item) {variable=Local Gizmo Ctrl}
   > RigVMFunction_Sequence = RigVMAggregateNode()
   n6 = RigVMVariableNode() {variable=Local Top Grp}
   n7 = &RigUnit_HierarchyAddNull(
    Name=Eye_Aim_NUL_GRP
    Space=LocalSpace
    Transform=(Rotation=(X=0,Y=0,Z=0,W=1),Translation=(X=0,Y=0,Z=0),Scale3D=(X=1,Y=1,Z=1))
    Parent=@n6.Value
    ExecutePin=@RigVMFunction_Sequence.C
   )
   > n8 = &RigUnit_HierarchyAddNull(
    Name=L_Eye_Aim_NUL_GRP
    Space=LocalSpace
    Transform=(Rotation=(X=-0.0001,Y=0.0074,Z=0.9998,W=0.0187),Translation=(X=3.1246,Y=6.1925,Z=158.5556),Scale3D=(X=0.89,Y=0.89,Z=0.89))
    Parent=@n7.Item
   )
   > &RigUnit_HierarchyAddNull(
    Name=L_Eye_Aim_NUL
    Space=LocalSpace
    Transform=(Rotation=(X=0,Y=0,Z=0,W=1),Translation=(X=0,Y=0,Z=0),Scale3D=(X=1,Y=1,Z=1))
    Parent=@n8.Item
   )
   > n9 = &RigUnit_HierarchyAddNull(
    Name=R_Eye_Aim_NUL_GRP
    Space=LocalSpace
    Transform=(Rotation=(X=0.0001,Y=0.0074,Z=0.9998,W=-0.0187),Translation=(X=-3.1878,Y=6.1966,Z=158.5556),Scale3D=(X=0.89,Y=0.89,Z=0.89))
    Parent=@n7.Item
   )
   > &RigUnit_HierarchyAddNull(
    Name=R_Eye_Aim_NUL
    Space=LocalSpace
    Transform=(Rotation=(X=0,Y=0,Z=0,W=1),Translation=(X=0,Y=0,Z=0),Scale3D=(X=1,Y=1,Z=1))
    Parent=@n9.Item
   )
   > n10 = &RigUnit_HierarchyAddNull(
    Name=C_Eye_Aim_GRP
    Space=LocalSpace
    Transform=(Rotation=(X=0,Y=0,Z=0,W=1),Translation=(X=-0.0316,Y=29.1945,Z=158.5556),Scale3D=(X=1,Y=1,Z=1))
    Parent=@n6.Value
   )
   > n11 = &RigUnit_HierarchyAddControlTransform(
    InitialValue=(Rotation=(X=0,Y=0,Z=0,W=1),Translation=(X=0,Y=0,Z=0),Scale3D=(X=1,Y=1,Z=1))
    Name=C_Eye_Aim
    OffsetSpace=LocalSpace
    OffsetTransform=(Rotation=(X=0,Y=0,Z=0,W=1),Translation=(X=0,Y=0,Z=0),Scale3D=(X=1,Y=1,Z=1))
    Settings="(InitialSpace=LocalSpace,bUsePreferredRotationOrder=False,PreferredRotationOrder=YZX,Limits=(LimitTranslationX=(bMinimum=False,bMaximum=False),LimitTranslationY=(bMinimum=False,bMaximum=False),LimitTranslationZ=(bMinimum=False,bMaximum=False),LimitPitch=(b ...(truncated)"
    Parent=@n10.Item
   ) {template="SpawnControl::Execute(in InitialValue,in Settings,in OffsetTransform,in OffsetSpace,in Parent,in Name,out Item)"}
   > &RigUnit_HierarchyAddAnimationChannelBool(InitialValue=False, LimitsEnabled=(), MaximumValue=True, MinimumValue=False, Name=eyeAim_follow_head, Parent=@n11.Item) {template="SpawnAnimationChannel::Execute(in InitialValue,in MinimumValue,in MaximumValue,in LimitsEnabled,in Parent,in Name,out Item)"}
   > n12 = &RigUnit_HierarchyAddNull(
    Name=L_Eye_Aim_GRP
    Space=LocalSpace
    Transform=(Rotation=(X=-0.0001,Y=0.0074,Z=0.9998,W=0.0187),Translation=(X=3.9875,Y=-0.77,Z=-0.329),Scale3D=(X=0.89,Y=0.89,Z=0.89))
    Parent=@n11.Item
   )
   > RigUnit_HierarchyAddControlVector = /Script/ControlRig.RigUnit_HierarchyAddControlVector(
    InitialValue=(X=0,Y=0,Z=0)
    Name=L_Eye_Aim
    OffsetSpace=LocalSpace
    OffsetTransform=(Rotation=(X=0,Y=0,Z=0,W=1),Translation=(X=0,Y=0,Z=0),Scale3D=(X=1,Y=1,Z=1))
    Settings="(InitialSpace=LocalSpace,bIsPosition=True,Limits=(LimitX=(bMinimum=True,bMaximum=True),LimitY=(bMinimum=True,bMaximum=True),LimitZ=(bMinimum=True,bMaximum=True),MinValue=(X=0,Y=0,Z=0),MaxValue=(X=0,Y=0,Z=0),bDrawLimits=True),Shape=(bVisible=True,Name=\"Sphe ...(truncated)"
    Parent=@n12.Item
   ) {
    is=struct RigVMUnitNode
    template="SpawnControl::Execute(in InitialValue,in Settings,in OffsetTransform,in OffsetSpace,in Parent,in Name,out Item)"
   }
   > n13 = &RigUnit_HierarchyAddNull(
    Name=R_Eye_Aim_GRP
    Space=LocalSpace
    Transform=(Rotation=(X=0.0001,Y=0.0074,Z=0.9998,W=-0.0187),Translation=(X=-3.9875,Y=-0.7659,Z=-0.329),Scale3D=(X=0.89,Y=0.89,Z=0.89))
    Parent=@n11.Item
   )
   > &RigUnit_HierarchyAddControlVector(
    InitialValue=(X=0,Y=0,Z=0)
    Name=R_Eye_Aim
    OffsetSpace=LocalSpace
    OffsetTransform=(Rotation=(X=0,Y=0,Z=0,W=1),Translation=(X=0,Y=0,Z=0),Scale3D=(X=1,Y=1,Z=1))
    Settings="(InitialSpace=LocalSpace,bIsPosition=True,Limits=(LimitX=(bMinimum=True,bMaximum=True),LimitY=(bMinimum=True,bMaximum=True),LimitZ=(bMinimum=True,bMaximum=True),MinValue=(X=0,Y=0,Z=0),MaxValue=(X=0,Y=0,Z=0),bDrawLimits=True),Shape=(bVisible=True,Name=\"Sphe ...(truncated)"
    Parent=@n13.Item
   ) {template="SpawnControl::Execute(in InitialValue,in Settings,in OffsetTransform,in OffsetSpace,in Parent,in Name,out Item)"}
   n14 = /Script/ControlRig.RigUnit_CollectionChildrenArray(Parent=(Type=Null,Name="FACE_CONTROL_GRP"), TypeToSearch=Control, bDefaultChildren=false, bIncludeParent=False, bRecursive=true) {is=struct RigVMUnitNode}
   > For_Each = RigVMDispatchNode(Array=@n14.Items) {template="DISPATCH_RigVMDispatch_ArrayIterator(in Array,out Element,out Index,out Count,out Ratio)"}
   n15 = /Script/ControlRig.RigUnit_GetShapeTransform(Control=@For_Each.Element.Name) {is=struct RigVMUnitNode}
   RigVMFunction_MathVectorFromDouble = /Script/RigVM.RigVMFunction_MathVectorFromDouble(Value=RigVMVariableNode() {variable=Control Scale}.Value) {is=struct RigVMUnitNode, template="Cast::Execute(in Value,out Result)"}
   Multiply = /Script/RigVM.RigVMFunction_MathVectorMul(A=@n15.Transform.Scale3D, B=@RigVMFunction_MathVectorFromDouble.Result) {is=struct RigVMUnitNode, template="Multiply::Execute(in A,in B,out Result)"}
   > /Script/ControlRig.RigUnit_SetShapeTransform(
    Transform=(Rotation=(X=0,Y=0,Z=0,W=1),Translation=(X=0,Y=0,Z=0),Scale3D=(X=1,Y=1,Z=1))
    Transform.Rotation=@n15.Transform.Rotation
    Transform.Translation=@n15.Transform.Translation
    Transform.Scale3D=@Multiply.Result
    Control=@For_Each.Element.Name
   ) {is=struct RigVMUnitNode}
   > SetControlColor = /Script/ControlRig.RigUnit_SetControlColor(Color=RigVMVariableNode() {variable=Control Color}.Value, Control=@For_Each.Element.Name) {is=struct RigVMUnitNode, template="SetControlColor::Execute(in Control,in Color)"}
   n16 = RigVMVariableNode() {variable=Local Gizmo Ctrl}
   n17 = &RigUnit_HierarchyAddNull(
    Name=Nose_GRP
    Space=LocalSpace
    Transform=(Rotation=(X=0,Y=-0.7071,Z=-0.7071,W=0),Translation=(X=0,Y=0,Z=14.3),Scale3D=(X=0.8,Y=0.8,Z=0.8))
    Parent=@n16.Value
    ExecutePin=@RigVMFunction_Sequence.B
   )
   > /Script/ControlRig.RigUnit_HierarchyAddControlFloat(
    InitialValue=0.0
    Name=Nose
    OffsetSpace=LocalSpace
    OffsetTransform=(Rotation=(X=0,Y=0,Z=0,W=1),Translation=(X=0,Y=0,Z=0),Scale3D=(X=1,Y=1,Z=1))
    Settings="(PrimaryAxis=Y,bIsScale=False,Limits=(Limit=(bMinimum=True,bMaximum=True),MinValue=0,MaxValue=1,bDrawLimits=True),Shape=(bVisible=True,Name=\"DefaultGizmoLibraryNormalized.Sphere_Solid\",Color=(R=1,G=1,B=0,A=1),Transform=(Rotation=(X=0,Y=0,Z=0,W=1),Translati ...(truncated)"
    Parent=@n17.Item
   ) {
    is=struct RigVMUnitNode
    template="SpawnControl::Execute(in InitialValue,in Settings,in OffsetTransform,in OffsetSpace,in Parent,in Name,out Item)"
   }
   > n18 = &RigUnit_HierarchyAddNull(
    Name=L_Cheek_GRP
    Space=LocalSpace
    Transform=(Rotation=(X=0,Y=-0.7071,Z=-0.7071,W=0),Translation=(X=3,Y=0,Z=14.1),Scale3D=(X=0.8,Y=0.8,Z=0.8))
    Parent=@n16.Value
   )
   > /Script/ControlRig.RigUnit_HierarchyAddControlVector2D(
    InitialValue=(X=0,Y=0)
    Name=L_Cheek
    OffsetSpace=LocalSpace
    OffsetTransform=(Rotation=(X=0,Y=0,Z=0,W=1),Translation=(X=0,Y=0,Z=0),Scale3D=(X=1,Y=1,Z=1))
    Settings="(PrimaryAxis=Z,Limits=(LimitX=(bMinimum=True,bMaximum=True),LimitY=(bMinimum=True,bMaximum=True),MinValue=(X=0,Y=0),MaxValue=(X=1,Y=1),bDrawLimits=True),Shape=(bVisible=True,Name=\"DefaultGizmoLibraryNormalized.Sphere_Solid\",Color=(R=0,G=1,B=0,A=1),Transfor ...(truncated)"
    Parent=@n18.Item
   ) {
    is=struct RigVMUnitNode
    template="SpawnControl::Execute(in InitialValue,in Settings,in OffsetTransform,in OffsetSpace,in Parent,in Name,out Item)"
   }
   > n19 = &RigUnit_HierarchyAddNull(
    Name=R_Cheek_GRP
    Space=LocalSpace
    Transform=(Rotation=(X=0.7071,Y=0,Z=0,W=0.7071),Translation=(X=-3,Y=0,Z=14.1),Scale3D=(X=0.8,Y=0.8,Z=0.8))
    Parent=@n16.Value
   )
   > &RigUnit_HierarchyAddControlVector2D(
    InitialValue=(X=0,Y=0)
    Name=R_Cheek
    OffsetSpace=LocalSpace
    OffsetTransform=(Rotation=(X=0,Y=0,Z=0,W=1),Translation=(X=0,Y=0,Z=0),Scale3D=(X=1,Y=1,Z=1))
    Settings="(PrimaryAxis=Z,Limits=(LimitX=(bMinimum=True,bMaximum=True),LimitY=(bMinimum=True,bMaximum=True),MinValue=(X=0,Y=0),MaxValue=(X=1,Y=1),bDrawLimits=True),Shape=(bVisible=True,Name=\"DefaultGizmoLibraryNormalized.Sphere_Solid\",Color=(R=1,G=0,B=0,A=1),Transfor ...(truncated)"
    Parent=@n19.Item
   ) {template="SpawnControl::Execute(in InitialValue,in Settings,in OffsetTransform,in OffsetSpace,in Parent,in Name,out Item)"}
   > n20 = &RigUnit_HierarchyAddNull(
    Name=Mouth_GRP
    Space=LocalSpace
    Transform=(Rotation=(X=0,Y=-0.7071,Z=-0.7071,W=0),Translation=(X=0,Y=0,Z=10.7),Scale3D=(X=0.8,Y=0.8,Z=0.8))
    Parent=RigVMVariableNode() {variable=Local Gizmo Ctrl}.Value
   )
   > n21 = &RigUnit_HierarchyAddControlFloat(
    InitialValue=0.0
    Name=Mouth
    OffsetSpace=LocalSpace
    OffsetTransform=(Rotation=(X=0,Y=0,Z=0,W=1),Translation=(X=0,Y=0,Z=0),Scale3D=(X=1,Y=1,Z=1))
    Settings="(PrimaryAxis=Y,bIsScale=False,Limits=(Limit=(bMinimum=True,bMaximum=True),MinValue=-1,MaxValue=1,bDrawLimits=True),Shape=(bVisible=True,Name=\"DefaultGizmoLibraryNormalized.Sphere_Solid\",Color=(R=0,G=0,B=1,A=1),Transform=(Rotation=(X=0,Y=0,Z=0,W=1),Translat ...(truncated)"
    Parent=@n20.Item
   ) {template="SpawnControl::Execute(in InitialValue,in Settings,in OffsetTransform,in OffsetSpace,in Parent,in Name,out Item)"}
   > &RigUnit_HierarchyAddAnimationChannelFloat(
    InitialValue=0.0
    LimitsEnabled=(Enabled=(bMinimum=True,bMaximum=True))
    MaximumValue=1.0
    MinimumValue=0.0
    Name=phoneme_ch
    Parent=@n21.Item
   ) {template="SpawnAnimationChannel::Execute(in InitialValue,in MinimumValue,in MaximumValue,in LimitsEnabled,in Parent,in Name,out Item)"}
   > &RigUnit_HierarchyAddAnimationChannelFloat(
    InitialValue=0.0
    LimitsEnabled=(Enabled=(bMinimum=True,bMaximum=True))
    MaximumValue=1.0
    MinimumValue=0.0
    Name=phoneme_fv
    Parent=@n21.Item
   ) {template="SpawnAnimationChannel::Execute(in InitialValue,in MinimumValue,in MaximumValue,in LimitsEnabled,in Parent,in Name,out Item)"}
   > &RigUnit_HierarchyAddAnimationChannelFloat(
    InitialValue=0.0
    LimitsEnabled=(Enabled=(bMinimum=True,bMaximum=True))
    MaximumValue=1.0
    MinimumValue=0.0
    Name=phoneme_mbp
    Parent=@n21.Item
   ) {template="SpawnAnimationChannel::Execute(in InitialValue,in MinimumValue,in MaximumValue,in LimitsEnabled,in Parent,in Name,out Item)"}
   > &RigUnit_HierarchyAddAnimationChannelFloat(
    InitialValue=0.0
    LimitsEnabled=(Enabled=(bMinimum=True,bMaximum=True))
    MaximumValue=1.0
    MinimumValue=0.0
    Name=phoneme_oo
    Parent=@n21.Item
   ) {template="SpawnAnimationChannel::Execute(in InitialValue,in MinimumValue,in MaximumValue,in LimitsEnabled,in Parent,in Name,out Item)"}
   n22 = RigVMVariableNode() {variable=Local Gizmo Ctrl}
   > n23 = &RigUnit_HierarchyAddNull(
    Name=Lip_Upper_GRP
    Space=LocalSpace
    Transform=(Rotation=(X=0,Y=-0.7071,Z=-0.7071,W=0),Translation=(X=0,Y=0,Z=12.65),Scale3D=(X=0.8,Y=0.8,Z=0.8))
    Parent=@n22.Value
   )
   > &RigUnit_HierarchyAddControlFloat(
    InitialValue=0.0
    Name=Lip_Upper
    OffsetSpace=LocalSpace
    OffsetTransform=(Rotation=(X=0,Y=0,Z=0,W=1),Translation=(X=0,Y=0,Z=0),Scale3D=(X=1,Y=1,Z=1))
    Settings="(PrimaryAxis=Y,bIsScale=False,Limits=(Limit=(bMinimum=True,bMaximum=True),MinValue=-1,MaxValue=1,bDrawLimits=True),Shape=(bVisible=True,Name=\"DefaultGizmoLibraryNormalized.Sphere_Solid\",Color=(R=1,G=1,B=0,A=1),Transform=(Rotation=(X=0,Y=0,Z=0,W=1),Translat ...(truncated)"
    Parent=@n23.Item
   ) {template="SpawnControl::Execute(in InitialValue,in Settings,in OffsetTransform,in OffsetSpace,in Parent,in Name,out Item)"}
   > n24 = &RigUnit_HierarchyAddNull(
    Name=Lip_Lower_GRP
    Space=LocalSpace
    Transform=(Rotation=(X=0,Y=-0.7071,Z=0.7071,W=0),Translation=(X=0,Y=0,Z=8.6),Scale3D=(X=0.8,Y=0.8,Z=0.8))
    Parent=@n22.Value
   )
   > &RigUnit_HierarchyAddControlFloat(
    InitialValue=0.0
    Name=Lip_Lower
    OffsetSpace=LocalSpace
    OffsetTransform=(Rotation=(X=0,Y=0,Z=0,W=1),Translation=(X=0,Y=0,Z=0),Scale3D=(X=1,Y=1,Z=1))
    Settings="(PrimaryAxis=Y,bIsScale=False,Limits=(Limit=(bMinimum=True,bMaximum=True),MinValue=-1,MaxValue=1,bDrawLimits=True),Shape=(bVisible=True,Name=\"DefaultGizmoLibraryNormalized.Sphere_Solid\",Color=(R=1,G=1,B=0,A=1),Transform=(Rotation=(X=0,Y=0,Z=0,W=1),Translat ...(truncated)"
    Parent=@n24.Item
   ) {template="SpawnControl::Execute(in InitialValue,in Settings,in OffsetTransform,in OffsetSpace,in Parent,in Name,out Item)"}
   n25 = RigVMVariableNode() {variable=Local Gizmo Ctrl}
   > n26 = &RigUnit_HierarchyAddNull(
    Name=L_Lip_Upper_GRP
    Space=LocalSpace
    Transform=(Rotation=(X=-0.2706,Y=-0.6533,Z=-0.6533,W=0.2706),Translation=(X=2.13,Y=0,Z=12.1),Scale3D=(X=0.8,Y=0.8,Z=0.8))
    Parent=@n25.Value
   )
   > &RigUnit_HierarchyAddControlFloat(
    InitialValue=0.0
    Name=L_Lip_Upper
    OffsetSpace=LocalSpace
    OffsetTransform=(Rotation=(X=0,Y=0,Z=0,W=1),Translation=(X=0,Y=0,Z=0),Scale3D=(X=1,Y=1,Z=1))
    Settings="(PrimaryAxis=Y,bIsScale=False,Limits=(Limit=(bMinimum=True,bMaximum=True),MinValue=-1,MaxValue=1,bDrawLimits=True),Shape=(bVisible=True,Name=\"DefaultGizmoLibraryNormalized.Sphere_Solid\",Color=(R=0,G=1,B=0,A=1),Transform=(Rotation=(X=0,Y=0,Z=0,W=1),Translat ...(truncated)"
    Parent=@n26.Item
   ) {template="SpawnControl::Execute(in InitialValue,in Settings,in OffsetTransform,in OffsetSpace,in Parent,in Name,out Item)"}
   > n27 = &RigUnit_HierarchyAddNull(
    Name=L_Lipcorner_GRP
    Space=LocalSpace
    Transform=(Rotation=(X=0,Y=-0.7071,Z=-0.7071,W=0),Translation=(X=3,Y=0,Z=10.7),Scale3D=(X=0.8,Y=0.8,Z=0.8))
    Parent=@n25.Value
   )
   > &RigUnit_HierarchyAddControlVector2D(
    InitialValue=(X=0,Y=0)
    Name=L_Lipcorner
    OffsetSpace=LocalSpace
    OffsetTransform=(Rotation=(X=0,Y=0,Z=0,W=1),Translation=(X=0,Y=0,Z=0),Scale3D=(X=1,Y=1,Z=1))
    Settings="(PrimaryAxis=Z,Limits=(LimitX=(bMinimum=True,bMaximum=True),LimitY=(bMinimum=True,bMaximum=True),MinValue=(X=-1,Y=-1),MaxValue=(X=1,Y=1),bDrawLimits=True),Shape=(bVisible=True,Name=\"DefaultGizmoLibraryNormalized.Sphere_Solid\",Color=(R=0,G=1,B=0,A=1),Transf ...(truncated)"
    Parent=@n27.Item
   ) {template="SpawnControl::Execute(in InitialValue,in Settings,in OffsetTransform,in OffsetSpace,in Parent,in Name,out Item)"}
   > n28 = &RigUnit_HierarchyAddNull(
    Name=L_Lip_Lower_GRP
    Space=LocalSpace
    Transform=(Rotation=(X=-0.2706,Y=-0.6533,Z=0.6533,W=-0.2706),Translation=(X=1.63,Y=0,Z=9.1),Scale3D=(X=0.8,Y=0.8,Z=0.8))
    Parent=@n25.Value
   )
   > &RigUnit_HierarchyAddControlFloat(
    InitialValue=0.0
    Name=L_Lip_Lower
    OffsetSpace=LocalSpace
    OffsetTransform=(Rotation=(X=0,Y=0,Z=0,W=1),Translation=(X=0,Y=0,Z=0),Scale3D=(X=1,Y=1,Z=1))
    Settings="(PrimaryAxis=Y,bIsScale=False,Limits=(Limit=(bMinimum=True,bMaximum=True),MinValue=-1,MaxValue=1,bDrawLimits=True),Shape=(bVisible=True,Name=\"DefaultGizmoLibraryNormalized.Sphere_Solid\",Color=(R=0,G=1,B=0,A=1),Transform=(Rotation=(X=0,Y=0,Z=0,W=1),Translat ...(truncated)"
    Parent=@n28.Item
   ) {template="SpawnControl::Execute(in InitialValue,in Settings,in OffsetTransform,in OffsetSpace,in Parent,in Name,out Item)"}
   > n29 = &RigUnit_HierarchyAddNull(
    Name=L_Frown_GRP
    Space=LocalSpace
    Transform=(Rotation=(X=0,Y=-0.7071,Z=-0.7071,W=0),Translation=(X=4.5,Y=0,Z=9),Scale3D=(X=0.8,Y=0.8,Z=0.8))
    Parent=@n25.Value
   )
   > &RigUnit_HierarchyAddControlFloat(
    InitialValue=0.0
    Name=L_Frown
    OffsetSpace=LocalSpace
    OffsetTransform=(Rotation=(X=0,Y=0,Z=0,W=1),Translation=(X=0,Y=0,Z=0),Scale3D=(X=1,Y=1,Z=1))
    Settings="(PrimaryAxis=Y,bIsScale=False,Limits=(Limit=(bMinimum=True,bMaximum=True),MinValue=-1,MaxValue=0,bDrawLimits=True),Shape=(bVisible=True,Name=\"DefaultGizmoLibraryNormalized.Sphere_Solid\",Color=(R=0,G=1,B=0,A=1),Transform=(Rotation=(X=0,Y=0,Z=0,W=1),Translat ...(truncated)"
    Parent=@n29.Item
   ) {template="SpawnControl::Execute(in InitialValue,in Settings,in OffsetTransform,in OffsetSpace,in Parent,in Name,out Item)"}
   n30 = RigVMVariableNode() {variable=Local Gizmo Ctrl}
   > n31 = &RigUnit_HierarchyAddNull(
    Name=R_Lip_Upper_GRP
    Space=LocalSpace
    Transform=(Rotation=(X=-0.2706,Y=0.6533,Z=0.6533,W=0.2706),Translation=(X=-2.13,Y=0,Z=12.1),Scale3D=(X=0.8,Y=0.8,Z=0.8))
    Parent=@n30.Value
   )
   > &RigUnit_HierarchyAddControlFloat(
    InitialValue=0.0
    Name=R_Lip_Upper
    OffsetSpace=LocalSpace
    OffsetTransform=(Rotation=(X=0,Y=0,Z=0,W=1),Translation=(X=0,Y=0,Z=0),Scale3D=(X=1,Y=1,Z=1))
    Settings="(PrimaryAxis=Y,bIsScale=False,Limits=(Limit=(bMinimum=True,bMaximum=True),MinValue=-1,MaxValue=1,bDrawLimits=True),Shape=(bVisible=True,Name=\"DefaultGizmoLibraryNormalized.Sphere_Solid\",Color=(R=1,G=0,B=0,A=1),Transform=(Rotation=(X=0,Y=0,Z=0,W=1),Translat ...(truncated)"
    Parent=@n31.Item
   ) {template="SpawnControl::Execute(in InitialValue,in Settings,in OffsetTransform,in OffsetSpace,in Parent,in Name,out Item)"}
   > n32 = &RigUnit_HierarchyAddNull(
    Name=R_Lipcorner_GRP
    Space=LocalSpace
    Transform=(Rotation=(X=0.7071,Y=0,Z=0,W=0.7071),Translation=(X=-3,Y=0,Z=10.7),Scale3D=(X=0.8,Y=0.8,Z=0.8))
    Parent=@n30.Value
   )
   > &RigUnit_HierarchyAddControlVector2D(
    InitialValue=(X=0,Y=0)
    Name=R_Lipcorner
    OffsetSpace=LocalSpace
    OffsetTransform=(Rotation=(X=0,Y=0,Z=0,W=1),Translation=(X=0,Y=0,Z=0),Scale3D=(X=1,Y=1,Z=1))
    Settings="(PrimaryAxis=Z,Limits=(LimitX=(bMinimum=True,bMaximum=True),LimitY=(bMinimum=True,bMaximum=True),MinValue=(X=-1,Y=-1),MaxValue=(X=1,Y=1),bDrawLimits=True),Shape=(bVisible=True,Name=\"DefaultGizmoLibraryNormalized.Sphere_Solid\",Color=(R=1,G=0,B=0,A=1),Transf ...(truncated)"
    Parent=@n32.Item
   ) {template="SpawnControl::Execute(in InitialValue,in Settings,in OffsetTransform,in OffsetSpace,in Parent,in Name,out Item)"}
   > n33 = &RigUnit_HierarchyAddNull(
    Name=R_Lip_Lower_GRP
    Space=LocalSpace
    Transform=(Rotation=(X=-0.2706,Y=0.6533,Z=-0.6533,W=-0.2706),Translation=(X=-1.63,Y=0,Z=9.1),Scale3D=(X=0.8,Y=0.8,Z=0.8))
    Parent=@n30.Value
   )
   > &RigUnit_HierarchyAddControlFloat(
    InitialValue=0.0
    Name=R_Lip_Lower
    OffsetSpace=LocalSpace
    OffsetTransform=(Rotation=(X=0,Y=0,Z=0,W=1),Translation=(X=0,Y=0,Z=0),Scale3D=(X=1,Y=1,Z=1))
    Settings="(PrimaryAxis=Y,bIsScale=False,Limits=(Limit=(bMinimum=True,bMaximum=True),MinValue=-1,MaxValue=1,bDrawLimits=True),Shape=(bVisible=True,Name=\"DefaultGizmoLibraryNormalized.Sphere_Solid\",Color=(R=1,G=0,B=0,A=1),Transform=(Rotation=(X=0,Y=0,Z=0,W=1),Translat ...(truncated)"
    Parent=@n33.Item
   ) {template="SpawnControl::Execute(in InitialValue,in Settings,in OffsetTransform,in OffsetSpace,in Parent,in Name,out Item)"}
   > n34 = &RigUnit_HierarchyAddNull(
    Name=R_Frown_GRP
    Space=LocalSpace
    Transform=(Rotation=(X=0,Y=-0.7071,Z=-0.7071,W=0),Translation=(X=-4.5,Y=0,Z=9),Scale3D=(X=0.8,Y=0.8,Z=0.8))
    Parent=@n30.Value
   )
   > &RigUnit_HierarchyAddControlFloat(
    InitialValue=0.0
    Name=R_Frown
    OffsetSpace=LocalSpace
    OffsetTransform=(Rotation=(X=0,Y=0,Z=0,W=1),Translation=(X=0,Y=0,Z=0),Scale3D=(X=1,Y=1,Z=1))
    Settings="(PrimaryAxis=Y,bIsScale=False,Limits=(Limit=(bMinimum=True,bMaximum=True),MinValue=-1,MaxValue=0,bDrawLimits=True),Shape=(bVisible=True,Name=\"DefaultGizmoLibraryNormalized.Sphere_Solid\",Color=(R=0,G=1,B=0,A=1),Transform=(Rotation=(X=0,Y=0,Z=0,W=1),Translat ...(truncated)"
    Parent=@n34.Item
   ) {template="SpawnControl::Execute(in InitialValue,in Settings,in OffsetTransform,in OffsetSpace,in Parent,in Name,out Item)"}
   n35 = RigVMVariableNode() {variable=Local Gizmo Ctrl}
   n36 = &RigUnit_HierarchyAddNull(
    Name=midbrow_GRP
    Space=LocalSpace
    Transform=(Rotation=(X=0,Y=-0.7071,Z=-0.7071,W=0),Translation=(X=0,Y=0,Z=19.63),Scale3D=(X=0.8,Y=0.8,Z=0.8))
    Parent=@n35.Value
    ExecutePin=@RigVMFunction_Sequence.A
   )
   > RigUnit_HierarchyAddControlFloat = &RigUnit_HierarchyAddControlFloat(
    InitialValue=0.0
    Name=midbrow
    OffsetSpace=LocalSpace
    OffsetTransform=(Rotation=(X=0,Y=0,Z=0,W=1),Translation=(X=0,Y=0,Z=0),Scale3D=(X=1,Y=1,Z=1))
    Settings="(PrimaryAxis=Y,bIsScale=False,Limits=(Limit=(bMinimum=True,bMaximum=True),MinValue=-1,MaxValue=1,bDrawLimits=True),Shape=(bVisible=True,Name=\"DefaultGizmoLibraryNormalized.Sphere_Solid\",Color=(R=1,G=1,B=0,A=1),Transform=(Rotation=(X=0,Y=0,Z=0,W=1),Translat ...(truncated)"
    Parent=@n36.Item
   ) {template="SpawnControl::Execute(in InitialValue,in Settings,in OffsetTransform,in OffsetSpace,in Parent,in Name,out Item)"}
   > n37 = &RigUnit_HierarchyAddNull(
    Name=L_Brow_GRP
    Space=LocalSpace
    Transform=(Rotation=(X=0,Y=-0.7071,Z=-0.7071,W=0),Translation=(X=3,Y=0,Z=19.2),Scale3D=(X=0.8,Y=0.8,Z=0.8))
    Parent=@n35.Value
   )
   > RigUnit_HierarchyAddControlVector2D = &RigUnit_HierarchyAddControlVector2D(
    InitialValue=(X=0,Y=0)
    Name=L_Brow
    OffsetSpace=LocalSpace
    OffsetTransform=(Rotation=(X=0,Y=0,Z=0,W=1),Translation=(X=0,Y=0,Z=0),Scale3D=(X=1,Y=1,Z=1))
    Settings="(PrimaryAxis=Z,Limits=(LimitX=(bMinimum=True,bMaximum=True),LimitY=(bMinimum=True,bMaximum=True),MinValue=(X=0,Y=-1),MaxValue=(X=1,Y=1),bDrawLimits=True),Shape=(bVisible=True,Name=\"DefaultGizmoLibraryNormalized.Sphere_Solid\",Color=(R=0,G=1,B=0,A=1),Transfo ...(truncated)"
    Parent=@n37.Item
   ) {template="SpawnControl::Execute(in InitialValue,in Settings,in OffsetTransform,in OffsetSpace,in Parent,in Name,out Item)"}
   > n38 = &RigUnit_HierarchyAddNull(
    Name=R_Brow_GRP
    Space=LocalSpace
    Transform=(Rotation=(X=0.7071,Y=0,Z=0,W=0.7071),Translation=(X=-3,Y=0,Z=19.2),Scale3D=(X=0.8,Y=0.8,Z=0.8))
    Parent=@n35.Value
   )
   > &RigUnit_HierarchyAddControlVector2D(
    InitialValue=(X=0,Y=0)
    Name=R_Brow
    OffsetSpace=LocalSpace
    OffsetTransform=(Rotation=(X=0,Y=0,Z=0,W=1),Translation=(X=0,Y=0,Z=0),Scale3D=(X=1,Y=1,Z=1))
    Settings="(PrimaryAxis=Z,Limits=(LimitX=(bMinimum=True,bMaximum=True),LimitY=(bMinimum=True,bMaximum=True),MinValue=(X=0,Y=-1),MaxValue=(X=1,Y=1),bDrawLimits=True),Shape=(bVisible=True,Name=\"DefaultGizmoLibraryNormalized.Sphere_Solid\",Color=(R=1,G=0,B=0,A=1),Transfo ...(truncated)"
    Parent=@n38.Item
   ) {template="SpawnControl::Execute(in InitialValue,in Settings,in OffsetTransform,in OffsetSpace,in Parent,in Name,out Item)"}
   n39 = RigVMVariableNode() {variable=Local Gizmo Ctrl}
   > n40 = &RigUnit_HierarchyAddNull(
    Name=L_Eyelid_GRP
    Space=LocalSpace
    Transform=(Rotation=(X=0,Y=-0.7071,Z=-0.7071,W=0),Translation=(X=3,Y=0,Z=16.3),Scale3D=(X=0.8,Y=0.8,Z=0.8))
    Parent=@n39.Value
   )
   > n41 = &RigUnit_HierarchyAddControlVector2D(
    InitialValue=(X=0,Y=0)
    Name=L_Eyelid
    OffsetSpace=LocalSpace
    OffsetTransform=(Rotation=(X=0,Y=0,Z=0,W=1),Translation=(X=0,Y=0,Z=0),Scale3D=(X=1,Y=1,Z=1))
    Settings="(PrimaryAxis=Z,Limits=(LimitX=(bMinimum=True,bMaximum=True),LimitY=(bMinimum=True,bMaximum=True),MinValue=(X=0,Y=-1),MaxValue=(X=1,Y=1),bDrawLimits=True),Shape=(bVisible=True,Name=\"DefaultGizmoLibraryNormalized.Sphere_Solid\",Color=(R=0,G=1,B=0,A=1),Transfo ...(truncated)"
    Parent=@n40.Item
   ) {template="SpawnControl::Execute(in InitialValue,in Settings,in OffsetTransform,in OffsetSpace,in Parent,in Name,out Item)"}
   > RigUnit_HierarchyAddAnimationChannelFloat = &RigUnit_HierarchyAddAnimationChannelFloat(
    InitialValue=0.0
    LimitsEnabled=(Enabled=(bMinimum=True,bMaximum=True))
    MaximumValue=1.0
    MinimumValue=0.0
    Name=upper_lid_down
    Parent=@n41.Item
   ) {template="SpawnAnimationChannel::Execute(in InitialValue,in MinimumValue,in MaximumValue,in LimitsEnabled,in Parent,in Name,out Item)"}
   > n42 = &RigUnit_HierarchyAddNull(
    Name=R_Eyelid_GRP
    Space=LocalSpace
    Transform=(Rotation=(X=0.7071,Y=0,Z=0,W=0.7071),Translation=(X=-3,Y=0,Z=16.3),Scale3D=(X=0.8,Y=0.8,Z=0.8))
    Parent=@n39.Value
   )
   > n43 = &RigUnit_HierarchyAddControlVector2D(
    InitialValue=(X=0,Y=0)
    Name=R_Eyelid
    OffsetSpace=LocalSpace
    OffsetTransform=(Rotation=(X=0,Y=0,Z=0,W=1),Translation=(X=0,Y=0,Z=0),Scale3D=(X=1,Y=1,Z=1))
    Settings="(PrimaryAxis=Z,Limits=(LimitX=(bMinimum=True,bMaximum=True),LimitY=(bMinimum=True,bMaximum=True),MinValue=(X=0,Y=-1),MaxValue=(X=1,Y=1),bDrawLimits=True),Shape=(bVisible=True,Name=\"DefaultGizmoLibraryNormalized.Sphere_Solid\",Color=(R=1,G=0,B=0,A=1),Transfo ...(truncated)"
    Parent=@n42.Item
   ) {template="SpawnControl::Execute(in InitialValue,in Settings,in OffsetTransform,in OffsetSpace,in Parent,in Name,out Item)"}
   > &RigUnit_HierarchyAddAnimationChannelFloat(
    InitialValue=0.0
    LimitsEnabled=(Enabled=(bMinimum=True,bMaximum=True))
    MaximumValue=1.0
    MinimumValue=0.0
    Name=upper_lid_down
    Parent=@n43.Item
   ) {template="SpawnAnimationChannel::Execute(in InitialValue,in MinimumValue,in MaximumValue,in LimitsEnabled,in Parent,in Name,out Item)"}
   > n44 = &RigUnit_HierarchyAddNull(
    Name=Look_GRP
    Space=LocalSpace
    Transform=(Rotation=(X=0,Y=-0.7071,Z=-0.7071,W=0),Translation=(X=0,Y=0,Z=16.3),Scale3D=(X=0.85,Y=0.85,Z=0.85))
    Parent=RigVMVariableNode() {variable=Local Gizmo Ctrl}.Value
   )
   > n45 = &RigUnit_HierarchyAddControlVector2D(
    InitialValue=(X=0,Y=0)
    Name=Look
    OffsetSpace=LocalSpace
    OffsetTransform=(Rotation=(X=0,Y=0,Z=0,W=1),Translation=(X=0,Y=0,Z=0),Scale3D=(X=1,Y=1,Z=1))
    Settings="(PrimaryAxis=Z,Limits=(LimitX=(bMinimum=True,bMaximum=True),LimitY=(bMinimum=True,bMaximum=True),MinValue=(X=-1,Y=-1),MaxValue=(X=1,Y=1),bDrawLimits=True),Shape=(bVisible=True,Name=\"DefaultGizmoLibraryNormalized.Sphere_Solid\",Color=(R=1,G=1,B=0,A=1),Transf ...(truncated)"
    Parent=@n44.Item
   ) {template="SpawnControl::Execute(in InitialValue,in Settings,in OffsetTransform,in OffsetSpace,in Parent,in Name,out Item)"}
   > &RigUnit_HierarchyAddAnimationChannelFloat(
    InitialValue=0.0
    LimitsEnabled=(Enabled=(bMinimum=True,bMaximum=True))
    MaximumValue=1.0
    MinimumValue=0.0
    Name=eye_convergence
    Parent=@n45.Item
   ) {template="SpawnAnimationChannel::Execute(in InitialValue,in MinimumValue,in MaximumValue,in LimitsEnabled,in Parent,in Name,out Item)"}
   > RigUnit_HierarchyAddAnimationChannelBool = &RigUnit_HierarchyAddAnimationChannelBool(InitialValue=False, LimitsEnabled=(), MaximumValue=True, MinimumValue=False, Name=use_eye_aim, Parent=@n45.Item) {template="SpawnAnimationChannel::Execute(in InitialValue,in MinimumValue,in MaximumValue,in LimitsEnabled,in Parent,in Name,out Item)"}
   n46 = RigVMVariableNode() {variable=Local Gizmo Ctrl}
   > n47 = &RigUnit_HierarchyAddNull(
    Name=Jaw_GRP
    Space=LocalSpace
    Transform=(Rotation=(X=0,Y=-0.7071,Z=-0.7071,W=0),Translation=(X=0,Y=0,Z=5.93),Scale3D=(X=0.85,Y=0.85,Z=0.85))
    Parent=@n46.Value
   )
   > &RigUnit_HierarchyAddControlVector2D(
    InitialValue=(X=0,Y=0)
    Name=Jaw
    OffsetSpace=LocalSpace
    OffsetTransform=(Rotation=(X=0,Y=0,Z=0,W=1),Translation=(X=0,Y=0,Z=0),Scale3D=(X=1,Y=1,Z=1))
    Settings="(PrimaryAxis=Z,Limits=(LimitX=(bMinimum=True,bMaximum=True),LimitY=(bMinimum=True,bMaximum=True),MinValue=(X=-1,Y=-1),MaxValue=(X=1,Y=1),bDrawLimits=True),Shape=(bVisible=True,Name=\"DefaultGizmoLibraryNormalized.Sphere_Solid\",Color=(R=1,G=1,B=0,A=1),Transf ...(truncated)"
    Parent=@n47.Item
   ) {template="SpawnControl::Execute(in InitialValue,in Settings,in OffsetTransform,in OffsetSpace,in Parent,in Name,out Item)"}
   > n48 = &RigUnit_HierarchyAddNull(
    Name=Jaw_FwdBk_GRP
    Space=LocalSpace
    Transform=(Rotation=(X=0,Y=-0.7071,Z=-0.7071,W=0),Translation=(X=-1.24,Y=0,Z=5.9),Scale3D=(X=1,Y=1,Z=1))
    Parent=@n46.Value
   )
   > &RigUnit_HierarchyAddControlFloat(
    InitialValue=0.0
    Name=Jaw_FwdBk
    OffsetSpace=LocalSpace
    OffsetTransform=(Rotation=(X=0,Y=0,Z=0,W=1),Translation=(X=0,Y=0,Z=0),Scale3D=(X=1,Y=1,Z=1))
    Settings="(PrimaryAxis=Y,bIsScale=False,Limits=(Limit=(bMinimum=True,bMaximum=True),MinValue=-1,MaxValue=1,bDrawLimits=True),Shape=(bVisible=True,Name=\"DefaultGizmoLibraryNormalized.Sphere_Solid\",Color=(R=1,G=1,B=0,A=1),Transform=(Rotation=(X=0,Y=0,Z=0,W=1),Translat ...(truncated)"
    Parent=@n48.Item
   ) {template="SpawnControl::Execute(in InitialValue,in Settings,in OffsetTransform,in OffsetSpace,in Parent,in Name,out Item)"}
   > n49 = &RigUnit_HierarchyAddNull(
    Name=Tongue_GRP
    Space=LocalSpace
    Transform=(Rotation=(X=0,Y=-0.7071,Z=-0.7071,W=0),Translation=(X=3,Y=0,Z=5.9),Scale3D=(X=0.8,Y=0.8,Z=0.8))
    Parent=RigVMVariableNode() {variable=Local Gizmo Ctrl}.Value
   )
   > &RigUnit_HierarchyAddControlVector2D(
    InitialValue=(X=0,Y=0)
    Name=tongue
    OffsetSpace=LocalSpace
    OffsetTransform=(Rotation=(X=0,Y=0,Z=0,W=1),Translation=(X=0,Y=0,Z=0),Scale3D=(X=1,Y=1,Z=1))
    Settings="(PrimaryAxis=Z,Limits=(LimitX=(bMinimum=True,bMaximum=True),LimitY=(bMinimum=True,bMaximum=True),MinValue=(X=-1,Y=-1),MaxValue=(X=1,Y=1),bDrawLimits=True),Shape=(bVisible=True,Name=\"DefaultGizmoLibraryNormalized.Sphere_Solid\",Color=(R=1,G=1,B=0,A=1),Transf ...(truncated)"
    Parent=@n49.Item
   ) {template="SpawnControl::Execute(in InitialValue,in Settings,in OffsetTransform,in OffsetSpace,in Parent,in Name,out Item)"}
   RigVMFunctionReturnNode(ExecuteContext=@n2.B)
  ]}
 RigVMFunctionLibrary {nodes [
   CRFL FN Legacy Backward v01 = RigVMCollapseNode(ExecuteContext=())
   CRFL FN Legacy Construction v01 = RigVMCollapseNode(Control Color=(R=1,G=1,B=0,A=1), Control Scale=1, ExecuteContext=())
   CRFL FN Legacy Forward v01 = RigVMCollapseNode(ExecuteContext=())
  ]}
 RigVMFunction_Sequence_1_ContainedGraph {nodes [
   n1 = RigVMFunctionEntryNode()
   RigVMFunction_Sequence = &RigVMFunction_Sequence(ExecuteContext=@n1.ExecuteContext)
   n2 = &RigVMFunction_Sequence(ExecuteContext=@RigVMFunction_Sequence.B)
   n3 = &RigVMFunction_Sequence(ExecuteContext=@n2.B)
   RigVMFunctionReturnNode(A=@RigVMFunction_Sequence.A, B=@n2.A, C=@n3.A, D=@n3.B)
  ]}
 RigVMModel {nodes [
   CRFL FN Legacy Forward v01 = RigVMFunctionReferenceNode(ExecuteContext=())
   RigUnit_BeginExecution = &RigUnit_BeginExecution()
  ]}
 RigVMModel Rig {nodes [
   CRFL FN Legacy Construction v01 = RigVMFunctionReferenceNode(Control Color=(R=1,G=1,B=0,A=1), Control Scale=1, ExecuteContext=())
   &RigUnit_PrepareForExecution()
  ]}
 RigVMModel Rig_1 {nodes [
   CRFL FN Legacy Backward v01 = RigVMFunctionReferenceNode(ExecuteContext=())
   /Script/ControlRig.RigUnit_InverseExecution() {is=struct RigVMUnitNode}
  ]}
 SequenceAggregate_1_ContainedGraph {nodes [
   n1 = RigVMFunctionEntryNode()
   SequenceAggregate = &RigVMFunction_Sequence(ExecuteContext=@n1.ExecuteContext)
   SequenceAggregate_1 = &RigVMFunction_Sequence(ExecuteContext=@SequenceAggregate.B)
   SequenceAggregate_2 = &RigVMFunction_Sequence(ExecuteContext=@SequenceAggregate_1.B)
   SequenceAggregate_3 = &RigVMFunction_Sequence(ExecuteContext=@SequenceAggregate_2.B)
   SequenceAggregate_4 = &RigVMFunction_Sequence(ExecuteContext=@SequenceAggregate_3.B)
   RigVMFunctionReturnNode(
    A=@SequenceAggregate.A
    B=@SequenceAggregate_1.A
    C=@SequenceAggregate_2.A
    D=@SequenceAggregate_3.A
    E=@SequenceAggregate_4.A
    F=@SequenceAggregate_4.B
   )
  ]}
 SequenceAggregate_1_ContainedGraph {nodes [
   n1 = RigVMFunctionEntryNode()
   SequenceAggregate = &RigVMFunction_Sequence(ExecuteContext=@n1.ExecuteContext)
   SequenceAggregate_1 = &RigVMFunction_Sequence(ExecuteContext=@SequenceAggregate.B)
   SequenceAggregate_2 = &RigVMFunction_Sequence(ExecuteContext=@SequenceAggregate_1.B)
   RigVMFunctionReturnNode(A=@SequenceAggregate.A, B=@SequenceAggregate_1.A, C=@SequenceAggregate_2.A, D=@SequenceAggregate_2.B)
  ]}
]
hierarchy [
 root {loc=<0 0 0>, rot=<0 0 0 1>}
 attach {parent=root, loc=<0 0 0>, rot=<0 0 0 1>}
 pelvis {parent=root, loc=<0 -0.7071 93.1191>, rot=<0.0127 0.707 -0.0127 -0.707>}
 spine_01 {parent=pelvis, loc=<6.7494 -0.6388 0>, rot=<0 0 0.034 -0.9994>}
 spine_02 {parent=spine_01, loc=<7.2518 -0.035 0>, rot=<0 0 0.0238 -0.9997>}
 spine_03 {parent=spine_02, loc=<6.0104 0 0>, rot=<0 0 0.0149 -0.9999>}
 spine_04 {parent=spine_03, loc=<8.1024 0 0>, rot=<0 0 -0.0132 -0.9999>}
 spine_05 {parent=spine_04, loc=<10.0146 0 0>, rot=<0 0 -0.0015 -1>}
 clavicle_l {parent=spine_05, loc=<4.3575 -0.1466 -2.3223>, rot=<0.0465 -0.7289 -0.009 -0.6829>}
 clavicle_r {parent=spine_05, loc=<4.3575 -0.1466 2.3223>, rot=<0.7289 0.0465 -0.6829 0.009>}
 neck_01 {parent=spine_05, loc=<11.3138 0.1529 0>, rot=<0 0 -0.2003 -0.9797>}
 neck_02 {parent=neck_01, loc=<4.5627 0 0>, rot=<0 0 0.0504 -0.9987>}
 head {parent=neck_02, loc=<5.3914 0 0>, rot=<0 0 0.115 -0.9934>}
 faceAttach {parent=head, loc=<-0.8266 -0.1832 0>, rot=<0.0044 -0.3827 -0.0044 -0.9239>, scale=<0.89 0.89 0.89>}
 C_brow_mid {parent=faceAttach, loc=<4.7209 11.2084 4.8303>, rot=<-0.0586 -0.317 0.0114 -0.9466>}
 L_brow_mid {parent=faceAttach, loc=<7.639 10.371 3.157>, rot=<0.0603 -0.3852 0.2489 -0.8866>}
 L_brow_outer {parent=faceAttach, loc=<9.1042 8.1218 1.4857>, rot=<0.2349 -0.2525 0.3549 -0.869>}
 L_cheek_inner {parent=faceAttach, loc=<4.2785 8.761 -1.4435>, rot=<0.0454 -0.6262 0.3191 -0.7099>}
 C_lip_upper_mid {parent=faceAttach, loc=<-0.3154 10.7989 -0.2464>, rot=<0.2097 -0.3713 -0.0897 -0.9001>}
 L_lip_upper_outer {parent=faceAttach, loc=<0.8467 10.1076 -1.4454>, rot=<0.1857 -0.3758 0.3015 -0.8564>}
 L_lip_corner {parent=faceAttach, loc=<1.613 9.1457 -2.3124>, rot=<0.0012 -0.375 -0.0189 -0.9269>}
 C_nose_bridge {parent=faceAttach, loc=<2.3602 13.1515 2.4853>, rot=<-0.2002 -0.3676 0.0825 -0.9044>}
 R_brow_mid {parent=faceAttach, loc=<3.0903 10.342 7.7132>, rot=<-0.1347 -0.3556 -0.2231 -0.8976>}
 R_brow_outer {parent=faceAttach, loc=<1.412 8.0675 9.1444>, rot=<-0.1251 -0.3913 -0.4176 -0.8104>}
 R_cheek_inner {parent=faceAttach, loc=<-1.4992 8.7305 4.3237>, rot=<-0.1979 -0.0545 -0.2598 -0.9436>}
 R_lip_corner {parent=faceAttach, loc=<-2.32 9.1134 1.692>, rot=<0.0127 -0.39 0.008 -0.9207>}
 C_jaw {parent=faceAttach, loc=<0.3123 3.2483 0.3582>, rot=<0.0044 -0.3827 -0.0044 -0.9239>}
 L_lip_lower_outer {parent=C_jaw, loc=<1.6966 6.8705 -0.8455>, rot=<-0.0146 -0.0279 0.3432 -0.9387>}
 C_lip_lower_mid {parent=C_jaw, loc=<0.0184 7.4191 -0.884>, rot=<0 -0.0099 -0.0021 -0.9999>}
 R_lip_lower_outer {parent=C_jaw, loc=<-1.6982 6.9042 -0.8853>, rot=<0.0021 -0.0171 -0.3152 -0.9489>}
 teeth_lower {parent=C_jaw, loc=<0.0169 5.3598 -2.8149>, rot=<0 0 0 1>}
 tongue {parent=C_jaw, loc=<0.0178 1.5245 -1.0398>, rot=<-0.1568 0 0 -0.9876>}
 R_eye {parent=faceAttach, loc=<2.2537 8.4351 6.6885>, rot=<0.3823 -0.0165 0.9223 -0.0548>}
 R_eye_lid_upper_mid {parent=R_eye, loc=<-0.0431 -1.1701 0.2679>, rot=<0.0541 -0.001 0.0182 -0.9984>}
 R_eye_lid_lower_mid {parent=R_eye, loc=<-0.0113 -0.9436 -0.7026>, rot=<0.0339 -0.0006 0.0182 -0.9993>}
 L_eye {parent=faceAttach, loc=<6.6314 8.4647 2.3108>, rot=<0.3819 0.0253 0.9227 0.046>}
 L_eye_lid_upper_mid {parent=L_eye, loc=<0.0251 -1.1623 0.2388>, rot=<0.0523 0 0 -0.9986>}
 L_eye_lid_lower_mid {parent=L_eye, loc=<0.0223 -0.923 -0.7488>, rot=<0.0329 0 0 -0.9995>}
 teeth_upper {parent=faceAttach, loc=<0.0954 9.3381 0.1585>, rot=<0.0044 -0.3827 -0.0044 -0.9239>}
 R_lip_upper_outer {parent=faceAttach, loc=<-1.497 10.1359 0.962>, rot=<-0.0848 -0.2926 -0.3215 -0.8966>}
 FX_EyeTrail_L {parent=head, type=Null, loc=<4.75 8 2.6>, rot=<-0.0051 0.6948 0.0011 0.7192>}
 FX_EyeTrail_R {parent=head, type=Null, loc=<4.75 8 -2.831>, rot=<0 0.7071 0 0.7071>}
 FX_Eye_L {parent=head, type=Null, loc=<3.4814 5.9142 -2.8258>, rot=<-0.0059 0.7071 0 0.7071>}
 FX_Eye_R {parent=head, type=Null, loc=<3.4814 5.9142 2.8258>, rot=<-0.0059 0.7071 0 0.7071>}
 earpiece {parent=head, type=Null, loc=<0 0 0>, rot=<0 0 0 1>}
 hat {parent=head, type=Null, loc=<0 0 0>, rot=<0 0.7071 0 0.7071>}
 FX_Visor {parent=head, type=Null, loc=<1.6781 9.0813 0>, rot=<0 0 0 1>}
 Charm_Attach_00 {parent=spine_05, type=Null, loc=<-0.095 14.4 -7.75>, rot=<-0.5245 0.4379 -0.6186 0.388>}
 look_left_pose {type=Curve, loc=<0 0 0>, rot=<0 0 0 1>}
 look_right_pose {type=Curve, loc=<0 0 0>, rot=<0 0 0 1>}
 look_up_pose {type=Curve, loc=<0 0 0>, rot=<0 0 0 1>}
 look_down_pose {type=Curve, loc=<0 0 0>, rot=<0 0 0 1>}
 L_blink_pose {type=Curve, loc=<0 0 0>, rot=<0 0 0 1>}
 R_blink_pose {type=Curve, loc=<0 0 0>, rot=<0 0 0 1>}
 L_wide_pose {type=Curve, loc=<0 0 0>, rot=<0 0 0 1>}
 R_wide_pose {type=Curve, loc=<0 0 0>, rot=<0 0 0 1>}
 R_squint_inner_pose {type=Curve, loc=<0 0 0>, rot=<0 0 0 1>}
 R_squeeze_pose {type=Curve, loc=<0 0 0>, rot=<0 0 0 1>}
 L_squint_inner_pose {type=Curve, loc=<0 0 0>, rot=<0 0 0 1>}
 L_squeeze_pose {type=Curve, loc=<0 0 0>, rot=<0 0 0 1>}
 L_brow_up_pose {type=Curve, loc=<0 0 0>, rot=<0 0 0 1>}
 L_brow_down_pose {type=Curve, loc=<0 0 0>, rot=<0 0 0 1>}
 R_brow_up_pose {type=Curve, loc=<0 0 0>, rot=<0 0 0 1>}
 R_brow_down_pose {type=Curve, loc=<0 0 0>, rot=<0 0 0 1>}
 C_glabella_up_pose {type=Curve, loc=<0 0 0>, rot=<0 0 0 1>}
 C_glabella_down_pose {type=Curve, loc=<0 0 0>, rot=<0 0 0 1>}
 L_brow_squeeze_in_pose {type=Curve, loc=<0 0 0>, rot=<0 0 0 1>}
 R_brow_squeeze_in_pose {type=Curve, loc=<0 0 0>, rot=<0 0 0 1>}
 L_upper_lip_raiser_pose {type=Curve, loc=<0 0 0>, rot=<0 0 0 1>}
 R_upper_lip_raiser_pose {type=Curve, loc=<0 0 0>, rot=<0 0 0 1>}
 L_upper_lip_lower_pose {type=Curve, loc=<0 0 0>, rot=<0 0 0 1>}
 R_upper_lip_lower_pose {type=Curve, loc=<0 0 0>, rot=<0 0 0 1>}
 L_lower_lip_up_pose {type=Curve, loc=<0 0 0>, rot=<0 0 0 1>}
 R_lower_lip_up_pose {type=Curve, loc=<0 0 0>, rot=<0 0 0 1>}
 L_lower_lip_down_pose {type=Curve, loc=<0 0 0>, rot=<0 0 0 1>}
 R_lower_lip_down_pose {type=Curve, loc=<0 0 0>, rot=<0 0 0 1>}
 L_lip_corner_narrow_pose {type=Curve, loc=<0 0 0>, rot=<0 0 0 1>}
 L_lip_corner_wide_pose {type=Curve, loc=<0 0 0>, rot=<0 0 0 1>}
 R_lip_corner_narrow_pose {type=Curve, loc=<0 0 0>, rot=<0 0 0 1>}
 R_lip_corner_wide_pose {type=Curve, loc=<0 0 0>, rot=<0 0 0 1>}
 L_smile_pose {type=Curve, loc=<0 0 0>, rot=<0 0 0 1>}
 R_smile_pose {type=Curve, loc=<0 0 0>, rot=<0 0 0 1>}
 L_frown_pose {type=Curve, loc=<0 0 0>, rot=<0 0 0 1>}
 R_frown_pose {type=Curve, loc=<0 0 0>, rot=<0 0 0 1>}
 phoneme_oo_pose {type=Curve, loc=<0 0 0>, rot=<0 0 0 1>}
 phoneme_ch_pose {type=Curve, loc=<0 0 0>, rot=<0 0 0 1>}
 phoneme_mbp_pose {type=Curve, loc=<0 0 0>, rot=<0 0 0 1>}
 phoneme_fv_pose {type=Curve, loc=<0 0 0>, rot=<0 0 0 1>}
 mouth_up_pose {type=Curve, loc=<0 0 0>, rot=<0 0 0 1>}
 mouth_down_pose {type=Curve, loc=<0 0 0>, rot=<0 0 0 1>}
 C_nose_wrinkler_pose {type=Curve, loc=<0 0 0>, rot=<0 0 0 1>}
 L_cheek_up_pose {type=Curve, loc=<0 0 0>, rot=<0 0 0 1>}
 R_cheek_up_pose {type=Curve, loc=<0 0 0>, rot=<0 0 0 1>}
 jaw_open_pose {type=Curve, loc=<0 0 0>, rot=<0 0 0 1>}
 jaw_right_pose {type=Curve, loc=<0 0 0>, rot=<0 0 0 1>}
 jaw_left_pose {type=Curve, loc=<0 0 0>, rot=<0 0 0 1>}
 eye_convergence_pose {type=Curve, loc=<0 0 0>, rot=<0 0 0 1>}
 L_upperLidDown_pose {type=Curve, loc=<0 0 0>, rot=<0 0 0 1>}
 R_upperLidDown_pose {type=Curve, loc=<0 0 0>, rot=<0 0 0 1>}
 jaw_up_pose {type=Curve, loc=<0 0 0>, rot=<0 0 0 1>}
 jaw_forward_pose {type=Curve, loc=<0 0 0>, rot=<0 0 0 1>}
 jaw_back_pose {type=Curve, loc=<0 0 0>, rot=<0 0 0 1>}
 tongue_forward_pose {type=Curve, loc=<0 0 0>, rot=<0 0 0 1>}
 tongue_back_pose {type=Curve, loc=<0 0 0>, rot=<0 0 0 1>}
 tongue_up_pose {type=Curve, loc=<0 0 0>, rot=<0 0 0 1>}
 tongue_down_pose {type=Curve, loc=<0 0 0>, rot=<0 0 0 1>}
 base_pose {type=Curve, loc=<0 0 0>, rot=<0 0 0 1>}
 Hair_Scarf {type=Curve, loc=<0 0 0>, rot=<0 0 0 1>}
 Hair_Hat {type=Curve, loc=<0 0 0>, rot=<0 0 0 1>}
 Hair_Helmet {type=Curve, loc=<0 0 0>, rot=<0 0 0 1>}
 Hair_Glasses {type=Curve, loc=<0 0 0>, rot=<0 0 0 1>}
 Hair_Cap {type=Curve, loc=<0 0 0>, rot=<0 0 0 1>}
 Hair_Headsets {type=Curve, loc=<0 0 0>, rot=<0 0 0 1>}
 Hair_Hood {type=Curve, loc=<0 0 0>, rot=<0 0 0 1>}
 Face_Mask {type=Curve, loc=<0 0 0>, rot=<0 0 0 1>}
 HairStrands_Hat {type=Curve, loc=<0 0 0>, rot=<0 0 0 1>}
 Hat_Hair {type=Curve, loc=<0 0 0>, rot=<0 0 0 1>}
 Helmet_Hair {type=Curve, loc=<0 0 0>, rot=<0 0 0 1>}
 Hair_Muff {type=Curve, loc=<0 0 0>, rot=<0 0 0 1>}
 Hair_SpaceSuit {type=Curve, loc=<0 0 0>, rot=<0 0 0 1>}
 Hair_Hide {type=Curve, loc=<0 0 0>, rot=<0 0 0 1>}
 Head_SpaceSuit {type=Curve, loc=<0 0 0>, rot=<0 0 0 1>}
 Hair_Null {type=Curve, loc=<0 0 0>, rot=<0 0 0 1>}
 DisableFaceOverride {type=Curve, loc=<0 0 0>, rot=<0 0 0 1>}
 LiveLink {type=Curve, loc=<0 0 0>, rot=<0 0 0 1>}
 Hat_MESH {type=Curve, loc=<0 0 0>, rot=<0 0 0 1>}
 Null_MESH {type=Curve, loc=<0 0 0>, rot=<0 0 0 1>}
 Pose_0 {type=Curve, loc=<0 0 0>, rot=<0 0 0 1>}
 Hair_Mask {type=Curve, loc=<0 0 0>, rot=<0 0 0 1>}
 eye_convergence {type=Curve, loc=<0 0 0>, rot=<0 0 0 1>}
 headTransform {type=Curve, loc=<0 0 0>, rot=<0 0 0 1>}
 Pose_1 {type=Curve, loc=<0 0 0>, rot=<0 0 0 1>}
 Pose_2 {type=Curve, loc=<0 0 0>, rot=<0 0 0 1>}
 Pose_3 {type=Curve, loc=<0 0 0>, rot=<0 0 0 1>}
 Pose_4 {type=Curve, loc=<0 0 0>, rot=<0 0 0 1>}
 Pose_5 {type=Curve, loc=<0 0 0>, rot=<0 0 0 1>}
 Pose_6 {type=Curve, loc=<0 0 0>, rot=<0 0 0 1>}
 Pose_7 {type=Curve, loc=<0 0 0>, rot=<0 0 0 1>}
 Pose_8 {type=Curve, loc=<0 0 0>, rot=<0 0 0 1>}
 Pose_9 {type=Curve, loc=<0 0 0>, rot=<0 0 0 1>}
 Pose_10 {type=Curve, loc=<0 0 0>, rot=<0 0 0 1>}
 Pose_11 {type=Curve, loc=<0 0 0>, rot=<0 0 0 1>}
 Pose_12 {type=Curve, loc=<0 0 0>, rot=<0 0 0 1>}
 Pose_13 {type=Curve, loc=<0 0 0>, rot=<0 0 0 1>}
 Pose_14 {type=Curve, loc=<0 0 0>, rot=<0 0 0 1>}
 Pose_15 {type=Curve, loc=<0 0 0>, rot=<0 0 0 1>}
 Pose_16 {type=Curve, loc=<0 0 0>, rot=<0 0 0 1>}
 Pose_17 {type=Curve, loc=<0 0 0>, rot=<0 0 0 1>}
 Pose_18 {type=Curve, loc=<0 0 0>, rot=<0 0 0 1>}
 Pose_19 {type=Curve, loc=<0 0 0>, rot=<0 0 0 1>}
 Pose_20 {type=Curve, loc=<0 0 0>, rot=<0 0 0 1>}
 Pose_21 {type=Curve, loc=<0 0 0>, rot=<0 0 0 1>}
 Pose_22 {type=Curve, loc=<0 0 0>, rot=<0 0 0 1>}
 Pose_23 {type=Curve, loc=<0 0 0>, rot=<0 0 0 1>}
 Pose_24 {type=Curve, loc=<0 0 0>, rot=<0 0 0 1>}
 Pose_25 {type=Curve, loc=<0 0 0>, rot=<0 0 0 1>}
 Pose_26 {type=Curve, loc=<0 0 0>, rot=<0 0 0 1>}
 Pose_27 {type=Curve, loc=<0 0 0>, rot=<0 0 0 1>}
 Pose_28 {type=Curve, loc=<0 0 0>, rot=<0 0 0 1>}
 Pose_29 {type=Curve, loc=<0 0 0>, rot=<0 0 0 1>}
 Pose_30 {type=Curve, loc=<0 0 0>, rot=<0 0 0 1>}
 Pose_31 {type=Curve, loc=<0 0 0>, rot=<0 0 0 1>}
 Pose_32 {type=Curve, loc=<0 0 0>, rot=<0 0 0 1>}
 Pose_33 {type=Curve, loc=<0 0 0>, rot=<0 0 0 1>}
 Pose_34 {type=Curve, loc=<0 0 0>, rot=<0 0 0 1>}
 Pose_35 {type=Curve, loc=<0 0 0>, rot=<0 0 0 1>}
 Pose_36 {type=Curve, loc=<0 0 0>, rot=<0 0 0 1>}
 Pose_37 {type=Curve, loc=<0 0 0>, rot=<0 0 0 1>}
 Pose_38 {type=Curve, loc=<0 0 0>, rot=<0 0 0 1>}
 Pose_39 {type=Curve, loc=<0 0 0>, rot=<0 0 0 1>}
 Pose_40 {type=Curve, loc=<0 0 0>, rot=<0 0 0 1>}
 Pose_41 {type=Curve, loc=<0 0 0>, rot=<0 0 0 1>}
 Pose_42 {type=Curve, loc=<0 0 0>, rot=<0 0 0 1>}
 Pose_43 {type=Curve, loc=<0 0 0>, rot=<0 0 0 1>}
 Pose_44 {type=Curve, loc=<0 0 0>, rot=<0 0 0 1>}
 Pose_45 {type=Curve, loc=<0 0 0>, rot=<0 0 0 1>}
 Pose_46 {type=Curve, loc=<0 0 0>, rot=<0 0 0 1>}
 Pose_47 {type=Curve, loc=<0 0 0>, rot=<0 0 0 1>}
 Pose_48 {type=Curve, loc=<0 0 0>, rot=<0 0 0 1>}
 Pose_49 {type=Curve, loc=<0 0 0>, rot=<0 0 0 1>}
 Pose_50 {type=Curve, loc=<0 0 0>, rot=<0 0 0 1>}
 Pose_51 {type=Curve, loc=<0 0 0>, rot=<0 0 0 1>}
 Pose_52 {type=Curve, loc=<0 0 0>, rot=<0 0 0 1>}
 Pose_53 {type=Curve, loc=<0 0 0>, rot=<0 0 0 1>}
 Pose_54 {type=Curve, loc=<0 0 0>, rot=<0 0 0 1>}
 Pose_55 {type=Curve, loc=<0 0 0>, rot=<0 0 0 1>}
 Pose_56 {type=Curve, loc=<0 0 0>, rot=<0 0 0 1>}
 Pose_57 {type=Curve, loc=<0 0 0>, rot=<0 0 0 1>}
 Pose_58 {type=Curve, loc=<0 0 0>, rot=<0 0 0 1>}
 Pose_59 {type=Curve, loc=<0 0 0>, rot=<0 0 0 1>}
 Pose_60 {type=Curve, loc=<0 0 0>, rot=<0 0 0 1>}
 Pose_61 {type=Curve, loc=<0 0 0>, rot=<0 0 0 1>}
 Pose_62 {type=Curve, loc=<0 0 0>, rot=<0 0 0 1>}
 Pose_63 {type=Curve, loc=<0 0 0>, rot=<0 0 0 1>}
 Pose_64 {type=Curve, loc=<0 0 0>, rot=<0 0 0 1>}
 Pose_65 {type=Curve, loc=<0 0 0>, rot=<0 0 0 1>}
 Pose_66 {type=Curve, loc=<0 0 0>, rot=<0 0 0 1>}
 Pose_67 {type=Curve, loc=<0 0 0>, rot=<0 0 0 1>}
 Pose_68 {type=Curve, loc=<0 0 0>, rot=<0 0 0 1>}
 Pose_69 {type=Curve, loc=<0 0 0>, rot=<0 0 0 1>}
 eyeLookDownRight {type=Curve, loc=<0 0 0>, rot=<0 0 0 1>}
 eyeLookOutRight {type=Curve, loc=<0 0 0>, rot=<0 0 0 1>}
 eyeLookUpRight {type=Curve, loc=<0 0 0>, rot=<0 0 0 1>}
 eyeLookOutLeft {type=Curve, loc=<0 0 0>, rot=<0 0 0 1>}
 eyeBlinkLeft {type=Curve, loc=<0 0 0>, rot=<0 0 0 1>}
 eyeBlinkRight {type=Curve, loc=<0 0 0>, rot=<0 0 0 1>}
 eyeSquintLeft {type=Curve, loc=<0 0 0>, rot=<0 0 0 1>}
 eyeSquintRight {type=Curve, loc=<0 0 0>, rot=<0 0 0 1>}
 eyeWideLeft {type=Curve, loc=<0 0 0>, rot=<0 0 0 1>}
 eyeWideRight {type=Curve, loc=<0 0 0>, rot=<0 0 0 1>}
 jawForward {type=Curve, loc=<0 0 0>, rot=<0 0 0 1>}
 jawLeft {type=Curve, loc=<0 0 0>, rot=<0 0 0 1>}
 jawRight {type=Curve, loc=<0 0 0>, rot=<0 0 0 1>}
 jawOpen {type=Curve, loc=<0 0 0>, rot=<0 0 0 1>}
 mouthClose {type=Curve, loc=<0 0 0>, rot=<0 0 0 1>}
 mouthFunnel {type=Curve, loc=<0 0 0>, rot=<0 0 0 1>}
 mouthPucker {type=Curve, loc=<0 0 0>, rot=<0 0 0 1>}
 mouthLeft {type=Curve, loc=<0 0 0>, rot=<0 0 0 1>}
 mouthRight {type=Curve, loc=<0 0 0>, rot=<0 0 0 1>}
 mouthSmileLeft {type=Curve, loc=<0 0 0>, rot=<0 0 0 1>}
 mouthSmileRight {type=Curve, loc=<0 0 0>, rot=<0 0 0 1>}
 mouthFrownLeft {type=Curve, loc=<0 0 0>, rot=<0 0 0 1>}
 mouthFrownRight {type=Curve, loc=<0 0 0>, rot=<0 0 0 1>}
 mouthDimpleLeft {type=Curve, loc=<0 0 0>, rot=<0 0 0 1>}
 mouthDimpleRight {type=Curve, loc=<0 0 0>, rot=<0 0 0 1>}
 mouthStretchLeft {type=Curve, loc=<0 0 0>, rot=<0 0 0 1>}
 mouthStretchRight {type=Curve, loc=<0 0 0>, rot=<0 0 0 1>}
 mouthRollLower {type=Curve, loc=<0 0 0>, rot=<0 0 0 1>}
 mouthRollUpper {type=Curve, loc=<0 0 0>, rot=<0 0 0 1>}
 mouthShrugLower {type=Curve, loc=<0 0 0>, rot=<0 0 0 1>}
 mouthShrugUpper {type=Curve, loc=<0 0 0>, rot=<0 0 0 1>}
 mouthPressLeft {type=Curve, loc=<0 0 0>, rot=<0 0 0 1>}
 mouthPressRight {type=Curve, loc=<0 0 0>, rot=<0 0 0 1>}
 mouthLowerDownLeft {type=Curve, loc=<0 0 0>, rot=<0 0 0 1>}
 mouthLowerDownRight {type=Curve, loc=<0 0 0>, rot=<0 0 0 1>}
 mouthUpperUpLeft {type=Curve, loc=<0 0 0>, rot=<0 0 0 1>}
 mouthUpperUpRight {type=Curve, loc=<0 0 0>, rot=<0 0 0 1>}
 browDownLeft {type=Curve, loc=<0 0 0>, rot=<0 0 0 1>}
 browDownRight {type=Curve, loc=<0 0 0>, rot=<0 0 0 1>}
 browInnerUp {type=Curve, loc=<0 0 0>, rot=<0 0 0 1>}
 browOuterUpLeft {type=Curve, loc=<0 0 0>, rot=<0 0 0 1>}
 browOuterUpRight {type=Curve, loc=<0 0 0>, rot=<0 0 0 1>}
 cheekPuff {type=Curve, loc=<0 0 0>, rot=<0 0 0 1>}
 cheekSquintLeft {type=Curve, loc=<0 0 0>, rot=<0 0 0 1>}
 cheekSquintRight {type=Curve, loc=<0 0 0>, rot=<0 0 0 1>}
 noseSneerLeft {type=Curve, loc=<0 0 0>, rot=<0 0 0 1>}
 noseSneerRight {type=Curve, loc=<0 0 0>, rot=<0 0 0 1>}
 C_glabella_down {type=Curve, loc=<0 0 0>, rot=<0 0 0 1>}
 L_brow_squeeze_in {type=Curve, loc=<0 0 0>, rot=<0 0 0 1>}
 R_brow_squeeze_in {type=Curve, loc=<0 0 0>, rot=<0 0 0 1>}
 L_upper_lip_lower {type=Curve, loc=<0 0 0>, rot=<0 0 0 1>}
 R_upper_lip_lower {type=Curve, loc=<0 0 0>, rot=<0 0 0 1>}
 L_lower_lip_up {type=Curve, loc=<0 0 0>, rot=<0 0 0 1>}
 R_lower_lip_up {type=Curve, loc=<0 0 0>, rot=<0 0 0 1>}
 L_lip_corner_narrow {type=Curve, loc=<0 0 0>, rot=<0 0 0 1>}
 R_lip_corner_narrow {type=Curve, loc=<0 0 0>, rot=<0 0 0 1>}
 phoneme_ch {type=Curve, loc=<0 0 0>, rot=<0 0 0 1>}
 phoneme_mbp {type=Curve, loc=<0 0 0>, rot=<0 0 0 1>}
 phoneme_fv {type=Curve, loc=<0 0 0>, rot=<0 0 0 1>}
 mouth_up {type=Curve, loc=<0 0 0>, rot=<0 0 0 1>}
 mouth_down {type=Curve, loc=<0 0 0>, rot=<0 0 0 1>}
 L_cheek_up {type=Curve, loc=<0 0 0>, rot=<0 0 0 1>}
 R_cheek_up {type=Curve, loc=<0 0 0>, rot=<0 0 0 1>}
 L_upperLidDown {type=Curve, loc=<0 0 0>, rot=<0 0 0 1>}
 R_upperLidDown {type=Curve, loc=<0 0 0>, rot=<0 0 0 1>}
 jaw_up {type=Curve, loc=<0 0 0>, rot=<0 0 0 1>}
 jaw_back {type=Curve, loc=<0 0 0>, rot=<0 0 0 1>}
 tongue_forward {type=Curve, loc=<0 0 0>, rot=<0 0 0 1>}
 tongue_back {type=Curve, loc=<0 0 0>, rot=<0 0 0 1>}
 tongue_up {type=Curve, loc=<0 0 0>, rot=<0 0 0 1>}
 tongue_down {type=Curve, loc=<0 0 0>, rot=<0 0 0 1>}
 basePose {type=Curve, loc=<0 0 0>, rot=<0 0 0 1>}
 neutral {type=Curve, loc=<0 0 0>, rot=<0 0 0 1>}
 cheek_puff_low {type=Curve, loc=<0 0 0>, rot=<0 0 0 1>}
 Desert_Island_Jonesy_Mouth {type=Curve, loc=<0 0 0>, rot=<0 0 0 1>}
 HT_L1_TransitionCurve {type=Curve, loc=<0 0 0>, rot=<0 0 0 1>}
 Hair_CapShape {type=Curve, loc=<0 0 0>, rot=<0 0 0 1>}
 Hair_CapShape1 {type=Curve, loc=<0 0 0>, rot=<0 0 0 1>}
 jekyll {type=Curve, loc=<0 0 0>, rot=<0 0 0 1>}
 JH_TransformCurve {type=Curve, loc=<0 0 0>, rot=<0 0 0 1>}
 eyes_LOD0 {type=Curve, loc=<0 0 0>, rot=<0 0 0 1>}
 MaterialCurveA {type=Curve, loc=<0 0 0>, rot=<0 0 0 1>}
 MaterialCurveB {type=Curve, loc=<0 0 0>, rot=<0 0 0 1>}
 MaterialCurveC {type=Curve, loc=<0 0 0>, rot=<0 0 0 1>}
 MaterialCurveD {type=Curve, loc=<0 0 0>, rot=<0 0 0 1>}
 MaterialCurveE {type=Curve, loc=<0 0 0>, rot=<0 0 0 1>}
 Custom_Frontend_Face_Idle {type=Curve, loc=<0 0 0>, rot=<0 0 0 1>}
 blendParent1 {type=Curve, loc=<0 0 0>, rot=<0 0 0 1>}
 BaseHeadDefaultFaceIdle {type=Curve, loc=<0 0 0>, rot=<0 0 0 1>}
 VeiledTransitionGlowAmount {type=Curve, loc=<0 0 0>, rot=<0 0 0 1>}
 is_3L {type=Curve, loc=<0 0 0>, rot=<0 0 0 1>}
 Hair_Hat_LOD0 {type=Curve, loc=<0 0 0>, rot=<0 0 0 1>}
 Hair_Helmet_LOD0 {type=Curve, loc=<0 0 0>, rot=<0 0 0 1>}
 Hair_Null_LOD0 {type=Curve, loc=<0 0 0>, rot=<0 0 0 1>}
 mouth_close_pose {type=Curve, loc=<0 0 0>, rot=<0 0 0 1>}
 mouth_pucker_pose {type=Curve, loc=<0 0 0>, rot=<0 0 0 1>}
 mouth_left_pose {type=Curve, loc=<0 0 0>, rot=<0 0 0 1>}
 mouth_right_pose {type=Curve, loc=<0 0 0>, rot=<0 0 0 1>}
 L_mouth_stretch_pose {type=Curve, loc=<0 0 0>, rot=<0 0 0 1>}
 R_mouth_stretch_pose {type=Curve, loc=<0 0 0>, rot=<0 0 0 1>}
 mouth_roll_lower_pose {type=Curve, loc=<0 0 0>, rot=<0 0 0 1>}
 mouth_roll_upper_pose {type=Curve, loc=<0 0 0>, rot=<0 0 0 1>}
 mouth_shrug_lower_pose {type=Curve, loc=<0 0 0>, rot=<0 0 0 1>}
 mouth_shrug_upper_pose {type=Curve, loc=<0 0 0>, rot=<0 0 0 1>}
 L_mouth_press_pose {type=Curve, loc=<0 0 0>, rot=<0 0 0 1>}
 R_mouth_press_pose {type=Curve, loc=<0 0 0>, rot=<0 0 0 1>}
 cheek_puff_pose {type=Curve, loc=<0 0 0>, rot=<0 0 0 1>}
 L_nose_wrinkler_pose {type=Curve, loc=<0 0 0>, rot=<0 0 0 1>}
 R_nose_wrinkler_pose {type=Curve, loc=<0 0 0>, rot=<0 0 0 1>}
 base {type=Curve, loc=<0 0 0>, rot=<0 0 0 1>}
 eyeLookDownLeft {type=Curve, loc=<0 0 0>, rot=<0 0 0 1>}
 eyeLookInLeft {type=Curve, loc=<0 0 0>, rot=<0 0 0 1>}
 eyeLookUpLeft {type=Curve, loc=<0 0 0>, rot=<0 0 0 1>}
 eyeLookInRight {type=Curve, loc=<0 0 0>, rot=<0 0 0 1>}
 is_face_bone_anim {type=Curve, loc=<0 0 0>, rot=<0 0 0 1>}
 FullHourglass {type=Curve, loc=<0 0 0>, rot=<0 0 0 1>}
 LeanRectangle {type=Curve, loc=<0 0 0>, rot=<0 0 0 1>}
 InvertedTriangle {type=Curve, loc=<0 0 0>, rot=<0 0 0 1>}
 L_eye {type=Curve, loc=<0 0 0>, rot=<0 0 0 1>}
 R_eye {type=Curve, loc=<0 0 0>, rot=<0 0 0 1>}
 head {type=Curve, loc=<0 0 0>, rot=<0 0 0 1>}
 dyn_ear_l_master {type=Curve, loc=<0 0 0>, rot=<0 0 0 1>}
 root {type=Curve, loc=<0 0 0>, rot=<0 0 0 1>}
 dyn_ear02_l {type=Curve, loc=<0 0 0>, rot=<0 0 0 1>}
 dyn_ear03_l {type=Curve, loc=<0 0 0>, rot=<0 0 0 1>}
 dyn_ear04_l {type=Curve, loc=<0 0 0>, rot=<0 0 0 1>}
 dyn_ear05_l {type=Curve, loc=<0 0 0>, rot=<0 0 0 1>}
 dyn_ear01_l {type=Curve, loc=<0 0 0>, rot=<0 0 0 1>}
 flipbook_face_L_eye_index {type=Curve, loc=<0 0 0>, rot=<0 0 0 1>}
 flipbook_face_R_eye_index {type=Curve, loc=<0 0 0>, rot=<0 0 0 1>}
 flipbook_face_L_brow_index {type=Curve, loc=<0 0 0>, rot=<0 0 0 1>}
 flipbook_face_R_brow_index {type=Curve, loc=<0 0 0>, rot=<0 0 0 1>}
 flipbook_face_mouth_index {type=Curve, loc=<0 0 0>, rot=<0 0 0 1>}
 flipbook_face_bypass_mapping {type=Curve, loc=<0 0 0>, rot=<0 0 0 1>}
 flipbook_face_is_idling {type=Curve, loc=<0 0 0>, rot=<0 0 0 1>}
]
