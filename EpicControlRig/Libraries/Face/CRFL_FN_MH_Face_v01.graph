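# Control Rig function library for the FN MetaHuman face, with forward and backward functions mapping facial board controls to CTRL_expressions curves and back.
# Shared float/vector helpers do the mapping in both directions, mirroring left-side entries to the right.
# Forward also handles eye aim and head-follow blending.
# Both set DisableFaceOverride and is_3L, sync WM_ offset curves with CTRL_WrinkleMap channels, call tweaker functions, and gate extra shapes on Is V2.
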ControlRigBlueprint /EpicControlRig/Libraries/Face/CRFL_FN_MH_Face_v01
graphs [
 CRFL FN MH Backward v01_ContainedGraph {nodes [
   n1 = RigVMFunctionEntryNode()
   RigVMFunction_Sequence = &RigVMFunction_Sequence(ExecuteContext=@n1.ExecuteContext)
   /Script/ControlRig.RigUnit_SetCurveValue(Curve=DisableFaceOverride, Value=1, ExecutePin=@RigVMFunction_Sequence.A) {is=struct RigVMUnitNode}
   > &RigUnit_SetCurveValue(Curve=is_3L, Value=1)
   > SequenceAggregate_1 = RigVMAggregateNode()
   FN_FloatCtrlToCurve_FNC_7 = RigVMFunctionReferenceNode(
    Ctrl_L_Token=CTRL_L_
    Ctrl_R_Token=CTRL_R_
    Curve_Side_Token_Last=true
    Pos_Neg_Curves=((Type=Curve,Name="CTRL_expressions_eyeBlinkL"),(Type=Curve,Name="CTRL_expressions_eyeWidenL"))
    backward=true
    control=(Type=Control,Name="CTRL_L_eye_blink")
    mirror=true
    vector_axis=X
    ExecuteContext=@SequenceAggregate_1.B
   )
   EyeAim_Inverse_CollapseNode = RigVMCollapseNode(ExecutePin=@FN_FloatCtrlToCurve_FNC_7.ExecuteContext)
   FN_VectorCtrlToCurve_FNC_14 = RigVMFunctionReferenceNode(
    Backward=true
    Control=(Type=Control,Name="CTRL_L_eye")
    Ctrl_L_Token=CTRL_L_
    Ctrl_R_Token=CTRL_R_
    Curve_Side_Token_Last=true
    X_Mirror=true
    X_Pos_Neg_Expr=((Type=Curve,Name="CTRL_expressions_eyeLookLeftL"),(Type=Curve,Name="CTRL_expressions_eyeLookRightL"))
    Y_Mirror=true
    Y_Pos_Neg_Expr=((Type=Curve,Name="CTRL_expressions_eyeLookUpL"),(Type=Curve,Name="CTRL_expressions_eyeLookDownL"))
    Z_Mirror=false
    Z_Pos_Neg_Expr=()
    ExecuteContext=@EyeAim_Inverse_CollapseNode.ExecutePin
   )
   > FN_VectorCtrlToCurve_FNC_4 = RigVMFunctionReferenceNode(
    Backward=true
    Control=(Type=Control,Name="CTRL_L_nose")
    Ctrl_L_Token=CTRL_L_
    Ctrl_R_Token=CTRL_R_
    Curve_Side_Token_Last=true
    X_Mirror=true
    X_Pos_Neg_Expr=((Type=Curve,Name="CTRL_expressions_noseNostrilDilateL"),(Type=Curve,Name="CTRL_expressions_noseNostrilCompressL"))
    Y_Mirror=true
    Y_Pos_Neg_Expr=((Type=Curve,Name="CTRL_expressions_noseWrinkleL"),(Type=Curve,Name="CTRL_expressions_noseNostrilDepressL"))
    Z_Mirror=false
    Z_Pos_Neg_Expr=()
   )
   > FN_FloatCtrlToCurve_FNC_8 = RigVMFunctionReferenceNode(
    Ctrl_L_Token=CTRL_L_
    Ctrl_R_Token=CTRL_R_
    Curve_Side_Token_Last=true
    Pos_Neg_Curves=((Type=Curve,Name="CTRL_expressions_noseWrinkleUpperL"))
    backward=true
    control=(Type=Control,Name="CTRL_L_nose_wrinkleUpper")
    mirror=true
    vector_axis=X
   )
   > FN_FloatCtrlToCurve_FNC_9 = RigVMFunctionReferenceNode(
    Ctrl_L_Token=CTRL_L_
    Ctrl_R_Token=CTRL_R_
    Curve_Side_Token_Last=true
    Pos_Neg_Curves=((Type=Curve,Name="CTRL_expressions_eyeSquintInnerL"))
    backward=true
    control=(Type=Control,Name="CTRL_L_eye_squintInner")
    mirror=true
    vector_axis=X
   )
   > FN_FloatCtrlToCurve_FNC_10 = RigVMFunctionReferenceNode(
    Ctrl_L_Token=CTRL_L_
    Ctrl_R_Token=CTRL_R_
    Curve_Side_Token_Last=true
    Pos_Neg_Curves=((Type=Curve,Name="CTRL_expressions_eyeCheekRaiseL"))
    backward=true
    control=(Type=Control,Name="CTRL_L_eye_cheekRaise")
    mirror=true
    vector_axis=X
   )
   > FN_FloatCtrlToCurve_FNC_11 = RigVMFunctionReferenceNode(
    Ctrl_L_Token=CTRL_L_
    Ctrl_R_Token=CTRL_R_
    Curve_Side_Token_Last=true
    Pos_Neg_Curves=((Type=Curve,Name="CTRL_expressions_mouthCheekBlowL"),(Type=Curve,Name="CTRL_expressions_mouthCheekSuckL"))
    backward=true
    control=(Type=Control,Name="CTRL_L_mouth_suckBlow")
    mirror=true
    vector_axis=X
   )
   > FN_FloatCtrlToCurve_FNC_12 = RigVMFunctionReferenceNode(
    Ctrl_L_Token=CTRL_L_
    Ctrl_R_Token=CTRL_R_
    Curve_Side_Token_Last=true
    Pos_Neg_Curves=((Type=Curve,Name="CTRL_expressions_mouthLipsBlowL"))
    backward=true
    control=(Type=Control,Name="CTRL_L_mouth_lipsBlow")
    mirror=true
    vector_axis=X
   )
   FN_FloatCtrlToCurve_FNC_13 = RigVMFunctionReferenceNode(
    Ctrl_L_Token=CTRL_L_
    Ctrl_R_Token=CTRL_R_
    Curve_Side_Token_Last=true
    Pos_Neg_Curves=((Type=Curve,Name="CTRL_expressions_mouthUpperLipRaiseL"))
    backward=true
    control=(Type=Control,Name="CTRL_L_mouth_upperLipRaise")
    mirror=true
    vector_axis=X
    ExecuteContext=@SequenceAggregate_1.C
   )
   > FN_FloatCtrlToCurve_FNC_14 = RigVMFunctionReferenceNode(
    Ctrl_L_Token=CTRL_L_
    Ctrl_R_Token=CTRL_R_
    Curve_Side_Token_Last=true
    Pos_Neg_Curves=((Type=Curve,Name="CTRL_expressions_mouthCornerPullL"))
    backward=true
    control=(Type=Control,Name="CTRL_L_mouth_cornerPull")
    mirror=true
    vector_axis=X
   )
   > FN_FloatCtrlToCurve_FNC_15 = RigVMFunctionReferenceNode(
    Ctrl_L_Token=CTRL_L_
    Ctrl_R_Token=CTRL_R_
    Curve_Side_Token_Last=true
    Pos_Neg_Curves=((Type=Curve,Name="CTRL_expressions_mouthDimpleL"))
    backward=true
    control=(Type=Control,Name="CTRL_L_mouth_dimple")
    mirror=true
    vector_axis=X
   )
   > FN_FloatCtrlToCurve_FNC_16 = RigVMFunctionReferenceNode(
    Ctrl_L_Token=CTRL_L_
    Ctrl_R_Token=CTRL_R_
    Curve_Side_Token_Last=true
    Pos_Neg_Curves=((Type=Curve,Name="CTRL_expressions_mouthCornerDepressL"))
    backward=true
    control=(Type=Control,Name="CTRL_L_mouth_cornerDepress")
    mirror=true
    vector_axis=X
   )
   > FN_FloatCtrlToCurve_FNC_17 = RigVMFunctionReferenceNode(
    Ctrl_L_Token=CTRL_L_
    Ctrl_R_Token=CTRL_R_
    Curve_Side_Token_Last=true
    Pos_Neg_Curves=((Type=Curve,Name="CTRL_expressions_mouthStretchL"))
    backward=true
    control=(Type=Control,Name="CTRL_L_mouth_stretch")
    mirror=true
    vector_axis=X
   )
   > FN_FloatCtrlToCurve_FNC_18 = RigVMFunctionReferenceNode(
    Ctrl_L_Token=CTRL_L_
    Ctrl_R_Token=CTRL_R_
    Curve_Side_Token_Last=true
    Pos_Neg_Curves=((Type=Curve,Name="CTRL_expressions_mouthLowerLipDepressL"))
    backward=true
    control=(Type=Control,Name="CTRL_L_mouth_lowerLipDepress")
    mirror=true
    vector_axis=X
   )
   > FN_VectorCtrlToCurve_FNC_6 = RigVMFunctionReferenceNode(
    Backward=true
    Control=(Type=Control,Name="CTRL_C_mouth")
    Ctrl_L_Token=CTRL_L_
    Ctrl_R_Token=CTRL_R_
    Curve_Side_Token_Last=true
    X_Mirror=false
    X_Pos_Neg_Expr=((Type=Curve,Name="CTRL_expressions_mouthLeft"),(Type=Curve,Name="CTRL_expressions_mouthRight"))
    Y_Mirror=false
    Y_Pos_Neg_Expr=((Type=Curve,Name="CTRL_expressions_mouthUp"),(Type=Curve,Name="CTRL_expressions_mouthDown"))
    Z_Mirror=false
    Z_Pos_Neg_Expr=()
   )
   > FN_FloatCtrlToCurve_FNC_31 = RigVMFunctionReferenceNode(
    Ctrl_L_Token=CTRL_L_
    Ctrl_R_Token=CTRL_R_
    Curve_Side_Token_Last=true
    Pos_Neg_Curves=((Type=Curve,Name="CTRL_expressions_tongueIn"),(Type=Curve,Name="CTRL_expressions_tongueOut"))
    backward=true
    control=(Type=Control,Name="CTRL_C_tongue_inOut")
    mirror=false
    vector_axis=X
   )
   > FN_VectorCtrlToCurve_FNC_11 = RigVMFunctionReferenceNode(
    Backward=true
    Control=(Type=Control,Name="CTRL_C_tongue")
    Ctrl_L_Token=CTRL_L_
    Ctrl_R_Token=CTRL_R_
    Curve_Side_Token_Last=true
    X_Mirror=false
    X_Pos_Neg_Expr=((Type=Curve,Name="CTRL_expressions_tongueLeft"),(Type=Curve,Name="CTRL_expressions_tongueRight"))
    Y_Mirror=false
    Y_Pos_Neg_Expr=((Type=Curve,Name="CTRL_expressions_tongueUp"),(Type=Curve,Name="CTRL_expressions_tongueDown"))
    Z_Mirror=false
    Z_Pos_Neg_Expr=()
   )
   > FN_VectorCtrlToCurve_FNC_12 = RigVMFunctionReferenceNode(
    Backward=true
    Control=(Type=Control,Name="CTRL_C_tongue_roll")
    Ctrl_L_Token=CTRL_L_
    Ctrl_R_Token=CTRL_R_
    Curve_Side_Token_Last=true
    X_Mirror=false
    X_Pos_Neg_Expr=((Type=Curve,Name="CTRL_expressions_tongueRollLeft"),(Type=Curve,Name="CTRL_expressions_tongueRollRight"))
    Y_Mirror=false
    Y_Pos_Neg_Expr=((Type=Curve,Name="CTRL_expressions_tongueRollUp"),(Type=Curve,Name="CTRL_expressions_tongueRollDown"))
    Z_Mirror=false
    Z_Pos_Neg_Expr=()
   )
   > FN_FloatCtrlToCurve_FNC_32 = RigVMFunctionReferenceNode(
    Ctrl_L_Token=CTRL_L_
    Ctrl_R_Token=CTRL_R_
    Curve_Side_Token_Last=true
    Pos_Neg_Curves=((Type=Curve,Name="CTRL_expressions_tongueWide"),(Type=Curve,Name="CTRL_expressions_tongueNarrow"))
    backward=true
    control=(Type=Control,Name="CTRL_C_tongue_narrowWide")
    mirror=false
    vector_axis=X
   )
   FN_FloatCtrlToCurve_FNC_33 = RigVMFunctionReferenceNode(
    Ctrl_L_Token=CTRL_L_
    Ctrl_R_Token=CTRL_R_
    Curve_Side_Token_Last=true
    Pos_Neg_Curves=((Type=Curve,Name="CTRL_expressions_mouthLipsTowardsUL"))
    backward=true
    control=(Type=Control,Name="CTRL_L_mouth_towardsU")
    mirror=true
    vector_axis=X
    ExecuteContext=@SequenceAggregate_1.D
   )
   > FN_FloatCtrlToCurve_FNC_34 = RigVMFunctionReferenceNode(
    Ctrl_L_Token=CTRL_L_
    Ctrl_R_Token=CTRL_R_
    Curve_Side_Token_Last=true
    Pos_Neg_Curves=((Type=Curve,Name="CTRL_expressions_mouthLipsTowardsDL"))
    backward=true
    control=(Type=Control,Name="CTRL_L_mouth_towardsD")
    mirror=true
    vector_axis=X
   )
   > FN_FloatCtrlToCurve_FNC_35 = RigVMFunctionReferenceNode(
    Ctrl_L_Token=CTRL_L_
    Ctrl_R_Token=CTRL_R_
    Curve_Side_Token_Last=true
    Pos_Neg_Curves=((Type=Curve,Name="CTRL_expressions_mouthLipsPurseUL"))
    backward=true
    control=(Type=Control,Name="CTRL_L_mouth_purseU")
    mirror=true
    vector_axis=X
   )
   > FN_FloatCtrlToCurve_FNC_36 = RigVMFunctionReferenceNode(
    Ctrl_L_Token=CTRL_L_
    Ctrl_R_Token=CTRL_R_
    Curve_Side_Token_Last=true
    Pos_Neg_Curves=((Type=Curve,Name="CTRL_expressions_mouthLipsPurseDL"))
    backward=true
    control=(Type=Control,Name="CTRL_L_mouth_purseD")
    mirror=true
    vector_axis=X
   )
   > FN_FloatCtrlToCurve_FNC_37 = RigVMFunctionReferenceNode(
    Ctrl_L_Token=CTRL_L_
    Ctrl_R_Token=CTRL_R_
    Curve_Side_Token_Last=true
    Pos_Neg_Curves=((Type=Curve,Name="CTRL_expressions_mouthFunnelUL"))
    backward=true
    control=(Type=Control,Name="CTRL_L_mouth_funnelU")
    mirror=true
    vector_axis=X
   )
   > FN_FloatCtrlToCurve_FNC_38 = RigVMFunctionReferenceNode(
    Ctrl_L_Token=CTRL_L_
    Ctrl_R_Token=CTRL_R_
    Curve_Side_Token_Last=true
    Pos_Neg_Curves=((Type=Curve,Name="CTRL_expressions_mouthFunnelDL"))
    backward=true
    control=(Type=Control,Name="CTRL_L_mouth_funnelD")
    mirror=true
    vector_axis=X
   )
   > FN_FloatCtrlToCurve_FNC_19 = RigVMFunctionReferenceNode(
    Ctrl_L_Token=CTRL_L_
    Ctrl_R_Token=CTRL_R_
    Curve_Side_Token_Last=true
    Pos_Neg_Curves=((Type=Curve,Name="CTRL_expressions_jawChinRaiseUL"))
    backward=true
    control=(Type=Control,Name="CTRL_L_jaw_ChinRaiseU")
    mirror=true
    vector_axis=X
   )
   > FN_FloatCtrlToCurve_FNC_20 = RigVMFunctionReferenceNode(
    Ctrl_L_Token=CTRL_L_
    Ctrl_R_Token=CTRL_R_
    Curve_Side_Token_Last=true
    Pos_Neg_Curves=((Type=Curve,Name="CTRL_expressions_jawChinRaiseDL"))
    backward=true
    control=(Type=Control,Name="CTRL_L_jaw_ChinRaiseD")
    mirror=true
    vector_axis=X
   )
   > FN_VectorCtrlToCurve_FNC_7 = RigVMFunctionReferenceNode(
    Backward=true
    Control=(Type=Control,Name="CTRL_C_jaw")
    Ctrl_L_Token=CTRL_L_
    Ctrl_R_Token=CTRL_R_
    Curve_Side_Token_Last=true
    X_Mirror=false
    X_Pos_Neg_Expr=((Type=Curve,Name="CTRL_expressions_jawRight"),(Type=Curve,Name="CTRL_expressions_jawLeft"))
    Y_Mirror=false
    Y_Pos_Neg_Expr=((Type=Curve,Name="CTRL_expressions_jawOpen"))
    Z_Mirror=false
    Z_Pos_Neg_Expr=()
   )
   > FN_FloatCtrlToCurve_FNC_22 = RigVMFunctionReferenceNode(
    Ctrl_L_Token=CTRL_L_
    Ctrl_R_Token=CTRL_R_
    Curve_Side_Token_Last=true
    Pos_Neg_Curves=((Type=Curve,Name="CTRL_expressions_jawOpenExtreme"))
    backward=true
    control=(Type=Control,Name="CTRL_C_jaw_openExtreme")
    mirror=false
    vector_axis=X
   )
   > FN_FloatCtrlToCurve_FNC_21 = RigVMFunctionReferenceNode(
    Ctrl_L_Token=CTRL_L_
    Ctrl_R_Token=CTRL_R_
    Curve_Side_Token_Last=true
    Pos_Neg_Curves=((Type=Curve,Name="CTRL_expressions_jawBack"),(Type=Curve,Name="CTRL_expressions_jawFwd"))
    backward=true
    control=(Type=Control,Name="CTRL_C_jaw_fwdBack")
    mirror=false
    vector_axis=X
   )
   > FN_FloatCtrlToCurve_FNC_41 = RigVMFunctionReferenceNode(
    Ctrl_L_Token=CTRL_L_
    Ctrl_R_Token=CTRL_R_
    Curve_Side_Token_Last=true
    Pos_Neg_Curves=((Type=Curve,Name="CTRL_expressions_mouthLipsTogetherUL"))
    backward=true
    control=(Type=Control,Name="CTRL_L_mouth_lipsTogetherU")
    mirror=true
    vector_axis=X
   )
   > FN_FloatCtrlToCurve_FNC_42 = RigVMFunctionReferenceNode(
    Ctrl_L_Token=CTRL_L_
    Ctrl_R_Token=CTRL_R_
    Curve_Side_Token_Last=true
    Pos_Neg_Curves=((Type=Curve,Name="CTRL_expressions_mouthLipsTogetherDL"))
    backward=true
    control=(Type=Control,Name="CTRL_L_mouth_lipsTogetherD")
    mirror=true
    vector_axis=X
   )
   > FN_FloatCtrlToCurve_FNC_43 = RigVMFunctionReferenceNode(
    Ctrl_L_Token=CTRL_L_
    Ctrl_R_Token=CTRL_R_
    Curve_Side_Token_Last=true
    Pos_Neg_Curves=((Type=Curve,Name="CTRL_expressions_mouthPressUL"))
    backward=true
    control=(Type=Control,Name="CTRL_L_mouth_pressU")
    mirror=true
    vector_axis=X
   )
   > FN_FloatCtrlToCurve_FNC_44 = RigVMFunctionReferenceNode(
    Ctrl_L_Token=CTRL_L_
    Ctrl_R_Token=CTRL_R_
    Curve_Side_Token_Last=true
    Pos_Neg_Curves=((Type=Curve,Name="CTRL_expressions_mouthPressDL"))
    backward=true
    control=(Type=Control,Name="CTRL_L_mouth_pressD")
    mirror=true
    vector_axis=X
   )
   > FN_FloatCtrlToCurve_FNC_39 = RigVMFunctionReferenceNode(
    Ctrl_L_Token=CTRL_L_
    Ctrl_R_Token=CTRL_R_
    Curve_Side_Token_Last=true
    Pos_Neg_Curves=((Type=Curve,Name="CTRL_expressions_mouthUpperLipBiteL"))
    backward=true
    control=(Type=Control,Name="CTRL_L_mouth_lipBiteU")
    mirror=true
    vector_axis=X
   )
   > FN_FloatCtrlToCurve_FNC_40 = RigVMFunctionReferenceNode(
    Ctrl_L_Token=CTRL_L_
    Ctrl_R_Token=CTRL_R_
    Curve_Side_Token_Last=true
    Pos_Neg_Curves=((Type=Curve,Name="CTRL_expressions_mouthLowerLipBiteL"))
    backward=true
    control=(Type=Control,Name="CTRL_L_mouth_lipBiteD")
    mirror=true
    vector_axis=X
   )
   n2 = &RigVMFunction_Sequence(ExecuteContext=@SequenceAggregate_1.F)
   For_Each = RigVMDispatchNode(
    Array=/Script/ControlRig.RigUnit_CollectionNameSearchArray(PartialName=_Offset_, TypeToSearch=Curve) {is=struct RigVMUnitNode}.Items
    ExecuteContext=@n2.A
   ) {template="DISPATCH_RigVMDispatch_ArrayIterator(in Array,out Element,out Index,out Count,out Ratio)"}
   StartsWith = /Script/RigVM.RigVMFunction_StartsWith(Start=WM_, Name=@For_Each.Element.Name) {is=struct RigVMUnitNode, template="StartsWith::Execute(in Name,in Start,out Result)"}
   n3 = &RigVMFunction_ControlFlowBranch(ExecuteContext=@For_Each.ExecuteContext, Condition=@StartsWith.Result)
   SetAnimationChannel = /Script/ControlRig.RigUnit_SetFloatAnimationChannel(
    Control=CTRL_WrinkleMap
    bInitial=False
    Value=/Script/ControlRig.RigUnit_GetCurveValue(Curve=@For_Each.Element.Name) {is=struct RigVMUnitNode}.Value
    ExecutePin=@n3.True
    Channel=@For_Each.Element.Name
   ) {is=struct RigVMUnitNode, template="SetAnimationChannel::Execute(in Value,in Control,in Channel,in bInitial)"}
   CRFL FN MH Tweak Backward v01 = RigVMFunctionReferenceNode(ExecuteContext=@n2.B, Is V2=RigVMVariableNode() {variable=Is V2}.Value)
   FN_FloatCtrlToCurve_FNC = RigVMFunctionReferenceNode(
    Ctrl_L_Token=CTRL_L_
    Ctrl_R_Token=CTRL_R_
    Curve_Side_Token_Last=true
    Pos_Neg_Curves=((Type=Curve,Name="CTRL_expressions_browRaiseInL"))
    backward=true
    control=(Type=Control,Name="CTRL_L_brow_raiseIn")
    mirror=true
    vector_axis=X
    ExecuteContext=@SequenceAggregate_1.A
   )
   > FN_FloatCtrlToCurve_FNC_1 = RigVMFunctionReferenceNode(
    Ctrl_L_Token=CTRL_L_
    Ctrl_R_Token=CTRL_R_
    Curve_Side_Token_Last=true
    Pos_Neg_Curves=((Type=Curve,Name="CTRL_expressions_browRaiseOuterL"))
    backward=true
    control=(Type=Control,Name="CTRL_L_brow_raiseOut")
    mirror=true
    vector_axis=X
   )
   > FN_FloatCtrlToCurve_FNC_2 = RigVMFunctionReferenceNode(
    Ctrl_L_Token=CTRL_L_
    Ctrl_R_Token=CTRL_R_
    Curve_Side_Token_Last=true
    Pos_Neg_Curves=((Type=Curve,Name="CTRL_expressions_browDownL"))
    backward=true
    control=(Type=Control,Name="CTRL_L_brow_down")
    mirror=true
    vector_axis=X
   )
   > FN_FloatCtrlToCurve_FNC_3 = RigVMFunctionReferenceNode(
    Ctrl_L_Token=CTRL_L_
    Ctrl_R_Token=CTRL_R_
    Curve_Side_Token_Last=true
    Pos_Neg_Curves=((Type=Curve,Name="CTRL_expressions_browLateralL"))
    backward=true
    control=(Type=Control,Name="CTRL_L_brow_lateral")
    mirror=true
    vector_axis=X
   )
   n4 = &RigVMFunction_ControlFlowBranch(Condition=RigVMVariableNode() {variable=Is V2}.Value, ExecuteContext=@SequenceAggregate_1.E)
   FN_FloatCtrlToCurve_FNC_45 = RigVMFunctionReferenceNode(
    Ctrl_L_Token=CTRL_L_
    Ctrl_R_Token=CTRL_R_
    Curve_Side_Token_Last=true
    Pos_Neg_Curves=((Type=Curve,Name="CTRL_expressions_eyePupilWideL"),(Type=Curve,Name="CTRL_expressions_eyePupilNarrowL"))
    backward=true
    control=(Type=Control,Name="CTRL_L_eye_pupil")
    mirror=true
    vector_axis=X
    ExecuteContext=@n4.True
   )
   > FN_FloatCtrlToCurve_FNC_46 = RigVMFunctionReferenceNode(
    Ctrl_L_Token=CTRL_L_
    Ctrl_R_Token=CTRL_R_
    Curve_Side_Token_Last=true
    Pos_Neg_Curves=((Type=Curve,Name="CTRL_expressions_mouthSharpCornerPullL"))
    backward=true
    control=(Type=Control,Name="CTRL_L_mouth_sharpCornerPull")
    mirror=true
    vector_axis=X
   )
   > FN_FloatCtrlToCurve_FNC_48 = RigVMFunctionReferenceNode(
    Ctrl_L_Token=CTRL_L_
    Ctrl_R_Token=CTRL_R_
    Curve_Side_Token_Last=true
    Pos_Neg_Curves=((Type=Curve,Name="CTRL_expressions_mouthLipsPressL"))
    backward=true
    control=(Type=Control,Name="CTRL_L_mouth_lipsPressU")
    mirror=true
    vector_axis=X
   )
   > FN_FloatCtrlToCurve_FNC_47 = RigVMFunctionReferenceNode(
    Ctrl_L_Token=CTRL_L_
    Ctrl_R_Token=CTRL_R_
    Curve_Side_Token_Last=true
    Pos_Neg_Curves=((Type=Curve,Name="CTRL_expressions_mouthLipsTightenUL"))
    backward=true
    control=(Type=Control,Name="CTRL_L_mouth_tightenU")
    mirror=true
    vector_axis=X
   )
   > FN_FloatCtrlToCurve_FNC_49 = RigVMFunctionReferenceNode(
    Ctrl_L_Token=CTRL_L_
    Ctrl_R_Token=CTRL_R_
    Curve_Side_Token_Last=true
    Pos_Neg_Curves=((Type=Curve,Name="CTRL_expressions_mouthLipsTightenDL"))
    backward=true
    control=(Type=Control,Name="CTRL_L_mouth_tightenD")
    mirror=true
    vector_axis=X
   )
   RigVMFunctionReturnNode(ExecuteContext=@RigVMFunction_Sequence.B)
   FN_VectorCtrlToCurve_FNC_13 = RigVMFunctionReferenceNode(
    Backward=true
    Control=(Type=Control,Name="CTRL_C_eye")
    Ctrl_L_Token=CTRL_L_
    Ctrl_R_Token=CTRL_R_
    Curve_Side_Token_Last=true
    ExecuteContext=()
    X_Mirror=false
    X_Pos_Neg_Expr=((Type=Curve,Name="CTRL_expressions_eyeLookLeftL"),((Type=Curve,Name="None"),(Type=Curve,Name="None"),Type=Curve,Name="CTRL_expressions_eyeLookRightL"))
    Y_Mirror=false
    Y_Pos_Neg_Expr=((Type=Curve,Name="CTRL_expressions_eyeLookUpL"),((Type=Curve,Name="None"),(Type=Curve,Name="None"),Type=Curve,Name="CTRL_expressions_eyeLookDownL"))
    Z_Mirror=false
    Z_Pos_Neg_Expr=()
   )
   Multiply = &RigVMFunction_MathDoubleMul(B=-1, A=&RigUnit_GetCurveValue(Curve=CTRL_expressions_eyeLookRightL).Value)
   Add = &RigVMFunction_MathDoubleAdd(A=@Multiply.Result, B=&RigUnit_GetCurveValue(Curve=CTRL_expressions_eyeLookRightR).Value)
   n5 = &RigUnit_GetCurveValue(Curve=CTRL_expressions_eyeLookLeftL)
   Subtract = &RigVMFunction_MathDoubleSub(
    B=@Add.Result
    A=&RigVMFunction_MathDoubleAdd(
     A=&RigVMFunction_MathDoubleMul(B=-1, A=@n5.Value).Result
     B=&RigUnit_GetCurveValue(Curve=CTRL_expressions_eyeLookLeftR).Value
    ).Result
   )
   n6 = &RigVMFunction_MathDoubleMul(B=-1, A=&RigUnit_GetCurveValue(Curve=CTRL_expressions_eyeLookUpL).Value)
   n7 = &RigUnit_GetCurveValue(Curve=CTRL_expressions_eyeLookUpR)
   n8 = &RigVMFunction_MathDoubleMul(B=-1, A=&RigUnit_GetCurveValue(Curve=CTRL_expressions_eyeLookDownL).Value)
   n9 = &RigUnit_GetCurveValue(Curve=CTRL_expressions_eyeLookDownR)
   > /Script/ControlRig.RigUnit_SetControlVector2D(
    Control=CTRL_R_eye
    Vector=(X=0,Y=0)
    Weight=1
    Vector.X=@Subtract.Result
    Vector.Y=&RigVMFunction_MathDoubleSub(
     A=&RigVMFunction_MathDoubleAdd(A=@n6.Result, B=@n7.Value).Result
     B=&RigVMFunction_MathDoubleAdd(A=@n8.Result, B=@n9.Value).Result
    ).Result
   ) {is=struct RigVMUnitNode, template="SetControlValue::Execute(in Control,in Weight,io Vector)"}
   /Script/ControlRig.RigUnit_FindItemsWithMetadataTag(NameSpace=Self, Tag=wm_curves) {is=struct RigVMUnitNode}
  ]}
 CRFL FN MH Forward v01_ContainedGraph {nodes [
   RigVMFunctionEntryNode()
   > RigVMFunction_Sequence = RigVMAggregateNode()
   &RigUnit_SetCurveValue(Curve=DisableFaceOverride, Value=1, ExecutePin=@RigVMFunction_Sequence.A)
   > &RigUnit_SetCurveValue(Curve=is_3L, Value=1)
   > SequenceAggregate_1 = RigVMAggregateNode()
   FN_FloatCtrlToCurve_FNC_7 = RigVMFunctionReferenceNode(
    Ctrl_L_Token=CTRL_L_
    Ctrl_R_Token=CTRL_R_
    Curve_Side_Token_Last=true
    Pos_Neg_Curves=((Type=Curve,Name="CTRL_expressions_eyeBlinkL"),(Type=Curve,Name="CTRL_expressions_eyeWidenL"))
    backward=False
    control=(Type=Control,Name="CTRL_L_eye_blink")
    mirror=true
    vector_axis=X
    ExecuteContext=@SequenceAggregate_1.B
   )
   n1 = /Script/ControlRig.RigUnit_ProjectTransformToNewParent(
    Child=(Type=Null,Name="EyeAim_GRP")
    NewParent=(Type=Bone,Name="head")
    OldParent=(Type=Bone,Name="head")
    bChildInitial=True
    bNewParentInitial=False
    bOldParentInitial=True
   ) {is=struct RigVMUnitNode}
   > &RigUnit_SetTransform(Item=(Type=Null,Name="EyeAim_GRP"), Weight=1, bInitial=False, Value=@n1.Transform)
   > /Script/ControlRig.RigUnit_AimItem(
    DebugSettings=(bEnabled=False,Scale=10,WorldOffset=(Rotation=(X=0,Y=0,Z=0,W=1),Translation=(X=0,Y=0,Z=0),Scale3D=(X=1,Y=1,Z=1)))
    Item=(Type=Null,Name="EyeAim_L")
    Primary=(Weight=1,Axis=(X=0,Y=0,Z=1),Target=(X=1,Y=0,Z=0),Kind=Location,Space=(Type=Bone,Name="None"))
    Secondary=(Weight=1,Axis=(X=1,Y=0,Z=0),Target=(X=0,Y=0,Z=-1),Kind=Direction,Space=(Type=Bone,Name="head"))
    Weight=1
    Primary.Target=&RigUnit_GetTransform(Item=(Type=Control,Name="CTRL_L_eyeAim"), bInitial=False).Transform.Translation
   ) {is=struct RigVMUnitNode}
   > n2 = &RigUnit_AimItem(
    DebugSettings=(bEnabled=False,Scale=10,WorldOffset=(Rotation=(X=0,Y=0,Z=0,W=1),Translation=(X=0,Y=0,Z=0),Scale3D=(X=1,Y=1,Z=1)))
    Item=(Type=Null,Name="EyeAim_R")
    Primary=(Weight=1,Axis=(X=0,Y=0,Z=1),Target=(X=1,Y=0,Z=0),Kind=Location,Space=(Type=Bone,Name="None"))
    Secondary=(Weight=1,Axis=(X=1,Y=0,Z=0),Target=(X=0,Y=0,Z=-1),Kind=Direction,Space=(Type=Bone,Name="head"))
    Weight=1
    Primary.Target=&RigUnit_GetTransform(Item=(Type=Control,Name="CTRL_R_eyeAim"), bInitial=False).Transform.Translation
   )
   Look_EyeDirection_CollapseNode = RigVMCollapseNode(ExecutePin=@n2.ExecutePin)
   Eye_Convergence_CollapseNode = RigVMCollapseNode(ExecutePin=@Look_EyeDirection_CollapseNode.ExecutePin)
   FN_VectorCtrlToCurve_FNC_4 = RigVMFunctionReferenceNode(
    Backward=false
    Control=(Type=Control,Name="CTRL_L_nose")
    Ctrl_L_Token=CTRL_L_
    Ctrl_R_Token=CTRL_R_
    Curve_Side_Token_Last=true
    X_Mirror=true
    X_Pos_Neg_Expr=((Type=Curve,Name="CTRL_expressions_noseNostrilDilateL"),(Type=Curve,Name="CTRL_expressions_noseNostrilCompressL"))
    Y_Mirror=true
    Y_Pos_Neg_Expr=((Type=Curve,Name="CTRL_expressions_noseWrinkleL"),(Type=Curve,Name="CTRL_expressions_noseNostrilDepressL"))
    Z_Mirror=true
    Z_Pos_Neg_Expr=()
    ExecuteContext=@SequenceAggregate_1.C
   )
   > FN_FloatCtrlToCurve_FNC_8 = RigVMFunctionReferenceNode(
    Ctrl_L_Token=CTRL_L_
    Ctrl_R_Token=CTRL_R_
    Curve_Side_Token_Last=true
    Pos_Neg_Curves=((Type=Curve,Name="CTRL_expressions_noseWrinkleUpperL"))
    backward=False
    control=(Type=Control,Name="CTRL_L_nose_wrinkleUpper")
    mirror=true
    vector_axis=X
   )
   > FN_FloatCtrlToCurve_FNC_9 = RigVMFunctionReferenceNode(
    Ctrl_L_Token=CTRL_L_
    Ctrl_R_Token=CTRL_R_
    Curve_Side_Token_Last=true
    Pos_Neg_Curves=((Type=Curve,Name="CTRL_expressions_eyeSquintInnerL"))
    backward=False
    control=(Type=Control,Name="CTRL_L_eye_squintInner")
    mirror=true
    vector_axis=X
   )
   > FN_FloatCtrlToCurve_FNC_10 = RigVMFunctionReferenceNode(
    Ctrl_L_Token=CTRL_L_
    Ctrl_R_Token=CTRL_R_
    Curve_Side_Token_Last=true
    Pos_Neg_Curves=((Type=Curve,Name="CTRL_expressions_eyeCheekRaiseL"))
    backward=False
    control=(Type=Control,Name="CTRL_L_eye_cheekRaise")
    mirror=true
    vector_axis=X
   )
   > FN_FloatCtrlToCurve_FNC_11 = RigVMFunctionReferenceNode(
    Ctrl_L_Token=CTRL_L_
    Ctrl_R_Token=CTRL_R_
    Curve_Side_Token_Last=true
    Pos_Neg_Curves=((Type=Curve,Name="CTRL_expressions_mouthCheekBlowL"),(Type=Curve,Name="CTRL_expressions_mouthCheekSuckL"))
    backward=False
    control=(Type=Control,Name="CTRL_L_mouth_suckBlow")
    mirror=true
    vector_axis=X
   )
   > FN_FloatCtrlToCurve_FNC_12 = RigVMFunctionReferenceNode(
    Ctrl_L_Token=CTRL_L_
    Ctrl_R_Token=CTRL_R_
    Curve_Side_Token_Last=true
    Pos_Neg_Curves=((Type=Curve,Name="CTRL_expressions_mouthLipsBlowL"))
    backward=False
    control=(Type=Control,Name="CTRL_L_mouth_lipsBlow")
    mirror=true
    vector_axis=X
   )
   > FN_FloatCtrlToCurve_FNC_13 = RigVMFunctionReferenceNode(
    Ctrl_L_Token=CTRL_L_
    Ctrl_R_Token=CTRL_R_
    Curve_Side_Token_Last=true
    Pos_Neg_Curves=((Type=Curve,Name="CTRL_expressions_mouthUpperLipRaiseL"))
    backward=False
    control=(Type=Control,Name="CTRL_L_mouth_upperLipRaise")
    mirror=true
    vector_axis=X
   )
   > FN_FloatCtrlToCurve_FNC_14 = RigVMFunctionReferenceNode(
    Ctrl_L_Token=CTRL_L_
    Ctrl_R_Token=CTRL_R_
    Curve_Side_Token_Last=true
    Pos_Neg_Curves=((Type=Curve,Name="CTRL_expressions_mouthCornerPullL"))
    backward=False
    control=(Type=Control,Name="CTRL_L_mouth_cornerPull")
    mirror=true
    vector_axis=X
   )
   > FN_FloatCtrlToCurve_FNC_15 = RigVMFunctionReferenceNode(
    Ctrl_L_Token=CTRL_L_
    Ctrl_R_Token=CTRL_R_
    Curve_Side_Token_Last=true
    Pos_Neg_Curves=((Type=Curve,Name="CTRL_expressions_mouthDimpleL"))
    backward=False
    control=(Type=Control,Name="CTRL_L_mouth_dimple")
    mirror=true
    vector_axis=X
   )
   > FN_FloatCtrlToCurve_FNC_16 = RigVMFunctionReferenceNode(
    Ctrl_L_Token=CTRL_L_
    Ctrl_R_Token=CTRL_R_
    Curve_Side_Token_Last=true
    Pos_Neg_Curves=((Type=Curve,Name="CTRL_expressions_mouthCornerDepressL"))
    backward=False
    control=(Type=Control,Name="CTRL_L_mouth_cornerDepress")
    mirror=true
    vector_axis=X
   )
   > FN_FloatCtrlToCurve_FNC_17 = RigVMFunctionReferenceNode(
    Ctrl_L_Token=CTRL_L_
    Ctrl_R_Token=CTRL_R_
    Curve_Side_Token_Last=true
    Pos_Neg_Curves=((Type=Curve,Name="CTRL_expressions_mouthStretchL"))
    backward=False
    control=(Type=Control,Name="CTRL_L_mouth_stretch")
    mirror=true
    vector_axis=X
   )
   > FN_FloatCtrlToCurve_FNC_18 = RigVMFunctionReferenceNode(
    Ctrl_L_Token=CTRL_L_
    Ctrl_R_Token=CTRL_R_
    Curve_Side_Token_Last=true
    Pos_Neg_Curves=((Type=Curve,Name="CTRL_expressions_mouthLowerLipDepressL"))
    backward=False
    control=(Type=Control,Name="CTRL_L_mouth_lowerLipDepress")
    mirror=true
    vector_axis=X
   )
   > FN_VectorCtrlToCurve_FNC_6 = RigVMFunctionReferenceNode(
    Backward=false
    Control=(Type=Control,Name="CTRL_C_mouth")
    Ctrl_L_Token=CTRL_L_
    Ctrl_R_Token=CTRL_R_
    Curve_Side_Token_Last=true
    X_Mirror=false
    X_Pos_Neg_Expr=((Type=Curve,Name="CTRL_expressions_mouthLeft"),(Type=Curve,Name="CTRL_expressions_mouthRight"))
    Y_Mirror=false
    Y_Pos_Neg_Expr=((Type=Curve,Name="CTRL_expressions_mouthUp"),(Type=Curve,Name="CTRL_expressions_mouthDown"))
    Z_Mirror=false
    Z_Pos_Neg_Expr=()
   )
   > FN_FloatCtrlToCurve_FNC_31 = RigVMFunctionReferenceNode(
    Ctrl_L_Token=CTRL_L_
    Ctrl_R_Token=CTRL_R_
    Curve_Side_Token_Last=true
    Pos_Neg_Curves=((Type=Curve,Name="CTRL_expressions_tongueIn"),(Type=Curve,Name="CTRL_expressions_tongueOut"))
    backward=False
    control=(Type=Control,Name="CTRL_C_tongue_inOut")
    mirror=false
    vector_axis=X
   )
   > FN_VectorCtrlToCurve_FNC_11 = RigVMFunctionReferenceNode(
    Backward=false
    Control=(Type=Control,Name="CTRL_C_tongue")
    Ctrl_L_Token=CTRL_L_
    Ctrl_R_Token=CTRL_R_
    Curve_Side_Token_Last=true
    X_Mirror=false
    X_Pos_Neg_Expr=((Type=Curve,Name="CTRL_expressions_tongueLeft"),(Type=Curve,Name="CTRL_expressions_tongueRight"))
    Y_Mirror=false
    Y_Pos_Neg_Expr=((Type=Curve,Name="CTRL_expressions_tongueUp"),(Type=Curve,Name="CTRL_expressions_tongueDown"))
    Z_Mirror=false
    Z_Pos_Neg_Expr=()
   )
   > FN_VectorCtrlToCurve_FNC_12 = RigVMFunctionReferenceNode(
    Backward=false
    Control=(Type=Control,Name="CTRL_C_tongue_roll")
    Ctrl_L_Token=CTRL_L_
    Ctrl_R_Token=CTRL_R_
    Curve_Side_Token_Last=true
    X_Mirror=false
    X_Pos_Neg_Expr=((Type=Curve,Name="CTRL_expressions_tongueRollLeft"),(Type=Curve,Name="CTRL_expressions_tongueRollRight"))
    Y_Mirror=false
    Y_Pos_Neg_Expr=((Type=Curve,Name="CTRL_expressions_tongueRollUp"),(Type=Curve,Name="CTRL_expressions_tongueRollDown"))
    Z_Mirror=false
    Z_Pos_Neg_Expr=()
   )
   > FN_FloatCtrlToCurve_FNC_32 = RigVMFunctionReferenceNode(
    Ctrl_L_Token=CTRL_L_
    Ctrl_R_Token=CTRL_R_
    Curve_Side_Token_Last=true
    Pos_Neg_Curves=((Type=Curve,Name="CTRL_expressions_tongueWide"),(Type=Curve,Name="CTRL_expressions_tongueNarrow"))
    backward=False
    control=(Type=Control,Name="CTRL_C_tongue_narrowWide")
    mirror=false
    vector_axis=X
   )
   FN_FloatCtrlToCurve_FNC_33 = RigVMFunctionReferenceNode(
    Ctrl_L_Token=CTRL_L_
    Ctrl_R_Token=CTRL_R_
    Curve_Side_Token_Last=true
    Pos_Neg_Curves=((Type=Curve,Name="CTRL_expressions_mouthLipsTowardsUL"))
    backward=False
    control=(Type=Control,Name="CTRL_L_mouth_towardsU")
    mirror=true
    vector_axis=X
    ExecuteContext=@SequenceAggregate_1.D
   )
   > FN_FloatCtrlToCurve_FNC_34 = RigVMFunctionReferenceNode(
    Ctrl_L_Token=CTRL_L_
    Ctrl_R_Token=CTRL_R_
    Curve_Side_Token_Last=true
    Pos_Neg_Curves=((Type=Curve,Name="CTRL_expressions_mouthLipsTowardsDL"))
    backward=False
    control=(Type=Control,Name="CTRL_L_mouth_towardsD")
    mirror=true
    vector_axis=X
   )
   > FN_FloatCtrlToCurve_FNC_35 = RigVMFunctionReferenceNode(
    Ctrl_L_Token=CTRL_L_
    Ctrl_R_Token=CTRL_R_
    Curve_Side_Token_Last=true
    Pos_Neg_Curves=((Type=Curve,Name="CTRL_expressions_mouthLipsPurseUL"))
    backward=False
    control=(Type=Control,Name="CTRL_L_mouth_purseU")
    mirror=true
    vector_axis=X
   )
   > FN_FloatCtrlToCurve_FNC_36 = RigVMFunctionReferenceNode(
    Ctrl_L_Token=CTRL_L_
    Ctrl_R_Token=CTRL_R_
    Curve_Side_Token_Last=true
    Pos_Neg_Curves=((Type=Curve,Name="CTRL_expressions_mouthLipsPurseDL"))
    backward=False
    control=(Type=Control,Name="CTRL_L_mouth_purseD")
    mirror=true
    vector_axis=X
   )
   > FN_FloatCtrlToCurve_FNC_37 = RigVMFunctionReferenceNode(
    Ctrl_L_Token=CTRL_L_
    Ctrl_R_Token=CTRL_R_
    Curve_Side_Token_Last=true
    Pos_Neg_Curves=((Type=Curve,Name="CTRL_expressions_mouthFunnelUL"))
    backward=False
    control=(Type=Control,Name="CTRL_L_mouth_funnelU")
    mirror=true
    vector_axis=X
   )
   > FN_FloatCtrlToCurve_FNC_38 = RigVMFunctionReferenceNode(
    Ctrl_L_Token=CTRL_L_
    Ctrl_R_Token=CTRL_R_
    Curve_Side_Token_Last=true
    Pos_Neg_Curves=((Type=Curve,Name="CTRL_expressions_mouthFunnelDL"))
    backward=False
    control=(Type=Control,Name="CTRL_L_mouth_funnelD")
    mirror=true
    vector_axis=X
   )
   > FN_FloatCtrlToCurve_FNC_19 = RigVMFunctionReferenceNode(
    Ctrl_L_Token=CTRL_L_
    Ctrl_R_Token=CTRL_R_
    Curve_Side_Token_Last=true
    Pos_Neg_Curves=((Type=Curve,Name="CTRL_expressions_jawChinRaiseUL"))
    backward=False
    control=(Type=Control,Name="CTRL_L_jaw_ChinRaiseU")
    mirror=true
    vector_axis=X
   )
   > FN_FloatCtrlToCurve_FNC_20 = RigVMFunctionReferenceNode(
    Ctrl_L_Token=CTRL_L_
    Ctrl_R_Token=CTRL_R_
    Curve_Side_Token_Last=true
    Pos_Neg_Curves=((Type=Curve,Name="CTRL_expressions_jawChinRaiseDL"))
    backward=False
    control=(Type=Control,Name="CTRL_L_jaw_ChinRaiseD")
    mirror=true
    vector_axis=X
   )
   > FN_VectorCtrlToCurve_FNC_7 = RigVMFunctionReferenceNode(
    Backward=false
    Control=(Type=Control,Name="CTRL_C_jaw")
    Ctrl_L_Token=CTRL_L_
    Ctrl_R_Token=CTRL_R_
    Curve_Side_Token_Last=true
    X_Mirror=false
    X_Pos_Neg_Expr=((Type=Curve,Name="CTRL_expressions_jawRight"),(Type=Curve,Name="CTRL_expressions_jawLeft"))
    Y_Mirror=false
    Y_Pos_Neg_Expr=((Type=Curve,Name="CTRL_expressions_jawOpen"))
    Z_Mirror=false
    Z_Pos_Neg_Expr=()
   )
   > FN_FloatCtrlToCurve_FNC_22 = RigVMFunctionReferenceNode(
    Ctrl_L_Token=CTRL_L_
    Ctrl_R_Token=CTRL_R_
    Curve_Side_Token_Last=true
    Pos_Neg_Curves=((Type=Curve,Name="CTRL_expressions_jawOpenExtreme"))
    backward=False
    control=(Type=Control,Name="CTRL_C_jaw_openExtreme")
    mirror=false
    vector_axis=X
   )
   > FN_FloatCtrlToCurve_FNC_21 = RigVMFunctionReferenceNode(
    Ctrl_L_Token=CTRL_L_
    Ctrl_R_Token=CTRL_R_
    Curve_Side_Token_Last=true
    Pos_Neg_Curves=((Type=Curve,Name="CTRL_expressions_jawBack"),(Type=Curve,Name="CTRL_expressions_jawFwd"))
    backward=False
    control=(Type=Control,Name="CTRL_C_jaw_fwdBack")
    mirror=false
    vector_axis=X
   )
   > FN_FloatCtrlToCurve_FNC_41 = RigVMFunctionReferenceNode(
    Ctrl_L_Token=CTRL_L_
    Ctrl_R_Token=CTRL_R_
    Curve_Side_Token_Last=true
    Pos_Neg_Curves=((Type=Curve,Name="CTRL_expressions_mouthLipsTogetherUL"))
    backward=False
    control=(Type=Control,Name="CTRL_L_mouth_lipsTogetherU")
    mirror=true
    vector_axis=X
   )
   > FN_FloatCtrlToCurve_FNC_42 = RigVMFunctionReferenceNode(
    Ctrl_L_Token=CTRL_L_
    Ctrl_R_Token=CTRL_R_
    Curve_Side_Token_Last=true
    Pos_Neg_Curves=((Type=Curve,Name="CTRL_expressions_mouthLipsTogetherDL"))
    backward=False
    control=(Type=Control,Name="CTRL_L_mouth_lipsTogetherD")
    mirror=true
    vector_axis=X
   )
   > FN_FloatCtrlToCurve_FNC_43 = RigVMFunctionReferenceNode(
    Ctrl_L_Token=CTRL_L_
    Ctrl_R_Token=CTRL_R_
    Curve_Side_Token_Last=true
    Pos_Neg_Curves=((Type=Curve,Name="CTRL_expressions_mouthPressUL"))
    backward=False
    control=(Type=Control,Name="CTRL_L_mouth_pressU")
    mirror=true
    vector_axis=X
   )
   > FN_FloatCtrlToCurve_FNC_44 = RigVMFunctionReferenceNode(
    Ctrl_L_Token=CTRL_L_
    Ctrl_R_Token=CTRL_R_
    Curve_Side_Token_Last=true
    Pos_Neg_Curves=((Type=Curve,Name="CTRL_expressions_mouthPressDL"))
    backward=False
    control=(Type=Control,Name="CTRL_L_mouth_pressD")
    mirror=true
    vector_axis=X
   )
   > FN_FloatCtrlToCurve_FNC_39 = RigVMFunctionReferenceNode(
    Ctrl_L_Token=CTRL_L_
    Ctrl_R_Token=CTRL_R_
    Curve_Side_Token_Last=true
    Pos_Neg_Curves=((Type=Curve,Name="CTRL_expressions_mouthUpperLipBiteL"))
    backward=False
    control=(Type=Control,Name="CTRL_L_mouth_lipBiteU")
    mirror=true
    vector_axis=X
   )
   > FN_FloatCtrlToCurve_FNC_40 = RigVMFunctionReferenceNode(
    Ctrl_L_Token=CTRL_L_
    Ctrl_R_Token=CTRL_R_
    Curve_Side_Token_Last=true
    Pos_Neg_Curves=((Type=Curve,Name="CTRL_expressions_mouthLowerLipBiteL"))
    backward=False
    control=(Type=Control,Name="CTRL_L_mouth_lipBiteD")
    mirror=true
    vector_axis=X
   )
   n3 = &RigVMFunction_ControlFlowBranch(Condition=RigVMVariableNode() {variable=Is V2}.Value, ExecuteContext=@SequenceAggregate_1.E)
   FN_FloatCtrlToCurve_FNC_4 = RigVMFunctionReferenceNode(
    Ctrl_L_Token=CTRL_L_
    Ctrl_R_Token=CTRL_R_
    Curve_Side_Token_Last=true
    Pos_Neg_Curves=((Type=Curve,Name="CTRL_expressions_eyePupilWideL"),(Type=Curve,Name="CTRL_expressions_eyePupilNarrowL"))
    backward=False
    control=(Type=Control,Name="CTRL_L_eye_pupil")
    mirror=true
    vector_axis=X
    ExecuteContext=@n3.True
   )
   > FN_FloatCtrlToCurve_FNC_5 = RigVMFunctionReferenceNode(
    Ctrl_L_Token=CTRL_L_
    Ctrl_R_Token=CTRL_R_
    Curve_Side_Token_Last=true
    Pos_Neg_Curves=((Type=Curve,Name="CTRL_expressions_mouthSharpCornerPullL"))
    backward=False
    control=(Type=Control,Name="CTRL_L_mouth_sharpCornerPull")
    mirror=true
    vector_axis=X
   )
   > FN_FloatCtrlToCurve_FNC_23 = RigVMFunctionReferenceNode(
    Ctrl_L_Token=CTRL_L_
    Ctrl_R_Token=CTRL_R_
    Curve_Side_Token_Last=true
    Pos_Neg_Curves=((Type=Curve,Name="CTRL_expressions_mouthLipsPressL"))
    backward=False
    control=(Type=Control,Name="CTRL_L_mouth_lipsPressU")
    mirror=true
    vector_axis=X
   )
   > FN_FloatCtrlToCurve_FNC_6 = RigVMFunctionReferenceNode(
    Ctrl_L_Token=CTRL_L_
    Ctrl_R_Token=CTRL_R_
    Curve_Side_Token_Last=true
    Pos_Neg_Curves=((Type=Curve,Name="CTRL_expressions_mouthLipsTightenUL"))
    backward=False
    control=(Type=Control,Name="CTRL_L_mouth_tightenU")
    mirror=true
    vector_axis=X
   )
   > FN_FloatCtrlToCurve_FNC_24 = RigVMFunctionReferenceNode(
    Ctrl_L_Token=CTRL_L_
    Ctrl_R_Token=CTRL_R_
    Curve_Side_Token_Last=true
    Pos_Neg_Curves=((Type=Curve,Name="CTRL_expressions_mouthLipsTightenDL"))
    backward=False
    control=(Type=Control,Name="CTRL_L_mouth_tightenD")
    mirror=true
    vector_axis=X
   )
   n4 = &RigVMFunction_Sequence(ExecuteContext=@SequenceAggregate_1.F)
   For_Each = RigVMDispatchNode(Array=&RigUnit_CollectionNameSearchArray(PartialName=_Offset_, TypeToSearch=Curve).Items, ExecuteContext=@n4.A) {template="DISPATCH_RigVMDispatch_ArrayIterator(in Array,out Element,out Index,out Count,out Ratio)"}
   StartsWith = &RigVMFunction_StartsWith(Start=WM_, Name=@For_Each.Element.Name) {template="StartsWith::Execute(in Name,in Start,out Result)"}
   n5 = &RigVMFunction_ControlFlowBranch(ExecuteContext=@For_Each.ExecuteContext, Condition=@StartsWith.Result)
   GetAnimationChannel = /Script/ControlRig.RigUnit_GetFloatAnimationChannel(Control=CTRL_WrinkleMap, bInitial=False, Channel=@For_Each.Element.Name) {is=struct RigVMUnitNode, template="GetAnimationChannel::Execute(out Value,in Control,in Channel,in bInitial)"}
   &RigUnit_SetCurveValue(Value=@GetAnimationChannel.Value, ExecutePin=@n5.True, Curve=@For_Each.Element.Name)
   CRFL FN MH Tweaker Forward v01 = RigVMFunctionReferenceNode(ExecuteContext=@n4.B, Is V2=RigVMVariableNode() {variable=Is V2}.Value)
   n6 = &RigUnit_ProjectTransformToNewParent(
    Child=(Type=Null,Name="Facial_Board_Follow")
    NewParent=(Type=Bone,Name="head")
    OldParent=(Type=Bone,Name="head")
    bChildInitial=True
    bNewParentInitial=False
    bOldParentInitial=True
   )
   &RigUnit_SetTransform(
    Item=(Type=Null,Name="Facial_Board_Follow")
    Weight=1
    bInitial=False
    Value=/Script/RigVM.RigVMFunction_MathTransformLerp(
     T=/Script/ControlRig.RigUnit_GetControlFloat(Control=CTRL_faceGUIfollowHead) {is=struct RigVMUnitNode, template="GetControlValue::Execute(in Control,out FloatValue,out Minimum,out Maximum)"}.FloatValue
     B=@n6.Transform
     A=&RigUnit_GetTransform(Item=(Type=Null,Name="Facial_Board_Follow"), bInitial=true).Transform
    ) {is=struct RigVMUnitNode, template="Interpolate::Execute(in A,in B,in T,out Result)"}.Result
    ExecutePin=@SequenceAggregate_1.A
   )
   n7 = &RigUnit_ProjectTransformToNewParent(
    Child=(Type=Null,Name="GRP_eyeControls")
    NewParent=(Type=Bone,Name="head")
    OldParent=(Type=Bone,Name="head")
    bChildInitial=True
    bNewParentInitial=False
    bOldParentInitial=True
   )
   > &RigUnit_SetTransform(
    Item=(Type=Null,Name="GRP_eyeControls")
    Weight=1
    bInitial=False
    Value=&RigVMFunction_MathTransformLerp(
     T=&RigUnit_GetControlFloat(Control=CTRL_eyesAimFollowHead) {template="GetControlValue::Execute(in Control,out FloatValue,out Minimum,out Maximum)"}.FloatValue
     B=@n7.Transform
     A=&RigUnit_GetTransform(Item=(Type=Null,Name="GRP_eyeControls"), bInitial=true).Transform
    ) {template="Interpolate::Execute(in A,in B,in T,out Result)"}.Result
   )
   > FN_FloatCtrlToCurve_FNC = RigVMFunctionReferenceNode(
    Ctrl_L_Token=CTRL_L_
    Ctrl_R_Token=CTRL_R_
    Curve_Side_Token_Last=true
    Pos_Neg_Curves=((Type=Curve,Name="CTRL_expressions_browRaiseInL"))
    backward=False
    control=(Type=Control,Name="CTRL_L_brow_raiseIn")
    mirror=true
    vector_axis=X
   )
   > FN_FloatCtrlToCurve_FNC_1 = RigVMFunctionReferenceNode(
    Ctrl_L_Token=CTRL_L_
    Ctrl_R_Token=CTRL_R_
    Curve_Side_Token_Last=true
    Pos_Neg_Curves=((Type=Curve,Name="CTRL_expressions_browRaiseOuterL"))
    backward=False
    control=(Type=Control,Name="CTRL_L_brow_raiseOut")
    mirror=true
    vector_axis=X
   )
   > FN_FloatCtrlToCurve_FNC_2 = RigVMFunctionReferenceNode(
    Ctrl_L_Token=CTRL_L_
    Ctrl_R_Token=CTRL_R_
    Curve_Side_Token_Last=true
    Pos_Neg_Curves=((Type=Curve,Name="CTRL_expressions_browDownL"))
    backward=False
    control=(Type=Control,Name="CTRL_L_brow_down")
    mirror=true
    vector_axis=X
   )
   > FN_FloatCtrlToCurve_FNC_3 = RigVMFunctionReferenceNode(
    Ctrl_L_Token=CTRL_L_
    Ctrl_R_Token=CTRL_R_
    Curve_Side_Token_Last=true
    Pos_Neg_Curves=((Type=Curve,Name="CTRL_expressions_browLateralL"))
    backward=False
    control=(Type=Control,Name="CTRL_L_brow_lateral")
    mirror=true
    vector_axis=X
   )
   > Eye_Aim_Control_Vis_CollapseNode = RigVMCollapseNode()
   RigVMFunctionReturnNode(ExecuteContext=@RigVMFunction_Sequence.C)
   &RigUnit_FindItemsWithMetadataTag(NameSpace=Root, Tag=wm_curves)
  ]}
 CRFL FN MH Tweak Backward v01_ContainedGraph {nodes [
   n1 = RigVMFunctionEntryNode()
   n2 = &RigVMFunction_Sequence(ExecuteContext=@n1.ExecuteContext)
   RigVMFunction_Sequence = RigVMAggregateNode(ExecuteContext=@n2.A)
   FN_FloatCtrlToCurve_FNC_4 = RigVMFunctionReferenceNode(
    Ctrl_L_Token=CTRL_L_
    Ctrl_R_Token=CTRL_R_
    Curve_Side_Token_Last=true
    Pos_Neg_Curves=((Type=Curve,Name="CTRL_expressions_eyeEyelidsReduce"))
    backward=true
    control=(Type=Control,Name="CTRL_C_eyelidsReduce")
    mirror=False
    vector_axis=X
    ExecuteContext=@RigVMFunction_Sequence.A
   )
   > FN_VectorCtrlToCurve_FNC_11 = RigVMFunctionReferenceNode(
    Backward=true
    Control=(Type=Control,Name="CTRL_L_eye_eyelidU")
    Ctrl_L_Token=CTRL_L_
    Ctrl_R_Token=CTRL_R_
    Curve_Side_Token_Last=true
    X_Mirror=false
    X_Pos_Neg_Expr=((Type=Curve,Name="CTRL_expressions_eyelidRollNegUL"),(Type=Curve,Name="CTRL_expressions_eyelidRollPosUL"))
    Y_Mirror=true
    Y_Pos_Neg_Expr=((Type=Curve,Name="CTRL_expressions_eyeRelaxL"),(Type=Curve,Name="CTRL_expressions_eyeUpperLidUpL"))
    Z_Mirror=false
    Z_Pos_Neg_Expr=()
   )
   > FN_VectorCtrlToCurve_FNC_12 = RigVMFunctionReferenceNode(
    Backward=true
    Control=(Type=Control,Name="CTRL_R_eye_eyelidU")
    Ctrl_L_Token=CTRL_L_
    Ctrl_R_Token=CTRL_R_
    Curve_Side_Token_Last=true
    X_Mirror=false
    X_Pos_Neg_Expr=((Type=Curve,Name="CTRL_expressions_eyelidRollPosUR"),(Type=Curve,Name="CTRL_expressions_eyelidRollNegUR"))
    Y_Mirror=false
    Y_Pos_Neg_Expr=((Type=Curve,Name="None"),(Type=Curve,Name="None"))
    Z_Mirror=false
    Z_Pos_Neg_Expr=()
   )
   > FN_VectorCtrlToCurve_FNC_13 = RigVMFunctionReferenceNode(
    Backward=true
    Control=(Type=Control,Name="CTRL_L_eye_eyelidD")
    Ctrl_L_Token=CTRL_L_
    Ctrl_R_Token=CTRL_R_
    Curve_Side_Token_Last=true
    X_Mirror=false
    X_Pos_Neg_Expr=((Type=Curve,Name="CTRL_expressions_eyelidRollNegDL"),(Type=Curve,Name="CTRL_expressions_eyelidRollPosDL"))
    Y_Mirror=true
    Y_Pos_Neg_Expr=((Type=Curve,Name="CTRL_expressions_eyeLowerLidUpL"),(Type=Curve,Name="CTRL_expressions_eyeLowerLidDownL"))
    Z_Mirror=false
    Z_Pos_Neg_Expr=()
   )
   > FN_VectorCtrlToCurve_FNC_14 = RigVMFunctionReferenceNode(
    Backward=true
    Control=(Type=Control,Name="CTRL_R_eye_eyelidD")
    Ctrl_L_Token=CTRL_L_
    Ctrl_R_Token=CTRL_R_
    Curve_Side_Token_Last=true
    X_Mirror=false
    X_Pos_Neg_Expr=((Type=Curve,Name="CTRL_expressions_eyelidRollPosDR"),(Type=Curve,Name="CTRL_expressions_eyelidRollNegDR"))
    Y_Mirror=False
    Y_Pos_Neg_Expr=((Type=Curve,Name="None"),(Type=Curve,Name="None"))
    Z_Mirror=false
    Z_Pos_Neg_Expr=()
   )
   > FN_FloatCtrlToCurve_FNC_23 = RigVMFunctionReferenceNode(
    Ctrl_L_Token=CTRL_L_
    Ctrl_R_Token=CTRL_R_
    Curve_Side_Token_Last=true
    Pos_Neg_Curves=((Type=Curve,Name="CTRL_expressions_mouthUpperLipTowardsTeethL"))
    backward=true
    control=(Type=Control,Name="CTRL_L_mouth_lipsTowardsTeethU")
    mirror=true
    vector_axis=X
   )
   > FN_FloatCtrlToCurve_FNC_24 = RigVMFunctionReferenceNode(
    Ctrl_L_Token=CTRL_L_
    Ctrl_R_Token=CTRL_R_
    Curve_Side_Token_Last=true
    Pos_Neg_Curves=((Type=Curve,Name="CTRL_expressions_mouthLowerLipTowardsTeethL"),(Type=Curve,Name="None"))
    backward=true
    control=(Type=Control,Name="CTRL_L_mouth_lipsTowardsTeethD")
    mirror=true
    vector_axis=X
   )
   > FN_VectorCtrlToCurve_FNC_5 = RigVMFunctionReferenceNode(
    Backward=true
    Control=(Type=Control,Name="CTRL_C_nose")
    Ctrl_L_Token=CTRL_L_
    Ctrl_R_Token=CTRL_R_
    Curve_Side_Token_Last=true
    X_Mirror=false
    X_Pos_Neg_Expr=((Type=Curve,Name="CTRL_expressions_noseRotateL"),(Type=Curve,Name="CTRL_expressions_noseRotateR"))
    Y_Mirror=false
    Y_Pos_Neg_Expr=((Type=Curve,Name="CTRL_expressions_noseSquash"),(Type=Curve,Name="CTRL_expressions_noseStretch"))
    Z_Mirror=false
    Z_Pos_Neg_Expr=()
   )
   > FN_VectorCtrlToCurve_FNC_8 = RigVMFunctionReferenceNode(
    Backward=true
    Control=(Type=Control,Name="CTRL_L_mouth_corner")
    Ctrl_L_Token=CTRL_L_
    Ctrl_R_Token=CTRL_R_
    Curve_Side_Token_Last=true
    X_Mirror=true
    X_Pos_Neg_Expr=((Type=Curve,Name="CTRL_expressions_mouthCornerWideL"),(Type=Curve,Name="CTRL_expressions_mouthCornerNarrowL"))
    Y_Mirror=true
    Y_Pos_Neg_Expr=((Type=Curve,Name="CTRL_expressions_mouthCornerUpL"),(Type=Curve,Name="CTRL_expressions_mouthCornerDownL"))
    Z_Mirror=false
    Z_Pos_Neg_Expr=()
   )
   FN_FloatCtrlToCurve_FNC_27 = RigVMFunctionReferenceNode(
    Ctrl_L_Token=CTRL_L_
    Ctrl_R_Token=CTRL_R_
    Curve_Side_Token_Last=true
    Pos_Neg_Curves=((Type=Curve,Name="CTRL_expressions_teethBackU"),(Type=Curve,Name="CTRL_expressions_teethFwdU"))
    backward=true
    control=(Type=Control,Name="CTRL_C_teeth_fwdBackU")
    mirror=false
    vector_axis=X
    ExecuteContext=@RigVMFunction_Sequence.B
   )
   > FN_FloatCtrlToCurve_FNC_28 = RigVMFunctionReferenceNode(
    Ctrl_L_Token=CTRL_L_
    Ctrl_R_Token=CTRL_R_
    Curve_Side_Token_Last=true
    Pos_Neg_Curves=((Type=Curve,Name="CTRL_expressions_teethBackD"),(Type=Curve,Name="CTRL_expressions_teethFwdD"))
    backward=true
    control=(Type=Control,Name="CTRL_C_teeth_fwdBackD")
    mirror=false
    vector_axis=X
   )
   > FN_VectorCtrlToCurve_FNC_9 = RigVMFunctionReferenceNode(
    Backward=true
    Control=(Type=Control,Name="CTRL_C_teethU")
    Ctrl_L_Token=CTRL_L_
    Ctrl_R_Token=CTRL_R_
    Curve_Side_Token_Last=true
    X_Mirror=false
    X_Pos_Neg_Expr=((Type=Curve,Name="CTRL_expressions_teethLeftU"),(Type=Curve,Name="CTRL_expressions_teethRightU"))
    Y_Mirror=false
    Y_Pos_Neg_Expr=((Type=Curve,Name="CTRL_expressions_teethUpU"),(Type=Curve,Name="CTRL_expressions_teethDownU"))
    Z_Mirror=false
    Z_Pos_Neg_Expr=()
   )
   > FN_VectorCtrlToCurve_FNC_10 = RigVMFunctionReferenceNode(
    Backward=true
    Control=(Type=Control,Name="CTRL_C_teethD")
    Ctrl_L_Token=CTRL_L_
    Ctrl_R_Token=CTRL_R_
    Curve_Side_Token_Last=true
    X_Mirror=false
    X_Pos_Neg_Expr=((Type=Curve,Name="CTRL_expressions_teethLeftD"),(Type=Curve,Name="CTRL_expressions_teethRightD"))
    Y_Mirror=false
    Y_Pos_Neg_Expr=((Type=Curve,Name="CTRL_expressions_teethUpD"),(Type=Curve,Name="CTRL_expressions_teethDownD"))
    Z_Mirror=false
    Z_Pos_Neg_Expr=()
   )
   > FN_FloatCtrlToCurve_FNC_29 = RigVMFunctionReferenceNode(
    Ctrl_L_Token=CTRL_L_
    Ctrl_R_Token=CTRL_R_
    Curve_Side_Token_Last=true
    Pos_Neg_Curves=((Type=Curve,Name="CTRL_expressions_teethRotateFwdU"),(Type=Curve,Name="CTRL_expressions_teethRotateBackU"))
    backward=true
    control=(Type=Control,Name="CTRL_C_teeth_rotateU")
    mirror=false
    vector_axis=X
   )
   > FN_FloatCtrlToCurve_FNC_30 = RigVMFunctionReferenceNode(
    Ctrl_L_Token=CTRL_L_
    Ctrl_R_Token=CTRL_R_
    Curve_Side_Token_Last=true
    Pos_Neg_Curves=((Type=Curve,Name="CTRL_expressions_teethRotateFwdD"),(Type=Curve,Name="CTRL_expressions_teethRotateBackD"))
    backward=true
    control=(Type=Control,Name="CTRL_C_teeth_rotateD")
    mirror=false
    vector_axis=X
   )
   > FN_FloatCtrlToCurve_FNC_25 = RigVMFunctionReferenceNode(
    Ctrl_L_Token=CTRL_L_
    Ctrl_R_Token=CTRL_R_
    Curve_Side_Token_Last=true
    Pos_Neg_Curves=((Type=Curve,Name="CTRL_expressions_mouthCornerSharpenUL"),(Type=Curve,Name="CTRL_expressions_mouthCornerRounderUL"))
    backward=true
    control=(Type=Control,Name="CTRL_L_mouth_cornerSharpnessU")
    mirror=true
    vector_axis=X
   )
   > FN_FloatCtrlToCurve_FNC_26 = RigVMFunctionReferenceNode(
    Ctrl_L_Token=CTRL_L_
    Ctrl_R_Token=CTRL_R_
    Curve_Side_Token_Last=true
    Pos_Neg_Curves=((Type=Curve,Name="CTRL_expressions_mouthCornerSharpenDL"),(Type=Curve,Name="CTRL_expressions_mouthCornerRounderDL"))
    backward=true
    control=(Type=Control,Name="CTRL_L_mouth_cornerSharpnessD")
    mirror=true
    vector_axis=X
   )
   RigVMFunction_MathBoolNot = &RigVMFunction_MathBoolNot(Value=RigVMVariableNode() {variable=Is V2}.Value)
   RigVMFunction_ControlFlowBranch = &RigVMFunction_ControlFlowBranch(Condition=@RigVMFunction_MathBoolNot.Result, ExecuteContext=@RigVMFunction_Sequence.C)
   FN_FloatCtrlToCurve_FNC_1 = RigVMFunctionReferenceNode(
    Ctrl_L_Token=CTRL_L_
    Ctrl_R_Token=CTRL_R_
    Curve_Side_Token_Last=true
    Pos_Neg_Curves=((Type=Curve,Name="CTRL_expressions_browUINL"),(Type=Curve,Name="CTRL_expressions_browDINL"))
    backward=true
    control=(Type=Control,Name="CTRL_L_brow_inner")
    mirror=true
    vector_axis=X
    ExecuteContext=@RigVMFunction_ControlFlowBranch.True
   )
   > FN_FloatCtrlToCurve_FNC_2 = RigVMFunctionReferenceNode(
    Ctrl_L_Token=CTRL_L_
    Ctrl_R_Token=CTRL_R_
    Curve_Side_Token_Last=true
    Pos_Neg_Curves=((Type=Curve,Name="CTRL_expressions_browUOUTL"),(Type=Curve,Name="CTRL_expressions_browDOUTL"))
    backward=true
    control=(Type=Control,Name="CTRL_L_brow_outer")
    mirror=true
    vector_axis=X
   )
   n3 = &RigVMFunction_ControlFlowBranch(Condition=RigVMVariableNode() {variable=Is V2}.Value, ExecuteContext=@RigVMFunction_Sequence.D)
   n4 = &RigVMFunction_Sequence(ExecuteContext=@n3.True)
   FN_VectorCtrlToCurve_FNC_4 = RigVMFunctionReferenceNode(
    Backward=true
    Control=(Type=Control,Name="CTRL_L_brow_inner")
    Ctrl_L_Token=CTRL_L_
    Ctrl_R_Token=CTRL_R_
    Curve_Side_Token_Last=true
    X_Mirror=true
    X_Pos_Neg_Expr=((Type=Curve,Name="CTRL_expressions_browInnerOutL"),(Type=Curve,Name="CTRL_expressions_browInnerInL"))
    Y_Mirror=true
    Y_Pos_Neg_Expr=((Type=Curve,Name="CTRL_expressions_browUINL"),(Type=Curve,Name="CTRL_expressions_browDINL"))
    Z_Mirror=false
    Z_Pos_Neg_Expr=()
    ExecuteContext=@n4.A
   )
   > FN_VectorCtrlToCurve_FNC_6 = RigVMFunctionReferenceNode(
    Backward=true
    Control=(Type=Control,Name="CTRL_L_brow_innerMid")
    Ctrl_L_Token=CTRL_L_
    Ctrl_R_Token=CTRL_R_
    Curve_Side_Token_Last=true
    X_Mirror=true
    X_Pos_Neg_Expr=((Type=Curve,Name="CTRL_expressions_browInnerMidOutL"),(Type=Curve,Name="CTRL_expressions_browInnerMidInL"))
    Y_Mirror=true
    Y_Pos_Neg_Expr=((Type=Curve,Name="CTRL_expressions_browInnerMidUpL"),(Type=Curve,Name="CTRL_expressions_browInnerMidDownL"))
    Z_Mirror=false
    Z_Pos_Neg_Expr=()
   )
   > FN_VectorCtrlToCurve_FNC_7 = RigVMFunctionReferenceNode(
    Backward=true
    Control=(Type=Control,Name="CTRL_L_brow_outerMid")
    Ctrl_L_Token=CTRL_L_
    Ctrl_R_Token=CTRL_R_
    Curve_Side_Token_Last=true
    X_Mirror=true
    X_Pos_Neg_Expr=((Type=Curve,Name="CTRL_expressions_browOuterMidOutL"),(Type=Curve,Name="CTRL_expressions_browOuterMidInL"))
    Y_Mirror=true
    Y_Pos_Neg_Expr=((Type=Curve,Name="CTRL_expressions_browOuterMidUpL"),(Type=Curve,Name="CTRL_expressions_browOuterMidDownL"))
    Z_Mirror=false
    Z_Pos_Neg_Expr=()
   )
   > FN_FloatCtrlToCurve_FNC_3 = RigVMFunctionReferenceNode(
    Ctrl_L_Token=CTRL_L_
    Ctrl_R_Token=CTRL_R_
    Curve_Side_Token_Last=true
    Pos_Neg_Curves=((Type=Curve,Name="CTRL_expressions_browUOUTL"),(Type=Curve,Name="CTRL_expressions_browDOUTL"))
    backward=true
    control=(Type=Control,Name="CTRL_L_brow_outer")
    mirror=true
    vector_axis=X
   )
   > FN_FloatCtrlToCurve_FNC_5 = RigVMFunctionReferenceNode(
    Ctrl_L_Token=CTRL_L_
    Ctrl_R_Token=CTRL_R_
    Curve_Side_Token_Last=true
    Pos_Neg_Curves=((Type=Curve,Name="CTRL_expressions_eyeFaceScrunchL"))
    backward=true
    control=(Type=Control,Name="CTRL_L_eye_faceScrunch")
    mirror=true
    vector_axis=X
   )
   > FN_FloatCtrlToCurve_FNC_6 = RigVMFunctionReferenceNode(
    Ctrl_L_Token=CTRL_L_
    Ctrl_R_Token=CTRL_R_
    Curve_Side_Token_Last=true
    Pos_Neg_Curves=((Type=Curve,Name="CTRL_expressions_noseNasolabialDeepenL"))
    backward=true
    control=(Type=Control,Name="CTRL_L_nose_nasolabialDeepen")
    mirror=true
    vector_axis=X
   )
   FN_FloatCtrlToCurve_FNC_7 = RigVMFunctionReferenceNode(
    Ctrl_L_Token=CTRL_L_
    Ctrl_R_Token=CTRL_R_
    Curve_Side_Token_Last=true
    Pos_Neg_Curves=((Type=Curve,Name="CTRL_expressions_mouthLipsThinUL"),(Type=Curve,Name="CTRL_expressions_mouthLipsThickUL"))
    backward=true
    control=(Type=Control,Name="CTRL_L_mouth_thicknessU")
    mirror=true
    vector_axis=X
    ExecuteContext=@n4.B
   )
   > FN_FloatCtrlToCurve_FNC_8 = RigVMFunctionReferenceNode(
    Ctrl_L_Token=CTRL_L_
    Ctrl_R_Token=CTRL_R_
    Curve_Side_Token_Last=true
    Pos_Neg_Curves=((Type=Curve,Name="CTRL_expressions_mouthLipsThinDL"),(Type=Curve,Name="CTRL_expressions_mouthLipsThickDL"))
    backward=true
    control=(Type=Control,Name="CTRL_L_mouth_thicknessD")
    mirror=true
    vector_axis=X
   )
   > FN_FloatCtrlToCurve_FNC_15 = RigVMFunctionReferenceNode(
    Ctrl_L_Token=CTRL_L_
    Ctrl_R_Token=CTRL_R_
    Curve_Side_Token_Last=true
    Pos_Neg_Curves=((Type=Curve,Name="CTRL_expressions_mouthLipsThinInwardUL"),(Type=Curve,Name="CTRL_expressions_mouthLipsThickInwardUL"))
    backward=true
    control=(Type=Control,Name="CTRL_L_mouth_thicknessInwardU")
    mirror=true
    vector_axis=X
   )
   > FN_FloatCtrlToCurve_FNC_16 = RigVMFunctionReferenceNode(
    Ctrl_L_Token=CTRL_L_
    Ctrl_R_Token=CTRL_R_
    Curve_Side_Token_Last=true
    Pos_Neg_Curves=((Type=Curve,Name="CTRL_expressions_mouthLipsThinInwardDL"),(Type=Curve,Name="CTRL_expressions_mouthLipsThickInwardDL"))
    backward=true
    control=(Type=Control,Name="CTRL_L_mouth_thicknessInwardD")
    mirror=true
    vector_axis=X
   )
   > FN_FloatCtrlToCurve_FNC_11 = RigVMFunctionReferenceNode(
    Ctrl_L_Token=CTRL_L_
    Ctrl_R_Token=CTRL_R_
    Curve_Side_Token_Last=true
    Pos_Neg_Curves=((Type=Curve,Name="CTRL_expressions_mouthUpperLipRollInL"),(Type=Curve,Name="CTRL_expressions_mouthUpperLipRollOutL"))
    backward=true
    control=(Type=Control,Name="CTRL_L_mouth_lipsRollU")
    mirror=true
    vector_axis=X
   )
   > FN_FloatCtrlToCurve_FNC_12 = RigVMFunctionReferenceNode(
    Ctrl_L_Token=CTRL_L_
    Ctrl_R_Token=CTRL_R_
    Curve_Side_Token_Last=true
    Pos_Neg_Curves=((Type=Curve,Name="CTRL_expressions_mouthLowerLipRollInL"),(Type=Curve,Name="CTRL_expressions_mouthLowerLipRollOutL"))
    backward=true
    control=(Type=Control,Name="CTRL_L_mouth_lipsRollD")
    mirror=true
    vector_axis=X
   )
   > FN_FloatCtrlToCurve_FNC_13 = RigVMFunctionReferenceNode(
    Ctrl_L_Token=CTRL_L_
    Ctrl_R_Token=CTRL_R_
    Curve_Side_Token_Last=true
    Pos_Neg_Curves=((Type=Curve,Name="CTRL_expressions_mouthLipsPushUL"),(Type=Curve,Name="CTRL_expressions_mouthLipsPullUL"))
    backward=true
    control=(Type=Control,Name="CTRL_L_mouth_pushPullU")
    mirror=true
    vector_axis=X
   )
   > FN_FloatCtrlToCurve_FNC_14 = RigVMFunctionReferenceNode(
    Ctrl_L_Token=CTRL_L_
    Ctrl_R_Token=CTRL_R_
    Curve_Side_Token_Last=true
    Pos_Neg_Curves=((Type=Curve,Name="CTRL_expressions_mouthLipsPushDL"),(Type=Curve,Name="CTRL_expressions_mouthLipsPullDL"))
    backward=true
    control=(Type=Control,Name="CTRL_L_mouth_pushPullD")
    mirror=true
    vector_axis=X
   )
   RigVMFunctionReturnNode(ExecuteContext=@n2.B)
  ]}
 CRFL FN MH Tweaker Forward v01_ContainedGraph {nodes [
   RigVMFunctionEntryNode()
   > RigVMFunction_Sequence_1 = RigVMAggregateNode()
   RigVMFunction_Sequence = RigVMAggregateNode(ExecuteContext=@RigVMFunction_Sequence_1.A)
   FN_FloatCtrlToCurve_FNC_4 = RigVMFunctionReferenceNode(
    Ctrl_L_Token=CTRL_L_
    Ctrl_R_Token=CTRL_R_
    Curve_Side_Token_Last=true
    Pos_Neg_Curves=((Type=Curve,Name="CTRL_expressions_eyeEyelidsReduce"))
    backward=False
    control=(Type=Control,Name="CTRL_C_eyelidsReduce")
    mirror=False
    vector_axis=X
    ExecuteContext=@RigVMFunction_Sequence.A
   )
   > FN_VectorCtrlToCurve_FNC_11 = RigVMFunctionReferenceNode(
    Backward=false
    Control=(Type=Control,Name="CTRL_L_eye_eyelidU")
    Ctrl_L_Token=CTRL_L_
    Ctrl_R_Token=CTRL_R_
    Curve_Side_Token_Last=true
    X_Mirror=false
    X_Pos_Neg_Expr=((Type=Curve,Name="CTRL_expressions_eyelidRollNegUL"),(Type=Curve,Name="CTRL_expressions_eyelidRollPosUL"))
    Y_Mirror=true
    Y_Pos_Neg_Expr=((Type=Curve,Name="CTRL_expressions_eyeRelaxL"),(Type=Curve,Name="CTRL_expressions_eyeUpperLidUpL"))
    Z_Mirror=false
    Z_Pos_Neg_Expr=()
   )
   > FN_VectorCtrlToCurve_FNC_12 = RigVMFunctionReferenceNode(
    Backward=false
    Control=(Type=Control,Name="CTRL_R_eye_eyelidU")
    Ctrl_L_Token=CTRL_L_
    Ctrl_R_Token=CTRL_R_
    Curve_Side_Token_Last=true
    X_Mirror=false
    X_Pos_Neg_Expr=((Type=Curve,Name="CTRL_expressions_eyelidRollPosUR"),(Type=Curve,Name="CTRL_expressions_eyelidRollNegUR"))
    Y_Mirror=false
    Y_Pos_Neg_Expr=((Type=Curve,Name="None"),(Type=Curve,Name="None"))
    Z_Mirror=false
    Z_Pos_Neg_Expr=()
   )
   > FN_VectorCtrlToCurve_FNC_13 = RigVMFunctionReferenceNode(
    Backward=false
    Control=(Type=Control,Name="CTRL_L_eye_eyelidD")
    Ctrl_L_Token=CTRL_L_
    Ctrl_R_Token=CTRL_R_
    Curve_Side_Token_Last=true
    X_Mirror=false
    X_Pos_Neg_Expr=((Type=Curve,Name="CTRL_expressions_eyelidRollNegDL"),(Type=Curve,Name="CTRL_expressions_eyelidRollPosDL"))
    Y_Mirror=true
    Y_Pos_Neg_Expr=((Type=Curve,Name="CTRL_expressions_eyeLowerLidUpL"),(Type=Curve,Name="CTRL_expressions_eyeLowerLidDownL"))
    Z_Mirror=false
    Z_Pos_Neg_Expr=()
   )
   > FN_VectorCtrlToCurve_FNC_14 = RigVMFunctionReferenceNode(
    Backward=false
    Control=(Type=Control,Name="CTRL_R_eye_eyelidD")
    Ctrl_L_Token=CTRL_L_
    Ctrl_R_Token=CTRL_R_
    Curve_Side_Token_Last=true
    X_Mirror=false
    X_Pos_Neg_Expr=((Type=Curve,Name="CTRL_expressions_eyelidRollPosDR"),(Type=Curve,Name="CTRL_expressions_eyelidRollNegDR"))
    Y_Mirror=False
    Y_Pos_Neg_Expr=((Type=Curve,Name="None"),(Type=Curve,Name="None"))
    Z_Mirror=false
    Z_Pos_Neg_Expr=()
   )
   > FN_VectorCtrlToCurve_FNC_5 = RigVMFunctionReferenceNode(
    Backward=false
    Control=(Type=Control,Name="CTRL_C_nose")
    Ctrl_L_Token=CTRL_L_
    Ctrl_R_Token=CTRL_R_
    Curve_Side_Token_Last=true
    X_Mirror=false
    X_Pos_Neg_Expr=((Type=Curve,Name="CTRL_expressions_noseRotateL"),(Type=Curve,Name="CTRL_expressions_noseRotateR"))
    Y_Mirror=false
    Y_Pos_Neg_Expr=((Type=Curve,Name="CTRL_expressions_noseSquash"),(Type=Curve,Name="CTRL_expressions_noseStretch"))
    Z_Mirror=false
    Z_Pos_Neg_Expr=()
   )
   FN_FloatCtrlToCurve_FNC_27 = RigVMFunctionReferenceNode(
    Ctrl_L_Token=CTRL_L_
    Ctrl_R_Token=CTRL_R_
    Curve_Side_Token_Last=true
    Pos_Neg_Curves=((Type=Curve,Name="CTRL_expressions_teethBackU"),(Type=Curve,Name="CTRL_expressions_teethFwdU"))
    backward=False
    control=(Type=Control,Name="CTRL_C_teeth_fwdBackU")
    mirror=false
    vector_axis=X
    ExecuteContext=@RigVMFunction_Sequence.B
   )
   > FN_FloatCtrlToCurve_FNC_28 = RigVMFunctionReferenceNode(
    Ctrl_L_Token=CTRL_L_
    Ctrl_R_Token=CTRL_R_
    Curve_Side_Token_Last=true
    Pos_Neg_Curves=((Type=Curve,Name="CTRL_expressions_teethBackD"),(Type=Curve,Name="CTRL_expressions_teethFwdD"))
    backward=False
    control=(Type=Control,Name="CTRL_C_teeth_fwdBackD")
    mirror=false
    vector_axis=X
   )
   > FN_VectorCtrlToCurve_FNC_9 = RigVMFunctionReferenceNode(
    Backward=false
    Control=(Type=Control,Name="CTRL_C_teethU")
    Ctrl_L_Token=CTRL_L_
    Ctrl_R_Token=CTRL_R_
    Curve_Side_Token_Last=true
    X_Mirror=false
    X_Pos_Neg_Expr=((Type=Curve,Name="CTRL_expressions_teethLeftU"),(Type=Curve,Name="CTRL_expressions_teethRightU"))
    Y_Mirror=false
    Y_Pos_Neg_Expr=((Type=Curve,Name="CTRL_expressions_teethUpU"),(Type=Curve,Name="CTRL_expressions_teethDownU"))
    Z_Mirror=false
    Z_Pos_Neg_Expr=()
   )
   > FN_VectorCtrlToCurve_FNC_10 = RigVMFunctionReferenceNode(
    Backward=false
    Control=(Type=Control,Name="CTRL_C_teethD")
    Ctrl_L_Token=CTRL_L_
    Ctrl_R_Token=CTRL_R_
    Curve_Side_Token_Last=true
    X_Mirror=false
    X_Pos_Neg_Expr=((Type=Curve,Name="CTRL_expressions_teethLeftD"),(Type=Curve,Name="CTRL_expressions_teethRightD"))
    Y_Mirror=false
    Y_Pos_Neg_Expr=((Type=Curve,Name="CTRL_expressions_teethUpD"),(Type=Curve,Name="CTRL_expressions_teethDownD"))
    Z_Mirror=false
    Z_Pos_Neg_Expr=()
   )
   > FN_FloatCtrlToCurve_FNC_29 = RigVMFunctionReferenceNode(
    Ctrl_L_Token=CTRL_L_
    Ctrl_R_Token=CTRL_R_
    Curve_Side_Token_Last=true
    Pos_Neg_Curves=((Type=Curve,Name="CTRL_expressions_teethRotateFwdU"),(Type=Curve,Name="CTRL_expressions_teethRotateBackU"))
    backward=False
    control=(Type=Control,Name="CTRL_C_teeth_rotateU")
    mirror=false
    vector_axis=X
   )
   > FN_FloatCtrlToCurve_FNC_30 = RigVMFunctionReferenceNode(
    Ctrl_L_Token=CTRL_L_
    Ctrl_R_Token=CTRL_R_
    Curve_Side_Token_Last=true
    Pos_Neg_Curves=((Type=Curve,Name="CTRL_expressions_teethRotateFwdD"),(Type=Curve,Name="CTRL_expressions_teethRotateBackD"))
    backward=False
    control=(Type=Control,Name="CTRL_C_teeth_rotateD")
    mirror=false
    vector_axis=X
   )
   > FN_FloatCtrlToCurve_FNC_25 = RigVMFunctionReferenceNode(
    Ctrl_L_Token=CTRL_L_
    Ctrl_R_Token=CTRL_R_
    Curve_Side_Token_Last=true
    Pos_Neg_Curves=((Type=Curve,Name="CTRL_expressions_mouthCornerSharpenUL"),(Type=Curve,Name="CTRL_expressions_mouthCornerRounderUL"))
    backward=False
    control=(Type=Control,Name="CTRL_L_mouth_cornerSharpnessU")
    mirror=true
    vector_axis=X
   )
   > FN_FloatCtrlToCurve_FNC_26 = RigVMFunctionReferenceNode(
    Ctrl_L_Token=CTRL_L_
    Ctrl_R_Token=CTRL_R_
    Curve_Side_Token_Last=true
    Pos_Neg_Curves=((Type=Curve,Name="CTRL_expressions_mouthCornerSharpenDL"),(Type=Curve,Name="CTRL_expressions_mouthCornerRounderDL"))
    backward=False
    control=(Type=Control,Name="CTRL_L_mouth_cornerSharpnessD")
    mirror=true
    vector_axis=X
   )
   > FN_FloatCtrlToCurve_FNC_23 = RigVMFunctionReferenceNode(
    Ctrl_L_Token=CTRL_L_
    Ctrl_R_Token=CTRL_R_
    Curve_Side_Token_Last=true
    Pos_Neg_Curves=((Type=Curve,Name="CTRL_expressions_mouthUpperLipTowardsTeethL"))
    backward=False
    control=(Type=Control,Name="CTRL_L_mouth_lipsTowardsTeethU")
    mirror=true
    vector_axis=X
   )
   > FN_FloatCtrlToCurve_FNC_24 = RigVMFunctionReferenceNode(
    Ctrl_L_Token=CTRL_L_
    Ctrl_R_Token=CTRL_R_
    Curve_Side_Token_Last=true
    Pos_Neg_Curves=((Type=Curve,Name="CTRL_expressions_mouthLowerLipTowardsTeethL"),(Type=Curve,Name="None"))
    backward=False
    control=(Type=Control,Name="CTRL_L_mouth_lipsTowardsTeethD")
    mirror=true
    vector_axis=X
   )
   > FN_VectorCtrlToCurve_FNC_8 = RigVMFunctionReferenceNode(
    Backward=false
    Control=(Type=Control,Name="CTRL_L_mouth_corner")
    Ctrl_L_Token=CTRL_L_
    Ctrl_R_Token=CTRL_R_
    Curve_Side_Token_Last=true
    X_Mirror=true
    X_Pos_Neg_Expr=((Type=Curve,Name="CTRL_expressions_mouthCornerWideL"),(Type=Curve,Name="CTRL_expressions_mouthCornerNarrowL"))
    Y_Mirror=true
    Y_Pos_Neg_Expr=((Type=Curve,Name="CTRL_expressions_mouthCornerUpL"),(Type=Curve,Name="CTRL_expressions_mouthCornerDownL"))
    Z_Mirror=false
    Z_Pos_Neg_Expr=()
   )
   RigVMFunction_MathBoolNot = &RigVMFunction_MathBoolNot(Value=RigVMVariableNode() {variable=Is V2}.Value)
   RigVMFunction_ControlFlowBranch = &RigVMFunction_ControlFlowBranch(Condition=@RigVMFunction_MathBoolNot.Result, ExecuteContext=@RigVMFunction_Sequence.C)
   FN_FloatCtrlToCurve_FNC_5 = RigVMFunctionReferenceNode(
    Ctrl_L_Token=CTRL_L_
    Ctrl_R_Token=CTRL_R_
    Curve_Side_Token_Last=true
    Pos_Neg_Curves=((Type=Curve,Name="CTRL_expressions_browUINL"),(Type=Curve,Name="CTRL_expressions_browDINL"))
    backward=False
    control=(Type=Control,Name="CTRL_L_brow_inner")
    mirror=true
    vector_axis=X
    ExecuteContext=@RigVMFunction_ControlFlowBranch.True
   )
   > FN_FloatCtrlToCurve_FNC_6 = RigVMFunctionReferenceNode(
    Ctrl_L_Token=CTRL_L_
    Ctrl_R_Token=CTRL_R_
    Curve_Side_Token_Last=true
    Pos_Neg_Curves=((Type=Curve,Name="CTRL_expressions_browUOUTL"),(Type=Curve,Name="CTRL_expressions_browDOUTL"))
    backward=False
    control=(Type=Control,Name="CTRL_L_brow_outer")
    mirror=true
    vector_axis=X
   )
   n1 = &RigVMFunction_ControlFlowBranch(Condition=RigVMVariableNode() {variable=Is V2}.Value, ExecuteContext=@RigVMFunction_Sequence.D)
   n2 = &RigVMFunction_Sequence(ExecuteContext=@n1.True)
   FN_VectorCtrlToCurve_FNC_4 = RigVMFunctionReferenceNode(
    Backward=false
    Control=(Type=Control,Name="CTRL_L_brow_inner")
    Ctrl_L_Token=CTRL_L_
    Ctrl_R_Token=CTRL_R_
    Curve_Side_Token_Last=true
    X_Mirror=true
    X_Pos_Neg_Expr=((Type=Curve,Name="CTRL_expressions_browInnerOutL"),(Type=Curve,Name="CTRL_expressions_browInnerInL"))
    Y_Mirror=true
    Y_Pos_Neg_Expr=((Type=Curve,Name="CTRL_expressions_browUINL"),(Type=Curve,Name="CTRL_expressions_browDINL"))
    Z_Mirror=false
    Z_Pos_Neg_Expr=()
    ExecuteContext=@n2.A
   )
   > FN_VectorCtrlToCurve_FNC_6 = RigVMFunctionReferenceNode(
    Backward=false
    Control=(Type=Control,Name="CTRL_L_brow_innerMid")
    Ctrl_L_Token=CTRL_L_
    Ctrl_R_Token=CTRL_R_
    Curve_Side_Token_Last=true
    X_Mirror=true
    X_Pos_Neg_Expr=((Type=Curve,Name="CTRL_expressions_browInnerMidOutL"),(Type=Curve,Name="CTRL_expressions_browInnerMidInL"))
    Y_Mirror=true
    Y_Pos_Neg_Expr=((Type=Curve,Name="CTRL_expressions_browInnerMidUpL"),(Type=Curve,Name="CTRL_expressions_browInnerMidDownL"))
    Z_Mirror=false
    Z_Pos_Neg_Expr=()
   )
   > FN_VectorCtrlToCurve_FNC_7 = RigVMFunctionReferenceNode(
    Backward=false
    Control=(Type=Control,Name="CTRL_L_brow_outerMid")
    Ctrl_L_Token=CTRL_L_
    Ctrl_R_Token=CTRL_R_
    Curve_Side_Token_Last=true
    X_Mirror=true
    X_Pos_Neg_Expr=((Type=Curve,Name="CTRL_expressions_browOuterMidOutL"),(Type=Curve,Name="CTRL_expressions_browOuterMidInL"))
    Y_Mirror=true
    Y_Pos_Neg_Expr=((Type=Curve,Name="CTRL_expressions_browOuterMidUpL"),(Type=Curve,Name="CTRL_expressions_browOuterMidDownL"))
    Z_Mirror=false
    Z_Pos_Neg_Expr=()
   )
   > FN_FloatCtrlToCurve_FNC_1 = RigVMFunctionReferenceNode(
    Ctrl_L_Token=CTRL_L_
    Ctrl_R_Token=CTRL_R_
    Curve_Side_Token_Last=true
    Pos_Neg_Curves=((Type=Curve,Name="CTRL_expressions_browUOUTL"),(Type=Curve,Name="CTRL_expressions_browDOUTL"))
    backward=False
    control=(Type=Control,Name="CTRL_L_brow_outer")
    mirror=true
    vector_axis=X
   )
   > FN_FloatCtrlToCurve_FNC_2 = RigVMFunctionReferenceNode(
    Ctrl_L_Token=CTRL_L_
    Ctrl_R_Token=CTRL_R_
    Curve_Side_Token_Last=true
    Pos_Neg_Curves=((Type=Curve,Name="CTRL_expressions_eyeFaceScrunchL"))
    backward=False
    control=(Type=Control,Name="CTRL_L_eye_faceScrunch")
    mirror=true
    vector_axis=X
   )
   > FN_FloatCtrlToCurve_FNC_3 = RigVMFunctionReferenceNode(
    Ctrl_L_Token=CTRL_L_
    Ctrl_R_Token=CTRL_R_
    Curve_Side_Token_Last=true
    Pos_Neg_Curves=((Type=Curve,Name="CTRL_expressions_noseNasolabialDeepenL"))
    backward=False
    control=(Type=Control,Name="CTRL_L_nose_nasolabialDeepen")
    mirror=true
    vector_axis=X
   )
   FN_FloatCtrlToCurve_FNC_7 = RigVMFunctionReferenceNode(
    Ctrl_L_Token=CTRL_L_
    Ctrl_R_Token=CTRL_R_
    Curve_Side_Token_Last=true
    Pos_Neg_Curves=((Type=Curve,Name="CTRL_expressions_mouthLipsThinUL"),(Type=Curve,Name="CTRL_expressions_mouthLipsThickUL"))
    backward=False
    control=(Type=Control,Name="CTRL_L_mouth_thicknessU")
    mirror=true
    vector_axis=X
    ExecuteContext=@n2.B
   )
   > FN_FloatCtrlToCurve_FNC_8 = RigVMFunctionReferenceNode(
    Ctrl_L_Token=CTRL_L_
    Ctrl_R_Token=CTRL_R_
    Curve_Side_Token_Last=true
    Pos_Neg_Curves=((Type=Curve,Name="CTRL_expressions_mouthLipsThinDL"),(Type=Curve,Name="CTRL_expressions_mouthLipsThickDL"))
    backward=False
    control=(Type=Control,Name="CTRL_L_mouth_thicknessD")
    mirror=true
    vector_axis=X
   )
   > FN_FloatCtrlToCurve_FNC_9 = RigVMFunctionReferenceNode(
    Ctrl_L_Token=CTRL_L_
    Ctrl_R_Token=CTRL_R_
    Curve_Side_Token_Last=true
    Pos_Neg_Curves=((Type=Curve,Name="CTRL_expressions_mouthLipsThinInwardUL"),(Type=Curve,Name="CTRL_expressions_mouthLipsThickInwardUL"))
    backward=False
    control=(Type=Control,Name="CTRL_L_mouth_thicknessInwardU")
    mirror=true
    vector_axis=X
   )
   > FN_FloatCtrlToCurve_FNC_10 = RigVMFunctionReferenceNode(
    Ctrl_L_Token=CTRL_L_
    Ctrl_R_Token=CTRL_R_
    Curve_Side_Token_Last=true
    Pos_Neg_Curves=((Type=Curve,Name="CTRL_expressions_mouthLipsThinInwardDL"),(Type=Curve,Name="CTRL_expressions_mouthLipsThickInwardDL"))
    backward=False
    control=(Type=Control,Name="CTRL_L_mouth_thicknessInwardD")
    mirror=true
    vector_axis=X
   )
   > FN_FloatCtrlToCurve_FNC_11 = RigVMFunctionReferenceNode(
    Ctrl_L_Token=CTRL_L_
    Ctrl_R_Token=CTRL_R_
    Curve_Side_Token_Last=true
    Pos_Neg_Curves=((Type=Curve,Name="CTRL_expressions_mouthUpperLipRollInL"),(Type=Curve,Name="CTRL_expressions_mouthUpperLipRollOutL"))
    backward=False
    control=(Type=Control,Name="CTRL_L_mouth_lipsRollU")
    mirror=true
    vector_axis=X
   )
   > FN_FloatCtrlToCurve_FNC_12 = RigVMFunctionReferenceNode(
    Ctrl_L_Token=CTRL_L_
    Ctrl_R_Token=CTRL_R_
    Curve_Side_Token_Last=true
    Pos_Neg_Curves=((Type=Curve,Name="CTRL_expressions_mouthLowerLipRollInL"),(Type=Curve,Name="CTRL_expressions_mouthLowerLipRollOutL"))
    backward=False
    control=(Type=Control,Name="CTRL_L_mouth_lipsRollD")
    mirror=true
    vector_axis=X
   )
   > FN_FloatCtrlToCurve_FNC_13 = RigVMFunctionReferenceNode(
    Ctrl_L_Token=CTRL_L_
    Ctrl_R_Token=CTRL_R_
    Curve_Side_Token_Last=true
    Pos_Neg_Curves=((Type=Curve,Name="CTRL_expressions_mouthLipsPushUL"),(Type=Curve,Name="CTRL_expressions_mouthLipsPullUL"))
    backward=False
    control=(Type=Control,Name="CTRL_L_mouth_pushPullU")
    mirror=true
    vector_axis=X
   )
   > FN_FloatCtrlToCurve_FNC_14 = RigVMFunctionReferenceNode(
    Ctrl_L_Token=CTRL_L_
    Ctrl_R_Token=CTRL_R_
    Curve_Side_Token_Last=true
    Pos_Neg_Curves=((Type=Curve,Name="CTRL_expressions_mouthLipsPushDL"),(Type=Curve,Name="CTRL_expressions_mouthLipsPullDL"))
    backward=False
    control=(Type=Control,Name="CTRL_L_mouth_pushPullD")
    mirror=true
    vector_axis=X
   )
   RigVMFunctionReturnNode(ExecuteContext=@RigVMFunction_Sequence_1.B)
  ]}
 CollapseNode_1_ContainedGraph {nodes [
   n1 = RigVMFunctionEntryNode()
   /Script/ControlRig.RigUnit_HierarchyAddControlFloat(
    InitialValue=0.0
    Name=CTRL_L_mouth_pressU
    OffsetSpace=LocalSpace
    OffsetTransform=(Rotation=(X=-0.5771,Y=0.5771,Z=-0.4086,W=0.4086),Translation=(X=7.09,Y=0.0221,Z=4.6303),Scale3D=(X=1,Y=1,Z=1))
    Settings="(PrimaryAxis=Y,bIsScale=False,Limits=(Limit=(bMinimum=True,bMaximum=True),MinValue=0,MaxValue=1,bDrawLimits=True),Shape=(bVisible=True,Name=\"DefaultGizmoLibraryNormalized.Sphere_Solid\",Color=(R=1,G=1,B=0,A=0),Transform=(Rotation=(X=0,Y=0,Z=0,W=1),Translati ...(truncated)"
    ExecutePin=@n1.ExecutePin
    Parent=@n1.Parent
   ) {
    is=struct RigVMUnitNode
    template="SpawnControl::Execute(in InitialValue,in Settings,in OffsetTransform,in OffsetSpace,in Parent,in Name,out Item)"
   }
   > &RigUnit_HierarchyAddControlFloat(
    InitialValue=0.0
    Name=CTRL_L_mouth_pressD
    OffsetSpace=LocalSpace
    OffsetTransform=(Rotation=(X=0.5797,Y=-0.5797,Z=-0.4049,W=0.4049),Translation=(X=7.09,Y=0.0221,Z=2.4568),Scale3D=(X=1,Y=1,Z=1))
    Settings="(PrimaryAxis=Y,bIsScale=False,Limits=(Limit=(bMinimum=True,bMaximum=True),MinValue=0,MaxValue=1,bDrawLimits=True),Shape=(bVisible=True,Name=\"DefaultGizmoLibraryNormalized.Sphere_Solid\",Color=(R=1,G=1,B=0,A=0),Transform=(Rotation=(X=0,Y=0,Z=0,W=1),Translati ...(truncated)"
    Parent=@n1.Parent
   ) {template="SpawnControl::Execute(in InitialValue,in Settings,in OffsetTransform,in OffsetSpace,in Parent,in Name,out Item)"}
   > &RigUnit_HierarchyAddControlFloat(
    InitialValue=0.0
    Name=CTRL_R_mouth_pressU
    OffsetSpace=LocalSpace
    OffsetTransform=(Rotation=(X=-0.4056,Y=0.4056,Z=-0.5792,W=0.5792),Translation=(X=6.1825,Y=0.0221,Z=4.6256),Scale3D=(X=1,Y=1,Z=1))
    Settings="(PrimaryAxis=Y,bIsScale=False,Limits=(Limit=(bMinimum=True,bMaximum=True),MinValue=0,MaxValue=1,bDrawLimits=True),Shape=(bVisible=True,Name=\"DefaultGizmoLibraryNormalized.Sphere_Solid\",Color=(R=1,G=1,B=0,A=0),Transform=(Rotation=(X=0,Y=0,Z=0,W=1),Translati ...(truncated)"
    Parent=@n1.Parent
   ) {template="SpawnControl::Execute(in InitialValue,in Settings,in OffsetTransform,in OffsetSpace,in Parent,in Name,out Item)"}
   > n2 = &RigUnit_HierarchyAddControlFloat(
    InitialValue=0.0
    Name=CTRL_R_mouth_pressD
    OffsetSpace=LocalSpace
    OffsetTransform=(Rotation=(X=0.4056,Y=-0.4056,Z=-0.5792,W=0.5792),Translation=(X=6.1825,Y=0.0221,Z=2.4594),Scale3D=(X=1,Y=1,Z=1))
    Settings="(PrimaryAxis=Y,bIsScale=False,Limits=(Limit=(bMinimum=True,bMaximum=True),MinValue=0,MaxValue=1,bDrawLimits=True),Shape=(bVisible=True,Name=\"DefaultGizmoLibraryNormalized.Sphere_Solid\",Color=(R=1,G=1,B=0,A=0),Transform=(Rotation=(X=0,Y=0,Z=0,W=1),Translati ...(truncated)"
    Parent=@n1.Parent
   ) {template="SpawnControl::Execute(in InitialValue,in Settings,in OffsetTransform,in OffsetSpace,in Parent,in Name,out Item)"}
   RigVMFunctionReturnNode(ExecutePin=@n2.ExecutePin)
  ]}
 CollapseNode_2_ContainedGraph {nodes [
   n1 = RigVMFunctionEntryNode()
   &RigUnit_HierarchyAddControlFloat(
    InitialValue=0.0
    Name=CTRL_L_mouth_funnelU
    OffsetSpace=LocalSpace
    OffsetTransform=(Rotation=(X=-0.6533,Y=0.6533,Z=-0.2706,W=0.2706),Translation=(X=-2.5998,Y=0.0221,Z=5.2205),Scale3D=(X=0.75,Y=0.75,Z=0.75))
    Settings="(PrimaryAxis=Y,bIsScale=False,Limits=(Limit=(bMinimum=True,bMaximum=True),MinValue=0,MaxValue=1,bDrawLimits=True),Shape=(bVisible=True,Name=\"DefaultGizmoLibraryNormalized.Sphere_Solid\",Color=(R=1,G=1,B=0,A=0),Transform=(Rotation=(X=0,Y=0,Z=0,W=1),Translati ...(truncated)"
    ExecutePin=@n1.ExecutePin
    Parent=@n1.Parent
   ) {template="SpawnControl::Execute(in InitialValue,in Settings,in OffsetTransform,in OffsetSpace,in Parent,in Name,out Item)"}
   > &RigUnit_HierarchyAddControlFloat(
    InitialValue=0.0
    Name=CTRL_L_mouth_funnelD
    OffsetSpace=LocalSpace
    OffsetTransform=(Rotation=(X=0.6533,Y=-0.6533,Z=-0.2706,W=0.2706),Translation=(X=-2.599,Y=0.0221,Z=3.94),Scale3D=(X=0.75,Y=0.75,Z=0.75))
    Settings="(PrimaryAxis=Y,bIsScale=False,Limits=(Limit=(bMinimum=True,bMaximum=True),MinValue=0,MaxValue=1,bDrawLimits=True),Shape=(bVisible=True,Name=\"DefaultGizmoLibraryNormalized.Sphere_Solid\",Color=(R=1,G=1,B=0,A=0),Transform=(Rotation=(X=0,Y=0,Z=0,W=1),Translati ...(truncated)"
    Parent=@n1.Parent
   ) {template="SpawnControl::Execute(in InitialValue,in Settings,in OffsetTransform,in OffsetSpace,in Parent,in Name,out Item)"}
   > &RigUnit_HierarchyAddControlFloat(
    InitialValue=0.0
    Name=CTRL_R_mouth_funnelU
    OffsetSpace=LocalSpace
    OffsetTransform=(Rotation=(X=-0.2706,Y=0.2706,Z=-0.6533,W=0.6533),Translation=(X=-3.882,Y=0.0221,Z=5.2239),Scale3D=(X=0.75,Y=0.75,Z=0.75))
    Settings="(PrimaryAxis=Y,bIsScale=False,Limits=(Limit=(bMinimum=True,bMaximum=True),MinValue=0,MaxValue=1,bDrawLimits=True),Shape=(bVisible=True,Name=\"DefaultGizmoLibraryNormalized.Sphere_Solid\",Color=(R=1,G=1,B=0,A=0),Transform=(Rotation=(X=0,Y=0,Z=0,W=1),Translati ...(truncated)"
    Parent=@n1.Parent
   ) {template="SpawnControl::Execute(in InitialValue,in Settings,in OffsetTransform,in OffsetSpace,in Parent,in Name,out Item)"}
   > n2 = &RigUnit_HierarchyAddControlFloat(
    InitialValue=0.0
    Name=CTRL_R_mouth_funnelD
    OffsetSpace=LocalSpace
    OffsetTransform=(Rotation=(X=0.2706,Y=-0.2706,Z=-0.6533,W=0.6533),Translation=(X=-3.8839,Y=0.0221,Z=3.938),Scale3D=(X=0.75,Y=0.75,Z=0.75))
    Settings="(PrimaryAxis=Y,bIsScale=False,Limits=(Limit=(bMinimum=True,bMaximum=True),MinValue=0,MaxValue=1,bDrawLimits=True),Shape=(bVisible=True,Name=\"DefaultGizmoLibraryNormalized.Sphere_Solid\",Color=(R=1,G=1,B=0,A=0),Transform=(Rotation=(X=0,Y=0,Z=0,W=1),Translati ...(truncated)"
    Parent=@n1.Parent
   ) {template="SpawnControl::Execute(in InitialValue,in Settings,in OffsetTransform,in OffsetSpace,in Parent,in Name,out Item)"}
   RigVMFunctionReturnNode(ExecutePin=@n2.ExecutePin)
  ]}
 CollapseNode_3_ContainedGraph {nodes [
   n1 = RigVMFunctionEntryNode()
   &RigUnit_HierarchyAddControlFloat(
    InitialValue=0.0
    Name=CTRL_L_mouth_purseU
    OffsetSpace=LocalSpace
    OffsetTransform=(Rotation=(X=-0.6533,Y=0.6533,Z=-0.2706,W=0.2706),Translation=(X=-4.5377,Y=0.0221,Z=5.2209),Scale3D=(X=0.75,Y=0.75,Z=0.75))
    Settings="(PrimaryAxis=Y,bIsScale=False,Limits=(Limit=(bMinimum=True,bMaximum=True),MinValue=0,MaxValue=1,bDrawLimits=True),Shape=(bVisible=True,Name=\"DefaultGizmoLibraryNormalized.Sphere_Solid\",Color=(R=1,G=1,B=0,A=0),Transform=(Rotation=(X=0,Y=0,Z=0,W=1),Translati ...(truncated)"
    ExecutePin=@n1.ExecutePin
    Parent=@n1.Parent
   ) {template="SpawnControl::Execute(in InitialValue,in Settings,in OffsetTransform,in OffsetSpace,in Parent,in Name,out Item)"}
   > &RigUnit_HierarchyAddControlFloat(
    InitialValue=0.0
    Name=CTRL_L_mouth_purseD
    OffsetSpace=LocalSpace
    OffsetTransform=(Rotation=(X=-0.6533,Y=0.6533,Z=0.2706,W=-0.2706),Translation=(X=-4.5369,Y=0.0221,Z=3.9396),Scale3D=(X=0.75,Y=0.75,Z=0.75))
    Settings="(PrimaryAxis=Y,bIsScale=False,Limits=(Limit=(bMinimum=True,bMaximum=True),MinValue=0,MaxValue=1,bDrawLimits=True),Shape=(bVisible=True,Name=\"DefaultGizmoLibraryNormalized.Sphere_Solid\",Color=(R=1,G=1,B=0,A=0),Transform=(Rotation=(X=0,Y=0,Z=0,W=1),Translati ...(truncated)"
    Parent=@n1.Parent
   ) {template="SpawnControl::Execute(in InitialValue,in Settings,in OffsetTransform,in OffsetSpace,in Parent,in Name,out Item)"}
   > &RigUnit_HierarchyAddControlFloat(
    InitialValue=0.0
    Name=CTRL_R_mouth_purseU
    OffsetSpace=LocalSpace
    OffsetTransform=(Rotation=(X=-0.2706,Y=0.2706,Z=-0.6533,W=0.6533),Translation=(X=-5.8206,Y=0.0221,Z=5.2243),Scale3D=(X=0.75,Y=0.75,Z=0.75))
    Settings="(PrimaryAxis=Y,bIsScale=False,Limits=(Limit=(bMinimum=True,bMaximum=True),MinValue=0,MaxValue=1,bDrawLimits=True),Shape=(bVisible=True,Name=\"DefaultGizmoLibraryNormalized.Sphere_Solid\",Color=(R=1,G=1,B=0,A=0),Transform=(Rotation=(X=0,Y=0,Z=0,W=1),Translati ...(truncated)"
    Parent=@n1.Parent
   ) {template="SpawnControl::Execute(in InitialValue,in Settings,in OffsetTransform,in OffsetSpace,in Parent,in Name,out Item)"}
   > n2 = &RigUnit_HierarchyAddControlFloat(
    InitialValue=0.0
    Name=CTRL_R_mouth_purseD
    OffsetSpace=LocalSpace
    OffsetTransform=(Rotation=(X=0.2706,Y=-0.2706,Z=-0.6533,W=0.6533),Translation=(X=-5.8225,Y=0.0221,Z=3.9377),Scale3D=(X=0.75,Y=0.75,Z=0.75))
    Settings="(PrimaryAxis=Y,bIsScale=False,Limits=(Limit=(bMinimum=True,bMaximum=True),MinValue=0,MaxValue=1,bDrawLimits=True),Shape=(bVisible=True,Name=\"DefaultGizmoLibraryNormalized.Sphere_Solid\",Color=(R=1,G=1,B=0,A=0),Transform=(Rotation=(X=0,Y=0,Z=0,W=1),Translati ...(truncated)"
    Parent=@n1.Parent
   ) {template="SpawnControl::Execute(in InitialValue,in Settings,in OffsetTransform,in OffsetSpace,in Parent,in Name,out Item)"}
   RigVMFunctionReturnNode(ExecutePin=@n2.ExecutePin)
  ]}
 CollapseNode_ContainedGraph {nodes [
   n1 = RigVMFunctionEntryNode()
   &RigUnit_HierarchyAddControlFloat(
    InitialValue=0.0
    Name=CTRL_L_mouth_lipsTogetherU
    OffsetSpace=LocalSpace
    OffsetTransform=(Rotation=(X=-0.5771,Y=0.5771,Z=-0.4086,W=0.4086),Translation=(X=4.7091,Y=0.0221,Z=4.6303),Scale3D=(X=1,Y=1,Z=1))
    Settings="(PrimaryAxis=Y,bIsScale=False,Limits=(Limit=(bMinimum=True,bMaximum=True),MinValue=0,MaxValue=1,bDrawLimits=True),Shape=(bVisible=True,Name=\"DefaultGizmoLibraryNormalized.Sphere_Solid\",Color=(R=1,G=1,B=0,A=0),Transform=(Rotation=(X=0,Y=0,Z=0,W=1),Translati ...(truncated)"
    ExecutePin=@n1.ExecutePin
    Parent=@n1.Parent
   ) {template="SpawnControl::Execute(in InitialValue,in Settings,in OffsetTransform,in OffsetSpace,in Parent,in Name,out Item)"}
   > &RigUnit_HierarchyAddControlFloat(
    InitialValue=0.0
    Name=CTRL_L_mouth_lipsTogetherD
    OffsetSpace=LocalSpace
    OffsetTransform=(Rotation=(X=0.5797,Y=-0.5797,Z=-0.4049,W=0.4049),Translation=(X=4.7091,Y=0.0221,Z=2.4568),Scale3D=(X=1,Y=1,Z=1))
    Settings="(PrimaryAxis=Y,bIsScale=False,Limits=(Limit=(bMinimum=True,bMaximum=True),MinValue=0,MaxValue=1,bDrawLimits=True),Shape=(bVisible=True,Name=\"DefaultGizmoLibraryNormalized.Sphere_Solid\",Color=(R=1,G=1,B=0,A=0),Transform=(Rotation=(X=0,Y=0,Z=0,W=1),Translati ...(truncated)"
    Parent=@n1.Parent
   ) {template="SpawnControl::Execute(in InitialValue,in Settings,in OffsetTransform,in OffsetSpace,in Parent,in Name,out Item)"}
   > &RigUnit_HierarchyAddControlFloat(
    InitialValue=0.0
    Name=CTRL_R_mouth_lipsTogetherU
    OffsetSpace=LocalSpace
    OffsetTransform=(Rotation=(X=-0.4056,Y=0.4056,Z=-0.5792,W=0.5792),Translation=(X=3.8016,Y=0.0221,Z=4.6256),Scale3D=(X=1,Y=1,Z=1))
    Settings="(PrimaryAxis=Y,bIsScale=False,Limits=(Limit=(bMinimum=True,bMaximum=True),MinValue=0,MaxValue=1,bDrawLimits=True),Shape=(bVisible=True,Name=\"DefaultGizmoLibraryNormalized.Sphere_Solid\",Color=(R=1,G=1,B=0,A=0),Transform=(Rotation=(X=0,Y=0,Z=0,W=1),Translati ...(truncated)"
    Parent=@n1.Parent
   ) {template="SpawnControl::Execute(in InitialValue,in Settings,in OffsetTransform,in OffsetSpace,in Parent,in Name,out Item)"}
   > n2 = &RigUnit_HierarchyAddControlFloat(
    InitialValue=0.0
    Name=CTRL_R_mouth_lipsTogetherD
    OffsetSpace=LocalSpace
    OffsetTransform=(Rotation=(X=0.4056,Y=-0.4056,Z=-0.5792,W=0.5792),Translation=(X=3.8016,Y=0.0221,Z=2.4594),Scale3D=(X=1,Y=1,Z=1))
    Settings="(PrimaryAxis=Y,bIsScale=False,Limits=(Limit=(bMinimum=True,bMaximum=True),MinValue=0,MaxValue=1,bDrawLimits=True),Shape=(bVisible=True,Name=\"DefaultGizmoLibraryNormalized.Sphere_Solid\",Color=(R=1,G=1,B=0,A=0),Transform=(Rotation=(X=0,Y=0,Z=0,W=1),Translati ...(truncated)"
    Parent=@n1.Parent
   ) {template="SpawnControl::Execute(in InitialValue,in Settings,in OffsetTransform,in OffsetSpace,in Parent,in Name,out Item)"}
   RigVMFunctionReturnNode(ExecutePin=@n2.ExecutePin)
  ]}
 CollapseNode_ContainedGraph {nodes [
   n1 = RigVMFunctionEntryNode()
   &RigUnit_HierarchyAddControlFloat(
    InitialValue=0.0
    Name=CTRL_L_mouth_towardsU
    OffsetSpace=LocalSpace
    OffsetTransform=(Rotation=(X=-0.6533,Y=0.6533,Z=-0.2706,W=0.2706),Translation=(X=-6.5202,Y=0.0221,Z=5.2209),Scale3D=(X=0.75,Y=0.75,Z=0.75))
    Settings="(PrimaryAxis=Y,bIsScale=False,Limits=(Limit=(bMinimum=True,bMaximum=True),MinValue=0,MaxValue=1,bDrawLimits=True),Shape=(bVisible=True,Name=\"DefaultGizmoLibraryNormalized.Sphere_Solid\",Color=(R=1,G=1,B=0,A=0),Transform=(Rotation=(X=0,Y=0,Z=0,W=1),Translati ...(truncated)"
    ExecutePin=@n1.ExecutePin
    Parent=@n1.Parent
   ) {template="SpawnControl::Execute(in InitialValue,in Settings,in OffsetTransform,in OffsetSpace,in Parent,in Name,out Item)"}
   > &RigUnit_HierarchyAddControlFloat(
    InitialValue=0.0
    Name=CTRL_L_mouth_towardsD
    OffsetSpace=LocalSpace
    OffsetTransform=(Rotation=(X=-0.6533,Y=0.6533,Z=0.2706,W=-0.2706),Translation=(X=-6.5193,Y=0.0221,Z=3.9396),Scale3D=(X=0.75,Y=0.75,Z=0.75))
    Settings="(PrimaryAxis=Y,bIsScale=False,Limits=(Limit=(bMinimum=True,bMaximum=True),MinValue=0,MaxValue=1,bDrawLimits=True),Shape=(bVisible=True,Name=\"DefaultGizmoLibraryNormalized.Sphere_Solid\",Color=(R=1,G=1,B=0,A=0),Transform=(Rotation=(X=0,Y=0,Z=0,W=1),Translati ...(truncated)"
    Parent=@n1.Parent
   ) {template="SpawnControl::Execute(in InitialValue,in Settings,in OffsetTransform,in OffsetSpace,in Parent,in Name,out Item)"}
   > &RigUnit_HierarchyAddControlFloat(
    InitialValue=0.0
    Name=CTRL_R_mouth_towardsU
    OffsetSpace=LocalSpace
    OffsetTransform=(Rotation=(X=-0.2706,Y=0.2706,Z=-0.6533,W=0.6533),Translation=(X=-7.8031,Y=0.0221,Z=5.2243),Scale3D=(X=0.75,Y=0.75,Z=0.75))
    Settings="(PrimaryAxis=Y,bIsScale=False,Limits=(Limit=(bMinimum=True,bMaximum=True),MinValue=0,MaxValue=1,bDrawLimits=True),Shape=(bVisible=True,Name=\"DefaultGizmoLibraryNormalized.Sphere_Solid\",Color=(R=1,G=1,B=0,A=0),Transform=(Rotation=(X=0,Y=0,Z=0,W=1),Translati ...(truncated)"
    Parent=@n1.Parent
   ) {template="SpawnControl::Execute(in InitialValue,in Settings,in OffsetTransform,in OffsetSpace,in Parent,in Name,out Item)"}
   > n2 = &RigUnit_HierarchyAddControlFloat(
    InitialValue=0.0
    Name=CTRL_R_mouth_towardsD
    OffsetSpace=LocalSpace
    OffsetTransform=(Rotation=(X=0.2706,Y=-0.2706,Z=-0.6533,W=0.6533),Translation=(X=-7.805,Y=0.0221,Z=3.9377),Scale3D=(X=0.75,Y=0.75,Z=0.75))
    Settings="(PrimaryAxis=Y,bIsScale=False,Limits=(Limit=(bMinimum=True,bMaximum=True),MinValue=0,MaxValue=1,bDrawLimits=True),Shape=(bVisible=True,Name=\"DefaultGizmoLibraryNormalized.Sphere_Solid\",Color=(R=1,G=1,B=0,A=0),Transform=(Rotation=(X=0,Y=0,Z=0,W=1),Translati ...(truncated)"
    Parent=@n1.Parent
   ) {template="SpawnControl::Execute(in InitialValue,in Settings,in OffsetTransform,in OffsetSpace,in Parent,in Name,out Item)"}
   RigVMFunctionReturnNode(ExecutePin=@n2.ExecutePin)
  ]}
 CollapseNode_ContainedGraph {nodes [
   n1 = RigVMFunctionEntryNode()
   n2 = &RigUnit_HierarchyAddNull(
    Name=GRP_browGUI
    Space=LocalSpace
    Transform=(Rotation=(X=0,Y=0,Z=0,W=1),Translation=(X=0,Y=0,Z=0),Scale3D=(X=1,Y=1,Z=1))
    ExecutePin=@n1.ExecutePin
    Parent=@n1.Parent
   )
   > &RigUnit_HierarchyAddControlFloat(
    InitialValue=0.0
    Name=CTRL_L_brow_lateral
    OffsetSpace=LocalSpace
    OffsetTransform=(Rotation=(X=0.0886,Y=0.0886,Z=-0.7015,W=-0.7015),Translation=(X=3.2459,Y=0,Z=19.1814),Scale3D=(X=2,Y=2,Z=2))
    Settings="(PrimaryAxis=Y,bIsScale=False,Limits=(Limit=(bMinimum=True,bMaximum=True),MinValue=0,MaxValue=1,bDrawLimits=True),Shape=(bVisible=True,Name=\"DefaultGizmoLibraryNormalized.Sphere_Solid\",Color=(R=1,G=1,B=0,A=1),Transform=(Rotation=(X=-0.7071,Y=0,Z=0,W=0.7071 ...(truncated)"
    Parent=@n2.Item
   ) {template="SpawnControl::Execute(in InitialValue,in Settings,in OffsetTransform,in OffsetSpace,in Parent,in Name,out Item)"}
   > &RigUnit_HierarchyAddControlFloat(
    InitialValue=0.0
    Name=CTRL_R_brow_lateral
    OffsetSpace=LocalSpace
    OffsetTransform=(Rotation=(X=0.7017,Y=0.7017,Z=-0.0874,W=-0.0874),Translation=(X=-3.2815,Y=0,Z=19.1814),Scale3D=(X=2,Y=2,Z=2))
    Settings="(PrimaryAxis=Y,bIsScale=False,Limits=(Limit=(bMinimum=True,bMaximum=True),MinValue=0,MaxValue=1,bDrawLimits=True),Shape=(bVisible=True,Name=\"DefaultGizmoLibraryNormalized.Sphere_Solid\",Color=(R=1,G=1,B=0,A=1),Transform=(Rotation=(X=-0.7071,Y=0,Z=0,W=0.7071 ...(truncated)"
    Parent=@n2.Item
   ) {template="SpawnControl::Execute(in InitialValue,in Settings,in OffsetTransform,in OffsetSpace,in Parent,in Name,out Item)"}
   > &RigUnit_HierarchyAddControlFloat(
    InitialValue=0.0
    Name=CTRL_L_brow_down
    OffsetSpace=LocalSpace
    OffsetTransform=(Rotation=(X=0.4307,Y=0.4307,Z=-0.5608,W=-0.5608),Translation=(X=0.9102,Y=0,Z=19.6952),Scale3D=(X=2,Y=2,Z=2))
    Settings="(PrimaryAxis=Y,bIsScale=False,Limits=(Limit=(bMinimum=True,bMaximum=True),MinValue=0,MaxValue=1,bDrawLimits=True),Shape=(bVisible=True,Name=\"DefaultGizmoLibraryNormalized.Sphere_Solid\",Color=(R=1,G=1,B=0,A=1),Transform=(Rotation=(X=-0.7071,Y=0,Z=0,W=0.7071 ...(truncated)"
    Parent=@n2.Item
   ) {template="SpawnControl::Execute(in InitialValue,in Settings,in OffsetTransform,in OffsetSpace,in Parent,in Name,out Item)"}
   > &RigUnit_HierarchyAddControlFloat(
    InitialValue=0.0
    Name=CTRL_R_brow_down
    OffsetSpace=LocalSpace
    OffsetTransform=(Rotation=(X=0.5622,Y=0.5622,Z=-0.4288,W=-0.4288),Translation=(X=-0.9194,Y=0,Z=19.6952),Scale3D=(X=2,Y=2,Z=2))
    Settings="(PrimaryAxis=Y,bIsScale=False,Limits=(Limit=(bMinimum=True,bMaximum=True),MinValue=0,MaxValue=1,bDrawLimits=True),Shape=(bVisible=True,Name=\"DefaultGizmoLibraryNormalized.Sphere_Solid\",Color=(R=1,G=1,B=0,A=1),Transform=(Rotation=(X=-0.7071,Y=0,Z=0,W=0.7071 ...(truncated)"
    Parent=@n2.Item
   ) {template="SpawnControl::Execute(in InitialValue,in Settings,in OffsetTransform,in OffsetSpace,in Parent,in Name,out Item)"}
   > &RigUnit_HierarchyAddControlFloat(
    InitialValue=0.0
    Name=CTRL_L_brow_raiseIn
    OffsetSpace=LocalSpace
    OffsetTransform=(Rotation=(X=0.5317,Y=0.5317,Z=0.4661,W=0.4661),Translation=(X=1.2087,Y=0,Z=20.4561),Scale3D=(X=2,Y=2,Z=2))
    Settings="(PrimaryAxis=Y,bIsScale=False,Limits=(Limit=(bMinimum=True,bMaximum=True),MinValue=0,MaxValue=1,bDrawLimits=True),Shape=(bVisible=True,Name=\"DefaultGizmoLibraryNormalized.Sphere_Solid\",Color=(R=1,G=1,B=0,A=1),Transform=(Rotation=(X=-0.7071,Y=0,Z=0,W=0.7071 ...(truncated)"
    Parent=@n2.Item
   ) {template="SpawnControl::Execute(in InitialValue,in Settings,in OffsetTransform,in OffsetSpace,in Parent,in Name,out Item)"}
   > &RigUnit_HierarchyAddControlFloat(
    InitialValue=0.0
    Name=CTRL_L_brow_raiseOut
    OffsetSpace=LocalSpace
    OffsetTransform=(Rotation=(X=0.5741,Y=0.5741,Z=0.4129,W=0.4129),Translation=(X=2.8573,Y=0,Z=20.1158),Scale3D=(X=2,Y=2,Z=2))
    Settings="(PrimaryAxis=Y,bIsScale=False,Limits=(Limit=(bMinimum=True,bMaximum=True),MinValue=0,MaxValue=1,bDrawLimits=True),Shape=(bVisible=True,Name=\"DefaultGizmoLibraryNormalized.Sphere_Solid\",Color=(R=1,G=1,B=0,A=1),Transform=(Rotation=(X=-0.7071,Y=0,Z=0,W=0.7071 ...(truncated)"
    Parent=@n2.Item
   ) {template="SpawnControl::Execute(in InitialValue,in Settings,in OffsetTransform,in OffsetSpace,in Parent,in Name,out Item)"}
   > &RigUnit_HierarchyAddControlFloat(
    InitialValue=0.0
    Name=CTRL_R_brow_raiseIn
    OffsetSpace=LocalSpace
    OffsetTransform=(Rotation=(X=0.4657,Y=0.4657,Z=0.5321,W=0.5321),Translation=(X=-1.1988,Y=0,Z=20.4528),Scale3D=(X=2,Y=2,Z=2))
    Settings="(PrimaryAxis=Y,bIsScale=False,Limits=(Limit=(bMinimum=True,bMaximum=True),MinValue=0,MaxValue=1,bDrawLimits=True),Shape=(bVisible=True,Name=\"DefaultGizmoLibraryNormalized.Sphere_Solid\",Color=(R=1,G=1,B=0,A=1),Transform=(Rotation=(X=-0.7071,Y=0,Z=0,W=0.7071 ...(truncated)"
    Parent=@n2.Item
   ) {template="SpawnControl::Execute(in InitialValue,in Settings,in OffsetTransform,in OffsetSpace,in Parent,in Name,out Item)"}
   > n3 = &RigUnit_HierarchyAddControlFloat(
    InitialValue=0.0
    Name=CTRL_R_brow_raiseOut
    OffsetSpace=LocalSpace
    OffsetTransform=(Rotation=(X=0.4129,Y=0.4129,Z=0.5741,W=0.5741),Translation=(X=-2.8421,Y=0,Z=20.1209),Scale3D=(X=2,Y=2,Z=2))
    Settings="(PrimaryAxis=Y,bIsScale=False,Limits=(Limit=(bMinimum=True,bMaximum=True),MinValue=0,MaxValue=1,bDrawLimits=True),Shape=(bVisible=True,Name=\"DefaultGizmoLibraryNormalized.Sphere_Solid\",Color=(R=1,G=1,B=0,A=1),Transform=(Rotation=(X=-0.7071,Y=0,Z=0,W=0.7071 ...(truncated)"
    Parent=@n2.Item
   ) {template="SpawnControl::Execute(in InitialValue,in Settings,in OffsetTransform,in OffsetSpace,in Parent,in Name,out Item)"}
   RigVMFunctionReturnNode(ExecutePin=@n3.ExecutePin)
  ]}
 CollapseNode_ContainedGraph {nodes [
   n1 = RigVMFunctionEntryNode()
   &RigUnit_HierarchyAddControlFloat(
    InitialValue=0.0
    Name=CTRL_L_mouth_lipBiteU
    OffsetSpace=LocalSpace
    OffsetTransform=(Rotation=(X=-0.5,Y=0.5,Z=-0.5,W=0.5),Translation=(X=-7.0481,Y=0.0221,Z=1.3765),Scale3D=(X=1,Y=1,Z=1))
    Settings="(PrimaryAxis=Y,bIsScale=False,Limits=(Limit=(bMinimum=True,bMaximum=True),MinValue=0,MaxValue=1,bDrawLimits=True),Shape=(bVisible=True,Name=\"DefaultGizmoLibraryNormalized.Sphere_Solid\",Color=(R=1,G=1,B=0,A=0),Transform=(Rotation=(X=0,Y=0,Z=0,W=1),Translati ...(truncated)"
    ExecutePin=@n1.ExecutePin
    Parent=@n1.Parent
   ) {template="SpawnControl::Execute(in InitialValue,in Settings,in OffsetTransform,in OffsetSpace,in Parent,in Name,out Item)"}
   > &RigUnit_HierarchyAddControlFloat(
    InitialValue=0.0
    Name=CTRL_L_mouth_lipBiteD
    OffsetSpace=LocalSpace
    OffsetTransform=(Rotation=(X=0.5,Y=-0.5,Z=-0.5,W=0.5),Translation=(X=-7.0481,Y=0.0221,Z=-0.797),Scale3D=(X=1,Y=1,Z=1))
    Settings="(PrimaryAxis=Y,bIsScale=False,Limits=(Limit=(bMinimum=True,bMaximum=True),MinValue=0,MaxValue=1,bDrawLimits=True),Shape=(bVisible=True,Name=\"DefaultGizmoLibraryNormalized.Sphere_Solid\",Color=(R=1,G=1,B=0,A=0),Transform=(Rotation=(X=0,Y=0,Z=0,W=1),Translati ...(truncated)"
    Parent=@n1.Parent
   ) {template="SpawnControl::Execute(in InitialValue,in Settings,in OffsetTransform,in OffsetSpace,in Parent,in Name,out Item)"}
   > &RigUnit_HierarchyAddControlFloat(
    InitialValue=0.0
    Name=CTRL_R_mouth_lipBiteU
    OffsetSpace=LocalSpace
    OffsetTransform=(Rotation=(X=-0.5,Y=0.5,Z=-0.5,W=0.5),Translation=(X=-7.5678,Y=0.0221,Z=1.3718),Scale3D=(X=1,Y=1,Z=1))
    Settings="(PrimaryAxis=Y,bIsScale=False,Limits=(Limit=(bMinimum=True,bMaximum=True),MinValue=0,MaxValue=1,bDrawLimits=True),Shape=(bVisible=True,Name=\"DefaultGizmoLibraryNormalized.Sphere_Solid\",Color=(R=1,G=1,B=0,A=0),Transform=(Rotation=(X=0,Y=0,Z=0,W=1),Translati ...(truncated)"
    Parent=@n1.Parent
   ) {template="SpawnControl::Execute(in InitialValue,in Settings,in OffsetTransform,in OffsetSpace,in Parent,in Name,out Item)"}
   > n2 = &RigUnit_HierarchyAddControlFloat(
    InitialValue=0.0
    Name=CTRL_R_mouth_lipBiteD
    OffsetSpace=LocalSpace
    OffsetTransform=(Rotation=(X=0.5,Y=-0.5,Z=-0.5,W=0.5),Translation=(X=-7.5678,Y=0.0221,Z=-0.7944),Scale3D=(X=1,Y=1,Z=1))
    Settings="(PrimaryAxis=Y,bIsScale=False,Limits=(Limit=(bMinimum=True,bMaximum=True),MinValue=0,MaxValue=1,bDrawLimits=True),Shape=(bVisible=True,Name=\"DefaultGizmoLibraryNormalized.Sphere_Solid\",Color=(R=1,G=1,B=0,A=0),Transform=(Rotation=(X=0,Y=0,Z=0,W=1),Translati ...(truncated)"
    Parent=@n1.Parent
   ) {template="SpawnControl::Execute(in InitialValue,in Settings,in OffsetTransform,in OffsetSpace,in Parent,in Name,out Item)"}
   RigVMFunctionReturnNode(ExecutePin=@n2.ExecutePin)
  ]}
 CollapseNode_ContainedGraph {nodes [
   n1 = RigVMFunctionEntryNode()
   /Script/ControlRig.RigUnit_HierarchyAddControlVector2D(
    InitialValue=(X=0,Y=0)
    Name=CTRL_C_eye
    OffsetSpace=LocalSpace
    OffsetTransform=(Rotation=(X=0.7071,Y=0,Z=0,W=0.7071),Translation=(X=0.0003,Y=0,Z=16.289),Scale3D=(X=0.85,Y=0.85,Z=0.85))
    Settings="(PrimaryAxis=Z,Limits=(LimitX=(bMinimum=True,bMaximum=True),LimitY=(bMinimum=True,bMaximum=True),MinValue=(X=-1,Y=-1),MaxValue=(X=1,Y=1),bDrawLimits=True),Shape=(bVisible=True,Name=\"DefaultGizmoLibraryNormalized.Sphere_Solid\",Color=(R=1,G=1,B=0,A=1),Transf ...(truncated)"
    ExecutePin=@n1.ExecutePin
    Parent=@n1.Parent
   ) {
    is=struct RigVMUnitNode
    template="SpawnControl::Execute(in InitialValue,in Settings,in OffsetTransform,in OffsetSpace,in Parent,in Name,out Item)"
   }
   > &RigUnit_HierarchyAddControlFloat(
    InitialValue=0.0
    Name=CTRL_C_eyesConvergence
    OffsetSpace=LocalSpace
    OffsetTransform=(Rotation=(X=0.7071,Y=0.7071,Z=0,W=0),Translation=(X=0,Y=0.0555,Z=17.2975),Scale3D=(X=0.75,Y=0.75,Z=0.75))
    Settings="(PrimaryAxis=Y,bIsScale=False,Limits=(Limit=(bMinimum=True,bMaximum=True),MinValue=-1,MaxValue=1,bDrawLimits=True),Shape=(bVisible=True,Name=\"DefaultGizmoLibraryNormalized.Sphere_Solid\",Color=(R=1,G=1,B=0,A=1),Transform=(Rotation=(X=-0.7071,Y=0,Z=0,W=0.707 ...(truncated)"
    Parent=@n1.Parent
   ) {template="SpawnControl::Execute(in InitialValue,in Settings,in OffsetTransform,in OffsetSpace,in Parent,in Name,out Item)"}
   > &RigUnit_HierarchyAddControlFloat(
    InitialValue=0.0
    Name=CTRL_C_eyelidsReduce
    OffsetSpace=LocalSpace
    OffsetTransform=(Rotation=(X=0.5,Y=0.5,Z=0.5,W=0.5),Translation=(X=0,Y=0,Z=17.9658),Scale3D=(X=1.2,Y=1.2,Z=1.2))
    Settings="(PrimaryAxis=Y,bIsScale=False,Limits=(Limit=(bMinimum=True,bMaximum=True),MinValue=0,MaxValue=1,bDrawLimits=True),Shape=(bVisible=True,Name=\"DefaultGizmoLibraryNormalized.Sphere_Solid\",Color=(R=0,G=0.0375,B=1,A=1),Transform=(Rotation=(X=-0.7071,Y=0,Z=0,W=0 ...(truncated)"
    Parent=@n1.Parent
   ) {template="SpawnControl::Execute(in InitialValue,in Settings,in OffsetTransform,in OffsetSpace,in Parent,in Name,out Item)"}
   > &RigUnit_HierarchyAddControlVector2D(
    InitialValue=(X=0,Y=0)
    Name=CTRL_L_eye
    OffsetSpace=LocalSpace
    OffsetTransform=(Rotation=(X=0.7071,Y=0,Z=0,W=0.7071),Translation=(X=2.5017,Y=0,Z=16.2836),Scale3D=(X=0.65,Y=0.65,Z=0.65))
    Settings="(PrimaryAxis=Z,Limits=(LimitX=(bMinimum=True,bMaximum=True),LimitY=(bMinimum=True,bMaximum=True),MinValue=(X=-1,Y=-1),MaxValue=(X=1,Y=1),bDrawLimits=True),Shape=(bVisible=True,Name=\"DefaultGizmoLibraryNormalized.Sphere_Solid\",Color=(R=1,G=1,B=0,A=1),Transf ...(truncated)"
    Parent=@n1.Parent
   ) {template="SpawnControl::Execute(in InitialValue,in Settings,in OffsetTransform,in OffsetSpace,in Parent,in Name,out Item)"}
   > &RigUnit_HierarchyAddControlVector2D(
    InitialValue=(X=0,Y=0)
    Name=CTRL_R_eye
    OffsetSpace=LocalSpace
    OffsetTransform=(Rotation=(X=0.7071,Y=0,Z=0,W=0.7071),Translation=(X=-2.4929,Y=0,Z=16.3101),Scale3D=(X=0.65,Y=0.65,Z=0.65))
    Settings="(PrimaryAxis=Z,Limits=(LimitX=(bMinimum=True,bMaximum=True),LimitY=(bMinimum=True,bMaximum=True),MinValue=(X=-1,Y=-1),MaxValue=(X=1,Y=1),bDrawLimits=True),Shape=(bVisible=True,Name=\"DefaultGizmoLibraryNormalized.Sphere_Solid\",Color=(R=1,G=1,B=0,A=1),Transf ...(truncated)"
    Parent=@n1.Parent
   ) {template="SpawnControl::Execute(in InitialValue,in Settings,in OffsetTransform,in OffsetSpace,in Parent,in Name,out Item)"}
   > &RigUnit_HierarchyAddControlFloat(
    InitialValue=0.0
    Name=CTRL_L_eye_blink
    OffsetSpace=LocalSpace
    OffsetTransform=(Rotation=(X=-0.7044,Y=0.0616,Z=0.0616,W=0.7044),Translation=(X=3.6658,Y=0,Z=16.436),Scale3D=(X=1,Y=1,Z=1))
    Settings="(PrimaryAxis=Y,bIsScale=False,Limits=(Limit=(bMinimum=True,bMaximum=True),MinValue=-1,MaxValue=1,bDrawLimits=True),Shape=(bVisible=True,Name=\"DefaultGizmoLibraryNormalized.Sphere_Solid\",Color=(R=1,G=1,B=0,A=1),Transform=(Rotation=(X=-0.7071,Y=0,Z=0,W=0.707 ...(truncated)"
    Parent=@n1.Parent
   ) {template="SpawnControl::Execute(in InitialValue,in Settings,in OffsetTransform,in OffsetSpace,in Parent,in Name,out Item)"}
   > n2 = &RigUnit_HierarchyAddControlFloat(
    InitialValue=0.0
    Name=CTRL_R_eye_blink
    OffsetSpace=LocalSpace
    OffsetTransform=(Rotation=(X=-0.5417,Y=-0.5417,Z=0.4545,W=0.4545),Translation=(X=-3.663,Y=0,Z=16.436),Scale3D=(X=1,Y=1,Z=1))
    Settings="(PrimaryAxis=Y,bIsScale=False,Limits=(Limit=(bMinimum=True,bMaximum=True),MinValue=-1,MaxValue=1,bDrawLimits=True),Shape=(bVisible=True,Name=\"DefaultGizmoLibraryNormalized.Sphere_Solid\",Color=(R=1,G=1,B=0,A=1),Transform=(Rotation=(X=-0.7071,Y=0,Z=0,W=0.707 ...(truncated)"
    Parent=@n1.Parent
   ) {template="SpawnControl::Execute(in InitialValue,in Settings,in OffsetTransform,in OffsetSpace,in Parent,in Name,out Item)"}
   RigVMFunctionReturnNode(ExecutePin=@n2.ExecutePin)
  ]}
 CollapseNode_ContainedGraph {nodes [
   n1 = RigVMFunctionEntryNode()
   n2 = &RigUnit_HierarchyAddControlFloat(
    InitialValue=0.0
    Name=CTRL_L_mouth_lipsPressU
    OffsetSpace=LocalSpace
    OffsetTransform=(Rotation=(X=-0.5771,Y=0.5771,Z=-0.4086,W=0.4086),Translation=(X=-4.9162,Y=0.0221,Z=1.3765),Scale3D=(X=1,Y=1,Z=1))
    Settings="(PrimaryAxis=Y,bIsScale=False,Limits=(Limit=(bMinimum=True,bMaximum=True),MinValue=0,MaxValue=1,bDrawLimits=True),Shape=(bVisible=True,Name=\"DefaultGizmoLibrary.Sphere_Solid\",Color=(R=1,G=1,B=0,A=0),Transform=(Rotation=(X=0,Y=0,Z=0,W=1),Translation=(X=0,Y= ...(truncated)"
    ExecutePin=@n1.ExecutePin
    Parent=@n1.Parent
   ) {template="SpawnControl::Execute(in InitialValue,in Settings,in OffsetTransform,in OffsetSpace,in Parent,in Name,out Item)"}
   > /Script/ControlRig.RigUnit_SetMetadataTag(NameSpace=Self, Tag=FN_MH_V2_Ctrl, Item=@n2.Item) {is=struct RigVMUnitNode}
   > n3 = &RigUnit_HierarchyAddControlFloat(
    InitialValue=0.0
    Name=CTRL_L_mouth_lipsPressD
    OffsetSpace=LocalSpace
    OffsetTransform=(Rotation=(X=0.6962,Y=-0.1236,Z=0.1236,W=0.6962),Translation=(X=-4.9108,Y=0,Z=-0.8028),Scale3D=(X=1,Y=1,Z=1))
    Settings="(PrimaryAxis=Y,bIsScale=False,Limits=(Limit=(bMinimum=True,bMaximum=True),MinValue=0,MaxValue=1,bDrawLimits=True),Shape=(bVisible=False,Name=\"DefaultGizmoLibrary.Sphere_Solid\",Color=(R=0.0503,G=0.0503,B=0,A=0),Transform=(Rotation=(X=0,Y=0,Z=0,W=1),Translat ...(truncated)"
    Parent=@n1.Parent
   ) {template="SpawnControl::Execute(in InitialValue,in Settings,in OffsetTransform,in OffsetSpace,in Parent,in Name,out Item)"}
   > &RigUnit_SetMetadataTag(NameSpace=Self, Tag=FN_MH_V2_Ctrl, Item=@n3.Item)
   > n4 = &RigUnit_HierarchyAddControlFloat(
    InitialValue=0.0
    Name=CTRL_R_mouth_lipsPressU
    OffsetSpace=LocalSpace
    OffsetTransform=(Rotation=(X=-0.4056,Y=0.4056,Z=-0.5792,W=0.5792),Translation=(X=-5.8236,Y=0.0221,Z=1.3718),Scale3D=(X=1,Y=1,Z=1))
    Settings="(PrimaryAxis=Y,bIsScale=False,Limits=(Limit=(bMinimum=True,bMaximum=True),MinValue=0,MaxValue=1,bDrawLimits=True),Shape=(bVisible=True,Name=\"DefaultGizmoLibrary.Sphere_Solid\",Color=(R=1,G=1,B=0,A=0),Transform=(Rotation=(X=0,Y=0,Z=0,W=1),Translation=(X=0,Y= ...(truncated)"
    Parent=@n1.Parent
   ) {template="SpawnControl::Execute(in InitialValue,in Settings,in OffsetTransform,in OffsetSpace,in Parent,in Name,out Item)"}
   > &RigUnit_SetMetadataTag(NameSpace=Self, Tag=FN_MH_V2_Ctrl, Item=@n4.Item)
   > n5 = &RigUnit_HierarchyAddControlFloat(
    InitialValue=0.0
    Name=CTRL_R_mouth_lipsPressD
    OffsetSpace=LocalSpace
    OffsetTransform=(Rotation=(X=0.6964,Y=0.1228,Z=-0.1228,W=0.6964),Translation=(X=-5.8183,Y=0,Z=-0.8003),Scale3D=(X=1,Y=1,Z=1))
    Settings="(PrimaryAxis=Y,bIsScale=False,Limits=(Limit=(bMinimum=True,bMaximum=True),MinValue=0,MaxValue=1,bDrawLimits=True),Shape=(bVisible=False,Name=\"DefaultGizmoLibrary.Sphere_Solid\",Color=(R=0.0503,G=0.0503,B=0,A=0),Transform=(Rotation=(X=0,Y=0,Z=0,W=1),Translat ...(truncated)"
    Parent=@n1.Parent
   ) {template="SpawnControl::Execute(in InitialValue,in Settings,in OffsetTransform,in OffsetSpace,in Parent,in Name,out Item)"}
   > n6 = &RigUnit_SetMetadataTag(NameSpace=Self, Tag=FN_MH_V2_Ctrl, Item=@n5.Item)
   RigVMFunctionReturnNode(ExecutePin=@n6.ExecutePin)
  ]}
 CollapseNode_ContainedGraph {nodes [
   n1 = RigVMFunctionEntryNode()
   n2 = &RigUnit_HierarchyAddControlFloat(
    InitialValue=0.0
    Name=CTRL_L_mouth_tightenU
    OffsetSpace=LocalSpace
    OffsetTransform=(Rotation=(X=-0.5771,Y=0.5771,Z=-0.4086,W=0.4086),Translation=(X=-3.0231,Y=0.0221,Z=1.3785),Scale3D=(X=1,Y=1,Z=1))
    Settings="(PrimaryAxis=Y,bIsScale=False,Limits=(Limit=(bMinimum=True,bMaximum=True),MinValue=0,MaxValue=1,bDrawLimits=True),Shape=(bVisible=True,Name=\"DefaultGizmoLibrary.Sphere_Solid\",Color=(R=1,G=1,B=0,A=0),Transform=(Rotation=(X=0,Y=0,Z=0,W=1),Translation=(X=0,Y= ...(truncated)"
    ExecutePin=@n1.ExecutePin
    Parent=@n1.Parent
   ) {template="SpawnControl::Execute(in InitialValue,in Settings,in OffsetTransform,in OffsetSpace,in Parent,in Name,out Item)"}
   > &RigUnit_SetMetadataTag(NameSpace=Self, Tag=FN_MH_V2_Ctrl, Item=@n2.Item)
   > n3 = &RigUnit_HierarchyAddControlFloat(
    InitialValue=0.0
    Name=CTRL_L_mouth_tightenD
    OffsetSpace=LocalSpace
    OffsetTransform=(Rotation=(X=0.5797,Y=-0.5797,Z=-0.4049,W=0.4049),Translation=(X=-3.0231,Y=0.0221,Z=-0.795),Scale3D=(X=1,Y=1,Z=1))
    Settings="(PrimaryAxis=Y,bIsScale=False,Limits=(Limit=(bMinimum=True,bMaximum=True),MinValue=0,MaxValue=1,bDrawLimits=True),Shape=(bVisible=True,Name=\"DefaultGizmoLibrary.Sphere_Solid\",Color=(R=1,G=1,B=0,A=0),Transform=(Rotation=(X=0,Y=0,Z=0,W=1),Translation=(X=0,Y= ...(truncated)"
    Parent=@n1.Parent
   ) {template="SpawnControl::Execute(in InitialValue,in Settings,in OffsetTransform,in OffsetSpace,in Parent,in Name,out Item)"}
   > &RigUnit_SetMetadataTag(NameSpace=Self, Tag=FN_MH_V2_Ctrl, Item=@n3.Item)
   > n4 = &RigUnit_HierarchyAddControlFloat(
    InitialValue=0.0
    Name=CTRL_R_mouth_tightenU
    OffsetSpace=LocalSpace
    OffsetTransform=(Rotation=(X=-0.4056,Y=0.4056,Z=-0.5792,W=0.5792),Translation=(X=-3.9306,Y=0.0221,Z=1.3737),Scale3D=(X=1,Y=1,Z=1))
    Settings="(PrimaryAxis=Y,bIsScale=False,Limits=(Limit=(bMinimum=True,bMaximum=True),MinValue=0,MaxValue=1,bDrawLimits=True),Shape=(bVisible=True,Name=\"DefaultGizmoLibrary.Sphere_Solid\",Color=(R=1,G=1,B=0,A=0),Transform=(Rotation=(X=0,Y=0,Z=0,W=1),Translation=(X=0,Y= ...(truncated)"
    Parent=@n1.Parent
   ) {template="SpawnControl::Execute(in InitialValue,in Settings,in OffsetTransform,in OffsetSpace,in Parent,in Name,out Item)"}
   > &RigUnit_SetMetadataTag(NameSpace=Self, Tag=FN_MH_V2_Ctrl, Item=@n4.Item)
   > n5 = &RigUnit_HierarchyAddControlFloat(
    InitialValue=0.0
    Name=CTRL_R_mouth_tightenD
    OffsetSpace=LocalSpace
    OffsetTransform=(Rotation=(X=0.4056,Y=-0.4056,Z=-0.5792,W=0.5792),Translation=(X=-3.9306,Y=0.0221,Z=-0.7924),Scale3D=(X=1,Y=1,Z=1))
    Settings="(PrimaryAxis=Y,bIsScale=False,Limits=(Limit=(bMinimum=True,bMaximum=True),MinValue=0,MaxValue=1,bDrawLimits=True),Shape=(bVisible=True,Name=\"DefaultGizmoLibrary.Sphere_Solid\",Color=(R=1,G=1,B=0,A=0),Transform=(Rotation=(X=0,Y=0,Z=0,W=1),Translation=(X=0,Y= ...(truncated)"
    Parent=@n1.Parent
   ) {template="SpawnControl::Execute(in InitialValue,in Settings,in OffsetTransform,in OffsetSpace,in Parent,in Name,out Item)"}
   > n6 = &RigUnit_SetMetadataTag(NameSpace=Self, Tag=FN_MH_V2_Ctrl, Item=@n5.Item)
   RigVMFunctionReturnNode(ExecutePin=@n6.ExecutePin)
  ]}
 CollapseNode_ContainedGraph {nodes [
   n1 = RigVMFunctionEntryNode()
   n2 = &RigUnit_HierarchyAddNull(
    Name=GRP_noseGUI
    Space=LocalSpace
    Transform=(Rotation=(X=0,Y=0,Z=0,W=1),Translation=(X=0,Y=0,Z=0),Scale3D=(X=1,Y=1,Z=1))
    ExecutePin=@n1.ExecutePin
    Parent=@n1.Parent
   )
   > &RigUnit_HierarchyAddControlVector2D(
    InitialValue=(X=0,Y=0)
    Name=CTRL_L_nose
    OffsetSpace=LocalSpace
    OffsetTransform=(Rotation=(X=0.7071,Y=0,Z=0,W=0.7071),Translation=(X=0.6892,Y=0,Z=14.4835),Scale3D=(X=0.65,Y=0.65,Z=0.65))
    Settings="(PrimaryAxis=Z,Limits=(LimitX=(bMinimum=True,bMaximum=True),LimitY=(bMinimum=True,bMaximum=True),MinValue=(X=-1,Y=-1),MaxValue=(X=1,Y=1),bDrawLimits=True),Shape=(bVisible=True,Name=\"DefaultGizmoLibraryNormalized.Sphere_Solid\",Color=(R=1,G=1,B=0,A=1),Transf ...(truncated)"
    Parent=@n2.Item
   ) {template="SpawnControl::Execute(in InitialValue,in Settings,in OffsetTransform,in OffsetSpace,in Parent,in Name,out Item)"}
   > n3 = &RigUnit_HierarchyAddControlVector2D(
    InitialValue=(X=0,Y=0)
    Name=CTRL_R_nose
    OffsetSpace=LocalSpace
    OffsetTransform=(Rotation=(X=0,Y=-0.7071,Z=-0.7071,W=0),Translation=(X=-0.6892,Y=0,Z=14.4835),Scale3D=(X=0.65,Y=0.65,Z=0.65))
    Settings="(PrimaryAxis=Z,Limits=(LimitX=(bMinimum=True,bMaximum=True),LimitY=(bMinimum=True,bMaximum=True),MinValue=(X=-1,Y=-1),MaxValue=(X=1,Y=1),bDrawLimits=True),Shape=(bVisible=True,Name=\"DefaultGizmoLibraryNormalized.Sphere_Solid\",Color=(R=1,G=1,B=0,A=1),Transf ...(truncated)"
    Parent=@n2.Item
   ) {template="SpawnControl::Execute(in InitialValue,in Settings,in OffsetTransform,in OffsetSpace,in Parent,in Name,out Item)"}
   RigVMFunctionReturnNode(ExecutePin=@n3.ExecutePin)
  ]}
 CollapseNode_ContainedGraph {nodes [
   n1 = RigVMFunctionEntryNode()
   n2 = &RigUnit_HierarchyAddNull(
    Name=GRP_chinGUI
    Space=LocalSpace
    Transform=(Rotation=(X=0,Y=0,Z=0,W=1),Translation=(X=0,Y=0,Z=0),Scale3D=(X=1,Y=1,Z=1))
    ExecutePin=@n1.ExecutePin
    Parent=@n1.Parent
   )
   > &RigUnit_HierarchyAddControlFloat(
    InitialValue=0.0
    Name=CTRL_L_jaw_ChinRaiseU
    OffsetSpace=LocalSpace
    OffsetTransform=(Rotation=(X=0.5,Y=-0.5,Z=-0.5,W=0.5),Translation=(X=0.2288,Y=0.0221,Z=6.067),Scale3D=(X=1,Y=1,Z=1))
    Settings="(PrimaryAxis=Y,bIsScale=False,Limits=(Limit=(bMinimum=True,bMaximum=True),MinValue=0,MaxValue=1,bDrawLimits=True),Shape=(bVisible=True,Name=\"DefaultGizmoLibraryNormalized.Sphere_Solid\",Color=(R=1,G=1,B=0,A=0),Transform=(Rotation=(X=0,Y=0,Z=0,W=1),Translati ...(truncated)"
    Parent=@n2.Item
   ) {template="SpawnControl::Execute(in InitialValue,in Settings,in OffsetTransform,in OffsetSpace,in Parent,in Name,out Item)"}
   > &RigUnit_HierarchyAddControlFloat(
    InitialValue=0.0
    Name=CTRL_L_jaw_ChinRaiseD
    OffsetSpace=LocalSpace
    OffsetTransform=(Rotation=(X=0.5,Y=-0.5,Z=-0.5,W=0.5),Translation=(X=0.229,Y=0.0221,Z=4.7736),Scale3D=(X=1,Y=1,Z=1))
    Settings="(PrimaryAxis=Y,bIsScale=False,Limits=(Limit=(bMinimum=True,bMaximum=True),MinValue=0,MaxValue=1,bDrawLimits=True),Shape=(bVisible=True,Name=\"DefaultGizmoLibraryNormalized.Sphere_Solid\",Color=(R=1,G=1,B=0,A=0),Transform=(Rotation=(X=0,Y=0,Z=0,W=1),Translati ...(truncated)"
    Parent=@n2.Item
   ) {template="SpawnControl::Execute(in InitialValue,in Settings,in OffsetTransform,in OffsetSpace,in Parent,in Name,out Item)"}
   > &RigUnit_HierarchyAddControlFloat(
    InitialValue=0.0
    Name=CTRL_R_jaw_ChinRaiseU
    OffsetSpace=LocalSpace
    OffsetTransform=(Rotation=(X=0.5,Y=-0.5,Z=-0.5,W=0.5),Translation=(X=-0.2914,Y=0.0221,Z=6.067),Scale3D=(X=1,Y=1,Z=1))
    Settings="(PrimaryAxis=Y,bIsScale=False,Limits=(Limit=(bMinimum=True,bMaximum=True),MinValue=0,MaxValue=1,bDrawLimits=True),Shape=(bVisible=True,Name=\"DefaultGizmoLibraryNormalized.Sphere_Solid\",Color=(R=1,G=1,B=0,A=0),Transform=(Rotation=(X=0,Y=0,Z=0,W=1),Translati ...(truncated)"
    Parent=@n2.Item
   ) {template="SpawnControl::Execute(in InitialValue,in Settings,in OffsetTransform,in OffsetSpace,in Parent,in Name,out Item)"}
   > n3 = &RigUnit_HierarchyAddControlFloat(
    InitialValue=0.0
    Name=CTRL_R_jaw_ChinRaiseD
    OffsetSpace=LocalSpace
    OffsetTransform=(Rotation=(X=0.5,Y=-0.5,Z=-0.5,W=0.5),Translation=(X=-0.2924,Y=0.0221,Z=4.7736),Scale3D=(X=1,Y=1,Z=1))
    Settings="(PrimaryAxis=Y,bIsScale=False,Limits=(Limit=(bMinimum=True,bMaximum=True),MinValue=0,MaxValue=1,bDrawLimits=True),Shape=(bVisible=True,Name=\"DefaultGizmoLibraryNormalized.Sphere_Solid\",Color=(R=1,G=1,B=0,A=0),Transform=(Rotation=(X=0,Y=0,Z=0,W=1),Translati ...(truncated)"
    Parent=@n2.Item
   ) {template="SpawnControl::Execute(in InitialValue,in Settings,in OffsetTransform,in OffsetSpace,in Parent,in Name,out Item)"}
   RigVMFunctionReturnNode(ExecutePin=@n3.ExecutePin)
  ]}
 CollapseNode_ContainedGraph {nodes [
   n1 = RigVMFunctionEntryNode()
   n2 = &RigUnit_HierarchyAddNull(
    Name=GRP_jawGUI
    Space=LocalSpace
    Transform=(Rotation=(X=0,Y=0,Z=0,W=1),Translation=(X=0,Y=0,Z=0),Scale3D=(X=1,Y=1,Z=1))
    ExecutePin=@n1.ExecutePin
    Parent=@n1.Parent
   )
   > RigUnit_HierarchyAddControlVector2D = &RigUnit_HierarchyAddControlVector2D(
    InitialValue=(X=0,Y=0)
    Name=CTRL_C_jaw
    OffsetSpace=LocalSpace
    OffsetTransform=(Rotation=(X=0,Y=0.7071,Z=-0.7071,W=0),Translation=(X=0,Y=0,Z=4.1),Scale3D=(X=1,Y=2,Z=1))
    Settings="(PrimaryAxis=Z,Limits=(LimitX=(bMinimum=True,bMaximum=True),LimitY=(bMinimum=True,bMaximum=True),MinValue=(X=-1,Y=0),MaxValue=(X=1,Y=1),bDrawLimits=True),Shape=(bVisible=True,Name=\"DefaultGizmoLibraryNormalized.Sphere_Solid\",Color=(R=1,G=1,B=0,A=1),Transfo ...(truncated)"
    Parent=@n2.Item
   ) {template="SpawnControl::Execute(in InitialValue,in Settings,in OffsetTransform,in OffsetSpace,in Parent,in Name,out Item)"}
   > &RigUnit_HierarchyAddControlFloat(
    InitialValue=0.0
    Name=CTRL_C_jaw_openExtreme
    OffsetSpace=LocalSpace
    OffsetTransform=(Rotation=(X=0,Y=0,Z=-0.7071,W=0.7071),Translation=(X=-0.9896,Y=0.0221,Z=1.6084),Scale3D=(X=2,Y=2,Z=2))
    Settings="(PrimaryAxis=Y,bIsScale=False,Limits=(Limit=(bMinimum=True,bMaximum=True),MinValue=0,MaxValue=1,bDrawLimits=True),Shape=(bVisible=True,Name=\"DefaultGizmoLibraryNormalized.Sphere_Solid\",Color=(R=1,G=1,B=0,A=0),Transform=(Rotation=(X=0,Y=0,Z=0,W=1),Translati ...(truncated)"
    Parent=@n2.Item
   ) {template="SpawnControl::Execute(in InitialValue,in Settings,in OffsetTransform,in OffsetSpace,in Parent,in Name,out Item)"}
   > n3 = &RigUnit_HierarchyAddControlFloat(
    InitialValue=0.0
    Name=CTRL_C_jaw_fwdBack
    OffsetSpace=LocalSpace
    OffsetTransform=(Rotation=(X=0.5,Y=-0.5,Z=-0.5,W=0.5),Translation=(X=2.0127,Y=0.0221,Z=3.1023),Scale3D=(X=1,Y=1,Z=1))
    Settings="(PrimaryAxis=Y,bIsScale=False,Limits=(Limit=(bMinimum=True,bMaximum=True),MinValue=-1,MaxValue=1,bDrawLimits=True),Shape=(bVisible=True,Name=\"DefaultGizmoLibraryNormalized.Sphere_Solid\",Color=(R=1,G=1,B=0,A=0),Transform=(Rotation=(X=0,Y=0,Z=0,W=1),Translat ...(truncated)"
    Parent=@n2.Item
   ) {template="SpawnControl::Execute(in InitialValue,in Settings,in OffsetTransform,in OffsetSpace,in Parent,in Name,out Item)"}
   RigVMFunctionReturnNode(ExecutePin=@n3.ExecutePin)
  ]}
 CollapseNode_ContainedGraph {nodes [
   n1 = RigVMFunctionEntryNode()
   n2 = &RigUnit_HierarchyAddNull(
    Name=GRP_tongueGUI
    Space=LocalSpace
    Transform=(Rotation=(X=0,Y=0,Z=0,W=1),Translation=(X=0,Y=0,Z=0),Scale3D=(X=1,Y=1,Z=1))
    ExecutePin=@n1.ExecutePin
    Parent=@n1.Parent
   )
   > &RigUnit_HierarchyAddControlVector2D(
    InitialValue=(X=0,Y=0)
    Name=CTRL_C_tongue_roll
    OffsetSpace=LocalSpace
    OffsetTransform=(Rotation=(X=0.7071,Y=0,Z=0,W=0.7071),Translation=(X=6.305,Y=0.0221,Z=9.109),Scale3D=(X=0.36,Y=0.36,Z=0.36))
    Settings="(PrimaryAxis=Z,Limits=(LimitX=(bMinimum=True,bMaximum=True),LimitY=(bMinimum=True,bMaximum=True),MinValue=(X=-1,Y=-1),MaxValue=(X=1,Y=1),bDrawLimits=True),Shape=(bVisible=True,Name=\"DefaultGizmoLibraryNormalized.Sphere_Solid\",Color=(R=1,G=1,B=0,A=0),Transf ...(truncated)"
    Parent=@n2.Item
   ) {template="SpawnControl::Execute(in InitialValue,in Settings,in OffsetTransform,in OffsetSpace,in Parent,in Name,out Item)"}
   > &RigUnit_HierarchyAddControlVector2D(
    InitialValue=(X=0,Y=0)
    Name=CTRL_C_tongue
    OffsetSpace=LocalSpace
    OffsetTransform=(Rotation=(X=0.7071,Y=0,Z=0,W=0.7071),Translation=(X=5.2647,Y=0.0221,Z=9.117),Scale3D=(X=0.36,Y=0.36,Z=0.36))
    Settings="(PrimaryAxis=Z,Limits=(LimitX=(bMinimum=True,bMaximum=True),LimitY=(bMinimum=True,bMaximum=True),MinValue=(X=-1,Y=-1),MaxValue=(X=1,Y=1),bDrawLimits=True),Shape=(bVisible=True,Name=\"DefaultGizmoLibraryNormalized.Sphere_Solid\",Color=(R=1,G=1,B=0,A=0),Transf ...(truncated)"
    Parent=@n2.Item
   ) {template="SpawnControl::Execute(in InitialValue,in Settings,in OffsetTransform,in OffsetSpace,in Parent,in Name,out Item)"}
   > &RigUnit_HierarchyAddControlFloat(
    InitialValue=0.0
    Name=CTRL_C_tongue_narrowWide
    OffsetSpace=LocalSpace
    OffsetTransform=(Rotation=(X=-0.7071,Y=-0.7071,Z=0,W=0),Translation=(X=6.25,Y=0.0221,Z=9.6391),Scale3D=(X=0.35,Y=0.35,Z=0.35))
    Settings="(PrimaryAxis=Y,bIsScale=False,Limits=(Limit=(bMinimum=True,bMaximum=True),MinValue=-1,MaxValue=1,bDrawLimits=True),Shape=(bVisible=True,Name=\"DefaultGizmoLibraryNormalized.Sphere_Solid\",Color=(R=1,G=1,B=0,A=0),Transform=(Rotation=(X=0,Y=0,Z=0,W=1),Translat ...(truncated)"
    Parent=@n2.Item
   ) {template="SpawnControl::Execute(in InitialValue,in Settings,in OffsetTransform,in OffsetSpace,in Parent,in Name,out Item)"}
   > n3 = &RigUnit_HierarchyAddControlFloat(
    InitialValue=0.0
    Name=CTRL_C_tongue_inOut
    OffsetSpace=LocalSpace
    OffsetTransform=(Rotation=(X=0.5,Y=0.5,Z=0.5,W=0.5),Translation=(X=4.7,Y=0.0221,Z=9.1398),Scale3D=(X=0.4,Y=0.4,Z=0.4))
    Settings="(PrimaryAxis=Y,bIsScale=False,Limits=(Limit=(bMinimum=True,bMaximum=True),MinValue=-1,MaxValue=1,bDrawLimits=True),Shape=(bVisible=True,Name=\"DefaultGizmoLibraryNormalized.Sphere_Solid\",Color=(R=1,G=1,B=0,A=0),Transform=(Rotation=(X=0,Y=0,Z=0,W=1),Translat ...(truncated)"
    Parent=@n2.Item
   ) {template="SpawnControl::Execute(in InitialValue,in Settings,in OffsetTransform,in OffsetSpace,in Parent,in Name,out Item)"}
   RigVMFunctionReturnNode(ExecutePin=@n3.ExecutePin)
  ]}
 EyeAim_Inverse_CollapseNode_ContainedGraph {nodes [
   n1 = RigVMFunctionEntryNode()
   n2 = &RigVMFunction_MathFloatRemap(
    SourceMaximum=1
    SourceMinimum=0
    TargetMaximum=40
    TargetMinimum=0
    bClamp=true
    Value=&RigUnit_GetCurveValue(Curve=CTRL_expressions_eyeLookDownL).Value
   )
   n3 = &RigVMFunction_MathFloatRemap(
    SourceMaximum=1
    SourceMinimum=0
    TargetMaximum=-40
    TargetMinimum=0
    bClamp=true
    Value=&RigUnit_GetCurveValue(Curve=CTRL_expressions_eyeLookUpL).Value
   )
   n4 = &RigVMFunction_MathFloatRemap(
    SourceMaximum=1
    SourceMinimum=0
    TargetMaximum=40
    TargetMinimum=0
    bClamp=true
    Value=&RigUnit_GetCurveValue(Curve=CTRL_expressions_eyeLookRightL).Value
   )
   n5 = &RigVMFunction_MathFloatRemap(
    SourceMaximum=1
    SourceMinimum=0
    TargetMaximum=-40
    TargetMinimum=0
    bClamp=true
    Value=&RigUnit_GetCurveValue(Curve=CTRL_expressions_eyeLookLeftL).Value
   )
   &RigUnit_SetRotation(
    Item=(Type=Null,Name="EyeAim_L")
    Weight=1
    bInitial=False
    ExecutePin=@n1.ExecutePin
    Value=/Script/RigVM.RigVMFunction_MathQuaternionFromRotatorV2(
     Value=(Pitch=0,Yaw=0,Roll=0)
     Value.Roll=&RigVMFunction_MathFloatAdd(A=@n2.Result, B=@n3.Result) {template="Add::Execute(in A,in B,out Result)"}.Result
     Value.Pitch=&RigVMFunction_MathFloatAdd(A=@n4.Result, B=@n5.Result) {template="Add::Execute(in A,in B,out Result)"}.Result
    ) {is=struct RigVMUnitNode, template="Cast::Execute(in Value,out Result)"}.Result
   )
   n6 = &RigVMFunction_MathFloatRemap(
    SourceMaximum=1
    SourceMinimum=0
    TargetMaximum=40
    TargetMinimum=0
    bClamp=true
    Value=&RigUnit_GetCurveValue(Curve=CTRL_expressions_eyeLookRightR).Value
   )
   n7 = &RigVMFunction_MathFloatRemap(
    SourceMaximum=1
    SourceMinimum=0
    TargetMaximum=-40
    TargetMinimum=0
    bClamp=true
    Value=&RigUnit_GetCurveValue(Curve=CTRL_expressions_eyeLookLeftR).Value
   )
   n8 = &RigVMFunction_MathFloatRemap(
    SourceMaximum=1
    SourceMinimum=0
    TargetMaximum=-40
    TargetMinimum=0
    bClamp=true
    Value=&RigUnit_GetCurveValue(Curve=CTRL_expressions_eyeLookUpR).Value
   )
   n9 = &RigVMFunction_MathFloatRemap(
    SourceMaximum=1
    SourceMinimum=0
    TargetMaximum=40
    TargetMinimum=0
    bClamp=true
    Value=&RigUnit_GetCurveValue(Curve=CTRL_expressions_eyeLookDownR).Value
   )
   > &RigUnit_SetRotation(
    Item=(Type=Null,Name="EyeAim_R")
    Weight=1
    bInitial=False
    Value=&RigVMFunction_MathQuaternionFromRotatorV2(
     Value=(Pitch=0,Yaw=0,Roll=0)
     Value.Pitch=&RigVMFunction_MathFloatAdd(A=@n6.Result, B=@n7.Result) {template="Add::Execute(in A,in B,out Result)"}.Result
     Value.Roll=&RigVMFunction_MathFloatAdd(A=@n8.Result, B=@n9.Result) {template="Add::Execute(in A,in B,out Result)"}.Result
    ) {template="Cast::Execute(in Value,out Result)"}.Result
   )
   n10 = &RigUnit_ProjectTransformToNewParent(
    Child=(Type=Control,Name="CTRL_C_eyesAim")
    NewParent=(Type=Null,Name="EyeAim_L")
    OldParent=(Type=Null,Name="EyeAim_L")
    bChildInitial=True
    bNewParentInitial=False
    bOldParentInitial=True
   )
   > &RigUnit_SetTransform(Item=(Type=Control,Name="CTRL_C_eyesAim"), Weight=1, bInitial=False, Value=@n10.Transform)
   n11 = &RigUnit_ProjectTransformToNewParent(
    Child=(Type=Control,Name="CTRL_L_eyeAim")
    NewParent=(Type=Null,Name="EyeAim_L")
    OldParent=(Type=Null,Name="EyeAim_L")
    bChildInitial=True
    bNewParentInitial=False
    bOldParentInitial=True
   )
   > &RigUnit_SetTransform(Item=(Type=Control,Name="CTRL_L_eyeAim"), Weight=1, bInitial=False, Value=@n11.Transform)
   n12 = &RigUnit_ProjectTransformToNewParent(
    Child=(Type=Control,Name="CTRL_R_eyeAim")
    NewParent=(Type=Null,Name="EyeAim_R")
    OldParent=(Type=Null,Name="EyeAim_R")
    bChildInitial=True
    bNewParentInitial=False
    bOldParentInitial=True
   )
   > n13 = &RigUnit_SetTransform(Item=(Type=Control,Name="CTRL_R_eyeAim"), Weight=1, bInitial=False, Value=@n12.Transform)
   RigVMFunctionReturnNode(ExecutePin=@n13.ExecutePin)
  ]}
 Eye_Aim_Control_Vis_CollapseNode_ContainedGraph {nodes [
   n1 = RigVMFunctionEntryNode()
   RigVMFunction_Sequence = &RigVMFunction_Sequence(ExecuteContext=@n1.ExecuteContext)
   n2 = /Script/ControlRig.RigUnit_CollectionItems(
    Items=((Type=Control,Name="CTRL_C_eyesAim"),(Type=Control,Name="CTRL_L_eyeAim"),(Type=Control,Name="CTRL_R_eyeAim"),(Type=Control,Name="FRM_convergenceGUI"),(Type=Control,Name="CTRL_convergenceSwitch"))
    bAllowDuplicates=False
   ) {is=struct RigVMUnitNode}
   n3 = RigVMDispatchNode(Array=@n2.Collection.Keys, ExecuteContext=@RigVMFunction_Sequence.A) {template="DISPATCH_RigVMDispatch_ArrayIterator(in Array,out Element,out Index,out Count,out Ratio)"}
   n4 = /Script/RigVM.RigVMFunction_MathFloatGreater(
    B=0.5
    A=&RigUnit_GetControlFloat(Control=CTRL_lookAtSwitch) {template="GetControlValue::Execute(in Control,out FloatValue,out Minimum,out Maximum)"}.FloatValue
   ) {is=struct RigVMUnitNode, template="Greater::Execute(in A,in B,out Result)"}
   > SetControlVisibility = /Script/ControlRig.RigUnit_SetControlVisibility(Item=@n3.Element, bVisible=@n4.Result) {is=struct RigVMUnitNode, template="SetControlVisibility::Execute(in Item,in Pattern,in bVisible)"}
   n5 = &RigVMFunction_ControlFlowBranch(ExecuteContext=@n3.Completed, Condition=@n4.Result)
   DebugLineItemSpace = /Script/RigVM.RigVMFunction_DebugLineNoSpace(
    Color=(R=0,G=0,B=0,A=1)
    DebugDrawSettings=(DepthPriority=SDPG_Foreground,Lifetime=-1)
    Thickness=0.05
    WorldOffset=(Rotation=(X=0,Y=0,Z=0,W=1),Translation=(X=0,Y=0,Z=0),Scale3D=(X=1,Y=1,Z=1))
    bEnabled=True
    ExecutePin=@n5.True
    A=&RigUnit_GetTransform(Item=(Type=Bone,Name="FACIAL_L_Eye"), bInitial=False).Transform.Translation
    B=&RigUnit_GetTransform(Item=(Type=Control,Name="CTRL_L_eyeAim"), bInitial=False).Transform.Translation
   ) {is=struct RigVMUnitNode}
   > DebugLineItemSpace_1 = &RigVMFunction_DebugLineNoSpace(
    Color=(R=0,G=0,B=0,A=1)
    DebugDrawSettings=(DepthPriority=SDPG_Foreground,Lifetime=-1)
    Thickness=0.05
    WorldOffset=(Rotation=(X=0,Y=0,Z=0,W=1),Translation=(X=0,Y=0,Z=0),Scale3D=(X=1,Y=1,Z=1))
    bEnabled=True
    B=&RigUnit_GetTransform(Item=(Type=Control,Name="CTRL_R_eyeAim"), bInitial=False).Transform.Translation
    A=&RigUnit_GetTransform(Item=(Type=Bone,Name="FACIAL_R_Eye"), bInitial=False).Transform.Translation
   )
   RigVMFunctionReturnNode(ExecuteContext=@RigVMFunction_Sequence.B)
  ]}
 Eye_Convergence_CollapseNode_ContainedGraph {nodes [
   n1 = RigVMFunctionEntryNode()
   n2 = &RigUnit_ProjectTransformToNewParent(
    Child=(Type=Null,Name="CTRL_L_eyeAim_space")
    NewParent=(Type=Control,Name="CTRL_C_eyesAim")
    OldParent=(Type=Control,Name="CTRL_C_eyesAim")
    bChildInitial=True
    bNewParentInitial=False
    bOldParentInitial=True
   )
   n3 = &RigUnit_GetControlFloat(Control=CTRL_convergenceSwitch) {template="GetControlValue::Execute(in Control,out FloatValue,out Minimum,out Maximum)"}
   n4 = &RigUnit_GetTransform(Item=(Type=Control,Name="CTRL_C_eyesAim"), bInitial=False)
   RigUnit_SetTransform = &RigUnit_SetTransform(
    Item=(Type=Null,Name="CTRL_L_eyeAim_space")
    Weight=1
    bInitial=False
    ExecutePin=@n1.ExecutePin
    Value=&RigVMFunction_MathTransformLerp(A=@n2.Transform, T=@n3.FloatValue, B=@n4.Transform) {template="Interpolate::Execute(in A,in B,in T,out Result)"}.Result
   )
   n5 = &RigUnit_ProjectTransformToNewParent(
    Child=(Type=Null,Name="CTRL_R_eyeAim_space")
    NewParent=(Type=Control,Name="CTRL_C_eyesAim")
    OldParent=(Type=Control,Name="CTRL_C_eyesAim")
    bChildInitial=True
    bNewParentInitial=False
    bOldParentInitial=True
   )
   > n6 = &RigUnit_SetTransform(
    Item=(Type=Null,Name="CTRL_R_eyeAim_space")
    Weight=1
    bInitial=False
    Value=&RigVMFunction_MathTransformLerp(A=@n5.Transform, B=@n4.Transform, T=@n3.FloatValue) {template="Interpolate::Execute(in A,in B,in T,out Result)"}.Result
   )
   RigVMFunctionReturnNode(ExecutePin=@n6.ExecutePin)
  ]}
 Look_EyeDirection_CollapseNode_ContainedGraph {nodes [
   n1 = RigVMFunctionEntryNode()
   RigVMFunction_Sequence = RigVMAggregateNode(ExecuteContext=@n1.ExecutePin)
   RigVMFunctionReturnNode(ExecutePin=@RigVMFunction_Sequence.C)
   n2 = &RigUnit_GetControlFloat(Control=CTRL_lookAtSwitch) {template="GetControlValue::Execute(in Control,out FloatValue,out Minimum,out Maximum)"}
   n3 = /Script/RigVM.RigVMFunction_MathQuaternionToRotator(Value=&RigUnit_GetTransform(Item=(Type=Null,Name="EyeAim_L"), Space=LocalSpace, bInitial=False).Transform.Rotation) {is=struct RigVMUnitNode, template="Cast::Execute(in Value,out Result)"}
   n4 = /Script/ControlRig.RigUnit_GetControlVector2D(Control=CTRL_L_eye) {is=struct RigVMUnitNode, template="GetControlValue::Execute(in Control,out Vector,out Minimum,out Maximum)"}
   n5 = &RigUnit_GetControlVector2D(Control=CTRL_C_eye) {template="GetControlValue::Execute(in Control,out Vector,out Minimum,out Maximum)"}
   Add = &RigVMFunction_MathDoubleAdd(
    A=&RigUnit_GetControlFloat(Control=CTRL_C_eyesConvergence) {template="GetControlValue::Execute(in Control,out FloatValue,out Minimum,out Maximum)"}.FloatValue
    B=/Script/RigVM.RigVMFunction_MathFloatLerp(
     T=@n2.FloatValue
     B=&RigVMFunction_MathFloatRemap(SourceMaximum=-40, SourceMinimum=40, TargetMinimum=-1, bClamp=true, Value=@n3.Result.Pitch).Result
     A=&RigVMFunction_MathFloatAdd(B=@n4.Vector.X, A=@n5.Vector.X) {template="Add::Execute(in A,in B,out Result)"}.Result
    ) {is=struct RigVMUnitNode, template="Interpolate::Execute(in A,in B,in T,out Result)"}.Result
   )
   Remap = &RigVMFunction_MathDoubleRemap(SourceMaximum=1, SourceMinimum=0, TargetMaximum=1, TargetMinimum=0, Value=@Add.Result)
   &RigUnit_SetCurveValue(Curve=CTRL_expressions_eyeLookLeftL, Value=@Remap.Result, ExecutePin=@RigVMFunction_Sequence.A)
   > &RigUnit_SetCurveValue(
    Curve=CTRL_expressions_eyeLookRightL
    Value=&RigVMFunction_MathDoubleRemap(SourceMaximum=-1, SourceMinimum=0, TargetMaximum=1, TargetMinimum=0, Value=@Add.Result).Result
   )
   n6 = &RigVMFunction_MathFloatLerp(
    B=&RigVMFunction_MathFloatRemap(SourceMaximum=50, SourceMinimum=-50, TargetMaximum=-1, TargetMinimum=1, bClamp=true, Value=@n3.Result.Roll).Result
    T=@n2.FloatValue
    A=&RigVMFunction_MathFloatAdd(B=@n4.Vector.Y, A=@n5.Vector.Y) {template="Add::Execute(in A,in B,out Result)"}.Result
   ) {template="Interpolate::Execute(in A,in B,in T,out Result)"}
   > &RigUnit_SetCurveValue(
    Curve=CTRL_expressions_eyeLookUpL
    Value=&RigVMFunction_MathDoubleRemap(SourceMaximum=1, SourceMinimum=0, TargetMaximum=1, TargetMinimum=0, Value=@n6.Result).Result
   )
   > &RigUnit_SetCurveValue(
    Curve=CTRL_expressions_eyeLookDownL
    Value=&RigVMFunction_MathDoubleRemap(SourceMaximum=-1, SourceMinimum=0, TargetMaximum=1, TargetMinimum=0, Value=@n6.Result).Result
   )
   n7 = &RigUnit_GetControlFloat(Control=CTRL_lookAtSwitch) {template="GetControlValue::Execute(in Control,out FloatValue,out Minimum,out Maximum)"}
   n8 = &RigVMFunction_MathQuaternionToRotator(Value=&RigUnit_GetTransform(Item=(Type=Null,Name="EyeAim_R"), Space=LocalSpace, bInitial=False).Transform.Rotation) {template="Cast::Execute(in Value,out Result)"}
   n9 = &RigUnit_GetControlVector2D(Control=CTRL_R_eye) {template="GetControlValue::Execute(in Control,out Vector,out Minimum,out Maximum)"}
   n10 = &RigUnit_GetControlVector2D(Control=CTRL_C_eye) {template="GetControlValue::Execute(in Control,out Vector,out Minimum,out Maximum)"}
   Subtract = &RigVMFunction_MathDoubleSub(
    B=&RigUnit_GetControlFloat(Control=CTRL_C_eyesConvergence) {template="GetControlValue::Execute(in Control,out FloatValue,out Minimum,out Maximum)"}.FloatValue
    A=&RigVMFunction_MathFloatLerp(
     T=@n7.FloatValue
     B=&RigVMFunction_MathFloatRemap(SourceMaximum=-40, SourceMinimum=40, TargetMinimum=-1, bClamp=true, Value=@n8.Result.Pitch).Result
     A=&RigVMFunction_MathFloatAdd(B=@n9.Vector.X, A=@n10.Vector.X) {template="Add::Execute(in A,in B,out Result)"}.Result
    ) {template="Interpolate::Execute(in A,in B,in T,out Result)"}.Result
   )
   &RigUnit_SetCurveValue(
    Curve=CTRL_expressions_eyeLookLeftR
    Value=&RigVMFunction_MathDoubleRemap(SourceMaximum=1, SourceMinimum=0, TargetMaximum=1, TargetMinimum=0, Value=@Subtract.Result).Result
    ExecutePin=@RigVMFunction_Sequence.B
   )
   > &RigUnit_SetCurveValue(
    Curve=CTRL_expressions_eyeLookRightR
    Value=&RigVMFunction_MathDoubleRemap(SourceMaximum=-1, SourceMinimum=0, TargetMaximum=1, TargetMinimum=0, Value=@Subtract.Result).Result
   )
   n11 = &RigVMFunction_MathFloatLerp(
    B=&RigVMFunction_MathFloatRemap(SourceMaximum=50, SourceMinimum=-50, TargetMaximum=-1, TargetMinimum=1, bClamp=true, Value=@n8.Result.Roll).Result
    T=@n7.FloatValue
    A=&RigVMFunction_MathFloatAdd(B=@n9.Vector.Y, A=@n10.Vector.Y) {template="Add::Execute(in A,in B,out Result)"}.Result
   ) {template="Interpolate::Execute(in A,in B,in T,out Result)"}
   > &RigUnit_SetCurveValue(
    Curve=CTRL_expressions_eyeLookUpR
    Value=&RigVMFunction_MathDoubleRemap(SourceMaximum=1, SourceMinimum=0, TargetMaximum=1, TargetMinimum=0, Value=@n11.Result).Result
   )
   > &RigUnit_SetCurveValue(
    Curve=CTRL_expressions_eyeLookDownR
    Value=&RigVMFunction_MathDoubleRemap(SourceMaximum=-1, SourceMinimum=0, TargetMaximum=1, TargetMinimum=0, Value=@n11.Result).Result
   )
  ]}
 New Function_ContainedGraph {nodes [
   n1 = RigVMFunctionEntryNode()
   RigVMFunction_Sequence = &RigVMFunction_Sequence(ExecuteContext=@n1.ExecuteContext)
   n2 = &RigUnit_HierarchyAddNull(
    Name=GRP_tweakerGUI
    Space=LocalSpace
    Transform=(Rotation=(X=0.7071,Y=0,Z=0,W=0.7071),Translation=(X=0,Y=0,Z=0),Scale3D=(X=1,Y=1,Z=1))
    ExecutePin=@RigVMFunction_Sequence.A
    Parent=@n1.Parent
   )
   > RigVMVariableNode(Value=@n2.Item) {variable=Local TweakerGUI}
   > RigVMFunction_Sequence_7 = RigVMAggregateNode()
   n3 = RigVMVariableNode() {variable=Local TweakerGUI}
   &RigUnit_HierarchyAddControlFloat(
    InitialValue=0.0
    Name=CTRL_L_mouth_lipsTowardsTeethU
    OffsetSpace=LocalSpace
    OffsetTransform=(Rotation=(X=0,Y=0,Z=0.9857,W=0.1684),Translation=(X=12.2122,Y=15.5614,Z=0.0001),Scale3D=(X=0.7,Y=0.7,Z=0.7))
    Settings="(PrimaryAxis=Y,bIsScale=False,Limits=(Limit=(bMinimum=True,bMaximum=True),MinValue=0,MaxValue=1,bDrawLimits=True),Shape=(bVisible=True,Name=\"DefaultGizmoLibraryNormalized.Sphere_Solid\",Color=(R=1,G=1,B=0,A=0),Transform=(Rotation=(X=0,Y=0,Z=0,W=1),Translati ...(truncated)"
    Parent=@n3.Value
    ExecutePin=@RigVMFunction_Sequence_7.B
   ) {template="SpawnControl::Execute(in InitialValue,in Settings,in OffsetTransform,in OffsetSpace,in Parent,in Name,out Item)"}
   > &RigUnit_HierarchyAddControlFloat(
    InitialValue=0.0
    Name=CTRL_L_mouth_lipsTowardsTeethD
    OffsetSpace=LocalSpace
    OffsetTransform=(Rotation=(X=0,Y=0,Z=0.1748,W=0.9846),Translation=(X=12.2122,Y=14.04,Z=0.0001),Scale3D=(X=0.7,Y=0.7,Z=0.7))
    Settings="(PrimaryAxis=Y,bIsScale=False,Limits=(Limit=(bMinimum=True,bMaximum=True),MinValue=0,MaxValue=1,bDrawLimits=True),Shape=(bVisible=True,Name=\"DefaultGizmoLibraryNormalized.Sphere_Solid\",Color=(R=1,G=1,B=0,A=0),Transform=(Rotation=(X=0,Y=0,Z=0,W=1),Translati ...(truncated)"
    Parent=@n3.Value
   ) {template="SpawnControl::Execute(in InitialValue,in Settings,in OffsetTransform,in OffsetSpace,in Parent,in Name,out Item)"}
   > &RigUnit_HierarchyAddControlFloat(
    InitialValue=0.0
    Name=CTRL_R_mouth_lipsTowardsTeethU
    OffsetSpace=LocalSpace
    OffsetTransform=(Rotation=(X=0,Y=0,Z=-0.9848,W=0.1736),Translation=(X=11.577,Y=15.5581,Z=0.0001),Scale3D=(X=0.7,Y=0.7,Z=0.7))
    Settings="(PrimaryAxis=Y,bIsScale=False,Limits=(Limit=(bMinimum=True,bMaximum=True),MinValue=0,MaxValue=1,bDrawLimits=True),Shape=(bVisible=True,Name=\"DefaultGizmoLibraryNormalized.Sphere_Solid\",Color=(R=1,G=1,B=0,A=0),Transform=(Rotation=(X=0,Y=0,Z=0,W=1),Translati ...(truncated)"
    Parent=@n3.Value
   ) {template="SpawnControl::Execute(in InitialValue,in Settings,in OffsetTransform,in OffsetSpace,in Parent,in Name,out Item)"}
   > &RigUnit_HierarchyAddControlFloat(
    InitialValue=0.0
    Name=CTRL_R_mouth_lipsTowardsTeethD
    OffsetSpace=LocalSpace
    OffsetTransform=(Rotation=(X=0,Y=0,Z=-0.1736,W=0.9848),Translation=(X=11.577,Y=14.0418,Z=0.0001),Scale3D=(X=0.7,Y=0.7,Z=0.7))
    Settings="(PrimaryAxis=Y,bIsScale=False,Limits=(Limit=(bMinimum=True,bMaximum=True),MinValue=0,MaxValue=1,bDrawLimits=True),Shape=(bVisible=True,Name=\"DefaultGizmoLibraryNormalized.Sphere_Solid\",Color=(R=1,G=1,B=0,A=0),Transform=(Rotation=(X=0,Y=0,Z=0,W=1),Translati ...(truncated)"
    Parent=@n3.Value
   ) {template="SpawnControl::Execute(in InitialValue,in Settings,in OffsetTransform,in OffsetSpace,in Parent,in Name,out Item)"}
   n4 = RigVMVariableNode() {variable=Local TweakerGUI}
   > &RigUnit_HierarchyAddControlVector2D(
    InitialValue=(X=0,Y=0)
    Name=CTRL_C_teethU
    OffsetSpace=LocalSpace
    OffsetTransform=(Rotation=(X=0,Y=0,Z=0,W=1),Translation=(X=13.725,Y=15.676,Z=0),Scale3D=(X=0.4,Y=0.4,Z=0.4))
    Settings="(PrimaryAxis=Z,Limits=(LimitX=(bMinimum=True,bMaximum=True),LimitY=(bMinimum=True,bMaximum=True),MinValue=(X=-1,Y=-1),MaxValue=(X=1,Y=1),bDrawLimits=True),Shape=(bVisible=True,Name=\"DefaultGizmoLibraryNormalized.Sphere_Solid\",Color=(R=1,G=1,B=0,A=0),Transf ...(truncated)"
    Parent=@n4.Value
   ) {template="SpawnControl::Execute(in InitialValue,in Settings,in OffsetTransform,in OffsetSpace,in Parent,in Name,out Item)"}
   > &RigUnit_HierarchyAddControlVector2D(
    InitialValue=(X=0,Y=0)
    Name=CTRL_C_teethD
    OffsetSpace=LocalSpace
    OffsetTransform=(Rotation=(X=0,Y=0,Z=0,W=1),Translation=(X=13.725,Y=14.672,Z=0),Scale3D=(X=0.4,Y=0.4,Z=0.4))
    Settings="(PrimaryAxis=Z,Limits=(LimitX=(bMinimum=True,bMaximum=True),LimitY=(bMinimum=True,bMaximum=True),MinValue=(X=-1,Y=-1),MaxValue=(X=1,Y=1),bDrawLimits=True),Shape=(bVisible=True,Name=\"DefaultGizmoLibraryNormalized.Sphere_Solid\",Color=(R=1,G=1,B=0,A=0),Transf ...(truncated)"
    Parent=@n4.Value
   ) {template="SpawnControl::Execute(in InitialValue,in Settings,in OffsetTransform,in OffsetSpace,in Parent,in Name,out Item)"}
   > &RigUnit_HierarchyAddControlFloat(
    InitialValue=0.0
    Name=CTRL_C_teeth_fwdBackU
    OffsetSpace=LocalSpace
    OffsetTransform=(Rotation=(X=-0.0008,Y=0.7071,Z=-0.0008,W=-0.7071),Translation=(X=13.118,Y=15.675,Z=0),Scale3D=(X=0.4,Y=0.4,Z=0.4))
    Settings="(PrimaryAxis=Y,bIsScale=False,Limits=(Limit=(bMinimum=True,bMaximum=True),MinValue=-1,MaxValue=1,bDrawLimits=True),Shape=(bVisible=True,Name=\"DefaultGizmoLibraryNormalized.Sphere_Solid\",Color=(R=1,G=1,B=0,A=0),Transform=(Rotation=(X=0,Y=0,Z=0,W=1),Translat ...(truncated)"
    Parent=@n4.Value
   ) {template="SpawnControl::Execute(in InitialValue,in Settings,in OffsetTransform,in OffsetSpace,in Parent,in Name,out Item)"}
   > &RigUnit_HierarchyAddControlFloat(
    InitialValue=0.0
    Name=CTRL_C_teeth_fwdBackD
    OffsetSpace=LocalSpace
    OffsetTransform=(Rotation=(X=0,Y=-0.7071,Z=0,W=0.7071),Translation=(X=13.118,Y=14.671,Z=0),Scale3D=(X=0.4,Y=0.4,Z=0.4))
    Settings="(PrimaryAxis=Y,bIsScale=False,Limits=(Limit=(bMinimum=True,bMaximum=True),MinValue=-1,MaxValue=1,bDrawLimits=True),Shape=(bVisible=True,Name=\"DefaultGizmoLibraryNormalized.Sphere_Solid\",Color=(R=1,G=1,B=0,A=0),Transform=(Rotation=(X=0,Y=0,Z=0,W=1),Translat ...(truncated)"
    Parent=@n4.Value
   ) {template="SpawnControl::Execute(in InitialValue,in Settings,in OffsetTransform,in OffsetSpace,in Parent,in Name,out Item)"}
   > &RigUnit_HierarchyAddControlFloat(
    InitialValue=0.0
    Name=CTRL_C_teeth_rotateU
    OffsetSpace=LocalSpace
    OffsetTransform=(Rotation=(X=0.0008,Y=-0.7071,Z=0.0008,W=0.7071),Translation=(X=14.35,Y=15.676,Z=0),Scale3D=(X=0.4,Y=0.4,Z=0.4))
    Settings="(PrimaryAxis=Y,bIsScale=False,Limits=(Limit=(bMinimum=True,bMaximum=True),MinValue=-1,MaxValue=1,bDrawLimits=True),Shape=(bVisible=True,Name=\"DefaultGizmoLibraryNormalized.Sphere_Solid\",Color=(R=0,G=0.0375,B=1,A=1),Transform=(Rotation=(X=0,Y=0,Z=0,W=1),Tra ...(truncated)"
    Parent=@n4.Value
   ) {template="SpawnControl::Execute(in InitialValue,in Settings,in OffsetTransform,in OffsetSpace,in Parent,in Name,out Item)"}
   > &RigUnit_HierarchyAddControlFloat(
    InitialValue=0.0
    Name=CTRL_C_teeth_rotateD
    OffsetSpace=LocalSpace
    OffsetTransform=(Rotation=(X=0,Y=-0.7071,Z=0,W=0.7071),Translation=(X=14.35,Y=14.672,Z=0),Scale3D=(X=0.4,Y=0.4,Z=0.4))
    Settings="(PrimaryAxis=Y,bIsScale=False,Limits=(Limit=(bMinimum=True,bMaximum=True),MinValue=-1,MaxValue=1,bDrawLimits=True),Shape=(bVisible=True,Name=\"DefaultGizmoLibraryNormalized.Sphere_Solid\",Color=(R=0,G=0.0375,B=1,A=1),Transform=(Rotation=(X=0,Y=0,Z=0,W=1),Tra ...(truncated)"
    Parent=@n4.Value
   ) {template="SpawnControl::Execute(in InitialValue,in Settings,in OffsetTransform,in OffsetSpace,in Parent,in Name,out Item)"}
   n5 = RigVMVariableNode() {variable=Local TweakerGUI}
   > &RigUnit_HierarchyAddControlFloat(
    InitialValue=0.0
    Name=CTRL_L_mouth_cornerSharpnessU
    OffsetSpace=LocalSpace
    OffsetTransform=(Rotation=(X=0,Y=0,Z=-1,W=0),Translation=(X=17.477,Y=12.6377,Z=0),Scale3D=(X=0.75,Y=0.75,Z=0.75))
    Settings="(PrimaryAxis=Y,bIsScale=False,Limits=(Limit=(bMinimum=True,bMaximum=True),MinValue=-1,MaxValue=1,bDrawLimits=True),Shape=(bVisible=True,Name=\"DefaultGizmoLibraryNormalized.Sphere_Solid\",Color=(R=1,G=1,B=0,A=0),Transform=(Rotation=(X=0,Y=0,Z=0,W=1),Translat ...(truncated)"
    Parent=@n5.Value
   ) {template="SpawnControl::Execute(in InitialValue,in Settings,in OffsetTransform,in OffsetSpace,in Parent,in Name,out Item)"}
   > &RigUnit_HierarchyAddControlFloat(
    InitialValue=0.0
    Name=CTRL_L_mouth_cornerSharpnessD
    OffsetSpace=LocalSpace
    OffsetTransform=(Rotation=(X=0,Y=0,Z=0,W=1),Translation=(X=17.477,Y=10.9297,Z=0),Scale3D=(X=0.75,Y=0.75,Z=0.75))
    Settings="(PrimaryAxis=Y,bIsScale=False,Limits=(Limit=(bMinimum=True,bMaximum=True),MinValue=-1,MaxValue=1,bDrawLimits=True),Shape=(bVisible=True,Name=\"DefaultGizmoLibraryNormalized.Sphere_Solid\",Color=(R=1,G=1,B=0,A=0),Transform=(Rotation=(X=0,Y=0,Z=0,W=1),Translat ...(truncated)"
    Parent=@n5.Value
   ) {template="SpawnControl::Execute(in InitialValue,in Settings,in OffsetTransform,in OffsetSpace,in Parent,in Name,out Item)"}
   > &RigUnit_HierarchyAddControlFloat(
    InitialValue=0.0
    Name=CTRL_R_mouth_cornerSharpnessU
    OffsetSpace=LocalSpace
    OffsetTransform=(Rotation=(X=0,Y=0,Z=-1,W=0),Translation=(X=17.025,Y=12.6377,Z=0),Scale3D=(X=0.75,Y=0.75,Z=0.75))
    Settings="(PrimaryAxis=Y,bIsScale=False,Limits=(Limit=(bMinimum=True,bMaximum=True),MinValue=-1,MaxValue=1,bDrawLimits=True),Shape=(bVisible=True,Name=\"DefaultGizmoLibraryNormalized.Sphere_Solid\",Color=(R=1,G=1,B=0,A=0),Transform=(Rotation=(X=0,Y=0,Z=0,W=1),Translat ...(truncated)"
    Parent=@n5.Value
   ) {template="SpawnControl::Execute(in InitialValue,in Settings,in OffsetTransform,in OffsetSpace,in Parent,in Name,out Item)"}
   > &RigUnit_HierarchyAddControlFloat(
    InitialValue=0.0
    Name=CTRL_R_mouth_cornerSharpnessD
    OffsetSpace=LocalSpace
    OffsetTransform=(Rotation=(X=0,Y=0,Z=0,W=1),Translation=(X=17.025,Y=10.9297,Z=0),Scale3D=(X=0.75,Y=0.75,Z=0.75))
    Settings="(PrimaryAxis=Y,bIsScale=False,Limits=(Limit=(bMinimum=True,bMaximum=True),MinValue=-1,MaxValue=1,bDrawLimits=True),Shape=(bVisible=True,Name=\"DefaultGizmoLibraryNormalized.Sphere_Solid\",Color=(R=1,G=1,B=0,A=0),Transform=(Rotation=(X=0,Y=0,Z=0,W=1),Translat ...(truncated)"
    Parent=@n5.Value
   ) {template="SpawnControl::Execute(in InitialValue,in Settings,in OffsetTransform,in OffsetSpace,in Parent,in Name,out Item)"}
   RigVMFunction_ControlFlowBranch = &RigVMFunction_ControlFlowBranch(Condition=RigVMVariableNode() {variable=Is V2}.Value, ExecuteContext=@RigVMFunction_Sequence_7.D)
   n6 = &RigVMFunction_Sequence(ExecuteContext=@RigVMFunction_ControlFlowBranch.True)
   n7 = RigVMVariableNode() {variable=Local TweakerGUI}
   n8 = &RigUnit_HierarchyAddControlFloat(
    InitialValue=0.0
    Name=CTRL_R_eye_faceScrunch
    OffsetSpace=LocalSpace
    OffsetTransform=(Rotation=(X=0,Y=0,Z=-1,W=0),Translation=(X=13.8909,Y=19.7318,Z=0),Scale3D=(X=1,Y=1,Z=1))
    Settings="(PrimaryAxis=Y,bIsScale=False,Limits=(Limit=(bMinimum=True,bMaximum=True),MinValue=0,MaxValue=1,bDrawLimits=True),Shape=(bVisible=True,Name=\"DefaultGizmoLibrary.Sphere_Solid\",Color=(R=1,G=1,B=0,A=0),Transform=(Rotation=(X=0,Y=0,Z=0,W=1),Translation=(X=0,Y= ...(truncated)"
    ExecutePin=@n6.A
    Parent=@n7.Value
   ) {template="SpawnControl::Execute(in InitialValue,in Settings,in OffsetTransform,in OffsetSpace,in Parent,in Name,out Item)"}
   > &RigUnit_SetMetadataTag(NameSpace=Self, Tag=FN_MH_V2_Ctrl, Item=@n8.Item)
   > n9 = &RigUnit_HierarchyAddControlFloat(
    InitialValue=0.0
    Name=CTRL_L_eye_faceScrunch
    OffsetSpace=LocalSpace
    OffsetTransform=(Rotation=(X=0,Y=0,Z=-1,W=0),Translation=(X=19.0221,Y=19.7318,Z=0),Scale3D=(X=1,Y=1,Z=1))
    Settings="(PrimaryAxis=Y,bIsScale=False,Limits=(Limit=(bMinimum=True,bMaximum=True),MinValue=0,MaxValue=1,bDrawLimits=True),Shape=(bVisible=True,Name=\"DefaultGizmoLibrary.Sphere_Solid\",Color=(R=1,G=1,B=0,A=0),Transform=(Rotation=(X=0,Y=0,Z=0,W=1),Translation=(X=0,Y= ...(truncated)"
    Parent=@n7.Value
   ) {template="SpawnControl::Execute(in InitialValue,in Settings,in OffsetTransform,in OffsetSpace,in Parent,in Name,out Item)"}
   > &RigUnit_SetMetadataTag(NameSpace=Self, Tag=FN_MH_V2_Ctrl, Item=@n9.Item)
   > &RigUnit_HierarchyAddControlFloat(
    InitialValue=0.0
    Name=CTRL_L_nose_nasolabialDeepen
    OffsetSpace=LocalSpace
    OffsetTransform=(Rotation=(X=0,Y=0,Z=-0.2588,W=0.9659),Translation=(X=17.4618,Y=15.7771,Z=0),Scale3D=(X=1,Y=1,Z=1))
    Settings="(PrimaryAxis=Y,bIsScale=False,Limits=(Limit=(bMinimum=True,bMaximum=True),MinValue=0,MaxValue=1,bDrawLimits=True),Shape=(bVisible=True,Name=\"DefaultGizmoLibrary.Sphere_Solid\",Color=(R=1,G=1,B=0,A=0),Transform=(Rotation=(X=0,Y=0,Z=0,W=1),Translation=(X=0,Y= ...(truncated)"
    Parent=@n7.Value
   ) {template="SpawnControl::Execute(in InitialValue,in Settings,in OffsetTransform,in OffsetSpace,in Parent,in Name,out Item)"}
   > &RigUnit_SetMetadataTag(Item=(Type=Bone,Name="None"), NameSpace=Self, Tag=FN_MH_V2_Ctrl)
   > &RigUnit_HierarchyAddControlFloat(
    InitialValue=0.0
    Name=CTRL_R_nose_nasolabialDeepen
    OffsetSpace=LocalSpace
    OffsetTransform=(Rotation=(X=0,Y=0,Z=0.2588,W=0.9659),Translation=(X=15.4658,Y=15.7771,Z=0),Scale3D=(X=1,Y=1,Z=1))
    Settings="(PrimaryAxis=Y,bIsScale=False,Limits=(Limit=(bMinimum=True,bMaximum=True),MinValue=0,MaxValue=1,bDrawLimits=True),Shape=(bVisible=True,Name=\"DefaultGizmoLibrary.Sphere_Solid\",Color=(R=1,G=1,B=0,A=0),Transform=(Rotation=(X=0,Y=0,Z=0,W=1),Translation=(X=0,Y= ...(truncated)"
    Parent=@n7.Value
   ) {template="SpawnControl::Execute(in InitialValue,in Settings,in OffsetTransform,in OffsetSpace,in Parent,in Name,out Item)"}
   > &RigUnit_SetMetadataTag(Item=(Type=Bone,Name="None"), NameSpace=Self, Tag=FN_MH_V2_Ctrl)
   n10 = RigVMVariableNode() {variable=Local TweakerGUI}
   > n11 = &RigUnit_HierarchyAddControlVector2D(
    InitialValue=(X=0,Y=0)
    Name=CTRL_L_brow_inner
    OffsetSpace=LocalSpace
    OffsetTransform=(Rotation=(X=0,Y=0,Z=0,W=1),Translation=(X=17.5,Y=21.9658,Z=0),Scale3D=(X=0.3,Y=0.3,Z=0.3))
    Settings="(PrimaryAxis=Z,Limits=(LimitX=(bMinimum=True,bMaximum=True),LimitY=(bMinimum=True,bMaximum=True),MinValue=(X=-1,Y=-1),MaxValue=(X=1,Y=1),bDrawLimits=True),Shape=(bVisible=True,Name=\"DefaultGizmoLibraryNormalized.Sphere_Solid\",Color=(R=0,G=0.0375,B=1,A=1),T ...(truncated)"
    Parent=@n10.Value
   ) {template="SpawnControl::Execute(in InitialValue,in Settings,in OffsetTransform,in OffsetSpace,in Parent,in Name,out Item)"}
   > &RigUnit_SetMetadataTag(NameSpace=Self, Tag=FN_MH_V2_Ctrl, Item=@n11.Item)
   > n12 = &RigUnit_HierarchyAddControlVector2D(
    InitialValue=(X=0,Y=0)
    Name=CTRL_L_brow_innerMid
    OffsetSpace=LocalSpace
    OffsetTransform=(Rotation=(X=0,Y=0,Z=0,W=1),Translation=(X=18.5,Y=21.9,Z=0),Scale3D=(X=0.3,Y=0.3,Z=0.3))
    Settings="(PrimaryAxis=Z,Limits=(LimitX=(bMinimum=True,bMaximum=True),LimitY=(bMinimum=True,bMaximum=True),MinValue=(X=-1,Y=-1),MaxValue=(X=1,Y=1),bDrawLimits=True),Shape=(bVisible=True,Name=\"DefaultGizmoLibraryNormalized.Sphere_Solid\",Color=(R=0,G=0.0375,B=1,A=1),T ...(truncated)"
    Parent=@n10.Value
   ) {template="SpawnControl::Execute(in InitialValue,in Settings,in OffsetTransform,in OffsetSpace,in Parent,in Name,out Item)"}
   > &RigUnit_SetMetadataTag(NameSpace=Self, Tag=FN_MH_V2_Ctrl, Item=@n12.Item)
   > n13 = &RigUnit_HierarchyAddControlVector2D(
    InitialValue=(X=0,Y=0)
    Name=CTRL_L_brow_outerMid
    OffsetSpace=LocalSpace
    OffsetTransform=(Rotation=(X=0,Y=0,Z=0,W=1),Translation=(X=19.5,Y=21.8,Z=0),Scale3D=(X=0.3,Y=0.3,Z=0.3))
    Settings="(PrimaryAxis=Z,Limits=(LimitX=(bMinimum=True,bMaximum=True),LimitY=(bMinimum=True,bMaximum=True),MinValue=(X=-1,Y=-1),MaxValue=(X=1,Y=1),bDrawLimits=True),Shape=(bVisible=True,Name=\"DefaultGizmoLibraryNormalized.Sphere_Solid\",Color=(R=0,G=0.0375,B=1,A=1),T ...(truncated)"
    Parent=@n10.Value
   ) {template="SpawnControl::Execute(in InitialValue,in Settings,in OffsetTransform,in OffsetSpace,in Parent,in Name,out Item)"}
   > &RigUnit_SetMetadataTag(NameSpace=Self, Tag=FN_MH_V2_Ctrl, Item=@n13.Item)
   > n14 = &RigUnit_HierarchyAddControlFloat(
    InitialValue=0.0
    Name=CTRL_L_brow_outer
    OffsetSpace=LocalSpace
    OffsetTransform=(Rotation=(X=0,Y=0,Z=0,W=1),Translation=(X=20.3,Y=21.7,Z=0),Scale3D=(X=0.4,Y=0.4,Z=0.4))
    Settings="(PrimaryAxis=Y,bIsScale=False,Limits=(Limit=(bMinimum=True,bMaximum=True),MinValue=-1,MaxValue=1,bDrawLimits=True),Shape=(bVisible=True,Name=\"DefaultGizmoLibraryNormalized.Sphere_Solid\",Color=(R=0,G=0.0375,B=1,A=1),Transform=(Rotation=(X=0,Y=0,Z=0,W=1),Tra ...(truncated)"
    Parent=@n10.Value
   ) {template="SpawnControl::Execute(in InitialValue,in Settings,in OffsetTransform,in OffsetSpace,in Parent,in Name,out Item)"}
   > &RigUnit_SetMetadataTag(NameSpace=Self, Tag=FN_MH_V2_Ctrl, Item=@n14.Item)
   n15 = RigVMVariableNode() {variable=Local TweakerGUI}
   > n16 = &RigUnit_HierarchyAddControlVector2D(
    InitialValue=(X=0,Y=0)
    Name=CTRL_R_brow_inner
    OffsetSpace=LocalSpace
    OffsetTransform=(Rotation=(X=0,Y=-1,Z=0,W=0),Translation=(X=15.5,Y=21.9658,Z=0),Scale3D=(X=0.3,Y=0.3,Z=0.3))
    Settings="(PrimaryAxis=Z,Limits=(LimitX=(bMinimum=True,bMaximum=True),LimitY=(bMinimum=True,bMaximum=True),MinValue=(X=-1,Y=-1),MaxValue=(X=1,Y=1),bDrawLimits=True),Shape=(bVisible=True,Name=\"DefaultGizmoLibraryNormalized.Sphere_Solid\",Color=(R=0,G=0.0375,B=1,A=1),T ...(truncated)"
    Parent=@n15.Value
   ) {template="SpawnControl::Execute(in InitialValue,in Settings,in OffsetTransform,in OffsetSpace,in Parent,in Name,out Item)"}
   > &RigUnit_SetMetadataTag(NameSpace=Self, Tag=FN_MH_V2_Ctrl, Item=@n16.Item)
   > n17 = &RigUnit_HierarchyAddControlVector2D(
    InitialValue=(X=0,Y=0)
    Name=CTRL_R_brow_innerMid
    OffsetSpace=LocalSpace
    OffsetTransform=(Rotation=(X=0,Y=-1,Z=0,W=0),Translation=(X=14.5,Y=21.9,Z=0),Scale3D=(X=0.3,Y=0.3,Z=0.3))
    Settings="(PrimaryAxis=Z,Limits=(LimitX=(bMinimum=True,bMaximum=True),LimitY=(bMinimum=True,bMaximum=True),MinValue=(X=-1,Y=-1),MaxValue=(X=1,Y=1),bDrawLimits=True),Shape=(bVisible=True,Name=\"DefaultGizmoLibraryNormalized.Sphere_Solid\",Color=(R=0,G=0.0375,B=1,A=1),T ...(truncated)"
    Parent=@n15.Value
   ) {template="SpawnControl::Execute(in InitialValue,in Settings,in OffsetTransform,in OffsetSpace,in Parent,in Name,out Item)"}
   > &RigUnit_SetMetadataTag(NameSpace=Self, Tag=FN_MH_V2_Ctrl, Item=@n17.Item)
   > n18 = &RigUnit_HierarchyAddControlVector2D(
    InitialValue=(X=0,Y=0)
    Name=CTRL_R_brow_outerMid
    OffsetSpace=LocalSpace
    OffsetTransform=(Rotation=(X=0,Y=-1,Z=0,W=0),Translation=(X=13.5,Y=21.8,Z=0),Scale3D=(X=0.3,Y=0.3,Z=0.3))
    Settings="(PrimaryAxis=Z,Limits=(LimitX=(bMinimum=True,bMaximum=True),LimitY=(bMinimum=True,bMaximum=True),MinValue=(X=-1,Y=-1),MaxValue=(X=1,Y=1),bDrawLimits=True),Shape=(bVisible=True,Name=\"DefaultGizmoLibraryNormalized.Sphere_Solid\",Color=(R=0,G=0.0375,B=1,A=1),T ...(truncated)"
    Parent=@n15.Value
   ) {template="SpawnControl::Execute(in InitialValue,in Settings,in OffsetTransform,in OffsetSpace,in Parent,in Name,out Item)"}
   > &RigUnit_SetMetadataTag(NameSpace=Self, Tag=FN_MH_V2_Ctrl, Item=@n18.Item)
   > n19 = &RigUnit_HierarchyAddControlFloat(
    InitialValue=0.0
    Name=CTRL_R_brow_outer
    OffsetSpace=LocalSpace
    OffsetTransform=(Rotation=(X=0,Y=-1,Z=0,W=0),Translation=(X=12.7,Y=21.7,Z=0),Scale3D=(X=0.4,Y=0.4,Z=0.4))
    Settings="(PrimaryAxis=Y,bIsScale=False,Limits=(Limit=(bMinimum=True,bMaximum=True),MinValue=-1,MaxValue=1,bDrawLimits=True),Shape=(bVisible=True,Name=\"DefaultGizmoLibraryNormalized.Sphere_Solid\",Color=(R=0,G=0.0375,B=1,A=1),Transform=(Rotation=(X=0,Y=0,Z=0,W=1),Tra ...(truncated)"
    Parent=@n15.Value
   ) {template="SpawnControl::Execute(in InitialValue,in Settings,in OffsetTransform,in OffsetSpace,in Parent,in Name,out Item)"}
   > &RigUnit_SetMetadataTag(NameSpace=Self, Tag=FN_MH_V2_Ctrl, Item=@n19.Item)
   n20 = RigVMVariableNode() {variable=Local TweakerGUI}
   n21 = &RigUnit_HierarchyAddControlFloat(
    InitialValue=0.0
    Name=CTRL_L_mouth_thicknessU
    OffsetSpace=LocalSpace
    OffsetTransform=(Rotation=(X=0,Y=0,Z=-1,W=0),Translation=(X=13.359,Y=12.636,Z=0),Scale3D=(X=0.75,Y=0.75,Z=0.75))
    Settings="(PrimaryAxis=Y,bIsScale=False,Limits=(Limit=(bMinimum=True,bMaximum=True),MinValue=-1,MaxValue=1,bDrawLimits=True),Shape=(bVisible=True,Name=\"DefaultGizmoLibrary.Sphere_Solid\",Color=(R=1,G=1,B=0,A=0),Transform=(Rotation=(X=0,Y=0,Z=0,W=1),Translation=(X=0,Y ...(truncated)"
    Parent=@n20.Value
    ExecutePin=@n6.B
   ) {template="SpawnControl::Execute(in InitialValue,in Settings,in OffsetTransform,in OffsetSpace,in Parent,in Name,out Item)"}
   > &RigUnit_SetMetadataTag(NameSpace=Self, Tag=FN_MH_V2_Ctrl, Item=@n21.Item)
   > n22 = &RigUnit_HierarchyAddControlFloat(
    InitialValue=0.0
    Name=CTRL_L_mouth_thicknessD
    OffsetSpace=LocalSpace
    OffsetTransform=(Rotation=(X=0,Y=0,Z=0,W=1),Translation=(X=13.359,Y=10.931,Z=0),Scale3D=(X=0.75,Y=0.75,Z=0.75))
    Settings="(PrimaryAxis=Y,bIsScale=False,Limits=(Limit=(bMinimum=True,bMaximum=True),MinValue=-1,MaxValue=1,bDrawLimits=True),Shape=(bVisible=True,Name=\"DefaultGizmoLibrary.Sphere_Solid\",Color=(R=1,G=1,B=0,A=0),Transform=(Rotation=(X=0,Y=0,Z=0,W=1),Translation=(X=0,Y ...(truncated)"
    Parent=@n20.Value
   ) {template="SpawnControl::Execute(in InitialValue,in Settings,in OffsetTransform,in OffsetSpace,in Parent,in Name,out Item)"}
   > &RigUnit_SetMetadataTag(NameSpace=Self, Tag=FN_MH_V2_Ctrl, Item=@n22.Item)
   > n23 = &RigUnit_HierarchyAddControlFloat(
    InitialValue=0.0
    Name=CTRL_R_mouth_thicknessU
    OffsetSpace=LocalSpace
    OffsetTransform=(Rotation=(X=0,Y=0,Z=-1,W=0),Translation=(X=12.908,Y=12.636,Z=0),Scale3D=(X=0.75,Y=0.75,Z=0.75))
    Settings="(PrimaryAxis=Y,bIsScale=False,Limits=(Limit=(bMinimum=True,bMaximum=True),MinValue=-1,MaxValue=1,bDrawLimits=True),Shape=(bVisible=True,Name=\"DefaultGizmoLibrary.Sphere_Solid\",Color=(R=1,G=1,B=0,A=0),Transform=(Rotation=(X=0,Y=0,Z=0,W=1),Translation=(X=0,Y ...(truncated)"
    Parent=@n20.Value
   ) {template="SpawnControl::Execute(in InitialValue,in Settings,in OffsetTransform,in OffsetSpace,in Parent,in Name,out Item)"}
   > &RigUnit_SetMetadataTag(NameSpace=Self, Tag=FN_MH_V2_Ctrl, Item=@n23.Item)
   > n24 = &RigUnit_HierarchyAddControlFloat(
    InitialValue=0.0
    Name=CTRL_R_mouth_thicknessD
    OffsetSpace=LocalSpace
    OffsetTransform=(Rotation=(X=0,Y=0,Z=0,W=1),Translation=(X=12.908,Y=10.931,Z=0),Scale3D=(X=0.75,Y=0.75,Z=0.75))
    Settings="(PrimaryAxis=Y,bIsScale=False,Limits=(Limit=(bMinimum=True,bMaximum=True),MinValue=-1,MaxValue=1,bDrawLimits=True),Shape=(bVisible=True,Name=\"DefaultGizmoLibrary.Sphere_Solid\",Color=(R=1,G=1,B=0,A=0),Transform=(Rotation=(X=0,Y=0,Z=0,W=1),Translation=(X=0,Y ...(truncated)"
    Parent=@n20.Value
   ) {template="SpawnControl::Execute(in InitialValue,in Settings,in OffsetTransform,in OffsetSpace,in Parent,in Name,out Item)"}
   > &RigUnit_SetMetadataTag(NameSpace=Self, Tag=FN_MH_V2_Ctrl, Item=@n24.Item)
   n25 = RigVMVariableNode() {variable=Local TweakerGUI}
   > n26 = &RigUnit_HierarchyAddControlFloat(
    InitialValue=0.0
    Name=CTRL_L_mouth_thicknessInwardU
    OffsetSpace=LocalSpace
    OffsetTransform=(Rotation=(X=0,Y=0,Z=-1,W=0),Translation=(X=14.615,Y=12.636,Z=0),Scale3D=(X=0.75,Y=0.75,Z=0.75))
    Settings="(PrimaryAxis=Y,bIsScale=False,Limits=(Limit=(bMinimum=True,bMaximum=True),MinValue=-1,MaxValue=1,bDrawLimits=True),Shape=(bVisible=True,Name=\"DefaultGizmoLibrary.Sphere_Solid\",Color=(R=1,G=1,B=0,A=0),Transform=(Rotation=(X=0,Y=0,Z=0,W=1),Translation=(X=0,Y ...(truncated)"
    Parent=@n25.Value
   ) {template="SpawnControl::Execute(in InitialValue,in Settings,in OffsetTransform,in OffsetSpace,in Parent,in Name,out Item)"}
   > &RigUnit_SetMetadataTag(NameSpace=Self, Tag=FN_MH_V2_Ctrl, Item=@n26.Item)
   > n27 = &RigUnit_HierarchyAddControlFloat(
    InitialValue=0.0
    Name=CTRL_L_mouth_thicknessInwardD
    OffsetSpace=LocalSpace
    OffsetTransform=(Rotation=(X=0,Y=0,Z=0,W=1),Translation=(X=14.615,Y=10.931,Z=0),Scale3D=(X=0.75,Y=0.75,Z=0.75))
    Settings="(PrimaryAxis=Y,bIsScale=False,Limits=(Limit=(bMinimum=True,bMaximum=True),MinValue=-1,MaxValue=1,bDrawLimits=True),Shape=(bVisible=True,Name=\"DefaultGizmoLibrary.Sphere_Solid\",Color=(R=1,G=1,B=0,A=0),Transform=(Rotation=(X=0,Y=0,Z=0,W=1),Translation=(X=0,Y ...(truncated)"
    Parent=@n25.Value
   ) {template="SpawnControl::Execute(in InitialValue,in Settings,in OffsetTransform,in OffsetSpace,in Parent,in Name,out Item)"}
   > &RigUnit_SetMetadataTag(NameSpace=Self, Tag=FN_MH_V2_Ctrl, Item=@n27.Item)
   > n28 = &RigUnit_HierarchyAddControlFloat(
    InitialValue=0.0
    Name=CTRL_R_mouth_thicknessInwardU
    OffsetSpace=LocalSpace
    OffsetTransform=(Rotation=(X=0,Y=0,Z=-1,W=0),Translation=(X=14.164,Y=12.636,Z=0),Scale3D=(X=0.75,Y=0.75,Z=0.75))
    Settings="(PrimaryAxis=Y,bIsScale=False,Limits=(Limit=(bMinimum=True,bMaximum=True),MinValue=-1,MaxValue=1,bDrawLimits=True),Shape=(bVisible=True,Name=\"DefaultGizmoLibrary.Sphere_Solid\",Color=(R=1,G=1,B=0,A=0),Transform=(Rotation=(X=0,Y=0,Z=0,W=1),Translation=(X=0,Y ...(truncated)"
    Parent=@n25.Value
   ) {template="SpawnControl::Execute(in InitialValue,in Settings,in OffsetTransform,in OffsetSpace,in Parent,in Name,out Item)"}
   > &RigUnit_SetMetadataTag(NameSpace=Self, Tag=FN_MH_V2_Ctrl, Item=@n28.Item)
   > n29 = &RigUnit_HierarchyAddControlFloat(
    InitialValue=0.0
    Name=CTRL_R_mouth_thicknessInwardD
    OffsetSpace=LocalSpace
    OffsetTransform=(Rotation=(X=0,Y=0,Z=0,W=1),Translation=(X=14.164,Y=10.931,Z=0),Scale3D=(X=0.75,Y=0.75,Z=0.75))
    Settings="(PrimaryAxis=Y,bIsScale=False,Limits=(Limit=(bMinimum=True,bMaximum=True),MinValue=-1,MaxValue=1,bDrawLimits=True),Shape=(bVisible=True,Name=\"DefaultGizmoLibrary.Sphere_Solid\",Color=(R=1,G=1,B=0,A=0),Transform=(Rotation=(X=0,Y=0,Z=0,W=1),Translation=(X=0,Y ...(truncated)"
    Parent=@n25.Value
   ) {template="SpawnControl::Execute(in InitialValue,in Settings,in OffsetTransform,in OffsetSpace,in Parent,in Name,out Item)"}
   > &RigUnit_SetMetadataTag(NameSpace=Self, Tag=FN_MH_V2_Ctrl, Item=@n29.Item)
   n30 = RigVMVariableNode() {variable=Local TweakerGUI}
   > n31 = &RigUnit_HierarchyAddControlFloat(
    InitialValue=0.0
    Name=CTRL_L_mouth_lipsRollU
    OffsetSpace=LocalSpace
    OffsetTransform=(Rotation=(X=0,Y=0,Z=-1,W=0),Translation=(X=16.047,Y=12.636,Z=0),Scale3D=(X=0.75,Y=0.75,Z=0.75))
    Settings="(PrimaryAxis=Y,bIsScale=False,Limits=(Limit=(bMinimum=True,bMaximum=True),MinValue=-1,MaxValue=1,bDrawLimits=True),Shape=(bVisible=True,Name=\"DefaultGizmoLibrary.Sphere_Solid\",Color=(R=1,G=1,B=0,A=0),Transform=(Rotation=(X=0,Y=0,Z=0,W=1),Translation=(X=0,Y ...(truncated)"
    Parent=@n30.Value
   ) {template="SpawnControl::Execute(in InitialValue,in Settings,in OffsetTransform,in OffsetSpace,in Parent,in Name,out Item)"}
   > &RigUnit_SetMetadataTag(NameSpace=Self, Tag=FN_MH_V2_Ctrl, Item=@n31.Item)
   > n32 = &RigUnit_HierarchyAddControlFloat(
    InitialValue=0.0
    Name=CTRL_L_mouth_lipsRollD
    OffsetSpace=LocalSpace
    OffsetTransform=(Rotation=(X=0,Y=0,Z=0,W=1),Translation=(X=16.047,Y=10.931,Z=0),Scale3D=(X=0.75,Y=0.75,Z=0.75))
    Settings="(PrimaryAxis=Y,bIsScale=False,Limits=(Limit=(bMinimum=True,bMaximum=True),MinValue=-1,MaxValue=1,bDrawLimits=True),Shape=(bVisible=True,Name=\"DefaultGizmoLibrary.Sphere_Solid\",Color=(R=1,G=1,B=0,A=0),Transform=(Rotation=(X=0,Y=0,Z=0,W=1),Translation=(X=0,Y ...(truncated)"
    Parent=@n30.Value
   ) {template="SpawnControl::Execute(in InitialValue,in Settings,in OffsetTransform,in OffsetSpace,in Parent,in Name,out Item)"}
   > &RigUnit_SetMetadataTag(NameSpace=Self, Tag=FN_MH_V2_Ctrl, Item=@n32.Item)
   > n33 = &RigUnit_HierarchyAddControlFloat(
    InitialValue=0.0
    Name=CTRL_R_mouth_lipsRollU
    OffsetSpace=LocalSpace
    OffsetTransform=(Rotation=(X=0,Y=0,Z=-1,W=0),Translation=(X=15.596,Y=12.636,Z=0),Scale3D=(X=0.75,Y=0.75,Z=0.75))
    Settings="(PrimaryAxis=Y,bIsScale=False,Limits=(Limit=(bMinimum=True,bMaximum=True),MinValue=-1,MaxValue=1,bDrawLimits=True),Shape=(bVisible=True,Name=\"DefaultGizmoLibrary.Sphere_Solid\",Color=(R=1,G=1,B=0,A=0),Transform=(Rotation=(X=0,Y=0,Z=0,W=1),Translation=(X=0,Y ...(truncated)"
    Parent=@n30.Value
   ) {template="SpawnControl::Execute(in InitialValue,in Settings,in OffsetTransform,in OffsetSpace,in Parent,in Name,out Item)"}
   > &RigUnit_SetMetadataTag(NameSpace=Self, Tag=FN_MH_V2_Ctrl, Item=@n33.Item)
   > n34 = &RigUnit_HierarchyAddControlFloat(
    InitialValue=0.0
    Name=CTRL_R_mouth_lipsRollD
    OffsetSpace=LocalSpace
    OffsetTransform=(Rotation=(X=0,Y=0,Z=0,W=1),Translation=(X=15.596,Y=10.931,Z=0),Scale3D=(X=0.75,Y=0.75,Z=0.75))
    Settings="(PrimaryAxis=Y,bIsScale=False,Limits=(Limit=(bMinimum=True,bMaximum=True),MinValue=-1,MaxValue=1,bDrawLimits=True),Shape=(bVisible=True,Name=\"DefaultGizmoLibrary.Sphere_Solid\",Color=(R=1,G=1,B=0,A=0),Transform=(Rotation=(X=0,Y=0,Z=0,W=1),Translation=(X=0,Y ...(truncated)"
    Parent=@n30.Value
   ) {template="SpawnControl::Execute(in InitialValue,in Settings,in OffsetTransform,in OffsetSpace,in Parent,in Name,out Item)"}
   > &RigUnit_SetMetadataTag(NameSpace=Self, Tag=FN_MH_V2_Ctrl, Item=@n34.Item)
   n35 = RigVMVariableNode() {variable=Local TweakerGUI}
   > n36 = &RigUnit_HierarchyAddControlFloat(
    InitialValue=0.0
    Name=CTRL_L_mouth_pushPullU
    OffsetSpace=LocalSpace
    OffsetTransform=(Rotation=(X=0,Y=0,Z=-0.3827,W=0.9239),Translation=(X=20.135,Y=12.402,Z=0),Scale3D=(X=0.75,Y=0.75,Z=0.75))
    Settings="(PrimaryAxis=Y,bIsScale=False,Limits=(Limit=(bMinimum=True,bMaximum=True),MinValue=-1,MaxValue=1,bDrawLimits=True),Shape=(bVisible=True,Name=\"DefaultGizmoLibrary.Sphere_Solid\",Color=(R=1,G=1,B=0,A=0),Transform=(Rotation=(X=0,Y=0,Z=0,W=1),Translation=(X=0,Y ...(truncated)"
    Parent=@n35.Value
   ) {template="SpawnControl::Execute(in InitialValue,in Settings,in OffsetTransform,in OffsetSpace,in Parent,in Name,out Item)"}
   > &RigUnit_SetMetadataTag(NameSpace=Self, Tag=FN_MH_V2_Ctrl, Item=@n36.Item)
   > n37 = &RigUnit_HierarchyAddControlFloat(
    InitialValue=0.0
    Name=CTRL_L_mouth_pushPullD
    OffsetSpace=LocalSpace
    OffsetTransform=(Rotation=(X=0,Y=0,Z=-0.9239,W=0.3827),Translation=(X=20.164,Y=11.218,Z=0),Scale3D=(X=0.75,Y=0.75,Z=0.75))
    Settings="(PrimaryAxis=Y,bIsScale=False,Limits=(Limit=(bMinimum=True,bMaximum=True),MinValue=-1,MaxValue=1,bDrawLimits=True),Shape=(bVisible=True,Name=\"DefaultGizmoLibrary.Sphere_Solid\",Color=(R=1,G=1,B=0,A=0),Transform=(Rotation=(X=0,Y=0,Z=0,W=1),Translation=(X=0,Y ...(truncated)"
    Parent=@n35.Value
   ) {template="SpawnControl::Execute(in InitialValue,in Settings,in OffsetTransform,in OffsetSpace,in Parent,in Name,out Item)"}
   > &RigUnit_SetMetadataTag(NameSpace=Self, Tag=FN_MH_V2_Ctrl, Item=@n37.Item)
   > n38 = &RigUnit_HierarchyAddControlFloat(
    InitialValue=0.0
    Name=CTRL_R_mouth_pushPullU
    OffsetSpace=LocalSpace
    OffsetTransform=(Rotation=(X=0,Y=0,Z=0.3827,W=0.9239),Translation=(X=18.887,Y=12.408,Z=0),Scale3D=(X=0.75,Y=0.75,Z=0.75))
    Settings="(PrimaryAxis=Y,bIsScale=False,Limits=(Limit=(bMinimum=True,bMaximum=True),MinValue=-1,MaxValue=1,bDrawLimits=True),Shape=(bVisible=True,Name=\"DefaultGizmoLibrary.Sphere_Solid\",Color=(R=1,G=1,B=0,A=0),Transform=(Rotation=(X=0,Y=0,Z=0,W=1),Translation=(X=0,Y ...(truncated)"
    Parent=@n35.Value
   ) {template="SpawnControl::Execute(in InitialValue,in Settings,in OffsetTransform,in OffsetSpace,in Parent,in Name,out Item)"}
   > &RigUnit_SetMetadataTag(NameSpace=Self, Tag=FN_MH_V2_Ctrl, Item=@n38.Item)
   > n39 = &RigUnit_HierarchyAddControlFloat(
    InitialValue=0.0
    Name=CTRL_R_mouth_pushPullD
    OffsetSpace=LocalSpace
    OffsetTransform=(Rotation=(X=0,Y=0,Z=0.9239,W=0.3827),Translation=(X=18.886,Y=11.231,Z=0),Scale3D=(X=0.75,Y=0.75,Z=0.75))
    Settings="(PrimaryAxis=Y,bIsScale=False,Limits=(Limit=(bMinimum=True,bMaximum=True),MinValue=-1,MaxValue=1,bDrawLimits=True),Shape=(bVisible=True,Name=\"DefaultGizmoLibrary.Sphere_Solid\",Color=(R=1,G=1,B=0,A=0),Transform=(Rotation=(X=0,Y=0,Z=0,W=1),Translation=(X=0,Y ...(truncated)"
    Parent=@n35.Value
   ) {template="SpawnControl::Execute(in InitialValue,in Settings,in OffsetTransform,in OffsetSpace,in Parent,in Name,out Item)"}
   > &RigUnit_SetMetadataTag(NameSpace=Self, Tag=FN_MH_V2_Ctrl, Item=@n39.Item)
   n40 = RigVMVariableNode() {variable=Local TweakerGUI}
   &RigUnit_HierarchyAddControlVector2D(
    InitialValue=(X=0,Y=0)
    Name=CTRL_L_eye_eyelidU
    OffsetSpace=LocalSpace
    OffsetTransform=(Rotation=(X=0,Y=0,Z=1,W=0),Translation=(X=19.9388,Y=18.9689,Z=0),Scale3D=(X=0.3,Y=0.3,Z=0.3))
    Settings="(PrimaryAxis=Z,Limits=(LimitX=(bMinimum=True,bMaximum=True),LimitY=(bMinimum=True,bMaximum=True),MinValue=(X=-1,Y=-1),MaxValue=(X=1,Y=1),bDrawLimits=True),Shape=(bVisible=True,Name=\"DefaultGizmoLibraryNormalized.Sphere_Solid\",Color=(R=0,G=0.0375,B=1,A=1),T ...(truncated)"
    Parent=@n40.Value
    ExecutePin=@RigVMFunction_Sequence_7.A
   ) {template="SpawnControl::Execute(in InitialValue,in Settings,in OffsetTransform,in OffsetSpace,in Parent,in Name,out Item)"}
   > &RigUnit_HierarchyAddControlVector2D(
    InitialValue=(X=0,Y=0)
    Name=CTRL_L_eye_eyelidD
    OffsetSpace=LocalSpace
    OffsetTransform=(Rotation=(X=0,Y=-1,Z=0,W=0),Translation=(X=19.9393,Y=17.7511,Z=0),Scale3D=(X=0.3,Y=0.3,Z=0.3))
    Settings="(PrimaryAxis=Z,Limits=(LimitX=(bMinimum=True,bMaximum=True),LimitY=(bMinimum=True,bMaximum=True),MinValue=(X=-1,Y=-1),MaxValue=(X=1,Y=1),bDrawLimits=True),Shape=(bVisible=True,Name=\"DefaultGizmoLibraryNormalized.Sphere_Solid\",Color=(R=0,G=0.0375,B=1,A=1),T ...(truncated)"
    Parent=@n40.Value
   ) {template="SpawnControl::Execute(in InitialValue,in Settings,in OffsetTransform,in OffsetSpace,in Parent,in Name,out Item)"}
   > &RigUnit_HierarchyAddControlVector2D(
    InitialValue=(X=0,Y=0)
    Name=CTRL_R_eye_eyelidU
    OffsetSpace=LocalSpace
    OffsetTransform=(Rotation=(X=-1,Y=0,Z=0,W=0),Translation=(X=12.9888,Y=18.9675,Z=0.0003),Scale3D=(X=0.3,Y=0.3,Z=0.3))
    Settings="(PrimaryAxis=Z,Limits=(LimitX=(bMinimum=True,bMaximum=True),LimitY=(bMinimum=True,bMaximum=True),MinValue=(X=-1,Y=-1),MaxValue=(X=1,Y=1),bDrawLimits=True),Shape=(bVisible=True,Name=\"DefaultGizmoLibraryNormalized.Sphere_Solid\",Color=(R=0,G=0.0375,B=1,A=1),T ...(truncated)"
    Parent=@n40.Value
   ) {template="SpawnControl::Execute(in InitialValue,in Settings,in OffsetTransform,in OffsetSpace,in Parent,in Name,out Item)"}
   > &RigUnit_HierarchyAddControlVector2D(
    InitialValue=(X=0,Y=0)
    Name=CTRL_R_eye_eyelidD
    OffsetSpace=LocalSpace
    OffsetTransform=(Rotation=(X=0,Y=0,Z=0,W=1),Translation=(X=12.9888,Y=17.7524,Z=0.0003),Scale3D=(X=0.3,Y=0.3,Z=0.3))
    Settings="(PrimaryAxis=Z,Limits=(LimitX=(bMinimum=True,bMaximum=True),LimitY=(bMinimum=True,bMaximum=True),MinValue=(X=-1,Y=-1),MaxValue=(X=1,Y=1),bDrawLimits=True),Shape=(bVisible=True,Name=\"DefaultGizmoLibraryNormalized.Sphere_Solid\",Color=(R=0,G=0.0375,B=1,A=1),T ...(truncated)"
    Parent=@n40.Value
   ) {template="SpawnControl::Execute(in InitialValue,in Settings,in OffsetTransform,in OffsetSpace,in Parent,in Name,out Item)"}
   > &RigUnit_HierarchyAddControlVector2D(
    InitialValue=(X=0,Y=0)
    Name=CTRL_C_nose
    OffsetSpace=LocalSpace
    OffsetTransform=(Rotation=(X=0,Y=0,Z=0,W=1),Translation=(X=16.5,Y=15.4658,Z=0),Scale3D=(X=0.32,Y=0.32,Z=0.32))
    Settings="(PrimaryAxis=Z,Limits=(LimitX=(bMinimum=True,bMaximum=True),LimitY=(bMinimum=True,bMaximum=True),MinValue=(X=-1,Y=-1),MaxValue=(X=1,Y=1),bDrawLimits=True),Shape=(bVisible=True,Name=\"DefaultGizmoLibraryNormalized.Sphere_Solid\",Color=(R=0,G=0.0375,B=1,A=1),T ...(truncated)"
    Parent=RigVMVariableNode() {variable=Local TweakerGUI}.Value
   ) {template="SpawnControl::Execute(in InitialValue,in Settings,in OffsetTransform,in OffsetSpace,in Parent,in Name,out Item)"}
   > n41 = &RigUnit_HierarchyAddNull(
    Name=GRP_faceAndEyesAimFollowHeadGUI
    Space=LocalSpace
    Transform=(Rotation=(X=-0.7071,Y=0,Z=0,W=0.7071),Translation=(X=16.5,Y=2,Z=0),Scale3D=(X=1,Y=1,Z=1))
    Parent=RigVMVariableNode() {variable=Local TweakerGUI}.Value
   )
   > n42 = /Script/ControlRig.RigUnit_HierarchyAddControlTransform(
    InitialValue=(Rotation=(X=0,Y=0,Z=0,W=1),Translation=(X=0,Y=0,Z=0),Scale3D=(X=1,Y=1,Z=1))
    Name=FRM_faceAndEyesAimFollowHeadGUI
    OffsetSpace=LocalSpace
    OffsetTransform=(Rotation=(X=0,Y=0,Z=0,W=1),Translation=(X=0,Y=0,Z=0),Scale3D=(X=1,Y=1,Z=1))
    Settings="(InitialSpace=LocalSpace,bUsePreferredRotationOrder=False,PreferredRotationOrder=YZX,Limits=(LimitTranslationX=(bMinimum=False,bMaximum=False),LimitTranslationY=(bMinimum=False,bMaximum=False),LimitTranslationZ=(bMinimum=False,bMaximum=False),LimitPitch=(b ...(truncated)"
    Parent=@n41.Item
   ) {
    is=struct RigVMUnitNode
    template="SpawnControl::Execute(in InitialValue,in Settings,in OffsetTransform,in OffsetSpace,in Parent,in Name,out Item)"
   }
   > &RigUnit_HierarchyAddControlFloat(
    InitialValue=1.0
    Name=CTRL_faceGUIfollowHead
    OffsetSpace=LocalSpace
    OffsetTransform=(Rotation=(X=0,Y=0,Z=-0.7071,W=0.7071),Translation=(X=1.55,Y=0,Z=1.92),Scale3D=(X=2,Y=2,Z=2))
    Settings="(PrimaryAxis=Y,bIsScale=False,Limits=(Limit=(bMinimum=True,bMaximum=True),MinValue=0,MaxValue=1,bDrawLimits=True),Shape=(bVisible=True,Name=\"DefaultGizmoLibraryNormalized.Sphere_Solid\",Color=(R=1,G=1,B=0,A=0),Transform=(Rotation=(X=-0.7071,Y=0,Z=0,W=0.7071 ...(truncated)"
    Parent=@n42.Item
   ) {template="SpawnControl::Execute(in InitialValue,in Settings,in OffsetTransform,in OffsetSpace,in Parent,in Name,out Item)"}
   > &RigUnit_HierarchyAddControlFloat(
    InitialValue=0.0
    Name=CTRL_eyesAimFollowHead
    OffsetSpace=LocalSpace
    OffsetTransform=(Rotation=(X=0,Y=0,Z=-0.7071,W=0.7071),Translation=(X=1.55,Y=0,Z=1.06),Scale3D=(X=2,Y=2,Z=2))
    Settings="(PrimaryAxis=Y,bIsScale=False,Limits=(Limit=(bMinimum=True,bMaximum=True),MinValue=0,MaxValue=1,bDrawLimits=True),Shape=(bVisible=True,Name=\"DefaultGizmoLibraryNormalized.Sphere_Solid\",Color=(R=1,G=1,B=0,A=0),Transform=(Rotation=(X=-0.7071,Y=0,Z=0,W=0.7071 ...(truncated)"
    Parent=@n42.Item
   ) {template="SpawnControl::Execute(in InitialValue,in Settings,in OffsetTransform,in OffsetSpace,in Parent,in Name,out Item)"}
   n43 = RigVMVariableNode() {variable=Local TweakerGUI}
   > &RigUnit_HierarchyAddControlVector2D(
    InitialValue=(X=0,Y=0)
    Name=CTRL_L_mouth_corner
    OffsetSpace=LocalSpace
    OffsetTransform=(Rotation=(X=0,Y=0,Z=0,W=1),Translation=(X=21.9933,Y=14.8865,Z=0),Scale3D=(X=0.3,Y=0.3,Z=0.3))
    Settings="(PrimaryAxis=Z,Limits=(LimitX=(bMinimum=True,bMaximum=True),LimitY=(bMinimum=True,bMaximum=True),MinValue=(X=-1,Y=-1),MaxValue=(X=1,Y=1),bDrawLimits=True),Shape=(bVisible=True,Name=\"DefaultGizmoLibraryNormalized.Sphere_Solid\",Color=(R=1,G=1,B=0,A=0),Transf ...(truncated)"
    Parent=@n43.Value
   ) {template="SpawnControl::Execute(in InitialValue,in Settings,in OffsetTransform,in OffsetSpace,in Parent,in Name,out Item)"}
   > &RigUnit_HierarchyAddControlVector2D(
    InitialValue=(X=0,Y=0)
    Name=CTRL_R_mouth_corner
    OffsetSpace=LocalSpace
    OffsetTransform=(Rotation=(X=0,Y=-1,Z=0,W=0),Translation=(X=19.2098,Y=14.8889,Z=0),Scale3D=(X=0.3,Y=0.3,Z=0.3))
    Settings="(PrimaryAxis=Z,Limits=(LimitX=(bMinimum=True,bMaximum=True),LimitY=(bMinimum=True,bMaximum=True),MinValue=(X=-1,Y=-1),MaxValue=(X=1,Y=1),bDrawLimits=True),Shape=(bVisible=True,Name=\"DefaultGizmoLibraryNormalized.Sphere_Solid\",Color=(R=1,G=1,B=0,A=0),Transf ...(truncated)"
    Parent=@n43.Value
   ) {template="SpawnControl::Execute(in InitialValue,in Settings,in OffsetTransform,in OffsetSpace,in Parent,in Name,out Item)"}
   RigVMFunction_MathBoolNot = &RigVMFunction_MathBoolNot(Value=RigVMVariableNode() {variable=Is V2}.Value)
   n44 = &RigVMFunction_ControlFlowBranch(Condition=@RigVMFunction_MathBoolNot.Result, ExecuteContext=@RigVMFunction_Sequence_7.C)
   n45 = RigVMVariableNode() {variable=Local TweakerGUI}
   n46 = &RigUnit_HierarchyAddControlFloat(
    InitialValue=0.0
    Name=CTRL_L_brow_inner
    OffsetSpace=LocalSpace
    OffsetTransform=(Rotation=(X=0,Y=0,Z=0,W=1),Translation=(X=17.5,Y=21.9658,Z=0),Scale3D=(X=0.75,Y=0.75,Z=0.75))
    Settings="(PrimaryAxis=Y,bIsScale=False,Limits=(Limit=(bMinimum=True,bMaximum=True),MinValue=-1,MaxValue=1,bDrawLimits=True),Shape=(bVisible=True,Name=\"DefaultGizmoLibraryNormalized.Sphere_Solid\",Color=(R=0,G=0.0375,B=1,A=1),Transform=(Rotation=(X=0,Y=0,Z=0,W=1),Tra ...(truncated)"
    ExecutePin=@n44.True
    Parent=@n45.Value
   ) {template="SpawnControl::Execute(in InitialValue,in Settings,in OffsetTransform,in OffsetSpace,in Parent,in Name,out Item)"}
   > &RigUnit_SetMetadataTag(NameSpace=Self, Tag=FN_MH_V1_Ctrl, Item=@n46.Item)
   > n47 = &RigUnit_HierarchyAddControlFloat(
    InitialValue=0.0
    Name=CTRL_R_brow_inner
    OffsetSpace=LocalSpace
    OffsetTransform=(Rotation=(X=0,Y=0,Z=0,W=1),Translation=(X=15.5,Y=21.9658,Z=0),Scale3D=(X=0.75,Y=0.75,Z=0.75))
    Settings="(PrimaryAxis=Y,bIsScale=False,Limits=(Limit=(bMinimum=True,bMaximum=True),MinValue=-1,MaxValue=1,bDrawLimits=True),Shape=(bVisible=True,Name=\"DefaultGizmoLibraryNormalized.Sphere_Solid\",Color=(R=0,G=0.0375,B=1,A=1),Transform=(Rotation=(X=0,Y=0,Z=0,W=1),Tra ...(truncated)"
    Parent=@n45.Value
   ) {template="SpawnControl::Execute(in InitialValue,in Settings,in OffsetTransform,in OffsetSpace,in Parent,in Name,out Item)"}
   > &RigUnit_SetMetadataTag(NameSpace=Self, Tag=FN_MH_V1_Ctrl, Item=@n47.Item)
   > n48 = &RigUnit_HierarchyAddControlFloat(
    InitialValue=0.0
    Name=CTRL_L_brow_outer
    OffsetSpace=LocalSpace
    OffsetTransform=(Rotation=(X=0,Y=0,Z=0,W=1),Translation=(X=19,Y=21.4658,Z=0),Scale3D=(X=0.75,Y=0.75,Z=0.75))
    Settings="(PrimaryAxis=Y,bIsScale=False,Limits=(Limit=(bMinimum=True,bMaximum=True),MinValue=-1,MaxValue=1,bDrawLimits=True),Shape=(bVisible=True,Name=\"DefaultGizmoLibraryNormalized.Sphere_Solid\",Color=(R=0,G=0.0375,B=1,A=1),Transform=(Rotation=(X=0,Y=0,Z=0,W=1),Tra ...(truncated)"
    Parent=@n45.Value
   ) {template="SpawnControl::Execute(in InitialValue,in Settings,in OffsetTransform,in OffsetSpace,in Parent,in Name,out Item)"}
   > &RigUnit_SetMetadataTag(NameSpace=Self, Tag=FN_MH_V1_Ctrl, Item=@n48.Item)
   > n49 = &RigUnit_HierarchyAddControlFloat(
    InitialValue=0.0
    Name=CTRL_R_brow_outer
    OffsetSpace=LocalSpace
    OffsetTransform=(Rotation=(X=0,Y=0,Z=0,W=1),Translation=(X=14,Y=21.4658,Z=0),Scale3D=(X=0.75,Y=0.75,Z=0.75))
    Settings="(PrimaryAxis=Y,bIsScale=False,Limits=(Limit=(bMinimum=True,bMaximum=True),MinValue=-1,MaxValue=1,bDrawLimits=True),Shape=(bVisible=True,Name=\"DefaultGizmoLibraryNormalized.Sphere_Solid\",Color=(R=0,G=0.0375,B=1,A=1),Transform=(Rotation=(X=0,Y=0,Z=0,W=1),Tra ...(truncated)"
    Parent=@n45.Value
   ) {template="SpawnControl::Execute(in InitialValue,in Settings,in OffsetTransform,in OffsetSpace,in Parent,in Name,out Item)"}
   > &RigUnit_SetMetadataTag(NameSpace=Self, Tag=FN_MH_V1_Ctrl, Item=@n49.Item)
   RigVMFunctionReturnNode(ExecuteContext=@RigVMFunction_Sequence.B)
  ]}
 New Function_ContainedGraph {nodes [
   RigVMFunctionEntryNode()
   > RigVMFunction_Sequence_1 = RigVMAggregateNode()
   n1 = &RigUnit_HierarchyAddNull(
    Name=FACIAL_BOARD
    Parent=(Type=None,Name="None")
    Space=LocalSpace
    Transform=(Rotation=(X=0,Y=0,Z=0,W=1),Translation=(X=0,Y=0,Z=0),Scale3D=(X=1,Y=1,Z=1))
    ExecutePin=@RigVMFunction_Sequence_1.A
   )
   > RigVMVariableNode(Value=@n1.Item) {variable=Local Top Null}
   > n2 = &RigUnit_HierarchyAddNull(
    Name=Facial_Board_Follow
    Space=LocalSpace
    Transform=(Rotation=(X=0,Y=0,Z=0,W=1),Translation=(X=25,Y=-1.6873,Z=142.0342),Scale3D=(X=1,Y=1,Z=1))
    Parent=@n1.Item
   )
   > n3 = &RigUnit_HierarchyAddControlTransform(
    InitialValue=(Rotation=(X=0,Y=0,Z=0,W=1),Translation=(X=0,Y=0,Z=0),Scale3D=(X=1,Y=1,Z=1))
    Name=CTRL_Master
    OffsetSpace=LocalSpace
    OffsetTransform=(Rotation=(X=0,Y=0,Z=0,W=1),Translation=(X=0,Y=0,Z=-6),Scale3D=(X=1,Y=1,Z=1))
    Settings="(InitialSpace=LocalSpace,bUsePreferredRotationOrder=False,PreferredRotationOrder=YZX,Limits=(LimitTranslationX=(bMinimum=False,bMaximum=False),LimitTranslationY=(bMinimum=False,bMaximum=False),LimitTranslationZ=(bMinimum=False,bMaximum=False),LimitPitch=(b ...(truncated)"
    Parent=@n2.Item
   ) {template="SpawnControl::Execute(in InitialValue,in Settings,in OffsetTransform,in OffsetSpace,in Parent,in Name,out Item)"}
   > n4 = &RigUnit_HierarchyAddControlTransform(
    InitialValue=(Rotation=(X=0,Y=0,Z=0,W=1),Translation=(X=0,Y=0,Z=0),Scale3D=(X=1,Y=1,Z=1))
    Name=Facial_Board_GUI
    OffsetSpace=LocalSpace
    OffsetTransform=(Rotation=(X=0,Y=0,Z=0,W=1),Translation=(X=0,Y=0,Z=6),Scale3D=(X=1,Y=1,Z=1))
    Settings="(InitialSpace=LocalSpace,bUsePreferredRotationOrder=False,PreferredRotationOrder=YZX,Limits=(LimitTranslationX=(bMinimum=False,bMaximum=False),LimitTranslationY=(bMinimum=False,bMaximum=False),LimitTranslationZ=(bMinimum=False,bMaximum=False),LimitPitch=(b ...(truncated)"
    Parent=@n3.Item
   ) {template="SpawnControl::Execute(in InitialValue,in Settings,in OffsetTransform,in OffsetSpace,in Parent,in Name,out Item)"}
   > RigVMVariableNode(Value=@n4.Item) {variable=Local Facial_Board Gui}
   > RigVMFunction_Sequence_4 = RigVMAggregateNode()
   Nose_Grp_CollapseNode = RigVMCollapseNode(ExecutePin=@RigVMFunction_Sequence_4.B, Parent=RigVMVariableNode() {variable=Local Facial_Board Gui}.Value)
   Chin_Grp_CollapseNode = RigVMCollapseNode(ExecutePin=@Nose_Grp_CollapseNode.ExecutePin, Parent=RigVMVariableNode() {variable=Local Facial_Board Gui}.Value)
   Jaw_Grp_CollapseNode = RigVMCollapseNode(ExecutePin=@Chin_Grp_CollapseNode.ExecutePin, Parent=RigVMVariableNode() {variable=Local Facial_Board Gui}.Value)
   Tongue_Grp_CollapseNode = RigVMCollapseNode(ExecutePin=@Jaw_Grp_CollapseNode.ExecutePin, Parent=RigVMVariableNode() {variable=Local Facial_Board Gui}.Value)
   n5 = &RigUnit_HierarchyAddNull(
    Name=GRP_switchesGUI
    Space=LocalSpace
    Transform=(Rotation=(X=0,Y=0,Z=0,W=1),Translation=(X=0.2021,Y=4.0214,Z=5.1338),Scale3D=(X=1,Y=1,Z=1))
    ExecutePin=@Tongue_Grp_CollapseNode.ExecutePin
    Parent=RigVMVariableNode() {variable=Local Facial_Board Gui}.Value
   )
   > &RigUnit_HierarchyAddControlFloat(
    InitialValue=1.0
    Name=CTRL_rigLogicSwitch
    OffsetSpace=LocalSpace
    OffsetTransform=(Rotation=(X=0,Y=0,Z=-0.7071,W=0.7071),Translation=(X=4.208,Y=-3.9993,Z=-5.5216),Scale3D=(X=2,Y=2,Z=2))
    Settings="(PrimaryAxis=Y,bIsScale=False,Limits=(Limit=(bMinimum=True,bMaximum=True),MinValue=1,MaxValue=1,bDrawLimits=True),Shape=(bVisible=False,Name=\"DefaultGizmoLibraryNormalized.Sphere_Solid\",Color=(R=0.0503,G=0.0503,B=0,A=0),Transform=(Rotation=(X=0,Y=0,Z=0,W=1 ...(truncated)"
    Parent=@n5.Item
   ) {template="SpawnControl::Execute(in InitialValue,in Settings,in OffsetTransform,in OffsetSpace,in Parent,in Name,out Item)"}
   > &RigUnit_HierarchyAddControlFloat(
    InitialValue=0.0
    Name=CTRL_lookAtSwitch
    OffsetSpace=LocalSpace
    OffsetTransform=(Rotation=(X=0,Y=0,Z=-0.7071,W=0.7071),Translation=(X=4.2083,Y=-3.9993,Z=-6.3814),Scale3D=(X=2,Y=2,Z=2))
    Settings="(PrimaryAxis=Y,bIsScale=False,Limits=(Limit=(bMinimum=True,bMaximum=True),MinValue=0,MaxValue=1,bDrawLimits=True),Shape=(bVisible=True,Name=\"DefaultGizmoLibraryNormalized.Sphere_Solid\",Color=(R=1,G=1,B=0,A=0),Transform=(Rotation=(X=0,Y=0,Z=0,W=1),Translati ...(truncated)"
    Parent=@n5.Item
   ) {template="SpawnControl::Execute(in InitialValue,in Settings,in OffsetTransform,in OffsetSpace,in Parent,in Name,out Item)"}
   n6 = &RigUnit_HierarchyAddNull(
    Name=GRP_mouthGUI
    Space=LocalSpace
    Transform=(Rotation=(X=0,Y=0,Z=0,W=1),Translation=(X=0,Y=0,Z=0),Scale3D=(X=1,Y=1,Z=1))
    ExecutePin=@RigVMFunction_Sequence_4.C
    Parent=RigVMVariableNode() {variable=Local Facial_Board Gui}.Value
   )
   > RigVMVariableNode(Value=@n6.Item) {variable=Local MouthGUI}
   > RigVMFunction_Sequence_8 = RigVMAggregateNode()
   Towards_Grp_CollapseNode = RigVMCollapseNode(ExecutePin=@RigVMFunction_Sequence_8.D, Parent=RigVMVariableNode() {variable=Local MouthGUI}.Value)
   Purse_Grp_CollapseNode = RigVMCollapseNode(ExecutePin=@Towards_Grp_CollapseNode.ExecutePin, Parent=RigVMVariableNode() {variable=Local MouthGUI}.Value)
   Funnel_Grp_CollapseNode = RigVMCollapseNode(ExecutePin=@Purse_Grp_CollapseNode.ExecutePin, Parent=RigVMVariableNode() {variable=Local MouthGUI}.Value)
   MouthPress_Grp_CollapseNode = RigVMCollapseNode(ExecutePin=@Funnel_Grp_CollapseNode.ExecutePin, Parent=RigVMVariableNode() {variable=Local MouthGUI}.Value)
   LipsTogether_Grp_CollapseNode = RigVMCollapseNode(ExecutePin=@MouthPress_Grp_CollapseNode.ExecutePin, Parent=RigVMVariableNode() {variable=Local MouthGUI}.Value)
   LipsBite_Grp_CollapseNode = RigVMCollapseNode(ExecutePin=@LipsTogether_Grp_CollapseNode.ExecutePin, Parent=RigVMVariableNode() {variable=Local MouthGUI}.Value)
   &RigUnit_HierarchyAddControlVector2D(
    InitialValue=(X=0,Y=0)
    Name=CTRL_C_mouth
    OffsetSpace=LocalSpace
    OffsetTransform=(Rotation=(X=0.7071,Y=0,Z=0,W=0.7071),Translation=(X=-0.0188,Y=0,Z=10.6925),Scale3D=(X=0.65,Y=0.65,Z=0.65))
    Settings="(PrimaryAxis=Z,Limits=(LimitX=(bMinimum=True,bMaximum=True),LimitY=(bMinimum=True,bMaximum=True),MinValue=(X=-1,Y=-1),MaxValue=(X=1,Y=1),bDrawLimits=True),Shape=(bVisible=True,Name=\"DefaultGizmoLibraryNormalized.Sphere_Solid\",Color=(R=1,G=1,B=0,A=1),Transf ...(truncated)"
    Parent=RigVMVariableNode() {variable=Local MouthGUI}.Value
    ExecutePin=@RigVMFunction_Sequence_8.A
   ) {template="SpawnControl::Execute(in InitialValue,in Settings,in OffsetTransform,in OffsetSpace,in Parent,in Name,out Item)"}
   n7 = RigVMVariableNode() {variable=Local MouthGUI}
   > &RigUnit_HierarchyAddControlFloat(
    InitialValue=0.0
    Name=CTRL_L_mouth_suckBlow
    OffsetSpace=LocalSpace
    OffsetTransform=(Rotation=(X=-0.7071,Y=-0.7071,Z=0,W=0),Translation=(X=6.4956,Y=0.0221,Z=11.8421),Scale3D=(X=1,Y=1,Z=1))
    Settings="(PrimaryAxis=Y,bIsScale=False,Limits=(Limit=(bMinimum=True,bMaximum=True),MinValue=-1,MaxValue=1,bDrawLimits=True),Shape=(bVisible=True,Name=\"DefaultGizmoLibraryNormalized.Sphere_Solid\",Color=(R=1,G=1,B=0,A=0),Transform=(Rotation=(X=0,Y=0,Z=0,W=1),Translat ...(truncated)"
    Parent=@n7.Value
   ) {template="SpawnControl::Execute(in InitialValue,in Settings,in OffsetTransform,in OffsetSpace,in Parent,in Name,out Item)"}
   > &RigUnit_HierarchyAddControlFloat(
    InitialValue=0.0
    Name=CTRL_R_mouth_suckBlow
    OffsetSpace=LocalSpace
    OffsetTransform=(Rotation=(X=0,Y=0,Z=-0.7071,W=-0.7071),Translation=(X=-6.5044,Y=0.0221,Z=11.8421),Scale3D=(X=1,Y=1,Z=1))
    Settings="(PrimaryAxis=Y,bIsScale=False,Limits=(Limit=(bMinimum=True,bMaximum=True),MinValue=-1,MaxValue=1,bDrawLimits=True),Shape=(bVisible=True,Name=\"DefaultGizmoLibraryNormalized.Sphere_Solid\",Color=(R=1,G=1,B=0,A=0),Transform=(Rotation=(X=0,Y=0,Z=0,W=1),Translat ...(truncated)"
    Parent=@n7.Value
   ) {template="SpawnControl::Execute(in InitialValue,in Settings,in OffsetTransform,in OffsetSpace,in Parent,in Name,out Item)"}
   n8 = RigVMVariableNode() {variable=Local MouthGUI}
   > &RigUnit_HierarchyAddControlFloat(
    InitialValue=0.0
    Name=CTRL_L_mouth_lipsBlow
    OffsetSpace=LocalSpace
    OffsetTransform=(Rotation=(X=-0.2418,Y=0.2418,Z=-0.6645,W=0.6645),Translation=(X=5.6381,Y=0.0221,Z=7.2983),Scale3D=(X=1,Y=1,Z=1))
    Settings="(PrimaryAxis=Y,bIsScale=False,Limits=(Limit=(bMinimum=True,bMaximum=True),MinValue=0,MaxValue=1,bDrawLimits=True),Shape=(bVisible=True,Name=\"DefaultGizmoLibraryNormalized.Sphere_Solid\",Color=(R=1,G=1,B=0,A=0),Transform=(Rotation=(X=0,Y=0,Z=0,W=1),Translati ...(truncated)"
    Parent=@n8.Value
   ) {template="SpawnControl::Execute(in InitialValue,in Settings,in OffsetTransform,in OffsetSpace,in Parent,in Name,out Item)"}
   > &RigUnit_HierarchyAddControlFloat(
    InitialValue=0.0
    Name=CTRL_R_mouth_lipsBlow
    OffsetSpace=LocalSpace
    OffsetTransform=(Rotation=(X=0.6645,Y=-0.6645,Z=0.2418,W=-0.2418),Translation=(X=5.2448,Y=0.0027,Z=7.2983),Scale3D=(X=1,Y=1,Z=1))
    Settings="(PrimaryAxis=Y,bIsScale=False,Limits=(Limit=(bMinimum=True,bMaximum=True),MinValue=0,MaxValue=1,bDrawLimits=True),Shape=(bVisible=True,Name=\"DefaultGizmoLibraryNormalized.Sphere_Solid\",Color=(R=1,G=1,B=0,A=0),Transform=(Rotation=(X=0,Y=0,Z=0,W=1),Translati ...(truncated)"
    Parent=@n8.Value
   ) {template="SpawnControl::Execute(in InitialValue,in Settings,in OffsetTransform,in OffsetSpace,in Parent,in Name,out Item)"}
   n9 = RigVMVariableNode() {variable=Local MouthGUI}
   &RigUnit_HierarchyAddControlFloat(
    InitialValue=0.0
    Name=CTRL_L_mouth_upperLipRaise
    OffsetSpace=LocalSpace
    OffsetTransform=(Rotation=(X=0.5889,Y=0.5889,Z=0.3914,W=0.3913),Translation=(X=0.5649,Y=0,Z=11.9007),Scale3D=(X=2,Y=2,Z=2))
    Settings="(PrimaryAxis=Y,bIsScale=False,Limits=(Limit=(bMinimum=True,bMaximum=True),MinValue=0,MaxValue=1,bDrawLimits=True),Shape=(bVisible=True,Name=\"DefaultGizmoLibraryNormalized.Sphere_Solid\",Color=(R=1,G=1,B=0,A=1),Transform=(Rotation=(X=-0.7071,Y=0,Z=0,W=0.7071 ...(truncated)"
    Parent=@n9.Value
    ExecutePin=@RigVMFunction_Sequence_8.B
   ) {template="SpawnControl::Execute(in InitialValue,in Settings,in OffsetTransform,in OffsetSpace,in Parent,in Name,out Item)"}
   > &RigUnit_HierarchyAddControlFloat(
    InitialValue=0.0
    Name=CTRL_R_mouth_upperLipRaise
    OffsetSpace=LocalSpace
    OffsetTransform=(Rotation=(X=0.3916,Y=0.3916,Z=0.5888,W=0.5888),Translation=(X=-0.5554,Y=0.0062,Z=11.888),Scale3D=(X=2,Y=2,Z=2))
    Settings="(PrimaryAxis=Y,bIsScale=False,Limits=(Limit=(bMinimum=True,bMaximum=True),MinValue=0,MaxValue=1,bDrawLimits=True),Shape=(bVisible=True,Name=\"DefaultGizmoLibraryNormalized.Sphere_Solid\",Color=(R=1,G=1,B=0,A=0),Transform=(Rotation=(X=0,Y=0,Z=0,W=1),Translati ...(truncated)"
    Parent=@n9.Value
   ) {template="SpawnControl::Execute(in InitialValue,in Settings,in OffsetTransform,in OffsetSpace,in Parent,in Name,out Item)"}
   n10 = RigVMVariableNode() {variable=Local MouthGUI}
   > &RigUnit_HierarchyAddControlFloat(
    InitialValue=0.0
    Name=CTRL_L_mouth_cornerPull
    OffsetSpace=LocalSpace
    OffsetTransform=(Rotation=(X=0.6437,Y=0.6437,Z=0.2928,W=0.2928),Translation=(X=2.1092,Y=0,Z=11.2272),Scale3D=(X=2,Y=2,Z=2))
    Settings="(PrimaryAxis=Y,bIsScale=False,Limits=(Limit=(bMinimum=True,bMaximum=True),MinValue=0,MaxValue=1,bDrawLimits=True),Shape=(bVisible=True,Name=\"DefaultGizmoLibraryNormalized.Sphere_Solid\",Color=(R=1,G=1,B=0,A=1),Transform=(Rotation=(X=-0.7071,Y=0,Z=0,W=0.7071 ...(truncated)"
    Parent=@n10.Value
   ) {template="SpawnControl::Execute(in InitialValue,in Settings,in OffsetTransform,in OffsetSpace,in Parent,in Name,out Item)"}
   > &RigUnit_HierarchyAddControlFloat(
    InitialValue=0.0
    Name=CTRL_R_mouth_cornerPull
    OffsetSpace=LocalSpace
    OffsetTransform=(Rotation=(X=0.2912,Y=0.2912,Z=0.6444,W=0.6444),Translation=(X=-2.0885,Y=0.0062,Z=11.2268),Scale3D=(X=2,Y=2,Z=2))
    Settings="(PrimaryAxis=Y,bIsScale=False,Limits=(Limit=(bMinimum=True,bMaximum=True),MinValue=0,MaxValue=1,bDrawLimits=True),Shape=(bVisible=True,Name=\"DefaultGizmoLibraryNormalized.Sphere_Solid\",Color=(R=1,G=1,B=0,A=0),Transform=(Rotation=(X=0,Y=0,Z=0,W=1),Translati ...(truncated)"
    Parent=@n10.Value
   ) {template="SpawnControl::Execute(in InitialValue,in Settings,in OffsetTransform,in OffsetSpace,in Parent,in Name,out Item)"}
   n11 = RigVMVariableNode() {variable=Local MouthGUI}
   > &RigUnit_HierarchyAddControlFloat(
    InitialValue=0.0
    Name=CTRL_L_mouth_dimple
    OffsetSpace=LocalSpace
    OffsetTransform=(Rotation=(X=0.705,Y=0.705,Z=0.0546,W=0.0547),Translation=(X=2.571,Y=0,Z=10.8119),Scale3D=(X=2,Y=2,Z=2))
    Settings="(PrimaryAxis=Y,bIsScale=False,Limits=(Limit=(bMinimum=True,bMaximum=True),MinValue=0,MaxValue=1,bDrawLimits=True),Shape=(bVisible=True,Name=\"DefaultGizmoLibraryNormalized.Sphere_Solid\",Color=(R=1,G=1,B=0,A=1),Transform=(Rotation=(X=-0.7071,Y=0,Z=0,W=0.7071 ...(truncated)"
    Parent=@n11.Value
   ) {template="SpawnControl::Execute(in InitialValue,in Settings,in OffsetTransform,in OffsetSpace,in Parent,in Name,out Item)"}
   > &RigUnit_HierarchyAddControlFloat(
    InitialValue=0.0
    Name=CTRL_R_mouth_dimple
    OffsetSpace=LocalSpace
    OffsetTransform=(Rotation=(X=0.0519,Y=0.0519,Z=0.7052,W=0.7052),Translation=(X=-2.5946,Y=0.0065,Z=10.8327),Scale3D=(X=2,Y=2,Z=2))
    Settings="(PrimaryAxis=Y,bIsScale=False,Limits=(Limit=(bMinimum=True,bMaximum=True),MinValue=0,MaxValue=1,bDrawLimits=True),Shape=(bVisible=True,Name=\"DefaultGizmoLibraryNormalized.Sphere_Solid\",Color=(R=1,G=1,B=0,A=0),Transform=(Rotation=(X=0,Y=0,Z=0,W=1),Translati ...(truncated)"
    Parent=@n11.Value
   ) {template="SpawnControl::Execute(in InitialValue,in Settings,in OffsetTransform,in OffsetSpace,in Parent,in Name,out Item)"}
   n12 = RigVMVariableNode() {variable=Local MouthGUI}
   &RigUnit_HierarchyAddControlFloat(
    InitialValue=0.0
    Name=CTRL_L_mouth_lowerLipDepress
    OffsetSpace=LocalSpace
    OffsetTransform=(Rotation=(X=0.5376,Y=0.5331,Z=-0.4594,W=-0.4645),Translation=(X=0.6405,Y=0,Z=9.4879),Scale3D=(X=2,Y=2,Z=2))
    Settings="(PrimaryAxis=Y,bIsScale=False,Limits=(Limit=(bMinimum=True,bMaximum=True),MinValue=0,MaxValue=1,bDrawLimits=True),Shape=(bVisible=True,Name=\"DefaultGizmoLibraryNormalized.Sphere_Solid\",Color=(R=1,G=1,B=0,A=1),Transform=(Rotation=(X=-0.7071,Y=0,Z=0,W=0.7071 ...(truncated)"
    Parent=@n12.Value
    ExecutePin=@RigVMFunction_Sequence_8.C
   ) {template="SpawnControl::Execute(in InitialValue,in Settings,in OffsetTransform,in OffsetSpace,in Parent,in Name,out Item)"}
   > &RigUnit_HierarchyAddControlFloat(
    InitialValue=0.0
    Name=CTRL_R_mouth_lowerLipDepress
    OffsetSpace=LocalSpace
    OffsetTransform=(Rotation=(X=0.4644,Y=0.4586,Z=-0.5315,W=-0.5399),Translation=(X=-0.6337,Y=-0.0001,Z=9.4868),Scale3D=(X=2,Y=2,Z=2))
    Settings="(PrimaryAxis=Y,bIsScale=False,Limits=(Limit=(bMinimum=True,bMaximum=True),MinValue=0,MaxValue=1,bDrawLimits=True),Shape=(bVisible=True,Name=\"DefaultGizmoLibraryNormalized.Sphere_Solid\",Color=(R=1,G=1,B=0,A=1),Transform=(Rotation=(X=-0.7071,Y=0,Z=0,W=0.7071 ...(truncated)"
    Parent=@n12.Value
   ) {template="SpawnControl::Execute(in InitialValue,in Settings,in OffsetTransform,in OffsetSpace,in Parent,in Name,out Item)"}
   n13 = RigVMVariableNode() {variable=Local MouthGUI}
   > &RigUnit_HierarchyAddControlFloat(
    InitialValue=0.0
    Name=CTRL_L_mouth_stretch
    OffsetSpace=LocalSpace
    OffsetTransform=(Rotation=(X=0.6804,Y=0.6804,Z=-0.1924,W=-0.1924),Translation=(X=1.3242,Y=0,Z=9.9288),Scale3D=(X=2,Y=2,Z=2))
    Settings="(PrimaryAxis=Y,bIsScale=False,Limits=(Limit=(bMinimum=True,bMaximum=True),MinValue=0,MaxValue=1,bDrawLimits=True),Shape=(bVisible=True,Name=\"DefaultGizmoLibraryNormalized.Sphere_Solid\",Color=(R=1,G=1,B=0,A=1),Transform=(Rotation=(X=-0.7071,Y=0,Z=0,W=0.7071 ...(truncated)"
    Parent=@n13.Value
   ) {template="SpawnControl::Execute(in InitialValue,in Settings,in OffsetTransform,in OffsetSpace,in Parent,in Name,out Item)"}
   > &RigUnit_HierarchyAddControlFloat(
    InitialValue=0.0
    Name=CTRL_R_mouth_stretch
    OffsetSpace=LocalSpace
    OffsetTransform=(Rotation=(X=-0.1924,Y=-0.1924,Z=0.6804,W=0.6804),Translation=(X=-1.3139,Y=0,Z=9.912),Scale3D=(X=2,Y=2,Z=2))
    Settings="(PrimaryAxis=Y,bIsScale=False,Limits=(Limit=(bMinimum=True,bMaximum=True),MinValue=0,MaxValue=1,bDrawLimits=True),Shape=(bVisible=True,Name=\"DefaultGizmoLibraryNormalized.Sphere_Solid\",Color=(R=1,G=1,B=0,A=1),Transform=(Rotation=(X=-0.7071,Y=0,Z=0,W=0.7071 ...(truncated)"
    Parent=@n13.Value
   ) {template="SpawnControl::Execute(in InitialValue,in Settings,in OffsetTransform,in OffsetSpace,in Parent,in Name,out Item)"}
   n14 = RigVMVariableNode() {variable=Local MouthGUI}
   > &RigUnit_HierarchyAddControlFloat(
    InitialValue=0.0
    Name=CTRL_L_mouth_cornerDepress
    OffsetSpace=LocalSpace
    OffsetTransform=(Rotation=(X=0.6381,Y=0.6381,Z=-0.3046,W=-0.3046),Translation=(X=2.1134,Y=0,Z=10.3735),Scale3D=(X=2,Y=2,Z=2))
    Settings="(PrimaryAxis=Y,bIsScale=False,Limits=(Limit=(bMinimum=True,bMaximum=True),MinValue=0,MaxValue=1,bDrawLimits=True),Shape=(bVisible=True,Name=\"DefaultGizmoLibraryNormalized.Sphere_Solid\",Color=(R=1,G=1,B=0,A=1),Transform=(Rotation=(X=-0.7071,Y=0,Z=0,W=0.7071 ...(truncated)"
    Parent=@n14.Value
   ) {template="SpawnControl::Execute(in InitialValue,in Settings,in OffsetTransform,in OffsetSpace,in Parent,in Name,out Item)"}
   > &RigUnit_HierarchyAddControlFloat(
    InitialValue=0.0
    Name=CTRL_R_mouth_cornerDepress
    OffsetSpace=LocalSpace
    OffsetTransform=(Rotation=(X=0.3044,Y=0.3044,Z=-0.6382,W=-0.6382),Translation=(X=-2.0731,Y=0.0208,Z=10.4121),Scale3D=(X=2,Y=2,Z=2))
    Settings="(PrimaryAxis=Y,bIsScale=False,Limits=(Limit=(bMinimum=True,bMaximum=True),MinValue=0,MaxValue=1,bDrawLimits=True),Shape=(bVisible=True,Name=\"DefaultGizmoLibraryNormalized.Sphere_Solid\",Color=(R=1,G=1,B=0,A=0),Transform=(Rotation=(X=0,Y=0,Z=0,W=1),Translati ...(truncated)"
    Parent=@n14.Value
   ) {template="SpawnControl::Execute(in InitialValue,in Settings,in OffsetTransform,in OffsetSpace,in Parent,in Name,out Item)"}
   RigVMFunction_ControlFlowBranch = &RigVMFunction_ControlFlowBranch(Condition=RigVMVariableNode() {variable=Is V2}.Value, ExecuteContext=@RigVMFunction_Sequence_4.D)
   n15 = RigVMVariableNode() {variable=Local EyeGUI}
   n16 = &RigUnit_HierarchyAddControlFloat(
    InitialValue=0.0
    Name=CTRL_L_eye_pupil
    OffsetSpace=LocalSpace
    OffsetTransform=(Rotation=(X=0.7071,Y=0.7071,Z=0,W=0),Translation=(X=2.4913,Y=0,Z=17.2402),Scale3D=(X=0.75,Y=0.75,Z=0.75))
    Settings="(PrimaryAxis=Y,bIsScale=False,Limits=(Limit=(bMinimum=True,bMaximum=True),MinValue=-1,MaxValue=1,bDrawLimits=True),Shape=(bVisible=True,Name=\"DefaultGizmoLibrary.Sphere_Solid\",Color=(R=1,G=1,B=0,A=1),Transform=(Rotation=(X=0,Y=0,Z=0,W=1),Translation=(X=0,Y ...(truncated)"
    Parent=@n15.Value
    ExecutePin=@RigVMFunction_ControlFlowBranch.True
   ) {template="SpawnControl::Execute(in InitialValue,in Settings,in OffsetTransform,in OffsetSpace,in Parent,in Name,out Item)"}
   > &RigUnit_SetMetadataTag(NameSpace=Self, Tag=FN_MH_V2_Ctrl, Item=@n16.Item)
   > n17 = &RigUnit_HierarchyAddControlFloat(
    InitialValue=0.0
    Name=CTRL_R_eye_pupil
    OffsetSpace=LocalSpace
    OffsetTransform=(Rotation=(X=0,Y=0,Z=0.7071,W=0.7071),Translation=(X=-2.509,Y=0,Z=17.2402),Scale3D=(X=0.75,Y=0.75,Z=0.75))
    Settings="(PrimaryAxis=Y,bIsScale=False,Limits=(Limit=(bMinimum=True,bMaximum=True),MinValue=-1,MaxValue=1,bDrawLimits=True),Shape=(bVisible=True,Name=\"DefaultGizmoLibrary.Sphere_Solid\",Color=(R=1,G=1,B=0,A=1),Transform=(Rotation=(X=0,Y=0,Z=0,W=1),Translation=(X=0,Y ...(truncated)"
    Parent=@n15.Value
   ) {template="SpawnControl::Execute(in InitialValue,in Settings,in OffsetTransform,in OffsetSpace,in Parent,in Name,out Item)"}
   > &RigUnit_SetMetadataTag(NameSpace=Self, Tag=FN_MH_V2_Ctrl, Item=@n17.Item)
   n18 = RigVMVariableNode() {variable=Local MouthGUI}
   > n19 = &RigUnit_HierarchyAddControlFloat(
    InitialValue=0.0
    Name=CTRL_L_mouth_sharpCornerPull
    OffsetSpace=LocalSpace
    OffsetTransform=(Rotation=(X=0.5654,Y=0.5654,Z=0.4246,W=0.4246),Translation=(X=1.3385,Y=0,Z=11.5698),Scale3D=(X=2,Y=2,Z=2))
    Settings="(PrimaryAxis=Y,bIsScale=False,Limits=(Limit=(bMinimum=True,bMaximum=True),MinValue=0,MaxValue=1,bDrawLimits=True),Shape=(bVisible=True,Name=\"DefaultGizmoLibrary.Sphere_Solid\",Color=(R=1,G=1,B=0,A=1),Transform=(Rotation=(X=0,Y=0,Z=0,W=1),Translation=(X=0,Y= ...(truncated)"
    Parent=@n18.Value
   ) {template="SpawnControl::Execute(in InitialValue,in Settings,in OffsetTransform,in OffsetSpace,in Parent,in Name,out Item)"}
   > &RigUnit_SetMetadataTag(NameSpace=Self, Tag=FN_MH_V2_Ctrl, Item=@n19.Item)
   > n20 = &RigUnit_HierarchyAddControlFloat(
    InitialValue=0.0
    Name=CTRL_R_mouth_sharpCornerPull
    OffsetSpace=LocalSpace
    OffsetTransform=(Rotation=(X=0.4278,Y=0.4278,Z=0.563,W=0.563),Translation=(X=-1.3465,Y=-0.0004,Z=11.5658),Scale3D=(X=2,Y=2,Z=2))
    Settings="(PrimaryAxis=Y,bIsScale=False,Limits=(Limit=(bMinimum=True,bMaximum=True),MinValue=0,MaxValue=1,bDrawLimits=True),Shape=(bVisible=True,Name=\"DefaultGizmoLibrary.Sphere_Solid\",Color=(R=1,G=1,B=0,A=0),Transform=(Rotation=(X=0,Y=0,Z=0,W=1),Translation=(X=0,Y= ...(truncated)"
    Parent=@n18.Value
   ) {template="SpawnControl::Execute(in InitialValue,in Settings,in OffsetTransform,in OffsetSpace,in Parent,in Name,out Item)"}
   > n21 = &RigUnit_SetMetadataTag(NameSpace=Self, Tag=FN_MH_V2_Ctrl, Item=@n20.Item)
   LipsPress_Grp_CollapseNode = RigVMCollapseNode(ExecutePin=@n21.ExecutePin, Parent=RigVMVariableNode() {variable=Local MouthGUI}.Value)
   MouthTighten_Grp_CollapseNode = RigVMCollapseNode(ExecutePin=@LipsPress_Grp_CollapseNode.ExecutePin, Parent=RigVMVariableNode() {variable=Local MouthGUI}.Value)
   n22 = &RigVMFunction_Sequence(ExecuteContext=@RigVMFunction_Sequence_4.E)
   SpawnControl = &RigUnit_HierarchyAddControlFloat(
    InitialValue=0
    Name=CTRL_WrinkleMap
    OffsetSpace=LocalSpace
    OffsetTransform=(Rotation=(X=0,Y=0,Z=0,W=1),Translation=(X=0,Y=0,Z=-4.5),Scale3D=(X=1,Y=1,Z=1))
    Settings="(PrimaryAxis=X,bIsScale=False,Limits=(Limit=(bMinimum=True,bMaximum=True),MinValue=0,MaxValue=0,bDrawLimits=False),Shape=(bVisible=True,Name=\"DefaultGizmoLibraryNormalized.Circle_Solid\",Color=(R=1,G=0.0406,B=0.8081,A=1),Transform=(Rotation=(X=-0.7071,Y=0,Z ...(truncated)"
    ExecutePin=@n22.A
    Parent=RigVMVariableNode() {variable=Local Facial_Board Gui}.Value
   ) {template="SpawnControl::Execute(in InitialValue,in Settings,in OffsetTransform,in OffsetSpace,in Parent,in Name,out Item)"}
   > For_Each = RigVMDispatchNode(Array=&RigUnit_CollectionNameSearchArray(PartialName=_Offset_, TypeToSearch=Curve).Items) {template="DISPATCH_RigVMDispatch_ArrayIterator(in Array,out Element,out Index,out Count,out Ratio)"}
   StartsWith = &RigVMFunction_StartsWith(Start=WM_, Name=@For_Each.Element.Name) {template="StartsWith::Execute(in Name,in Start,out Result)"}
   n23 = &RigVMFunction_ControlFlowBranch(ExecuteContext=@For_Each.ExecuteContext, Condition=@StartsWith.Result)
   /Script/ControlRig.RigUnit_HierarchyAddAnimationChannelFloat(
    InitialValue=0.0
    LimitsEnabled=(Enabled=(bMinimum=true,bMaximum=true))
    MaximumValue=10
    MinimumValue=-10
    ExecutePin=@n23.True
    Name=@For_Each.Element.Name
    Parent=@SpawnControl.Item
   ) {
    is=struct RigVMUnitNode
    template="SpawnAnimationChannel::Execute(in InitialValue,in MinimumValue,in MaximumValue,in LimitsEnabled,in Parent,in Name,out Item)"
   }
   CRFL FN MH Tweaker Construction v01 = RigVMFunctionReferenceNode(ExecuteContext=@n22.B, Is V2=RigVMVariableNode() {variable=Is V2}.Value, Parent=RigVMVariableNode() {variable=Local Facial_Board Gui}.Value)
   n24 = /Script/ControlRig.RigUnit_CollectionChildrenArray(Parent=(Type=Null,Name="FACIAL_BOARD"), TypeToSearch=Control, bDefaultChildren=false, bIncludeParent=False, bRecursive=true) {is=struct RigVMUnitNode}
   > For_Each_1 = RigVMDispatchNode(Array=@n24.Items) {template="DISPATCH_RigVMDispatch_ArrayIterator(in Array,out Element,out Index,out Count,out Ratio)"}
   EndsWith = /Script/RigVM.RigVMFunction_EndsWith(Ending=GUI, Name=@For_Each_1.Element.Name) {is=struct RigVMUnitNode, template="EndsWith::Execute(in Name,in Ending,out Result)"}
   RigVMFunction_MathBoolNot = &RigVMFunction_MathBoolNot(Value=@EndsWith.Result)
   n25 = &RigVMFunction_ControlFlowBranch(ExecuteContext=@For_Each_1.ExecuteContext, Condition=@RigVMFunction_MathBoolNot.Result)
   n26 = /Script/ControlRig.RigUnit_GetShapeTransform(Control=@For_Each_1.Element.Name) {is=struct RigVMUnitNode}
   RigVMFunction_MathVectorFromDouble = /Script/RigVM.RigVMFunction_MathVectorFromDouble(Value=RigVMVariableNode() {variable=Control Scale}.Value) {is=struct RigVMUnitNode, template="Cast::Execute(in Value,out Result)"}
   Multiply = /Script/RigVM.RigVMFunction_MathVectorMul(A=@n26.Transform.Scale3D, B=@RigVMFunction_MathVectorFromDouble.Result) {is=struct RigVMUnitNode, template="Multiply::Execute(in A,in B,out Result)"}
   /Script/ControlRig.RigUnit_SetShapeTransform(
    Transform=(Rotation=(X=0,Y=0,Z=0,W=1),Translation=(X=0,Y=0,Z=0),Scale3D=(X=1,Y=1,Z=1))
    Transform.Rotation=@n26.Transform.Rotation
    Transform.Translation=@n26.Transform.Translation
    Transform.Scale3D=@Multiply.Result
    ExecutePin=@n25.True
    Control=@For_Each_1.Element.Name
   ) {is=struct RigVMUnitNode}
   > SetControlColor = /Script/ControlRig.RigUnit_SetControlColor(Color=RigVMVariableNode() {variable=Control Color}.Value, Control=@For_Each_1.Element.Name) {is=struct RigVMUnitNode, template="SetControlColor::Execute(in Control,in Color)"}
   Brow_Grp_CollapseNode = RigVMCollapseNode(Parent=RigVMVariableNode() {variable=Local Facial_Board Gui}.Value, ExecutePin=@RigVMFunction_Sequence_4.A)
   n27 = &RigUnit_HierarchyAddNull(
    Name=GRP_eyeGUI
    Space=LocalSpace
    Transform=(Rotation=(X=0,Y=0,Z=0,W=1),Translation=(X=0,Y=0,Z=0),Scale3D=(X=1,Y=1,Z=1))
    ExecutePin=@Brow_Grp_CollapseNode.ExecutePin
    Parent=RigVMVariableNode() {variable=Local Facial_Board Gui}.Value
   )
   > n28 = RigVMVariableNode(Value=@n27.Item) {variable=Local EyeGUI}
   Eye_Grp_CollapseNode = RigVMCollapseNode(ExecutePin=@n28.ExecuteContext, Parent=RigVMVariableNode() {variable=Local EyeGUI}.Value)
   n29 = RigVMVariableNode() {variable=Local EyeGUI}
   &RigUnit_HierarchyAddControlFloat(
    InitialValue=0.0
    Name=CTRL_L_eye_cheekRaise
    OffsetSpace=LocalSpace
    OffsetTransform=(Rotation=(X=0.5,Y=0.5,Z=0.5,W=0.5),Translation=(X=4.5005,Y=0,Z=14.9949),Scale3D=(X=1.2,Y=1.2,Z=1.2))
    Settings="(PrimaryAxis=Y,bIsScale=False,Limits=(Limit=(bMinimum=True,bMaximum=True),MinValue=0,MaxValue=1,bDrawLimits=True),Shape=(bVisible=True,Name=\"DefaultGizmoLibraryNormalized.Sphere_Solid\",Color=(R=1,G=1,B=0,A=1),Transform=(Rotation=(X=-0.7071,Y=0,Z=0,W=0.7071 ...(truncated)"
    ExecutePin=@Eye_Grp_CollapseNode.ExecutePin
    Parent=@n29.Value
   ) {template="SpawnControl::Execute(in InitialValue,in Settings,in OffsetTransform,in OffsetSpace,in Parent,in Name,out Item)"}
   > &RigUnit_HierarchyAddControlFloat(
    InitialValue=0.0
    Name=CTRL_R_eye_cheekRaise
    OffsetSpace=LocalSpace
    OffsetTransform=(Rotation=(X=0.5,Y=0.5,Z=0.5,W=0.5),Translation=(X=-4.5005,Y=0,Z=14.9949),Scale3D=(X=1.2,Y=1.2,Z=1.2))
    Settings="(PrimaryAxis=Y,bIsScale=False,Limits=(Limit=(bMinimum=True,bMaximum=True),MinValue=0,MaxValue=1,bDrawLimits=True),Shape=(bVisible=True,Name=\"DefaultGizmoLibraryNormalized.Sphere_Solid\",Color=(R=1,G=1,B=0,A=1),Transform=(Rotation=(X=-0.7071,Y=0,Z=0,W=0.7071 ...(truncated)"
    Parent=@n29.Value
   ) {template="SpawnControl::Execute(in InitialValue,in Settings,in OffsetTransform,in OffsetSpace,in Parent,in Name,out Item)"}
   n30 = RigVMVariableNode() {variable=Local EyeGUI}
   > &RigUnit_HierarchyAddControlFloat(
    InitialValue=0.0
    Name=CTRL_L_eye_squintInner
    OffsetSpace=LocalSpace
    OffsetTransform=(Rotation=(X=0.1828,Y=0.1828,Z=0.6831,W=0.6831),Translation=(X=2.5695,Y=0,Z=14.9544),Scale3D=(X=1,Y=1,Z=1))
    Settings="(PrimaryAxis=Y,bIsScale=False,Limits=(Limit=(bMinimum=True,bMaximum=True),MinValue=0,MaxValue=1,bDrawLimits=True),Shape=(bVisible=True,Name=\"DefaultGizmoLibraryNormalized.Sphere_Solid\",Color=(R=1,G=1,B=0,A=1),Transform=(Rotation=(X=-0.7071,Y=0,Z=0,W=0.7071 ...(truncated)"
    Parent=@n30.Value
   ) {template="SpawnControl::Execute(in InitialValue,in Settings,in OffsetTransform,in OffsetSpace,in Parent,in Name,out Item)"}
   > &RigUnit_HierarchyAddControlFloat(
    InitialValue=0.0
    Name=CTRL_R_eye_squintInner
    OffsetSpace=LocalSpace
    OffsetTransform=(Rotation=(X=0.6831,Y=0.6831,Z=0.1827,W=0.1827),Translation=(X=-2.5695,Y=0,Z=14.9544),Scale3D=(X=1,Y=1,Z=1))
    Settings="(PrimaryAxis=Y,bIsScale=False,Limits=(Limit=(bMinimum=True,bMaximum=True),MinValue=0,MaxValue=1,bDrawLimits=True),Shape=(bVisible=True,Name=\"DefaultGizmoLibraryNormalized.Sphere_Solid\",Color=(R=1,G=1,B=0,A=1),Transform=(Rotation=(X=-0.7071,Y=0,Z=0,W=0.7071 ...(truncated)"
    Parent=@n30.Value
   ) {template="SpawnControl::Execute(in InitialValue,in Settings,in OffsetTransform,in OffsetSpace,in Parent,in Name,out Item)"}
   n31 = RigVMVariableNode() {variable=Local EyeGUI}
   > &RigUnit_HierarchyAddControlFloat(
    InitialValue=0.0
    Name=CTRL_L_nose_wrinkleUpper
    OffsetSpace=LocalSpace
    OffsetTransform=(Rotation=(X=-0.5,Y=-0.5,Z=0.5,W=0.5),Translation=(X=1.3472,Y=0.057,Z=17.6119),Scale3D=(X=1,Y=1,Z=1))
    Settings="(PrimaryAxis=Y,bIsScale=False,Limits=(Limit=(bMinimum=True,bMaximum=True),MinValue=0,MaxValue=1,bDrawLimits=True),Shape=(bVisible=True,Name=\"DefaultGizmoLibraryNormalized.Sphere_Solid\",Color=(R=1,G=1,B=0,A=1),Transform=(Rotation=(X=-0.7071,Y=0,Z=0,W=0.7071 ...(truncated)"
    Parent=@n31.Value
   ) {template="SpawnControl::Execute(in InitialValue,in Settings,in OffsetTransform,in OffsetSpace,in Parent,in Name,out Item)"}
   > &RigUnit_HierarchyAddControlFloat(
    InitialValue=0.0
    Name=CTRL_R_nose_wrinkleUpper
    OffsetSpace=LocalSpace
    OffsetTransform=(Rotation=(X=-0.5,Y=-0.5,Z=0.5,W=0.5),Translation=(X=-1.3366,Y=0.057,Z=17.6138),Scale3D=(X=1,Y=1,Z=1))
    Settings="(PrimaryAxis=Y,bIsScale=False,Limits=(Limit=(bMinimum=True,bMaximum=True),MinValue=0,MaxValue=1,bDrawLimits=True),Shape=(bVisible=True,Name=\"DefaultGizmoLibraryNormalized.Sphere_Solid\",Color=(R=1,G=1,B=0,A=1),Transform=(Rotation=(X=-0.7071,Y=0,Z=0,W=0.7071 ...(truncated)"
    Parent=@n31.Value
   ) {template="SpawnControl::Execute(in InitialValue,in Settings,in OffsetTransform,in OffsetSpace,in Parent,in Name,out Item)"}
   > n32 = &RigUnit_HierarchyAddNull(
    Name=EyeAim_GRP
    Space=LocalSpace
    Transform=(Rotation=(X=0,Y=0,Z=0,W=1),Translation=(X=0,Y=0,Z=0),Scale3D=(X=1,Y=1,Z=1))
    Parent=RigVMVariableNode() {variable=Local EyeGUI}.Value
   )
   > n33 = &RigUnit_HierarchyAddNull(
    Name=EyeAim_L_GRP
    Space=LocalSpace
    Transform=(Rotation=(X=0.7071,Y=0,Z=0,W=-0.7071),Translation=(X=-22.2521,Y=7.2952,Z=14.7727),Scale3D=(X=1,Y=1,Z=1))
    Parent=@n32.Item
   )
   > &RigUnit_HierarchyAddNull(Name=EyeAim_L, Space=LocalSpace, Transform=(Rotation=(X=0,Y=0,Z=0,W=1),Translation=(X=0,Y=0,Z=0),Scale3D=(X=1,Y=1,Z=1)), Parent=@n33.Item)
   > n34 = &RigUnit_HierarchyAddNull(
    Name=EyeAim_R_GRP
    Space=LocalSpace
    Transform=(Rotation=(X=0.7071,Y=0,Z=0,W=-0.7071),Translation=(X=-27.748,Y=7.2952,Z=14.7727),Scale3D=(X=1,Y=1,Z=1))
    Parent=@n32.Item
   )
   > &RigUnit_HierarchyAddNull(Name=EyeAim_R, Space=LocalSpace, Transform=(Rotation=(X=0,Y=0,Z=0,W=1),Translation=(X=0,Y=0,Z=0),Scale3D=(X=1,Y=1,Z=1)), Parent=@n34.Item)
   RigUnit_HierarchyAddNull = &RigUnit_HierarchyAddNull(
    Name=GRP_eyeControls
    Space=LocalSpace
    Transform=(Rotation=(X=0,Y=0,Z=0,W=1),Translation=(X=17.7372,Y=7.4878,Z=140.2747),Scale3D=(X=1,Y=1,Z=1))
    ExecutePin=@RigVMFunction_Sequence_1.B
    Parent=RigVMVariableNode() {variable=Local Top Null}.Value
   )
   > RigUnit_HierarchyAddControlTransform = &RigUnit_HierarchyAddControlTransform(
    InitialValue=(Rotation=(X=0,Y=0,Z=0,W=1),Translation=(X=0,Y=0,Z=0),Scale3D=(X=1,Y=1,Z=1))
    Name=CTRL_C_eyesAim
    OffsetSpace=LocalSpace
    OffsetTransform=(Rotation=(X=0.7071,Y=0,Z=0.7071,W=0),Translation=(X=-17.7372,Y=22.5122,Z=16.5322),Scale3D=(X=1,Y=1,Z=1))
    Settings="(InitialSpace=LocalSpace,bUsePreferredRotationOrder=False,PreferredRotationOrder=YZX,Limits=(LimitTranslationX=(bMinimum=False,bMaximum=False),LimitTranslationY=(bMinimum=False,bMaximum=False),LimitTranslationZ=(bMinimum=False,bMaximum=False),LimitPitch=(b ...(truncated)"
    Parent=@RigUnit_HierarchyAddNull.Item
   ) {template="SpawnControl::Execute(in InitialValue,in Settings,in OffsetTransform,in OffsetSpace,in Parent,in Name,out Item)"}
   > RigVMFunction_Sequence = RigVMAggregateNode()
   n35 = &RigUnit_HierarchyAddNull(
    Name=CTRL_L_eyeAim_space
    Space=LocalSpace
    Transform=(Rotation=(X=0,Y=0,Z=0,W=1),Translation=(X=0,Y=0,Z=2.748),Scale3D=(X=1,Y=1,Z=1))
    ExecutePin=@RigVMFunction_Sequence.A
    Parent=@RigUnit_HierarchyAddControlTransform.Item
   )
   > &RigUnit_HierarchyAddControlTransform(
    InitialValue=(Rotation=(X=0,Y=0,Z=0,W=1),Translation=(X=0,Y=0,Z=0),Scale3D=(X=1,Y=1,Z=1))
    Name=CTRL_L_eyeAim
    OffsetSpace=LocalSpace
    OffsetTransform=(Rotation=(X=0,Y=0,Z=0,W=1),Translation=(X=0.0067,Y=0,Z=0.0632),Scale3D=(X=1,Y=1,Z=1))
    Settings="(InitialSpace=LocalSpace,bUsePreferredRotationOrder=False,PreferredRotationOrder=YZX,Limits=(LimitTranslationX=(bMinimum=False,bMaximum=False),LimitTranslationY=(bMinimum=False,bMaximum=False),LimitTranslationZ=(bMinimum=False,bMaximum=False),LimitPitch=(b ...(truncated)"
    Parent=@n35.Item
   ) {template="SpawnControl::Execute(in InitialValue,in Settings,in OffsetTransform,in OffsetSpace,in Parent,in Name,out Item)"}
   n36 = &RigUnit_HierarchyAddNull(
    Name=CTRL_R_eyeAim_space
    Space=LocalSpace
    Transform=(Rotation=(X=0,Y=0,Z=0,W=1),Translation=(X=0,Y=0,Z=-2.748),Scale3D=(X=1,Y=1,Z=1))
    ExecutePin=@RigVMFunction_Sequence.B
    Parent=@RigUnit_HierarchyAddControlTransform.Item
   )
   > &RigUnit_HierarchyAddControlTransform(
    InitialValue=(Rotation=(X=0,Y=0,Z=0,W=1),Translation=(X=0,Y=0,Z=0),Scale3D=(X=1,Y=1,Z=1))
    Name=CTRL_R_eyeAim
    OffsetSpace=LocalSpace
    OffsetTransform=(Rotation=(X=0,Y=0,Z=0,W=1),Translation=(X=-0.0067,Y=0,Z=-0.0632),Scale3D=(X=1,Y=1,Z=1))
    Settings="(InitialSpace=LocalSpace,bUsePreferredRotationOrder=False,PreferredRotationOrder=YZX,Limits=(LimitTranslationX=(bMinimum=False,bMaximum=False),LimitTranslationY=(bMinimum=False,bMaximum=False),LimitTranslationZ=(bMinimum=False,bMaximum=False),LimitPitch=(b ...(truncated)"
    Parent=@n36.Item
   ) {template="SpawnControl::Execute(in InitialValue,in Settings,in OffsetTransform,in OffsetSpace,in Parent,in Name,out Item)"}
   n37 = &RigUnit_HierarchyAddControlTransform(
    InitialValue=(Rotation=(X=0,Y=0,Z=0,W=1),Translation=(X=0,Y=0,Z=0),Scale3D=(X=1,Y=1,Z=1))
    Name=FRM_convergenceGUI
    OffsetSpace=LocalSpace
    OffsetTransform=(Rotation=(X=0,Y=0,Z=0,W=1),Translation=(X=-4.3604,Y=0,Z=0),Scale3D=(X=1.2,Y=1.2,Z=1.2))
    Settings="(InitialSpace=LocalSpace,bUsePreferredRotationOrder=False,PreferredRotationOrder=YZX,Limits=(LimitTranslationX=(bMinimum=False,bMaximum=False),LimitTranslationY=(bMinimum=False,bMaximum=False),LimitTranslationZ=(bMinimum=False,bMaximum=False),LimitPitch=(b ...(truncated)"
    ExecutePin=@RigVMFunction_Sequence.C
    Parent=@RigUnit_HierarchyAddControlTransform.Item
   ) {template="SpawnControl::Execute(in InitialValue,in Settings,in OffsetTransform,in OffsetSpace,in Parent,in Name,out Item)"}
   > n38 = &RigUnit_HierarchyAddNull(
    Name=CTRL_convergenceSwitch_space
    Space=LocalSpace
    Transform=(Rotation=(X=0,Y=-0.7071,Z=-0.7071,W=0),Translation=(X=-0.44,Y=0,Z=-1),Scale3D=(X=2,Y=2,Z=2))
    Parent=@n37.Item
   )
   > RigUnit_HierarchyAddControlFloat = &RigUnit_HierarchyAddControlFloat(
    InitialValue=0.0
    Name=CTRL_convergenceSwitch
    OffsetSpace=LocalSpace
    OffsetTransform=(Rotation=(X=0,Y=0,Z=0,W=1),Translation=(X=0,Y=0,Z=0),Scale3D=(X=1,Y=1,Z=1))
    Settings="(PrimaryAxis=Y,bIsScale=False,Limits=(Limit=(bMinimum=True,bMaximum=True),MinValue=0,MaxValue=1,bDrawLimits=True),Shape=(bVisible=True,Name=\"DefaultGizmoLibraryNormalized.Sphere_Solid\",Color=(R=1,G=1,B=0,A=0),Transform=(Rotation=(X=0,Y=0,Z=0,W=1),Translati ...(truncated)"
    Parent=@n38.Item
   ) {template="SpawnControl::Execute(in InitialValue,in Settings,in OffsetTransform,in OffsetSpace,in Parent,in Name,out Item)"}
   RigUnit_GetTransform = &RigUnit_GetTransform(Item=(Type=Bone,Name="head"), bInitial=true)
   &RigUnit_SetTransform(
    Item=(Type=Control,Name="CTRL_Master")
    Value=(Rotation=(X=0,Y=0,Z=0,W=1),Translation=(X=17.7372,Y=7.4878,Z=140.2747),Scale3D=(X=1,Y=1,Z=1))
    Weight=1
    bInitial=true
    Value.Translation=/Script/RigVM.RigVMFunction_MathVectorAdd(B=(X=23,Y=3,Z=-10), A=@RigUnit_GetTransform.Transform.Translation) {is=struct RigVMUnitNode, template="Add::Execute(in A,in B,out Result)"}.Result
    ExecutePin=@RigVMFunction_Sequence_1.C
   )
   n39 = &RigUnit_GetTransform(Item=(Type=Bone,Name="FACIAL_L_Eye"), bInitial=true)
   n40 = &RigUnit_GetTransform(Item=(Type=Bone,Name="FACIAL_R_Eye"), bInitial=true)
   n41 = &RigVMFunction_MathTransformLerp(T=0.5, A=@n39.Transform, B=@n40.Transform) {template="Interpolate::Execute(in A,in B,in T,out Result)"}
   > &RigUnit_SetTransform(
    Item=(Type=Control,Name="CTRL_C_eyesAim")
    Value=(Rotation=(X=0.7071,Y=0,Z=0.7071,W=0),Translation=(X=0,Y=30,Z=0),Scale3D=(X=1,Y=1,Z=1))
    Weight=1
    bInitial=true
    Value.Translation.X=@n41.Result.Translation.X
    Value.Translation.Z=@n41.Result.Translation.Z
   )
   > &RigUnit_SetTransform(
    Item=(Type=Null,Name="CTRL_L_eyeAim_space")
    Value=(Rotation=(X=0.7071,Y=0,Z=0.7071,W=0),Translation=(X=0,Y=30,Z=0),Scale3D=(X=1,Y=1,Z=1))
    Weight=1
    bInitial=true
    Value.Translation.X=@n39.Transform.Translation.X
    Value.Translation.Z=@n39.Transform.Translation.Z
   )
   > &RigUnit_SetTransform(
    Item=(Type=Null,Name="CTRL_R_eyeAim_space")
    Value=(Rotation=(X=0.7071,Y=0,Z=0.7071,W=0),Translation=(X=0,Y=30,Z=0),Scale3D=(X=1,Y=1,Z=1))
    Weight=1
    bInitial=true
    Value.Translation.X=@n40.Transform.Translation.X
    Value.Translation.Z=@n40.Transform.Translation.Z
   )
   > &RigUnit_SetTransform(Item=(Type=Null,Name="EyeAim_L_GRP"), Weight=1, bInitial=true, Value=@n39.Transform)
   > &RigUnit_SetTransform(Item=(Type=Null,Name="EyeAim_R_GRP"), Weight=1, bInitial=true, Value=@n40.Transform)
   RigVMFunctionReturnNode(ExecuteContext=@RigVMFunction_Sequence_1.D)
   n42 = &RigUnit_HierarchyAddNull(
    ExecutePin=()
    Name=CTRL_C_mouth_stickyD_GRP
    Parent=(Type=Null,Name="GRP_mouthGUI")
    Space=LocalSpace
    Transform=(Rotation=(X=0.7071,Y=0,Z=0,W=0.7071),Translation=(X=-6.1394,Y=0.0221,Z=6.8351),Scale3D=(X=0.8,Y=0.8,Z=0.8))
   )
   > &RigUnit_HierarchyAddControlFloat(
    InitialValue=0.0
    Name=CTRL_C_mouth_stickyD
    OffsetSpace=LocalSpace
    OffsetTransform=(Rotation=(X=0,Y=0,Z=0,W=1),Translation=(X=0,Y=0,Z=0),Scale3D=(X=1,Y=1,Z=1))
    Settings="(PrimaryAxis=Y,bIsScale=False,Limits=(Limit=(bMinimum=True,bMaximum=True),MinValue=0,MaxValue=1,bDrawLimits=True),Shape=(bVisible=True,Name=\"DefaultGizmoLibrary.Sphere_Solid\",Color=(R=1,G=1,B=0,A=0),Transform=(Rotation=(X=0,Y=0,Z=0,W=1),Translation=(X=0,Y= ...(truncated)"
    Parent=@n42.Item
   ) {template="SpawnControl::Execute(in InitialValue,in Settings,in OffsetTransform,in OffsetSpace,in Parent,in Name,out Item)"}
   > n43 = &RigUnit_HierarchyAddNull(
    Name=CTRL_C_mouth_stickyU_GRP
    Parent=(Type=Null,Name="GRP_mouthGUI")
    Space=LocalSpace
    Transform=(Rotation=(X=-0.7071,Y=0,Z=0,W=0.7071),Translation=(X=-6.1394,Y=0.0221,Z=8.8351),Scale3D=(X=0.8,Y=0.8,Z=0.8))
   )
   > &RigUnit_HierarchyAddControlFloat(
    InitialValue=0.0
    Name=CTRL_C_mouth_stickyU
    OffsetSpace=LocalSpace
    OffsetTransform=(Rotation=(X=0,Y=0,Z=0,W=1),Translation=(X=0,Y=0,Z=0),Scale3D=(X=1,Y=1,Z=1))
    Settings="(PrimaryAxis=Y,bIsScale=False,Limits=(Limit=(bMinimum=True,bMaximum=True),MinValue=0,MaxValue=1,bDrawLimits=True),Shape=(bVisible=True,Name=\"DefaultGizmoLibrary.Sphere_Solid\",Color=(R=1,G=1,B=0,A=0),Transform=(Rotation=(X=0,Y=0,Z=0,W=1),Translation=(X=0,Y= ...(truncated)"
    Parent=@n43.Item
   ) {template="SpawnControl::Execute(in InitialValue,in Settings,in OffsetTransform,in OffsetSpace,in Parent,in Name,out Item)"}
   > n44 = &RigUnit_HierarchyAddNull(
    Name=CTRL_L_mouth_lipSticky_GRP
    Parent=(Type=Null,Name="GRP_mouthGUI")
    Space=LocalSpace
    Transform=(Rotation=(X=0.7071,Y=-0.7071,Z=0,W=0),Translation=(X=-4.4918,Y=0.0221,Z=7.8527),Scale3D=(X=0.8,Y=0.8,Z=0.8))
   )
   > &RigUnit_HierarchyAddControlFloat(
    InitialValue=0
    Name=CTRL_L_mouth_lipSticky
    OffsetSpace=LocalSpace
    OffsetTransform=(Rotation=(X=0,Y=0,Z=0,W=1),Translation=(X=0,Y=0,Z=0),Scale3D=(X=1,Y=1,Z=1))
    Settings="(PrimaryAxis=Y,bIsScale=False,Limits=(Limit=(bMinimum=True,bMaximum=True),MinValue=0,MaxValue=1,bDrawLimits=True),Shape=(bVisible=True,Name=\"DefaultGizmoLibrary.Sphere_Solid\",Color=(R=1,G=1,B=0,A=0),Transform=(Rotation=(X=0,Y=0,Z=0,W=1),Translation=(X=0,Y= ...(truncated)"
    Parent=@n44.Item
   ) {template="SpawnControl::Execute(in InitialValue,in Settings,in OffsetTransform,in OffsetSpace,in Parent,in Name,out Item)"}
   > n45 = &RigUnit_HierarchyAddNull(
    Name=CTRL_L_mouth_stickyInnerU_GRP
    Parent=(Type=Null,Name="GRP_mouthGUI")
    Space=LocalSpace
    Transform=(Rotation=(X=-0.7071,Y=0,Z=0,W=0.7071),Translation=(X=-5.5118,Y=0.0221,Z=8.8351),Scale3D=(X=0.8,Y=0.8,Z=0.8))
   )
   > &RigUnit_HierarchyAddControlFloat(
    InitialValue=0.0
    Name=CTRL_L_mouth_stickyInnerU
    OffsetSpace=LocalSpace
    OffsetTransform=(Rotation=(X=0,Y=0,Z=0,W=1),Translation=(X=0,Y=0,Z=0),Scale3D=(X=1,Y=1,Z=1))
    Settings="(PrimaryAxis=Y,bIsScale=False,Limits=(Limit=(bMinimum=True,bMaximum=True),MinValue=0,MaxValue=1,bDrawLimits=True),Shape=(bVisible=True,Name=\"DefaultGizmoLibrary.Sphere_Solid\",Color=(R=1,G=1,B=0,A=0),Transform=(Rotation=(X=0,Y=0,Z=0,W=1),Translation=(X=0,Y= ...(truncated)"
    Parent=@n45.Item
   ) {template="SpawnControl::Execute(in InitialValue,in Settings,in OffsetTransform,in OffsetSpace,in Parent,in Name,out Item)"}
   > n46 = &RigUnit_HierarchyAddNull(
    Name=CTRL_L_mouth_stickyInnerD_GRP
    Parent=(Type=Null,Name="GRP_mouthGUI")
    Space=LocalSpace
    Transform=(Rotation=(X=0.7071,Y=0,Z=0,W=0.7071),Translation=(X=-5.5118,Y=0.0221,Z=6.8351),Scale3D=(X=0.8,Y=0.8,Z=0.8))
   )
   > &RigUnit_HierarchyAddControlFloat(
    InitialValue=0.0
    Name=CTRL_L_mouth_stickyInnerD
    OffsetSpace=LocalSpace
    OffsetTransform=(Rotation=(X=0,Y=0,Z=0,W=1),Translation=(X=0,Y=0,Z=0),Scale3D=(X=1,Y=1,Z=1))
    Settings="(PrimaryAxis=Y,bIsScale=False,Limits=(Limit=(bMinimum=True,bMaximum=True),MinValue=0,MaxValue=1,bDrawLimits=True),Shape=(bVisible=True,Name=\"DefaultGizmoLibrary.Sphere_Solid\",Color=(R=1,G=1,B=0,A=0),Transform=(Rotation=(X=0,Y=0,Z=0,W=1),Translation=(X=0,Y= ...(truncated)"
    Parent=@n46.Item
   ) {template="SpawnControl::Execute(in InitialValue,in Settings,in OffsetTransform,in OffsetSpace,in Parent,in Name,out Item)"}
   > n47 = &RigUnit_HierarchyAddNull(
    Name=CTRL_L_mouth_stickyOuterU_GRP
    Parent=(Type=Null,Name="GRP_mouthGUI")
    Space=LocalSpace
    Transform=(Rotation=(X=-0.7071,Y=0,Z=0,W=0.7071),Translation=(X=-4.749,Y=0.0221,Z=8.8351),Scale3D=(X=0.8,Y=0.8,Z=0.8))
   )
   > &RigUnit_HierarchyAddControlFloat(
    InitialValue=0.0
    Name=CTRL_L_mouth_stickyOuterU
    OffsetSpace=LocalSpace
    OffsetTransform=(Rotation=(X=0,Y=0,Z=0,W=1),Translation=(X=0,Y=0,Z=0),Scale3D=(X=1,Y=1,Z=1))
    Settings="(PrimaryAxis=Y,bIsScale=False,Limits=(Limit=(bMinimum=True,bMaximum=True),MinValue=0,MaxValue=1,bDrawLimits=True),Shape=(bVisible=True,Name=\"DefaultGizmoLibrary.Sphere_Solid\",Color=(R=1,G=1,B=0,A=0),Transform=(Rotation=(X=0,Y=0,Z=0,W=1),Translation=(X=0,Y= ...(truncated)"
    Parent=@n47.Item
   ) {template="SpawnControl::Execute(in InitialValue,in Settings,in OffsetTransform,in OffsetSpace,in Parent,in Name,out Item)"}
   > n48 = &RigUnit_HierarchyAddNull(
    Name=CTRL_L_mouth_stickyOuterD_GRP
    Parent=(Type=Null,Name="GRP_mouthGUI")
    Space=LocalSpace
    Transform=(Rotation=(X=0.7071,Y=0,Z=0,W=0.7071),Translation=(X=-4.749,Y=0.0221,Z=6.8351),Scale3D=(X=0.8,Y=0.8,Z=0.8))
   )
   > &RigUnit_HierarchyAddControlFloat(
    InitialValue=0.0
    Name=CTRL_L_mouth_stickyOuterD
    OffsetSpace=LocalSpace
    OffsetTransform=(Rotation=(X=0,Y=0,Z=0,W=1),Translation=(X=0,Y=0,Z=0),Scale3D=(X=1,Y=1,Z=1))
    Settings="(PrimaryAxis=Y,bIsScale=False,Limits=(Limit=(bMinimum=True,bMaximum=True),MinValue=0,MaxValue=1,bDrawLimits=True),Shape=(bVisible=True,Name=\"DefaultGizmoLibrary.Sphere_Solid\",Color=(R=1,G=1,B=0,A=0),Transform=(Rotation=(X=0,Y=0,Z=0,W=1),Translation=(X=0,Y= ...(truncated)"
    Parent=@n48.Item
   ) {template="SpawnControl::Execute(in InitialValue,in Settings,in OffsetTransform,in OffsetSpace,in Parent,in Name,out Item)"}
   > n49 = &RigUnit_HierarchyAddNull(
    Name=CTRL_R_mouth_lipSticky_GRP
    Parent=(Type=Null,Name="GRP_mouthGUI")
    Space=LocalSpace
    Transform=(Rotation=(X=0,Y=0,Z=-0.7071,W=0.7071),Translation=(X=-7.8646,Y=0.0221,Z=7.8527),Scale3D=(X=0.8,Y=0.8,Z=0.8))
   )
   > &RigUnit_HierarchyAddControlFloat(
    InitialValue=0
    Name=CTRL_R_mouth_lipSticky
    OffsetSpace=LocalSpace
    OffsetTransform=(Rotation=(X=0,Y=0,Z=0,W=1),Translation=(X=0,Y=0,Z=0),Scale3D=(X=1,Y=1,Z=1))
    Settings="(PrimaryAxis=Y,bIsScale=False,Limits=(Limit=(bMinimum=True,bMaximum=True),MinValue=0,MaxValue=1,bDrawLimits=True),Shape=(bVisible=True,Name=\"DefaultGizmoLibrary.Sphere_Solid\",Color=(R=1,G=1,B=0,A=0),Transform=(Rotation=(X=0,Y=0,Z=0,W=1),Translation=(X=0,Y= ...(truncated)"
    Parent=@n49.Item
   ) {template="SpawnControl::Execute(in InitialValue,in Settings,in OffsetTransform,in OffsetSpace,in Parent,in Name,out Item)"}
   > n50 = &RigUnit_HierarchyAddNull(
    Name=CTRL_R_mouth_stickyInnerU_GRP
    Parent=(Type=Null,Name="GRP_mouthGUI")
    Space=LocalSpace
    Transform=(Rotation=(X=-0.7071,Y=0,Z=0,W=0.7071),Translation=(X=-6.8529,Y=0.0221,Z=8.8351),Scale3D=(X=0.8,Y=0.8,Z=0.8))
   )
   > &RigUnit_HierarchyAddControlFloat(
    InitialValue=0.0
    Name=CTRL_R_mouth_stickyInnerU
    OffsetSpace=LocalSpace
    OffsetTransform=(Rotation=(X=0,Y=0,Z=0,W=1),Translation=(X=0,Y=0,Z=0),Scale3D=(X=1,Y=1,Z=1))
    Settings="(PrimaryAxis=Y,bIsScale=False,Limits=(Limit=(bMinimum=True,bMaximum=True),MinValue=0,MaxValue=1,bDrawLimits=True),Shape=(bVisible=True,Name=\"DefaultGizmoLibrary.Sphere_Solid\",Color=(R=1,G=1,B=0,A=0),Transform=(Rotation=(X=0,Y=0,Z=0,W=1),Translation=(X=0,Y= ...(truncated)"
    Parent=@n50.Item
   ) {template="SpawnControl::Execute(in InitialValue,in Settings,in OffsetTransform,in OffsetSpace,in Parent,in Name,out Item)"}
   > n51 = &RigUnit_HierarchyAddNull(
    Name=CTRL_R_mouth_stickyInnerD_GRP
    Parent=(Type=Null,Name="GRP_mouthGUI")
    Space=LocalSpace
    Transform=(Rotation=(X=0.7071,Y=0,Z=0,W=0.7071),Translation=(X=-6.8529,Y=0.0221,Z=6.8351),Scale3D=(X=0.8,Y=0.8,Z=0.8))
   )
   > &RigUnit_HierarchyAddControlFloat(
    InitialValue=0.0
    Name=CTRL_R_mouth_stickyInnerD
    OffsetSpace=LocalSpace
    OffsetTransform=(Rotation=(X=0,Y=0,Z=0,W=1),Translation=(X=0,Y=0,Z=0),Scale3D=(X=1,Y=1,Z=1))
    Settings="(PrimaryAxis=Y,bIsScale=False,Limits=(Limit=(bMinimum=True,bMaximum=True),MinValue=0,MaxValue=1,bDrawLimits=True),Shape=(bVisible=True,Name=\"DefaultGizmoLibrary.Sphere_Solid\",Color=(R=1,G=1,B=0,A=0),Transform=(Rotation=(X=0,Y=0,Z=0,W=1),Translation=(X=0,Y= ...(truncated)"
    Parent=@n51.Item
   ) {template="SpawnControl::Execute(in InitialValue,in Settings,in OffsetTransform,in OffsetSpace,in Parent,in Name,out Item)"}
   > n52 = &RigUnit_HierarchyAddNull(
    Name=CTRL_R_mouth_stickyOuterU_GRP
    Parent=(Type=Null,Name="GRP_mouthGUI")
    Space=LocalSpace
    Transform=(Rotation=(X=-0.7071,Y=0,Z=0,W=0.7071),Translation=(X=-7.5686,Y=0.0221,Z=8.8351),Scale3D=(X=0.8,Y=0.8,Z=0.8))
   )
   > &RigUnit_HierarchyAddControlFloat(
    InitialValue=0.0
    Name=CTRL_R_mouth_stickyOuterU
    OffsetSpace=LocalSpace
    OffsetTransform=(Rotation=(X=0,Y=0,Z=0,W=1),Translation=(X=0,Y=0,Z=0),Scale3D=(X=1,Y=1,Z=1))
    Settings="(PrimaryAxis=Y,bIsScale=False,Limits=(Limit=(bMinimum=True,bMaximum=True),MinValue=0,MaxValue=1,bDrawLimits=True),Shape=(bVisible=True,Name=\"DefaultGizmoLibrary.Sphere_Solid\",Color=(R=1,G=1,B=0,A=0),Transform=(Rotation=(X=0,Y=0,Z=0,W=1),Translation=(X=0,Y= ...(truncated)"
    Parent=@n52.Item
   ) {template="SpawnControl::Execute(in InitialValue,in Settings,in OffsetTransform,in OffsetSpace,in Parent,in Name,out Item)"}
   > n53 = &RigUnit_HierarchyAddNull(
    Name=CTRL_R_mouth_stickyOuterD_GRP
    Parent=(Type=Null,Name="GRP_mouthGUI")
    Space=LocalSpace
    Transform=(Rotation=(X=0.7071,Y=0,Z=0,W=0.7071),Translation=(X=-7.5686,Y=0.0221,Z=6.8351),Scale3D=(X=0.8,Y=0.8,Z=0.8))
   )
   > &RigUnit_HierarchyAddControlFloat(
    InitialValue=0.0
    Name=CTRL_R_mouth_stickyOuterD
    OffsetSpace=LocalSpace
    OffsetTransform=(Rotation=(X=0,Y=0,Z=0,W=1),Translation=(X=0,Y=0,Z=0),Scale3D=(X=1,Y=1,Z=1))
    Settings="(PrimaryAxis=Y,bIsScale=False,Limits=(Limit=(bMinimum=True,bMaximum=True),MinValue=0,MaxValue=1,bDrawLimits=True),Shape=(bVisible=True,Name=\"DefaultGizmoLibrary.Sphere_Solid\",Color=(R=1,G=1,B=0,A=0),Transform=(Rotation=(X=0,Y=0,Z=0,W=1),Translation=(X=0,Y= ...(truncated)"
    Parent=@n53.Item
   ) {template="SpawnControl::Execute(in InitialValue,in Settings,in OffsetTransform,in OffsetSpace,in Parent,in Name,out Item)"}
   &RigUnit_SetMetadataTag(ExecutePin=(), Item=(Type=Bone,Name="None"), NameSpace=Root, Tag=wm_curves)
  ]}
 RigVMFunctionLibrary {nodes [
   CRFL FN MH Backward v01 = RigVMCollapseNode(ExecuteContext=())
   CRFL FN MH Construction v01 = RigVMCollapseNode(Control Color=(R=1,G=1,B=0,A=1), Control Scale=1, ExecuteContext=())
   CRFL FN MH Forward v01 = RigVMCollapseNode(ExecuteContext=())
   CRFL FN MH Tweak Backward v01 = RigVMCollapseNode(ExecuteContext=())
   CRFL FN MH Tweaker Construction v01 = RigVMCollapseNode(ExecuteContext=(), Parent=(Type=None,Name="None"))
   CRFL FN MH Tweaker Forward v01 = RigVMCollapseNode(ExecuteContext=())
  ]}
 RigVMFunction_Sequence_1_ContainedGraph {nodes [
   n1 = RigVMFunctionEntryNode()
   RigVMFunction_Sequence = &RigVMFunction_Sequence(ExecuteContext=@n1.ExecuteContext)
   n2 = &RigVMFunction_Sequence(ExecuteContext=@RigVMFunction_Sequence.B)
   RigVMFunctionReturnNode(A=@RigVMFunction_Sequence.A, B=@n2.A, C=@n2.B)
  ]}
 RigVMFunction_Sequence_1_ContainedGraph {nodes [
   n1 = RigVMFunctionEntryNode()
   RigVMFunction_Sequence = &RigVMFunction_Sequence(ExecuteContext=@n1.ExecuteContext)
   n2 = &RigVMFunction_Sequence(ExecuteContext=@RigVMFunction_Sequence.B)
   RigVMFunctionReturnNode(A=@RigVMFunction_Sequence.A, B=@n2.A, C=@n2.B)
  ]}
 RigVMFunction_Sequence_1_ContainedGraph {nodes [
   n1 = RigVMFunctionEntryNode()
   RigVMFunction_Sequence = &RigVMFunction_Sequence(ExecuteContext=@n1.ExecuteContext)
   n2 = &RigVMFunction_Sequence(ExecuteContext=@RigVMFunction_Sequence.B)
   RigVMFunctionReturnNode(A=@RigVMFunction_Sequence.A, B=@n2.A, C=@n2.B)
  ]}
 RigVMFunction_Sequence_1_ContainedGraph {nodes [
   n1 = RigVMFunctionEntryNode()
   RigVMFunction_Sequence = &RigVMFunction_Sequence(ExecuteContext=@n1.ExecuteContext)
   n2 = &RigVMFunction_Sequence(ExecuteContext=@RigVMFunction_Sequence.B)
   n3 = &RigVMFunction_Sequence(ExecuteContext=@n2.B)
   RigVMFunctionReturnNode(A=@RigVMFunction_Sequence.A, B=@n2.A, C=@n3.A, D=@n3.B)
  ]}
 RigVMFunction_Sequence_1_ContainedGraph {nodes [
   n1 = RigVMFunctionEntryNode()
   RigVMFunction_Sequence = &RigVMFunction_Sequence(ExecuteContext=@n1.ExecuteContext)
   n2 = &RigVMFunction_Sequence(ExecuteContext=@RigVMFunction_Sequence.B)
   n3 = &RigVMFunction_Sequence(ExecuteContext=@n2.B)
   RigVMFunctionReturnNode(A=@RigVMFunction_Sequence.A, B=@n2.A, C=@n3.A, D=@n3.B)
  ]}
 RigVMFunction_Sequence_2_ContainedGraph {nodes [
   n1 = RigVMFunctionEntryNode()
   n2 = &RigVMFunction_Sequence(ExecuteContext=@n1.ExecuteContext)
   n3 = &RigVMFunction_Sequence(ExecuteContext=@n2.B)
   n4 = &RigVMFunction_Sequence(ExecuteContext=@n3.B)
   RigVMFunctionReturnNode(B=@n3.A, C=@n4.A, D=@n4.B, A=@n2.A)
  ]}
 RigVMFunction_Sequence_2_ContainedGraph {nodes [
   n1 = RigVMFunctionEntryNode()
   n2 = &RigVMFunction_Sequence(ExecuteContext=@n1.ExecuteContext)
   n3 = &RigVMFunction_Sequence(ExecuteContext=@n2.B)
   n4 = &RigVMFunction_Sequence(ExecuteContext=@n3.B)
   RigVMFunctionReturnNode(A=@n2.A, B=@n3.A, C=@n4.A, D=@n4.B)
  ]}
 RigVMFunction_Sequence_2_ContainedGraph {nodes [
   n1 = RigVMFunctionEntryNode()
   n2 = &RigVMFunction_Sequence(ExecuteContext=@n1.ExecuteContext)
   n3 = &RigVMFunction_Sequence(ExecuteContext=@n2.B)
   RigVMFunctionReturnNode(A=@n2.A, B=@n3.A, C=@n3.B)
  ]}
 RigVMFunction_Sequence_6_ContainedGraph {nodes [
   n1 = RigVMFunctionEntryNode()
   n2 = &RigVMFunction_Sequence(ExecuteContext=@n1.ExecuteContext)
   n3 = &RigVMFunction_Sequence(ExecuteContext=@n2.B)
   n4 = &RigVMFunction_Sequence(ExecuteContext=@n3.B)
   n5 = &RigVMFunction_Sequence(ExecuteContext=@n4.B)
   RigVMFunctionReturnNode(B=@n3.A, C=@n4.A, D=@n5.A, E=@n5.B, A=@n2.A)
  ]}
 RigVMFunction_Sequence_9_ContainedGraph {nodes [
   n1 = RigVMFunctionEntryNode()
   n2 = &RigVMFunction_Sequence(ExecuteContext=@n1.ExecuteContext)
   n3 = &RigVMFunction_Sequence(ExecuteContext=@n2.B)
   n4 = &RigVMFunction_Sequence(ExecuteContext=@n3.B)
   n5 = &RigVMFunction_Sequence(ExecuteContext=@n4.B)
   n6 = &RigVMFunction_Sequence(ExecuteContext=@n5.B)
   RigVMFunctionReturnNode(B=@n3.A, C=@n4.A, D=@n5.A, E=@n6.A, F=@n6.B, A=@n2.A)
  ]}
 RigVMModel {nodes [
   CRFL FN MH Forward v01 = RigVMFunctionReferenceNode(ExecuteContext=(), Is V2=true)
   RigUnit_BeginExecution = &RigUnit_BeginExecution()
  ]}
 RigVMModel Rig {nodes [
   CRFL FN MH Backward v01 = RigVMFunctionReferenceNode(ExecuteContext=(), Is V2=true)
   /Script/ControlRig.RigUnit_InverseExecution() {is=struct RigVMUnitNode}
  ]}
 RigVMModel Rig_1 {nodes [
   CRFL FN MH Construction v01 = RigVMFunctionReferenceNode(Control Color=(R=1,G=1,B=0,A=1), Control Scale=1, ExecuteContext=(), Is V2=true)
   &RigUnit_PrepareForExecution()
  ]}
 SequenceAggregate_2_ContainedGraph {nodes [
   n1 = RigVMFunctionEntryNode()
   SequenceAggregate_1 = &RigVMFunction_Sequence(ExecuteContext=@n1.ExecuteContext)
   SequenceAggregate_2 = &RigVMFunction_Sequence(ExecuteContext=@SequenceAggregate_1.B)
   SequenceAggregate_3 = &RigVMFunction_Sequence(ExecuteContext=@SequenceAggregate_2.B)
   SequenceAggregate_4 = &RigVMFunction_Sequence(ExecuteContext=@SequenceAggregate_3.B)
   SequenceAggregate_5 = &RigVMFunction_Sequence(ExecuteContext=@SequenceAggregate_4.B)
   RigVMFunctionReturnNode(
    A=@SequenceAggregate_1.A
    B=@SequenceAggregate_2.A
    C=@SequenceAggregate_3.A
    D=@SequenceAggregate_4.A
    E=@SequenceAggregate_5.A
    F=@SequenceAggregate_5.B
   )
  ]}
 SequenceAggregate_2_ContainedGraph {nodes [
   n1 = RigVMFunctionEntryNode()
   SequenceAggregate_1 = &RigVMFunction_Sequence(ExecuteContext=@n1.ExecuteContext)
   SequenceAggregate_2 = &RigVMFunction_Sequence(ExecuteContext=@SequenceAggregate_1.B)
   SequenceAggregate_3 = &RigVMFunction_Sequence(ExecuteContext=@SequenceAggregate_2.B)
   SequenceAggregate_4 = &RigVMFunction_Sequence(ExecuteContext=@SequenceAggregate_3.B)
   SequenceAggregate_5 = &RigVMFunction_Sequence(ExecuteContext=@SequenceAggregate_4.B)
   RigVMFunctionReturnNode(
    A=@SequenceAggregate_1.A
    B=@SequenceAggregate_2.A
    C=@SequenceAggregate_3.A
    D=@SequenceAggregate_4.A
    E=@SequenceAggregate_5.A
    F=@SequenceAggregate_5.B
   )
  ]}
]
hierarchy [
 root {loc=<0 0 0>, rot=<0 0 0 1>}
 pelvis {parent=root, loc=<0 -0.7071 93.1191>, rot=<0.0127 0.707 -0.0127 -0.707>}
 spine_01 {parent=pelvis, loc=<6.7494 -0.6388 0>, rot=<0 0 0.034 -0.9994>}
 spine_02 {parent=spine_01, loc=<7.2518 -0.035 0>, rot=<0 0 0.0238 -0.9997>}
 spine_03 {parent=spine_02, loc=<6.0104 0 0>, rot=<0 0 0.0149 -0.9999>}
 spine_04 {parent=spine_03, loc=<8.1024 0 0>, rot=<0 0 -0.0132 -0.9999>}
 spine_05 {parent=spine_04, loc=<10.0146 0 0>, rot=<0 0 -0.0015 -1>}
 neck_01 {parent=spine_05, loc=<11.3138 0.1529 0>, rot=<0 0 -0.2003 -0.9797>}
 neck_02 {parent=neck_01, loc=<4.5627 0 0>, rot=<0 0 0.0504 -0.9987>}
 head {parent=neck_02, loc=<5.3914 0 0>, rot=<0 0 0.115 -0.9934>}
 FACIAL_C_FacialRoot {parent=head, loc=<2.9712 -0.4622 0>, rot=<0.5021 -0.4979 -0.5021 -0.4979>}
 FACIAL_C_Forehead {parent=FACIAL_C_FacialRoot, loc=<0 -3.1308 9.3974>, rot=<0.0207 0 0 -0.9998>}
 FACIAL_L_ForeheadIn {parent=FACIAL_C_FacialRoot, loc=<1.5783 -2.9083 9.9779>, rot=<0.0031 -0.0674 0.0105 -0.9977>}
 FACIAL_R_ForeheadIn {parent=FACIAL_C_FacialRoot, loc=<-1.5783 -2.9083 9.9779>, rot=<0.0031 0.0674 -0.0105 -0.9977>}
 FACIAL_L_ForeheadMid {parent=FACIAL_C_FacialRoot, loc=<2.867 -3.3915 9.4422>, rot=<-0.0082 -0.1749 -0.0076 -0.9845>}
 FACIAL_L_ForeheadMid1 {parent=FACIAL_L_ForeheadMid, loc=<-0.5077 0.2381 -0.0995>, rot=<0 0 0 1>}
 FACIAL_L_ForeheadMid2 {parent=FACIAL_L_ForeheadMid, loc=<0.674 0.1716 -0.0884>, rot=<0 0 0 1>}
 FACIAL_R_ForeheadMid {parent=FACIAL_C_FacialRoot, loc=<-2.867 -3.3915 9.4422>, rot=<-0.0082 0.1749 0.0076 -0.9845>}
 FACIAL_R_ForeheadMid1 {parent=FACIAL_R_ForeheadMid, loc=<0.5077 0.2381 -0.0995>, rot=<0 0 0 1>}
 FACIAL_R_ForeheadMid2 {parent=FACIAL_R_ForeheadMid, loc=<-0.674 0.1716 -0.0884>, rot=<0 0 0 1>}
 FACIAL_L_ForeheadOut {parent=FACIAL_C_FacialRoot, loc=<4.0465 -2.9836 8.7213>, rot=<-0.0602 -0.2584 -0.0611 -0.9622>}
 FACIAL_R_ForeheadOut {parent=FACIAL_C_FacialRoot, loc=<-4.0465 -2.9836 8.7213>, rot=<-0.0602 0.2584 0.0611 -0.9622>}
 FACIAL_L_EyesackUpper {parent=FACIAL_C_FacialRoot, loc=<2.8439 -2.3877 9.3504>, rot=<0.1124 -0.0867 -0.0571 -0.9882>}
 FACIAL_R_EyesackUpper {parent=FACIAL_C_FacialRoot, loc=<-2.8439 -2.3877 9.3504>, rot=<0.1124 0.0867 0.0571 -0.9882>}
 FACIAL_L_EyelidUpperFurrow {parent=FACIAL_C_FacialRoot, loc=<2.6585 -2.0955 9.0119>, rot=<0.0331 -0.154 -0.068 -0.9852>}
 FACIAL_R_EyelidUpperFurrow {parent=FACIAL_C_FacialRoot, loc=<-2.6585 -2.0955 9.0119>, rot=<0.0331 0.154 0.068 -0.9852>}
 FACIAL_L_EyelidUpperB {parent=FACIAL_C_FacialRoot, loc=<2.6949 -1.4649 7.7068>, rot=<-0.1618 0.002 0.0014 -0.9868>}
 FACIAL_R_EyelidUpperB {parent=FACIAL_C_FacialRoot, loc=<-2.6949 -1.4649 7.7068>, rot=<-0.1618 -0.002 -0.0014 -0.9868>}
 FACIAL_L_EyelidUpperA {parent=FACIAL_C_FacialRoot, loc=<2.6949 -1.4649 7.7068>, rot=<-0.0839 0.0019 0.0016 -0.9965>}
 FACIAL_L_EyelidUpperA1 {parent=FACIAL_L_EyelidUpperA, loc=<-0.6817 0.0696 1.1303>, rot=<0.0106 0.1214 0.0865 -0.9888>}
 FACIAL_L_EyelidUpperA3 {parent=FACIAL_L_EyelidUpperA, loc=<0.7566 -0.0356 1.1039>, rot=<-0.0086 -0.1657 -0.0487 -0.9849>}
 FACIAL_R_EyelidUpperA {parent=FACIAL_C_FacialRoot, loc=<-2.6949 -1.4649 7.7068>, rot=<-0.0839 -0.0019 -0.0016 -0.9965>}
 FACIAL_R_EyelidUpperA1 {parent=FACIAL_R_EyelidUpperA, loc=<0.6817 0.0696 1.1303>, rot=<0.0106 -0.1214 -0.0865 -0.9888>}
 FACIAL_R_EyelidUpperA3 {parent=FACIAL_R_EyelidUpperA, loc=<-0.7566 -0.0356 1.1039>, rot=<-0.0086 0.1657 0.0487 -0.9849>}
 FACIAL_L_Eye {parent=FACIAL_C_FacialRoot, loc=<2.6949 -1.4649 7.7068>, rot=<0.0643 0.0017 0.0019 -0.9979>}
 FACIAL_L_Pupil {parent=FACIAL_L_Eye, loc=<0.1563 -0.1178 1.1715>, rot=<0 0 0 1>}
 FACIAL_R_Eye {parent=FACIAL_C_FacialRoot, loc=<-2.6949 -1.4649 7.7068>, rot=<0.0643 -0.0017 -0.0019 -0.9979>}
 FACIAL_R_Pupil {parent=FACIAL_R_Eye, loc=<-0.1563 -0.1178 1.1715>, rot=<0 0 0 1>}
 FACIAL_L_EyelidLowerA {parent=FACIAL_C_FacialRoot, loc=<2.6949 -1.4649 7.7068>, rot=<0.3451 0.0011 0.0023 -0.9386>}
 FACIAL_L_EyelidLowerA1 {parent=FACIAL_L_EyelidLowerA, loc=<-0.7483 -0.3072 1.1061>, rot=<0 0.0785 0 -0.9969>}
 FACIAL_L_EyelidLowerA3 {parent=FACIAL_L_EyelidLowerA, loc=<0.6735 -0.1663 1.1484>, rot=<0 -0.1822 0 -0.9833>}
 FACIAL_R_EyelidLowerA {parent=FACIAL_C_FacialRoot, loc=<-2.6949 -1.4649 7.7068>, rot=<0.3451 -0.0011 -0.0023 -0.9386>}
 FACIAL_R_EyelidLowerA1 {parent=FACIAL_R_EyelidLowerA, loc=<0.7483 -0.3072 1.1061>, rot=<0 -0.0785 0 -0.9969>}
 FACIAL_R_EyelidLowerA3 {parent=FACIAL_R_EyelidLowerA, loc=<-0.6735 -0.1663 1.1484>, rot=<0 0.1822 0 -0.9833>}
 FACIAL_L_EyelidLowerB {parent=FACIAL_C_FacialRoot, loc=<2.6949 -1.4649 7.7068>, rot=<0.3776 0.001 0.0023 -0.926>}
 FACIAL_L_EyelidLowerB1 {parent=FACIAL_L_EyelidLowerB, loc=<-0.7153 -0.1974 1.2808>, rot=<0 0.0785 0 -0.9969>}
 FACIAL_L_EyelidLowerB3 {parent=FACIAL_L_EyelidLowerB, loc=<0.7461 0.0443 1.2303>, rot=<0 -0.1822 0 -0.9833>}
 FACIAL_R_EyelidLowerB {parent=FACIAL_C_FacialRoot, loc=<-2.6949 -1.4649 7.7068>, rot=<0.3776 -0.001 -0.0023 -0.926>}
 FACIAL_R_EyelidLowerB1 {parent=FACIAL_R_EyelidLowerB, loc=<0.7153 -0.1974 1.2808>, rot=<0 -0.0785 0 -0.9969>}
 FACIAL_R_EyelidLowerB3 {parent=FACIAL_R_EyelidLowerB, loc=<-0.7461 0.0443 1.2303>, rot=<0 0.1822 0 -0.9833>}
 FACIAL_L_EyeCornerInner {parent=FACIAL_C_FacialRoot, loc=<1.7099 -1.1996 8.5226>, rot=<0.0643 0.0017 0.0019 -0.9979>}
 FACIAL_R_EyeCornerInner {parent=FACIAL_C_FacialRoot, loc=<-1.7099 -1.1996 8.5226>, rot=<0.0643 -0.0017 -0.0019 -0.9979>}
 FACIAL_L_EyeCornerOuter {parent=FACIAL_C_FacialRoot, loc=<3.9155 -1.2959 7.935>, rot=<0.0587 -0.3803 0.0263 -0.9226>}
 FACIAL_R_EyeCornerOuter {parent=FACIAL_C_FacialRoot, loc=<-3.9155 -1.2959 7.935>, rot=<0.0587 0.3803 -0.0263 -0.9226>}
 FACIAL_L_EyesackLower {parent=FACIAL_C_FacialRoot, loc=<2.8327 -0.1952 8.2854>, rot=<0.0412 -0.1392 -0.0114 -0.9893>}
 FACIAL_L_EyesackLower1 {parent=FACIAL_L_EyesackLower, loc=<-0.992 -0.1607 0.241>, rot=<0.0204 0.1999 -0.0204 -0.9794>}
 FACIAL_L_EyesackLower2 {parent=FACIAL_L_EyesackLower, loc=<0.9592 -0.1302 0.2257>, rot=<0.0082 -0.0885 0.1026 -0.9907>}
 FACIAL_R_EyesackLower {parent=FACIAL_C_FacialRoot, loc=<-2.8327 -0.1952 8.2854>, rot=<0.0412 0.1392 0.0114 -0.9893>}
 FACIAL_R_EyesackLower1 {parent=FACIAL_R_EyesackLower, loc=<0.992 -0.1607 0.241>, rot=<0.0204 -0.1999 0.0204 -0.9794>}
 FACIAL_R_EyesackLower2 {parent=FACIAL_R_EyesackLower, loc=<-0.9592 -0.1302 0.2257>, rot=<0.0082 0.0885 -0.1026 -0.9907>}
 FACIAL_L_CheekInner {parent=FACIAL_C_FacialRoot, loc=<1.8741 0.5243 8.2984>, rot=<-0.0606 -0.0987 -0.1895 -0.975>}
 FACIAL_L_CheekInner1 {parent=FACIAL_L_CheekInner, loc=<-0.5285 -0.2617 0.8209>, rot=<0.0509 -0.0536 0.0234 -0.997>}
 FACIAL_L_CheekInner2 {parent=FACIAL_L_CheekInner, loc=<-0.319 0.3869 1.159>, rot=<0.0349 -0.0565 -0.1491 -0.9866>}
 FACIAL_L_CheekInner3 {parent=FACIAL_L_CheekInner, loc=<0.5554 -0.1496 0.7785>, rot=<0.0509 -0.0536 0.0234 -0.997>}
 FACIAL_L_CheekInner4 {parent=FACIAL_L_CheekInner, loc=<0.6441 0.7311 0.9265>, rot=<0.0349 -0.0565 -0.1491 -0.9866>}
 FACIAL_R_CheekInner {parent=FACIAL_C_FacialRoot, loc=<-1.8741 0.5243 8.2984>, rot=<-0.0606 0.0987 0.1895 -0.975>}
 FACIAL_R_CheekInner1 {parent=FACIAL_R_CheekInner, loc=<0.5285 -0.2617 0.8209>, rot=<0.0509 0.0536 -0.0234 -0.997>}
 FACIAL_R_CheekInner2 {parent=FACIAL_R_CheekInner, loc=<0.319 0.3869 1.159>, rot=<0.0349 0.0565 0.1491 -0.9866>}
 FACIAL_R_CheekInner3 {parent=FACIAL_R_CheekInner, loc=<-0.5554 -0.1496 0.7785>, rot=<0.0509 0.0536 -0.0234 -0.997>}
 FACIAL_R_CheekInner4 {parent=FACIAL_R_CheekInner, loc=<-0.6441 0.7311 0.9265>, rot=<0.0349 0.0565 0.1491 -0.9866>}
 FACIAL_L_CheekOuter {parent=FACIAL_C_FacialRoot, loc=<3.3414 0.7955 6.9957>, rot=<0.1132 -0.3684 0.1262 -0.9141>}
 FACIAL_L_CheekOuter1 {parent=FACIAL_L_CheekOuter, loc=<-0.5843 -0.5617 1.2992>, rot=<-0.0677 -0.0465 -0.0404 -0.9958>}
 FACIAL_L_CheekOuter2 {parent=FACIAL_L_CheekOuter, loc=<0.0775 1.0309 1.24>, rot=<0.0801 -0.0306 -0.0549 -0.9948>}
 FACIAL_L_CheekOuter3 {parent=FACIAL_L_CheekOuter, loc=<1.0345 -0.5023 1.2182>, rot=<-0.0491 -0.0605 0.2182 -0.9728>}
 FACIAL_R_CheekOuter {parent=FACIAL_C_FacialRoot, loc=<-3.3414 0.7955 6.9957>, rot=<0.1132 0.3684 -0.1262 -0.9141>}
 FACIAL_R_CheekOuter1 {parent=FACIAL_R_CheekOuter, loc=<0.5843 -0.5617 1.2992>, rot=<-0.0677 0.0465 0.0404 -0.9958>}
 FACIAL_R_CheekOuter2 {parent=FACIAL_R_CheekOuter, loc=<-0.0775 1.0309 1.24>, rot=<0.0801 0.0306 0.0549 -0.9948>}
 FACIAL_R_CheekOuter3 {parent=FACIAL_R_CheekOuter, loc=<-1.0345 -0.5023 1.2182>, rot=<-0.0491 0.0605 -0.2182 -0.9728>}
 FACIAL_C_NoseBridge {parent=FACIAL_C_FacialRoot, loc=<0 -1.8959 10.1812>, rot=<0.1078 0 0 -0.9942>}
 FACIAL_L_NasolabialBulge1 {parent=FACIAL_C_FacialRoot, loc=<3.2494 2.3983 8.6467>, rot=<-0.025 -0.2497 -0.3656 -0.8963>}
 FACIAL_R_NasolabialBulge1 {parent=FACIAL_C_FacialRoot, loc=<-3.2494 2.3983 8.6467>, rot=<-0.025 0.2497 0.3656 -0.8963>}
 FACIAL_L_NasolabialBulge {parent=FACIAL_C_FacialRoot, loc=<3.0622 4.338 7.415>, rot=<0.1164 -0.3737 0.0498 -0.9189>}
 FACIAL_L_NasolabialBulge2 {parent=FACIAL_L_NasolabialBulge, loc=<0.0915 -0.9316 0.7307>, rot=<-0.1022 0.041 0.057 -0.9923>}
 FACIAL_L_NasolabialBulge3 {parent=FACIAL_L_NasolabialBulge, loc=<-0.1478 0.616 0.8109>, rot=<-0.0099 0.0168 -0.0699 -0.9974>}
 FACIAL_R_NasolabialBulge {parent=FACIAL_C_FacialRoot, loc=<-3.0622 4.338 7.415>, rot=<0.1164 0.3737 -0.0498 -0.9189>}
 FACIAL_R_NasolabialBulge2 {parent=FACIAL_R_NasolabialBulge, loc=<-0.0915 -0.9316 0.7307>, rot=<-0.1022 -0.041 -0.057 -0.9923>}
 FACIAL_R_NasolabialBulge3 {parent=FACIAL_R_NasolabialBulge, loc=<0.1478 0.616 0.8109>, rot=<-0.0099 -0.0168 0.0699 -0.9974>}
 FACIAL_L_NasolabialFurrow {parent=FACIAL_C_FacialRoot, loc=<2.7204 2.6119 9.0327>, rot=<0.0264 -0.1474 -0.3533 -0.9235>}
 FACIAL_R_NasolabialFurrow {parent=FACIAL_C_FacialRoot, loc=<-2.7204 2.6119 9.0327>, rot=<0.0264 0.1474 0.3533 -0.9235>}
 FACIAL_L_CheekLower {parent=FACIAL_C_FacialRoot, loc=<3.5517 3.9289 6.316>, rot=<0.0916 -0.4209 0.0868 -0.8983>}
 FACIAL_L_CheekLower1 {parent=FACIAL_L_CheekLower, loc=<-0.3052 0.2977 0.8335>, rot=<0 0.0872 0 -0.9962>}
 FACIAL_L_CheekLower2 {parent=FACIAL_L_CheekLower, loc=<0.7344 0.1917 0.5743>, rot=<0 -0.0872 0 -0.9962>}
 FACIAL_R_CheekLower {parent=FACIAL_C_FacialRoot, loc=<-3.5517 3.9289 6.316>, rot=<0.0916 0.4209 -0.0868 -0.8983>}
 FACIAL_R_CheekLower1 {parent=FACIAL_R_CheekLower, loc=<0.3052 0.2977 0.8335>, rot=<0 -0.0872 0 -0.9962>}
 FACIAL_R_CheekLower2 {parent=FACIAL_R_CheekLower, loc=<-0.7344 0.1917 0.5743>, rot=<0 0.0872 0 -0.9962>}
 FACIAL_L_Ear {parent=FACIAL_C_FacialRoot, loc=<6.4207 0.0696 1.4054>, rot=<0.0382 -0.0006 -0.0243 -0.999>}
 FACIAL_R_Ear {parent=FACIAL_C_FacialRoot, loc=<-6.4207 0.0696 1.4054>, rot=<0.0382 0.0006 0.0243 -0.999>}
 FACIAL_C_Nose {parent=FACIAL_C_FacialRoot, loc=<0 0.1218 10.1331>, rot=<-0.0665 0 0 -0.9978>}
 FACIAL_C_NoseLower {parent=FACIAL_C_Nose, loc=<0 2.611 0.1959>, rot=<0.1305 0 0 -0.9914>}
 FACIAL_C_NoseTip {parent=FACIAL_C_Nose, loc=<0 1.7381 0.7792>, rot=<0.0872 0 0 -0.9962>}
 FACIAL_L_Nostril {parent=FACIAL_C_Nose, loc=<1.3616 1.8117 0.0109>, rot=<0.0173 0.1305 0.0023 -0.9913>}
 FACIAL_R_Nostril {parent=FACIAL_C_Nose, loc=<-1.3616 1.8117 0.0109>, rot=<0.0173 -0.1305 -0.0023 -0.9913>}
 FACIAL_C_LipUpperSkin {parent=FACIAL_C_FacialRoot, loc=<0 3.1175 10.6656>, rot=<0.0207 0 0 -0.9998>}
 FACIAL_L_LipUpperOuterSkin {parent=FACIAL_C_FacialRoot, loc=<2.4826 3.4496 9.4024>, rot=<-0.0193 -0.1728 -0.0993 -0.9797>}
 FACIAL_R_LipUpperOuterSkin {parent=FACIAL_C_FacialRoot, loc=<-2.4826 3.4496 9.4024>, rot=<-0.0193 0.1728 0.0993 -0.9797>}
 FACIAL_C_MouthUpper {parent=FACIAL_C_FacialRoot, loc=<0 4.4644 7.5076>, rot=<0.0643 0 0 -0.9979>}
 FACIAL_C_LipUpper {parent=FACIAL_C_MouthUpper, loc=<0 -0.9938 2.4973>, rot=<-0.0872 0 0 -0.9962>}
 FACIAL_C_LipUpper1 {parent=FACIAL_C_LipUpper, loc=<0 -0.1465 0.8579>, rot=<0 0 0 1>}
 FACIAL_C_LipUpper2 {parent=FACIAL_C_LipUpper, loc=<0 0.6591 0.1943>, rot=<0 0 0 1>}
 FACIAL_L_LipUpper {parent=FACIAL_C_MouthUpper, loc=<1.1829 -0.8863 2.1826>, rot=<-0.0778 -0.1601 -0.0379 -0.9833>}
 FACIAL_L_LipUpper1 {parent=FACIAL_L_LipUpper, loc=<-0.0571 -0.1403 0.6326>, rot=<0 0 0 1>}
 FACIAL_L_LipUpper2 {parent=FACIAL_L_LipUpper, loc=<-0.1281 0.5669 0.1244>, rot=<0 0 0 1>}
 FACIAL_R_LipUpper {parent=FACIAL_C_MouthUpper, loc=<-1.1829 -0.8863 2.1826>, rot=<-0.0778 0.1601 0.0379 -0.9833>}
 FACIAL_R_LipUpper1 {parent=FACIAL_R_LipUpper, loc=<0.0571 -0.1403 0.6326>, rot=<0 0 0 1>}
 FACIAL_R_LipUpper2 {parent=FACIAL_R_LipUpper, loc=<0.1281 0.5669 0.1244>, rot=<0 0 0 1>}
 FACIAL_L_LipUpperOuter {parent=FACIAL_C_MouthUpper, loc=<1.9999 -0.6965 1.5448>, rot=<-0.0983 -0.2294 -0.0983 -0.9634>}
 FACIAL_L_LipUpperOuter1 {parent=FACIAL_L_LipUpperOuter, loc=<-0.0959 0.1942 0.5095>, rot=<0 0 0 1>}
 FACIAL_L_LipUpperOuter2 {parent=FACIAL_L_LipUpperOuter, loc=<-0.2529 0.5561 0.2264>, rot=<0 0 0 1>}
 FACIAL_R_LipUpperOuter {parent=FACIAL_C_MouthUpper, loc=<-1.9999 -0.6965 1.5448>, rot=<-0.0983 0.2294 0.0983 -0.9634>}
 FACIAL_R_LipUpperOuter1 {parent=FACIAL_R_LipUpperOuter, loc=<0.0959 0.1942 0.5095>, rot=<0 0 0 1>}
 FACIAL_R_LipUpperOuter2 {parent=FACIAL_R_LipUpperOuter, loc=<0.2529 0.5561 0.2264>, rot=<0 0 0 1>}
 FACIAL_L_LipCorner {parent=FACIAL_C_MouthUpper, loc=<2.619 -0.1516 1.046>, rot=<0 -0.2079 0 -0.9781>}
 FACIAL_L_LipCorner2 {parent=FACIAL_L_LipCorner, loc=<-0.1424 -0.0129 -0.7513>, rot=<0 0.2588 0 -0.9659>}
 FACIAL_L_LipCorner3 {parent=FACIAL_L_LipCorner, loc=<-0.4025 0.5791 -0.5869>, rot=<0 0.2588 0 -0.9659>}
 FACIAL_R_LipCorner {parent=FACIAL_C_MouthUpper, loc=<-2.6191 -0.1516 1.046>, rot=<0 0.2079 0 -0.9781>}
 FACIAL_R_LipCorner2 {parent=FACIAL_R_LipCorner, loc=<0.1424 -0.0129 -0.7513>, rot=<0 -0.2588 0 -0.9659>}
 FACIAL_R_LipCorner3 {parent=FACIAL_R_LipCorner, loc=<0.4025 0.5791 -0.5869>, rot=<0 -0.2588 0 -0.9659>}
 FACIAL_L_JawBulge {parent=FACIAL_C_FacialRoot, loc=<5.2193 3.4628 3.4629>, rot=<0.1506 0.0886 0.004 -0.9846>}
 FACIAL_R_JawBulge {parent=FACIAL_C_FacialRoot, loc=<-5.2193 3.4628 3.4629>, rot=<0.1506 -0.0886 -0.004 -0.9846>}
 FACIAL_L_Masseter {parent=FACIAL_C_FacialRoot, loc=<5.3206 1.6285 4.5667>, rot=<0.1331 0.1147 0.0026 -0.9844>}
 FACIAL_R_Masseter {parent=FACIAL_C_FacialRoot, loc=<-5.3206 1.6285 4.5667>, rot=<0.1331 -0.1147 -0.0026 -0.9844>}
 FACIAL_C_UnderChin {parent=FACIAL_C_FacialRoot, loc=<0 8.5606 6.1576>, rot=<0.5546 0 0 -0.8321>}
 FACIAL_L_UnderChin {parent=FACIAL_C_FacialRoot, loc=<3.1037 7.3473 4.3729>, rot=<0.4548 -0.3483 0.1072 -0.8126>}
 FACIAL_R_UnderChin {parent=FACIAL_C_FacialRoot, loc=<-3.1037 7.3473 4.3729>, rot=<0.4548 0.3483 -0.1072 -0.8126>}
 FACIAL_C_TeethUpper {parent=FACIAL_C_FacialRoot, loc=<0 4.3737 9.6558>, rot=<0.0643 0 0 -0.9979>}
 FACIAL_C_LowerLipRotation {parent=FACIAL_C_FacialRoot, loc=<0 1.9413 2.9092>, rot=<0.4017 0 0 -0.9158>}
 FACIAL_C_LipLowerSkin {parent=FACIAL_C_LowerLipRotation, loc=<0 -2.158 8.1295>, rot=<-0.2588 0 0 -0.9659>}
 FACIAL_L_LipLowerOuterSkin {parent=FACIAL_C_LowerLipRotation, loc=<2.5044 -2.0345 6.9892>, rot=<-0.2809 -0.2529 0.0021 -0.9258>}
 FACIAL_R_LipLowerOuterSkin {parent=FACIAL_C_LowerLipRotation, loc=<-2.5044 -2.0345 6.9892>, rot=<-0.2809 0.2529 -0.0021 -0.9258>}
 FACIAL_C_MouthLower {parent=FACIAL_C_LowerLipRotation, loc=<0 -1.6746 4.9707>, rot=<-0.342 0 0 -0.9397>}
 FACIAL_C_LipLower {parent=FACIAL_C_MouthLower, loc=<0 0.5544 2.7134>, rot=<0.0872 0 0 -0.9962>}
 FACIAL_C_LipLower1 {parent=FACIAL_C_LipLower, loc=<0 0.168 0.5115>, rot=<0 0 0 1>}
 FACIAL_C_LipLower2 {parent=FACIAL_C_LipLower, loc=<0 -0.8438 -0.3887>, rot=<0 0 0 1>}
 FACIAL_L_LipLower {parent=FACIAL_C_MouthLower, loc=<1.1262 0.4533 2.3553>, rot=<0.0927 -0.1295 0.0273 -0.9869>}
 FACIAL_L_LipLower1 {parent=FACIAL_L_LipLower, loc=<0.2662 0.1675 0.5708>, rot=<0 0 0 1>}
 FACIAL_L_LipLower2 {parent=FACIAL_L_LipLower, loc=<-0.0195 -0.7652 -0.2768>, rot=<0 0 0 1>}
 FACIAL_R_LipLower {parent=FACIAL_C_MouthLower, loc=<-1.1263 0.4533 2.3553>, rot=<0.0927 0.1295 -0.0273 -0.9869>}
 FACIAL_R_LipLower1 {parent=FACIAL_R_LipLower, loc=<-0.2662 0.1675 0.5708>, rot=<0 0 0 1>}
 FACIAL_R_LipLower2 {parent=FACIAL_R_LipLower, loc=<0.0195 -0.7652 -0.2768>, rot=<0 0 0 1>}
 FACIAL_L_LipLowerOuter {parent=FACIAL_C_MouthLower, loc=<1.8907 0.2989 1.9184>, rot=<0.0913 -0.2322 0.0318 -0.9679>}
 FACIAL_L_LipLowerOuter1 {parent=FACIAL_L_LipLowerOuter, loc=<0.1099 0.008 0.4186>, rot=<0 0 0 -1>}
 FACIAL_L_LipLowerOuter2 {parent=FACIAL_L_LipLowerOuter, loc=<-0.0021 -0.562 -0.1956>, rot=<0 0 0 -1>}
 FACIAL_R_LipLowerOuter {parent=FACIAL_C_MouthLower, loc=<-1.8907 0.2989 1.9184>, rot=<0.0913 0.2322 -0.0318 -0.9679>}
 FACIAL_R_LipLowerOuter1 {parent=FACIAL_R_LipLowerOuter, loc=<-0.1099 0.008 0.4186>, rot=<0 0 0 -1>}
 FACIAL_R_LipLowerOuter2 {parent=FACIAL_R_LipLowerOuter, loc=<0.0021 -0.562 -0.1956>, rot=<0 0 0 -1>}
 FACIAL_C_TeethLower {parent=FACIAL_C_LowerLipRotation, loc=<0 -3.3107 6.3943>, rot=<-0.342 0 0 -0.9397>}
 FACIAL_C_Tongue1 {parent=FACIAL_C_TeethLower, loc=<0 0.8814 -4.9974>, rot=<-0.2588 0 0 -0.9659>}
 FACIAL_C_Tongue2 {parent=FACIAL_C_Tongue1, loc=<0 0.4172 1.4811>, rot=<0.2588 0 0 -0.9659>}
 FACIAL_C_TongueUpper2 {parent=FACIAL_C_Tongue2, loc=<-0.0004 -0.5559 -0.0593>, rot=<0 0 0 1>}
 FACIAL_L_TongueSide2 {parent=FACIAL_C_Tongue2, loc=<1.7649 0.1107 0.3121>, rot=<0 0 0 1>}
 FACIAL_R_TongueSide2 {parent=FACIAL_C_Tongue2, loc=<-1.7649 0.1107 0.3121>, rot=<0 0 0 -1>}
 FACIAL_C_Tongue3 {parent=FACIAL_C_Tongue2, loc=<0 0.2069 1.3354>, rot=<0.1305 0 0 -0.9914>}
 FACIAL_C_TongueUpper3 {parent=FACIAL_C_Tongue3, loc=<-0.0003 -0.3509 -0.1572>, rot=<0 0 0 -1>}
 FACIAL_C_TongueLower3 {parent=FACIAL_C_Tongue3, loc=<0 0.8092 0.0947>, rot=<0 0 0 -1>}
 FACIAL_L_TongueSide3 {parent=FACIAL_C_Tongue3, loc=<1.5325 0.0121 0.1487>, rot=<0 0 0 -1>}
 FACIAL_R_TongueSide3 {parent=FACIAL_C_Tongue3, loc=<-1.5325 0.0121 0.1487>, rot=<0 0 0 -1>}
 FACIAL_C_Tongue4 {parent=FACIAL_C_Tongue3, loc=<0.0006 -0.1867 1.1247>, rot=<0.0436 0 0 -0.999>}
 FACIAL_C_Jaw {parent=FACIAL_C_FacialRoot, loc=<0 1.9413 2.9092>, rot=<0.4017 0 0 -0.9158>}
 FACIAL_C_Jawline {parent=FACIAL_C_Jaw, loc=<0 0.9567 8.3398>, rot=<0.2588 0 0 -0.9659>}
 FACIAL_L_Jawline {parent=FACIAL_C_Jaw, loc=<4.3349 0.5774 4.2447>, rot=<-0.1329 0.167 0.0026 -0.977>}
 FACIAL_R_Jawline {parent=FACIAL_C_Jaw, loc=<-4.3349 0.5774 4.2447>, rot=<-0.1329 -0.167 -0.0026 -0.977>}
 FACIAL_L_ChinSide {parent=FACIAL_C_Jaw, loc=<2.9641 -0.948 6.7957>, rot=<-0.2207 -0.3739 0.05 -0.8994>}
 FACIAL_R_ChinSide {parent=FACIAL_C_Jaw, loc=<-2.9641 -0.948 6.7957>, rot=<-0.2207 0.3739 -0.05 -0.8994>}
 FACIAL_C_Chin {parent=FACIAL_C_Jaw, loc=<0 -1.7958 8.7593>, rot=<-0.4226 0 0 -0.9063>}
 FACIAL_C_Chin3 {parent=FACIAL_C_Jaw, loc=<0 -0.3859 9.3079>, rot=<0.0436 0 0 -0.999>}
 clavicle_l {parent=spine_05, loc=<4.3575 -0.1466 -2.3223>, rot=<0.0465 -0.7289 -0.009 -0.6829>}
 upperarm_l {parent=clavicle_l, loc=<15.5111 0 0>, rot=<0.0226 -0.3234 -0.0088 -0.9459>}
 clavicle_r {parent=spine_05, loc=<4.3575 -0.1466 2.3223>, rot=<0.7289 0.0465 -0.6829 0.009>}
 upperarm_r {parent=clavicle_r, loc=<-15.5111 0 0>, rot=<0.0226 -0.3234 -0.0088 -0.9459>}
 CTRL_expressions_browDINL {type=Curve, loc=<0 0 0>, rot=<0 0 0 1>}
 CTRL_expressions_browDINR {type=Curve, loc=<0 0 0>, rot=<0 0 0 1>}
 CTRL_expressions_browDOUTL {type=Curve, loc=<0 0 0>, rot=<0 0 0 1>}
 CTRL_expressions_browDOUTR {type=Curve, loc=<0 0 0>, rot=<0 0 0 1>}
 CTRL_expressions_browDownL {type=Curve, loc=<0 0 0>, rot=<0 0 0 1>}
 CTRL_expressions_browDownR {type=Curve, loc=<0 0 0>, rot=<0 0 0 1>}
 CTRL_expressions_browInnerInL {type=Curve, loc=<0 0 0>, rot=<0 0 0 1>}
 CTRL_expressions_browInnerInR {type=Curve, loc=<0 0 0>, rot=<0 0 0 1>}
 CTRL_expressions_browInnerMidDownL {type=Curve, loc=<0 0 0>, rot=<0 0 0 1>}
 CTRL_expressions_browInnerMidDownR {type=Curve, loc=<0 0 0>, rot=<0 0 0 1>}
 CTRL_expressions_browInnerMidInL {type=Curve, loc=<0 0 0>, rot=<0 0 0 1>}
 CTRL_expressions_browInnerMidInR {type=Curve, loc=<0 0 0>, rot=<0 0 0 1>}
 CTRL_expressions_browInnerMidOutL {type=Curve, loc=<0 0 0>, rot=<0 0 0 1>}
 CTRL_expressions_browInnerMidOutR {type=Curve, loc=<0 0 0>, rot=<0 0 0 1>}
 CTRL_expressions_browInnerMidUpL {type=Curve, loc=<0 0 0>, rot=<0 0 0 1>}
 CTRL_expressions_browInnerMidUpR {type=Curve, loc=<0 0 0>, rot=<0 0 0 1>}
 CTRL_expressions_browInnerOutL {type=Curve, loc=<0 0 0>, rot=<0 0 0 1>}
 CTRL_expressions_browInnerOutR {type=Curve, loc=<0 0 0>, rot=<0 0 0 1>}
 CTRL_expressions_browLateralL {type=Curve, loc=<0 0 0>, rot=<0 0 0 1>}
 CTRL_expressions_browLateralR {type=Curve, loc=<0 0 0>, rot=<0 0 0 1>}
 CTRL_expressions_browOuterMidDownL {type=Curve, loc=<0 0 0>, rot=<0 0 0 1>}
 CTRL_expressions_browOuterMidDownR {type=Curve, loc=<0 0 0>, rot=<0 0 0 1>}
 CTRL_expressions_browOuterMidInL {type=Curve, loc=<0 0 0>, rot=<0 0 0 1>}
 CTRL_expressions_browOuterMidInR {type=Curve, loc=<0 0 0>, rot=<0 0 0 1>}
 CTRL_expressions_browOuterMidOutL {type=Curve, loc=<0 0 0>, rot=<0 0 0 1>}
 CTRL_expressions_browOuterMidOutR {type=Curve, loc=<0 0 0>, rot=<0 0 0 1>}
 CTRL_expressions_browOuterMidUpL {type=Curve, loc=<0 0 0>, rot=<0 0 0 1>}
 CTRL_expressions_browOuterMidUpR {type=Curve, loc=<0 0 0>, rot=<0 0 0 1>}
 CTRL_expressions_browRaiseInL {type=Curve, loc=<0 0 0>, rot=<0 0 0 1>}
 CTRL_expressions_browRaiseInR {type=Curve, loc=<0 0 0>, rot=<0 0 0 1>}
 CTRL_expressions_browRaiseOuterL {type=Curve, loc=<0 0 0>, rot=<0 0 0 1>}
 CTRL_expressions_browRaiseOuterR {type=Curve, loc=<0 0 0>, rot=<0 0 0 1>}
 CTRL_expressions_browRollNegInL {type=Curve, loc=<0 0 0>, rot=<0 0 0 1>}
 CTRL_expressions_browRollNegInR {type=Curve, loc=<0 0 0>, rot=<0 0 0 1>}
 CTRL_expressions_browRollNegOutL {type=Curve, loc=<0 0 0>, rot=<0 0 0 1>}
 CTRL_expressions_browRollNegOutR {type=Curve, loc=<0 0 0>, rot=<0 0 0 1>}
 CTRL_expressions_browRollPosInL {type=Curve, loc=<0 0 0>, rot=<0 0 0 1>}
 CTRL_expressions_browRollPosInR {type=Curve, loc=<0 0 0>, rot=<0 0 0 1>}
 CTRL_expressions_browRollPosOutL {type=Curve, loc=<0 0 0>, rot=<0 0 0 1>}
 CTRL_expressions_browRollPosOutR {type=Curve, loc=<0 0 0>, rot=<0 0 0 1>}
 CTRL_expressions_browUINL {type=Curve, loc=<0 0 0>, rot=<0 0 0 1>}
 CTRL_expressions_browUINR {type=Curve, loc=<0 0 0>, rot=<0 0 0 1>}
 CTRL_expressions_browUOUTL {type=Curve, loc=<0 0 0>, rot=<0 0 0 1>}
 CTRL_expressions_browUOUTR {type=Curve, loc=<0 0 0>, rot=<0 0 0 1>}
 CTRL_expressions_eyeBlinkL {type=Curve, loc=<0 0 0>, rot=<0 0 0 1>}
 CTRL_expressions_eyeBlinkR {type=Curve, loc=<0 0 0>, rot=<0 0 0 1>}
 CTRL_expressions_eyeCheekRaiseL {type=Curve, loc=<0 0 0>, rot=<0 0 0 1>}
 CTRL_expressions_eyeCheekRaiseR {type=Curve, loc=<0 0 0>, rot=<0 0 0 1>}
 CTRL_expressions_eyeEyelidsReduce {type=Curve, loc=<0 0 0>, rot=<0 0 0 1>}
 CTRL_expressions_eyeFaceScrunchL {type=Curve, loc=<0 0 0>, rot=<0 0 0 1>}
 CTRL_expressions_eyeFaceScrunchR {type=Curve, loc=<0 0 0>, rot=<0 0 0 1>}
 CTRL_expressions_eyelidRollNegDL {type=Curve, loc=<0 0 0>, rot=<0 0 0 1>}
 CTRL_expressions_eyelidRollNegDR {type=Curve, loc=<0 0 0>, rot=<0 0 0 1>}
 CTRL_expressions_eyelidRollNegUL {type=Curve, loc=<0 0 0>, rot=<0 0 0 1>}
 CTRL_expressions_eyelidRollNegUR {type=Curve, loc=<0 0 0>, rot=<0 0 0 1>}
 CTRL_expressions_eyelidRollPosDL {type=Curve, loc=<0 0 0>, rot=<0 0 0 1>}
 CTRL_expressions_eyelidRollPosDR {type=Curve, loc=<0 0 0>, rot=<0 0 0 1>}
 CTRL_expressions_eyelidRollPosUL {type=Curve, loc=<0 0 0>, rot=<0 0 0 1>}
 CTRL_expressions_eyelidRollPosUR {type=Curve, loc=<0 0 0>, rot=<0 0 0 1>}
 CTRL_expressions_eyeLookDownL {type=Curve, loc=<0 0 0>, rot=<0 0 0 1>}
 CTRL_expressions_eyeLookDownR {type=Curve, loc=<0 0 0>, rot=<0 0 0 1>}
 CTRL_expressions_eyeLookLeftL {type=Curve, loc=<0 0 0>, rot=<0 0 0 1>}
 CTRL_expressions_eyeLookLeftR {type=Curve, loc=<0 0 0>, rot=<0 0 0 1>}
 CTRL_expressions_eyeLookRightL {type=Curve, loc=<0 0 0>, rot=<0 0 0 1>}
 CTRL_expressions_eyeLookRightR {type=Curve, loc=<0 0 0>, rot=<0 0 0 1>}
 CTRL_expressions_eyeLookUpL {type=Curve, loc=<0 0 0>, rot=<0 0 0 1>}
 CTRL_expressions_eyeLookUpR {type=Curve, loc=<0 0 0>, rot=<0 0 0 1>}
 CTRL_expressions_eyeLowerLidDownL {type=Curve, loc=<0 0 0>, rot=<0 0 0 1>}
 CTRL_expressions_eyeLowerLidDownR {type=Curve, loc=<0 0 0>, rot=<0 0 0 1>}
 CTRL_expressions_eyeLowerLidUpL {type=Curve, loc=<0 0 0>, rot=<0 0 0 1>}
 CTRL_expressions_eyeLowerLidUpR {type=Curve, loc=<0 0 0>, rot=<0 0 0 1>}
 CTRL_expressions_eyePupilNarrowL {type=Curve, loc=<0 0 0>, rot=<0 0 0 1>}
 CTRL_expressions_eyePupilNarrowR {type=Curve, loc=<0 0 0>, rot=<0 0 0 1>}
 CTRL_expressions_eyePupilWideL {type=Curve, loc=<0 0 0>, rot=<0 0 0 1>}
 CTRL_expressions_eyePupilWideR {type=Curve, loc=<0 0 0>, rot=<0 0 0 1>}
 CTRL_expressions_eyeRelaxL {type=Curve, loc=<0 0 0>, rot=<0 0 0 1>}
 CTRL_expressions_eyeRelaxR {type=Curve, loc=<0 0 0>, rot=<0 0 0 1>}
 CTRL_expressions_eyeSquintInnerL {type=Curve, loc=<0 0 0>, rot=<0 0 0 1>}
 CTRL_expressions_eyeSquintInnerR {type=Curve, loc=<0 0 0>, rot=<0 0 0 1>}
 CTRL_expressions_eyeUpperLidUpL {type=Curve, loc=<0 0 0>, rot=<0 0 0 1>}
 CTRL_expressions_eyeUpperLidUpR {type=Curve, loc=<0 0 0>, rot=<0 0 0 1>}
 CTRL_expressions_eyeWidenL {type=Curve, loc=<0 0 0>, rot=<0 0 0 1>}
 CTRL_expressions_eyeWidenR {type=Curve, loc=<0 0 0>, rot=<0 0 0 1>}
 CTRL_expressions_jawBack {type=Curve, loc=<0 0 0>, rot=<0 0 0 1>}
 CTRL_expressions_jawChinRaiseDL {type=Curve, loc=<0 0 0>, rot=<0 0 0 1>}
 CTRL_expressions_jawChinRaiseDR {type=Curve, loc=<0 0 0>, rot=<0 0 0 1>}
 CTRL_expressions_jawChinRaiseUL {type=Curve, loc=<0 0 0>, rot=<0 0 0 1>}
 CTRL_expressions_jawChinRaiseUR {type=Curve, loc=<0 0 0>, rot=<0 0 0 1>}
 CTRL_expressions_jawFwd {type=Curve, loc=<0 0 0>, rot=<0 0 0 1>}
 CTRL_expressions_jawLeft {type=Curve, loc=<0 0 0>, rot=<0 0 0 1>}
 CTRL_expressions_jawOpen {type=Curve, loc=<0 0 0>, rot=<0 0 0 1>}
 CTRL_expressions_jawOpenExtreme {type=Curve, loc=<0 0 0>, rot=<0 0 0 1>}
 CTRL_expressions_jawRight {type=Curve, loc=<0 0 0>, rot=<0 0 0 1>}
 CTRL_expressions_mouthCheekBlowL {type=Curve, loc=<0 0 0>, rot=<0 0 0 1>}
 CTRL_expressions_mouthCheekBlowR {type=Curve, loc=<0 0 0>, rot=<0 0 0 1>}
 CTRL_expressions_mouthCheekSuckL {type=Curve, loc=<0 0 0>, rot=<0 0 0 1>}
 CTRL_expressions_mouthCheekSuckR {type=Curve, loc=<0 0 0>, rot=<0 0 0 1>}
 CTRL_expressions_mouthCornerDepressL {type=Curve, loc=<0 0 0>, rot=<0 0 0 1>}
 CTRL_expressions_mouthCornerDepressR {type=Curve, loc=<0 0 0>, rot=<0 0 0 1>}
 CTRL_expressions_mouthCornerDownL {type=Curve, loc=<0 0 0>, rot=<0 0 0 1>}
 CTRL_expressions_mouthCornerDownR {type=Curve, loc=<0 0 0>, rot=<0 0 0 1>}
 CTRL_expressions_mouthCornerNarrowL {type=Curve, loc=<0 0 0>, rot=<0 0 0 1>}
 CTRL_expressions_mouthCornerNarrowR {type=Curve, loc=<0 0 0>, rot=<0 0 0 1>}
 CTRL_expressions_mouthCornerPullL {type=Curve, loc=<0 0 0>, rot=<0 0 0 1>}
 CTRL_expressions_mouthCornerPullR {type=Curve, loc=<0 0 0>, rot=<0 0 0 1>}
 CTRL_expressions_mouthCornerRounderDL {type=Curve, loc=<0 0 0>, rot=<0 0 0 1>}
 CTRL_expressions_mouthCornerRounderDR {type=Curve, loc=<0 0 0>, rot=<0 0 0 1>}
 CTRL_expressions_mouthCornerRounderUL {type=Curve, loc=<0 0 0>, rot=<0 0 0 1>}
 CTRL_expressions_mouthCornerRounderUR {type=Curve, loc=<0 0 0>, rot=<0 0 0 1>}
 CTRL_expressions_mouthCornerSharpenDL {type=Curve, loc=<0 0 0>, rot=<0 0 0 1>}
 CTRL_expressions_mouthCornerSharpenDR {type=Curve, loc=<0 0 0>, rot=<0 0 0 1>}
 CTRL_expressions_mouthCornerSharpenUL {type=Curve, loc=<0 0 0>, rot=<0 0 0 1>}
 CTRL_expressions_mouthCornerSharpenUR {type=Curve, loc=<0 0 0>, rot=<0 0 0 1>}
 CTRL_expressions_mouthCornerUpL {type=Curve, loc=<0 0 0>, rot=<0 0 0 1>}
 CTRL_expressions_mouthCornerUpR {type=Curve, loc=<0 0 0>, rot=<0 0 0 1>}
 CTRL_expressions_mouthCornerWideL {type=Curve, loc=<0 0 0>, rot=<0 0 0 1>}
 CTRL_expressions_mouthCornerWideR {type=Curve, loc=<0 0 0>, rot=<0 0 0 1>}
 CTRL_expressions_mouthDimpleL {type=Curve, loc=<0 0 0>, rot=<0 0 0 1>}
 CTRL_expressions_mouthDimpleR {type=Curve, loc=<0 0 0>, rot=<0 0 0 1>}
 CTRL_expressions_mouthDown {type=Curve, loc=<0 0 0>, rot=<0 0 0 1>}
 CTRL_expressions_mouthFunnelDL {type=Curve, loc=<0 0 0>, rot=<0 0 0 1>}
 CTRL_expressions_mouthFunnelDR {type=Curve, loc=<0 0 0>, rot=<0 0 0 1>}
 CTRL_expressions_mouthFunnelUL {type=Curve, loc=<0 0 0>, rot=<0 0 0 1>}
 CTRL_expressions_mouthFunnelUR {type=Curve, loc=<0 0 0>, rot=<0 0 0 1>}
 CTRL_expressions_mouthLeft {type=Curve, loc=<0 0 0>, rot=<0 0 0 1>}
 CTRL_expressions_mouthLipsBlowL {type=Curve, loc=<0 0 0>, rot=<0 0 0 1>}
 CTRL_expressions_mouthLipsBlowR {type=Curve, loc=<0 0 0>, rot=<0 0 0 1>}
 CTRL_expressions_mouthLipsPressL {type=Curve, loc=<0 0 0>, rot=<0 0 0 1>}
 CTRL_expressions_mouthLipsPressR {type=Curve, loc=<0 0 0>, rot=<0 0 0 1>}
 CTRL_expressions_mouthLipsPullDL {type=Curve, loc=<0 0 0>, rot=<0 0 0 1>}
 CTRL_expressions_mouthLipsPullDR {type=Curve, loc=<0 0 0>, rot=<0 0 0 1>}
 CTRL_expressions_mouthLipsPullUL {type=Curve, loc=<0 0 0>, rot=<0 0 0 1>}
 CTRL_expressions_mouthLipsPullUR {type=Curve, loc=<0 0 0>, rot=<0 0 0 1>}
 CTRL_expressions_mouthLipsPurseDL {type=Curve, loc=<0 0 0>, rot=<0 0 0 1>}
 CTRL_expressions_mouthLipsPurseDR {type=Curve, loc=<0 0 0>, rot=<0 0 0 1>}
 CTRL_expressions_mouthLipsPurseUL {type=Curve, loc=<0 0 0>, rot=<0 0 0 1>}
 CTRL_expressions_mouthLipsPurseUR {type=Curve, loc=<0 0 0>, rot=<0 0 0 1>}
 CTRL_expressions_mouthLipsPushDL {type=Curve, loc=<0 0 0>, rot=<0 0 0 1>}
 CTRL_expressions_mouthLipsPushDR {type=Curve, loc=<0 0 0>, rot=<0 0 0 1>}
 CTRL_expressions_mouthLipsPushUL {type=Curve, loc=<0 0 0>, rot=<0 0 0 1>}
 CTRL_expressions_mouthLipsPushUR {type=Curve, loc=<0 0 0>, rot=<0 0 0 1>}
 CTRL_expressions_mouthLipsThickDL {type=Curve, loc=<0 0 0>, rot=<0 0 0 1>}
 CTRL_expressions_mouthLipsThickDR {type=Curve, loc=<0 0 0>, rot=<0 0 0 1>}
 CTRL_expressions_mouthLipsThickInwardDL {type=Curve, loc=<0 0 0>, rot=<0 0 0 1>}
 CTRL_expressions_mouthLipsThickInwardDR {type=Curve, loc=<0 0 0>, rot=<0 0 0 1>}
 CTRL_expressions_mouthLipsThickInwardUL {type=Curve, loc=<0 0 0>, rot=<0 0 0 1>}
 CTRL_expressions_mouthLipsThickInwardUR {type=Curve, loc=<0 0 0>, rot=<0 0 0 1>}
 CTRL_expressions_mouthLipsThickUL {type=Curve, loc=<0 0 0>, rot=<0 0 0 1>}
 CTRL_expressions_mouthLipsThickUR {type=Curve, loc=<0 0 0>, rot=<0 0 0 1>}
 CTRL_expressions_mouthLipsThinDL {type=Curve, loc=<0 0 0>, rot=<0 0 0 1>}
 CTRL_expressions_mouthLipsThinDR {type=Curve, loc=<0 0 0>, rot=<0 0 0 1>}
 CTRL_expressions_mouthLipsThinInwardDL {type=Curve, loc=<0 0 0>, rot=<0 0 0 1>}
 CTRL_expressions_mouthLipsThinInwardDR {type=Curve, loc=<0 0 0>, rot=<0 0 0 1>}
 CTRL_expressions_mouthLipsThinInwardUL {type=Curve, loc=<0 0 0>, rot=<0 0 0 1>}
 CTRL_expressions_mouthLipsThinInwardUR {type=Curve, loc=<0 0 0>, rot=<0 0 0 1>}
 CTRL_expressions_mouthLipsThinUL {type=Curve, loc=<0 0 0>, rot=<0 0 0 1>}
 CTRL_expressions_mouthLipsThinUR {type=Curve, loc=<0 0 0>, rot=<0 0 0 1>}
 CTRL_expressions_mouthLipsTightenDL {type=Curve, loc=<0 0 0>, rot=<0 0 0 1>}
 CTRL_expressions_mouthLipsTightenDR {type=Curve, loc=<0 0 0>, rot=<0 0 0 1>}
 CTRL_expressions_mouthLipsTightenUL {type=Curve, loc=<0 0 0>, rot=<0 0 0 1>}
 CTRL_expressions_mouthLipsTightenUR {type=Curve, loc=<0 0 0>, rot=<0 0 0 1>}
 CTRL_expressions_mouthLipsTogetherDL {type=Curve, loc=<0 0 0>, rot=<0 0 0 1>}
 CTRL_expressions_mouthLipsTogetherDR {type=Curve, loc=<0 0 0>, rot=<0 0 0 1>}
 CTRL_expressions_mouthLipsTogetherUL {type=Curve, loc=<0 0 0>, rot=<0 0 0 1>}
 CTRL_expressions_mouthLipsTogetherUR {type=Curve, loc=<0 0 0>, rot=<0 0 0 1>}
 CTRL_expressions_mouthLipsTowardsDL {type=Curve, loc=<0 0 0>, rot=<0 0 0 1>}
 CTRL_expressions_mouthLipsTowardsDR {type=Curve, loc=<0 0 0>, rot=<0 0 0 1>}
 CTRL_expressions_mouthLipsTowardsUL {type=Curve, loc=<0 0 0>, rot=<0 0 0 1>}
 CTRL_expressions_mouthLipsTowardsUR {type=Curve, loc=<0 0 0>, rot=<0 0 0 1>}
 CTRL_expressions_mouthLowerLipBiteL {type=Curve, loc=<0 0 0>, rot=<0 0 0 1>}
 CTRL_expressions_mouthLowerLipBiteR {type=Curve, loc=<0 0 0>, rot=<0 0 0 1>}
 CTRL_expressions_mouthLowerLipDepressL {type=Curve, loc=<0 0 0>, rot=<0 0 0 1>}
 CTRL_expressions_mouthLowerLipDepressR {type=Curve, loc=<0 0 0>, rot=<0 0 0 1>}
 CTRL_expressions_mouthLowerLipRollInL {type=Curve, loc=<0 0 0>, rot=<0 0 0 1>}
 CTRL_expressions_mouthLowerLipRollInR {type=Curve, loc=<0 0 0>, rot=<0 0 0 1>}
 CTRL_expressions_mouthLowerLipRollOutL {type=Curve, loc=<0 0 0>, rot=<0 0 0 1>}
 CTRL_expressions_mouthLowerLipRollOutR {type=Curve, loc=<0 0 0>, rot=<0 0 0 1>}
 CTRL_expressions_mouthLowerLipTowardsTeethL {type=Curve, loc=<0 0 0>, rot=<0 0 0 1>}
 CTRL_expressions_mouthLowerLipTowardsTeethR {type=Curve, loc=<0 0 0>, rot=<0 0 0 1>}
 CTRL_expressions_mouthPressDL {type=Curve, loc=<0 0 0>, rot=<0 0 0 1>}
 CTRL_expressions_mouthPressDR {type=Curve, loc=<0 0 0>, rot=<0 0 0 1>}
 CTRL_expressions_mouthPressUL {type=Curve, loc=<0 0 0>, rot=<0 0 0 1>}
 CTRL_expressions_mouthPressUR {type=Curve, loc=<0 0 0>, rot=<0 0 0 1>}
 CTRL_expressions_mouthRight {type=Curve, loc=<0 0 0>, rot=<0 0 0 1>}
 CTRL_expressions_mouthSharpCornerPullL {type=Curve, loc=<0 0 0>, rot=<0 0 0 1>}
 CTRL_expressions_mouthSharpCornerPullR {type=Curve, loc=<0 0 0>, rot=<0 0 0 1>}
 CTRL_expressions_mouthStretchL {type=Curve, loc=<0 0 0>, rot=<0 0 0 1>}
 CTRL_expressions_mouthStretchR {type=Curve, loc=<0 0 0>, rot=<0 0 0 1>}
 CTRL_expressions_mouthUp {type=Curve, loc=<0 0 0>, rot=<0 0 0 1>}
 CTRL_expressions_mouthUpperLipBiteL {type=Curve, loc=<0 0 0>, rot=<0 0 0 1>}
 CTRL_expressions_mouthUpperLipBiteR {type=Curve, loc=<0 0 0>, rot=<0 0 0 1>}
 CTRL_expressions_mouthUpperLipRaiseL {type=Curve, loc=<0 0 0>, rot=<0 0 0 1>}
 CTRL_expressions_mouthUpperLipRaiseR {type=Curve, loc=<0 0 0>, rot=<0 0 0 1>}
 CTRL_expressions_mouthUpperLipRollInL {type=Curve, loc=<0 0 0>, rot=<0 0 0 1>}
 CTRL_expressions_mouthUpperLipRollInR {type=Curve, loc=<0 0 0>, rot=<0 0 0 1>}
 CTRL_expressions_mouthUpperLipRollOutL {type=Curve, loc=<0 0 0>, rot=<0 0 0 1>}
 CTRL_expressions_mouthUpperLipRollOutR {type=Curve, loc=<0 0 0>, rot=<0 0 0 1>}
 CTRL_expressions_mouthUpperLipTowardsTeethL {type=Curve, loc=<0 0 0>, rot=<0 0 0 1>}
 CTRL_expressions_mouthUpperLipTowardsTeethR {type=Curve, loc=<0 0 0>, rot=<0 0 0 1>}
 CTRL_expressions_noseNasolabialDeepenL {type=Curve, loc=<0 0 0>, rot=<0 0 0 1>}
 CTRL_expressions_noseNasolabialDeepenR {type=Curve, loc=<0 0 0>, rot=<0 0 0 1>}
 CTRL_expressions_noseNostrilCompressL {type=Curve, loc=<0 0 0>, rot=<0 0 0 1>}
 CTRL_expressions_noseNostrilCompressR {type=Curve, loc=<0 0 0>, rot=<0 0 0 1>}
 CTRL_expressions_noseNostrilDepressL {type=Curve, loc=<0 0 0>, rot=<0 0 0 1>}
 CTRL_expressions_noseNostrilDepressR {type=Curve, loc=<0 0 0>, rot=<0 0 0 1>}
 CTRL_expressions_noseNostrilDilateL {type=Curve, loc=<0 0 0>, rot=<0 0 0 1>}
 CTRL_expressions_noseNostrilDilateR {type=Curve, loc=<0 0 0>, rot=<0 0 0 1>}
 CTRL_expressions_noseRotateL {type=Curve, loc=<0 0 0>, rot=<0 0 0 1>}
 CTRL_expressions_noseRotateR {type=Curve, loc=<0 0 0>, rot=<0 0 0 1>}
 CTRL_expressions_noseSquash {type=Curve, loc=<0 0 0>, rot=<0 0 0 1>}
 CTRL_expressions_noseStretch {type=Curve, loc=<0 0 0>, rot=<0 0 0 1>}
 CTRL_expressions_noseWrinkleL {type=Curve, loc=<0 0 0>, rot=<0 0 0 1>}
 CTRL_expressions_noseWrinkleR {type=Curve, loc=<0 0 0>, rot=<0 0 0 1>}
 CTRL_expressions_noseWrinkleUpperL {type=Curve, loc=<0 0 0>, rot=<0 0 0 1>}
 CTRL_expressions_noseWrinkleUpperR {type=Curve, loc=<0 0 0>, rot=<0 0 0 1>}
 CTRL_expressions_teethBackD {type=Curve, loc=<0 0 0>, rot=<0 0 0 1>}
 CTRL_expressions_teethBackU {type=Curve, loc=<0 0 0>, rot=<0 0 0 1>}
 CTRL_expressions_teethDownD {type=Curve, loc=<0 0 0>, rot=<0 0 0 1>}
 CTRL_expressions_teethDownU {type=Curve, loc=<0 0 0>, rot=<0 0 0 1>}
 CTRL_expressions_teethFwdD {type=Curve, loc=<0 0 0>, rot=<0 0 0 1>}
 CTRL_expressions_teethFwdU {type=Curve, loc=<0 0 0>, rot=<0 0 0 1>}
 CTRL_expressions_teethLeftD {type=Curve, loc=<0 0 0>, rot=<0 0 0 1>}
 CTRL_expressions_teethLeftU {type=Curve, loc=<0 0 0>, rot=<0 0 0 1>}
 CTRL_expressions_teethRightD {type=Curve, loc=<0 0 0>, rot=<0 0 0 1>}
 CTRL_expressions_teethRightU {type=Curve, loc=<0 0 0>, rot=<0 0 0 1>}
 CTRL_expressions_teethRotateBackD {type=Curve, loc=<0 0 0>, rot=<0 0 0 1>}
 CTRL_expressions_teethRotateBackU {type=Curve, loc=<0 0 0>, rot=<0 0 0 1>}
 CTRL_expressions_teethRotateFwdD {type=Curve, loc=<0 0 0>, rot=<0 0 0 1>}
 CTRL_expressions_teethRotateFwdU {type=Curve, loc=<0 0 0>, rot=<0 0 0 1>}
 CTRL_expressions_teethUpD {type=Curve, loc=<0 0 0>, rot=<0 0 0 1>}
 CTRL_expressions_teethUpU {type=Curve, loc=<0 0 0>, rot=<0 0 0 1>}
 CTRL_expressions_tongueBendDown {type=Curve, loc=<0 0 0>, rot=<0 0 0 1>}
 CTRL_expressions_tongueBendUp {type=Curve, loc=<0 0 0>, rot=<0 0 0 1>}
 CTRL_expressions_tongueDown {type=Curve, loc=<0 0 0>, rot=<0 0 0 1>}
 CTRL_expressions_tongueIn {type=Curve, loc=<0 0 0>, rot=<0 0 0 1>}
 CTRL_expressions_tongueLeft {type=Curve, loc=<0 0 0>, rot=<0 0 0 1>}
 CTRL_expressions_tongueNarrow {type=Curve, loc=<0 0 0>, rot=<0 0 0 1>}
 CTRL_expressions_tongueOut {type=Curve, loc=<0 0 0>, rot=<0 0 0 1>}
 CTRL_expressions_tongueRight {type=Curve, loc=<0 0 0>, rot=<0 0 0 1>}
 CTRL_expressions_tongueRollDown {type=Curve, loc=<0 0 0>, rot=<0 0 0 1>}
 CTRL_expressions_tongueRollLeft {type=Curve, loc=<0 0 0>, rot=<0 0 0 1>}
 CTRL_expressions_tongueRollRight {type=Curve, loc=<0 0 0>, rot=<0 0 0 1>}
 CTRL_expressions_tongueRollUp {type=Curve, loc=<0 0 0>, rot=<0 0 0 1>}
 CTRL_expressions_tongueTwistLeft {type=Curve, loc=<0 0 0>, rot=<0 0 0 1>}
 CTRL_expressions_tongueTwistRight {type=Curve, loc=<0 0 0>, rot=<0 0 0 1>}
 CTRL_expressions_tongueUp {type=Curve, loc=<0 0 0>, rot=<0 0 0 1>}
 CTRL_expressions_tongueWide {type=Curve, loc=<0 0 0>, rot=<0 0 0 1>}
 DisableFaceOverride {type=Curve, loc=<0 0 0>, rot=<0 0 0 1>}
 is_3L {type=Curve, loc=<0 0 0>, rot=<0 0 0 1>}
 is_face_bone_anim {type=Curve, loc=<0 0 0>, rot=<0 0 0 1>}
 WM_BrowLateral_L {type=Curve, loc=<0 0 0>, rot=<0 0 0 1>}
 WM_BrowLateral_Offset_L {type=Curve, loc=<0 0 0>, rot=<0 0 0 1>}
 WM_BrowLateral_Offset_R {type=Curve, loc=<0 0 0>, rot=<0 0 0 1>}
 WM_BrowLateral_R {type=Curve, loc=<0 0 0>, rot=<0 0 0 1>}
 WM_BrowRaiseInner_L {type=Curve, loc=<0 0 0>, rot=<0 0 0 1>}
 WM_BrowRaiseInner_Offset_L {type=Curve, loc=<0 0 0>, rot=<0 0 0 1>}
 WM_BrowRaiseInner_Offset_R {type=Curve, loc=<0 0 0>, rot=<0 0 0 1>}
 WM_BrowRaiseInner_R {type=Curve, loc=<0 0 0>, rot=<0 0 0 1>}
 WM_BrowRaiseOutter_L {type=Curve, loc=<0 0 0>, rot=<0 0 0 1>}
 WM_BrowRaiseOutter_Offset_L {type=Curve, loc=<0 0 0>, rot=<0 0 0 1>}
 WM_BrowRaiseOutter_Offset_R {type=Curve, loc=<0 0 0>, rot=<0 0 0 1>}
 WM_BrowRaiseOutter_R {type=Curve, loc=<0 0 0>, rot=<0 0 0 1>}
 WM_CheekRaise_L {type=Curve, loc=<0 0 0>, rot=<0 0 0 1>}
 WM_CheekRaise_Offset_L {type=Curve, loc=<0 0 0>, rot=<0 0 0 1>}
 WM_CheekRaise_Offset_R {type=Curve, loc=<0 0 0>, rot=<0 0 0 1>}
 WM_CheekRaise_R {type=Curve, loc=<0 0 0>, rot=<0 0 0 1>}
 WM_Lips_L {type=Curve, loc=<0 0 0>, rot=<0 0 0 1>}
 WM_Lips_Offset_L {type=Curve, loc=<0 0 0>, rot=<0 0 0 1>}
 WM_Lips_Offset_R {type=Curve, loc=<0 0 0>, rot=<0 0 0 1>}
 WM_Lips_R {type=Curve, loc=<0 0 0>, rot=<0 0 0 1>}
 WM_MouthStretch_L {type=Curve, loc=<0 0 0>, rot=<0 0 0 1>}
 WM_MouthStretch_Offset_L {type=Curve, loc=<0 0 0>, rot=<0 0 0 1>}
 WM_MouthStretch_Offset_R {type=Curve, loc=<0 0 0>, rot=<0 0 0 1>}
 WM_MouthStretch_R {type=Curve, loc=<0 0 0>, rot=<0 0 0 1>}
 WM_NoseWrinkle_L {type=Curve, loc=<0 0 0>, rot=<0 0 0 1>}
 WM_NoseWrinkle_Offset_L {type=Curve, loc=<0 0 0>, rot=<0 0 0 1>}
 WM_NoseWrinkle_Offset_R {type=Curve, loc=<0 0 0>, rot=<0 0 0 1>}
 WM_NoseWrinkle_R {type=Curve, loc=<0 0 0>, rot=<0 0 0 1>}
 WM_Smile_Offset_L {type=Curve, loc=<0 0 0>, rot=<0 0 0 1>}
 WM_Smile_Offset_R {type=Curve, loc=<0 0 0>, rot=<0 0 0 1>}
 WM_Smile_R {type=Curve, loc=<0 0 0>, rot=<0 0 0 1>}
 WM_Smile_L {type=Curve, loc=<0 0 0>, rot=<0 0 0 1>}
]
