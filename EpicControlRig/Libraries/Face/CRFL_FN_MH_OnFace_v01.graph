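# Control Rig function library for a MetaHuman face rig, with a function that procedurally spawns the facial animation control hierarchy.
# It builds nested nulls down to Face_Ctrl_Master, then branches through sequences into brow, eye, jaw, mouth, tongue and teeth controls.
# Local variables pass parent items along, and bounded float channels add secondary shapes.
# CTRL_WrinkleMap gets channels dynamically from curves containing "_Offset_" and starting with "WM_".
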ControlRigBlueprint /EpicControlRig/Libraries/Face/CRFL_FN_MH_OnFace_v01
graphs [
 New Function_ContainedGraph {nodes [
   RigVMFunctionEntryNode()
   > RigVMFunction_Sequence_9 = RigVMAggregateNode()
   n1 = &RigUnit_HierarchyAddNull(
    Name=FACE_CONTROL_GRP
    Parent=(Type=Bone,Name="None")
    Space=LocalSpace
    Transform=(Rotation=(X=0,Y=0,Z=0,W=1),Translation=(X=0,Y=0,Z=0),Scale3D=(X=1,Y=1,Z=1))
    ExecutePin=@RigVMFunction_Sequence_9.A
   )
   > n2 = &RigUnit_HierarchyAddNull(
    Name=Face_Ctrl_Master_Follow
    Space=LocalSpace
    Transform=(Rotation=(X=-0.0029,Y=0.7071,Z=0.0029,W=-0.7071),Translation=(X=0,Y=-1.6873,Z=152.0342),Scale3D=(X=1,Y=1,Z=1))
    Parent=@n1.Item
   )
   > n3 = &RigUnit_HierarchyAddNull(
    Name=Face_Ctrl_Master_GRP
    Space=LocalSpace
    Transform=(Rotation=(X=0,Y=0.7071,Z=0,W=0.7071),Translation=(X=-8,Y=15,Z=0),Scale3D=(X=1,Y=1,Z=1))
    Parent=@n2.Item
   )
   > n4 = /Script/ControlRig.RigUnit_HierarchyAddControlTransform(
    InitialValue=(Rotation=(X=0,Y=0,Z=0,W=1),Translation=(X=0,Y=0,Z=0),Scale3D=(X=1,Y=1,Z=1))
    Name=Face_Ctrl_Master
    OffsetSpace=LocalSpace
    OffsetTransform=(Rotation=(X=0,Y=0,Z=0,W=1),Translation=(X=0,Y=0,Z=0),Scale3D=(X=0.75,Y=0.75,Z=0.75))
    Settings="(InitialSpace=LocalSpace,bUsePreferredRotationOrder=False,PreferredRotationOrder=YZX,Limits=(LimitTranslationX=(bMinimum=False,bMaximum=False),LimitTranslationY=(bMinimum=False,bMaximum=False),LimitTranslationZ=(bMinimum=False,bMaximum=False),LimitPitch=(b ...(truncated)"
    Parent=@n3.Item
   ) {
    is=struct RigVMUnitNode
    template="SpawnControl::Execute(in InitialValue,in Settings,in OffsetTransform,in OffsetSpace,in Parent,in Name,out Item)"
   }
   > RigVMVariableNode(Value=@n4.Item) {variable=Local master_ctrl}
   > /Script/ControlRig.RigUnit_HierarchyAddAnimationChannelFloat(
    InitialValue=1.0
    LimitsEnabled=(Enabled=(bMinimum=True,bMaximum=True))
    MaximumValue=1.0
    MinimumValue=0.0
    Name=gui_follow_head
    Parent=@n4.Item
   ) {
    is=struct RigVMUnitNode
    template="SpawnAnimationChannel::Execute(in InitialValue,in MinimumValue,in MaximumValue,in LimitsEnabled,in Parent,in Name,out Item)"
   }
   > &RigUnit_HierarchyAddAnimationChannelFloat(
    InitialValue=0.0
    LimitsEnabled=(Enabled=(bMinimum=True,bMaximum=True))
    MaximumValue=1.0
    MinimumValue=0.0
    Name=lookAtSwitch
    Parent=@n4.Item
   ) {template="SpawnAnimationChannel::Execute(in InitialValue,in MinimumValue,in MaximumValue,in LimitsEnabled,in Parent,in Name,out Item)"}
   > RigVMFunction_Sequence_8 = RigVMAggregateNode()
   n5 = RigVMVariableNode() {variable=Local master_ctrl}
   /Script/ControlRig.RigUnit_HierarchyAddControlFloat(
    InitialValue=0.0
    Name=CTRL_L_mouth_suckBlow
    OffsetSpace=LocalSpace
    OffsetTransform=(Rotation=(X=0.5154,Y=0.4842,Z=-0.5154,W=0.4842),Translation=(X=5,Y=0,Z=12),Scale3D=(X=1,Y=1,Z=1))
    Settings="(PrimaryAxis=Y,bIsScale=False,Limits=(Limit=(bMinimum=True,bMaximum=True),MinValue=-1,MaxValue=1,bDrawLimits=False),Shape=(bVisible=True,Name=\"DefaultGizmoLibraryNormalized.Triangle_Solid\",Color=(R=1,G=1,B=0,A=1),Transform=(Rotation=(X=-1,Y=0,Z=0,W=0),Tran ...(truncated)"
    ExecutePin=@RigVMFunction_Sequence_8.C
    Parent=@n5.Value
   ) {
    is=struct RigVMUnitNode
    template="SpawnControl::Execute(in InitialValue,in Settings,in OffsetTransform,in OffsetSpace,in Parent,in Name,out Item)"
   }
   > &RigUnit_HierarchyAddControlFloat(
    InitialValue=0.0
    Name=CTRL_R_mouth_suckBlow
    OffsetSpace=LocalSpace
    OffsetTransform=(Rotation=(X=0.5154,Y=-0.4842,Z=-0.5154,W=-0.4842),Translation=(X=-5,Y=0,Z=12),Scale3D=(X=1,Y=1,Z=1))
    Settings="(PrimaryAxis=Y,bIsScale=False,Limits=(Limit=(bMinimum=True,bMaximum=True),MinValue=-1,MaxValue=1,bDrawLimits=False),Shape=(bVisible=True,Name=\"DefaultGizmoLibraryNormalized.Triangle_Solid\",Color=(R=1,G=1,B=0,A=1),Transform=(Rotation=(X=-1,Y=0,Z=0,W=0),Tran ...(truncated)"
    Parent=@n5.Value
   ) {template="SpawnControl::Execute(in InitialValue,in Settings,in OffsetTransform,in OffsetSpace,in Parent,in Name,out Item)"}
   > RigUnit_HierarchyAddControlVector = /Script/ControlRig.RigUnit_HierarchyAddControlVector(
    InitialValue=(X=0,Y=0,Z=0)
    Name=CTRL_C_jaw
    OffsetSpace=LocalSpace
    OffsetTransform=(Rotation=(X=0,Y=-0.7071,Z=0.7071,W=0),Translation=(X=0,Y=0,Z=5.5),Scale3D=(X=1.5,Y=1.5,Z=1.5))
    Settings="(InitialSpace=LocalSpace,bIsPosition=True,Limits=(LimitX=(bMinimum=True,bMaximum=True),LimitY=(bMinimum=True,bMaximum=True),LimitZ=(bMinimum=True,bMaximum=True),MinValue=(X=-1,Y=-1,Z=-1),MaxValue=(X=1,Y=1,Z=1),bDrawLimits=False),Shape=(bVisible=True,Name=\" ...(truncated)"
    Parent=RigVMVariableNode() {variable=Local master_ctrl}.Value
   ) {
    is=struct RigVMUnitNode
    template="SpawnControl::Execute(in InitialValue,in Settings,in OffsetTransform,in OffsetSpace,in Parent,in Name,out Item)"
   }
   > &RigUnit_HierarchyAddAnimationChannelFloat(
    InitialValue=0.0
    LimitsEnabled=(Enabled=(bMinimum=True,bMaximum=True))
    MaximumValue=1.0
    MinimumValue=0.0
    Name=jawOpenExtreme
    Parent=@RigUnit_HierarchyAddControlVector.Item
   ) {template="SpawnAnimationChannel::Execute(in InitialValue,in MinimumValue,in MaximumValue,in LimitsEnabled,in Parent,in Name,out Item)"}
   > &RigUnit_HierarchyAddAnimationChannelFloat(
    InitialValue=0.0
    LimitsEnabled=(Enabled=(bMinimum=True,bMaximum=True))
    MaximumValue=1.0
    MinimumValue=0.0
    Name=chinRaiseUL
    Parent=@RigUnit_HierarchyAddControlVector.Item
   ) {template="SpawnAnimationChannel::Execute(in InitialValue,in MinimumValue,in MaximumValue,in LimitsEnabled,in Parent,in Name,out Item)"}
   > &RigUnit_HierarchyAddAnimationChannelFloat(
    InitialValue=0.0
    LimitsEnabled=(Enabled=(bMinimum=True,bMaximum=True))
    MaximumValue=1.0
    MinimumValue=0.0
    Name=chinRaiseUR
    Parent=@RigUnit_HierarchyAddControlVector.Item
   ) {template="SpawnAnimationChannel::Execute(in InitialValue,in MinimumValue,in MaximumValue,in LimitsEnabled,in Parent,in Name,out Item)"}
   > &RigUnit_HierarchyAddAnimationChannelFloat(
    InitialValue=0.0
    LimitsEnabled=(Enabled=(bMinimum=True,bMaximum=True))
    MaximumValue=1.0
    MinimumValue=0.0
    Name=chinRaiseDL
    Parent=@RigUnit_HierarchyAddControlVector.Item
   ) {template="SpawnAnimationChannel::Execute(in InitialValue,in MinimumValue,in MaximumValue,in LimitsEnabled,in Parent,in Name,out Item)"}
   > &RigUnit_HierarchyAddAnimationChannelFloat(
    InitialValue=0.0
    LimitsEnabled=(Enabled=(bMinimum=True,bMaximum=True))
    MaximumValue=1.0
    MinimumValue=0.0
    Name=chinRaiseDR
    Parent=@RigUnit_HierarchyAddControlVector.Item
   ) {template="SpawnAnimationChannel::Execute(in InitialValue,in MinimumValue,in MaximumValue,in LimitsEnabled,in Parent,in Name,out Item)"}
   > RigVMFunction_Sequence_7 = RigVMAggregateNode()
   n6 = RigVMVariableNode() {variable=Local master_ctrl}
   n7 = &RigUnit_HierarchyAddControlVector(
    InitialValue=(X=0,Y=0,Z=0)
    Name=CTRL_C_tongue
    OffsetSpace=LocalSpace
    OffsetTransform=(Rotation=(X=0.7071,Y=0,Z=0,W=0.7071),Translation=(X=8,Y=0,Z=10),Scale3D=(X=0.5,Y=0.5,Z=0.5))
    Settings="(InitialSpace=LocalSpace,bIsPosition=True,Limits=(LimitX=(bMinimum=True,bMaximum=True),LimitY=(bMinimum=True,bMaximum=True),LimitZ=(bMinimum=True,bMaximum=True),MinValue=(X=-1,Y=-1,Z=-1),MaxValue=(X=1,Y=1,Z=1),bDrawLimits=False),Shape=(bVisible=True,Name=\" ...(truncated)"
    ExecutePin=@RigVMFunction_Sequence_7.A
    Parent=@n6.Value
   ) {template="SpawnControl::Execute(in InitialValue,in Settings,in OffsetTransform,in OffsetSpace,in Parent,in Name,out Item)"}
   > &RigUnit_HierarchyAddAnimationChannelFloat(
    InitialValue=0.0
    LimitsEnabled=(Enabled=(bMinimum=True,bMaximum=True))
    MaximumValue=1.0
    MinimumValue=-1.0
    Name=wideNarrow
    Parent=@n7.Item
   ) {template="SpawnAnimationChannel::Execute(in InitialValue,in MinimumValue,in MaximumValue,in LimitsEnabled,in Parent,in Name,out Item)"}
   > &RigUnit_HierarchyAddAnimationChannelFloat(
    InitialValue=0.0
    LimitsEnabled=(Enabled=(bMinimum=True,bMaximum=True))
    MaximumValue=1.0
    MinimumValue=-1.0
    Name=rollUpDown
    Parent=@n7.Item
   ) {template="SpawnAnimationChannel::Execute(in InitialValue,in MinimumValue,in MaximumValue,in LimitsEnabled,in Parent,in Name,out Item)"}
   > &RigUnit_HierarchyAddAnimationChannelFloat(
    InitialValue=0.0
    LimitsEnabled=(Enabled=(bMinimum=True,bMaximum=True))
    MaximumValue=1.0
    MinimumValue=-1.0
    Name=rollLeftRight
    Parent=@n7.Item
   ) {template="SpawnAnimationChannel::Execute(in InitialValue,in MinimumValue,in MaximumValue,in LimitsEnabled,in Parent,in Name,out Item)"}
   > SpawnControl = &RigUnit_HierarchyAddControlFloat(
    InitialValue=0
    Name=CTRL_WrinkleMap
    OffsetSpace=LocalSpace
    OffsetTransform=(Rotation=(X=0,Y=0,Z=0,W=1),Translation=(X=0,Y=0,Z=2),Scale3D=(X=1,Y=1,Z=1))
    Settings="(PrimaryAxis=X,bIsScale=False,Limits=(Limit=(bMinimum=True,bMaximum=True),MinValue=0,MaxValue=0,bDrawLimits=False),Shape=(bVisible=True,Name=\"DefaultGizmoLibraryNormalized.Circle_Solid\",Color=(R=1,G=0.0406,B=0.8081,A=1),Transform=(Rotation=(X=-0.7071,Y=0,Z ...(truncated)"
    Parent=RigVMVariableNode() {variable=Local master_ctrl}.Value
   ) {template="SpawnControl::Execute(in InitialValue,in Settings,in OffsetTransform,in OffsetSpace,in Parent,in Name,out Item)"}
   > For_Each = RigVMDispatchNode(Array=/Script/ControlRig.RigUnit_CollectionNameSearchArray(PartialName=_Offset_, TypeToSearch=Curve) {is=struct RigVMUnitNode}.Items) {template="DISPATCH_RigVMDispatch_ArrayIterator(in Array,out Element,out Index,out Count,out Ratio)"}
   StartsWith = /Script/RigVM.RigVMFunction_StartsWith(Start=WM_, Name=@For_Each.Element.Name) {is=struct RigVMUnitNode, template="StartsWith::Execute(in Name,in Start,out Result)"}
   n8 = &RigVMFunction_ControlFlowBranch(ExecuteContext=@For_Each.ExecuteContext, Condition=@StartsWith.Result)
   &RigUnit_HierarchyAddAnimationChannelFloat(
    InitialValue=0.0
    LimitsEnabled=(Enabled=(bMinimum=true,bMaximum=true))
    MaximumValue=10
    MinimumValue=-10
    ExecutePin=@n8.True
    Parent=@SpawnControl.Item
    Name=@For_Each.Element.Name
   ) {template="SpawnAnimationChannel::Execute(in InitialValue,in MinimumValue,in MaximumValue,in LimitsEnabled,in Parent,in Name,out Item)"}
   n9 = &RigUnit_HierarchyAddControlVector(
    InitialValue=(X=0,Y=0,Z=0)
    Name=CTRL_C_teethU
    OffsetSpace=LocalSpace
    OffsetTransform=(Rotation=(X=0.7071,Y=0,Z=0,W=0.7071),Translation=(X=8,Y=0,Z=11),Scale3D=(X=0.5,Y=0.5,Z=0.5))
    Settings="(InitialSpace=LocalSpace,bIsPosition=True,Limits=(LimitX=(bMinimum=True,bMaximum=True),LimitY=(bMinimum=True,bMaximum=True),LimitZ=(bMinimum=True,bMaximum=True),MinValue=(X=-1,Y=-1,Z=-1),MaxValue=(X=1,Y=1,Z=1),bDrawLimits=False),Shape=(bVisible=True,Name=\" ...(truncated)"
    ExecutePin=@RigVMFunction_Sequence_7.B
    Parent=@n6.Value
   ) {template="SpawnControl::Execute(in InitialValue,in Settings,in OffsetTransform,in OffsetSpace,in Parent,in Name,out Item)"}
   > &RigUnit_HierarchyAddAnimationChannelFloat(
    InitialValue=0.0
    LimitsEnabled=(Enabled=(bMinimum=True,bMaximum=True))
    MaximumValue=1.0
    MinimumValue=-1.0
    Name=rotateFwdBack
    Parent=@n9.Item
   ) {template="SpawnAnimationChannel::Execute(in InitialValue,in MinimumValue,in MaximumValue,in LimitsEnabled,in Parent,in Name,out Item)"}
   n10 = &RigUnit_HierarchyAddControlVector(
    InitialValue=(X=0,Y=0,Z=0)
    Name=CTRL_C_teethD
    OffsetSpace=LocalSpace
    OffsetTransform=(Rotation=(X=0.7071,Y=0,Z=0,W=0.7071),Translation=(X=8,Y=0,Z=9),Scale3D=(X=0.5,Y=0.5,Z=0.5))
    Settings="(InitialSpace=LocalSpace,bIsPosition=True,Limits=(LimitX=(bMinimum=True,bMaximum=True),LimitY=(bMinimum=True,bMaximum=True),LimitZ=(bMinimum=True,bMaximum=True),MinValue=(X=-1,Y=-1,Z=-1),MaxValue=(X=1,Y=1,Z=1),bDrawLimits=False),Shape=(bVisible=True,Name=\" ...(truncated)"
    ExecutePin=@RigVMFunction_Sequence_7.C
    Parent=@n6.Value
   ) {template="SpawnControl::Execute(in InitialValue,in Settings,in OffsetTransform,in OffsetSpace,in Parent,in Name,out Item)"}
   > &RigUnit_HierarchyAddAnimationChannelFloat(
    InitialValue=0.0
    LimitsEnabled=(Enabled=(bMinimum=True,bMaximum=True))
    MaximumValue=1.0
    MinimumValue=-1.0
    Name=rotateFwdBack
    Parent=@n10.Item
   ) {template="SpawnAnimationChannel::Execute(in InitialValue,in MinimumValue,in MaximumValue,in LimitsEnabled,in Parent,in Name,out Item)"}
   n11 = /Script/ControlRig.RigUnit_HierarchyAddControlVector2D(
    InitialValue=(X=0,Y=0)
    Name=CTRL_C_mouth
    OffsetSpace=LocalSpace
    OffsetTransform=(Rotation=(X=0.7071,Y=0,Z=0,W=0.7071),Translation=(X=0,Y=0,Z=10.5),Scale3D=(X=1,Y=1,Z=1))
    Settings="(PrimaryAxis=Z,Limits=(LimitX=(bMinimum=True,bMaximum=True),LimitY=(bMinimum=True,bMaximum=True),MinValue=(X=-1,Y=-1),MaxValue=(X=1,Y=1),bDrawLimits=False),Shape=(bVisible=True,Name=\"DefaultGizmoLibraryNormalized.Square_Solid\",Color=(R=1,G=1,B=0,A=1),Trans ...(truncated)"
    ExecutePin=@RigVMFunction_Sequence_8.D
    Parent=RigVMVariableNode() {variable=Local master_ctrl}.Value
   ) {
    is=struct RigVMUnitNode
    template="SpawnControl::Execute(in InitialValue,in Settings,in OffsetTransform,in OffsetSpace,in Parent,in Name,out Item)"
   }
   > RigVMVariableNode(Value=@n11.Item) {variable=Local c_mouth_ctrl}
   > RigVMFunction_Sequence_1 = RigVMAggregateNode()
   n12 = &RigVMFunction_Sequence(ExecuteContext=@RigVMFunction_Sequence_1.A)
   n13 = RigVMVariableNode() {variable=Local c_mouth_ctrl}
   n14 = &RigUnit_HierarchyAddControlVector2D(
    InitialValue=(X=0,Y=0)
    Name=CTRL_L_mouth_lipU
    OffsetSpace=LocalSpace
    OffsetTransform=(Rotation=(X=0,Y=0.7071,Z=0,W=0.7071),Translation=(X=1.5,Y=1,Z=0),Scale3D=(X=0.8,Y=0.8,Z=0.8))
    Settings="(PrimaryAxis=Z,Limits=(LimitX=(bMinimum=True,bMaximum=True),LimitY=(bMinimum=True,bMaximum=True),MinValue=(X=-1,Y=-1),MaxValue=(X=1,Y=1),bDrawLimits=False),Shape=(bVisible=True,Name=\"DefaultGizmoLibraryNormalized.Circle_Solid\",Color=(R=1,G=1,B=0,A=1),Trans ...(truncated)"
    ExecutePin=@n12.A
    Parent=@n13.Value
   ) {template="SpawnControl::Execute(in InitialValue,in Settings,in OffsetTransform,in OffsetSpace,in Parent,in Name,out Item)"}
   > &RigUnit_HierarchyAddAnimationChannelFloat(
    InitialValue=0.0
    LimitsEnabled=(Enabled=(bMinimum=True,bMaximum=True))
    MaximumValue=1.0
    MinimumValue=0.0
    Name=mouthPress
    Parent=@n14.Item
   ) {template="SpawnAnimationChannel::Execute(in InitialValue,in MinimumValue,in MaximumValue,in LimitsEnabled,in Parent,in Name,out Item)"}
   > &RigUnit_HierarchyAddAnimationChannelFloat(
    InitialValue=0.0
    LimitsEnabled=(Enabled=(bMinimum=True,bMaximum=True))
    MaximumValue=1.0
    MinimumValue=0.0
    Name=lipsTogether
    Parent=@n14.Item
   ) {template="SpawnAnimationChannel::Execute(in InitialValue,in MinimumValue,in MaximumValue,in LimitsEnabled,in Parent,in Name,out Item)"}
   n15 = &RigUnit_HierarchyAddControlVector2D(
    InitialValue=(X=0,Y=0)
    Name=CTRL_R_mouth_lipU
    OffsetSpace=LocalSpace
    OffsetTransform=(Rotation=(X=0,Y=-0.7071,Z=0,W=-0.7071),Translation=(X=-1.5,Y=1,Z=0),Scale3D=(X=0.8,Y=0.8,Z=0.8))
    Settings="(PrimaryAxis=Z,Limits=(LimitX=(bMinimum=True,bMaximum=True),LimitY=(bMinimum=True,bMaximum=True),MinValue=(X=-1,Y=-1),MaxValue=(X=1,Y=1),bDrawLimits=False),Shape=(bVisible=True,Name=\"DefaultGizmoLibraryNormalized.Circle_Solid\",Color=(R=1,G=1,B=0,A=1),Trans ...(truncated)"
    ExecutePin=@n12.B
    Parent=@n13.Value
   ) {template="SpawnControl::Execute(in InitialValue,in Settings,in OffsetTransform,in OffsetSpace,in Parent,in Name,out Item)"}
   > &RigUnit_HierarchyAddAnimationChannelFloat(
    InitialValue=0.0
    LimitsEnabled=(Enabled=(bMinimum=True,bMaximum=True))
    MaximumValue=1.0
    MinimumValue=0.0
    Name=mouthPress
    Parent=@n15.Item
   ) {template="SpawnAnimationChannel::Execute(in InitialValue,in MinimumValue,in MaximumValue,in LimitsEnabled,in Parent,in Name,out Item)"}
   > n16 = &RigUnit_HierarchyAddAnimationChannelFloat(
    InitialValue=0.0
    LimitsEnabled=(Enabled=(bMinimum=True,bMaximum=True))
    MaximumValue=1.0
    MinimumValue=0.0
    Name=lipsTogether
    Parent=@n15.Item
   ) {template="SpawnAnimationChannel::Execute(in InitialValue,in MinimumValue,in MaximumValue,in LimitsEnabled,in Parent,in Name,out Item)"}
   n17 = &RigVMFunction_Sequence(ExecuteContext=@n16.ExecutePin)
   n18 = RigVMVariableNode() {variable=Local c_mouth_ctrl}
   n19 = &RigUnit_HierarchyAddNull(
    Name=CTRL_L_mouth_lipD_GRP
    Space=LocalSpace
    Transform=(Rotation=(X=-0.7071,Y=0,Z=0.7071,W=0),Translation=(X=1.5,Y=-1,Z=0),Scale3D=(X=0.8,Y=0.8,Z=0.8))
    ExecutePin=@n17.A
    Parent=@n18.Value
   )
   > n20 = &RigUnit_HierarchyAddNull(
    Name=CTRL_L_mouth_lipD_offset
    Space=LocalSpace
    Transform=(Rotation=(X=0,Y=0,Z=0,W=1),Translation=(X=0,Y=0,Z=0),Scale3D=(X=1,Y=1,Z=1))
    Parent=@n19.Item
   )
   > n21 = &RigUnit_HierarchyAddControlVector2D(
    InitialValue=(X=0,Y=0)
    Name=CTRL_L_mouth_lipD
    OffsetSpace=LocalSpace
    OffsetTransform=(Rotation=(X=0,Y=0,Z=0,W=1),Translation=(X=0,Y=0,Z=0),Scale3D=(X=1,Y=1,Z=1))
    Settings="(PrimaryAxis=Z,Limits=(LimitX=(bMinimum=True,bMaximum=True),LimitY=(bMinimum=True,bMaximum=True),MinValue=(X=-1,Y=-1),MaxValue=(X=1,Y=1),bDrawLimits=False),Shape=(bVisible=True,Name=\"DefaultGizmoLibraryNormalized.Circle_Solid\",Color=(R=1,G=1,B=0,A=1),Trans ...(truncated)"
    Parent=@n20.Item
   ) {template="SpawnControl::Execute(in InitialValue,in Settings,in OffsetTransform,in OffsetSpace,in Parent,in Name,out Item)"}
   > &RigUnit_HierarchyAddAnimationChannelFloat(
    InitialValue=0.0
    LimitsEnabled=(Enabled=(bMinimum=True,bMaximum=True))
    MaximumValue=1.0
    MinimumValue=0.0
    Name=mouthPress
    Parent=@n21.Item
   ) {template="SpawnAnimationChannel::Execute(in InitialValue,in MinimumValue,in MaximumValue,in LimitsEnabled,in Parent,in Name,out Item)"}
   > &RigUnit_HierarchyAddAnimationChannelFloat(
    InitialValue=0.0
    LimitsEnabled=(Enabled=(bMinimum=True,bMaximum=True))
    MaximumValue=1.0
    MinimumValue=0.0
    Name=lipsTogether
    Parent=@n21.Item
   ) {template="SpawnAnimationChannel::Execute(in InitialValue,in MinimumValue,in MaximumValue,in LimitsEnabled,in Parent,in Name,out Item)"}
   n22 = &RigUnit_HierarchyAddNull(
    Name=CTRL_R_mouth_lipD_GRP
    Space=LocalSpace
    Transform=(Rotation=(X=-0.7071,Y=0,Z=0.7071,W=0),Translation=(X=-1.5,Y=-1,Z=0),Scale3D=(X=0.8,Y=0.8,Z=0.8))
    ExecutePin=@n17.B
    Parent=@n18.Value
   )
   > n23 = &RigUnit_HierarchyAddNull(
    Name=CTRL_R_mouth_lipD_offset
    Space=LocalSpace
    Transform=(Rotation=(X=0,Y=0,Z=0,W=1),Translation=(X=0,Y=0,Z=0),Scale3D=(X=1,Y=1,Z=1))
    Parent=@n22.Item
   )
   > n24 = &RigUnit_HierarchyAddControlVector2D(
    InitialValue=(X=0,Y=0)
    Name=CTRL_R_mouth_lipD
    OffsetSpace=LocalSpace
    OffsetTransform=(Rotation=(X=0,Y=0,Z=0,W=1),Translation=(X=0,Y=0,Z=0),Scale3D=(X=1,Y=1,Z=1))
    Settings="(PrimaryAxis=Z,Limits=(LimitX=(bMinimum=True,bMaximum=True),LimitY=(bMinimum=True,bMaximum=True),MinValue=(X=-1,Y=-1),MaxValue=(X=1,Y=1),bDrawLimits=False),Shape=(bVisible=True,Name=\"DefaultGizmoLibraryNormalized.Circle_Solid\",Color=(R=1,G=1,B=0,A=1),Trans ...(truncated)"
    Parent=@n23.Item
   ) {template="SpawnControl::Execute(in InitialValue,in Settings,in OffsetTransform,in OffsetSpace,in Parent,in Name,out Item)"}
   > &RigUnit_HierarchyAddAnimationChannelFloat(
    InitialValue=0.0
    LimitsEnabled=(Enabled=(bMinimum=True,bMaximum=True))
    MaximumValue=1.0
    MinimumValue=0.0
    Name=mouthPress
    Parent=@n24.Item
   ) {template="SpawnAnimationChannel::Execute(in InitialValue,in MinimumValue,in MaximumValue,in LimitsEnabled,in Parent,in Name,out Item)"}
   > &RigUnit_HierarchyAddAnimationChannelFloat(
    InitialValue=0.0
    LimitsEnabled=(Enabled=(bMinimum=True,bMaximum=True))
    MaximumValue=1.0
    MinimumValue=0.0
    Name=lipsTogether
    Parent=@n24.Item
   ) {template="SpawnAnimationChannel::Execute(in InitialValue,in MinimumValue,in MaximumValue,in LimitsEnabled,in Parent,in Name,out Item)"}
   n25 = &RigVMFunction_Sequence(ExecuteContext=@RigVMFunction_Sequence_1.B)
   n26 = RigVMVariableNode() {variable=Local c_mouth_ctrl}
   n27 = &RigUnit_HierarchyAddControlVector2D(
    InitialValue=(X=0,Y=0)
    Name=CTRL_L_mouth_smileDepress
    OffsetSpace=LocalSpace
    OffsetTransform=(Rotation=(X=0,Y=0,Z=-0.3827,W=0.9239),Translation=(X=3,Y=0,Z=0),Scale3D=(X=1,Y=1,Z=1))
    Settings="(PrimaryAxis=Z,Limits=(LimitX=(bMinimum=True,bMaximum=True),LimitY=(bMinimum=True,bMaximum=True),MinValue=(X=0,Y=0),MaxValue=(X=1,Y=1),bDrawLimits=False),Shape=(bVisible=True,Name=\"DefaultGizmoLibraryNormalized.Square_Solid\",Color=(R=1,G=1,B=0,A=1),Transfo ...(truncated)"
    ExecutePin=@n25.A
    Parent=@n26.Value
   ) {template="SpawnControl::Execute(in InitialValue,in Settings,in OffsetTransform,in OffsetSpace,in Parent,in Name,out Item)"}
   > n28 = &RigUnit_HierarchyAddControlVector2D(
    InitialValue=(X=0,Y=0)
    Name=CTRL_L_mouth_corner
    OffsetSpace=LocalSpace
    OffsetTransform=(Rotation=(X=0,Y=0,Z=0.3827,W=0.9239),Translation=(X=0.7071,Y=0.7071,Z=0),Scale3D=(X=0.6,Y=0.6,Z=0.6))
    Settings="(PrimaryAxis=Z,Limits=(LimitX=(bMinimum=True,bMaximum=True),LimitY=(bMinimum=True,bMaximum=True),MinValue=(X=-1,Y=-1),MaxValue=(X=1,Y=1),bDrawLimits=False),Shape=(bVisible=True,Name=\"DefaultGizmoLibraryNormalized.Circle_Solid\",Color=(R=1,G=1,B=0,A=1),Trans ...(truncated)"
    Parent=@n27.Item
   ) {template="SpawnControl::Execute(in InitialValue,in Settings,in OffsetTransform,in OffsetSpace,in Parent,in Name,out Item)"}
   > RigVMFunction_Sequence_5 = RigVMAggregateNode()
   &RigUnit_HierarchyAddAnimationChannelFloat(
    InitialValue=0.0
    LimitsEnabled=(Enabled=(bMinimum=True,bMaximum=True))
    MaximumValue=1.0
    MinimumValue=0.0
    Name=lipsBlow
    Parent=@n28.Item
    ExecutePin=@RigVMFunction_Sequence_5.A
   ) {template="SpawnAnimationChannel::Execute(in InitialValue,in MinimumValue,in MaximumValue,in LimitsEnabled,in Parent,in Name,out Item)"}
   > &RigUnit_HierarchyAddAnimationChannelFloat(
    InitialValue=0.0
    LimitsEnabled=(Enabled=(bMinimum=True,bMaximum=True))
    MaximumValue=1.0
    MinimumValue=0.0
    Name=mouthDimple
    Parent=@n27.Item
   ) {template="SpawnAnimationChannel::Execute(in InitialValue,in MinimumValue,in MaximumValue,in LimitsEnabled,in Parent,in Name,out Item)"}
   > &RigUnit_HierarchyAddAnimationChannelFloat(
    InitialValue=0.0
    LimitsEnabled=(Enabled=(bMinimum=True,bMaximum=True))
    MaximumValue=1.0
    MinimumValue=0.0
    Name=mouthStretch
    Parent=@n27.Item
   ) {template="SpawnAnimationChannel::Execute(in InitialValue,in MinimumValue,in MaximumValue,in LimitsEnabled,in Parent,in Name,out Item)"}
   &RigUnit_HierarchyAddAnimationChannelFloat(
    InitialValue=0.0
    LimitsEnabled=(Enabled=(bMinimum=True,bMaximum=True))
    MaximumValue=1.0
    MinimumValue=0.0
    Name=mouthPurseU
    ExecutePin=@RigVMFunction_Sequence_5.B
    Parent=@n27.Item
   ) {template="SpawnAnimationChannel::Execute(in InitialValue,in MinimumValue,in MaximumValue,in LimitsEnabled,in Parent,in Name,out Item)"}
   > &RigUnit_HierarchyAddAnimationChannelFloat(
    InitialValue=0.0
    LimitsEnabled=(Enabled=(bMinimum=True,bMaximum=True))
    MaximumValue=1.0
    MinimumValue=0.0
    Name=mouthPurseD
    Parent=@n27.Item
   ) {template="SpawnAnimationChannel::Execute(in InitialValue,in MinimumValue,in MaximumValue,in LimitsEnabled,in Parent,in Name,out Item)"}
   > &RigUnit_HierarchyAddAnimationChannelFloat(
    InitialValue=0.0
    LimitsEnabled=(Enabled=(bMinimum=True,bMaximum=True))
    MaximumValue=1.0
    MinimumValue=0.0
    Name=mouthFunnelU
    Parent=@n27.Item
   ) {template="SpawnAnimationChannel::Execute(in InitialValue,in MinimumValue,in MaximumValue,in LimitsEnabled,in Parent,in Name,out Item)"}
   &RigUnit_HierarchyAddAnimationChannelFloat(
    InitialValue=0.0
    LimitsEnabled=(Enabled=(bMinimum=True,bMaximum=True))
    MaximumValue=1.0
    MinimumValue=0.0
    Name=mouthFunnelD
    ExecutePin=@RigVMFunction_Sequence_5.C
    Parent=@n27.Item
   ) {template="SpawnAnimationChannel::Execute(in InitialValue,in MinimumValue,in MaximumValue,in LimitsEnabled,in Parent,in Name,out Item)"}
   > &RigUnit_HierarchyAddAnimationChannelFloat(
    InitialValue=0.0
    LimitsEnabled=(Enabled=(bMinimum=True,bMaximum=True))
    MaximumValue=1.0
    MinimumValue=-1.0
    Name=mouthCornerSharpnessU
    Parent=@n27.Item
   ) {template="SpawnAnimationChannel::Execute(in InitialValue,in MinimumValue,in MaximumValue,in LimitsEnabled,in Parent,in Name,out Item)"}
   > &RigUnit_HierarchyAddAnimationChannelFloat(
    InitialValue=0.0
    LimitsEnabled=(Enabled=(bMinimum=True,bMaximum=True))
    MaximumValue=1.0
    MinimumValue=-1.0
    Name=mouthCornerSharpnessD
    Parent=@n27.Item
   ) {template="SpawnAnimationChannel::Execute(in InitialValue,in MinimumValue,in MaximumValue,in LimitsEnabled,in Parent,in Name,out Item)"}
   n29 = &RigUnit_HierarchyAddControlVector2D(
    InitialValue=(X=0,Y=0)
    Name=CTRL_R_mouth_smileDepress
    OffsetSpace=LocalSpace
    OffsetTransform=(Rotation=(X=0.3827,Y=-0.9239,Z=0,W=0),Translation=(X=-3,Y=0,Z=0),Scale3D=(X=1,Y=1,Z=1))
    Settings="(PrimaryAxis=Z,Limits=(LimitX=(bMinimum=True,bMaximum=True),LimitY=(bMinimum=True,bMaximum=True),MinValue=(X=0,Y=0),MaxValue=(X=1,Y=1),bDrawLimits=False),Shape=(bVisible=True,Name=\"DefaultGizmoLibraryNormalized.Square_Solid\",Color=(R=1,G=1,B=0,A=1),Transfo ...(truncated)"
    ExecutePin=@n25.B
    Parent=@n26.Value
   ) {template="SpawnControl::Execute(in InitialValue,in Settings,in OffsetTransform,in OffsetSpace,in Parent,in Name,out Item)"}
   > n30 = &RigUnit_HierarchyAddControlVector2D(
    InitialValue=(X=0,Y=0)
    Name=CTRL_R_mouth_corner
    OffsetSpace=LocalSpace
    OffsetTransform=(Rotation=(X=0,Y=0,Z=0.3827,W=0.9239),Translation=(X=0.7071,Y=0.7071,Z=0),Scale3D=(X=0.6,Y=0.6,Z=0.6))
    Settings="(PrimaryAxis=Z,Limits=(LimitX=(bMinimum=True,bMaximum=True),LimitY=(bMinimum=True,bMaximum=True),MinValue=(X=-1,Y=-1),MaxValue=(X=1,Y=1),bDrawLimits=False),Shape=(bVisible=True,Name=\"DefaultGizmoLibraryNormalized.Circle_Solid\",Color=(R=1,G=1,B=0,A=1),Trans ...(truncated)"
    Parent=@n29.Item
   ) {template="SpawnControl::Execute(in InitialValue,in Settings,in OffsetTransform,in OffsetSpace,in Parent,in Name,out Item)"}
   > RigVMFunction_Sequence_6 = RigVMAggregateNode()
   &RigUnit_HierarchyAddAnimationChannelFloat(
    InitialValue=0.0
    LimitsEnabled=(Enabled=(bMinimum=True,bMaximum=True))
    MaximumValue=1.0
    MinimumValue=0.0
    Name=lipsBlow
    Parent=@n30.Item
    ExecutePin=@RigVMFunction_Sequence_6.A
   ) {template="SpawnAnimationChannel::Execute(in InitialValue,in MinimumValue,in MaximumValue,in LimitsEnabled,in Parent,in Name,out Item)"}
   > &RigUnit_HierarchyAddAnimationChannelFloat(
    InitialValue=0.0
    LimitsEnabled=(Enabled=(bMinimum=True,bMaximum=True))
    MaximumValue=1.0
    MinimumValue=0.0
    Name=mouthDimple
    Parent=@n29.Item
   ) {template="SpawnAnimationChannel::Execute(in InitialValue,in MinimumValue,in MaximumValue,in LimitsEnabled,in Parent,in Name,out Item)"}
   > &RigUnit_HierarchyAddAnimationChannelFloat(
    InitialValue=0.0
    LimitsEnabled=(Enabled=(bMinimum=True,bMaximum=True))
    MaximumValue=1.0
    MinimumValue=0.0
    Name=mouthStretch
    Parent=@n29.Item
   ) {template="SpawnAnimationChannel::Execute(in InitialValue,in MinimumValue,in MaximumValue,in LimitsEnabled,in Parent,in Name,out Item)"}
   &RigUnit_HierarchyAddAnimationChannelFloat(
    InitialValue=0.0
    LimitsEnabled=(Enabled=(bMinimum=True,bMaximum=True))
    MaximumValue=1.0
    MinimumValue=0.0
    Name=mouthPurseU
    ExecutePin=@RigVMFunction_Sequence_6.B
    Parent=@n29.Item
   ) {template="SpawnAnimationChannel::Execute(in InitialValue,in MinimumValue,in MaximumValue,in LimitsEnabled,in Parent,in Name,out Item)"}
   > &RigUnit_HierarchyAddAnimationChannelFloat(
    InitialValue=0.0
    LimitsEnabled=(Enabled=(bMinimum=True,bMaximum=True))
    MaximumValue=1.0
    MinimumValue=0.0
    Name=mouthPurseD
    Parent=@n29.Item
   ) {template="SpawnAnimationChannel::Execute(in InitialValue,in MinimumValue,in MaximumValue,in LimitsEnabled,in Parent,in Name,out Item)"}
   > &RigUnit_HierarchyAddAnimationChannelFloat(
    InitialValue=0.0
    LimitsEnabled=(Enabled=(bMinimum=True,bMaximum=True))
    MaximumValue=1.0
    MinimumValue=0.0
    Name=mouthFunnelU
    Parent=@n29.Item
   ) {template="SpawnAnimationChannel::Execute(in InitialValue,in MinimumValue,in MaximumValue,in LimitsEnabled,in Parent,in Name,out Item)"}
   &RigUnit_HierarchyAddAnimationChannelFloat(
    InitialValue=0.0
    LimitsEnabled=(Enabled=(bMinimum=True,bMaximum=True))
    MaximumValue=1.0
    MinimumValue=0.0
    Name=mouthFunnelD
    ExecutePin=@RigVMFunction_Sequence_6.C
    Parent=@n29.Item
   ) {template="SpawnAnimationChannel::Execute(in InitialValue,in MinimumValue,in MaximumValue,in LimitsEnabled,in Parent,in Name,out Item)"}
   > &RigUnit_HierarchyAddAnimationChannelFloat(
    InitialValue=0.0
    LimitsEnabled=(Enabled=(bMinimum=True,bMaximum=True))
    MaximumValue=1.0
    MinimumValue=-1.0
    Name=mouthCornerSharpnessU
    Parent=@n29.Item
   ) {template="SpawnAnimationChannel::Execute(in InitialValue,in MinimumValue,in MaximumValue,in LimitsEnabled,in Parent,in Name,out Item)"}
   > &RigUnit_HierarchyAddAnimationChannelFloat(
    InitialValue=0.0
    LimitsEnabled=(Enabled=(bMinimum=True,bMaximum=True))
    MaximumValue=1.0
    MinimumValue=-1.0
    Name=mouthCornerSharpnessD
    Parent=@n29.Item
   ) {template="SpawnAnimationChannel::Execute(in InitialValue,in MinimumValue,in MaximumValue,in LimitsEnabled,in Parent,in Name,out Item)"}
   n31 = &RigUnit_HierarchyAddNull(
    Name=Brow_Grp
    Parent=(Type=Control,Name="Face_Ctrl_Master")
    Space=LocalSpace
    Transform=(Rotation=(X=0,Y=0,Z=0,W=1),Translation=(X=0,Y=0,Z=0),Scale3D=(X=1,Y=1,Z=1))
    ExecutePin=@RigVMFunction_Sequence_8.A
   )
   > n32 = &RigUnit_HierarchyAddControlVector2D(
    InitialValue=(X=0,Y=0)
    Name=CTRL_L_brow
    OffsetSpace=LocalSpace
    OffsetTransform=(Rotation=(X=0,Y=0.7071,Z=0.7071,W=0),Translation=(X=3.5,Y=0,Z=19.5),Scale3D=(X=1,Y=1,Z=1))
    Settings="(PrimaryAxis=Z,Limits=(LimitX=(bMinimum=True,bMaximum=True),LimitY=(bMinimum=True,bMaximum=True),MinValue=(X=0,Y=-1),MaxValue=(X=1,Y=1),bDrawLimits=False),Shape=(bVisible=True,Name=\"DefaultGizmoLibraryNormalized.Square_Solid\",Color=(R=1,G=1,B=0,A=1),Transf ...(truncated)"
    Parent=@n31.Item
   ) {template="SpawnControl::Execute(in InitialValue,in Settings,in OffsetTransform,in OffsetSpace,in Parent,in Name,out Item)"}
   > &RigUnit_HierarchyAddControlFloat(
    InitialValue=0.0
    Name=CTRL_L_brow_inner
    OffsetSpace=LocalSpace
    OffsetTransform=(Rotation=(X=0,Y=0,Z=0,W=-1),Translation=(X=2,Y=-0.5,Z=0),Scale3D=(X=1,Y=1,Z=1))
    Settings="(PrimaryAxis=Y,bIsScale=False,Limits=(Limit=(bMinimum=True,bMaximum=True),MinValue=-1,MaxValue=1,bDrawLimits=False),Shape=(bVisible=True,Name=\"DefaultGizmoLibraryNormalized.Circle_Solid\",Color=(R=1,G=1,B=0,A=1),Transform=(Rotation=(X=0,Y=0,Z=0,W=1),Transla ...(truncated)"
    Parent=@n32.Item
   ) {template="SpawnControl::Execute(in InitialValue,in Settings,in OffsetTransform,in OffsetSpace,in Parent,in Name,out Item)"}
   > &RigUnit_HierarchyAddControlFloat(
    InitialValue=0.0
    Name=CTRL_L_brow_outer
    OffsetSpace=LocalSpace
    OffsetTransform=(Rotation=(X=0,Y=0,Z=0,W=1),Translation=(X=-1.5,Y=-0.5,Z=0),Scale3D=(X=1,Y=1,Z=1))
    Settings="(PrimaryAxis=Y,bIsScale=False,Limits=(Limit=(bMinimum=True,bMaximum=True),MinValue=-1,MaxValue=1,bDrawLimits=False),Shape=(bVisible=True,Name=\"DefaultGizmoLibraryNormalized.Circle_Solid\",Color=(R=1,G=1,B=0,A=1),Transform=(Rotation=(X=0,Y=0,Z=0,W=1),Transla ...(truncated)"
    Parent=@n32.Item
   ) {template="SpawnControl::Execute(in InitialValue,in Settings,in OffsetTransform,in OffsetSpace,in Parent,in Name,out Item)"}
   > n33 = &RigUnit_HierarchyAddControlVector2D(
    InitialValue=(X=0,Y=0)
    Name=CTRL_R_brow
    OffsetSpace=LocalSpace
    OffsetTransform=(Rotation=(X=-0.7071,Y=0,Z=0,W=-0.7071),Translation=(X=-3.5,Y=0,Z=19.5),Scale3D=(X=1,Y=1,Z=1))
    Settings="(PrimaryAxis=Z,Limits=(LimitX=(bMinimum=True,bMaximum=True),LimitY=(bMinimum=True,bMaximum=True),MinValue=(X=0,Y=-1),MaxValue=(X=1,Y=1),bDrawLimits=False),Shape=(bVisible=True,Name=\"DefaultGizmoLibraryNormalized.Square_Solid\",Color=(R=1,G=0,B=0,A=1),Transf ...(truncated)"
    Parent=@n31.Item
   ) {template="SpawnControl::Execute(in InitialValue,in Settings,in OffsetTransform,in OffsetSpace,in Parent,in Name,out Item)"}
   > &RigUnit_HierarchyAddControlFloat(
    InitialValue=0.0
    Name=CTRL_R_brow_inner
    OffsetSpace=LocalSpace
    OffsetTransform=(Rotation=(X=0,Y=0,Z=0,W=1),Translation=(X=2,Y=-0.5,Z=0),Scale3D=(X=1,Y=1,Z=1))
    Settings="(PrimaryAxis=Y,bIsScale=False,Limits=(Limit=(bMinimum=True,bMaximum=True),MinValue=-1,MaxValue=1,bDrawLimits=False),Shape=(bVisible=True,Name=\"DefaultGizmoLibraryNormalized.Circle_Solid\",Color=(R=1,G=1,B=0,A=1),Transform=(Rotation=(X=0,Y=0,Z=0,W=1),Transla ...(truncated)"
    Parent=@n33.Item
   ) {template="SpawnControl::Execute(in InitialValue,in Settings,in OffsetTransform,in OffsetSpace,in Parent,in Name,out Item)"}
   > &RigUnit_HierarchyAddControlFloat(
    InitialValue=0.0
    Name=CTRL_R_brow_outer
    OffsetSpace=LocalSpace
    OffsetTransform=(Rotation=(X=0,Y=0,Z=0,W=1),Translation=(X=-1.5,Y=-0.5,Z=0),Scale3D=(X=1,Y=1,Z=1))
    Settings="(PrimaryAxis=Y,bIsScale=False,Limits=(Limit=(bMinimum=True,bMaximum=True),MinValue=-1,MaxValue=1,bDrawLimits=False),Shape=(bVisible=True,Name=\"DefaultGizmoLibraryNormalized.Circle_Solid\",Color=(R=1,G=1,B=0,A=1),Transform=(Rotation=(X=0,Y=0,Z=0,W=1),Transla ...(truncated)"
    Parent=@n33.Item
   ) {template="SpawnControl::Execute(in InitialValue,in Settings,in OffsetTransform,in OffsetSpace,in Parent,in Name,out Item)"}
   n34 = &RigUnit_HierarchyAddNull(
    Name=Eye_Grp
    Parent=(Type=Control,Name="Face_Ctrl_Master")
    Space=LocalSpace
    Transform=(Rotation=(X=0,Y=0,Z=0,W=1),Translation=(X=0,Y=0,Z=0),Scale3D=(X=1,Y=1,Z=1))
    ExecutePin=@RigVMFunction_Sequence_8.B
   )
   > RigVMVariableNode(Value=@n34.Item) {variable=Local eye_grp}
   > n35 = &RigUnit_HierarchyAddControlVector2D(
    InitialValue=(X=0,Y=0)
    Name=CTRL_C_eye
    OffsetSpace=LocalSpace
    OffsetTransform=(Rotation=(X=0.7071,Y=0,Z=0,W=0.7071),Translation=(X=0,Y=0,Z=17),Scale3D=(X=1.1,Y=1.1,Z=1.1))
    Settings="(PrimaryAxis=Z,Limits=(LimitX=(bMinimum=True,bMaximum=True),LimitY=(bMinimum=True,bMaximum=True),MinValue=(X=-1,Y=-1),MaxValue=(X=1,Y=1),bDrawLimits=False),Shape=(bVisible=True,Name=\"DefaultGizmoLibraryNormalized.RoundedSquare_Thick\",Color=(R=1,G=1,B=0,A=1 ...(truncated)"
    Parent=@n34.Item
   ) {template="SpawnControl::Execute(in InitialValue,in Settings,in OffsetTransform,in OffsetSpace,in Parent,in Name,out Item)"}
   > &RigUnit_HierarchyAddControlVector2D(
    InitialValue=(X=0,Y=0)
    Name=CTRL_L_eye
    OffsetSpace=LocalSpace
    OffsetTransform=(Rotation=(X=0,Y=0,Z=0,W=1),Translation=(X=0.6612,Y=0,Z=0),Scale3D=(X=0.2893,Y=0.2893,Z=0.2893))
    Settings="(PrimaryAxis=Z,Limits=(LimitX=(bMinimum=True,bMaximum=True),LimitY=(bMinimum=True,bMaximum=True),MinValue=(X=-1,Y=-1),MaxValue=(X=1,Y=1),bDrawLimits=False),Shape=(bVisible=True,Name=\"DefaultGizmoLibraryNormalized.Circle_Thick\",Color=(R=0,G=1,B=0,A=1),Trans ...(truncated)"
    Parent=@n35.Item
   ) {template="SpawnControl::Execute(in InitialValue,in Settings,in OffsetTransform,in OffsetSpace,in Parent,in Name,out Item)"}
   > &RigUnit_HierarchyAddControlVector2D(
    InitialValue=(X=0,Y=0)
    Name=CTRL_R_eye
    OffsetSpace=LocalSpace
    OffsetTransform=(Rotation=(X=0,Y=0,Z=0,W=1),Translation=(X=-0.6612,Y=0,Z=0),Scale3D=(X=0.2893,Y=0.2893,Z=0.2893))
    Settings="(PrimaryAxis=Z,Limits=(LimitX=(bMinimum=True,bMaximum=True),LimitY=(bMinimum=True,bMaximum=True),MinValue=(X=-1,Y=-1),MaxValue=(X=1,Y=1),bDrawLimits=False),Shape=(bVisible=True,Name=\"DefaultGizmoLibraryNormalized.Circle_Thick\",Color=(R=0,G=1,B=0,A=1),Trans ...(truncated)"
    Parent=@n35.Item
   ) {template="SpawnControl::Execute(in InitialValue,in Settings,in OffsetTransform,in OffsetSpace,in Parent,in Name,out Item)"}
   > &RigUnit_HierarchyAddAnimationChannelFloat(
    InitialValue=0.0
    LimitsEnabled=(Enabled=(bMinimum=True,bMaximum=True))
    MaximumValue=1.0
    MinimumValue=-1.0
    Name=eyesConvergence
    Parent=@n35.Item
   ) {template="SpawnAnimationChannel::Execute(in InitialValue,in MinimumValue,in MaximumValue,in LimitsEnabled,in Parent,in Name,out Item)"}
   > &RigUnit_HierarchyAddAnimationChannelFloat(
    InitialValue=0.0
    LimitsEnabled=(Enabled=(bMinimum=True,bMaximum=True))
    MaximumValue=1.0
    MinimumValue=0.0
    Name=eyelidsReduce
    Parent=@n35.Item
   ) {template="SpawnAnimationChannel::Execute(in InitialValue,in MinimumValue,in MaximumValue,in LimitsEnabled,in Parent,in Name,out Item)"}
   n36 = RigVMVariableNode() {variable=Local eye_grp}
   > &RigUnit_HierarchyAddControlFloat(
    InitialValue=0.0
    Name=CTRL_L_eye_blink
    OffsetSpace=LocalSpace
    OffsetTransform=(Rotation=(X=0,Y=-0.7071,Z=0.7071,W=0),Translation=(X=4.5,Y=0,Z=17),Scale3D=(X=0.8,Y=0.8,Z=0.8))
    Settings="(PrimaryAxis=Y,bIsScale=False,Limits=(Limit=(bMinimum=True,bMaximum=True),MinValue=-1,MaxValue=1,bDrawLimits=False),Shape=(bVisible=True,Name=\"DefaultGizmoLibraryNormalized.Square_Solid\",Color=(R=1,G=1,B=0,A=1),Transform=(Rotation=(X=0,Y=0,Z=0,W=1),Transla ...(truncated)"
    Parent=@n36.Value
   ) {template="SpawnControl::Execute(in InitialValue,in Settings,in OffsetTransform,in OffsetSpace,in Parent,in Name,out Item)"}
   > &RigUnit_HierarchyAddControlFloat(
    InitialValue=0.0
    Name=CTRL_R_eye_blink
    OffsetSpace=LocalSpace
    OffsetTransform=(Rotation=(X=-0.7071,Y=0,Z=0,W=0.7071),Translation=(X=-4.5,Y=0,Z=17),Scale3D=(X=0.8,Y=0.8,Z=0.8))
    Settings="(PrimaryAxis=Y,bIsScale=False,Limits=(Limit=(bMinimum=True,bMaximum=True),MinValue=-1,MaxValue=1,bDrawLimits=False),Shape=(bVisible=True,Name=\"DefaultGizmoLibraryNormalized.Square_Solid\",Color=(R=1,G=1,B=0,A=1),Transform=(Rotation=(X=0,Y=0,Z=0,W=1),Transla ...(truncated)"
    Parent=@n36.Value
   ) {template="SpawnControl::Execute(in InitialValue,in Settings,in OffsetTransform,in OffsetSpace,in Parent,in Name,out Item)"}
   n37 = RigVMVariableNode() {variable=Local eye_grp}
   > n38 = &RigUnit_HierarchyAddNull(
    Name=CTRL_L_eye_eyelidU_GRP
    Space=LocalSpace
    Transform=(Rotation=(X=0,Y=0.7071,Z=0.7071,W=0),Translation=(X=3,Y=0,Z=17.5),Scale3D=(X=0.6,Y=0.6,Z=0.6))
    Parent=@n37.Value
   )
   > n39 = &RigUnit_HierarchyAddNull(
    Name=CTRL_L_eye_eyelidU_offset
    Space=LocalSpace
    Transform=(Rotation=(X=0,Y=0,Z=0,W=1),Translation=(X=0,Y=0,Z=0),Scale3D=(X=1,Y=1,Z=1))
    Parent=@n38.Item
   )
   > RigUnit_HierarchyAddControlVector2D = &RigUnit_HierarchyAddControlVector2D(
    InitialValue=(X=0,Y=0)
    Name=CTRL_L_eye_eyelidU
    OffsetSpace=LocalSpace
    OffsetTransform=(Rotation=(X=0,Y=0,Z=0,W=1),Translation=(X=0,Y=0,Z=0),Scale3D=(X=1,Y=1,Z=1))
    Settings="(PrimaryAxis=Z,Limits=(LimitX=(bMinimum=True,bMaximum=True),LimitY=(bMinimum=True,bMaximum=True),MinValue=(X=-1,Y=-1),MaxValue=(X=1,Y=1),bDrawLimits=False),Shape=(bVisible=True,Name=\"DefaultGizmoLibraryNormalized.Circle_Solid\",Color=(R=1,G=1,B=0,A=1),Trans ...(truncated)"
    Parent=@n39.Item
   ) {template="SpawnControl::Execute(in InitialValue,in Settings,in OffsetTransform,in OffsetSpace,in Parent,in Name,out Item)"}
   > n40 = &RigUnit_HierarchyAddNull(
    Name=CTRL_R_eye_eyelidU_GRP
    Space=LocalSpace
    Transform=(Rotation=(X=-0.7071,Y=0,Z=0,W=-0.7071),Translation=(X=-3.4,Y=0,Z=17.5),Scale3D=(X=0.6,Y=0.6,Z=0.6))
    Parent=@n37.Value
   )
   > n41 = &RigUnit_HierarchyAddNull(
    Name=CTRL_R_eye_eyelidU_offset
    Space=LocalSpace
    Transform=(Rotation=(X=0,Y=0,Z=0,W=1),Translation=(X=0,Y=0,Z=0),Scale3D=(X=1,Y=1,Z=1))
    Parent=@n40.Item
   )
   > &RigUnit_HierarchyAddControlVector2D(
    InitialValue=(X=0,Y=0)
    Name=CTRL_R_eye_eyelidU
    OffsetSpace=LocalSpace
    OffsetTransform=(Rotation=(X=0,Y=0,Z=0,W=1),Translation=(X=0,Y=0,Z=0),Scale3D=(X=1,Y=1,Z=1))
    Settings="(PrimaryAxis=Z,Limits=(LimitX=(bMinimum=True,bMaximum=True),LimitY=(bMinimum=True,bMaximum=True),MinValue=(X=-1,Y=-1),MaxValue=(X=1,Y=1),bDrawLimits=False),Shape=(bVisible=True,Name=\"DefaultGizmoLibraryNormalized.Circle_Solid\",Color=(R=1,G=1,B=0,A=1),Trans ...(truncated)"
    Parent=@n41.Item
   ) {template="SpawnControl::Execute(in InitialValue,in Settings,in OffsetTransform,in OffsetSpace,in Parent,in Name,out Item)"}
   n42 = RigVMVariableNode() {variable=Local eye_grp}
   > &RigUnit_HierarchyAddControlVector2D(
    InitialValue=(X=0,Y=0)
    Name=CTRL_L_eye_eyelidD
    OffsetSpace=LocalSpace
    OffsetTransform=(Rotation=(X=0,Y=0.7071,Z=0.7071,W=0),Translation=(X=3,Y=0,Z=16.5),Scale3D=(X=0.6,Y=0.6,Z=0.6))
    Settings="(PrimaryAxis=Z,Limits=(LimitX=(bMinimum=True,bMaximum=True),LimitY=(bMinimum=True,bMaximum=True),MinValue=(X=-1,Y=-1),MaxValue=(X=1,Y=1),bDrawLimits=False),Shape=(bVisible=True,Name=\"DefaultGizmoLibraryNormalized.Circle_Solid\",Color=(R=1,G=1,B=0,A=1),Trans ...(truncated)"
    Parent=@n42.Value
   ) {template="SpawnControl::Execute(in InitialValue,in Settings,in OffsetTransform,in OffsetSpace,in Parent,in Name,out Item)"}
   > &RigUnit_HierarchyAddControlVector2D(
    InitialValue=(X=0,Y=0)
    Name=CTRL_R_eye_eyelidD
    OffsetSpace=LocalSpace
    OffsetTransform=(Rotation=(X=-0.7071,Y=0,Z=0,W=-0.7071),Translation=(X=-3.4,Y=0,Z=16.5),Scale3D=(X=0.6,Y=0.6,Z=0.6))
    Settings="(PrimaryAxis=Z,Limits=(LimitX=(bMinimum=True,bMaximum=True),LimitY=(bMinimum=True,bMaximum=True),MinValue=(X=-1,Y=-1),MaxValue=(X=1,Y=1),bDrawLimits=False),Shape=(bVisible=True,Name=\"DefaultGizmoLibraryNormalized.Circle_Solid\",Color=(R=1,G=1,B=0,A=1),Trans ...(truncated)"
    Parent=@n42.Value
   ) {template="SpawnControl::Execute(in InitialValue,in Settings,in OffsetTransform,in OffsetSpace,in Parent,in Name,out Item)"}
   > n43 = &RigUnit_HierarchyAddControlVector2D(
    InitialValue=(X=0,Y=0)
    Name=CTRL_C_nose
    OffsetSpace=LocalSpace
    OffsetTransform=(Rotation=(X=0.7071,Y=0,Z=0,W=0.7071),Translation=(X=0,Y=0,Z=13),Scale3D=(X=1,Y=1,Z=1))
    Settings="(PrimaryAxis=Z,Limits=(LimitX=(bMinimum=True,bMaximum=True),LimitY=(bMinimum=True,bMaximum=True),MinValue=(X=-1,Y=-1),MaxValue=(X=1,Y=1),bDrawLimits=False),Shape=(bVisible=True,Name=\"DefaultGizmoLibraryNormalized.Square_Solid\",Color=(R=1,G=1,B=0,A=1),Trans ...(truncated)"
    Parent=RigVMVariableNode() {variable=Local master_ctrl}.Value
   ) {template="SpawnControl::Execute(in InitialValue,in Settings,in OffsetTransform,in OffsetSpace,in Parent,in Name,out Item)"}
   > n44 = &RigUnit_HierarchyAddControlVector2D(
    InitialValue=(X=0,Y=0)
    Name=CTRL_L_nose
    OffsetSpace=LocalSpace
    OffsetTransform=(Rotation=(X=0,Y=0,Z=0,W=1),Translation=(X=2,Y=0.5,Z=0),Scale3D=(X=0.5,Y=0.5,Z=0.5))
    Settings="(PrimaryAxis=Z,Limits=(LimitX=(bMinimum=True,bMaximum=True),LimitY=(bMinimum=True,bMaximum=True),MinValue=(X=-1,Y=-1),MaxValue=(X=1,Y=1),bDrawLimits=False),Shape=(bVisible=True,Name=\"DefaultGizmoLibraryNormalized.Circle_Solid\",Color=(R=1,G=1,B=0,A=1),Trans ...(truncated)"
    Parent=@n43.Item
   ) {template="SpawnControl::Execute(in InitialValue,in Settings,in OffsetTransform,in OffsetSpace,in Parent,in Name,out Item)"}
   > &RigUnit_HierarchyAddAnimationChannelFloat(
    InitialValue=0.0
    LimitsEnabled=(Enabled=(bMinimum=True,bMaximum=True))
    MaximumValue=1.0
    MinimumValue=0.0
    Name=noseWrinkleUpper
    Parent=@n44.Item
   ) {template="SpawnAnimationChannel::Execute(in InitialValue,in MinimumValue,in MaximumValue,in LimitsEnabled,in Parent,in Name,out Item)"}
   > n45 = &RigUnit_HierarchyAddControlVector2D(
    InitialValue=(X=0,Y=0)
    Name=CTRL_R_nose
    OffsetSpace=LocalSpace
    OffsetTransform=(Rotation=(X=0,Y=-1,Z=0,W=0),Translation=(X=-2,Y=0.5,Z=0),Scale3D=(X=0.5,Y=0.5,Z=0.5))
    Settings="(PrimaryAxis=Z,Limits=(LimitX=(bMinimum=True,bMaximum=True),LimitY=(bMinimum=True,bMaximum=True),MinValue=(X=-1,Y=-1),MaxValue=(X=1,Y=1),bDrawLimits=False),Shape=(bVisible=True,Name=\"DefaultGizmoLibraryNormalized.Circle_Solid\",Color=(R=1,G=1,B=0,A=1),Trans ...(truncated)"
    Parent=@n43.Item
   ) {template="SpawnControl::Execute(in InitialValue,in Settings,in OffsetTransform,in OffsetSpace,in Parent,in Name,out Item)"}
   > &RigUnit_HierarchyAddAnimationChannelFloat(
    InitialValue=0.0
    LimitsEnabled=(Enabled=(bMinimum=True,bMaximum=True))
    MaximumValue=1.0
    MinimumValue=0.0
    Name=noseWrinkleUpper
    Parent=@n45.Item
   ) {template="SpawnAnimationChannel::Execute(in InitialValue,in MinimumValue,in MaximumValue,in LimitsEnabled,in Parent,in Name,out Item)"}
   n46 = RigVMVariableNode() {variable=Local master_ctrl}
   > &RigUnit_HierarchyAddControlFloat(
    InitialValue=0.0
    Name=CTRL_L_eye_squintInner
    OffsetSpace=LocalSpace
    OffsetTransform=(Rotation=(X=-0.2706,Y=0.6533,Z=0.6533,W=0.2706),Translation=(X=3.5,Y=0,Z=15),Scale3D=(X=1,Y=1,Z=1))
    Settings="(PrimaryAxis=Y,bIsScale=False,Limits=(Limit=(bMinimum=True,bMaximum=True),MinValue=0,MaxValue=1,bDrawLimits=False),Shape=(bVisible=True,Name=\"DefaultGizmoLibraryNormalized.Triangle_Solid\",Color=(R=1,G=1,B=0,A=1),Transform=(Rotation=(X=-1,Y=0,Z=0,W=0),Trans ...(truncated)"
    Parent=@n46.Value
   ) {template="SpawnControl::Execute(in InitialValue,in Settings,in OffsetTransform,in OffsetSpace,in Parent,in Name,out Item)"}
   > &RigUnit_HierarchyAddControlFloat(
    InitialValue=0.0
    Name=CTRL_L_eye_cheekRaise
    OffsetSpace=LocalSpace
    OffsetTransform=(Rotation=(X=0,Y=-0.7071,Z=-0.7071,W=0),Translation=(X=5,Y=0,Z=14),Scale3D=(X=1,Y=1,Z=1))
    Settings="(PrimaryAxis=Y,bIsScale=False,Limits=(Limit=(bMinimum=True,bMaximum=True),MinValue=0,MaxValue=1,bDrawLimits=False),Shape=(bVisible=True,Name=\"DefaultGizmoLibraryNormalized.Triangle_Solid\",Color=(R=1,G=1,B=0,A=1),Transform=(Rotation=(X=-1,Y=0,Z=0,W=0),Trans ...(truncated)"
    Parent=@n46.Value
   ) {template="SpawnControl::Execute(in InitialValue,in Settings,in OffsetTransform,in OffsetSpace,in Parent,in Name,out Item)"}
   n47 = RigVMVariableNode() {variable=Local master_ctrl}
   > &RigUnit_HierarchyAddControlFloat(
    InitialValue=0.0
    Name=CTRL_R_eye_squintInner
    OffsetSpace=LocalSpace
    OffsetTransform=(Rotation=(X=-0.6533,Y=-0.2706,Z=0.2706,W=-0.6533),Translation=(X=-3.5,Y=0,Z=15),Scale3D=(X=1,Y=1,Z=1))
    Settings="(PrimaryAxis=Y,bIsScale=False,Limits=(Limit=(bMinimum=True,bMaximum=True),MinValue=0,MaxValue=1,bDrawLimits=False),Shape=(bVisible=True,Name=\"DefaultGizmoLibraryNormalized.Triangle_Solid\",Color=(R=1,G=1,B=0,A=1),Transform=(Rotation=(X=-1,Y=0,Z=0,W=0),Trans ...(truncated)"
    Parent=@n47.Value
   ) {template="SpawnControl::Execute(in InitialValue,in Settings,in OffsetTransform,in OffsetSpace,in Parent,in Name,out Item)"}
   > &RigUnit_HierarchyAddControlFloat(
    InitialValue=0.0
    Name=CTRL_R_eye_cheekRaise
    OffsetSpace=LocalSpace
    OffsetTransform=(Rotation=(X=-0.7071,Y=0,Z=0,W=-0.7071),Translation=(X=-5,Y=0,Z=14),Scale3D=(X=1,Y=1,Z=1))
    Settings="(PrimaryAxis=Y,bIsScale=False,Limits=(Limit=(bMinimum=True,bMaximum=True),MinValue=0,MaxValue=1,bDrawLimits=False),Shape=(bVisible=True,Name=\"DefaultGizmoLibraryNormalized.Triangle_Solid\",Color=(R=1,G=1,B=0,A=1),Transform=(Rotation=(X=-1,Y=0,Z=0,W=0),Trans ...(truncated)"
    Parent=@n47.Value
   ) {template="SpawnControl::Execute(in InitialValue,in Settings,in OffsetTransform,in OffsetSpace,in Parent,in Name,out Item)"}
   RigUnit_HierarchyAddNull = &RigUnit_HierarchyAddNull(
    Name=GRP_eyeControls
    Parent=(Type=Bone,Name="None")
    Space=LocalSpace
    Transform=(Rotation=(X=0,Y=0,Z=0,W=1),Translation=(X=17.7372,Y=7.4878,Z=140.2747),Scale3D=(X=1,Y=1,Z=1))
    ExecutePin=@RigVMFunction_Sequence_9.B
   )
   RigVMFunction_Sequence = &RigVMFunction_Sequence(ExecuteContext=@RigUnit_HierarchyAddNull.ExecutePin)
   RigUnit_HierarchyAddControlTransform = &RigUnit_HierarchyAddControlTransform(
    InitialValue=(Rotation=(X=0,Y=0,Z=0,W=1),Translation=(X=0,Y=0,Z=0),Scale3D=(X=1,Y=1,Z=1))
    Name=CTRL_C_eyesAim
    OffsetSpace=LocalSpace
    OffsetTransform=(Rotation=(X=0.7071,Y=0,Z=0.7071,W=0),Translation=(X=-17.7372,Y=22.5122,Z=16.5322),Scale3D=(X=1,Y=1,Z=1))
    Settings="(InitialSpace=LocalSpace,bUsePreferredRotationOrder=False,PreferredRotationOrder=YZX,Limits=(LimitTranslationX=(bMinimum=False,bMaximum=False),LimitTranslationY=(bMinimum=False,bMaximum=False),LimitTranslationZ=(bMinimum=False,bMaximum=False),LimitPitch=(b ...(truncated)"
    Parent=@RigUnit_HierarchyAddNull.Item
    ExecutePin=@RigVMFunction_Sequence.A
   ) {template="SpawnControl::Execute(in InitialValue,in Settings,in OffsetTransform,in OffsetSpace,in Parent,in Name,out Item)"}
   > n48 = &RigUnit_HierarchyAddNull(
    Name=CTRL_L_eyeAim_space
    Space=LocalSpace
    Transform=(Rotation=(X=0,Y=0,Z=0,W=1),Translation=(X=0,Y=0,Z=2.748),Scale3D=(X=1,Y=1,Z=1))
    Parent=@RigUnit_HierarchyAddControlTransform.Item
   )
   > &RigUnit_HierarchyAddControlTransform(
    InitialValue=(Rotation=(X=0,Y=0,Z=0,W=1),Translation=(X=0,Y=0,Z=0),Scale3D=(X=1,Y=1,Z=1))
    Name=CTRL_L_eyeAim
    OffsetSpace=LocalSpace
    OffsetTransform=(Rotation=(X=0,Y=0,Z=0,W=1),Translation=(X=0.0067,Y=0,Z=0.0632),Scale3D=(X=1,Y=1,Z=1))
    Settings="(InitialSpace=LocalSpace,bUsePreferredRotationOrder=False,PreferredRotationOrder=YZX,Limits=(LimitTranslationX=(bMinimum=False,bMaximum=False),LimitTranslationY=(bMinimum=False,bMaximum=False),LimitTranslationZ=(bMinimum=False,bMaximum=False),LimitPitch=(b ...(truncated)"
    Parent=@n48.Item
   ) {template="SpawnControl::Execute(in InitialValue,in Settings,in OffsetTransform,in OffsetSpace,in Parent,in Name,out Item)"}
   > n49 = &RigUnit_HierarchyAddNull(
    Name=CTRL_R_eyeAim_space
    Space=LocalSpace
    Transform=(Rotation=(X=0,Y=0,Z=0,W=1),Translation=(X=0,Y=0,Z=-2.748),Scale3D=(X=1,Y=1,Z=1))
    Parent=@RigUnit_HierarchyAddControlTransform.Item
   )
   > &RigUnit_HierarchyAddControlTransform(
    InitialValue=(Rotation=(X=0,Y=0,Z=0,W=1),Translation=(X=0,Y=0,Z=0),Scale3D=(X=1,Y=1,Z=1))
    Name=CTRL_R_eyeAim
    OffsetSpace=LocalSpace
    OffsetTransform=(Rotation=(X=0,Y=0,Z=0,W=1),Translation=(X=-0.0067,Y=0,Z=-0.0632),Scale3D=(X=1,Y=1,Z=1))
    Settings="(InitialSpace=LocalSpace,bUsePreferredRotationOrder=False,PreferredRotationOrder=YZX,Limits=(LimitTranslationX=(bMinimum=False,bMaximum=False),LimitTranslationY=(bMinimum=False,bMaximum=False),LimitTranslationZ=(bMinimum=False,bMaximum=False),LimitPitch=(b ...(truncated)"
    Parent=@n49.Item
   ) {template="SpawnControl::Execute(in InitialValue,in Settings,in OffsetTransform,in OffsetSpace,in Parent,in Name,out Item)"}
   > n50 = &RigUnit_HierarchyAddControlTransform(
    InitialValue=(Rotation=(X=0,Y=0,Z=0,W=1),Translation=(X=0,Y=0,Z=0),Scale3D=(X=1,Y=1,Z=1))
    Name=FRM_convergenceGUI
    OffsetSpace=LocalSpace
    OffsetTransform=(Rotation=(X=0,Y=0,Z=0,W=1),Translation=(X=-4.3604,Y=0,Z=0),Scale3D=(X=1.2,Y=1.2,Z=1.2))
    Settings="(InitialSpace=LocalSpace,bUsePreferredRotationOrder=False,PreferredRotationOrder=YZX,Limits=(LimitTranslationX=(bMinimum=False,bMaximum=False),LimitTranslationY=(bMinimum=False,bMaximum=False),LimitTranslationZ=(bMinimum=False,bMaximum=False),LimitPitch=(b ...(truncated)"
    Parent=@RigUnit_HierarchyAddControlTransform.Item
   ) {template="SpawnControl::Execute(in InitialValue,in Settings,in OffsetTransform,in OffsetSpace,in Parent,in Name,out Item)"}
   > n51 = &RigUnit_HierarchyAddNull(
    Name=CTRL_convergenceSwitch_space
    Space=LocalSpace
    Transform=(Rotation=(X=0,Y=-0.7071,Z=-0.7071,W=0),Translation=(X=-0.44,Y=0,Z=-1),Scale3D=(X=2,Y=2,Z=2))
    Parent=@n50.Item
   )
   > RigUnit_HierarchyAddControlFloat = &RigUnit_HierarchyAddControlFloat(
    InitialValue=0.0
    Name=CTRL_convergenceSwitch
    OffsetSpace=LocalSpace
    OffsetTransform=(Rotation=(X=0,Y=0,Z=0,W=1),Translation=(X=0,Y=0,Z=0),Scale3D=(X=1,Y=1,Z=1))
    Settings="(PrimaryAxis=Y,bIsScale=False,Limits=(Limit=(bMinimum=True,bMaximum=True),MinValue=0,MaxValue=1,bDrawLimits=True),Shape=(bVisible=True,Name=\"DefaultGizmoLibraryNormalized.Sphere_Solid\",Color=(R=1,G=1,B=0,A=0),Transform=(Rotation=(X=0,Y=0,Z=0,W=1),Translati ...(truncated)"
    Parent=@n51.Item
   ) {template="SpawnControl::Execute(in InitialValue,in Settings,in OffsetTransform,in OffsetSpace,in Parent,in Name,out Item)"}
   > RigUnit_HierarchyAddAnimationChannelFloat = &RigUnit_HierarchyAddAnimationChannelFloat(
    InitialValue=0.0
    LimitsEnabled=(Enabled=(bMinimum=True,bMaximum=True))
    MaximumValue=1.0
    MinimumValue=0.0
    Name=follow_head
    Parent=@RigUnit_HierarchyAddControlTransform.Item
   ) {template="SpawnAnimationChannel::Execute(in InitialValue,in MinimumValue,in MaximumValue,in LimitsEnabled,in Parent,in Name,out Item)"}
   n52 = &RigUnit_HierarchyAddNull(
    Name=EyeAim_GRP
    Space=LocalSpace
    Transform=(Rotation=(X=0,Y=0,Z=0,W=1),Translation=(X=7.2628,Y=-9.1751,Z=1.7596),Scale3D=(X=1,Y=1,Z=1))
    Parent=@RigUnit_HierarchyAddNull.Item
    ExecutePin=@RigVMFunction_Sequence.B
   )
   > n53 = &RigUnit_HierarchyAddNull(
    Name=EyeAim_L_GRP
    Space=LocalSpace
    Transform=(Rotation=(X=0.7071,Y=0,Z=0,W=-0.7071),Translation=(X=-22.2521,Y=7.2952,Z=14.7727),Scale3D=(X=1,Y=1,Z=1))
    Parent=@n52.Item
   )
   > &RigUnit_HierarchyAddNull(Name=EyeAim_L, Space=LocalSpace, Transform=(Rotation=(X=0,Y=0,Z=0,W=1),Translation=(X=0,Y=0,Z=0),Scale3D=(X=1,Y=1,Z=1)), Parent=@n53.Item)
   > n54 = &RigUnit_HierarchyAddNull(
    Name=EyeAim_R_GRP
    Space=LocalSpace
    Transform=(Rotation=(X=0.7071,Y=0,Z=0,W=-0.7071),Translation=(X=-27.748,Y=7.2952,Z=14.7726),Scale3D=(X=1,Y=1,Z=1))
    Parent=@n52.Item
   )
   > &RigUnit_HierarchyAddNull(Name=EyeAim_R, Space=LocalSpace, Transform=(Rotation=(X=0,Y=0,Z=0,W=1),Translation=(X=0,Y=0,Z=0),Scale3D=(X=1,Y=1,Z=1)), Parent=@n54.Item)
   n55 = &RigUnit_GetTransform(Item=(Type=Bone,Name="FACIAL_L_Eye"), bInitial=true)
   n56 = &RigUnit_GetTransform(Item=(Type=Bone,Name="FACIAL_R_Eye"), bInitial=true)
   n57 = /Script/RigVM.RigVMFunction_MathTransformLerp(T=0.5, A=@n55.Transform, B=@n56.Transform) {is=struct RigVMUnitNode, template="Interpolate::Execute(in A,in B,in T,out Result)"}
   &RigUnit_SetTransform(
    Item=(Type=Control,Name="CTRL_C_eyesAim")
    Value=(Rotation=(X=0.7071,Y=0,Z=0.7071,W=0),Translation=(X=0,Y=30,Z=0),Scale3D=(X=1,Y=1,Z=1))
    Weight=1
    bInitial=true
    Value.Translation.X=@n57.Result.Translation.X
    Value.Translation.Z=@n57.Result.Translation.Z
    ExecutePin=@RigVMFunction_Sequence_9.C
   )
   > &RigUnit_SetTransform(
    Item=(Type=Null,Name="CTRL_L_eyeAim_space")
    Value=(Rotation=(X=0.7071,Y=0,Z=0.7071,W=0),Translation=(X=0,Y=30,Z=0),Scale3D=(X=1,Y=1,Z=1))
    Weight=1
    bInitial=true
    Value.Translation.X=@n55.Transform.Translation.X
    Value.Translation.Z=@n55.Transform.Translation.Z
   )
   > &RigUnit_SetTransform(
    Item=(Type=Null,Name="CTRL_R_eyeAim_space")
    Value=(Rotation=(X=0.7071,Y=0,Z=0.7071,W=0),Translation=(X=0,Y=30,Z=0),Scale3D=(X=1,Y=1,Z=1))
    Weight=1
    bInitial=true
    Value.Translation.X=@n56.Transform.Translation.X
    Value.Translation.Z=@n56.Transform.Translation.Z
   )
   > &RigUnit_SetTransform(Item=(Type=Null,Name="EyeAim_L_GRP"), Weight=1, bInitial=true, Value=@n55.Transform)
   > &RigUnit_SetTransform(Item=(Type=Null,Name="EyeAim_R_GRP"), Weight=1, bInitial=true, Value=@n56.Transform)
   > &RigUnit_SetTransform(
    Item=(Type=Null,Name="Face_Ctrl_Master_Follow")
    Weight=1
    bInitial=true
    Value=&RigUnit_GetTransform(Item=(Type=Bone,Name="head"), bInitial=true).Transform
   )
   n58 = /Script/ControlRig.RigUnit_CollectionChildrenArray(Parent=(Type=Null,Name="Face_Ctrl_Master_GRP"), TypeToSearch=Control, bDefaultChildren=false, bIncludeParent=False, bRecursive=true) {is=struct RigVMUnitNode}
   > For_Each_1 = RigVMDispatchNode(Array=@n58.Items) {template="DISPATCH_RigVMDispatch_ArrayIterator(in Array,out Element,out Index,out Count,out Ratio)"}
   n59 = /Script/ControlRig.RigUnit_GetShapeTransform(Control=@For_Each_1.Element.Name) {is=struct RigVMUnitNode}
   Multiply = /Script/RigVM.RigVMFunction_MathVectorMul(
    A=@n59.Transform.Scale3D
    B=/Script/RigVM.RigVMFunction_MathVectorFromDouble(Value=RigVMVariableNode() {variable=Control Scale}.Value) {is=struct RigVMUnitNode, template="Cast::Execute(in Value,out Result)"}.Result
   ) {is=struct RigVMUnitNode, template="Multiply::Execute(in A,in B,out Result)"}
   > /Script/ControlRig.RigUnit_SetShapeTransform(
    Transform=(Rotation=(X=0,Y=0,Z=0,W=1),Translation=(X=0,Y=0,Z=0),Scale3D=(X=1,Y=1,Z=1))
    Transform.Rotation=@n59.Transform.Rotation
    Transform.Translation=@n59.Transform.Translation
    Transform.Scale3D=@Multiply.Result
    Control=@For_Each_1.Element.Name
   ) {is=struct RigVMUnitNode}
   > SetControlColor = /Script/ControlRig.RigUnit_SetControlColor(Color=RigVMVariableNode() {variable=Control Color}.Value, Control=@For_Each_1.Element.Name) {is=struct RigVMUnitNode, template="SetControlColor::Execute(in Control,in Color)"}
   RigVMFunctionReturnNode(ExecuteContext=@RigVMFunction_Sequence_9.D)
   /Script/ControlRig.RigUnit_SetMetadataTag(ExecutePin=(), Item=(Type=Bone,Name="None"), NameSpace=Self, Tag=wm_curves) {is=struct RigVMUnitNode}
  ]}
 New Function_ContainedGraph {nodes [
   RigVMFunctionEntryNode()
   > SequenceAggregate_4 = RigVMAggregateNode()
   /Script/ControlRig.RigUnit_SetCurveValue(Curve=DisableFaceOverride, Value=1, ExecutePin=@SequenceAggregate_4.A) {is=struct RigVMUnitNode}
   > &RigUnit_SetCurveValue(Curve=is_3L, Value=1)
   > SequenceAggregate = RigVMAggregateNode()
   SequenceAggregate_1 = RigVMAggregateNode(ExecuteContext=@SequenceAggregate.A)
   FN_FloatCtrlToCurve_FNC_15_2 = RigVMFunctionReferenceNode(
    Ctrl_L_Token=CTRL_L_
    Ctrl_R_Token=CTRL_R_
    Curve_Side_Token_Last=true
    Pos_Neg_Curves=((Type=Curve,Name="CTRL_expressions_eyeBlinkL"),(Type=Curve,Name="CTRL_expressions_eyeWidenL"))
    backward=False
    control=(Type=Control,Name="CTRL_L_eye_blink")
    mirror=true
    vector_axis=X
    ExecuteContext=@SequenceAggregate_1.C
   )
   > FN_VectorCtrlToCurve_FNC_2 = RigVMFunctionReferenceNode(
    Backward=false
    Control=(Type=Control,Name="CTRL_L_eye_eyelidU")
    Ctrl_L_Token=CTRL_L_
    Ctrl_R_Token=CTRL_R_
    Curve_Side_Token_Last=true
    X_Mirror=false
    X_Pos_Neg_Expr=((Type=Curve,Name="CTRL_expressions_eyelidRollNegUL"),(Type=Curve,Name="CTRL_expressions_eyelidRollPosUL"))
    Y_Mirror=true
    Y_Pos_Neg_Expr=((Type=Curve,Name="CTRL_expressions_eyeUpperLidUpL"),(Type=Curve,Name="CTRL_expressions_eyeRelaxL"))
    Z_Mirror=false
    Z_Pos_Neg_Expr=()
   )
   > FN_VectorCtrlToCurve_FNC_8 = RigVMFunctionReferenceNode(
    Backward=false
    Control=(Type=Control,Name="CTRL_R_eye_eyelidU")
    Ctrl_L_Token=CTRL_L_
    Ctrl_R_Token=CTRL_R_
    Curve_Side_Token_Last=true
    X_Mirror=false
    X_Pos_Neg_Expr=((Type=Curve,Name="CTRL_expressions_eyelidRollPosUR"),(Type=Curve,Name="CTRL_expressions_eyelidRollNegUR"))
    Y_Mirror=false
    Y_Pos_Neg_Expr=()
    Z_Mirror=false
    Z_Pos_Neg_Expr=()
   )
   > FN_VectorCtrlToCurve_FNC_3 = RigVMFunctionReferenceNode(
    Backward=false
    Control=(Type=Control,Name="CTRL_L_eye_eyelidD")
    Ctrl_L_Token=CTRL_L_
    Ctrl_R_Token=CTRL_R_
    Curve_Side_Token_Last=true
    X_Mirror=false
    X_Pos_Neg_Expr=((Type=Curve,Name="CTRL_expressions_eyelidRollPosDL"),(Type=Curve,Name="CTRL_expressions_eyelidRollNegDL"))
    Y_Mirror=true
    Y_Pos_Neg_Expr=((Type=Curve,Name="CTRL_expressions_eyeLowerLidUpL"),(Type=Curve,Name="CTRL_expressions_eyeLowerLidDownL"))
    Z_Mirror=false
    Z_Pos_Neg_Expr=()
   )
   > FN_VectorCtrlToCurve_FNC_9 = RigVMFunctionReferenceNode(
    Backward=false
    Control=(Type=Control,Name="CTRL_R_eye_eyelidD")
    Ctrl_L_Token=CTRL_L_
    Ctrl_R_Token=CTRL_R_
    Curve_Side_Token_Last=true
    X_Mirror=false
    X_Pos_Neg_Expr=((Type=Curve,Name="CTRL_expressions_eyelidRollNegDR"),(Type=Curve,Name="CTRL_expressions_eyelidRollPosDR"))
    Y_Mirror=false
    Y_Pos_Neg_Expr=()
    Z_Mirror=false
    Z_Pos_Neg_Expr=()
   )
   n1 = /Script/ControlRig.RigUnit_GetFloatAnimationChannel(Channel=lookAtSwitch, Control=Face_Ctrl_Master, bInitial=False) {is=struct RigVMUnitNode, template="GetAnimationChannel::Execute(out Value,in Control,in Channel,in bInitial)"}
   n2 = /Script/RigVM.RigVMFunction_MathQuaternionToRotator(Value=&RigUnit_GetTransform(Item=(Type=Null,Name="EyeAim_L"), Space=LocalSpace).Transform.Rotation) {is=struct RigVMUnitNode, template="Cast::Execute(in Value,out Result)"}
   n3 = /Script/ControlRig.RigUnit_GetControlVector2D(Control=CTRL_L_eye) {is=struct RigVMUnitNode, template="GetControlValue::Execute(in Control,out Vector,out Minimum,out Maximum)"}
   n4 = &RigUnit_GetControlVector2D(Control=CTRL_C_eye) {template="GetControlValue::Execute(in Control,out Vector,out Minimum,out Maximum)"}
   Add = &RigVMFunction_MathDoubleAdd(
    B=&RigUnit_GetFloatAnimationChannel(Channel=eyesConvergence, Control=CTRL_C_eye, bInitial=False) {template="GetAnimationChannel::Execute(out Value,in Control,in Channel,in bInitial)"}.Value
    A=&RigVMFunction_MathFloatAdd(B=@n3.Vector.X, A=@n4.Vector.X) {template="Add::Execute(in A,in B,out Result)"}.Result
   )
   n5 = /Script/RigVM.RigVMFunction_MathFloatLerp(
    T=@n1.Value
    B=&RigVMFunction_MathFloatRemap(SourceMaximum=-40, SourceMinimum=40, TargetMinimum=-1, bClamp=true, Value=@n2.Result.Pitch).Result
    A=@Add.Result
   ) {is=struct RigVMUnitNode, template="Interpolate::Execute(in A,in B,in T,out Result)"}
   Remap = &RigVMFunction_MathDoubleRemap(SourceMaximum=1, SourceMinimum=0, TargetMaximum=1, TargetMinimum=0, Value=@n5.Result)
   > &RigUnit_SetCurveValue(Curve=CTRL_expressions_eyeLookLeftL, Value=@Remap.Result)
   > &RigUnit_SetCurveValue(
    Curve=CTRL_expressions_eyeLookRightL
    Value=&RigVMFunction_MathDoubleRemap(SourceMaximum=-1, SourceMinimum=0, TargetMaximum=1, TargetMinimum=0, Value=@n5.Result).Result
   )
   n6 = &RigVMFunction_MathFloatLerp(
    T=@n1.Value
    B=&RigVMFunction_MathFloatRemap(SourceMaximum=50, SourceMinimum=-50, TargetMaximum=-1, TargetMinimum=1, bClamp=true, Value=@n2.Result.Roll).Result
    A=&RigVMFunction_MathFloatAdd(B=@n3.Vector.Y, A=@n4.Vector.Y) {template="Add::Execute(in A,in B,out Result)"}.Result
   ) {template="Interpolate::Execute(in A,in B,in T,out Result)"}
   > &RigUnit_SetCurveValue(
    Curve=CTRL_expressions_eyeLookUpL
    Value=&RigVMFunction_MathDoubleRemap(SourceMaximum=1, SourceMinimum=0, TargetMaximum=1, TargetMinimum=0, Value=@n6.Result).Result
   )
   > &RigUnit_SetCurveValue(
    Curve=CTRL_expressions_eyeLookDownL
    Value=&RigVMFunction_MathDoubleRemap(SourceMaximum=-1, SourceMinimum=0, TargetMaximum=1, TargetMinimum=0, Value=@n6.Result).Result
   )
   n7 = &RigUnit_GetFloatAnimationChannel(Channel=lookAtSwitch, Control=Face_Ctrl_Master, bInitial=False) {template="GetAnimationChannel::Execute(out Value,in Control,in Channel,in bInitial)"}
   n8 = &RigVMFunction_MathQuaternionToRotator(Value=&RigUnit_GetTransform(Item=(Type=Null,Name="EyeAim_R"), Space=LocalSpace).Transform.Rotation) {template="Cast::Execute(in Value,out Result)"}
   n9 = &RigUnit_GetControlVector2D(Control=CTRL_R_eye) {template="GetControlValue::Execute(in Control,out Vector,out Minimum,out Maximum)"}
   n10 = &RigUnit_GetControlVector2D(Control=CTRL_C_eye) {template="GetControlValue::Execute(in Control,out Vector,out Minimum,out Maximum)"}
   Subtract = &RigVMFunction_MathDoubleSub(
    B=&RigUnit_GetFloatAnimationChannel(Channel=eyesConvergence, Control=CTRL_C_eye, bInitial=False) {template="GetAnimationChannel::Execute(out Value,in Control,in Channel,in bInitial)"}.Value
    A=&RigVMFunction_MathFloatAdd(B=@n9.Vector.X, A=@n10.Vector.X) {template="Add::Execute(in A,in B,out Result)"}.Result
   )
   n11 = &RigVMFunction_MathFloatLerp(
    T=@n7.Value
    B=&RigVMFunction_MathFloatRemap(SourceMaximum=-40, SourceMinimum=40, TargetMinimum=-1, bClamp=true, Value=@n8.Result.Pitch).Result
    A=@Subtract.Result
   ) {template="Interpolate::Execute(in A,in B,in T,out Result)"}
   > &RigUnit_SetCurveValue(
    Curve=CTRL_expressions_eyeLookLeftR
    Value=&RigVMFunction_MathDoubleRemap(SourceMaximum=1, SourceMinimum=0, TargetMaximum=1, TargetMinimum=0, Value=@n11.Result).Result
   )
   > &RigUnit_SetCurveValue(
    Curve=CTRL_expressions_eyeLookRightR
    Value=&RigVMFunction_MathDoubleRemap(SourceMaximum=-1, SourceMinimum=0, TargetMaximum=1, TargetMinimum=0, Value=@n11.Result).Result
   )
   n12 = &RigVMFunction_MathFloatLerp(
    T=@n7.Value
    B=&RigVMFunction_MathFloatRemap(SourceMaximum=50, SourceMinimum=-50, TargetMaximum=-1, TargetMinimum=1, bClamp=true, Value=@n8.Result.Roll).Result
    A=&RigVMFunction_MathFloatAdd(B=@n9.Vector.Y, A=@n10.Vector.Y) {template="Add::Execute(in A,in B,out Result)"}.Result
   ) {template="Interpolate::Execute(in A,in B,in T,out Result)"}
   > &RigUnit_SetCurveValue(
    Curve=CTRL_expressions_eyeLookUpR
    Value=&RigVMFunction_MathDoubleRemap(SourceMaximum=1, SourceMinimum=0, TargetMaximum=1, TargetMinimum=0, Value=@n12.Result).Result
   )
   > &RigUnit_SetCurveValue(
    Curve=CTRL_expressions_eyeLookDownR
    Value=&RigVMFunction_MathDoubleRemap(SourceMaximum=-1, SourceMinimum=0, TargetMaximum=1, TargetMinimum=0, Value=@n12.Result).Result
   )
   n13 = /Script/ControlRig.RigUnit_ProjectTransformToNewParent(
    Child=(Type=Null,Name="CTRL_L_eyeAim_space")
    NewParent=(Type=Control,Name="CTRL_C_eyesAim")
    OldParent=(Type=Control,Name="CTRL_C_eyesAim")
    bChildInitial=True
    bOldParentInitial=True
   ) {is=struct RigVMUnitNode}
   n14 = /Script/ControlRig.RigUnit_GetControlFloat(Control=CTRL_convergenceSwitch) {is=struct RigVMUnitNode, template="GetControlValue::Execute(in Control,out FloatValue,out Minimum,out Maximum)"}
   n15 = &RigUnit_GetTransform(Item=(Type=Control,Name="CTRL_C_eyesAim"))
   > RigUnit_SetTransform = &RigUnit_SetTransform(
    Item=(Type=Null,Name="CTRL_L_eyeAim_space")
    Weight=1
    Value=&RigVMFunction_MathTransformLerp(A=@n13.Transform, T=@n14.FloatValue, B=@n15.Transform) {template="Interpolate::Execute(in A,in B,in T,out Result)"}.Result
   )
   n16 = &RigUnit_ProjectTransformToNewParent(
    Child=(Type=Null,Name="CTRL_R_eyeAim_space")
    NewParent=(Type=Control,Name="CTRL_C_eyesAim")
    OldParent=(Type=Control,Name="CTRL_C_eyesAim")
    bChildInitial=True
    bOldParentInitial=True
   )
   > &RigUnit_SetTransform(
    Item=(Type=Null,Name="CTRL_R_eyeAim_space")
    Weight=1
    Value=&RigVMFunction_MathTransformLerp(A=@n16.Transform, B=@n15.Transform, T=@n14.FloatValue) {template="Interpolate::Execute(in A,in B,in T,out Result)"}.Result
   )
   > CRFL AnimAttrToCurve v01_34 = RigVMFunctionReferenceNode(
    Backward=False
    Channel=eyelidsReduce
    Control=(Type=Control,Name="CTRL_C_eye")
    Pos_Neg_Curves=((Type=Curve,Name="CTRL_expressions_eyeEyelidsReduce"))
   )
   FN_FloatCtrlToCurve_FNC_15 = RigVMFunctionReferenceNode(
    Ctrl_L_Token=CTRL_L_
    Ctrl_R_Token=CTRL_R_
    Curve_Side_Token_Last=true
    Pos_Neg_Curves=((Type=Curve,Name="CTRL_expressions_browRaiseInL"),(Type=Curve,Name="CTRL_expressions_browDINL"))
    backward=False
    control=(Type=Control,Name="CTRL_L_brow_inner")
    mirror=true
    vector_axis=X
    ExecuteContext=@SequenceAggregate_1.B
   )
   > FN_VectorCtrlToCurve_FNC = RigVMFunctionReferenceNode(
    Backward=false
    Control=(Type=Control,Name="CTRL_L_brow")
    Ctrl_L_Token=CTRL_L_
    Ctrl_R_Token=CTRL_R_
    Curve_Side_Token_Last=true
    X_Mirror=true
    X_Pos_Neg_Expr=((Type=Curve,Name="CTRL_expressions_browLateralL"))
    Y_Mirror=true
    Y_Pos_Neg_Expr=((Type=Curve,Name="CTRL_expressions_browUINL"),(Type=Curve,Name="CTRL_expressions_browDownL"))
    Z_Mirror=false
    Z_Pos_Neg_Expr=()
   )
   > FN_VectorCtrlToCurve_FNC_1 = RigVMFunctionReferenceNode(
    Backward=false
    Control=(Type=Control,Name="CTRL_L_brow")
    Ctrl_L_Token=CTRL_L_
    Ctrl_R_Token=CTRL_R_
    Curve_Side_Token_Last=true
    X_Mirror=false
    X_Pos_Neg_Expr=()
    Y_Mirror=true
    Y_Pos_Neg_Expr=((Type=Curve,Name="CTRL_expressions_browUOUTL"))
    Z_Mirror=false
    Z_Pos_Neg_Expr=()
   )
   > FN_FloatCtrlToCurve_FNC_15_1 = RigVMFunctionReferenceNode(
    Ctrl_L_Token=CTRL_L_
    Ctrl_R_Token=CTRL_R_
    Curve_Side_Token_Last=true
    Pos_Neg_Curves=((Type=Curve,Name="CTRL_expressions_browRaiseOuterL"),(Type=Curve,Name="CTRL_expressions_browDOUTL"))
    backward=False
    control=(Type=Control,Name="CTRL_L_brow_outer")
    mirror=true
    vector_axis=X
   )
   n17 = &RigUnit_ProjectTransformToNewParent(
    Child=(Type=Null,Name="Face_Ctrl_Master_Follow")
    NewParent=(Type=Bone,Name="head")
    OldParent=(Type=Bone,Name="head")
    bChildInitial=True
    bOldParentInitial=True
   )
   &RigUnit_SetTransform(
    Item=(Type=Null,Name="Face_Ctrl_Master_Follow")
    Weight=1
    Value=&RigVMFunction_MathTransformLerp(
     T=&RigUnit_GetFloatAnimationChannel(Channel=gui_follow_head, Control=Face_Ctrl_Master, bInitial=false) {template="GetAnimationChannel::Execute(out Value,in Control,in Channel,in bInitial)"}.Value
     B=@n17.Transform
     A=&RigUnit_GetTransform(Item=(Type=Null,Name="Face_Ctrl_Master_Follow"), bInitial=true).Transform
    ) {template="Interpolate::Execute(in A,in B,in T,out Result)"}.Result
    ExecutePin=@SequenceAggregate_1.A
   )
   n18 = &RigUnit_ProjectTransformToNewParent(
    Child=(Type=Null,Name="GRP_eyeControls")
    NewParent=(Type=Bone,Name="head")
    OldParent=(Type=Bone,Name="head")
    bChildInitial=True
    bOldParentInitial=True
   )
   > &RigUnit_SetTransform(
    Item=(Type=Null,Name="GRP_eyeControls")
    Weight=1
    Value=&RigVMFunction_MathTransformLerp(
     T=&RigUnit_GetFloatAnimationChannel(Channel=follow_head, Control=CTRL_C_eyesAim, bInitial=False) {template="GetAnimationChannel::Execute(out Value,in Control,in Channel,in bInitial)"}.Value
     B=@n18.Transform
     A=&RigUnit_GetTransform(Item=(Type=Null,Name="GRP_eyeControls"), bInitial=true).Transform
    ) {template="Interpolate::Execute(in A,in B,in T,out Result)"}.Result
   )
   n19 = &RigUnit_ProjectTransformToNewParent(
    Child=(Type=Null,Name="EyeAim_GRP")
    NewParent=(Type=Bone,Name="head")
    OldParent=(Type=Bone,Name="head")
    bChildInitial=True
    bOldParentInitial=True
   )
   > &RigUnit_SetTransform(Item=(Type=Null,Name="EyeAim_GRP"), Weight=1, Value=@n19.Transform)
   > /Script/ControlRig.RigUnit_AimItem(
    DebugSettings=(Scale=10,WorldOffset=(Rotation=(X=0,Y=0,Z=0,W=1),Translation=(X=0,Y=0,Z=0),Scale3D=(X=1,Y=1,Z=1)))
    Item=(Type=Null,Name="EyeAim_L")
    Primary=(Weight=1,Axis=(X=0,Y=0,Z=1),Target=(X=1,Y=0,Z=0),Kind=Location,Space=(Type=Bone))
    Secondary=(Weight=1,Axis=(X=1,Y=0,Z=0),Target=(X=0,Y=0,Z=-1),Kind=Direction,Space=(Type=Bone,Name="head"))
    Weight=1
    Primary.Target=&RigUnit_GetTransform(Item=(Type=Control,Name="CTRL_L_eyeAim")).Transform.Translation
   ) {is=struct RigVMUnitNode}
   > &RigUnit_AimItem(
    DebugSettings=(Scale=10,WorldOffset=(Rotation=(X=0,Y=0,Z=0,W=1),Translation=(X=0,Y=0,Z=0),Scale3D=(X=1,Y=1,Z=1)))
    Item=(Type=Null,Name="EyeAim_R")
    Primary=(Weight=1,Axis=(X=0,Y=0,Z=1),Target=(X=1,Y=0,Z=0),Kind=Location,Space=(Type=Bone))
    Secondary=(Weight=1,Axis=(X=1,Y=0,Z=0),Target=(X=0,Y=0,Z=-1),Kind=Direction,Space=(Type=Bone,Name="head"))
    Weight=1
    Primary.Target=&RigUnit_GetTransform(Item=(Type=Control,Name="CTRL_R_eyeAim")).Transform.Translation
   )
   RerouteNode_7 = RigVMDispatchNode(Values=((Type=Control,Name="CTRL_C_eye"),((),Type=Control,Name="CTRL_L_eye"),((),Type=Control,Name="CTRL_R_eye"))) {template="DISPATCH_RigVMDispatch_ArrayMake(in Values,out Array)"}
   > n20 = RigVMDispatchNode(Array=@RerouteNode_7.Array) {template="DISPATCH_RigVMDispatch_ArrayIterator(in Array,out Element,out Index,out Count,out Ratio)"}
   GetAnimationChannel = &RigUnit_GetFloatAnimationChannel(Channel=lookAtSwitch, Control=Face_Ctrl_Master, bInitial=False) {template="GetAnimationChannel::Execute(out Value,in Control,in Channel,in bInitial)"}
   Greater = /Script/RigVM.RigVMFunction_MathDoubleGreater(B=0.5, A=@GetAnimationChannel.Value) {is=struct RigVMUnitNode, template="Greater::Execute(in A,in B,out Result)"}
   > /Script/ControlRig.RigUnit_SetControlVisibility(Item=@n20.Element, bVisible=&RigVMFunction_MathBoolNot(Value=@Greater.Result).Result) {is=struct RigVMUnitNode, template="SetControlVisibility::Execute(in Item,in Pattern,in bVisible)"}
   RerouteNode_1 = RigVMDispatchNode(Values=((Type=Control,Name="CTRL_C_eyesAim"),((),Type=Control,Name="CTRL_L_eyeAim"),((),Type=Control,Name="CTRL_R_eyeAim"),((),Type=Control,Name="FRM_convergenceGUI"),((),Type=Control,Name="CTRL_convergenceSwitch"))) {template="DISPATCH_RigVMDispatch_ArrayMake(in Values,out Array)"}
   > n21 = RigVMDispatchNode(Array=@RerouteNode_1.Array) {template="DISPATCH_RigVMDispatch_ArrayIterator(in Array,out Element,out Index,out Count,out Ratio)"}
   > SetControlVisibility = &RigUnit_SetControlVisibility(Item=@n21.Element, bVisible=@Greater.Result) {template="SetControlVisibility::Execute(in Item,in Pattern,in bVisible)"}
   Branch_RigVMUnitNode = &RigVMFunction_ControlFlowBranch(ExecuteContext=@SetControlVisibility.ExecutePin, Condition=@Greater.Result)
   DebugLineItemSpace = /Script/RigVM.RigVMFunction_DebugLineNoSpace(
    Color=(R=0,G=0,B=0,A=1)
    DebugDrawSettings=(DepthPriority=SDPG_Foreground,Lifetime=-1)
    Thickness=0.05
    WorldOffset=(Rotation=(X=0,Y=0,Z=0,W=1),Translation=(X=0,Y=0,Z=0),Scale3D=(X=1,Y=1,Z=1))
    bEnabled=True
    ExecutePin=@Branch_RigVMUnitNode.True
    A=&RigUnit_GetTransform(Item=(Type=Bone,Name="FACIAL_L_Eye")).Transform.Translation
    B=&RigUnit_GetTransform(Item=(Type=Control,Name="CTRL_L_eyeAim")).Transform.Translation
   ) {is=struct RigVMUnitNode}
   > DebugLineItemSpace_1 = &RigVMFunction_DebugLineNoSpace(
    Color=(R=0,G=0,B=0,A=1)
    DebugDrawSettings=(DepthPriority=SDPG_Foreground,Lifetime=-1)
    Thickness=0.05
    WorldOffset=(Rotation=(X=0,Y=0,Z=0,W=1),Translation=(X=0,Y=0,Z=0),Scale3D=(X=1,Y=1,Z=1))
    bEnabled=True
    B=&RigUnit_GetTransform(Item=(Type=Control,Name="CTRL_R_eyeAim")).Transform.Translation
    A=&RigUnit_GetTransform(Item=(Type=Bone,Name="FACIAL_R_Eye")).Transform.Translation
   )
   SequenceAggregate_2 = RigVMAggregateNode(ExecuteContext=@SequenceAggregate.B)
   FN_VectorCtrlToCurve_FNC_4 = RigVMFunctionReferenceNode(
    Backward=false
    Control=(Type=Control,Name="CTRL_C_nose")
    Ctrl_L_Token=CTRL_L_
    Ctrl_R_Token=CTRL_R_
    Curve_Side_Token_Last=true
    X_Mirror=false
    X_Pos_Neg_Expr=((Type=Curve,Name="CTRL_expressions_noseRotateL"),(Type=Curve,Name="CTRL_expressions_noseRotateR"))
    Y_Mirror=false
    Y_Pos_Neg_Expr=((Type=Curve,Name="CTRL_expressions_noseSquash"),(Type=Curve,Name="CTRL_expressions_noseStretch"))
    Z_Mirror=false
    Z_Pos_Neg_Expr=()
    ExecuteContext=@SequenceAggregate_2.A
   )
   > FN_VectorCtrlToCurve_FNC_5 = RigVMFunctionReferenceNode(
    Backward=false
    Control=(Type=Control,Name="CTRL_L_nose")
    Ctrl_L_Token=CTRL_L_
    Ctrl_R_Token=CTRL_R_
    Curve_Side_Token_Last=true
    X_Mirror=true
    X_Pos_Neg_Expr=((Type=Curve,Name="CTRL_expressions_noseNostrilDilateL"),(Type=Curve,Name="CTRL_expressions_noseNostrilCompressL"))
    Y_Mirror=true
    Y_Pos_Neg_Expr=((Type=Curve,Name="CTRL_expressions_noseWrinkleL"),(Type=Curve,Name="CTRL_expressions_noseNostrilDepressL"))
    Z_Mirror=false
    Z_Pos_Neg_Expr=()
   )
   > CRFL AnimAttrToCurve v01_30 = RigVMFunctionReferenceNode(
    Backward=False
    Channel=noseWrinkleUpper
    Control=(Type=Control,Name="CTRL_L_nose")
    Pos_Neg_Curves=((Type=Curve,Name="CTRL_expressions_noseWrinkleUpperL"))
   )
   > CRFL AnimAttrToCurve v01_31 = RigVMFunctionReferenceNode(
    Backward=False
    Channel=noseWrinkleUpper
    Control=(Type=Control,Name="CTRL_R_nose")
    Pos_Neg_Curves=((Type=Curve,Name="CTRL_expressions_noseWrinkleUpperR"))
   )
   FN_FloatCtrlToCurve_FNC_15_2_1 = RigVMFunctionReferenceNode(
    Ctrl_L_Token=CTRL_L_
    Ctrl_R_Token=CTRL_R_
    Curve_Side_Token_Last=true
    Pos_Neg_Curves=((Type=Curve,Name="CTRL_expressions_eyeSquintInnerL"))
    backward=False
    control=(Type=Control,Name="CTRL_L_eye_squintInner")
    mirror=true
    vector_axis=X
    ExecuteContext=@SequenceAggregate_2.B
   )
   > FN_FloatCtrlToCurve_FNC_15_2_1_1 = RigVMFunctionReferenceNode(
    Ctrl_L_Token=CTRL_L_
    Ctrl_R_Token=CTRL_R_
    Curve_Side_Token_Last=true
    Pos_Neg_Curves=((Type=Curve,Name="CTRL_expressions_eyeCheekRaiseL"))
    backward=False
    control=(Type=Control,Name="CTRL_L_eye_cheekRaise")
    mirror=true
    vector_axis=X
   )
   > FN_FloatCtrlToCurve_FNC_15_2_1_1_1 = RigVMFunctionReferenceNode(
    Ctrl_L_Token=CTRL_L_
    Ctrl_R_Token=CTRL_R_
    Curve_Side_Token_Last=true
    Pos_Neg_Curves=((Type=Curve,Name="CTRL_expressions_mouthCheekBlowL"),(Type=Curve,Name="CTRL_expressions_mouthCheekSuckL"))
    backward=False
    control=(Type=Control,Name="CTRL_L_mouth_suckBlow")
    mirror=true
    vector_axis=X
   )
   FN_VectorCtrlToCurve_FNC_6 = RigVMFunctionReferenceNode(
    Backward=false
    Control=(Type=Control,Name="CTRL_C_mouth")
    Ctrl_L_Token=CTRL_L_
    Ctrl_R_Token=CTRL_R_
    Curve_Side_Token_Last=true
    X_Mirror=false
    X_Pos_Neg_Expr=((Type=Curve,Name="CTRL_expressions_mouthLeft"),(Type=Curve,Name="CTRL_expressions_mouthRight"))
    Y_Mirror=false
    Y_Pos_Neg_Expr=((Type=Curve,Name="CTRL_expressions_mouthUp"),(Type=Curve,Name="CTRL_expressions_mouthDown"))
    Z_Mirror=false
    Z_Pos_Neg_Expr=()
    ExecuteContext=@SequenceAggregate_2.C
   )
   > FN_VectorCtrlToCurve_FNC_7 = RigVMFunctionReferenceNode(
    Backward=false
    Control=(Type=Control,Name="CTRL_L_mouth_lipU")
    Ctrl_L_Token=CTRL_L_
    Ctrl_R_Token=CTRL_R_
    Curve_Side_Token_Last=true
    X_Mirror=true
    X_Pos_Neg_Expr=((Type=Curve,Name="CTRL_expressions_mouthLipsTowardsUL"),(Type=Curve,Name="CTRL_expressions_mouthUpperLipBiteL"))
    Y_Mirror=true
    Y_Pos_Neg_Expr=((Type=Curve,Name="CTRL_expressions_mouthUpperLipRaiseL"),(Type=Curve,Name="CTRL_expressions_mouthUpperLipTowardsTeethL"))
    Z_Mirror=false
    Z_Pos_Neg_Expr=()
   )
   > CRFL AnimAttrToCurve v01_22 = RigVMFunctionReferenceNode(
    Backward=False
    Channel=mouthPress
    Control=(Type=Control,Name="CTRL_L_mouth_lipU")
    Pos_Neg_Curves=((Type=Curve,Name="CTRL_expressions_mouthPressUL"))
   )
   > CRFL AnimAttrToCurve v01_23 = RigVMFunctionReferenceNode(
    Backward=False
    Channel=lipsTogether
    Control=(Type=Control,Name="CTRL_L_mouth_lipU")
    Pos_Neg_Curves=((Type=Curve,Name="CTRL_expressions_mouthLipsTogetherUL"))
   )
   > CRFL AnimAttrToCurve v01_17 = RigVMFunctionReferenceNode(
    Backward=False
    Channel=mouthPress
    Control=(Type=Control,Name="CTRL_R_mouth_lipU")
    Pos_Neg_Curves=((Type=Curve,Name="CTRL_expressions_mouthPressUR"))
   )
   > CRFL AnimAttrToCurve v01_16 = RigVMFunctionReferenceNode(
    Backward=False
    Channel=lipsTogether
    Control=(Type=Control,Name="CTRL_R_mouth_lipU")
    Pos_Neg_Curves=((Type=Curve,Name="CTRL_expressions_mouthLipsTogetherUR"))
   )
   > FN_VectorCtrlToCurve_FNC_11 = RigVMFunctionReferenceNode(
    Backward=false
    Control=(Type=Control,Name="CTRL_L_mouth_lipD")
    Ctrl_L_Token=CTRL_L_
    Ctrl_R_Token=CTRL_R_
    Curve_Side_Token_Last=true
    X_Mirror=true
    X_Pos_Neg_Expr=((Type=Curve,Name="CTRL_expressions_mouthLipsTowardsDL"),(Type=Curve,Name="CTRL_expressions_mouthLowerLipBiteL"))
    Y_Mirror=true
    Y_Pos_Neg_Expr=((Type=Curve,Name="CTRL_expressions_mouthLowerLipDepressL"),(Type=Curve,Name="CTRL_expressions_mouthLowerLipTowardsTeethL"))
    Z_Mirror=false
    Z_Pos_Neg_Expr=()
   )
   > CRFL AnimAttrToCurve v01_20 = RigVMFunctionReferenceNode(
    Backward=False
    Channel=mouthPress
    Control=(Type=Control,Name="CTRL_L_mouth_lipD")
    Pos_Neg_Curves=((Type=Curve,Name="CTRL_expressions_mouthPressDL"))
   )
   > CRFL AnimAttrToCurve v01_21 = RigVMFunctionReferenceNode(
    Backward=False
    Channel=lipsTogether
    Control=(Type=Control,Name="CTRL_L_mouth_lipD")
    Pos_Neg_Curves=((Type=Curve,Name="CTRL_expressions_mouthLipsTogetherDL"))
   )
   > CRFL AnimAttrToCurve v01_18 = RigVMFunctionReferenceNode(
    Backward=False
    Channel=mouthPress
    Control=(Type=Control,Name="CTRL_R_mouth_lipD")
    Pos_Neg_Curves=((Type=Curve,Name="CTRL_expressions_mouthPressDR"))
   )
   > CRFL AnimAttrToCurve v01_19 = RigVMFunctionReferenceNode(
    Backward=False
    Channel=lipsTogether
    Control=(Type=Control,Name="CTRL_R_mouth_lipD")
    Pos_Neg_Curves=((Type=Curve,Name="CTRL_expressions_mouthLipsTogetherDR"))
   )
   > FN_VectorCtrlToCurve_FNC_12 = RigVMFunctionReferenceNode(
    Backward=false
    Control=(Type=Control,Name="CTRL_L_mouth_smileDepress")
    Ctrl_L_Token=CTRL_L_
    Ctrl_R_Token=CTRL_R_
    Curve_Side_Token_Last=true
    X_Mirror=true
    X_Pos_Neg_Expr=((Type=Curve,Name="CTRL_expressions_mouthCornerDepressL"))
    Y_Mirror=true
    Y_Pos_Neg_Expr=((Type=Curve,Name="CTRL_expressions_mouthCornerPullL"))
    Z_Mirror=false
    Z_Pos_Neg_Expr=()
   )
   > CRFL AnimAttrToCurve v01_6 = RigVMFunctionReferenceNode(
    Backward=False
    Channel=mouthDimple
    Control=(Type=Control,Name="CTRL_L_mouth_smileDepress")
    Pos_Neg_Curves=((Type=Curve,Name="CTRL_expressions_mouthDimpleL"))
   )
   > CRFL AnimAttrToCurve v01_7 = RigVMFunctionReferenceNode(
    Backward=False
    Channel=mouthStretch
    Control=(Type=Control,Name="CTRL_L_mouth_smileDepress")
    Pos_Neg_Curves=((Type=Curve,Name="CTRL_expressions_mouthStretchL"))
   )
   > CRFL AnimAttrToCurve v01_5 = RigVMFunctionReferenceNode(
    Backward=False
    Channel=mouthPurseU
    Control=(Type=Control,Name="CTRL_L_mouth_smileDepress")
    Pos_Neg_Curves=((Type=Curve,Name="CTRL_expressions_mouthLipsPurseUL"))
   )
   > CRFL AnimAttrToCurve v01_4 = RigVMFunctionReferenceNode(
    Backward=False
    Channel=mouthPurseD
    Control=(Type=Control,Name="CTRL_L_mouth_smileDepress")
    Pos_Neg_Curves=((Type=Curve,Name="CTRL_expressions_mouthLipsPurseDL"))
   )
   > CRFL AnimAttrToCurve v01_3 = RigVMFunctionReferenceNode(
    Backward=False
    Channel=mouthFunnelU
    Control=(Type=Control,Name="CTRL_L_mouth_smileDepress")
    Pos_Neg_Curves=((Type=Curve,Name="CTRL_expressions_mouthFunnelUL"))
   )
   > CRFL AnimAttrToCurve v01_2 = RigVMFunctionReferenceNode(
    Backward=False
    Channel=mouthFunnelD
    Control=(Type=Control,Name="CTRL_L_mouth_smileDepress")
    Pos_Neg_Curves=((Type=Curve,Name="CTRL_expressions_mouthFunnelDL"))
   )
   > CRFL AnimAttrToCurve v01 = RigVMFunctionReferenceNode(
    Backward=False
    Channel=mouthCornerSharpnessU
    Control=(Type=Control,Name="CTRL_L_mouth_smileDepress")
    Pos_Neg_Curves=((Type=Curve,Name="CTRL_expressions_mouthCornerRounderUL"),(Type=Curve,Name="CTRL_expressions_mouthCornerSharpenUL"))
   )
   > CRFL AnimAttrToCurve v01_1 = RigVMFunctionReferenceNode(
    Backward=False
    Channel=mouthCornerSharpnessD
    Control=(Type=Control,Name="CTRL_L_mouth_smileDepress")
    Pos_Neg_Curves=((Type=Curve,Name="CTRL_expressions_mouthCornerRounderDL"),(Type=Curve,Name="CTRL_expressions_mouthCornerSharpenDL"))
   )
   > CRFL AnimAttrToCurve v01_14 = RigVMFunctionReferenceNode(
    Backward=False
    Channel=mouthDimple
    Control=(Type=Control,Name="CTRL_R_mouth_smileDepress")
    Pos_Neg_Curves=((Type=Curve,Name="CTRL_expressions_mouthDimpleR"))
   )
   > CRFL AnimAttrToCurve v01_15 = RigVMFunctionReferenceNode(
    Backward=False
    Channel=mouthStretch
    Control=(Type=Control,Name="CTRL_R_mouth_smileDepress")
    Pos_Neg_Curves=((Type=Curve,Name="CTRL_expressions_mouthStretchR"))
   )
   > CRFL AnimAttrToCurve v01_13 = RigVMFunctionReferenceNode(
    Backward=False
    Channel=mouthPurseU
    Control=(Type=Control,Name="CTRL_R_mouth_smileDepress")
    Pos_Neg_Curves=((Type=Curve,Name="CTRL_expressions_mouthLipsPurseUR"))
   )
   > CRFL AnimAttrToCurve v01_12 = RigVMFunctionReferenceNode(
    Backward=False
    Channel=mouthPurseD
    Control=(Type=Control,Name="CTRL_R_mouth_smileDepress")
    Pos_Neg_Curves=((Type=Curve,Name="CTRL_expressions_mouthLipsPurseDR"))
   )
   > CRFL AnimAttrToCurve v01_11 = RigVMFunctionReferenceNode(
    Backward=False
    Channel=mouthFunnelU
    Control=(Type=Control,Name="CTRL_R_mouth_smileDepress")
    Pos_Neg_Curves=((Type=Curve,Name="CTRL_expressions_mouthFunnelUR"))
   )
   > CRFL AnimAttrToCurve v01_10 = RigVMFunctionReferenceNode(
    Backward=False
    Channel=mouthFunnelD
    Control=(Type=Control,Name="CTRL_R_mouth_smileDepress")
    Pos_Neg_Curves=((Type=Curve,Name="CTRL_expressions_mouthFunnelDR"))
   )
   > CRFL AnimAttrToCurve v01_8 = RigVMFunctionReferenceNode(
    Backward=False
    Channel=mouthCornerSharpnessU
    Control=(Type=Control,Name="CTRL_R_mouth_smileDepress")
    Pos_Neg_Curves=((Type=Curve,Name="CTRL_expressions_mouthCornerRounderUR"),(Type=Curve,Name="CTRL_expressions_mouthCornerSharpenUR"))
   )
   > CRFL AnimAttrToCurve v01_9 = RigVMFunctionReferenceNode(
    Backward=False
    Channel=mouthCornerSharpnessD
    Control=(Type=Control,Name="CTRL_R_mouth_smileDepress")
    Pos_Neg_Curves=((Type=Curve,Name="CTRL_expressions_mouthCornerRounderDR"),(Type=Curve,Name="CTRL_expressions_mouthCornerSharpenDR"))
   )
   > FN_VectorCtrlToCurve_FNC_13 = RigVMFunctionReferenceNode(
    Backward=false
    Control=(Type=Control,Name="CTRL_L_mouth_corner")
    Ctrl_L_Token=CTRL_L_
    Ctrl_R_Token=CTRL_R_
    Curve_Side_Token_Last=true
    X_Mirror=true
    X_Pos_Neg_Expr=((Type=Curve,Name="CTRL_expressions_mouthCornerWideL"),(Type=Curve,Name="CTRL_expressions_mouthCornerNarrowL"))
    Y_Mirror=true
    Y_Pos_Neg_Expr=((Type=Curve,Name="CTRL_expressions_mouthCornerUpL"),(Type=Curve,Name="CTRL_expressions_mouthCornerDownL"))
    Z_Mirror=false
    Z_Pos_Neg_Expr=()
   )
   > CRFL AnimAttrToCurve v01_32 = RigVMFunctionReferenceNode(
    Backward=False
    Channel=lipsBlow
    Control=(Type=Control,Name="CTRL_L_mouth_corner")
    Pos_Neg_Curves=((Type=Curve,Name="CTRL_expressions_mouthLipsBlowL"))
   )
   > CRFL AnimAttrToCurve v01_33 = RigVMFunctionReferenceNode(
    Backward=False
    Channel=lipsBlow
    Control=(Type=Control,Name="CTRL_R_mouth_corner")
    Pos_Neg_Curves=((Type=Curve,Name="CTRL_expressions_mouthLipsBlowR"))
   )
   SequenceAggregate_3 = &RigVMFunction_Sequence(ExecuteContext=@SequenceAggregate.C)
   n22 = /Script/ControlRig.RigUnit_GetControlVector(Control=CTRL_C_jaw, Space=LocalSpace) {is=struct RigVMUnitNode, template="GetControlValue::Execute(in Control,in Space,out Vector,out Minimum,out Maximum)"}
   Set Transform = /Script/ControlRig.RigUnit_SetTranslation(
    Item=(Type=Null,Name="CTRL_L_mouth_lipD_offset")
    Space=LocalSpace
    Value=(X=0,Y=0,Z=0)
    Weight=1
    bInitial=false
    bPropagateToChildren=True
    Value.Z=@n22.Vector.X
    Value.Y=@n22.Vector.Y
    ExecutePin=@SequenceAggregate_3.B
   ) {is=struct RigVMUnitNode, template="Set Transform::Execute(in Item,in Space,in bInitial,in Value,in Weight,in bPropagateToChildren)"}
   > Set Transform_1 = &RigUnit_SetTranslation(
    Item=(Type=Null,Name="CTRL_R_mouth_lipD_offset")
    Space=LocalSpace
    Value=(X=0,Y=0,Z=0)
    Weight=1
    bInitial=false
    bPropagateToChildren=True
    Value.Z=@n22.Vector.X
    Value.Y=@n22.Vector.Y
   ) {template="Set Transform::Execute(in Item,in Space,in bInitial,in Value,in Weight,in bPropagateToChildren)"}
   Multiply = &RigVMFunction_MathDoubleMul(
    B=-1
    A=&RigUnit_GetControlFloat(Control=CTRL_L_eye_blink) {template="GetControlValue::Execute(in Control,out FloatValue,out Minimum,out Maximum)"}.FloatValue
   )
   > Set Transform_2 = &RigUnit_SetTranslation(
    Item=(Type=Null,Name="CTRL_L_eye_eyelidU_offset")
    Space=LocalSpace
    Value=(X=0,Y=0,Z=0)
    Weight=1
    bInitial=false
    bPropagateToChildren=True
    Value.Y=@Multiply.Result
   ) {template="Set Transform::Execute(in Item,in Space,in bInitial,in Value,in Weight,in bPropagateToChildren)"}
   > Set Transform_2_1 = &RigUnit_SetTranslation(
    Item=(Type=Null,Name="CTRL_R_eye_eyelidU_offset")
    Space=LocalSpace
    Value=(X=0,Y=0,Z=0)
    Weight=1
    bInitial=false
    bPropagateToChildren=True
    Value.Y=&RigVMFunction_MathDoubleMul(
     B=-1
     A=&RigUnit_GetControlFloat(Control=CTRL_R_eye_blink) {template="GetControlValue::Execute(in Control,out FloatValue,out Minimum,out Maximum)"}.FloatValue
    ).Result
   ) {template="Set Transform::Execute(in Item,in Space,in bInitial,in Value,in Weight,in bPropagateToChildren)"}
   > For_Each_1 = RigVMDispatchNode(Array=&RigUnit_CollectionNameSearchArray(PartialName=_Offset_, TypeToSearch=Curve).Items) {template="DISPATCH_RigVMDispatch_ArrayIterator(in Array,out Element,out Index,out Count,out Ratio)"}
   StartsWith = &RigVMFunction_StartsWith(Start=WM_, Name=@For_Each_1.Element.Name) {template="StartsWith::Execute(in Name,in Start,out Result)"}
   n23 = &RigVMFunction_ControlFlowBranch(ExecuteContext=@For_Each_1.ExecuteContext, Condition=@StartsWith.Result)
   &RigUnit_SetCurveValue(
    Value=&RigUnit_GetFloatAnimationChannel(Control=CTRL_WrinkleMap, bInitial=False, Channel=@For_Each_1.Element.Name) {template="GetAnimationChannel::Execute(out Value,in Control,in Channel,in bInitial)"}.Value
    ExecutePin=@n23.True
    Curve=@For_Each_1.Element.Name
   )
   FN_VectorCtrlToCurve_FNC_10 = RigVMFunctionReferenceNode(
    Backward=false
    Control=(Type=Control,Name="CTRL_C_jaw")
    Ctrl_L_Token=CTRL_L_
    Ctrl_R_Token=CTRL_R_
    Curve_Side_Token_Last=true
    X_Mirror=false
    X_Pos_Neg_Expr=((Type=Curve,Name="CTRL_expressions_jawRight"),(Type=Curve,Name="CTRL_expressions_jawLeft"))
    Y_Mirror=false
    Y_Pos_Neg_Expr=((Type=Curve,Name="CTRL_expressions_jawOpen"))
    Z_Mirror=false
    Z_Pos_Neg_Expr=((Type=Curve,Name="CTRL_expressions_jawBack"),(Type=Curve,Name="CTRL_expressions_jawFwd"))
    ExecuteContext=@SequenceAggregate_3.A
   )
   n24 = &RigUnit_GetControlVector2D(Control=CTRL_C_jaw) {template="GetControlValue::Execute(in Control,out Vector,out Minimum,out Maximum)"}
   RigUnit_AlphaInterp_1 = &RigVMFunction_AlphaInterp(
    InRange=(Min=0,Max=-1)
    InterpSpeedDecreasing=10
    InterpSpeedIncreasing=10
    Scale=1
    bClampResult=True
    bInterpResult=False
    bMapRange=True
    Value=&RigVMFunction_MathDoubleSub(
     B=&RigUnit_GetFloatAnimationChannel(Channel=chinRaiseDL, Control=CTRL_C_jaw, bInitial=False) {template="GetAnimationChannel::Execute(out Value,in Control,in Channel,in bInitial)"}.Value
     A=@n24.Vector.Y
    ).Result
   )
   > &RigUnit_SetCurveValue(Curve=CTRL_expressions_jawChinRaiseDL, Value=@RigUnit_AlphaInterp_1.Result)
   RigUnit_AlphaInterp_4_2 = &RigVMFunction_AlphaInterp(
    InRange=(Min=0,Max=-1)
    InterpSpeedDecreasing=10
    InterpSpeedIncreasing=10
    Scale=1
    bClampResult=True
    bInterpResult=False
    bMapRange=True
    Value=&RigVMFunction_MathDoubleSub(
     B=&RigUnit_GetFloatAnimationChannel(Channel=chinRaiseDR, Control=CTRL_C_jaw, bInitial=False) {template="GetAnimationChannel::Execute(out Value,in Control,in Channel,in bInitial)"}.Value
     A=@n24.Vector.Y
    ).Result
   )
   > &RigUnit_SetCurveValue(Curve=CTRL_expressions_jawChinRaiseDR, Value=@RigUnit_AlphaInterp_4_2.Result)
   RigUnit_AlphaInterp_4_1_2 = &RigVMFunction_AlphaInterp(
    InRange=(Min=0,Max=-1)
    InterpSpeedDecreasing=10
    InterpSpeedIncreasing=10
    Scale=1
    bClampResult=True
    bInterpResult=False
    bMapRange=True
    Value=&RigVMFunction_MathDoubleSub(
     B=&RigUnit_GetFloatAnimationChannel(Channel=chinRaiseUL, Control=CTRL_C_jaw, bInitial=False) {template="GetAnimationChannel::Execute(out Value,in Control,in Channel,in bInitial)"}.Value
     A=@n24.Vector.Y
    ).Result
   )
   > &RigUnit_SetCurveValue(Curve=CTRL_expressions_jawChinRaiseUL, Value=@RigUnit_AlphaInterp_4_1_2.Result)
   RigUnit_AlphaInterp_4_1_1_2 = &RigVMFunction_AlphaInterp(
    InRange=(Min=0,Max=-1)
    InterpSpeedDecreasing=10
    InterpSpeedIncreasing=10
    Scale=1
    bClampResult=True
    bInterpResult=False
    bMapRange=True
    Value=&RigVMFunction_MathDoubleSub(
     B=&RigUnit_GetFloatAnimationChannel(Channel=chinRaiseUR, Control=CTRL_C_jaw, bInitial=False) {template="GetAnimationChannel::Execute(out Value,in Control,in Channel,in bInitial)"}.Value
     A=@n24.Vector.Y
    ).Result
   )
   > &RigUnit_SetCurveValue(Curve=CTRL_expressions_jawChinRaiseUR, Value=@RigUnit_AlphaInterp_4_1_1_2.Result)
   > CRFL AnimAttrToCurve v01_24 = RigVMFunctionReferenceNode(
    Backward=False
    Channel=jawOpenExtreme
    Control=(Type=Control,Name="CTRL_C_jaw")
    Pos_Neg_Curves=((Type=Curve,Name="CTRL_expressions_jawOpenExtreme"))
   )
   > FN_VectorCtrlToCurve_FNC_14 = RigVMFunctionReferenceNode(
    Backward=false
    Control=(Type=Control,Name="CTRL_C_tongue")
    Ctrl_L_Token=CTRL_L_
    Ctrl_R_Token=CTRL_R_
    Curve_Side_Token_Last=true
    X_Mirror=false
    X_Pos_Neg_Expr=((Type=Curve,Name="CTRL_expressions_tongueLeft"),(Type=Curve,Name="CTRL_expressions_tongueRight"))
    Y_Mirror=false
    Y_Pos_Neg_Expr=((Type=Curve,Name="CTRL_expressions_tongueUp"),(Type=Curve,Name="CTRL_expressions_tongueDown"))
    Z_Mirror=false
    Z_Pos_Neg_Expr=((Type=Curve,Name="CTRL_expressions_tongueIn"),(Type=Curve,Name="CTRL_expressions_tongueOut"))
   )
   > CRFL AnimAttrToCurve v01_25 = RigVMFunctionReferenceNode(
    Backward=False
    Channel=wideNarrow
    Control=(Type=Control,Name="CTRL_C_tongue")
    Pos_Neg_Curves=((Type=Curve,Name="CTRL_expressions_tongueWide"),(Type=Curve,Name="CTRL_expressions_tongueNarrow"))
   )
   > CRFL AnimAttrToCurve v01_26 = RigVMFunctionReferenceNode(
    Backward=False
    Channel=rollUpDown
    Control=(Type=Control,Name="CTRL_C_tongue")
    Pos_Neg_Curves=((Type=Curve,Name="CTRL_expressions_tongueRollUp"),(Type=Curve,Name="CTRL_expressions_tongueRollDown"))
   )
   > CRFL AnimAttrToCurve v01_27 = RigVMFunctionReferenceNode(
    Backward=False
    Channel=rollLeftRight
    Control=(Type=Control,Name="CTRL_C_tongue")
    Pos_Neg_Curves=((Type=Curve,Name="CTRL_expressions_tongueRollLeft"),(Type=Curve,Name="CTRL_expressions_tongueRollRight"))
   )
   > FN_VectorCtrlToCurve_FNC_15 = RigVMFunctionReferenceNode(
    Backward=false
    Control=(Type=Control,Name="CTRL_C_teethU")
    Ctrl_L_Token=CTRL_L_
    Ctrl_R_Token=CTRL_R_
    Curve_Side_Token_Last=true
    X_Mirror=false
    X_Pos_Neg_Expr=((Type=Curve,Name="CTRL_expressions_teethLeftU"),(Type=Curve,Name="CTRL_expressions_teethRightU"))
    Y_Mirror=false
    Y_Pos_Neg_Expr=((Type=Curve,Name="CTRL_expressions_teethUpU"),(Type=Curve,Name="CTRL_expressions_teethDownU"))
    Z_Mirror=false
    Z_Pos_Neg_Expr=((Type=Curve,Name="CTRL_expressions_teethBackU"),(Type=Curve,Name="CTRL_expressions_teethFwdU"))
   )
   > CRFL AnimAttrToCurve v01_28 = RigVMFunctionReferenceNode(
    Backward=False
    Channel=rotateFwdBack
    Control=(Type=Control,Name="CTRL_C_teethU")
    Pos_Neg_Curves=((Type=Curve,Name="CTRL_expressions_teethRotateFwdU"),(Type=Curve,Name="CTRL_expressions_teethRotateBackU"))
   )
   > FN_VectorCtrlToCurve_FNC_16 = RigVMFunctionReferenceNode(
    Backward=false
    Control=(Type=Control,Name="CTRL_C_teethD")
    Ctrl_L_Token=CTRL_L_
    Ctrl_R_Token=CTRL_R_
    Curve_Side_Token_Last=true
    X_Mirror=false
    X_Pos_Neg_Expr=((Type=Curve,Name="CTRL_expressions_teethLeftD"),(Type=Curve,Name="CTRL_expressions_teethRightD"))
    Y_Mirror=false
    Y_Pos_Neg_Expr=((Type=Curve,Name="CTRL_expressions_teethUpD"),(Type=Curve,Name="CTRL_expressions_teethDownD"))
    Z_Mirror=false
    Z_Pos_Neg_Expr=((Type=Curve,Name="CTRL_expressions_teethBackD"),(Type=Curve,Name="CTRL_expressions_teethFwdD"))
   )
   > CRFL AnimAttrToCurve v01_29 = RigVMFunctionReferenceNode(
    Backward=False
    Channel=rotateFwdBack
    Control=(Type=Control,Name="CTRL_C_teethD")
    Pos_Neg_Curves=((Type=Curve,Name="CTRL_expressions_teethRotateFwdD"),(Type=Curve,Name="CTRL_expressions_teethRotateBackD"))
   )
   RigVMFunctionReturnNode(ExecuteContext=@SequenceAggregate_4.B)
  ]}
 New Function_ContainedGraph {nodes [
   RigVMFunctionEntryNode()
   > RigVMFunction_Sequence = RigVMAggregateNode()
   &RigUnit_SetCurveValue(Curve=DisableFaceOverride, Value=1, ExecutePin=@RigVMFunction_Sequence.A)
   > &RigUnit_SetCurveValue(Curve=is_3L, Value=1)
   > SequenceAggregate = RigVMAggregateNode()
   SequenceAggregate_1 = RigVMAggregateNode(ExecuteContext=@SequenceAggregate.A)
   FN_FloatCtrlToCurve_FNC_15_2 = RigVMFunctionReferenceNode(
    Ctrl_L_Token=CTRL_L_
    Ctrl_R_Token=CTRL_R_
    Curve_Side_Token_Last=true
    Pos_Neg_Curves=((Type=Curve,Name="CTRL_expressions_eyeBlinkL"),(Type=Curve,Name="CTRL_expressions_eyeWidenL"))
    backward=true
    control=(Type=Control,Name="CTRL_L_eye_blink")
    mirror=true
    vector_axis=X
    ExecuteContext=@SequenceAggregate_1.C
   )
   > FN_VectorCtrlToCurve_FNC_2 = RigVMFunctionReferenceNode(
    Backward=true
    Control=(Type=Control,Name="CTRL_L_eye_eyelidU")
    Ctrl_L_Token=CTRL_L_
    Ctrl_R_Token=CTRL_R_
    Curve_Side_Token_Last=true
    X_Mirror=false
    X_Pos_Neg_Expr=((Type=Curve,Name="CTRL_expressions_eyelidRollNegUL"),(Type=Curve,Name="CTRL_expressions_eyelidRollPosUL"))
    Y_Mirror=true
    Y_Pos_Neg_Expr=((Type=Curve,Name="CTRL_expressions_eyeUpperLidUpL"),(Type=Curve,Name="CTRL_expressions_eyeRelaxL"))
    Z_Mirror=false
    Z_Pos_Neg_Expr=()
   )
   > FN_VectorCtrlToCurve_FNC_8 = RigVMFunctionReferenceNode(
    Backward=true
    Control=(Type=Control,Name="CTRL_R_eye_eyelidU")
    Ctrl_L_Token=CTRL_L_
    Ctrl_R_Token=CTRL_R_
    Curve_Side_Token_Last=true
    X_Mirror=false
    X_Pos_Neg_Expr=((Type=Curve,Name="CTRL_expressions_eyelidRollPosUR"),(Type=Curve,Name="CTRL_expressions_eyelidRollNegUR"))
    Y_Mirror=false
    Y_Pos_Neg_Expr=()
    Z_Mirror=false
    Z_Pos_Neg_Expr=()
   )
   > FN_VectorCtrlToCurve_FNC_3 = RigVMFunctionReferenceNode(
    Backward=true
    Control=(Type=Control,Name="CTRL_L_eye_eyelidD")
    Ctrl_L_Token=CTRL_L_
    Ctrl_R_Token=CTRL_R_
    Curve_Side_Token_Last=true
    X_Mirror=false
    X_Pos_Neg_Expr=((Type=Curve,Name="CTRL_expressions_eyelidRollPosDL"),(Type=Curve,Name="CTRL_expressions_eyelidRollNegDL"))
    Y_Mirror=true
    Y_Pos_Neg_Expr=((Type=Curve,Name="CTRL_expressions_eyeLowerLidUpL"),(Type=Curve,Name="CTRL_expressions_eyeLowerLidDownL"))
    Z_Mirror=false
    Z_Pos_Neg_Expr=()
   )
   > FN_VectorCtrlToCurve_FNC_9 = RigVMFunctionReferenceNode(
    Backward=true
    Control=(Type=Control,Name="CTRL_R_eye_eyelidD")
    Ctrl_L_Token=CTRL_L_
    Ctrl_R_Token=CTRL_R_
    Curve_Side_Token_Last=true
    X_Mirror=false
    X_Pos_Neg_Expr=((Type=Curve,Name="CTRL_expressions_eyelidRollNegDR"),(Type=Curve,Name="CTRL_expressions_eyelidRollPosDR"))
    Y_Mirror=false
    Y_Pos_Neg_Expr=()
    Z_Mirror=false
    Z_Pos_Neg_Expr=()
   )
   FN_FloatCtrlToCurve_FNC_15 = RigVMFunctionReferenceNode(
    Ctrl_L_Token=CTRL_L_
    Ctrl_R_Token=CTRL_R_
    Curve_Side_Token_Last=true
    Pos_Neg_Curves=((Type=Curve,Name="CTRL_expressions_browRaiseInL"),(Type=Curve,Name="CTRL_expressions_browDINL"))
    backward=true
    control=(Type=Control,Name="CTRL_L_brow_inner")
    mirror=true
    vector_axis=X
    ExecuteContext=@SequenceAggregate_1.A
   )
   > FN_VectorCtrlToCurve_FNC = RigVMFunctionReferenceNode(
    Backward=true
    Control=(Type=Control,Name="CTRL_L_brow")
    Ctrl_L_Token=CTRL_L_
    Ctrl_R_Token=CTRL_R_
    Curve_Side_Token_Last=true
    X_Mirror=true
    X_Pos_Neg_Expr=((Type=Curve,Name="CTRL_expressions_browLateralL"))
    Y_Mirror=true
    Y_Pos_Neg_Expr=((Type=Curve,Name="CTRL_expressions_browUINL"),(Type=Curve,Name="CTRL_expressions_browDownL"))
    Z_Mirror=false
    Z_Pos_Neg_Expr=()
   )
   > FN_FloatCtrlToCurve_FNC_15_1 = RigVMFunctionReferenceNode(
    Ctrl_L_Token=CTRL_L_
    Ctrl_R_Token=CTRL_R_
    Curve_Side_Token_Last=true
    Pos_Neg_Curves=((Type=Curve,Name="CTRL_expressions_browRaiseOuterL"),(Type=Curve,Name="CTRL_expressions_browDOUTL"))
    backward=true
    control=(Type=Control,Name="CTRL_L_brow_outer")
    mirror=true
    vector_axis=X
   )
   n1 = &RigVMFunction_MathFloatRemap(
    SourceMaximum=1
    TargetMaximum=40
    bClamp=true
    Value=/Script/ControlRig.RigUnit_GetCurveValue(Curve=CTRL_expressions_eyeLookDownL) {is=struct RigVMUnitNode}.Value
   )
   n2 = &RigVMFunction_MathFloatRemap(SourceMaximum=1, TargetMaximum=-40, bClamp=true, Value=&RigUnit_GetCurveValue(Curve=CTRL_expressions_eyeLookUpL).Value)
   n3 = &RigVMFunction_MathFloatRemap(SourceMaximum=1, TargetMaximum=40, bClamp=true, Value=&RigUnit_GetCurveValue(Curve=CTRL_expressions_eyeLookRightL).Value)
   n4 = &RigVMFunction_MathFloatRemap(SourceMaximum=1, TargetMaximum=-40, bClamp=true, Value=&RigUnit_GetCurveValue(Curve=CTRL_expressions_eyeLookLeftL).Value)
   &RigUnit_SetRotation(
    Item=(Type=Null,Name="EyeAim_L")
    Weight=1
    Value=/Script/RigVM.RigVMFunction_MathQuaternionFromRotatorV2(
     Value=(Pitch=0,Yaw=0,Roll=0)
     Value.Roll=&RigVMFunction_MathFloatAdd(A=@n1.Result, B=@n2.Result) {template="Add::Execute(in A,in B,out Result)"}.Result
     Value.Pitch=&RigVMFunction_MathFloatAdd(A=@n3.Result, B=@n4.Result) {template="Add::Execute(in A,in B,out Result)"}.Result
    ) {is=struct RigVMUnitNode, template="Cast::Execute(in Value,out Result)"}.Result
    ExecutePin=@SequenceAggregate_1.B
   )
   n5 = &RigVMFunction_MathFloatRemap(SourceMaximum=1, TargetMaximum=40, bClamp=true, Value=&RigUnit_GetCurveValue(Curve=CTRL_expressions_eyeLookRightR).Value)
   n6 = &RigVMFunction_MathFloatRemap(SourceMaximum=1, TargetMaximum=-40, bClamp=true, Value=&RigUnit_GetCurveValue(Curve=CTRL_expressions_eyeLookLeftR).Value)
   n7 = &RigVMFunction_MathFloatRemap(SourceMaximum=1, TargetMaximum=-40, bClamp=true, Value=&RigUnit_GetCurveValue(Curve=CTRL_expressions_eyeLookUpR).Value)
   n8 = &RigVMFunction_MathFloatRemap(SourceMaximum=1, TargetMaximum=40, bClamp=true, Value=&RigUnit_GetCurveValue(Curve=CTRL_expressions_eyeLookDownR).Value)
   > &RigUnit_SetRotation(
    Item=(Type=Null,Name="EyeAim_R")
    Weight=1
    Value=&RigVMFunction_MathQuaternionFromRotatorV2(
     Value=(Pitch=0,Yaw=0,Roll=0)
     Value.Pitch=&RigVMFunction_MathFloatAdd(A=@n5.Result, B=@n6.Result) {template="Add::Execute(in A,in B,out Result)"}.Result
     Value.Roll=&RigVMFunction_MathFloatAdd(A=@n7.Result, B=@n8.Result) {template="Add::Execute(in A,in B,out Result)"}.Result
    ) {template="Cast::Execute(in Value,out Result)"}.Result
   )
   n9 = &RigUnit_ProjectTransformToNewParent(
    Child=(Type=Control,Name="CTRL_C_eyesAim")
    NewParent=(Type=Null,Name="EyeAim_L")
    OldParent=(Type=Null,Name="EyeAim_L")
    bChildInitial=True
    bOldParentInitial=True
   )
   > &RigUnit_SetTransform(Item=(Type=Control,Name="CTRL_C_eyesAim"), Weight=1, Value=@n9.Transform)
   n10 = &RigUnit_ProjectTransformToNewParent(
    Child=(Type=Control,Name="CTRL_L_eyeAim")
    NewParent=(Type=Null,Name="EyeAim_L")
    OldParent=(Type=Null,Name="EyeAim_L")
    bChildInitial=True
    bOldParentInitial=True
   )
   > &RigUnit_SetTransform(Item=(Type=Control,Name="CTRL_L_eyeAim"), Weight=1, Value=@n10.Transform)
   n11 = &RigUnit_ProjectTransformToNewParent(
    Child=(Type=Control,Name="CTRL_R_eyeAim")
    NewParent=(Type=Null,Name="EyeAim_R")
    OldParent=(Type=Null,Name="EyeAim_R")
    bChildInitial=True
    bOldParentInitial=True
   )
   > &RigUnit_SetTransform(Item=(Type=Control,Name="CTRL_R_eyeAim"), Weight=1, Value=@n11.Transform)
   > FN_VectorCtrlToCurve_FNC_14 = RigVMFunctionReferenceNode(
    Backward=true
    Control=(Type=Control,Name="CTRL_L_eye")
    Ctrl_L_Token=CTRL_L_
    Ctrl_R_Token=CTRL_R_
    Curve_Side_Token_Last=true
    X_Mirror=true
    X_Pos_Neg_Expr=((Type=Curve,Name="CTRL_expressions_eyeLookLeftL"),(Type=Curve,Name="CTRL_expressions_eyeLookRightL"))
    Y_Mirror=true
    Y_Pos_Neg_Expr=((Type=Curve,Name="CTRL_expressions_eyeLookUpL"),(Type=Curve,Name="CTRL_expressions_eyeLookDownL"))
    Z_Mirror=false
    Z_Pos_Neg_Expr=()
   )
   > CRFL AnimAttrToCurve v01_34 = RigVMFunctionReferenceNode(
    Backward=true
    Channel=eyelidsReduce
    Control=(Type=Control,Name="CTRL_C_eye")
    Pos_Neg_Curves=((Type=Curve,Name="CTRL_expressions_eyeEyelidsReduce"))
   )
   SequenceAggregate_2 = RigVMAggregateNode(ExecuteContext=@SequenceAggregate.B)
   FN_VectorCtrlToCurve_FNC_4 = RigVMFunctionReferenceNode(
    Backward=true
    Control=(Type=Control,Name="CTRL_C_nose")
    Ctrl_L_Token=CTRL_L_
    Ctrl_R_Token=CTRL_R_
    Curve_Side_Token_Last=true
    X_Mirror=false
    X_Pos_Neg_Expr=((Type=Curve,Name="CTRL_expressions_noseRotateL"),(Type=Curve,Name="CTRL_expressions_noseRotateR"))
    Y_Mirror=false
    Y_Pos_Neg_Expr=((Type=Curve,Name="CTRL_expressions_noseSquash"),(Type=Curve,Name="CTRL_expressions_noseStretch"))
    Z_Mirror=false
    Z_Pos_Neg_Expr=()
    ExecuteContext=@SequenceAggregate_2.A
   )
   > FN_VectorCtrlToCurve_FNC_5 = RigVMFunctionReferenceNode(
    Backward=true
    Control=(Type=Control,Name="CTRL_L_nose")
    Ctrl_L_Token=CTRL_L_
    Ctrl_R_Token=CTRL_R_
    Curve_Side_Token_Last=true
    X_Mirror=true
    X_Pos_Neg_Expr=((Type=Curve,Name="CTRL_expressions_noseNostrilDilateL"),(Type=Curve,Name="CTRL_expressions_noseNostrilCompressL"))
    Y_Mirror=true
    Y_Pos_Neg_Expr=((Type=Curve,Name="CTRL_expressions_noseWrinkleL"),(Type=Curve,Name="CTRL_expressions_noseNostrilDepressL"))
    Z_Mirror=false
    Z_Pos_Neg_Expr=()
   )
   > CRFL AnimAttrToCurve v01_31 = RigVMFunctionReferenceNode(
    Backward=true
    Channel=noseWrinkleUpper
    Control=(Type=Control,Name="CTRL_L_nose")
    Pos_Neg_Curves=((Type=Curve,Name="CTRL_expressions_noseWrinkleUpperL"))
   )
   > CRFL AnimAttrToCurve v01_30 = RigVMFunctionReferenceNode(
    Backward=true
    Channel=noseWrinkleUpper
    Control=(Type=Control,Name="CTRL_R_nose")
    Pos_Neg_Curves=((Type=Curve,Name="CTRL_expressions_noseWrinkleUpperR"))
   )
   FN_FloatCtrlToCurve_FNC_15_2_1 = RigVMFunctionReferenceNode(
    Ctrl_L_Token=CTRL_L_
    Ctrl_R_Token=CTRL_R_
    Curve_Side_Token_Last=true
    Pos_Neg_Curves=((Type=Curve,Name="CTRL_expressions_eyeSquintInnerL"))
    backward=true
    control=(Type=Control,Name="CTRL_L_eye_squintInner")
    mirror=true
    vector_axis=X
    ExecuteContext=@SequenceAggregate_2.B
   )
   > FN_FloatCtrlToCurve_FNC_15_2_1_1 = RigVMFunctionReferenceNode(
    Ctrl_L_Token=CTRL_L_
    Ctrl_R_Token=CTRL_R_
    Curve_Side_Token_Last=true
    Pos_Neg_Curves=((Type=Curve,Name="CTRL_expressions_eyeCheekRaiseL"))
    backward=true
    control=(Type=Control,Name="CTRL_L_eye_cheekRaise")
    mirror=true
    vector_axis=X
   )
   > FN_FloatCtrlToCurve_FNC_15_2_1_1_1 = RigVMFunctionReferenceNode(
    Ctrl_L_Token=CTRL_L_
    Ctrl_R_Token=CTRL_R_
    Curve_Side_Token_Last=true
    Pos_Neg_Curves=((Type=Curve,Name="CTRL_expressions_mouthCheekBlowL"),(Type=Curve,Name="CTRL_expressions_mouthCheekSuckL"))
    backward=true
    control=(Type=Control,Name="CTRL_L_mouth_suckBlow")
    mirror=true
    vector_axis=X
   )
   FN_VectorCtrlToCurve_FNC_6 = RigVMFunctionReferenceNode(
    Backward=true
    Control=(Type=Control,Name="CTRL_C_mouth")
    Ctrl_L_Token=CTRL_L_
    Ctrl_R_Token=CTRL_R_
    Curve_Side_Token_Last=true
    X_Mirror=false
    X_Pos_Neg_Expr=((Type=Curve,Name="CTRL_expressions_mouthLeft"),(Type=Curve,Name="CTRL_expressions_mouthRight"))
    Y_Mirror=false
    Y_Pos_Neg_Expr=((Type=Curve,Name="CTRL_expressions_mouthUp"),(Type=Curve,Name="CTRL_expressions_mouthDown"))
    Z_Mirror=false
    Z_Pos_Neg_Expr=()
    ExecuteContext=@SequenceAggregate_2.C
   )
   > FN_VectorCtrlToCurve_FNC_7 = RigVMFunctionReferenceNode(
    Backward=true
    Control=(Type=Control,Name="CTRL_L_mouth_lipU")
    Ctrl_L_Token=CTRL_L_
    Ctrl_R_Token=CTRL_R_
    Curve_Side_Token_Last=true
    X_Mirror=true
    X_Pos_Neg_Expr=((Type=Curve,Name="CTRL_expressions_mouthLipsTowardsUL"),(Type=Curve,Name="CTRL_expressions_mouthUpperLipBiteL"))
    Y_Mirror=true
    Y_Pos_Neg_Expr=((Type=Curve,Name="CTRL_expressions_mouthUpperLipRaiseL"),(Type=Curve,Name="CTRL_expressions_mouthUpperLipTowardsTeethL"))
    Z_Mirror=false
    Z_Pos_Neg_Expr=()
   )
   > CRFL AnimAttrToCurve v01_18 = RigVMFunctionReferenceNode(
    Backward=true
    Channel=mouthPress
    Control=(Type=Control,Name="CTRL_L_mouth_lipU")
    Pos_Neg_Curves=((Type=Curve,Name="CTRL_expressions_mouthPressUL"))
   )
   > CRFL AnimAttrToCurve v01_19 = RigVMFunctionReferenceNode(
    Backward=true
    Channel=lipsTogether
    Control=(Type=Control,Name="CTRL_L_mouth_lipU")
    Pos_Neg_Curves=((Type=Curve,Name="CTRL_expressions_mouthLipsTogetherUL"))
   )
   > CRFL AnimAttrToCurve v01_17 = RigVMFunctionReferenceNode(
    Backward=true
    Channel=mouthPress
    Control=(Type=Control,Name="CTRL_R_mouth_lipU")
    Pos_Neg_Curves=((Type=Curve,Name="CTRL_expressions_mouthPressUR"))
   )
   > CRFL AnimAttrToCurve v01_16 = RigVMFunctionReferenceNode(
    Backward=true
    Channel=lipsTogether
    Control=(Type=Control,Name="CTRL_R_mouth_lipU")
    Pos_Neg_Curves=((Type=Curve,Name="CTRL_expressions_mouthLipsTogetherUR"))
   )
   > FN_VectorCtrlToCurve_FNC_11 = RigVMFunctionReferenceNode(
    Backward=true
    Control=(Type=Control,Name="CTRL_L_mouth_lipD")
    Ctrl_L_Token=CTRL_L_
    Ctrl_R_Token=CTRL_R_
    Curve_Side_Token_Last=true
    X_Mirror=true
    X_Pos_Neg_Expr=((Type=Curve,Name="CTRL_expressions_mouthLipsTowardsDL"),(Type=Curve,Name="CTRL_expressions_mouthLowerLipBiteL"))
    Y_Mirror=true
    Y_Pos_Neg_Expr=((Type=Curve,Name="CTRL_expressions_mouthLowerLipDepressL"),(Type=Curve,Name="CTRL_expressions_mouthLowerLipTowardsTeethL"))
    Z_Mirror=false
    Z_Pos_Neg_Expr=()
   )
   > CRFL AnimAttrToCurve v01_22 = RigVMFunctionReferenceNode(
    Backward=true
    Channel=mouthPress
    Control=(Type=Control,Name="CTRL_L_mouth_lipD")
    Pos_Neg_Curves=((Type=Curve,Name="CTRL_expressions_mouthPressDL"))
   )
   > CRFL AnimAttrToCurve v01_23 = RigVMFunctionReferenceNode(
    Backward=true
    Channel=lipsTogether
    Control=(Type=Control,Name="CTRL_L_mouth_lipD")
    Pos_Neg_Curves=((Type=Curve,Name="CTRL_expressions_mouthLipsTogetherDR"))
   )
   > CRFL AnimAttrToCurve v01_20 = RigVMFunctionReferenceNode(
    Backward=true
    Channel=mouthPress
    Control=(Type=Control,Name="CTRL_R_mouth_lipD")
    Pos_Neg_Curves=((Type=Curve,Name="CTRL_expressions_mouthPressDR"))
   )
   > CRFL AnimAttrToCurve v01_21 = RigVMFunctionReferenceNode(
    Backward=true
    Channel=lipsTogether
    Control=(Type=Control,Name="CTRL_R_mouth_lipD")
    Pos_Neg_Curves=((Type=Curve,Name="CTRL_expressions_mouthLipsTogetherDR"))
   )
   > FN_VectorCtrlToCurve_FNC_12 = RigVMFunctionReferenceNode(
    Backward=true
    Control=(Type=Control,Name="CTRL_L_mouth_smileDepress")
    Ctrl_L_Token=CTRL_L_
    Ctrl_R_Token=CTRL_R_
    Curve_Side_Token_Last=true
    X_Mirror=true
    X_Pos_Neg_Expr=((Type=Curve,Name="CTRL_expressions_mouthCornerDepressL"))
    Y_Mirror=true
    Y_Pos_Neg_Expr=((Type=Curve,Name="CTRL_expressions_mouthCornerPullL"))
    Z_Mirror=false
    Z_Pos_Neg_Expr=()
   )
   > CRFL AnimAttrToCurve v01_6 = RigVMFunctionReferenceNode(
    Backward=true
    Channel=mouthDimple
    Control=(Type=Control,Name="CTRL_L_mouth_smileDepress")
    Pos_Neg_Curves=((Type=Curve,Name="CTRL_expressions_mouthDimpleL"))
   )
   > CRFL AnimAttrToCurve v01_7 = RigVMFunctionReferenceNode(
    Backward=true
    Channel=mouthStretch
    Control=(Type=Control,Name="CTRL_L_mouth_smileDepress")
    Pos_Neg_Curves=((Type=Curve,Name="CTRL_expressions_mouthStretchL"))
   )
   > CRFL AnimAttrToCurve v01_5 = RigVMFunctionReferenceNode(
    Backward=true
    Channel=mouthPurseU
    Control=(Type=Control,Name="CTRL_L_mouth_smileDepress")
    Pos_Neg_Curves=((Type=Curve,Name="CTRL_expressions_mouthLipsPurseUL"))
   )
   > CRFL AnimAttrToCurve v01_4 = RigVMFunctionReferenceNode(
    Backward=true
    Channel=mouthPurseD
    Control=(Type=Control,Name="CTRL_L_mouth_smileDepress")
    Pos_Neg_Curves=((Type=Curve,Name="CTRL_expressions_mouthLipsPurseDL"))
   )
   > CRFL AnimAttrToCurve v01_3 = RigVMFunctionReferenceNode(
    Backward=true
    Channel=mouthFunnelU
    Control=(Type=Control,Name="CTRL_L_mouth_smileDepress")
    Pos_Neg_Curves=((Type=Curve,Name="CTRL_expressions_mouthFunnelUL"))
   )
   > CRFL AnimAttrToCurve v01_2 = RigVMFunctionReferenceNode(
    Backward=true
    Channel=mouthFunnelD
    Control=(Type=Control,Name="CTRL_L_mouth_smileDepress")
    Pos_Neg_Curves=((Type=Curve,Name="CTRL_expressions_mouthFunnelDL"))
   )
   > CRFL AnimAttrToCurve v01 = RigVMFunctionReferenceNode(
    Backward=true
    Channel=mouthCornerSharpnessU
    Control=(Type=Control,Name="CTRL_L_mouth_smileDepress")
    Pos_Neg_Curves=((Type=Curve,Name="CTRL_expressions_mouthCornerRounderUL"),(Type=Curve,Name="CTRL_expressions_mouthCornerSharpenUL"))
   )
   > CRFL AnimAttrToCurve v01_1 = RigVMFunctionReferenceNode(
    Backward=true
    Channel=mouthCornerSharpnessD
    Control=(Type=Control,Name="CTRL_L_mouth_smileDepress")
    Pos_Neg_Curves=((Type=Curve,Name="CTRL_expressions_mouthCornerRounderDL"),(Type=Curve,Name="CTRL_expressions_mouthCornerSharpenDL"))
   )
   > CRFL AnimAttrToCurve v01_14 = RigVMFunctionReferenceNode(
    Backward=true
    Channel=mouthDimple
    Control=(Type=Control,Name="CTRL_R_mouth_smileDepress")
    Pos_Neg_Curves=((Type=Curve,Name="CTRL_expressions_mouthDimpleR"))
   )
   > CRFL AnimAttrToCurve v01_15 = RigVMFunctionReferenceNode(
    Backward=true
    Channel=mouthStretch
    Control=(Type=Control,Name="CTRL_R_mouth_smileDepress")
    Pos_Neg_Curves=((Type=Curve,Name="CTRL_expressions_mouthStretchR"))
   )
   > CRFL AnimAttrToCurve v01_13 = RigVMFunctionReferenceNode(
    Backward=true
    Channel=mouthPurseU
    Control=(Type=Control,Name="CTRL_R_mouth_smileDepress")
    Pos_Neg_Curves=((Type=Curve,Name="CTRL_expressions_mouthLipsPurseUR"))
   )
   > CRFL AnimAttrToCurve v01_12 = RigVMFunctionReferenceNode(
    Backward=true
    Channel=mouthPurseD
    Control=(Type=Control,Name="CTRL_R_mouth_smileDepress")
    Pos_Neg_Curves=((Type=Curve,Name="CTRL_expressions_mouthLipsPurseDR"))
   )
   > CRFL AnimAttrToCurve v01_11 = RigVMFunctionReferenceNode(
    Backward=true
    Channel=mouthFunnelU
    Control=(Type=Control,Name="CTRL_R_mouth_smileDepress")
    Pos_Neg_Curves=((Type=Curve,Name="CTRL_expressions_mouthFunnelUR"))
   )
   > CRFL AnimAttrToCurve v01_10 = RigVMFunctionReferenceNode(
    Backward=true
    Channel=mouthFunnelD
    Control=(Type=Control,Name="CTRL_R_mouth_smileDepress")
    Pos_Neg_Curves=((Type=Curve,Name="CTRL_expressions_mouthFunnelDR"))
   )
   > CRFL AnimAttrToCurve v01_8 = RigVMFunctionReferenceNode(
    Backward=true
    Channel=mouthCornerSharpnessU
    Control=(Type=Control,Name="CTRL_R_mouth_smileDepress")
    Pos_Neg_Curves=((Type=Curve,Name="CTRL_expressions_mouthCornerRounderUR"),(Type=Curve,Name="CTRL_expressions_mouthCornerSharpenUR"))
   )
   > CRFL AnimAttrToCurve v01_9 = RigVMFunctionReferenceNode(
    Backward=true
    Channel=mouthCornerSharpnessD
    Control=(Type=Control,Name="CTRL_R_mouth_smileDepress")
    Pos_Neg_Curves=((Type=Curve,Name="CTRL_expressions_mouthCornerRounderDR"),(Type=Curve,Name="CTRL_expressions_mouthCornerSharpenDR"))
   )
   > FN_VectorCtrlToCurve_FNC_13 = RigVMFunctionReferenceNode(
    Backward=true
    Control=(Type=Control,Name="CTRL_L_mouth_corner")
    Ctrl_L_Token=CTRL_L_
    Ctrl_R_Token=CTRL_R_
    Curve_Side_Token_Last=true
    X_Mirror=true
    X_Pos_Neg_Expr=((Type=Curve,Name="CTRL_expressions_mouthCornerWideL"),(Type=Curve,Name="CTRL_expressions_mouthCornerNarrowL"))
    Y_Mirror=true
    Y_Pos_Neg_Expr=((Type=Curve,Name="CTRL_expressions_mouthCornerUpL"),(Type=Curve,Name="CTRL_expressions_mouthCornerDownL"))
    Z_Mirror=false
    Z_Pos_Neg_Expr=()
   )
   > CRFL AnimAttrToCurve v01_32 = RigVMFunctionReferenceNode(
    Backward=true
    Channel=lipsBlow
    Control=(Type=Control,Name="CTRL_L_mouth_corner")
    Pos_Neg_Curves=((Type=Curve,Name="CTRL_expressions_mouthLipsBlowL"))
   )
   > CRFL AnimAttrToCurve v01_33 = RigVMFunctionReferenceNode(
    Backward=true
    Channel=lipsBlow
    Control=(Type=Control,Name="CTRL_R_mouth_corner")
    Pos_Neg_Curves=((Type=Curve,Name="CTRL_expressions_mouthLipsBlowR"))
   )
   SequenceAggregate_3 = &RigVMFunction_Sequence(ExecuteContext=@SequenceAggregate.C)
   FN_VectorCtrlToCurve_FNC_10_1 = RigVMFunctionReferenceNode(
    Backward=true
    Control=(Type=Control,Name="CTRL_C_jaw")
    Ctrl_L_Token=CTRL_L_
    Ctrl_R_Token=CTRL_R_
    Curve_Side_Token_Last=true
    X_Mirror=false
    X_Pos_Neg_Expr=((Type=Curve,Name="CTRL_expressions_jawRight"),(Type=Curve,Name="CTRL_expressions_jawLeft"))
    Y_Mirror=false
    Y_Pos_Neg_Expr=((Type=Curve,Name="CTRL_expressions_jawOpen"))
    Z_Mirror=false
    Z_Pos_Neg_Expr=((Type=Curve,Name="CTRL_expressions_jawBack"),(Type=Curve,Name="CTRL_expressions_jawFwd"))
    ExecuteContext=@SequenceAggregate_3.A
   )
   > CRFL AnimAttrToCurve v01_24 = RigVMFunctionReferenceNode(
    Backward=true
    Channel=jawOpenExtreme
    Control=(Type=Control,Name="CTRL_C_jaw")
    Pos_Neg_Curves=((Type=Curve,Name="CTRL_expressions_jawOpenExtreme"))
   )
   > CRFL AnimAttrToCurve v01_36 = RigVMFunctionReferenceNode(
    Backward=true
    Channel=chinRaiseUL
    Control=(Type=Control,Name="CTRL_C_jaw")
    Pos_Neg_Curves=((Type=Curve,Name="CTRL_expressions_jawChinRaiseUL"))
   )
   > CRFL AnimAttrToCurve v01_37 = RigVMFunctionReferenceNode(
    Backward=true
    Channel=chinRaiseUR
    Control=(Type=Control,Name="CTRL_C_jaw")
    Pos_Neg_Curves=((Type=Curve,Name="CTRL_expressions_jawChinRaiseUR"))
   )
   > CRFL AnimAttrToCurve v01_38 = RigVMFunctionReferenceNode(
    Backward=true
    Channel=chinRaiseDL
    Control=(Type=Control,Name="CTRL_C_jaw")
    Pos_Neg_Curves=((Type=Curve,Name="CTRL_expressions_jawChinRaiseDL"))
   )
   > CRFL AnimAttrToCurve v01_39 = RigVMFunctionReferenceNode(
    Backward=true
    Channel=chinRaiseDR
    Control=(Type=Control,Name="CTRL_C_jaw")
    Pos_Neg_Curves=((Type=Curve,Name="CTRL_expressions_jawChinRaiseDR"))
   )
   > FN_VectorCtrlToCurve_FNC_14_1 = RigVMFunctionReferenceNode(
    Backward=true
    Control=(Type=Control,Name="CTRL_C_tongue")
    Ctrl_L_Token=CTRL_L_
    Ctrl_R_Token=CTRL_R_
    Curve_Side_Token_Last=true
    X_Mirror=false
    X_Pos_Neg_Expr=((Type=Curve,Name="CTRL_expressions_tongueLeft"),(Type=Curve,Name="CTRL_expressions_tongueRight"))
    Y_Mirror=false
    Y_Pos_Neg_Expr=((Type=Curve,Name="CTRL_expressions_tongueUp"),(Type=Curve,Name="CTRL_expressions_tongueDown"))
    Z_Mirror=false
    Z_Pos_Neg_Expr=((Type=Curve,Name="CTRL_expressions_tongueIn"),(Type=Curve,Name="CTRL_expressions_tongueOut"))
   )
   > CRFL AnimAttrToCurve v01_25 = RigVMFunctionReferenceNode(
    Backward=true
    Channel=wideNarrow
    Control=(Type=Control,Name="CTRL_C_tongue")
    Pos_Neg_Curves=((Type=Curve,Name="CTRL_expressions_tongueWide"),(Type=Curve,Name="CTRL_expressions_tongueNarrow"))
   )
   > CRFL AnimAttrToCurve v01_26 = RigVMFunctionReferenceNode(
    Backward=true
    Channel=rollUpDown
    Control=(Type=Control,Name="CTRL_C_tongue")
    Pos_Neg_Curves=((Type=Curve,Name="CTRL_expressions_tongueRollUp"),(Type=Curve,Name="CTRL_expressions_tongueRollDown"))
   )
   > CRFL AnimAttrToCurve v01_27 = RigVMFunctionReferenceNode(
    Backward=true
    Channel=rollLeftRight
    Control=(Type=Control,Name="CTRL_C_tongue")
    Pos_Neg_Curves=((Type=Curve,Name="CTRL_expressions_tongueRollLeft"),(Type=Curve,Name="CTRL_expressions_tongueRollRight"))
   )
   > FN_VectorCtrlToCurve_FNC_15_1 = RigVMFunctionReferenceNode(
    Backward=true
    Control=(Type=Control,Name="CTRL_C_teethU")
    Ctrl_L_Token=CTRL_L_
    Ctrl_R_Token=CTRL_R_
    Curve_Side_Token_Last=true
    X_Mirror=false
    X_Pos_Neg_Expr=((Type=Curve,Name="CTRL_expressions_teethLeftU"),(Type=Curve,Name="CTRL_expressions_teethRightU"))
    Y_Mirror=false
    Y_Pos_Neg_Expr=((Type=Curve,Name="CTRL_expressions_teethUpU"),(Type=Curve,Name="CTRL_expressions_teethDownU"))
    Z_Mirror=false
    Z_Pos_Neg_Expr=((Type=Curve,Name="CTRL_expressions_teethBackU"),(Type=Curve,Name="CTRL_expressions_teethFwdU"))
   )
   > CRFL AnimAttrToCurve v01_28 = RigVMFunctionReferenceNode(
    Backward=true
    Channel=rotateFwdBack
    Control=(Type=Control,Name="CTRL_C_teethU")
    Pos_Neg_Curves=((Type=Curve,Name="CTRL_expressions_teethRotateFwdU"),(Type=Curve,Name="CTRL_expressions_teethRotateBackU"))
   )
   > FN_VectorCtrlToCurve_FNC_16_1 = RigVMFunctionReferenceNode(
    Backward=true
    Control=(Type=Control,Name="CTRL_C_teethD")
    Ctrl_L_Token=CTRL_L_
    Ctrl_R_Token=CTRL_R_
    Curve_Side_Token_Last=true
    X_Mirror=false
    X_Pos_Neg_Expr=((Type=Curve,Name="CTRL_expressions_teethLeftD"),(Type=Curve,Name="CTRL_expressions_teethRightD"))
    Y_Mirror=false
    Y_Pos_Neg_Expr=((Type=Curve,Name="CTRL_expressions_teethUpD"),(Type=Curve,Name="CTRL_expressions_teethDownD"))
    Z_Mirror=false
    Z_Pos_Neg_Expr=((Type=Curve,Name="CTRL_expressions_teethBackD"),(Type=Curve,Name="CTRL_expressions_teethFwdD"))
   )
   > CRFL AnimAttrToCurve v01_29 = RigVMFunctionReferenceNode(
    Backward=true
    Channel=rotateFwdBack
    Control=(Type=Control,Name="CTRL_C_teethD")
    Pos_Neg_Curves=((Type=Curve,Name="CTRL_expressions_teethRotateFwdD"),(Type=Curve,Name="CTRL_expressions_teethRotateBackD"))
   )
   > For_Each_1 = RigVMDispatchNode(Array=&RigUnit_CollectionNameSearchArray(PartialName=_Offset_, TypeToSearch=Curve).Items) {template="DISPATCH_RigVMDispatch_ArrayIterator(in Array,out Element,out Index,out Count,out Ratio)"}
   StartsWith = &RigVMFunction_StartsWith(Start=WM_, Name=@For_Each_1.Element.Name) {template="StartsWith::Execute(in Name,in Start,out Result)"}
   n12 = &RigVMFunction_ControlFlowBranch(ExecuteContext=@For_Each_1.ExecuteContext, Condition=@StartsWith.Result)
   /Script/ControlRig.RigUnit_SetFloatAnimationChannel(
    Control=CTRL_WrinkleMap
    bInitial=False
    Value=&RigUnit_GetCurveValue(Curve=@For_Each_1.Element.Name).Value
    ExecutePin=@n12.True
    Channel=@For_Each_1.Element.Name
   ) {is=struct RigVMUnitNode, template="SetAnimationChannel::Execute(in Value,in Control,in Channel,in bInitial)"}
   RigVMFunctionReturnNode(ExecuteContext=@RigVMFunction_Sequence.B)
   /Script/ControlRig.RigUnit_FindItemsWithMetadataTag(NameSpace=Self, Tag=wm_curves) {is=struct RigVMUnitNode}
  ]}
 RigVMFunctionLibrary {nodes [
   CRFL FN MH OnFace Backward v01 = RigVMCollapseNode(ExecuteContext=())
   CRFL FN MH OnFace Construction v01 = RigVMCollapseNode(Control Color=(R=1,G=1,B=0,A=1), Control Scale=1, ExecuteContext=())
   CRFL FN MH OnFace Forward v01 = RigVMCollapseNode(ExecuteContext=())
  ]}
 RigVMFunction_Sequence_10_ContainedGraph {nodes [
   n1 = RigVMFunctionEntryNode()
   n2 = &RigVMFunction_Sequence(ExecuteContext=@n1.ExecuteContext)
   n3 = &RigVMFunction_Sequence(ExecuteContext=@n2.B)
   n4 = &RigVMFunction_Sequence(ExecuteContext=@n3.B)
   RigVMFunctionReturnNode(B=@n3.A, C=@n4.A, D=@n4.B, A=@n2.A)
  ]}
 RigVMFunction_Sequence_1_ContainedGraph {nodes [
   n1 = RigVMFunctionEntryNode()
   RigVMFunction_Sequence = &RigVMFunction_Sequence(ExecuteContext=@n1.ExecuteContext)
   n2 = &RigVMFunction_Sequence(ExecuteContext=@RigVMFunction_Sequence.B)
   RigVMFunctionReturnNode(A=@RigVMFunction_Sequence.A, B=@n2.A, C=@n2.B)
  ]}
 RigVMFunction_Sequence_6_ContainedGraph {nodes [
   n1 = RigVMFunctionEntryNode()
   n2 = &RigVMFunction_Sequence(ExecuteContext=@n1.ExecuteContext)
   n3 = &RigVMFunction_Sequence(ExecuteContext=@n2.B)
   RigVMFunctionReturnNode(B=@n3.A, C=@n3.B, A=@n2.A)
  ]}
 RigVMFunction_Sequence_6_ContainedGraph {nodes [
   n1 = RigVMFunctionEntryNode()
   n2 = &RigVMFunction_Sequence(ExecuteContext=@n1.ExecuteContext)
   n3 = &RigVMFunction_Sequence(ExecuteContext=@n2.B)
   RigVMFunctionReturnNode(B=@n3.A, C=@n3.B, A=@n2.A)
  ]}
 RigVMFunction_Sequence_8_ContainedGraph {nodes [
   n1 = RigVMFunctionEntryNode()
   n2 = &RigVMFunction_Sequence(ExecuteContext=@n1.ExecuteContext)
   n3 = &RigVMFunction_Sequence(ExecuteContext=@n2.B)
   RigVMFunctionReturnNode(B=@n3.A, C=@n3.B, A=@n2.A)
  ]}
 RigVMFunction_Sequence_8_ContainedGraph {nodes [
   n1 = RigVMFunctionEntryNode()
   n2 = &RigVMFunction_Sequence(ExecuteContext=@n1.ExecuteContext)
   n3 = &RigVMFunction_Sequence(ExecuteContext=@n2.B)
   RigVMFunctionReturnNode(A=@n2.A, B=@n3.A, C=@n3.B)
  ]}
 RigVMFunction_Sequence_9_ContainedGraph {nodes [
   n1 = RigVMFunctionEntryNode()
   n2 = &RigVMFunction_Sequence(ExecuteContext=@n1.ExecuteContext)
   n3 = &RigVMFunction_Sequence(ExecuteContext=@n2.B)
   n4 = &RigVMFunction_Sequence(ExecuteContext=@n3.B)
   RigVMFunctionReturnNode(B=@n3.A, C=@n4.A, D=@n4.B, A=@n2.A)
  ]}
 RigVMModel {nodes [
   CRFL FN MH OnFace Forward v01 = RigVMFunctionReferenceNode(ExecuteContext=())
   RigUnit_BeginExecution = &RigUnit_BeginExecution()
  ]}
 RigVMModel Rig {nodes [
   CRFL FN MH OnFace Construction v01 = RigVMFunctionReferenceNode(Control Color=(R=1,G=1,B=0,A=1), Control Scale=1, ExecuteContext=())
   &RigUnit_PrepareForExecution()
  ]}
 RigVMModel Rig_1 {nodes [
   CRFL FN MH OnFace Backward v01 = RigVMFunctionReferenceNode(ExecuteContext=())
   /Script/ControlRig.RigUnit_InverseExecution() {is=struct RigVMUnitNode}
  ]}
 SequenceAggregate_1_1_ContainedGraph {nodes [
   n1 = RigVMFunctionEntryNode()
   SequenceAggregate_1 = &RigVMFunction_Sequence(ExecuteContext=@n1.ExecuteContext)
   SequenceAggregate_1_1 = &RigVMFunction_Sequence(ExecuteContext=@SequenceAggregate_1.B)
   RigVMFunctionReturnNode(A=@SequenceAggregate_1.A, B=@SequenceAggregate_1_1.A, C=@SequenceAggregate_1_1.B)
  ]}
 SequenceAggregate_1_1_ContainedGraph {nodes [
   n1 = RigVMFunctionEntryNode()
   SequenceAggregate_1 = &RigVMFunction_Sequence(ExecuteContext=@n1.ExecuteContext)
   SequenceAggregate_1_1 = &RigVMFunction_Sequence(ExecuteContext=@SequenceAggregate_1.B)
   RigVMFunctionReturnNode(A=@SequenceAggregate_1.A, B=@SequenceAggregate_1_1.A, C=@SequenceAggregate_1_1.B)
  ]}
 SequenceAggregate_1_ContainedGraph {nodes [
   n1 = RigVMFunctionEntryNode()
   SequenceAggregate = &RigVMFunction_Sequence(ExecuteContext=@n1.ExecuteContext)
   SequenceAggregate_1 = &RigVMFunction_Sequence(ExecuteContext=@SequenceAggregate.B)
   SequenceAggregate_2 = &RigVMFunction_Sequence(ExecuteContext=@SequenceAggregate_1.B)
   RigVMFunctionReturnNode(A=@SequenceAggregate.A, B=@SequenceAggregate_1.A, C=@SequenceAggregate_2.A, D=@SequenceAggregate_2.B)
  ]}
 SequenceAggregate_1_ContainedGraph {nodes [
   n1 = RigVMFunctionEntryNode()
   SequenceAggregate = &RigVMFunction_Sequence(ExecuteContext=@n1.ExecuteContext)
   SequenceAggregate_1 = &RigVMFunction_Sequence(ExecuteContext=@SequenceAggregate.B)
   SequenceAggregate_2 = &RigVMFunction_Sequence(ExecuteContext=@SequenceAggregate_1.B)
   RigVMFunctionReturnNode(A=@SequenceAggregate.A, B=@SequenceAggregate_1.A, C=@SequenceAggregate_2.A, D=@SequenceAggregate_2.B)
  ]}
 SequenceAggregate_2_1_ContainedGraph {nodes [
   n1 = RigVMFunctionEntryNode()
   SequenceAggregate_2 = &RigVMFunction_Sequence(ExecuteContext=@n1.ExecuteContext)
   SequenceAggregate_2_1 = &RigVMFunction_Sequence(ExecuteContext=@SequenceAggregate_2.B)
   RigVMFunctionReturnNode(A=@SequenceAggregate_2.A, B=@SequenceAggregate_2_1.A, C=@SequenceAggregate_2_1.B)
  ]}
 SequenceAggregate_2_1_ContainedGraph {nodes [
   n1 = RigVMFunctionEntryNode()
   SequenceAggregate_2 = &RigVMFunction_Sequence(ExecuteContext=@n1.ExecuteContext)
   SequenceAggregate_2_1 = &RigVMFunction_Sequence(ExecuteContext=@SequenceAggregate_2.B)
   RigVMFunctionReturnNode(A=@SequenceAggregate_2.A, B=@SequenceAggregate_2_1.A, C=@SequenceAggregate_2_1.B)
  ]}
 SequenceAggregate_5_ContainedGraph {nodes [
   n1 = RigVMFunctionEntryNode()
   SequenceAggregate_4 = &RigVMFunction_Sequence(ExecuteContext=@n1.ExecuteContext)
   SequenceAggregate_1 = &RigVMFunction_Sequence(ExecuteContext=@SequenceAggregate_4.B)
   RigVMFunctionReturnNode(B=@SequenceAggregate_1.A, C=@SequenceAggregate_1.B, A=@SequenceAggregate_4.A)
  ]}
]
hierarchy [
 root {loc=<0 0 0>, rot=<0 0 0 1>}
 pelvis {parent=root, loc=<0 -0.7071 93.1191>, rot=<0.0127 0.707 -0.0127 -0.707>}
 spine_01 {parent=pelvis, loc=<6.7494 -0.6388 0>, rot=<0 0 0.034 -0.9994>}
 spine_02 {parent=spine_01, loc=<7.2518 -0.035 0>, rot=<0 0 0.0238 -0.9997>}
 spine_03 {parent=spine_02, loc=<6.0104 0 0>, rot=<0 0 0.0149 -0.9999>}
 spine_04 {parent=spine_03, loc=<8.1024 0 0>, rot=<0 0 -0.0132 -0.9999>}
 spine_05 {parent=spine_04, loc=<10.0146 0 0>, rot=<0 0 -0.0015 -1>}
 neck_01 {parent=spine_05, loc=<11.3138 0.1529 0>, rot=<0 0 -0.2003 -0.9797>}
 neck_02 {parent=neck_01, loc=<4.5627 0 0>, rot=<0 0 0.0504 -0.9987>}
 head {parent=neck_02, loc=<5.3914 0 0>, rot=<0 0 0.115 -0.9934>}
 FACIAL_C_FacialRoot {parent=head, loc=<2.9712 -0.4622 0>, rot=<0.5021 -0.4979 -0.5021 -0.4979>}
 FACIAL_C_Forehead {parent=FACIAL_C_FacialRoot, loc=<0 -3.1308 9.3974>, rot=<0.0207 0 0 -0.9998>}
 FACIAL_L_ForeheadIn {parent=FACIAL_C_FacialRoot, loc=<1.5783 -2.9083 9.9779>, rot=<0.0031 -0.0674 0.0105 -0.9977>}
 FACIAL_R_ForeheadIn {parent=FACIAL_C_FacialRoot, loc=<-1.5783 -2.9083 9.9779>, rot=<0.0031 0.0674 -0.0105 -0.9977>}
 FACIAL_L_ForeheadMid {parent=FACIAL_C_FacialRoot, loc=<2.867 -3.3915 9.4422>, rot=<-0.0082 -0.1749 -0.0076 -0.9845>}
 FACIAL_L_ForeheadMid1 {parent=FACIAL_L_ForeheadMid, loc=<-0.5077 0.2381 -0.0995>, rot=<0 0 0 1>}
 FACIAL_L_ForeheadMid2 {parent=FACIAL_L_ForeheadMid, loc=<0.674 0.1716 -0.0884>, rot=<0 0 0 1>}
 FACIAL_R_ForeheadMid {parent=FACIAL_C_FacialRoot, loc=<-2.867 -3.3915 9.4422>, rot=<-0.0082 0.1749 0.0076 -0.9845>}
 FACIAL_R_ForeheadMid1 {parent=FACIAL_R_ForeheadMid, loc=<0.5077 0.2381 -0.0995>, rot=<0 0 0 1>}
 FACIAL_R_ForeheadMid2 {parent=FACIAL_R_ForeheadMid, loc=<-0.674 0.1716 -0.0884>, rot=<0 0 0 1>}
 FACIAL_L_ForeheadOut {parent=FACIAL_C_FacialRoot, loc=<4.0465 -2.9836 8.7213>, rot=<-0.0602 -0.2584 -0.0611 -0.9622>}
 FACIAL_R_ForeheadOut {parent=FACIAL_C_FacialRoot, loc=<-4.0465 -2.9836 8.7213>, rot=<-0.0602 0.2584 0.0611 -0.9622>}
 FACIAL_L_EyesackUpper {parent=FACIAL_C_FacialRoot, loc=<2.8439 -2.3877 9.3504>, rot=<0.1124 -0.0867 -0.0571 -0.9882>}
 FACIAL_R_EyesackUpper {parent=FACIAL_C_FacialRoot, loc=<-2.8439 -2.3877 9.3504>, rot=<0.1124 0.0867 0.0571 -0.9882>}
 FACIAL_L_EyelidUpperFurrow {parent=FACIAL_C_FacialRoot, loc=<2.6585 -2.0955 9.0119>, rot=<0.0331 -0.154 -0.068 -0.9852>}
 FACIAL_R_EyelidUpperFurrow {parent=FACIAL_C_FacialRoot, loc=<-2.6585 -2.0955 9.0119>, rot=<0.0331 0.154 0.068 -0.9852>}
 FACIAL_L_EyelidUpperB {parent=FACIAL_C_FacialRoot, loc=<2.6949 -1.4649 7.7068>, rot=<-0.1618 0.002 0.0014 -0.9868>}
 FACIAL_R_EyelidUpperB {parent=FACIAL_C_FacialRoot, loc=<-2.6949 -1.4649 7.7068>, rot=<-0.1618 -0.002 -0.0014 -0.9868>}
 FACIAL_L_EyelidUpperA {parent=FACIAL_C_FacialRoot, loc=<2.6949 -1.4649 7.7068>, rot=<-0.0839 0.0019 0.0016 -0.9965>}
 FACIAL_L_EyelidUpperA1 {parent=FACIAL_L_EyelidUpperA, loc=<-0.6817 0.0696 1.1303>, rot=<0.0106 0.1214 0.0865 -0.9888>}
 FACIAL_L_EyelidUpperA3 {parent=FACIAL_L_EyelidUpperA, loc=<0.7566 -0.0356 1.1039>, rot=<-0.0086 -0.1657 -0.0487 -0.9849>}
 FACIAL_R_EyelidUpperA {parent=FACIAL_C_FacialRoot, loc=<-2.6949 -1.4649 7.7068>, rot=<-0.0839 -0.0019 -0.0016 -0.9965>}
 FACIAL_R_EyelidUpperA1 {parent=FACIAL_R_EyelidUpperA, loc=<0.6817 0.0696 1.1303>, rot=<0.0106 -0.1214 -0.0865 -0.9888>}
 FACIAL_R_EyelidUpperA3 {parent=FACIAL_R_EyelidUpperA, loc=<-0.7566 -0.0356 1.1039>, rot=<-0.0086 0.1657 0.0487 -0.9849>}
 FACIAL_L_Eye {parent=FACIAL_C_FacialRoot, loc=<2.6949 -1.4649 7.7068>, rot=<0.0643 0.0017 0.0019 -0.9979>}
 FACIAL_L_Pupil {parent=FACIAL_L_Eye, loc=<0.1563 -0.1178 1.1715>, rot=<0 0 0 1>}
 FACIAL_R_Eye {parent=FACIAL_C_FacialRoot, loc=<-2.6949 -1.4649 7.7068>, rot=<0.0643 -0.0017 -0.0019 -0.9979>}
 FACIAL_R_Pupil {parent=FACIAL_R_Eye, loc=<-0.1563 -0.1178 1.1715>, rot=<0 0 0 1>}
 FACIAL_L_EyelidLowerA {parent=FACIAL_C_FacialRoot, loc=<2.6949 -1.4649 7.7068>, rot=<0.3451 0.0011 0.0023 -0.9386>}
 FACIAL_L_EyelidLowerA1 {parent=FACIAL_L_EyelidLowerA, loc=<-0.7483 -0.3072 1.1061>, rot=<0 0.0785 0 -0.9969>}
 FACIAL_L_EyelidLowerA3 {parent=FACIAL_L_EyelidLowerA, loc=<0.6735 -0.1663 1.1484>, rot=<0 -0.1822 0 -0.9833>}
 FACIAL_R_EyelidLowerA {parent=FACIAL_C_FacialRoot, loc=<-2.6949 -1.4649 7.7068>, rot=<0.3451 -0.0011 -0.0023 -0.9386>}
 FACIAL_R_EyelidLowerA1 {parent=FACIAL_R_EyelidLowerA, loc=<0.7483 -0.3072 1.1061>, rot=<0 -0.0785 0 -0.9969>}
 FACIAL_R_EyelidLowerA3 {parent=FACIAL_R_EyelidLowerA, loc=<-0.6735 -0.1663 1.1484>, rot=<0 0.1822 0 -0.9833>}
 FACIAL_L_EyelidLowerB {parent=FACIAL_C_FacialRoot, loc=<2.6949 -1.4649 7.7068>, rot=<0.3776 0.001 0.0023 -0.926>}
 FACIAL_L_EyelidLowerB1 {parent=FACIAL_L_EyelidLowerB, loc=<-0.7153 -0.1974 1.2808>, rot=<0 0.0785 0 -0.9969>}
 FACIAL_L_EyelidLowerB3 {parent=FACIAL_L_EyelidLowerB, loc=<0.7461 0.0443 1.2303>, rot=<0 -0.1822 0 -0.9833>}
 FACIAL_R_EyelidLowerB {parent=FACIAL_C_FacialRoot, loc=<-2.6949 -1.4649 7.7068>, rot=<0.3776 -0.001 -0.0023 -0.926>}
 FACIAL_R_EyelidLowerB1 {parent=FACIAL_R_EyelidLowerB, loc=<0.7153 -0.1974 1.2808>, rot=<0 -0.0785 0 -0.9969>}
 FACIAL_R_EyelidLowerB3 {parent=FACIAL_R_EyelidLowerB, loc=<-0.7461 0.0443 1.2303>, rot=<0 0.1822 0 -0.9833>}
 FACIAL_L_EyeCornerInner {parent=FACIAL_C_FacialRoot, loc=<1.7099 -1.1996 8.5226>, rot=<0.0643 0.0017 0.0019 -0.9979>}
 FACIAL_R_EyeCornerInner {parent=FACIAL_C_FacialRoot, loc=<-1.7099 -1.1996 8.5226>, rot=<0.0643 -0.0017 -0.0019 -0.9979>}
 FACIAL_L_EyeCornerOuter {parent=FACIAL_C_FacialRoot, loc=<3.9155 -1.2959 7.935>, rot=<0.0587 -0.3803 0.0263 -0.9226>}
 FACIAL_R_EyeCornerOuter {parent=FACIAL_C_FacialRoot, loc=<-3.9155 -1.2959 7.935>, rot=<0.0587 0.3803 -0.0263 -0.9226>}
 FACIAL_L_EyesackLower {parent=FACIAL_C_FacialRoot, loc=<2.8327 -0.1952 8.2854>, rot=<0.0412 -0.1392 -0.0114 -0.9893>}
 FACIAL_L_EyesackLower1 {parent=FACIAL_L_EyesackLower, loc=<-0.992 -0.1607 0.241>, rot=<0.0204 0.1999 -0.0204 -0.9794>}
 FACIAL_L_EyesackLower2 {parent=FACIAL_L_EyesackLower, loc=<0.9592 -0.1302 0.2257>, rot=<0.0082 -0.0885 0.1026 -0.9907>}
 FACIAL_R_EyesackLower {parent=FACIAL_C_FacialRoot, loc=<-2.8327 -0.1952 8.2854>, rot=<0.0412 0.1392 0.0114 -0.9893>}
 FACIAL_R_EyesackLower1 {parent=FACIAL_R_EyesackLower, loc=<0.992 -0.1607 0.241>, rot=<0.0204 -0.1999 0.0204 -0.9794>}
 FACIAL_R_EyesackLower2 {parent=FACIAL_R_EyesackLower, loc=<-0.9592 -0.1302 0.2257>, rot=<0.0082 0.0885 -0.1026 -0.9907>}
 FACIAL_L_CheekInner {parent=FACIAL_C_FacialRoot, loc=<1.8741 0.5243 8.2984>, rot=<-0.0606 -0.0987 -0.1895 -0.975>}
 FACIAL_L_CheekInner1 {parent=FACIAL_L_CheekInner, loc=<-0.5285 -0.2617 0.8209>, rot=<0.0509 -0.0536 0.0234 -0.997>}
 FACIAL_L_CheekInner2 {parent=FACIAL_L_CheekInner, loc=<-0.319 0.3869 1.159>, rot=<0.0349 -0.0565 -0.1491 -0.9866>}
 FACIAL_L_CheekInner3 {parent=FACIAL_L_CheekInner, loc=<0.5554 -0.1496 0.7785>, rot=<0.0509 -0.0536 0.0234 -0.997>}
 FACIAL_L_CheekInner4 {parent=FACIAL_L_CheekInner, loc=<0.6441 0.7311 0.9265>, rot=<0.0349 -0.0565 -0.1491 -0.9866>}
 FACIAL_R_CheekInner {parent=FACIAL_C_FacialRoot, loc=<-1.8741 0.5243 8.2984>, rot=<-0.0606 0.0987 0.1895 -0.975>}
 FACIAL_R_CheekInner1 {parent=FACIAL_R_CheekInner, loc=<0.5285 -0.2617 0.8209>, rot=<0.0509 0.0536 -0.0234 -0.997>}
 FACIAL_R_CheekInner2 {parent=FACIAL_R_CheekInner, loc=<0.319 0.3869 1.159>, rot=<0.0349 0.0565 0.1491 -0.9866>}
 FACIAL_R_CheekInner3 {parent=FACIAL_R_CheekInner, loc=<-0.5554 -0.1496 0.7785>, rot=<0.0509 0.0536 -0.0234 -0.997>}
 FACIAL_R_CheekInner4 {parent=FACIAL_R_CheekInner, loc=<-0.6441 0.7311 0.9265>, rot=<0.0349 0.0565 0.1491 -0.9866>}
 FACIAL_L_CheekOuter {parent=FACIAL_C_FacialRoot, loc=<3.3414 0.7955 6.9957>, rot=<0.1132 -0.3684 0.1262 -0.9141>}
 FACIAL_L_CheekOuter1 {parent=FACIAL_L_CheekOuter, loc=<-0.5843 -0.5617 1.2992>, rot=<-0.0677 -0.0465 -0.0404 -0.9958>}
 FACIAL_L_CheekOuter2 {parent=FACIAL_L_CheekOuter, loc=<0.0775 1.0309 1.24>, rot=<0.0801 -0.0306 -0.0549 -0.9948>}
 FACIAL_L_CheekOuter3 {parent=FACIAL_L_CheekOuter, loc=<1.0345 -0.5023 1.2182>, rot=<-0.0491 -0.0605 0.2182 -0.9728>}
 FACIAL_R_CheekOuter {parent=FACIAL_C_FacialRoot, loc=<-3.3414 0.7955 6.9957>, rot=<0.1132 0.3684 -0.1262 -0.9141>}
 FACIAL_R_CheekOuter1 {parent=FACIAL_R_CheekOuter, loc=<0.5843 -0.5617 1.2992>, rot=<-0.0677 0.0465 0.0404 -0.9958>}
 FACIAL_R_CheekOuter2 {parent=FACIAL_R_CheekOuter, loc=<-0.0775 1.0309 1.24>, rot=<0.0801 0.0306 0.0549 -0.9948>}
 FACIAL_R_CheekOuter3 {parent=FACIAL_R_CheekOuter, loc=<-1.0345 -0.5023 1.2182>, rot=<-0.0491 0.0605 -0.2182 -0.9728>}
 FACIAL_C_NoseBridge {parent=FACIAL_C_FacialRoot, loc=<0 -1.8959 10.1812>, rot=<0.1078 0 0 -0.9942>}
 FACIAL_L_NasolabialBulge1 {parent=FACIAL_C_FacialRoot, loc=<3.2494 2.3983 8.6467>, rot=<-0.025 -0.2497 -0.3656 -0.8963>}
 FACIAL_R_NasolabialBulge1 {parent=FACIAL_C_FacialRoot, loc=<-3.2494 2.3983 8.6467>, rot=<-0.025 0.2497 0.3656 -0.8963>}
 FACIAL_L_NasolabialBulge {parent=FACIAL_C_FacialRoot, loc=<3.0622 4.338 7.415>, rot=<0.1164 -0.3737 0.0498 -0.9189>}
 FACIAL_L_NasolabialBulge2 {parent=FACIAL_L_NasolabialBulge, loc=<0.0915 -0.9316 0.7307>, rot=<-0.1022 0.041 0.057 -0.9923>}
 FACIAL_L_NasolabialBulge3 {parent=FACIAL_L_NasolabialBulge, loc=<-0.1478 0.616 0.8109>, rot=<-0.0099 0.0168 -0.0699 -0.9974>}
 FACIAL_R_NasolabialBulge {parent=FACIAL_C_FacialRoot, loc=<-3.0622 4.338 7.415>, rot=<0.1164 0.3737 -0.0498 -0.9189>}
 FACIAL_R_NasolabialBulge2 {parent=FACIAL_R_NasolabialBulge, loc=<-0.0915 -0.9316 0.7307>, rot=<-0.1022 -0.041 -0.057 -0.9923>}
 FACIAL_R_NasolabialBulge3 {parent=FACIAL_R_NasolabialBulge, loc=<0.1478 0.616 0.8109>, rot=<-0.0099 -0.0168 0.0699 -0.9974>}
 FACIAL_L_NasolabialFurrow {parent=FACIAL_C_FacialRoot, loc=<2.7204 2.6119 9.0327>, rot=<0.0264 -0.1474 -0.3533 -0.9235>}
 FACIAL_R_NasolabialFurrow {parent=FACIAL_C_FacialRoot, loc=<-2.7204 2.6119 9.0327>, rot=<0.0264 0.1474 0.3533 -0.9235>}
 FACIAL_L_CheekLower {parent=FACIAL_C_FacialRoot, loc=<3.5517 3.9289 6.316>, rot=<0.0916 -0.4209 0.0868 -0.8983>}
 FACIAL_L_CheekLower1 {parent=FACIAL_L_CheekLower, loc=<-0.3052 0.2977 0.8335>, rot=<0 0.0872 0 -0.9962>}
 FACIAL_L_CheekLower2 {parent=FACIAL_L_CheekLower, loc=<0.7344 0.1917 0.5743>, rot=<0 -0.0872 0 -0.9962>}
 FACIAL_R_CheekLower {parent=FACIAL_C_FacialRoot, loc=<-3.5517 3.9289 6.316>, rot=<0.0916 0.4209 -0.0868 -0.8983>}
 FACIAL_R_CheekLower1 {parent=FACIAL_R_CheekLower, loc=<0.3052 0.2977 0.8335>, rot=<0 -0.0872 0 -0.9962>}
 FACIAL_R_CheekLower2 {parent=FACIAL_R_CheekLower, loc=<-0.7344 0.1917 0.5743>, rot=<0 0.0872 0 -0.9962>}
 FACIAL_L_Ear {parent=FACIAL_C_FacialRoot, loc=<6.4207 0.0696 1.4054>, rot=<0.0382 -0.0006 -0.0243 -0.999>}
 FACIAL_R_Ear {parent=FACIAL_C_FacialRoot, loc=<-6.4207 0.0696 1.4054>, rot=<0.0382 0.0006 0.0243 -0.999>}
 FACIAL_C_Nose {parent=FACIAL_C_FacialRoot, loc=<0 0.1218 10.1331>, rot=<-0.0665 0 0 -0.9978>}
 FACIAL_C_NoseLower {parent=FACIAL_C_Nose, loc=<0 2.611 0.1959>, rot=<0.1305 0 0 -0.9914>}
 FACIAL_C_NoseTip {parent=FACIAL_C_Nose, loc=<0 1.7381 0.7792>, rot=<0.0872 0 0 -0.9962>}
 FACIAL_L_Nostril {parent=FACIAL_C_Nose, loc=<1.3616 1.8117 0.0109>, rot=<0.0173 0.1305 0.0023 -0.9913>}
 FACIAL_R_Nostril {parent=FACIAL_C_Nose, loc=<-1.3616 1.8117 0.0109>, rot=<0.0173 -0.1305 -0.0023 -0.9913>}
 FACIAL_C_LipUpperSkin {parent=FACIAL_C_FacialRoot, loc=<0 3.1175 10.6656>, rot=<0.0207 0 0 -0.9998>}
 FACIAL_L_LipUpperOuterSkin {parent=FACIAL_C_FacialRoot, loc=<2.4826 3.4496 9.4024>, rot=<-0.0193 -0.1728 -0.0993 -0.9797>}
 FACIAL_R_LipUpperOuterSkin {parent=FACIAL_C_FacialRoot, loc=<-2.4826 3.4496 9.4024>, rot=<-0.0193 0.1728 0.0993 -0.9797>}
 FACIAL_C_MouthUpper {parent=FACIAL_C_FacialRoot, loc=<0 4.4644 7.5076>, rot=<0.0643 0 0 -0.9979>}
 FACIAL_C_LipUpper {parent=FACIAL_C_MouthUpper, loc=<0 -0.9938 2.4973>, rot=<-0.0872 0 0 -0.9962>}
 FACIAL_C_LipUpper1 {parent=FACIAL_C_LipUpper, loc=<0 -0.1465 0.8579>, rot=<0 0 0 1>}
 FACIAL_C_LipUpper2 {parent=FACIAL_C_LipUpper, loc=<0 0.6591 0.1943>, rot=<0 0 0 1>}
 FACIAL_L_LipUpper {parent=FACIAL_C_MouthUpper, loc=<1.1829 -0.8863 2.1826>, rot=<-0.0778 -0.1601 -0.0379 -0.9833>}
 FACIAL_L_LipUpper1 {parent=FACIAL_L_LipUpper, loc=<-0.0571 -0.1403 0.6326>, rot=<0 0 0 1>}
 FACIAL_L_LipUpper2 {parent=FACIAL_L_LipUpper, loc=<-0.1281 0.5669 0.1244>, rot=<0 0 0 1>}
 FACIAL_R_LipUpper {parent=FACIAL_C_MouthUpper, loc=<-1.1829 -0.8863 2.1826>, rot=<-0.0778 0.1601 0.0379 -0.9833>}
 FACIAL_R_LipUpper1 {parent=FACIAL_R_LipUpper, loc=<0.0571 -0.1403 0.6326>, rot=<0 0 0 1>}
 FACIAL_R_LipUpper2 {parent=FACIAL_R_LipUpper, loc=<0.1281 0.5669 0.1244>, rot=<0 0 0 1>}
 FACIAL_L_LipUpperOuter {parent=FACIAL_C_MouthUpper, loc=<1.9999 -0.6965 1.5448>, rot=<-0.0983 -0.2294 -0.0983 -0.9634>}
 FACIAL_L_LipUpperOuter1 {parent=FACIAL_L_LipUpperOuter, loc=<-0.0959 0.1942 0.5095>, rot=<0 0 0 1>}
 FACIAL_L_LipUpperOuter2 {parent=FACIAL_L_LipUpperOuter, loc=<-0.2529 0.5561 0.2264>, rot=<0 0 0 1>}
 FACIAL_R_LipUpperOuter {parent=FACIAL_C_MouthUpper, loc=<-1.9999 -0.6965 1.5448>, rot=<-0.0983 0.2294 0.0983 -0.9634>}
 FACIAL_R_LipUpperOuter1 {parent=FACIAL_R_LipUpperOuter, loc=<0.0959 0.1942 0.5095>, rot=<0 0 0 1>}
 FACIAL_R_LipUpperOuter2 {parent=FACIAL_R_LipUpperOuter, loc=<0.2529 0.5561 0.2264>, rot=<0 0 0 1>}
 FACIAL_L_LipCorner {parent=FACIAL_C_MouthUpper, loc=<2.619 -0.1516 1.046>, rot=<0 -0.2079 0 -0.9781>}
 FACIAL_L_LipCorner2 {parent=FACIAL_L_LipCorner, loc=<-0.1424 -0.0129 -0.7513>, rot=<0 0.2588 0 -0.9659>}
 FACIAL_L_LipCorner3 {parent=FACIAL_L_LipCorner, loc=<-0.4025 0.5791 -0.5869>, rot=<0 0.2588 0 -0.9659>}
 FACIAL_R_LipCorner {parent=FACIAL_C_MouthUpper, loc=<-2.6191 -0.1516 1.046>, rot=<0 0.2079 0 -0.9781>}
 FACIAL_R_LipCorner2 {parent=FACIAL_R_LipCorner, loc=<0.1424 -0.0129 -0.7513>, rot=<0 -0.2588 0 -0.9659>}
 FACIAL_R_LipCorner3 {parent=FACIAL_R_LipCorner, loc=<0.4025 0.5791 -0.5869>, rot=<0 -0.2588 0 -0.9659>}
 FACIAL_L_JawBulge {parent=FACIAL_C_FacialRoot, loc=<5.2193 3.4628 3.4629>, rot=<0.1506 0.0886 0.004 -0.9846>}
 FACIAL_R_JawBulge {parent=FACIAL_C_FacialRoot, loc=<-5.2193 3.4628 3.4629>, rot=<0.1506 -0.0886 -0.004 -0.9846>}
 FACIAL_L_Masseter {parent=FACIAL_C_FacialRoot, loc=<5.3206 1.6285 4.5667>, rot=<0.1331 0.1147 0.0026 -0.9844>}
 FACIAL_R_Masseter {parent=FACIAL_C_FacialRoot, loc=<-5.3206 1.6285 4.5667>, rot=<0.1331 -0.1147 -0.0026 -0.9844>}
 FACIAL_C_UnderChin {parent=FACIAL_C_FacialRoot, loc=<0 8.5606 6.1576>, rot=<0.5546 0 0 -0.8321>}
 FACIAL_L_UnderChin {parent=FACIAL_C_FacialRoot, loc=<3.1037 7.3473 4.3729>, rot=<0.4548 -0.3483 0.1072 -0.8126>}
 FACIAL_R_UnderChin {parent=FACIAL_C_FacialRoot, loc=<-3.1037 7.3473 4.3729>, rot=<0.4548 0.3483 -0.1072 -0.8126>}
 FACIAL_C_TeethUpper {parent=FACIAL_C_FacialRoot, loc=<0 4.3737 9.6558>, rot=<0.0643 0 0 -0.9979>}
 FACIAL_C_LowerLipRotation {parent=FACIAL_C_FacialRoot, loc=<0 1.9413 2.9092>, rot=<0.4017 0 0 -0.9158>}
 FACIAL_C_LipLowerSkin {parent=FACIAL_C_LowerLipRotation, loc=<0 -2.158 8.1295>, rot=<-0.2588 0 0 -0.9659>}
 FACIAL_L_LipLowerOuterSkin {parent=FACIAL_C_LowerLipRotation, loc=<2.5044 -2.0345 6.9892>, rot=<-0.2809 -0.2529 0.0021 -0.9258>}
 FACIAL_R_LipLowerOuterSkin {parent=FACIAL_C_LowerLipRotation, loc=<-2.5044 -2.0345 6.9892>, rot=<-0.2809 0.2529 -0.0021 -0.9258>}
 FACIAL_C_MouthLower {parent=FACIAL_C_LowerLipRotation, loc=<0 -1.6746 4.9707>, rot=<-0.342 0 0 -0.9397>}
 FACIAL_C_LipLower {parent=FACIAL_C_MouthLower, loc=<0 0.5544 2.7134>, rot=<0.0872 0 0 -0.9962>}
 FACIAL_C_LipLower1 {parent=FACIAL_C_LipLower, loc=<0 0.168 0.5115>, rot=<0 0 0 1>}
 FACIAL_C_LipLower2 {parent=FACIAL_C_LipLower, loc=<0 -0.8438 -0.3887>, rot=<0 0 0 1>}
 FACIAL_L_LipLower {parent=FACIAL_C_MouthLower, loc=<1.1262 0.4533 2.3553>, rot=<0.0927 -0.1295 0.0273 -0.9869>}
 FACIAL_L_LipLower1 {parent=FACIAL_L_LipLower, loc=<0.2662 0.1675 0.5708>, rot=<0 0 0 1>}
 FACIAL_L_LipLower2 {parent=FACIAL_L_LipLower, loc=<-0.0195 -0.7652 -0.2768>, rot=<0 0 0 1>}
 FACIAL_R_LipLower {parent=FACIAL_C_MouthLower, loc=<-1.1263 0.4533 2.3553>, rot=<0.0927 0.1295 -0.0273 -0.9869>}
 FACIAL_R_LipLower1 {parent=FACIAL_R_LipLower, loc=<-0.2662 0.1675 0.5708>, rot=<0 0 0 1>}
 FACIAL_R_LipLower2 {parent=FACIAL_R_LipLower, loc=<0.0195 -0.7652 -0.2768>, rot=<0 0 0 1>}
 FACIAL_L_LipLowerOuter {parent=FACIAL_C_MouthLower, loc=<1.8907 0.2989 1.9184>, rot=<0.0913 -0.2322 0.0318 -0.9679>}
 FACIAL_L_LipLowerOuter1 {parent=FACIAL_L_LipLowerOuter, loc=<0.1099 0.008 0.4186>, rot=<0 0 0 -1>}
 FACIAL_L_LipLowerOuter2 {parent=FACIAL_L_LipLowerOuter, loc=<-0.0021 -0.562 -0.1956>, rot=<0 0 0 -1>}
 FACIAL_R_LipLowerOuter {parent=FACIAL_C_MouthLower, loc=<-1.8907 0.2989 1.9184>, rot=<0.0913 0.2322 -0.0318 -0.9679>}
 FACIAL_R_LipLowerOuter1 {parent=FACIAL_R_LipLowerOuter, loc=<-0.1099 0.008 0.4186>, rot=<0 0 0 -1>}
 FACIAL_R_LipLowerOuter2 {parent=FACIAL_R_LipLowerOuter, loc=<0.0021 -0.562 -0.1956>, rot=<0 0 0 -1>}
 FACIAL_C_TeethLower {parent=FACIAL_C_LowerLipRotation, loc=<0 -3.3107 6.3943>, rot=<-0.342 0 0 -0.9397>}
 FACIAL_C_Tongue1 {parent=FACIAL_C_TeethLower, loc=<0 0.8814 -4.9974>, rot=<-0.2588 0 0 -0.9659>}
 FACIAL_C_Tongue2 {parent=FACIAL_C_Tongue1, loc=<0 0.4172 1.4811>, rot=<0.2588 0 0 -0.9659>}
 FACIAL_C_TongueUpper2 {parent=FACIAL_C_Tongue2, loc=<-0.0004 -0.5559 -0.0593>, rot=<0 0 0 1>}
 FACIAL_L_TongueSide2 {parent=FACIAL_C_Tongue2, loc=<1.7649 0.1107 0.3121>, rot=<0 0 0 1>}
 FACIAL_R_TongueSide2 {parent=FACIAL_C_Tongue2, loc=<-1.7649 0.1107 0.3121>, rot=<0 0 0 -1>}
 FACIAL_C_Tongue3 {parent=FACIAL_C_Tongue2, loc=<0 0.2069 1.3354>, rot=<0.1305 0 0 -0.9914>}
 FACIAL_C_TongueUpper3 {parent=FACIAL_C_Tongue3, loc=<-0.0003 -0.3509 -0.1572>, rot=<0 0 0 -1>}
 FACIAL_C_TongueLower3 {parent=FACIAL_C_Tongue3, loc=<0 0.8092 0.0947>, rot=<0 0 0 -1>}
 FACIAL_L_TongueSide3 {parent=FACIAL_C_Tongue3, loc=<1.5325 0.0121 0.1487>, rot=<0 0 0 -1>}
 FACIAL_R_TongueSide3 {parent=FACIAL_C_Tongue3, loc=<-1.5325 0.0121 0.1487>, rot=<0 0 0 -1>}
 FACIAL_C_Tongue4 {parent=FACIAL_C_Tongue3, loc=<0.0006 -0.1867 1.1247>, rot=<0.0436 0 0 -0.999>}
 FACIAL_C_Jaw {parent=FACIAL_C_FacialRoot, loc=<0 1.9413 2.9092>, rot=<0.4017 0 0 -0.9158>}
 FACIAL_C_Jawline {parent=FACIAL_C_Jaw, loc=<0 0.9567 8.3398>, rot=<0.2588 0 0 -0.9659>}
 FACIAL_L_Jawline {parent=FACIAL_C_Jaw, loc=<4.3349 0.5774 4.2447>, rot=<-0.1329 0.167 0.0026 -0.977>}
 FACIAL_R_Jawline {parent=FACIAL_C_Jaw, loc=<-4.3349 0.5774 4.2447>, rot=<-0.1329 -0.167 -0.0026 -0.977>}
 FACIAL_L_ChinSide {parent=FACIAL_C_Jaw, loc=<2.9641 -0.948 6.7957>, rot=<-0.2207 -0.3739 0.05 -0.8994>}
 FACIAL_R_ChinSide {parent=FACIAL_C_Jaw, loc=<-2.9641 -0.948 6.7957>, rot=<-0.2207 0.3739 -0.05 -0.8994>}
 FACIAL_C_Chin {parent=FACIAL_C_Jaw, loc=<0 -1.7958 8.7593>, rot=<-0.4226 0 0 -0.9063>}
 FACIAL_C_Chin3 {parent=FACIAL_C_Jaw, loc=<0 -0.3859 9.3079>, rot=<0.0436 0 0 -0.999>}
 clavicle_l {parent=spine_05, loc=<4.3575 -0.1466 -2.3223>, rot=<0.0465 -0.7289 -0.009 -0.6829>}
 upperarm_l {parent=clavicle_l, loc=<15.5111 0 0>, rot=<0.0226 -0.3234 -0.0088 -0.9459>}
 clavicle_r {parent=spine_05, loc=<4.3575 -0.1466 2.3223>, rot=<0.7289 0.0465 -0.6829 0.009>}
 upperarm_r {parent=clavicle_r, loc=<-15.5111 0 0>, rot=<0.0226 -0.3234 -0.0088 -0.9459>}
 CTRL_expressions_browDINL {type=Curve, loc=<0 0 0>, rot=<0 0 0 1>}
 CTRL_expressions_browDINR {type=Curve, loc=<0 0 0>, rot=<0 0 0 1>}
 CTRL_expressions_browDOUTL {type=Curve, loc=<0 0 0>, rot=<0 0 0 1>}
 CTRL_expressions_browDOUTR {type=Curve, loc=<0 0 0>, rot=<0 0 0 1>}
 CTRL_expressions_browDownL {type=Curve, loc=<0 0 0>, rot=<0 0 0 1>}
 CTRL_expressions_browDownR {type=Curve, loc=<0 0 0>, rot=<0 0 0 1>}
 CTRL_expressions_browInnerInL {type=Curve, loc=<0 0 0>, rot=<0 0 0 1>}
 CTRL_expressions_browInnerInR {type=Curve, loc=<0 0 0>, rot=<0 0 0 1>}
 CTRL_expressions_browInnerMidDownL {type=Curve, loc=<0 0 0>, rot=<0 0 0 1>}
 CTRL_expressions_browInnerMidDownR {type=Curve, loc=<0 0 0>, rot=<0 0 0 1>}
 CTRL_expressions_browInnerMidInL {type=Curve, loc=<0 0 0>, rot=<0 0 0 1>}
 CTRL_expressions_browInnerMidInR {type=Curve, loc=<0 0 0>, rot=<0 0 0 1>}
 CTRL_expressions_browInnerMidOutL {type=Curve, loc=<0 0 0>, rot=<0 0 0 1>}
 CTRL_expressions_browInnerMidOutR {type=Curve, loc=<0 0 0>, rot=<0 0 0 1>}
 CTRL_expressions_browInnerMidUpL {type=Curve, loc=<0 0 0>, rot=<0 0 0 1>}
 CTRL_expressions_browInnerMidUpR {type=Curve, loc=<0 0 0>, rot=<0 0 0 1>}
 CTRL_expressions_browInnerOutL {type=Curve, loc=<0 0 0>, rot=<0 0 0 1>}
 CTRL_expressions_browInnerOutR {type=Curve, loc=<0 0 0>, rot=<0 0 0 1>}
 CTRL_expressions_browLateralL {type=Curve, loc=<0 0 0>, rot=<0 0 0 1>}
 CTRL_expressions_browLateralR {type=Curve, loc=<0 0 0>, rot=<0 0 0 1>}
 CTRL_expressions_browOuterMidDownL {type=Curve, loc=<0 0 0>, rot=<0 0 0 1>}
 CTRL_expressions_browOuterMidDownR {type=Curve, loc=<0 0 0>, rot=<0 0 0 1>}
 CTRL_expressions_browOuterMidInL {type=Curve, loc=<0 0 0>, rot=<0 0 0 1>}
 CTRL_expressions_browOuterMidInR {type=Curve, loc=<0 0 0>, rot=<0 0 0 1>}
 CTRL_expressions_browOuterMidOutL {type=Curve, loc=<0 0 0>, rot=<0 0 0 1>}
 CTRL_expressions_browOuterMidOutR {type=Curve, loc=<0 0 0>, rot=<0 0 0 1>}
 CTRL_expressions_browOuterMidUpL {type=Curve, loc=<0 0 0>, rot=<0 0 0 1>}
 CTRL_expressions_browOuterMidUpR {type=Curve, loc=<0 0 0>, rot=<0 0 0 1>}
 CTRL_expressions_browRaiseInL {type=Curve, loc=<0 0 0>, rot=<0 0 0 1>}
 CTRL_expressions_browRaiseInR {type=Curve, loc=<0 0 0>, rot=<0 0 0 1>}
 CTRL_expressions_browRaiseOuterL {type=Curve, loc=<0 0 0>, rot=<0 0 0 1>}
 CTRL_expressions_browRaiseOuterR {type=Curve, loc=<0 0 0>, rot=<0 0 0 1>}
 CTRL_expressions_browRollNegInL {type=Curve, loc=<0 0 0>, rot=<0 0 0 1>}
 CTRL_expressions_browRollNegInR {type=Curve, loc=<0 0 0>, rot=<0 0 0 1>}
 CTRL_expressions_browRollNegOutL {type=Curve, loc=<0 0 0>, rot=<0 0 0 1>}
 CTRL_expressions_browRollNegOutR {type=Curve, loc=<0 0 0>, rot=<0 0 0 1>}
 CTRL_expressions_browRollPosInL {type=Curve, loc=<0 0 0>, rot=<0 0 0 1>}
 CTRL_expressions_browRollPosInR {type=Curve, loc=<0 0 0>, rot=<0 0 0 1>}
 CTRL_expressions_browRollPosOutL {type=Curve, loc=<0 0 0>, rot=<0 0 0 1>}
 CTRL_expressions_browRollPosOutR {type=Curve, loc=<0 0 0>, rot=<0 0 0 1>}
 CTRL_expressions_browUINL {type=Curve, loc=<0 0 0>, rot=<0 0 0 1>}
 CTRL_expressions_browUINR {type=Curve, loc=<0 0 0>, rot=<0 0 0 1>}
 CTRL_expressions_browUOUTL {type=Curve, loc=<0 0 0>, rot=<0 0 0 1>}
 CTRL_expressions_browUOUTR {type=Curve, loc=<0 0 0>, rot=<0 0 0 1>}
 CTRL_expressions_eyeBlinkL {type=Curve, loc=<0 0 0>, rot=<0 0 0 1>}
 CTRL_expressions_eyeBlinkR {type=Curve, loc=<0 0 0>, rot=<0 0 0 1>}
 CTRL_expressions_eyeCheekRaiseL {type=Curve, loc=<0 0 0>, rot=<0 0 0 1>}
 CTRL_expressions_eyeCheekRaiseR {type=Curve, loc=<0 0 0>, rot=<0 0 0 1>}
 CTRL_expressions_eyeEyelidsReduce {type=Curve, loc=<0 0 0>, rot=<0 0 0 1>}
 CTRL_expressions_eyeFaceScrunchL {type=Curve, loc=<0 0 0>, rot=<0 0 0 1>}
 CTRL_expressions_eyeFaceScrunchR {type=Curve, loc=<0 0 0>, rot=<0 0 0 1>}
 CTRL_expressions_eyelidRollNegDL {type=Curve, loc=<0 0 0>, rot=<0 0 0 1>}
 CTRL_expressions_eyelidRollNegDR {type=Curve, loc=<0 0 0>, rot=<0 0 0 1>}
 CTRL_expressions_eyelidRollNegUL {type=Curve, loc=<0 0 0>, rot=<0 0 0 1>}
 CTRL_expressions_eyelidRollNegUR {type=Curve, loc=<0 0 0>, rot=<0 0 0 1>}
 CTRL_expressions_eyelidRollPosDL {type=Curve, loc=<0 0 0>, rot=<0 0 0 1>}
 CTRL_expressions_eyelidRollPosDR {type=Curve, loc=<0 0 0>, rot=<0 0 0 1>}
 CTRL_expressions_eyelidRollPosUL {type=Curve, loc=<0 0 0>, rot=<0 0 0 1>}
 CTRL_expressions_eyelidRollPosUR {type=Curve, loc=<0 0 0>, rot=<0 0 0 1>}
 CTRL_expressions_eyeLookDownL {type=Curve, loc=<0 0 0>, rot=<0 0 0 1>}
 CTRL_expressions_eyeLookDownR {type=Curve, loc=<0 0 0>, rot=<0 0 0 1>}
 CTRL_expressions_eyeLookLeftL {type=Curve, loc=<0 0 0>, rot=<0 0 0 1>}
 CTRL_expressions_eyeLookLeftR {type=Curve, loc=<0 0 0>, rot=<0 0 0 1>}
 CTRL_expressions_eyeLookRightL {type=Curve, loc=<0 0 0>, rot=<0 0 0 1>}
 CTRL_expressions_eyeLookRightR {type=Curve, loc=<0 0 0>, rot=<0 0 0 1>}
 CTRL_expressions_eyeLookUpL {type=Curve, loc=<0 0 0>, rot=<0 0 0 1>}
 CTRL_expressions_eyeLookUpR {type=Curve, loc=<0 0 0>, rot=<0 0 0 1>}
 CTRL_expressions_eyeLowerLidDownL {type=Curve, loc=<0 0 0>, rot=<0 0 0 1>}
 CTRL_expressions_eyeLowerLidDownR {type=Curve, loc=<0 0 0>, rot=<0 0 0 1>}
 CTRL_expressions_eyeLowerLidUpL {type=Curve, loc=<0 0 0>, rot=<0 0 0 1>}
 CTRL_expressions_eyeLowerLidUpR {type=Curve, loc=<0 0 0>, rot=<0 0 0 1>}
 CTRL_expressions_eyePupilNarrowL {type=Curve, loc=<0 0 0>, rot=<0 0 0 1>}
 CTRL_expressions_eyePupilNarrowR {type=Curve, loc=<0 0 0>, rot=<0 0 0 1>}
 CTRL_expressions_eyePupilWideL {type=Curve, loc=<0 0 0>, rot=<0 0 0 1>}
 CTRL_expressions_eyePupilWideR {type=Curve, loc=<0 0 0>, rot=<0 0 0 1>}
 CTRL_expressions_eyeRelaxL {type=Curve, loc=<0 0 0>, rot=<0 0 0 1>}
 CTRL_expressions_eyeRelaxR {type=Curve, loc=<0 0 0>, rot=<0 0 0 1>}
 CTRL_expressions_eyeSquintInnerL {type=Curve, loc=<0 0 0>, rot=<0 0 0 1>}
 CTRL_expressions_eyeSquintInnerR {type=Curve, loc=<0 0 0>, rot=<0 0 0 1>}
 CTRL_expressions_eyeUpperLidUpL {type=Curve, loc=<0 0 0>, rot=<0 0 0 1>}
 CTRL_expressions_eyeUpperLidUpR {type=Curve, loc=<0 0 0>, rot=<0 0 0 1>}
 CTRL_expressions_eyeWidenL {type=Curve, loc=<0 0 0>, rot=<0 0 0 1>}
 CTRL_expressions_eyeWidenR {type=Curve, loc=<0 0 0>, rot=<0 0 0 1>}
 CTRL_expressions_jawBack {type=Curve, loc=<0 0 0>, rot=<0 0 0 1>}
 CTRL_expressions_jawChinRaiseDL {type=Curve, loc=<0 0 0>, rot=<0 0 0 1>}
 CTRL_expressions_jawChinRaiseDR {type=Curve, loc=<0 0 0>, rot=<0 0 0 1>}
 CTRL_expressions_jawChinRaiseUL {type=Curve, loc=<0 0 0>, rot=<0 0 0 1>}
 CTRL_expressions_jawChinRaiseUR {type=Curve, loc=<0 0 0>, rot=<0 0 0 1>}
 CTRL_expressions_jawFwd {type=Curve, loc=<0 0 0>, rot=<0 0 0 1>}
 CTRL_expressions_jawLeft {type=Curve, loc=<0 0 0>, rot=<0 0 0 1>}
 CTRL_expressions_jawOpen {type=Curve, loc=<0 0 0>, rot=<0 0 0 1>}
 CTRL_expressions_jawOpenExtreme {type=Curve, loc=<0 0 0>, rot=<0 0 0 1>}
 CTRL_expressions_jawRight {type=Curve, loc=<0 0 0>, rot=<0 0 0 1>}
 CTRL_expressions_mouthCheekBlowL {type=Curve, loc=<0 0 0>, rot=<0 0 0 1>}
 CTRL_expressions_mouthCheekBlowR {type=Curve, loc=<0 0 0>, rot=<0 0 0 1>}
 CTRL_expressions_mouthCheekSuckL {type=Curve, loc=<0 0 0>, rot=<0 0 0 1>}
 CTRL_expressions_mouthCheekSuckR {type=Curve, loc=<0 0 0>, rot=<0 0 0 1>}
 CTRL_expressions_mouthCornerDepressL {type=Curve, loc=<0 0 0>, rot=<0 0 0 1>}
 CTRL_expressions_mouthCornerDepressR {type=Curve, loc=<0 0 0>, rot=<0 0 0 1>}
 CTRL_expressions_mouthCornerDownL {type=Curve, loc=<0 0 0>, rot=<0 0 0 1>}
 CTRL_expressions_mouthCornerDownR {type=Curve, loc=<0 0 0>, rot=<0 0 0 1>}
 CTRL_expressions_mouthCornerNarrowL {type=Curve, loc=<0 0 0>, rot=<0 0 0 1>}
 CTRL_expressions_mouthCornerNarrowR {type=Curve, loc=<0 0 0>, rot=<0 0 0 1>}
 CTRL_expressions_mouthCornerPullL {type=Curve, loc=<0 0 0>, rot=<0 0 0 1>}
 CTRL_expressions_mouthCornerPullR {type=Curve, loc=<0 0 0>, rot=<0 0 0 1>}
 CTRL_expressions_mouthCornerRounderDL {type=Curve, loc=<0 0 0>, rot=<0 0 0 1>}
 CTRL_expressions_mouthCornerRounderDR {type=Curve, loc=<0 0 0>, rot=<0 0 0 1>}
 CTRL_expressions_mouthCornerRounderUL {type=Curve, loc=<0 0 0>, rot=<0 0 0 1>}
 CTRL_expressions_mouthCornerRounderUR {type=Curve, loc=<0 0 0>, rot=<0 0 0 1>}
 CTRL_expressions_mouthCornerSharpenDL {type=Curve, loc=<0 0 0>, rot=<0 0 0 1>}
 CTRL_expressions_mouthCornerSharpenDR {type=Curve, loc=<0 0 0>, rot=<0 0 0 1>}
 CTRL_expressions_mouthCornerSharpenUL {type=Curve, loc=<0 0 0>, rot=<0 0 0 1>}
 CTRL_expressions_mouthCornerSharpenUR {type=Curve, loc=<0 0 0>, rot=<0 0 0 1>}
 CTRL_expressions_mouthCornerUpL {type=Curve, loc=<0 0 0>, rot=<0 0 0 1>}
 CTRL_expressions_mouthCornerUpR {type=Curve, loc=<0 0 0>, rot=<0 0 0 1>}
 CTRL_expressions_mouthCornerWideL {type=Curve, loc=<0 0 0>, rot=<0 0 0 1>}
 CTRL_expressions_mouthCornerWideR {type=Curve, loc=<0 0 0>, rot=<0 0 0 1>}
 CTRL_expressions_mouthDimpleL {type=Curve, loc=<0 0 0>, rot=<0 0 0 1>}
 CTRL_expressions_mouthDimpleR {type=Curve, loc=<0 0 0>, rot=<0 0 0 1>}
 CTRL_expressions_mouthDown {type=Curve, loc=<0 0 0>, rot=<0 0 0 1>}
 CTRL_expressions_mouthFunnelDL {type=Curve, loc=<0 0 0>, rot=<0 0 0 1>}
 CTRL_expressions_mouthFunnelDR {type=Curve, loc=<0 0 0>, rot=<0 0 0 1>}
 CTRL_expressions_mouthFunnelUL {type=Curve, loc=<0 0 0>, rot=<0 0 0 1>}
 CTRL_expressions_mouthFunnelUR {type=Curve, loc=<0 0 0>, rot=<0 0 0 1>}
 CTRL_expressions_mouthLeft {type=Curve, loc=<0 0 0>, rot=<0 0 0 1>}
 CTRL_expressions_mouthLipsBlowL {type=Curve, loc=<0 0 0>, rot=<0 0 0 1>}
 CTRL_expressions_mouthLipsBlowR {type=Curve, loc=<0 0 0>, rot=<0 0 0 1>}
 CTRL_expressions_mouthLipsPressL {type=Curve, loc=<0 0 0>, rot=<0 0 0 1>}
 CTRL_expressions_mouthLipsPressR {type=Curve, loc=<0 0 0>, rot=<0 0 0 1>}
 CTRL_expressions_mouthLipsPullDL {type=Curve, loc=<0 0 0>, rot=<0 0 0 1>}
 CTRL_expressions_mouthLipsPullDR {type=Curve, loc=<0 0 0>, rot=<0 0 0 1>}
 CTRL_expressions_mouthLipsPullUL {type=Curve, loc=<0 0 0>, rot=<0 0 0 1>}
 CTRL_expressions_mouthLipsPullUR {type=Curve, loc=<0 0 0>, rot=<0 0 0 1>}
 CTRL_expressions_mouthLipsPurseDL {type=Curve, loc=<0 0 0>, rot=<0 0 0 1>}
 CTRL_expressions_mouthLipsPurseDR {type=Curve, loc=<0 0 0>, rot=<0 0 0 1>}
 CTRL_expressions_mouthLipsPurseUL {type=Curve, loc=<0 0 0>, rot=<0 0 0 1>}
 CTRL_expressions_mouthLipsPurseUR {type=Curve, loc=<0 0 0>, rot=<0 0 0 1>}
 CTRL_expressions_mouthLipsPushDL {type=Curve, loc=<0 0 0>, rot=<0 0 0 1>}
 CTRL_expressions_mouthLipsPushDR {type=Curve, loc=<0 0 0>, rot=<0 0 0 1>}
 CTRL_expressions_mouthLipsPushUL {type=Curve, loc=<0 0 0>, rot=<0 0 0 1>}
 CTRL_expressions_mouthLipsPushUR {type=Curve, loc=<0 0 0>, rot=<0 0 0 1>}
 CTRL_expressions_mouthLipsThickDL {type=Curve, loc=<0 0 0>, rot=<0 0 0 1>}
 CTRL_expressions_mouthLipsThickDR {type=Curve, loc=<0 0 0>, rot=<0 0 0 1>}
 CTRL_expressions_mouthLipsThickInwardDL {type=Curve, loc=<0 0 0>, rot=<0 0 0 1>}
 CTRL_expressions_mouthLipsThickInwardDR {type=Curve, loc=<0 0 0>, rot=<0 0 0 1>}
 CTRL_expressions_mouthLipsThickInwardUL {type=Curve, loc=<0 0 0>, rot=<0 0 0 1>}
 CTRL_expressions_mouthLipsThickInwardUR {type=Curve, loc=<0 0 0>, rot=<0 0 0 1>}
 CTRL_expressions_mouthLipsThickUL {type=Curve, loc=<0 0 0>, rot=<0 0 0 1>}
 CTRL_expressions_mouthLipsThickUR {type=Curve, loc=<0 0 0>, rot=<0 0 0 1>}
 CTRL_expressions_mouthLipsThinDL {type=Curve, loc=<0 0 0>, rot=<0 0 0 1>}
 CTRL_expressions_mouthLipsThinDR {type=Curve, loc=<0 0 0>, rot=<0 0 0 1>}
 CTRL_expressions_mouthLipsThinInwardDL {type=Curve, loc=<0 0 0>, rot=<0 0 0 1>}
 CTRL_expressions_mouthLipsThinInwardDR {type=Curve, loc=<0 0 0>, rot=<0 0 0 1>}
 CTRL_expressions_mouthLipsThinInwardUL {type=Curve, loc=<0 0 0>, rot=<0 0 0 1>}
 CTRL_expressions_mouthLipsThinInwardUR {type=Curve, loc=<0 0 0>, rot=<0 0 0 1>}
 CTRL_expressions_mouthLipsThinUL {type=Curve, loc=<0 0 0>, rot=<0 0 0 1>}
 CTRL_expressions_mouthLipsThinUR {type=Curve, loc=<0 0 0>, rot=<0 0 0 1>}
 CTRL_expressions_mouthLipsTightenDL {type=Curve, loc=<0 0 0>, rot=<0 0 0 1>}
 CTRL_expressions_mouthLipsTightenDR {type=Curve, loc=<0 0 0>, rot=<0 0 0 1>}
 CTRL_expressions_mouthLipsTightenUL {type=Curve, loc=<0 0 0>, rot=<0 0 0 1>}
 CTRL_expressions_mouthLipsTightenUR {type=Curve, loc=<0 0 0>, rot=<0 0 0 1>}
 CTRL_expressions_mouthLipsTogetherDL {type=Curve, loc=<0 0 0>, rot=<0 0 0 1>}
 CTRL_expressions_mouthLipsTogetherDR {type=Curve, loc=<0 0 0>, rot=<0 0 0 1>}
 CTRL_expressions_mouthLipsTogetherUL {type=Curve, loc=<0 0 0>, rot=<0 0 0 1>}
 CTRL_expressions_mouthLipsTogetherUR {type=Curve, loc=<0 0 0>, rot=<0 0 0 1>}
 CTRL_expressions_mouthLipsTowardsDL {type=Curve, loc=<0 0 0>, rot=<0 0 0 1>}
 CTRL_expressions_mouthLipsTowardsDR {type=Curve, loc=<0 0 0>, rot=<0 0 0 1>}
 CTRL_expressions_mouthLipsTowardsUL {type=Curve, loc=<0 0 0>, rot=<0 0 0 1>}
 CTRL_expressions_mouthLipsTowardsUR {type=Curve, loc=<0 0 0>, rot=<0 0 0 1>}
 CTRL_expressions_mouthLowerLipBiteL {type=Curve, loc=<0 0 0>, rot=<0 0 0 1>}
 CTRL_expressions_mouthLowerLipBiteR {type=Curve, loc=<0 0 0>, rot=<0 0 0 1>}
 CTRL_expressions_mouthLowerLipDepressL {type=Curve, loc=<0 0 0>, rot=<0 0 0 1>}
 CTRL_expressions_mouthLowerLipDepressR {type=Curve, loc=<0 0 0>, rot=<0 0 0 1>}
 CTRL_expressions_mouthLowerLipRollInL {type=Curve, loc=<0 0 0>, rot=<0 0 0 1>}
 CTRL_expressions_mouthLowerLipRollInR {type=Curve, loc=<0 0 0>, rot=<0 0 0 1>}
 CTRL_expressions_mouthLowerLipRollOutL {type=Curve, loc=<0 0 0>, rot=<0 0 0 1>}
 CTRL_expressions_mouthLowerLipRollOutR {type=Curve, loc=<0 0 0>, rot=<0 0 0 1>}
 CTRL_expressions_mouthLowerLipTowardsTeethL {type=Curve, loc=<0 0 0>, rot=<0 0 0 1>}
 CTRL_expressions_mouthLowerLipTowardsTeethR {type=Curve, loc=<0 0 0>, rot=<0 0 0 1>}
 CTRL_expressions_mouthPressDL {type=Curve, loc=<0 0 0>, rot=<0 0 0 1>}
 CTRL_expressions_mouthPressDR {type=Curve, loc=<0 0 0>, rot=<0 0 0 1>}
 CTRL_expressions_mouthPressUL {type=Curve, loc=<0 0 0>, rot=<0 0 0 1>}
 CTRL_expressions_mouthPressUR {type=Curve, loc=<0 0 0>, rot=<0 0 0 1>}
 CTRL_expressions_mouthRight {type=Curve, loc=<0 0 0>, rot=<0 0 0 1>}
 CTRL_expressions_mouthSharpCornerPullL {type=Curve, loc=<0 0 0>, rot=<0 0 0 1>}
 CTRL_expressions_mouthSharpCornerPullR {type=Curve, loc=<0 0 0>, rot=<0 0 0 1>}
 CTRL_expressions_mouthStretchL {type=Curve, loc=<0 0 0>, rot=<0 0 0 1>}
 CTRL_expressions_mouthStretchR {type=Curve, loc=<0 0 0>, rot=<0 0 0 1>}
 CTRL_expressions_mouthUp {type=Curve, loc=<0 0 0>, rot=<0 0 0 1>}
 CTRL_expressions_mouthUpperLipBiteL {type=Curve, loc=<0 0 0>, rot=<0 0 0 1>}
 CTRL_expressions_mouthUpperLipBiteR {type=Curve, loc=<0 0 0>, rot=<0 0 0 1>}
 CTRL_expressions_mouthUpperLipRaiseL {type=Curve, loc=<0 0 0>, rot=<0 0 0 1>}
 CTRL_expressions_mouthUpperLipRaiseR {type=Curve, loc=<0 0 0>, rot=<0 0 0 1>}
 CTRL_expressions_mouthUpperLipRollInL {type=Curve, loc=<0 0 0>, rot=<0 0 0 1>}
 CTRL_expressions_mouthUpperLipRollInR {type=Curve, loc=<0 0 0>, rot=<0 0 0 1>}
 CTRL_expressions_mouthUpperLipRollOutL {type=Curve, loc=<0 0 0>, rot=<0 0 0 1>}
 CTRL_expressions_mouthUpperLipRollOutR {type=Curve, loc=<0 0 0>, rot=<0 0 0 1>}
 CTRL_expressions_mouthUpperLipTowardsTeethL {type=Curve, loc=<0 0 0>, rot=<0 0 0 1>}
 CTRL_expressions_mouthUpperLipTowardsTeethR {type=Curve, loc=<0 0 0>, rot=<0 0 0 1>}
 CTRL_expressions_noseNasolabialDeepenL {type=Curve, loc=<0 0 0>, rot=<0 0 0 1>}
 CTRL_expressions_noseNasolabialDeepenR {type=Curve, loc=<0 0 0>, rot=<0 0 0 1>}
 CTRL_expressions_noseNostrilCompressL {type=Curve, loc=<0 0 0>, rot=<0 0 0 1>}
 CTRL_expressions_noseNostrilCompressR {type=Curve, loc=<0 0 0>, rot=<0 0 0 1>}
 CTRL_expressions_noseNostrilDepressL {type=Curve, loc=<0 0 0>, rot=<0 0 0 1>}
 CTRL_expressions_noseNostrilDepressR {type=Curve, loc=<0 0 0>, rot=<0 0 0 1>}
 CTRL_expressions_noseNostrilDilateL {type=Curve, loc=<0 0 0>, rot=<0 0 0 1>}
 CTRL_expressions_noseNostrilDilateR {type=Curve, loc=<0 0 0>, rot=<0 0 0 1>}
 CTRL_expressions_noseRotateL {type=Curve, loc=<0 0 0>, rot=<0 0 0 1>}
 CTRL_expressions_noseRotateR {type=Curve, loc=<0 0 0>, rot=<0 0 0 1>}
 CTRL_expressions_noseSquash {type=Curve, loc=<0 0 0>, rot=<0 0 0 1>}
 CTRL_expressions_noseStretch {type=Curve, loc=<0 0 0>, rot=<0 0 0 1>}
 CTRL_expressions_noseWrinkleL {type=Curve, loc=<0 0 0>, rot=<0 0 0 1>}
 CTRL_expressions_noseWrinkleR {type=Curve, loc=<0 0 0>, rot=<0 0 0 1>}
 CTRL_expressions_noseWrinkleUpperL {type=Curve, loc=<0 0 0>, rot=<0 0 0 1>}
 CTRL_expressions_noseWrinkleUpperR {type=Curve, loc=<0 0 0>, rot=<0 0 0 1>}
 CTRL_expressions_teethBackD {type=Curve, loc=<0 0 0>, rot=<0 0 0 1>}
 CTRL_expressions_teethBackU {type=Curve, loc=<0 0 0>, rot=<0 0 0 1>}
 CTRL_expressions_teethDownD {type=Curve, loc=<0 0 0>, rot=<0 0 0 1>}
 CTRL_expressions_teethDownU {type=Curve, loc=<0 0 0>, rot=<0 0 0 1>}
 CTRL_expressions_teethFwdD {type=Curve, loc=<0 0 0>, rot=<0 0 0 1>}
 CTRL_expressions_teethFwdU {type=Curve, loc=<0 0 0>, rot=<0 0 0 1>}
 CTRL_expressions_teethLeftD {type=Curve, loc=<0 0 0>, rot=<0 0 0 1>}
 CTRL_expressions_teethLeftU {type=Curve, loc=<0 0 0>, rot=<0 0 0 1>}
 CTRL_expressions_teethRightD {type=Curve, loc=<0 0 0>, rot=<0 0 0 1>}
 CTRL_expressions_teethRightU {type=Curve, loc=<0 0 0>, rot=<0 0 0 1>}
 CTRL_expressions_teethRotateBackD {type=Curve, loc=<0 0 0>, rot=<0 0 0 1>}
 CTRL_expressions_teethRotateBackU {type=Curve, loc=<0 0 0>, rot=<0 0 0 1>}
 CTRL_expressions_teethRotateFwdD {type=Curve, loc=<0 0 0>, rot=<0 0 0 1>}
 CTRL_expressions_teethRotateFwdU {type=Curve, loc=<0 0 0>, rot=<0 0 0 1>}
 CTRL_expressions_teethUpD {type=Curve, loc=<0 0 0>, rot=<0 0 0 1>}
 CTRL_expressions_teethUpU {type=Curve, loc=<0 0 0>, rot=<0 0 0 1>}
 CTRL_expressions_tongueBendDown {type=Curve, loc=<0 0 0>, rot=<0 0 0 1>}
 CTRL_expressions_tongueBendUp {type=Curve, loc=<0 0 0>, rot=<0 0 0 1>}
 CTRL_expressions_tongueDown {type=Curve, loc=<0 0 0>, rot=<0 0 0 1>}
 CTRL_expressions_tongueIn {type=Curve, loc=<0 0 0>, rot=<0 0 0 1>}
 CTRL_expressions_tongueLeft {type=Curve, loc=<0 0 0>, rot=<0 0 0 1>}
 CTRL_expressions_tongueNarrow {type=Curve, loc=<0 0 0>, rot=<0 0 0 1>}
 CTRL_expressions_tongueOut {type=Curve, loc=<0 0 0>, rot=<0 0 0 1>}
 CTRL_expressions_tongueRight {type=Curve, loc=<0 0 0>, rot=<0 0 0 1>}
 CTRL_expressions_tongueRollDown {type=Curve, loc=<0 0 0>, rot=<0 0 0 1>}
 CTRL_expressions_tongueRollLeft {type=Curve, loc=<0 0 0>, rot=<0 0 0 1>}
 CTRL_expressions_tongueRollRight {type=Curve, loc=<0 0 0>, rot=<0 0 0 1>}
 CTRL_expressions_tongueRollUp {type=Curve, loc=<0 0 0>, rot=<0 0 0 1>}
 CTRL_expressions_tongueTwistLeft {type=Curve, loc=<0 0 0>, rot=<0 0 0 1>}
 CTRL_expressions_tongueTwistRight {type=Curve, loc=<0 0 0>, rot=<0 0 0 1>}
 CTRL_expressions_tongueUp {type=Curve, loc=<0 0 0>, rot=<0 0 0 1>}
 CTRL_expressions_tongueWide {type=Curve, loc=<0 0 0>, rot=<0 0 0 1>}
 DisableFaceOverride {type=Curve, loc=<0 0 0>, rot=<0 0 0 1>}
 is_3L {type=Curve, loc=<0 0 0>, rot=<0 0 0 1>}
 is_face_bone_anim {type=Curve, loc=<0 0 0>, rot=<0 0 0 1>}
 WM_BrowLateral_L {type=Curve, loc=<0 0 0>, rot=<0 0 0 1>}
 WM_BrowLateral_Offset_L {type=Curve, loc=<0 0 0>, rot=<0 0 0 1>}
 WM_BrowLateral_Offset_R {type=Curve, loc=<0 0 0>, rot=<0 0 0 1>}
 WM_BrowLateral_R {type=Curve, loc=<0 0 0>, rot=<0 0 0 1>}
 WM_BrowRaiseInner_L {type=Curve, loc=<0 0 0>, rot=<0 0 0 1>}
 WM_BrowRaiseInner_Offset_L {type=Curve, loc=<0 0 0>, rot=<0 0 0 1>}
 WM_BrowRaiseInner_Offset_R {type=Curve, loc=<0 0 0>, rot=<0 0 0 1>}
 WM_BrowRaiseInner_R {type=Curve, loc=<0 0 0>, rot=<0 0 0 1>}
 WM_BrowRaiseOutter_L {type=Curve, loc=<0 0 0>, rot=<0 0 0 1>}
 WM_BrowRaiseOutter_Offset_L {type=Curve, loc=<0 0 0>, rot=<0 0 0 1>}
 WM_BrowRaiseOutter_Offset_R {type=Curve, loc=<0 0 0>, rot=<0 0 0 1>}
 WM_BrowRaiseOutter_R {type=Curve, loc=<0 0 0>, rot=<0 0 0 1>}
 WM_CheekRaise_L {type=Curve, loc=<0 0 0>, rot=<0 0 0 1>}
 WM_CheekRaise_Offset_L {type=Curve, loc=<0 0 0>, rot=<0 0 0 1>}
 WM_CheekRaise_Offset_R {type=Curve, loc=<0 0 0>, rot=<0 0 0 1>}
 WM_CheekRaise_R {type=Curve, loc=<0 0 0>, rot=<0 0 0 1>}
 WM_Lips_L {type=Curve, loc=<0 0 0>, rot=<0 0 0 1>}
 WM_Lips_Offset_L {type=Curve, loc=<0 0 0>, rot=<0 0 0 1>}
 WM_Lips_Offset_R {type=Curve, loc=<0 0 0>, rot=<0 0 0 1>}
 WM_Lips_R {type=Curve, loc=<0 0 0>, rot=<0 0 0 1>}
 WM_MouthStretch_L {type=Curve, loc=<0 0 0>, rot=<0 0 0 1>}
 WM_MouthStretch_Offset_L {type=Curve, loc=<0 0 0>, rot=<0 0 0 1>}
 WM_MouthStretch_Offset_R {type=Curve, loc=<0 0 0>, rot=<0 0 0 1>}
 WM_MouthStretch_R {type=Curve, loc=<0 0 0>, rot=<0 0 0 1>}
 WM_NoseWrinkle_L {type=Curve, loc=<0 0 0>, rot=<0 0 0 1>}
 WM_NoseWrinkle_Offset_L {type=Curve, loc=<0 0 0>, rot=<0 0 0 1>}
 WM_NoseWrinkle_Offset_R {type=Curve, loc=<0 0 0>, rot=<0 0 0 1>}
 WM_NoseWrinkle_R {type=Curve, loc=<0 0 0>, rot=<0 0 0 1>}
 WM_Smile_Offset_L {type=Curve, loc=<0 0 0>, rot=<0 0 0 1>}
 WM_Smile_Offset_R {type=Curve, loc=<0 0 0>, rot=<0 0 0 1>}
 WM_Smile_R {type=Curve, loc=<0 0 0>, rot=<0 0 0 1>}
 WM_Smile_L {type=Curve, loc=<0 0 0>, rot=<0 0 0 1>}
]
